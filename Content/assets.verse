UserInterface<public> := module:
    Icons<public> := module{}
    HUD<public> := module{}
    News<public> := module{}

NPC<public> := module:
    AnimationSequences<public> := module{}

Props<public> := module:
    ItemDrops<public> := module{}
    