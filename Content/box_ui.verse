using { /Verse.org/Colors }
using { /UnrealEngine.com/Temporary/SpatialMath }
using { /UnrealEngine.com/Temporary/UI }

<# ----------------------------------------------------------------------------------------------------
    Constructor
---------------------------------------------------------------------------------------------------- #>
MakeBoxUI<constructor>(
    ?DefaultColor:color = NamedColors.Black, ?DefaultOpacity:type{_X:float where 0.000000 <= _X, _X <= 1.000000} = 0.1,
    ?DefaultBorderRadius:float = 1.0, ?DefaultBorderColor:color = NamedColors.White, ?DefaultBorderOpacity:type{_X:float where 0.000000 <= _X, _X <= 1.000000} = 0.1,
    ?DefaultWidth:float = 0.0, ?DefaultHeight:float = 0.0
) := box_ui:
    DefaultColor := DefaultColor
    DefaultOpacity := DefaultOpacity
    DefaultWidth := DefaultWidth
    DefaultHeight := DefaultHeight
    DefaultBorderRadius := DefaultBorderRadius
    DefaultBorderColor := DefaultBorderColor
    DefaultBorderOpacity := DefaultBorderOpacity

    let:
        DefaultBox := color_block:
            DefaultColor := DefaultColor
            DefaultOpacity := DefaultOpacity
            DefaultDesiredSize := vector2{X := DefaultWidth, Y := DefaultHeight}

        DefaultBorderTop := color_block:
            DefaultColor := DefaultBorderColor
            DefaultOpacity := DefaultBorderOpacity
            DefaultDesiredSize := vector2{X := 0.0, Y := DefaultBorderRadius}

        DefaultBorderBottom := color_block:
            DefaultColor := DefaultBorderColor
            DefaultOpacity := DefaultBorderOpacity
            DefaultDesiredSize := vector2{X := 0.0, Y := DefaultBorderRadius}

        DefaultBorderLeft := color_block:
            DefaultColor := DefaultBorderColor
            DefaultOpacity := DefaultBorderOpacity
            DefaultDesiredSize := vector2{X := DefaultBorderRadius, Y := 0.0}

        DefaultBorderRight := color_block:
            DefaultColor := DefaultBorderColor
            DefaultOpacity := DefaultBorderOpacity
            DefaultDesiredSize := vector2{X := DefaultBorderRadius, Y := 0.0}

    BoxColor := DefaultBox
    BorderTopColor := DefaultBorderTop
    BorderBottomColor := DefaultBorderBottom
    BorderLeftColor := DefaultBorderLeft
    BorderRightColor := DefaultBorderRight

    Widget := overlay:
        Slots := array:
            overlay_slot:
                HorizontalAlignment := horizontal_alignment.Fill
                VerticalAlignment := vertical_alignment.Fill
                Padding := margin{Left := DefaultBorderRadius, Top := DefaultBorderRadius, Right := DefaultBorderRadius, Bottom := DefaultBorderRadius}
                Widget := DefaultBox
            overlay_slot:
                HorizontalAlignment := horizontal_alignment.Fill
                VerticalAlignment := vertical_alignment.Top
                Padding := margin{Left := DefaultBorderRadius, Top := 0.0, Right := DefaultBorderRadius, Bottom := 0.0}
                Widget := DefaultBorderTop
            overlay_slot:
                HorizontalAlignment := horizontal_alignment.Fill
                VerticalAlignment := vertical_alignment.Bottom
                Padding := margin{Left := DefaultBorderRadius, Top := 0.0, Right := DefaultBorderRadius, Bottom := 0.0}
                Widget := DefaultBorderBottom
            overlay_slot:
                HorizontalAlignment := horizontal_alignment.Left
                VerticalAlignment := vertical_alignment.Fill
                Widget := DefaultBorderLeft
            overlay_slot:
                HorizontalAlignment := horizontal_alignment.Right
                VerticalAlignment := vertical_alignment.Fill
                Widget := DefaultBorderRight

<# ----------------------------------------------------------------------------------------------------
    box_ui
---------------------------------------------------------------------------------------------------- #>
box_ui := class:
    DefaultColor:color
    DefaultOpacity:type{_X:float where 0.000000 <= _X, _X <= 1.000000}

    DefaultWidth:float
    DefaultHeight:float

    DefaultBorderRadius:float
    DefaultBorderColor:color
    DefaultBorderOpacity:type{_X:float where 0.000000 <= _X, _X <= 1.000000}

    BoxColor:color_block = color_block{}
    BorderTopColor:color_block = color_block{}
    BorderBottomColor:color_block = color_block{}
    BorderLeftColor:color_block = color_block{}
    BorderRightColor:color_block = color_block{}
    
    var Widget:widget = canvas{}

    <# ----------------------------------------------------------------------------------------------------
        Get Functions
    ---------------------------------------------------------------------------------------------------- #>
    GetWidget()<transacts>:widget =
        Widget

    <# ----------------------------------------------------------------------------------------------------
        Update Functions
    ---------------------------------------------------------------------------------------------------- #>
    SetColor(Color:color):void =
        BoxColor.SetColor(Color)

    SetOpacity(Opacity:type{_X:float where 0.000000 <= _X, _X <= 1.000000}):void =
        BoxColor.SetOpacity(Opacity)
    
    SetBorderColor(Color:color):void =
        BorderTopColor.SetColor(Color)
        BorderBottomColor.SetColor(Color)
        BorderLeftColor.SetColor(Color)
        BorderRightColor.SetColor(Color)
    
    SetBorderOpacity(Opacity:type{_X:float where 0.000000 <= _X, _X <= 1.000000}):void =
        BorderTopColor.SetOpacity(Opacity)
        BorderBottomColor.SetOpacity(Opacity)
        BorderLeftColor.SetOpacity(Opacity)
        BorderRightColor.SetOpacity(Opacity)