events_cancelable<public> := class:
    CancelEvent:event() = event(){}
    Cancel():void =
        CancelEvent.Signal()

events_subscribable<public>(t:type)<computes> := interface:
    Subscribe<public>(Callback:type {__(:t):void}):events_cancelable

events_subscribable<public>()<computes> := subscribable(tuple())

events<public>(t:type)<computes> := class(event(t), events_subscribable(t)):
    Event:event(t) = event(t){}

    Signal<override>(Payload:t):void =
        Event.Signal(Payload)

    Await<override>()<suspends>:t =
        Event.Await()

    Subscribe<override>(Callback:type{_(:t):void}):events_cancelable =
        Cancelable:events_cancelable = events_cancelable{}
        spawn{AwaitSignal(Callback, Cancelable.CancelEvent)}
        Cancelable

    AwaitSignal(Callback:type{_(:t):void}, CancelEvent:event())<suspends>:void =
        race:
            CancelEvent.Await()
            loop:
                Callback(Event.Await())