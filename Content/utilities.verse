using { /Fortnite.com/Characters }
using { /Fortnite.com/Devices }
using { /Verse.org/Colors }
using { /Verse.org/Random }
using { /Verse.org/Simulation }
using { /UnrealEngine.com/Temporary/Diagnostics }
using { /UnrealEngine.com/Temporary/SpatialMath }
using { /UnrealEngine.com/Temporary/UI }

<#>
    Useful constants.

TickTime:float = 1.0 / 30.0
CharacterHeightOffset:float = 77.0
PlayerCharacterSize:float = 25.0
UnitVector:vector3 = vector3{X := 1.0, Y := 1.0, Z := 1.0}
TimeDilationFactor:float = 10000.0
MapIndicatorDisabledPosition:vector3 = vector3{X := 10000.0}
MapIndicatorEnabledPosition:vector3 = vector3{Z := 100.0}

<#>
    Make a message from a string.

MakeMessage<localizes>(Message:string):message = "{Message}"
MakeMessage<localizes>(Agent:agent):message = "{Agent}"
operator'+'<public><localizes>(MessageA:message, MessageB:message)<computes>:message = "{MessageA}{MessageB}"

<#>
    Wrap text function.

CharacterSizeMapping:[char]int = map:
    ' ' => 8, '!' => 11, '"' => 13, '#' => 20, '$' => 20, '%' => 27, '&' => 24, ''' => 6, '(' => 11, ')' => 11, '*' => 13, '+' => 16, ',' => 9, '-' => 12,
    '.' => 9, '/' => 15, '0' => 21, '1' => 14, '2' => 19, '3' => 20, '4' => 20, '5' => 20, '6' => 21, '7' => 18, '8' => 21, '9' => 20, ':' => 9, ';' => 10,
    '<' => 16, '=' => 15, '>' => 16, '?' => 21, '@' => 34, 'A' => 22, 'B' => 20, 'C' => 22, 'D' => 21, 'E' => 18, 'F' => 17, 'G' => 22, 'H' => 20, 'I' => 11,
    'J' => 19, 'K' => 21, 'L' => 17, 'M' => 30, 'N' => 21, 'O' => 24, 'P' => 19, 'Q' => 23, 'R' => 20, 'S' => 20, 'T' => 19, 'U' => 21, 'V' => 21, 'W' => 29,
    'X' => 21, 'Y' => 22, 'Z' => 18, '[' => 12, '\' => 15, ']' => 12, '^' => 17, '_' => 16, '`' => 18, 'a' => 18, 'b' => 19, 'c' => 18, 'd' => 19, 'e' => 18,
    'f' => 14, 'g' => 18, 'h' => 18, 'i' => 9, 'j' => 9, 'k' => 18, 'l' => 9, 'm' => 28, 'n' => 18, 'o' => 19, 'p' => 20, 'q' => 19, 'r' => 15, 's' => 17,
    't' => 14, 'u' => 18, 'v' => 18, 'w' => 26, 'x' => 19, 'y' => 18, 'z' => 15, '{' => 15, '|' => 10, '}' => 15, '~' => 17

(Character:char).GetSize()<transacts>:int =
    CharacterSizeMapping[Character] or 0

GetStringSize(String:string)<transacts>:float =
    var TotalStringSize:float = 0.0
    for (CharacterIndex -> Character : String):
        CharacterSize := Character.GetSize()
        set TotalStringSize += CharacterSize * 1.0
    TotalStringSize

(Text:string).WrapText(AvailableWidth:float):string =
    var WrappedText:string = ""
    var CurrentTextSize:int = 0
    var TextSizeFromLastSpaceCharacter:int = 0
    var LastSpaceCharacterIndex:int = 0

    for (CharacterIndex -> Character : Text):
        CharacterSize := Character.GetSize()
        set CurrentTextSize += CharacterSize
        set TextSizeFromLastSpaceCharacter += CharacterSize

        if (CurrentTextSize * 1.0 >= AvailableWidth):
            if (set WrappedText[LastSpaceCharacterIndex] = '\n') {}
            set CurrentTextSize = TextSizeFromLastSpaceCharacter
            
        if (Character = ' '):
            set LastSpaceCharacterIndex = CharacterIndex
            set TextSizeFromLastSpaceCharacter = 0

        set WrappedText += ToString(Character)

    WrappedText

(Character:char).ToLower<public>()<computes>:char =
    case(Character):
        'A' => 'a', 'B' => 'b', 'C' => 'c', 'D' => 'd',
        'E' => 'e', 'F' => 'f', 'G' => 'g', 'H' => 'h',
        'I' => 'i', 'J' => 'j', 'K' => 'k', 'L' => 'l',
        'M' => 'm', 'N' => 'n', 'O' => 'o', 'P' => 'p',
        'Q' => 'q', 'R' => 'r', 'S' => 's', 'T' => 't',
        'U' => 'u', 'V' => 'v', 'W' => 'w', 'X' => 'x',
        'Y' => 'y', 'Z' => 'z', _ => Character

(Character:char).ToUpper<public>()<computes>:char =
    case(Character):
        'a' => 'A', 'b' => 'B', 'c' => 'C', 'd' => 'D',
        'e' => 'E', 'f' => 'F', 'g' => 'G', 'h' => 'H',
        'i' => 'I', 'j' => 'J', 'k' => 'K', 'l' => 'L',
        'm' => 'M', 'n' => 'N', 'o' => 'O', 'p' => 'P',
        'q' => 'Q', 'r' => 'R', 's' => 'S', 't' => 'T',
        'u' => 'U', 'v' => 'V', 'w' => 'W', 'x' => 'X',
        'y' => 'Y', 'z' => 'Z', _ => Character

(String:string).ToLower<public>()<computes>:string =
    for (Character : String):
        Character.ToLower()

(String:string).ToUpper<public>()<computes>:string =
    for (Character : String):
        Character.ToUpper()


<#>
    Utility function for replacing identifier in strings e.g. "Hi [PlayerName].".Replace("PlayerName", "John") => "Hi John."

(InputString:string).Replace(IdentifierToReplace:string, ReplacementString:string):string =
    var FinalString:string = InputString
    loop:
        if:
            StartIndex := FinalString.Find['[']
            StopIndex := FinalString.Find[']']
            Identifier := FinalString.Slice[StartIndex + 1, StopIndex]
            Identifier = IdentifierToReplace
            BeforeString := FinalString.Slice[0, StartIndex]
            AfterString := FinalString.Slice[StopIndex + 1, FinalString.Length]
        then:
            set FinalString = BeforeString + ReplacementString + AfterString
        else:
            break

    FinalString

<#>
    Dispose of a spawned prop after the specified duration.

DisposeProp(MaybeProp:?creative_prop, Duration:float)<suspends>:void =
    if (Prop := MaybeProp?):
        Sleep(Duration)
        Prop.Dispose()

<#>
    Toggle widget visibility Hidden/Visible to force widget to be displayed immediately.

(Widget:widget).ToggleVisibility():void =
    WidgetVisibility := Widget.GetVisibility()
    Widget.SetVisibility(WidgetVisibility = widget_visibility.Visible and widget_visibility.Hidden or widget_visibility.Visible)

<#>
    Returns the average value of the values of the array.

Average(Numbers:[]float)<transacts>:float =
    var TotalValue:float = 0.0
    for (Number : Numbers):
        set TotalValue += Number
    TotalValue / (Numbers.Length * 1.0)

<#>
    A parametric remove key from map function.

RemoveKeyFromMap(Map:[t]u, KeyToRemove:t where t:subtype(comparable), u:type)<transacts>:[t]u =
    var NewMap:[t]u = map{}
    for (Key -> Value : Map, Key <> KeyToRemove):
        set NewMap = ConcatenateMaps(NewMap, map{Key => Value})
    NewMap

RemoveKeysFromMap(Map:[t]u, KeysToRemove:[]t where t:subtype(comparable), u:type)<transacts>:[t]u =
    var NewMap:[t]u = map{}
    for (Key -> Value : Map, not KeysToRemove.Find[Key]):
        set NewMap = ConcatenateMaps(NewMap, map{Key => Value})
    NewMap

<#>
    Global log function.

global_log := class(log_channel) {}

Logger()<transacts>:log =
    log{Channel := global_log}

<#>
    Utility function to get a random position on a circle.

GetRandomPositionOnCircle(Radius:float, ?Origin:vector3 = vector3{})<transacts>:vector3 =
    Angle := GetRandomFloat(0.0, 1.0) * PiFloat * 2.0;
    vector3{X := Origin.X + Cos(Angle) * Radius, Y := Origin.Y + Sin(Angle) * Radius, Z := Origin.Z}

GetRandomPositionInCircle(Radius:float, ?Origin:vector3 = vector3{})<transacts>:vector3 =
    Angle := GetRandomFloat(0.0, 1.0) * PiFloat * 2.0;
    RandomRadius := GetRandomFloat(0.0, Radius)
    vector3{X := Origin.X + Cos(Angle) * RandomRadius, Y := Origin.Y + Sin(Angle) * RandomRadius, Z := Origin.Z}

GetRandomPositionInDonut(InnerRadius:float, OuterRadius:float, ?Origin:vector3 = vector3{})<transacts>:vector3 =
    Angle := GetRandomFloat(0.0, 1.0) * PiFloat * 2.0;
    RandomRadius := GetRandomFloat(InnerRadius, OuterRadius)
    vector3{X := Origin.X + Cos(Angle) * RandomRadius, Y := Origin.Y + Sin(Angle) * RandomRadius, Z := Origin.Z}

<#>
    Utility class to calculate DeltaTime/ElapsedTime for animations.

delta_timer := class:
    var StartTime:float = 0.0
    var PreviousTime:float = 0.0
    
    block:
        if (StartTime = 0.0):
            set StartTime = GetSimulationElapsedTime()

        if (PreviousTime = 0.0):
            set PreviousTime = GetSimulationElapsedTime()

    GetStartTime()<transacts>:float =
        StartTime

    GetDeltaTime()<transacts>:float =
        CurrentTime := GetSimulationElapsedTime()
        DeltaTime := CurrentTime - PreviousTime
        set PreviousTime = CurrentTime
        DeltaTime

    GetElapsedTime()<transacts>:float =
        GetSimulationElapsedTime() - StartTime

    Reset():void =
        set StartTime = GetSimulationElapsedTime()
        set PreviousTime = GetSimulationElapsedTime()

<#>
    Utility function to get tier colors from integer values (1 = Common .. 7 = Exotic).

(Tier:int).ToTierColor()<transacts>:color =
    case (Tier):
        1 => ThemeColors.RarityCommon
        2 => ThemeColors.RarityUncommon
        3 => ThemeColors.RarityRare
        4 => ThemeColors.RarityEpic
        5 => ThemeColors.RarityLegendary
        6 => ThemeColors.RarityMythic
        7 => ThemeColors.RarityExotic
        8 => ThemeColors.RarityImmortal
        9 => ThemeColors.RarityImmortal
        10 => ThemeColors.RarityImmortal
        _ => ThemeColors.RarityCommon

<#>
    Utility function to get tier string from integer values (1 = Common .. 9 = Immortal).

(Tier:int).ToTierString()<transacts>:string =
    case (Tier):
        1 => "Common"
        2 => "Uncommon"
        3 => "Rare"
        4 => "Epic"
        5 => "Legendary"
        6 => "Mythic"
        7 => "Exotic"
        _ => "Common"

<#>
    Utility function to make a string of a decimal number rounded to a specified decimal point
    Reference: https://forums.unrealengine.com/t/how-to-round-float-to-x-decimal-places/821189/5?u=jutpz1517

GetRoundedString(Input:float, DecimalPoints:int)<transacts>:string =
    if:
        Multiplier := Pow(10.0, DecimalPoints * 1.0)
        RoundedValue := float[Round[Input * Multiplier] * 1.0] / Multiplier
        BeforeDecimal := Floor[RoundedValue]
        var AfterDecimal:string = ToString(Abs(Round[(RoundedValue - BeforeDecimal * 1.0) * Multiplier]))

        for (Index := 0..(DecimalPoints - AfterDecimal.Length - 1)):
            set AfterDecimal = array{'0'} + AfterDecimal
    then:       
        if (DecimalPoints = 0):
            "{BeforeDecimal}"
        else:
            "{BeforeDecimal}.{AfterDecimal}"
    else:
        ToString(Input)
        
<#>
    Sleep function that scales with Time Dilation (only use when the game is paused).

SleepTimeDilation(Duration:float)<suspends>:void =
    StartTime := GetSecondsSinceEpoch()
    loop:
        Sleep(0.0)
        if (GetSecondsSinceEpoch() - StartTime >= Duration):
            break

<#>
    Function to calculate weighted random, returns a random index from array of weights.

GetWeightedRandom(Weights:[]int)<transacts>:int =
    var RandomIndex:int = -1
    var CumulativeWeights:[]float = array{}
    var MaxCumulativeWeight:float = 0.0

    for (Index -> Weight : Weights):
        PreviousCumulativeWeight := CumulativeWeights[Index - 1] or 0.0
        set CumulativeWeights = CumulativeWeights + array{Weight * 1.0 + PreviousCumulativeWeight}
        set MaxCumulativeWeight += Weight * 1.0

    RandomNumber := MaxCumulativeWeight * GetRandomFloat(0.0, 1.0)
    for:
        Index -> Weight : Weights
        CumulativeWeight := CumulativeWeights[Index]
        CumulativeWeight >= RandomNumber
        RandomIndex = -1
    do:
        set RandomIndex = Index

    RandomIndex

<#>
    Queue Implementation for handling data in a First In First Out (FIFO) order.

queue<public>(t:type) := class:
    Elements<internal>:[]t = array{}

    Enqueue<public>(NewElement:t)<transacts>:queue(t) =
        queue(t){Elements := Elements + array{NewElement}}
    
    Dequeue<public>()<decides><transacts>:tuple(queue(t), t) =
        FirstElement := Front[]
        (queue(t){Elements := Elements.RemoveElement[0]}, FirstElement)

    Size<public>()<transacts>:int =
        Elements.Length
    
    IsEmpty<public>()<transacts>:logic =
        logic{Size() = 0}
    
    Front<public>()<decides><transacts>:t =
        Elements[0]
    
    Rear<public>()<decides><transacts>:t =
        Elements[Elements.Length - 1]
        
CreateQueue<public><constructor>(InitialElements:[]t where t:type) := queue(t):
    Elements := InitialElements

<#>
    Utility function to format float into time format string (mm:ss).

(Time:float).FormatAsTime()<transacts>:string =
    Time.FormatAsTime(?IsHour := false)

(Time:float).FormatAsTime(?IsHour:logic = false)<transacts>:string =
    var FormattedTime:string = "00:00:00"
    if :
        var Seconds:int = Floor[Time]
        Hours := Floor[Time / 3600.0]
        SecondsFromHours := Mod[Seconds, 3600]
        set Seconds = SecondsFromHours
        Minutes := Floor[Seconds * 1.0 / 60.0]
        SecondsFromMinutes := Mod[Seconds,  60]
        set Seconds = SecondsFromMinutes
    then:
        if (IsHour?):
            set FormattedTime = "{Hours.FormatLeadingZeros(2)}:{Minutes.FormatLeadingZeros(2)}:{Seconds.FormatLeadingZeros(2)}"
        else:
            set FormattedTime = "{Minutes.FormatLeadingZeros(2)}:{Seconds.FormatLeadingZeros(2)}"

    FormattedTime

<#>
    Formats a number with leading zeros.

(Number:int).FormatLeadingZeros(Digits:int)<transacts>:string =
    var FormattedNumber:string = ToString(Number)
    Difference := Max(0, Digits - FormattedNumber.Length)
    if (Difference > 0):
        for (Index := 1..Difference):
            set FormattedNumber = "0" + FormattedNumber
    FormattedNumber

<#>
    Grant percentage health/shield to player.

(Player:player).HealPercent(Percent:type{_X:float where 0.000000 <= _X, _X <= 1.000000}):void =
    if (Character := Player.GetFortCharacter[]):
        Character.Heal(Character.GetMaxHealth() * Percent)

(Player:player).ShieldPercent(Percent:type{_X:float where 0.000000 <= _X, _X <= 1.000000}):void =
    if (Character := Player.GetFortCharacter[]):
        Character.SetShield(Character.GetShield() + Character.GetMaxShield() * Percent)

(Player:player).IsMaxHealth()<transacts>:logic =
    logic{Player.GetFortCharacter[].GetHealth() = Player.GetFortCharacter[].GetMaxHealth()}

(Player:player).IsMaxShield()<transacts>:logic =
    logic{Player.GetFortCharacter[].GetShield() = Player.GetFortCharacter[].GetMaxShield()}

<#>
    Returns vector3 with the z-axis zeroed out.

(Vector:vector3).Make2D()<transacts>:vector3 =
    vector3{X := Vector.X, Y := Vector.Y, Z := 0.0}

<#>
    Random functions.

GetRandomSign()<transacts>:float =
    if (GetRandomFloat(0.0, 1.0) <= 0.5) then 1.0 else -1.0

GetRandomSign(Factor:float)<transacts>:float =
    Factor * GetRandomSign()

<#>
    Check if position is within the area of a rectangle.

rectangle := struct:
    A:vector3
    B:vector3
    C:vector3

MakeRectangle(Position:vector3, Rotation:rotation, Width:float, Length:float)<transacts>:rectangle =
    ForwardDirection := Rotation.GetLocalForward()
    RightDirection := Rotation.GetLocalRight()
    
    A := Position - RightDirection * Width / 2.0
    B := Position + RightDirection * Width / 2.0
    C := B + ForwardDirection * Length
    
    rectangle:
        A := A
        B := B
        C := C

IsInRectangle(Position:vector3, Rectangle:rectangle)<transacts>:logic =
    DotABAM := DotProduct(Rectangle.B - Rectangle.A, Position - Rectangle.A)
    DotBCBM := DotProduct(Rectangle.C - Rectangle.B, Position - Rectangle.B)
    logic:
        0.0 <= DotABAM and DotABAM <= DotProduct(Rectangle.B - Rectangle.A, Rectangle.B - Rectangle.A) and
        0.0 <= DotBCBM and DotBCBM <= DotProduct(Rectangle.C - Rectangle.B, Rectangle.C - Rectangle.B)

<#>
    Function to get the nearest target agents to the specified agent.

GetNearestPosition(Position:vector3, TargetPositions:[]vector3)<transacts>:?int =
    var MaybeNearestIndex:?int = false
    var NearestDistenceSquared:float = 0.0
    
    for:
        Index -> TargetPosition : TargetPositions
        DistanceSquaredFromTarget := DistanceSquared(Position, TargetPosition)
        not MaybeNearestIndex? or DistanceSquaredFromTarget < NearestDistenceSquared
    do:
        set MaybeNearestIndex = option{Index}
        set NearestDistenceSquared = DistanceSquaredFromTarget

    MaybeNearestIndex

GetNearestAgent(Agent:agent, Targets:[]agent)<transacts>:?agent =
    var MaybeNearestTarget:?agent = false
    var NearestDistenceSquared:float = 0.0
    
    for:
        Target : Targets
        DistanceSquaredFromTarget := DistanceSquared(Agent.GetFortCharacter[].GetTransform().Translation, Target.GetFortCharacter[].GetTransform().Translation)
        not MaybeNearestTarget? or DistanceSquaredFromTarget < NearestDistenceSquared
    do:
        set MaybeNearestTarget = option{Target}
        set NearestDistenceSquared = DistanceSquaredFromTarget

    MaybeNearestTarget

<#>
    Calculate the sum of the array of numbers.

Sum(Numbers:[]int)<transacts>:int =
    var Total:int = 0
    for (Number : Numbers):
        set Total += Number
    Total

SumFloat(Numbers:[]float)<transacts>:float =
    var Total:float = 0.0
    for (Number : Numbers):
        set Total += Number
    Total

<#>
    Mod for floating point numbers.

FMod(X:float, Y:float)<decides><transacts>:float =
    if (X.IsFinite[] and not Y.IsFinite[] and Y <> NaN):
        X
    else:
        X - Int[X / Y] * Y

<#>
    Makes rotation that points from an initial position to a target position.

MakeRotation(From:vector3, To:vector3)<transacts>:rotation =
    Direction := (To - From).MakeUnitVector[] or vector3{}
    MakeRotation(vector3{X := 0.0, Y := 0.0, Z := 1.0}, ArcTan(Direction.Y, Direction.X))

<#>
    Utility function to reverse an array.

(Val:int).ToRoman()<transacts>:string =
    Letters := array{"M", "CM", "D", "CD", "C", "XC", "L", "XL", "X", "IX", "V", "IV", "I"}
    LookupValues := array{1000, 900, 500, 400, 100, 90, 50, 40, 10, 9, 5, 4, 1}
        
    var Result:string = ""
    var Number:int = Val
    
    for (Index := 0..LookupValues.Length - 1, Number > 0):
        loop:
            if:
                Number >= LookupValues[Index]
                Value := LookupValues[Index]
                Letter := Letters[Index]
            then:
                set Number -= Value
                set Result += Letter
            else:
                break

    Result

<#>
    Less functions used for sorting.

LessPlayerFloat(PlayerFloatA:tuple(player, float), PlayerFloatB:tuple(player, float))<computes><decides>:void =
    PlayerFloatA(1) > PlayerFloatB(1)