using { /Fortnite.com/Characters }
using { /Verse.org/Simulation }
using { /UnrealEngine.com/Temporary/SpatialMath }
using { /UnrealEngine.com/Temporary/UI }

<# ----------------------------------------------------------------------------------------------------
    Constants
---------------------------------------------------------------------------------------------------- #>
AbilityDashCooldownTime:float = 5.0

<# ----------------------------------------------------------------------------------------------------
    ability_dash
---------------------------------------------------------------------------------------------------- #>
ability_dash := class:
    Player:player

    RechargeCancelledEvent:event(void) = event(void){}
    EndEvent:event(void) = event(void){}

    var MaybeUI:?ability_dash_ui = false
    var Charge:int = 1
    var HasAirDashed:logic = false
    var IsRecharging:logic = false
    var PreviousPosition:vector3 = vector3{}
    var CurrentPosition:vector3 = vector3{}

    <# ----------------------------------------------------------------------------------------------------
        Initialization
    ---------------------------------------------------------------------------------------------------- #>
    Initialize()<suspends>:void =
        WaitForUIManager(Player)
        WaitForPlayerManager(Player)
        WaitForInputManager()
        
        AbilityHUDUI := ability_dash_ui{Player := Player}
        AbilityHUDUI.Initialize()
        
        Player.AddWidgetToHUD of canvas_slot:
            Anchors := anchors{Minimum := vector2{X := 0.5, Y := 1.0}, Maximum := vector2{X := 0.5, Y := 1.0}}
            Offsets := margin{Left := 0.0, Top := -62.0, Right := 0.0, Bottom := 50.0}
            Alignment := vector2{X := 0.5, Y := 1.0}
            Widget := AbilityHUDUI.GetUI()

        set MaybeUI = option{AbilityHUDUI}

        if (Player.GetDashMode() = dash_mode.InputTrigger):
            Player.RegisterInputs(DirectionInputTriggerIds)
        else:
            spawn{TrackPosition()}

    <# ----------------------------------------------------------------------------------------------------
        Get Functions
    ---------------------------------------------------------------------------------------------------- #>
    GetCooldownTime()<transacts>:float =
        AbilityDashCooldownTime / (1.0 + Player.GetAbilityValue(ability_id.DashCooldown))

    GetMaxCharge()<transacts>:int =
        1 + Floor[Player.GetPowerUpValue(power_up_id.DashCount)] or 0

    <# ----------------------------------------------------------------------------------------------------
        Update Functions
    ---------------------------------------------------------------------------------------------------- #>
    TrackPosition()<suspends>:void =
        if (Character := Player.GetFortCharacter[]):
            race:
                EndEvent.Await()
                loop:
                    set CurrentPosition = Character.GetTransform().Translation
                    Sleep(TickTime * 2)
                    set PreviousPosition = CurrentPosition

    Activate():void =
        if (Charge = 0):
            return

        ActivateDash()
        
        if (Charge < GetMaxCharge(), not IsRecharging?):
            spawn{Recharge()}

    Recharge()<suspends>:void =
        set IsRecharging = true
        defer:
            set IsRecharging = false

        race:
            RechargeCancelledEvent.Await()
            block:
                DeltaTimer := delta_timer{}
                if (UI := MaybeUI?):
                    loop:
                        Sleep(0.0)
                        Progress := DeltaTimer.GetElapsedTime() / GetCooldownTime()
                        UI.UpdateProgress(Progress)
                        if (DeltaTimer.GetElapsedTime() >= GetCooldownTime()):
                            set Charge = Min(GetMaxCharge(), Charge + 1)
                            UI.UpdateCharge(Charge)
                            if (Charge = GetMaxCharge()):
                                break
                            else:
                                DeltaTimer.Reset()

    Reset():void =
        RechargeCancelledEvent.Signal()
        if (UI := MaybeUI?):
            set Charge = GetMaxCharge()
            UI.UpdateProgress(1.0)
            UI.UpdateCharge(GetMaxCharge())

    ActivateDash():void =
        if:
            not Player.IsInStasis()?
            Character := Player.GetFortCharacter[]
            Character.IsActive[]
            not Character.IsDownButNotOut[]
        then:
            if (Character.IsOnGround[]):
                set HasAirDashed = false
            else:
                if (Player.GetPowerUpLevel(power_up_id.AirDash) <= 0 or HasAirDashed?):
                    return

                set HasAirDashed = true

            DashRotation := if (Player.GetDashMode() = dash_mode.InputTrigger):
                YawAmount := if (Player.IsInputHeld[input_id.Forward] and Player.IsInputHeld[input_id.Right]) {45.0}
                    else if (Player.IsInputHeld[input_id.Right] and Player.IsInputHeld[input_id.Backward]) {135.0}
                    else if (Player.IsInputHeld[input_id.Backward] and Player.IsInputHeld[input_id.Left]) {225.0}
                    else if (Player.IsInputHeld[input_id.Left] and Player.IsInputHeld[input_id.Forward]) {315.0}
                    else if (Player.IsInputHeld[input_id.Forward]) {0.0}
                    else if (Player.IsInputHeld[input_id.Right]) {90.0}
                    else if (Player.IsInputHeld[input_id.Backward]) {180.0}
                    else if (Player.IsInputHeld[input_id.Left]) {270.0}
                    else 0.0

                if (Player.GetActiveCamera() = camera_type.Orbit or Player.GetActiveCamera() = camera_type.OrbitFree):
                    if (Player.GetData[].Settings.AutoAim?):
                        Character.GetTransform().Rotation
                    else:
                        Character.GetTransform().Rotation.ApplyYaw(DegreesToRadians(YawAmount))
                else:
                    if (not Player.IsInputHeld[input_id.Forward] and not Player.IsInputHeld[input_id.Backward] and not Player.IsInputHeld[input_id.Left] and not Player.IsInputHeld[input_id.Right]):
                        Character.GetTransform().Rotation
                    else:
                        MakeRotationFromYawPitchRollDegrees(YawAmount, 0.0, 0.0)
            else:
                if (Direction := (CurrentPosition - PreviousPosition).MakeUnitVector[]):
                    MakeRotation(vector3{X := 0.0, Y := 0.0, Z := 1.0}, ArcTan(Direction.Y, Direction.X))
                else:
                    Character.GetTransform().Rotation

            Player.ApplyImpulse(Player.HasBoon(boon_id.DashingWings)? and impulse_id.DashBoon or impulse_id.Dash, ?Rotation := DashRotation)
            
            for:
                DownedPlayer : GetActivePlayers()
                DownedCharacter := DownedPlayer.GetFortCharacter[]
                DownedCharacter.IsDownButNotOut[]
                Distance(Character.GetTransform().Translation, DownedCharacter.GetTransform().Translation) <= 300.0
            do:
                DownedPlayer.ApplyImpulse(impulse_id.Dash, ?Rotation := DashRotation)
            
            Player.PlayAudio(audio_id.Impulse)
            Player.SetAFK(false)
            KnockbackNPCs(Character, DashRotation)
            set Charge -= 1
            if (UI := MaybeUI?):
                UI.UpdateCharge(Charge)

    KnockbackNPCs(Character:fort_character, DashRotationRotation:rotation):void =
        DashPosition := Character.GetTransform().Translation.Make2D()
        Rectangle := MakeRectangle(DashPosition, DashRotationRotation, 300.0, 700.0)

        for:
            NPCBehaviorBase : GetAllNPCs()
            NPCCharacter := NPCBehaviorBase.GetAgent[].GetFortCharacter[]
            IsInRectangle(NPCCharacter.GetTransform().Translation, Rectangle)?
        do:
            spawn{NPCBehaviorBase.Knockback(1.0, DashRotationRotation)}

    <#
        DashDamageMultiplier := Player.GetAbilityValue(ability_id.PlayerDashDamage)
        if (DashDamageMultiplier > 0.0):
            spawn{DashAttack(DashPosition, DashDamageMultiplier)}

    DashAttack(Position:vector3, DashDamageMultiplier:float)<suspends>:void =
        AttackTime := 1.0
        AttackRange := 200.0
        SpawnVFX(vfx_id.DashAttack, Position, ?Scale := vector3{X := AttackRange * 0.1, Y := AttackRange * 0.1, Z := AttackTime})
        Sleep(AttackTime)

        for:
            NPCAgent : GetAllNPCs()
            NPCCharacter := NPCAgent.GetFortCharacter[]
            NPCBehaviorBase := npc_behavior_base[NPCAgent.GetNPCBehavior[]]
            DistanceSquared(NPCCharacter.GetTransform().Translation, Position) <= Pow(AttackRange, 2.0)
        do:
            NPCBehaviorBase.Damage(Player, Player.GetDamageArgs(?Multiplier := DashDamageMultiplier))
    #>

    <# ----------------------------------------------------------------------------------------------------
        Event Handlers
    ---------------------------------------------------------------------------------------------------- #>
   