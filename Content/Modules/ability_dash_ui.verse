using { /Fortnite.com/UI }
using { /Verse.org/Colors }
using { /UnrealEngine.com/Temporary/SpatialMath }
using { /UnrealEngine.com/Temporary/UI }

ability_dash_ui := class(ui_component):       
    AbilityChargeText:text_block = text_block:
        DefaultTextColor := NamedColors.Orange
        DefaultShadowOpacity := 0.5
        DefaultShadowOffset := option{vector2{X := 2.0, Y := 2.0}}

    ProgressColorBlock:color_block = color_block:
        DefaultColor := NamedColors.White
        DefaultOpacity := 0.3
        DefaultDesiredSize := vector2{X := 0.0, Y := 0.0}

    Padding:float = 10.0
    IconSize:float = 40.0
    BorderWidth:float = 2.0

    <# ----------------------------------------------------------------------------------------------------
        Initialization
    ---------------------------------------------------------------------------------------------------- #>
    CreateUI<override>():widget =
        AbilityChargeText.SetText(MakeMessage("1"))

        stack_box:
            Orientation := orientation.Vertical
            Slots := array:
                stack_box_slot:
                    HorizontalAlignment := horizontal_alignment.Center
                    VerticalAlignment := vertical_alignment.Fill
                    Widget := AbilityChargeText
                stack_box_slot:
                    HorizontalAlignment := horizontal_alignment.Fill
                    VerticalAlignment := vertical_alignment.Fill
                    Widget := overlay:
                        Slots := array:
                            overlay_slot:
                                HorizontalAlignment := horizontal_alignment.Fill
                                VerticalAlignment := vertical_alignment.Fill
                                Widget := MakeBox(?BorderRadius := BorderWidth, ?Opacity := 0.6, ?BorderColor := NamedColors.Black, ?BorderOpacity := 0.3)
                            overlay_slot:
                                HorizontalAlignment := horizontal_alignment.Fill
                                VerticalAlignment := vertical_alignment.Bottom
                                Padding := margin{Left := BorderWidth + 1.0, Top := BorderWidth + 1.0, Right := BorderWidth + 1.0, Bottom := BorderWidth + 1.0}
                                Widget := ProgressColorBlock
                            overlay_slot:
                                Padding := margin{Left := Padding, Top := Padding, Right := Padding, Bottom := Padding}
                                Widget := texture_block:
                                    DefaultImage := UserInterface.Icons.DashCooldown_64
                                    DefaultDesiredSize := vector2{X := IconSize, Y := IconSize}

    <# ----------------------------------------------------------------------------------------------------
        Update Functions
    ---------------------------------------------------------------------------------------------------- #>
    UpdateProgress(Progress:float):void =      
        MaxHeight := IconSize + Padding * 2.0 - BorderWidth * 2.0
        Height := MaxHeight * (1.0 - Progress)
        ProgressColorBlock.SetDesiredSize(vector2{X := 1.0, Y := Height})

    UpdateCharge(Amount:int):void =
        AbilityChargeText.SetText(MakeMessage("{Amount}"))

    <# ----------------------------------------------------------------------------------------------------
        Event Handlers
    ---------------------------------------------------------------------------------------------------- #>
