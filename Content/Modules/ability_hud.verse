using { /Verse.org/Simulation }
using { /UnrealEngine.com/Temporary/SpatialMath }
using { /UnrealEngine.com/Temporary/UI }

ability_hud := class:
    Player:player

    var MaybeAbilityHUDUI:?ability_hud_ui = false

    <# ----------------------------------------------------------------------------------------------------
        Initialization
    ---------------------------------------------------------------------------------------------------- #>
    Initialize()<suspends>:void =
        WaitForGameManager()
        if (GameManager := GetGameManager[]):
            GameManager.GameStartedEvent.Subscribe(OnGameStarted)
            GameManager.GameEndedEvent.Subscribe(OnGameEnded)

        WaitForUIManager(Player)
        AbilityHUDUI := ability_hud_ui{Player := Player}
        AbilityHUDUI.Initialize()
        set MaybeAbilityHUDUI = option{AbilityHUDUI}

        WaitForPlayerManager(Player)
        if (PlayerStat := Player.GetStats[]):
            PlayerStat.AbilityChangedEvent.Subscribe(OnAbilityChanged)

        Player.AddWidgetToHUD of canvas_slot:
            Anchors := anchors{Minimum := vector2{X := 1.0, Y := 1.0}, Maximum := vector2{X := 1.0, Y := 1.0}}
            Offsets := margin{Left := -20.0, Top := -62.0, Right := 0.0, Bottom := 0.0}
            Alignment := vector2{X := 1.0, Y := 1.0}
            Widget := AbilityHUDUI.GetUI()

        AbilityHUDUI.Hide()

    <# ----------------------------------------------------------------------------------------------------
        Update Functions
    ---------------------------------------------------------------------------------------------------- #>
    Show():void =
        if (AbilityHUDUI := MaybeAbilityHUDUI?):
            AbilityHUDUI.Show()

    Hide():void =
        if (AbilityHUDUI := MaybeAbilityHUDUI?):
            AbilityHUDUI.Hide()

    Update(Index:int, WeaponInfo:weapon_info):void =
        if (AbilityHUDUI := MaybeAbilityHUDUI?):
            AbilityHUDUI.Update(Index, WeaponInfo)

    <# ----------------------------------------------------------------------------------------------------
        Event Handlers
    ---------------------------------------------------------------------------------------------------- #>
    OnGameStarted():void =
        if (AbilityHUDUI := MaybeAbilityHUDUI?):
            AbilityHUDUI.Reset()
            if (EquippedWeapon := Player.GetEquippedWeapon()?, WeaponInfo := GetWeaponInfo[EquippedWeapon]):
                AbilityHUDUI.Update(0, WeaponInfo)

        Show()

    OnGameEnded():void =
        Hide()

    OnAbilityChanged(AbilityId:ability_id):void =
        var Index:int = 0
        for (Id -> WeaponLevel : Player.GetWeaponLevels(), WeaponInfo := GetWeaponInfo[Id]):
            Update(Index, WeaponInfo)
            set Index += 1