using { /Fortnite.com/UI }
using { /Verse.org/Colors }
using { /UnrealEngine.com/Temporary/SpatialMath }
using { /UnrealEngine.com/Temporary/UI }

ability_hud_ui := class(ui_component):
    
    var Icons:[]texture_block = array{}
    var LevelTexts:[]text_block = array{}

    <# ----------------------------------------------------------------------------------------------------
        Initialization
    ---------------------------------------------------------------------------------------------------- #>
    CreateUI<override>():widget =
        overlay:
            Slots := array:
                overlay_slot:
                    HorizontalAlignment := horizontal_alignment.Fill
                    VerticalAlignment := vertical_alignment.Bottom
                    Widget := texture_block:
                        DefaultImage := UserInterface.HUD.AbilityHUD
                        DefaultDesiredSize := vector2{X := 256.0, Y := 70.0}
                overlay_slot:
                    HorizontalAlignment := horizontal_alignment.Fill
                    VerticalAlignment := vertical_alignment.Fill
                    Widget := stack_box:
                        Orientation := orientation.Horizontal
                        Slots := for (Index := 1..4):
                            Icon := texture_block:
                                DefaultImage := UserInterface.Icons.Lock_64
                                DefaultDesiredSize := vector2{X := 40.0, Y := 40.0}

                            LevelText := text_block{}

                            set Icons += array{Icon}
                            set LevelTexts += array{LevelText}

                            stack_box_slot:
                                HorizontalAlignment := horizontal_alignment.Fill
                                VerticalAlignment := vertical_alignment.Fill
                                Padding := margin{Left := (Index = 0) and 0.0 or 4.0, Top := 0.0, Right := 0.0, Bottom := 5.0}
                                Widget := stack_box:
                                    Orientation := orientation.Vertical
                                    Slots := array:
                                        stack_box_slot:
                                            HorizontalAlignment := horizontal_alignment.Center
                                            VerticalAlignment := vertical_alignment.Top
                                            Padding := margin{Left := 0.0, Top := 0.0, Right := 0.0, Bottom := 5.0}
                                            Widget := LevelText
                                        stack_box_slot:
                                            HorizontalAlignment := horizontal_alignment.Fill
                                            VerticalAlignment := vertical_alignment.Fill
                                            Padding := margin{Left := 10.0, Top := 10.0, Right := 10.0, Bottom := 10.0}
                                            Widget := Icon

    <# ----------------------------------------------------------------------------------------------------
        Get Functions
    ---------------------------------------------------------------------------------------------------- #>
    GetLevel(Index:int)<transacts>:string =
        Level := if (Index = 1). ""
            else if (Index = 2). "{WeaponUnlockLevel1}"
            else if (Index = 3). "{WeaponUnlockLevel2}"
            else. "{WeaponUnlockLevel3}"

    <# ----------------------------------------------------------------------------------------------------
        Update Functions
    ---------------------------------------------------------------------------------------------------- #>
    Update(Index:int, WeaponInfo:weapon_info):void =
        if (Icon := Icons[Index], LevelText := LevelTexts[Index]):
            Icon.SetImage(WeaponInfo.IconSmall)
            Icon.SetTint(NamedColors.White)
            LevelText.SetText(MakeMessage("{Player.GetWeaponLevel(WeaponInfo.Id)}"))
            LevelText.SetColor(ThemeColors.Highlight)

    Reset():void =
        for (Index := 1..Icons.Length):
            if (Icon := Icons[Index - 1], LevelText := LevelTexts[Index - 1]):
                Icon.SetImage(UserInterface.Icons.Lock_64)
                Icon.SetTint(ThemeColors.Locked)
                LevelText.SetText(MakeMessage("{GetLevel(Index)}"))
                LevelText.SetColor(ThemeColors.Failure)
