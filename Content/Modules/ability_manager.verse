using { /Fortnite.com/Characters }
using { /Fortnite.com/Devices }
using { /Verse.org/Random }
using { /Verse.org/Simulation }
using { /UnrealEngine.com/Temporary/SpatialMath }
using { /UnrealEngine.com/Temporary/UI }
using { WeaponAbilities }

<# ----------------------------------------------------------------------------------------------------
    Data Definitions
---------------------------------------------------------------------------------------------------- #>
ability_damage_args := struct:
    Target:npc_behavior_base
    Amount:float
    IsCriticalHit:logic = false

ability<public> := class<abstract>:
    Id<public>:weapon_id
    Player<public>:player

    var IsActive:logic = false
    var StartTime:float = 0.0
    var CastTime:float = 0.0
    var CooldownTimeRemaining:float = 0.0
    var TotalDamage:float = 0.0

    Initialize():void =
        set StartTime = GetSimulationElapsedTime()
        set TotalDamage = 0.0
        OnInitialize()
    
    OnInitialize():void = {}
    Activate():void
    Deactivate(?IsPause:logic = false):void
    Damage(DamageArgs:ability_damage_args, ?IsKnockback:logic = true):void =
        set TotalDamage += DamageArgs.Amount
        Player.AddGameStat(player_game_stat_id.DamageDealt, DamageArgs.Amount)
        DamageArgs.Target.Damage(Player, DamageArgs, ?IsKnockback := IsKnockback)

    GetIsActive()<transacts>:logic =
        IsActive

    GetTotalDamage()<transacts>:float =
        TotalDamage

    GetDuration(?EndTime:?float = false)<transacts>:float =
        if (Time := EndTime?):
            Time - StartTime
        else:
            GetSimulationElapsedTime() - StartTime

    GetDPS(?EndTime:?float = false)<transacts>:float =
        GetTotalDamage() / GetDuration(?EndTime := EndTime)

dash_mode := enum:
    InputTrigger
    Velocity

<# ----------------------------------------------------------------------------------------------------
    Global Functions
---------------------------------------------------------------------------------------------------- #>
var GlobalAbilityManager:weak_map(session, ability_manager) = map{}

GetAbilityManager()<decides><transacts>:ability_manager =
    GlobalAbilityManager[GetSession()]

(Player:player).GrantAbilityUpgrade():void =
    if (AbilityManager := GetAbilityManager[]):
        AbilityManager.GrantUpgrade(Player)

(Player:player).ShowAbilitySelectUI(Level:int):void =
    if (AbilityManager := GetAbilityManager[]):
        AbilityManager.ShowAbilitySelectUI(Player, Level)

(Player:player).ShowWeaponEvolutionUI(Id:weapon_id, ?IsItemShop:logic = false, ?IsPreview:logic = false):void =
    if (AbilityManager := GetAbilityManager[]):
        AbilityManager.ShowWeaponEvolutionUI(Player, Id, ?IsItemShop := IsItemShop, ?IsPreview := IsPreview)

(Player:player).ShowBoonSelectUI():void =
    if (AbilityManager := GetAbilityManager[]):
        AbilityManager.ShowBoonSelectUI(Player)

(Player:player).ActivateAbility():void =
    if (AbilityManager := GetAbilityManager[]):
        AbilityManager.ActivateAbility(Player)

(Player:player).ActivateNewAbility(Id:weapon_id):void =
    if (AbilityManager := GetAbilityManager[]):
        AbilityManager.ActivateNewAbility(Player, Id)

(Player:player).DeactivateAbility(?IsPause:logic = false):void =
    if (AbilityManager := GetAbilityManager[]):
        AbilityManager.DeactivateAbility(Player, ?IsPause := IsPause)

(Player:player).DeactivateAbility():void =
    Player.DeactivateAbility(?IsPause := false)

(Player:player).ReactivateAbility():void =
    if (AbilityManager := GetAbilityManager[]):
        AbilityManager.ReactivateAbility(Player)

(Player:player).ToggleAbility():void =
    if (AbilityManager := GetAbilityManager[]):
        AbilityManager.ToggleAbility(Player)

(Player:player).GetDamageArgs(Target:npc_behavior_base, Id:weapon_id, WeaponDamage:float, ?Multiplier:float = 1.0):ability_damage_args =
    var DamageMultplier:float = 1.0 + Player.GetAbilityValue(ability_id.Damage) + Player.GetPowerUpValue(power_up_id.Damage) + Multiplier
    set DamageMultplier += (Player.HasBoon(boon_id.BossHunter)? and Target.GetIsBoss()?) and BoonBossHunterMultiplier or 0.0
    set DamageMultplier += (Player.HasBoon(boon_id.WoodenClub)?) and BoonGetValueWoodenClub(Player) or 0.0
    set DamageMultplier += (Player.HasBoon(boon_id.AllForOne)?) and (Id = Player.GetEquippedWeapon()? and 0.5 or -0.25) or 0.0
    set DamageMultplier += (Player.HasBoon(boon_id.MultiTool)?) and BoonGetValueMultiTool(Player) or 0.0
    var DamageAmount:float = WeaponDamage * Max(1.0, DamageMultplier)

    var CriticalRate:float = Player.GetAbilityValue(ability_id.CriticalRate) + Player.GetPowerUpValue(power_up_id.PlayerCriticalRate)
    set CriticalRate += (Player.HasBoon(boon_id.WoodenArrow)?) and BoonGetValueWoodenArrow(Player) or 0.0

    IsCriticalHit := logic{GetRandomFloat(0.0, 1.0) <= CriticalRate}
    if (IsCriticalHit?):
        CriticalOverdrive := Player.HasBoon(boon_id.CriticalOverdrive)? and Max(0.0, (CriticalRate - 1.0) * 2) or 0.0
        CriticalMultiplier := 1.0 + Player.GetAbilityValue(ability_id.CriticalDamage) + Player.GetPowerUpValue(power_up_id.PlayerCriticalDamage) + CriticalOverdrive
        set DamageAmount = DamageAmount * (IsCriticalHit? and CriticalMultiplier or 0.0)

    ability_damage_args:
        Target := Target
        Amount := DamageAmount
        IsCriticalHit := IsCriticalHit

(Player:player).GetDashMode()<transacts>:dash_mode =
    GetAbilityManager[].GetDashMode(Player) or dash_mode.InputTrigger

GetPlayerActiveAbilities()<transacts>:[player][weapon_id]ability =
    GetAbilityManager[].PlayerActiveAbilities or map{}

(Player:player).GetWeaponDPS(WeaponId:weapon_id, ?EndTime:?float = false)<transacts>:float =
    GetAbilityManager[].GetWeaponDPS(Player, WeaponId, ?EndTime := EndTime) or 0.0

(Player:player).GetWeaponTotalDamage(WeaponId:weapon_id)<transacts>:float =
    GetAbilityManager[].GetWeaponTotalDamage(Player, WeaponId) or 0.0

(Player:player).GetWeaponDuration(WeaponId:weapon_id, ?EndTime:?float = false)<transacts>:float =
    GetAbilityManager[].GetWeaponDuration(Player, WeaponId, ?EndTime := EndTime) or 0.0

(NPCBehavior:npc_behavior_base).ActivateMovementModulator(Index:int):void =
    if:
        AbilityManager := GetAbilityManager[]
        MovementModulator := AbilityManager.MovementModulators[Clamp(Index, 0, AbilityManager.MovementModulators.Length - 1)]
        Agent := NPCBehavior.GetAgent[]
    then:
        MovementModulator.Activate(Agent)

(NPCBehavior:npc_behavior_base).DeactivateMovementModulator():void =
    if (AbilityManager := GetAbilityManager[], Agent := NPCBehavior.GetAgent[]):
        AbilityManager.MovementModulatorDisable.Activate(Agent)

(Player:player).UpdateMaxHealth():void =
    if (AbilityManager := GetAbilityManager[]):
        AbilityManager.UpdateMaxHealth(Player)

<# ----------------------------------------------------------------------------------------------------
    ability_manager
---------------------------------------------------------------------------------------------------- #>
ability_manager := class(creative_device):
    @editable
    HealthPowerup<private>:health_powerup_device = health_powerup_device{}
    
    @editable
    HealthPowerupDisable<private>:health_powerup_device = health_powerup_device{}

    @editable
    MovementModulators:[]movement_modulator_device = array{}

    @editable
    MovementModulatorDisable:movement_modulator_device = movement_modulator_device{}

    @editable
    InputTriggerDash:input_trigger_device = input_trigger_device{}

    @editable
    SwitchDashMode:switch_device = switch_device{}

    var PlayerAbilityUpgradeUI:[player]ability_upgrade_ui = map{}
    var PlayerAbilitySelectUI:[player]ability_select_ui = map{}
    var PlayerBoonSelectUI:[player]boon_select_ui = map{}
    var PlayerWeaponEvolutionUI:[player]weapon_evolution_ui = map{}
    var PlayerDashAbility:[player]ability_dash = map{}
    var PlayerAbilityHUD:[player]ability_hud = map{}
    var PlayerIsInitialized:[player]logic = map{}

    var PlayerActiveAbilities:[player][weapon_id]ability = map{}
    var WeaponAbilityInfo:[]ability_info = array{}
    var PlayerAbilityUpgradeCounter:[player]int = map{}

    <# ----------------------------------------------------------------------------------------------------
        Initialization
    ---------------------------------------------------------------------------------------------------- #>
    OnBegin<override>()<suspends>:void =
        if (set GlobalAbilityManager[GetSession()] = Self) {}

        Playspace := GetPlayspace()
        Playspace.PlayerAddedEvent().Subscribe(OnPlayerAdded)
        Playspace.PlayerRemovedEvent().Subscribe(OnPlayerRemoved)

        for (Player : Playspace.GetPlayers()):
            spawn{Initialize(Player)}

        WaitForGameManager()
        if (GameManager := GetGameManager[]):
            GameManager.GameStartedEvent.Subscribe(OnGameStarted)
            GameManager.GameEndedEvent.Subscribe(OnGameEnded)

        set WeaponAbilityInfo = GetWeaponAbilityInfo()

        InputTriggerDash.PressedEvent.Subscribe(OnPressedDash)
        SwitchDashMode.TurnedOffEvent.Subscribe(OnTurnedOffDashMode)
        SwitchDashMode.TurnedOnEvent.Subscribe(OnTurnedOnDashMode)

    Initialize(Player:player)<suspends>:void =
        if (PlayerIsInitialized[Player]?) then return
        if (set PlayerIsInitialized[Player] = true) {}

        WaitForPlayerManager(Player)
        if (PlayerStats := Player.GetStats[]):
            spawn{AwaitAbilityChanged(Player, PlayerStats.AbilityChangedEvent)}
            spawn{AwaitStatsReset(Player, PlayerStats.StatsResetEvent)}

        AbilityDash := ability_dash{Player := Player}
        AbilityDash.Initialize()
        if (set PlayerDashAbility[Player] = AbilityDash) {}

        AbilityHUD := ability_hud{Player := Player}
        AbilityHUD.Initialize()
        if (set PlayerAbilityHUD[Player] = AbilityHUD) {}

        WaitForInputManager()
        Player.RegisterInput(input_id.Dash)
        if (InputPressedEvent := GetInputPressedEvent[input_id.OpenTreasureChest]):
            InputPressedEvent.Subscribe(OnInputPressedOpenTreasureChest)

        if (InputPressedEvent := GetInputPressedEvent[input_id.AutoAim]):
            InputPressedEvent.Subscribe(OnInputPressedAutoAim)

    <# ----------------------------------------------------------------------------------------------------
        Get Functions
    ---------------------------------------------------------------------------------------------------- #>
    GetAbility(Player:player, Id:weapon_id)<transacts>:ability =
        case (Id):
            weapon_id.Sword => ability_sword{Player := Player}
            weapon_id.Bow => ability_bow{Player := Player}
            weapon_id.Shuriken => ability_shuriken{Player := Player}
            weapon_id.Sentry => ability_sentry{Player := Player}
            weapon_id.Aura => ability_aura{Player := Player}
            weapon_id.Torrent => ability_torrent{Player := Player}
            weapon_id.Lightning => ability_lightning{Player := Player}
            weapon_id.Trail => ability_trail{Player := Player}
            weapon_id.Shield => ability_shield{Player := Player}
            weapon_id.ChainLightning => ability_chain_lightning{Player := Player}
            weapon_id.EnergyBomb => ability_energy_bomb{Player := Player}
            weapon_id.Boomerang => ability_boomerang{Player := Player}

    GetWeaponAbilityInfo(Player:player)<transacts>:[]ability_info =
        for:
            AbilityInfo : WeaponAbilityInfo
            WeaponId := AbilityInfo.MaybeWeaponId?
            Player.HasWeapon(WeaponId)?
            not AbilityInfo.MaybeMinLevel? or GetLevel() >= AbilityInfo.MaybeMinLevel?
            not AbilityInfo.Id = ability_id.WeaponLevel or (AbilityInfo.Id = ability_id.WeaponLevel and Player.GetWeaponLevel(WeaponId) < MaxWeaponLevel - 1)
        do:
            AbilityInfo

    GetWeaponAbilityInfo(WeaponId:weapon_id, AbilityId:ability_id)<transacts>:?ability_info =
        var WeaponAbilityInfoResult:?ability_info = false
        for (AbilityInfo : WeaponAbilityInfo, not WeaponAbilityInfoResult?):
            if (WeaponId = AbilityInfo.MaybeWeaponId?, AbilityId = AbilityInfo.Id):
                set WeaponAbilityInfoResult = option{AbilityInfo}

        WeaponAbilityInfoResult

    GetWeaponAbilityInfo()<transacts>:[]ability_info =
        Concatenate of array:
            for (WeaponInfo : WeaponInfos):
                ability_info:
                    MaybeWeaponId := option{WeaponInfo.Id}
                    Id := ability_id.Damage
                    Name := "{WeaponInfo.Id} Damage"
                    Description := "Increase damage of {WeaponInfo.Id} weapon."
                    Icon := WeaponInfo.Icon
                    IconSmall := WeaponInfo.IconSmall
                    MaybeStatIcon := option{UserInterface.Icons.ProjectileDamage_64}
                    StatIconColor := ThemeColors.WeaponDamage
                    TierValue := array{option{0.1}, option{0.15}, option{0.25}, option{0.35}, option{0.5}}
                    FormatValue := FormatAsPercent
    
            for (WeaponInfo : WeaponInfos):
                ability_info:
                    MaybeWeaponId := option{WeaponInfo.Id}
                    Id := ability_id.Cooldown
                    Name := (WeaponInfo.Id = weapon_id.Trail) and "{WeaponInfo.Id} Duration" or "{WeaponInfo.Id} Cooldown"
                    Description := (WeaponInfo.Id = weapon_id.Trail) and "Increase duration of {WeaponInfo.Id} weapon." or "Reduce cooldown of {WeaponInfo.Id} weapon."
                    Icon := WeaponInfo.Icon
                    IconSmall := WeaponInfo.IconSmall
                    MaybeStatIcon := option{UserInterface.Icons.AbilityCooldown_64}
                    StatIconColor := ThemeColors.WeaponCooldown
                    TierValue := array{option{0.1}, option{0.15}, option{0.25}, option{0.35}, option{0.5}}
                    FormatValue := FormatAsPercent
    
            for (WeaponInfo : WeaponInfos):
                ability_info:
                    MaybeWeaponId := option{WeaponInfo.Id}
                    Id := ability_id.Size
                    Name := "{WeaponInfo.Id} Size"
                    Description := "Increase size of {WeaponInfo.Id} weapon."
                    Icon := WeaponInfo.Icon
                    IconSmall := WeaponInfo.IconSmall
                    MaybeStatIcon := option{UserInterface.Icons.AbilitySize_64}
                    StatIconColor := ThemeColors.WeaponSize
                    TierValue := array{option{0.1}, option{0.15}, option{0.25}, option{0.35}, option{0.5}}
                    FormatValue := FormatAsPercent

            for (WeaponInfo : WeaponInfos):
                ability_info:
                    MaybeWeaponId := option{WeaponInfo.Id}
                    Id := ability_id.WeaponLevel
                    Name := "{WeaponInfo.Id} Level"
                    Description := "Increase the level of {WeaponInfo.Id} weapon."
                    Icon := WeaponInfo.Icon
                    IconSmall := WeaponInfo.IconSmall
                    MaybeStatIcon := option{UserInterface.Icons.WeaponLevel_64}
                    StatIconColor := ThemeColors.WeaponLevel
                    TierValue := array{false, false, option{1.0}, option{2.0}, false}
                    MaybeMinLevel := option{6}
                    FormatValue := FormatAsInt

    GetDashMode(Player:player)<transacts>:dash_mode =
        SwitchDashMode.GetCurrentState[Player] and dash_mode.InputTrigger or dash_mode.Velocity

    GetWeaponDPS(Player:player, WeaponId:weapon_id, ?EndTime:?float = false)<transacts>:float =
        PlayerActiveAbilities[Player][WeaponId].GetDPS(?EndTime := EndTime) or 0.0

    GetWeaponTotalDamage(Player:player, WeaponId:weapon_id)<transacts>:float =
        PlayerActiveAbilities[Player][WeaponId].GetTotalDamage() or 0.0

    GetWeaponDuration(Player:player, WeaponId:weapon_id, ?EndTime:?float = false)<transacts>:float =
        PlayerActiveAbilities[Player][WeaponId].GetDuration(?EndTime := EndTime) or 0.0

    <# ----------------------------------------------------------------------------------------------------
        Update Functions
    ---------------------------------------------------------------------------------------------------- #>
    ShowAbilitySelectUI(Player:player, Level:int, ?IsItemShop:logic = false):void =
        UI := if (AbilitySelectUI := PlayerAbilitySelectUI[Player]):
            AbilitySelectUI
        else:
            AbilitySelectUI := ability_select_ui{Player := Player}
            if (set PlayerAbilitySelectUI[Player] = AbilitySelectUI) {}
            AbilitySelectUI

        UI.Show()
        UI.OnShow(Level)

    ShowWeaponEvolutionUI(Player:player, Id:weapon_id, ?IsItemShop:logic = false, ?IsPreview:logic = false):void =
        UI := if (WeaponEvolutionUI := PlayerWeaponEvolutionUI[Player]):
            WeaponEvolutionUI
        else:
            WeaponEvolutionUI := weapon_evolution_ui{Player := Player}
            if (set PlayerWeaponEvolutionUI[Player] = WeaponEvolutionUI) {}
            WeaponEvolutionUI

        UI.Show()
        UI.OnShow(Id, ?IsItemShop := IsItemShop, ?InIsPreview := IsPreview)

    GrantUpgrade(Player:player):void =
        UI := if (AbilityUpgradeUI := PlayerAbilityUpgradeUI[Player]):
            AbilityUpgradeUI
        else:
            AbilityUpgradeUI := ability_upgrade_ui{Player := Player}
            if (set PlayerAbilityUpgradeUI[Player] = AbilityUpgradeUI) {}
            AbilityUpgradeUI
            
        if (RandomAbility := GetTreasureChestRandomAbility(Player)?):
            UI.Show()
            UI.Update(RandomAbility)

    ShowBoonSelectUI(Player:player):void =
        BoonSelectUI := if (UI := PlayerBoonSelectUI[Player]):
            UI
        else:
            UI := boon_select_ui{Player := Player}
            if (set PlayerBoonSelectUI[Player] = UI) {}
            UI

        BoonSelectUI.Show()

    ActivateAbility(Player:player):void =
        if (EquippedWeapon := Player.GetEquippedWeapon()?):
            Ability := GetAbility(Player, EquippedWeapon)
            Ability.Initialize()
            Ability.Activate()
            if (set PlayerActiveAbilities[Player] = map{EquippedWeapon => Ability}) {}

    ActivateNewAbility(Player:player, Id:weapon_id):void =
        Ability := GetAbility(Player, Id)
        Ability.Initialize()

        if (Character := Player.GetFortCharacter[], not Character.IsDownButNotOut[], not Player.IsPlayerPaused()?):
            Ability.Activate()
        
        if (AbilityHUD := PlayerAbilityHUD[Player], AbilityIndex := PlayerActiveAbilities[Player].Length, WeaponInfo := GetWeaponInfo[Id]):
            AbilityHUD.Update(AbilityIndex, WeaponInfo)

        if:
            (set PlayerActiveAbilities[Player][Id] = Ability) or
            (set PlayerActiveAbilities[Player] = map{Id => Ability})

    DeactivateAbility(Player:player, ?IsPause:logic = false):void =
        for (ActiveAbility : PlayerActiveAbilities[Player], ActiveAbility.IsActive?):
            ActiveAbility.Deactivate(?IsPause := IsPause)

    ReactivateAbility(Player:player):void =
        if (Character := Player.GetFortCharacter[], Character.IsDownButNotOut[]) then return
        if (Player.IsInStageVolume()?) then return

        for (ActiveAbility : PlayerActiveAbilities[Player], not ActiveAbility.IsActive?):
            ActiveAbility.Activate()

    ToggleAbility(Player:player):void =
        if (ActiveAbilities := PlayerActiveAbilities[Player]):
            for (ActiveAbility : ActiveAbilities):
                if (ActiveAbility.GetIsActive()?):
                    DeactivateAbility(Player)
                else:
                    ReactivateAbility(Player)

        else if (EquippedWeapon := Player.GetEquippedWeapon()?):
            ActivateNewAbility(Player, EquippedWeapon)

    UpdateMaxHealth(Player:player):void =
        if (Character := Player.GetFortCharacter[]):
            var MaxHealth:float = Player.GetAbilityValue(ability_id.MaxHealth)          
            set MaxHealth += Player.HasBoon(boon_id.BossSiphon)? and BoonGetValueBossSiphon(Player) or 0.0
            set MaxHealth += Player.HasBoon(boon_id.ImmortalHeart)? and BoonGetValueImmortalHeart(Player) or 0.0
            Character.SetMaxHealth(MaxHealth)

    UpdateHealthRegeneration(Player:player, AbilityValue:float):void =
        HealthPowerup.Spawn()
        HealthPowerup.SetMagnitude(AbilityValue)
        HealthPowerup.Pickup(Player)

    GrantAbilityUpgrade(Player:player)<suspends>:void =
        if (set PlayerAbilityUpgradeCounter[Player] = 1) {}
        Player.ActivateVFXPowerup(vfx_powerup_id.TreasureChestOpen, ?Duration := 1.0)
        
        Sleep(1.0)
        for (Index := 0..PlayerAbilityUpgradeCounter[Player] - 1):
            PauseGame(Player, pause_type.AbilityUpgrade, 0)

        if (set PlayerAbilityUpgradeCounter[Player] = 0) {}
    
    <# ----------------------------------------------------------------------------------------------------
        Event Handlers
    ---------------------------------------------------------------------------------------------------- #>
    OnPlayerAdded(Player:player):void =
        spawn{Initialize(Player)}

    OnPlayerRemoved(Player:player):void =
        if (DashAbility := PlayerDashAbility[Player]):
            DashAbility.EndEvent.Signal()

        if (ActiveAbilities := PlayerActiveAbilities[Player]):
            for (ActiveAbility : ActiveAbilities):
                ActiveAbility.Deactivate()

        set PlayerAbilityUpgradeUI = RemoveKeyFromMap(PlayerAbilityUpgradeUI, Player)
        set PlayerAbilitySelectUI = RemoveKeyFromMap(PlayerAbilitySelectUI, Player)
        set PlayerWeaponEvolutionUI = RemoveKeyFromMap(PlayerWeaponEvolutionUI, Player)
        set PlayerBoonSelectUI = RemoveKeyFromMap(PlayerBoonSelectUI, Player)
        set PlayerDashAbility = RemoveKeyFromMap(PlayerDashAbility, Player)
        set PlayerActiveAbilities = RemoveKeyFromMap(PlayerActiveAbilities, Player)
        set PlayerAbilityHUD = RemoveKeyFromMap(PlayerAbilityHUD, Player)
        set PlayerAbilityUpgradeCounter = RemoveKeyFromMap(PlayerAbilityUpgradeCounter, Player)
        set PlayerIsInitialized = RemoveKeyFromMap(PlayerIsInitialized, Player)

    AwaitAbilityChanged(Player:player, AbilityChangedEvent:events(ability_id))<suspends>:void =
        loop:
            AbilityId := AbilityChangedEvent.Await()
            AbilityLevel := Player.GetAbilityLevel(AbilityId)
            AbilityValue := Player.GetAbilityValue(AbilityId)

            case (AbilityId):
                ability_id.HealthRegeneration =>
                    if (AbilityLevel = 0):
                        HealthPowerupDisable.Pickup(Player)
                    else:
                        UpdateHealthRegeneration(Player, AbilityValue)

                ability_id.MaxHealth =>
                    UpdateMaxHealth(Player)

                ability_id.MoveSpeed =>                    
                    if (AbilityLevel = 0):
                        MovementModulatorDisable.Activate(Player)
                    else if (MovementModulator := MovementModulators[Clamp(AbilityLevel - 1, 0, MovementModulators.Length - 1)]):
                        MovementModulator.Activate(Player)

                _ => {}

    AwaitStatsReset(Player:player, StatsResetEvent:events(tuple()))<suspends>:void =
        loop:
            StatsResetEvent.Await()
            if (Character := Player.GetFortCharacter[]):
                Character.SetMaxHealth(100.0)
                Character.SetHealth(100.0)
                Character.SetMaxShield(100.0)
                Character.SetShield(0.0)
                HealthPowerupDisable.Pickup(Player)
                MovementModulatorDisable.Activate(Player)

    OnTurnedOffDashMode(Agent:agent):void =
        if (Player := player[Agent], DashAbility := PlayerDashAbility[Player]):
            Player.UnregisterInputs(DirectionInputTriggerIds)
            spawn{DashAbility.TrackPosition()}

    OnTurnedOnDashMode(Agent:agent):void =
        if (Player := player[Agent], DashAbility := PlayerDashAbility[Player]):
            Player.RegisterInputs(DirectionInputTriggerIds)
            DashAbility.EndEvent.Signal()

    OnPressedDash(Agent:agent):void =
        if (Player := player[Agent], DashAbility := PlayerDashAbility[Player]):
            DashAbility.Activate()

    OnGameStarted():void =
        for (Player : GetActivePlayers(), DashAbility := PlayerDashAbility[Player]):
            DashAbility.Reset()
            UpdateMaxHealth(Player)
            UpdateHealthRegeneration(Player, Player.GetAbilityValue(ability_id.HealthRegeneration))

    OnGameEnded():void =
        for (Player : GetAllPlayers()):
            Player.DeactivateAbility()
        
        set PlayerActiveAbilities = map{}

    OnInputPressedOpenTreasureChest(Agent:agent):void =
        if (Player := player[Agent]):
            if (Player.IsInStageVolume()?) then return

            if (Player.GetGameCurrency(game_currency_id.TreasureChest) > 0.0):
                if (PlayerAbilityUpgradeCounter[Player] > 0):
                    if (set PlayerAbilityUpgradeCounter[Player] += 1) {}
                else:
                    spawn{GrantAbilityUpgrade(Player)}

                Player.RemoveGameCurrency(game_currency_id.TreasureChest, 1.0)
                Player.PlayAudio(audio_id.OpenTreasureChest)

    OnInputPressedAutoAim(Agent:agent):void =
        if (Player := player[Agent]):
            Player.ToggleAutoAim()
            Player.RefreshControls()
            Player.PlayAudio(audio_id.Activate)
            Player.ShowMessage(message_id.AutoAim, "{Player.GetData[].Settings.AutoAim? and "ON" or "OFF"}", ?DisplayTime := 3.0)