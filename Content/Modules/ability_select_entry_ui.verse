using { /Fortnite.com/UI }
using { /Verse.org/Colors }
using { /Verse.org/Simulation }
using { /UnrealEngine.com/Temporary/SpatialMath }
using { /UnrealEngine.com/Temporary/UI }

<# ----------------------------------------------------------------------------------------------------
    Constants
---------------------------------------------------------------------------------------------------- #>
AbilitySelectHealPercent:type{_X:float where 0.000000 <= _X, _X <= 1.000000} = 0.25
AbilitySelectShieldPercent:type{_X:float where 0.000000 <= _X, _X <= 1.000000} = 0.25

<# ----------------------------------------------------------------------------------------------------
    Data Definitions
---------------------------------------------------------------------------------------------------- #>
ability_select_mode := enum:
    Default
    Banish
    Boost

<# ----------------------------------------------------------------------------------------------------
    ability_select_entry_ui
---------------------------------------------------------------------------------------------------- #>
ability_select_entry_ui := class(ui_component):
    HasButton:logic = true
    SelectEvent:events(tuple(ability_info, int, logic)) = events(tuple(ability_info, int, logic)){}
    SelectWeaponEvent:events(weapon_info) = events(weapon_info){}
    CancelAnimationEvent:events(tuple()) = events(tuple()){}

    SelectButton:button_quiet = button_quiet{}

    Icon:texture_block = texture_block:
        DefaultImage := UserInterface.Icons.WeaponSword_128
        DefaultDesiredSize := vector2{X := 128.0, Y := 128.0}

    StatIcon:texture_block = texture_block:
        DefaultImage := UserInterface.Icons.WeaponSword_64
        DefaultDesiredSize := vector2{X := 64.0, Y := 64.0}

    NameText:text_block = text_block:
        DefaultTextColor := ThemeColors.PrimaryText
        DefaultShadowOffset := option{vector2{X := 2.0, Y := 2.0}}
        DefaultShadowOpacity := 0.5

    DescriptionText:text_block = text_block:
        DefaultTextColor := ThemeColors.PrimaryText
        DefaultTextOpacity := 0.5

    TierText:text_block = text_block:
        DefaultShadowOffset := option{vector2{X := 2.0, Y := 2.0}}
        DefaultShadowOpacity := 1.0

    TierProjectionText:text_block = text_block:
        DefaultShadowOffset := option{vector2{X := 2.0, Y := 2.0}}
        DefaultShadowOpacity := 1.0

    StatText:text_block = text_block:
        DefaultTextColor := ThemeColors.PrimaryText
        DefaultShadowOffset := option{vector2{X := 2.0, Y := 2.0}}
        DefaultShadowOpacity := 0.5

    StatDeltaText:text_block = text_block:
        DefaultTextColor := ThemeColors.Success
        DefaultShadowColor := ThemeColors.Success * 0.125
        DefaultShadowOffset := option{vector2{X := 2.0, Y := 2.0}}
        DefaultShadowOpacity := 1.0

    BanishIcon:texture_block = texture_block:
        DefaultImage := UserInterface.Icons.GameProgressBar_Cross_64
        DefaultDesiredSize := vector2{X := 40.0, Y := 40.0}
        #DefaultTint := ThemeColors.Failure

    var EvolutionStackBox:stack_box = stack_box:
        Orientation := orientation.Horizontal

    var BoxUIs:[]box_ui = array{}
    var ProgressUI:widget = canvas{}
    var InfoOverlay:overlay = overlay{}
    var MaybeAbilityInfo:?ability_info = false
    var MaybeWeaponInfo:?weapon_info = false
    var ProjectionBoxUIs:[]box_ui = array{}
    var MaybeBoxUI:?box_ui = false

    var ActiveAbilitySelectMode:ability_select_mode = ability_select_mode.Default
    var Tier:int = 1
    var IsEvolution:logic = false
        
    <# ----------------------------------------------------------------------------------------------------
        Create UI
    ---------------------------------------------------------------------------------------------------- #>
    CreateUI<override>():widget =
        BoxUI := MakeBoxUI(?DefaultBorderRadius := 2.0)
        set MaybeBoxUI = option{BoxUI}

        StackBox := stack_box:
            Orientation := orientation.Vertical
            Slots := array:
                stack_box_slot:
                    HorizontalAlignment := horizontal_alignment.Fill
                    VerticalAlignment := vertical_alignment.Fill
                    Distribution := option{1.0}
                    Widget := set InfoOverlay = overlay:
                        Slots := array:
                            overlay_slot:
                                HorizontalAlignment := horizontal_alignment.Fill
                                VerticalAlignment := vertical_alignment.Fill
                                Widget :=  BoxUI.GetWidget()
                            overlay_slot:
                                HorizontalAlignment := horizontal_alignment.Center
                                VerticalAlignment := vertical_alignment.Center
                                Widget := stack_box:
                                    Orientation := orientation.Horizontal
                                    Slots := array:
                                        stack_box_slot:
                                            HorizontalAlignment := horizontal_alignment.Fill
                                            VerticalAlignment := vertical_alignment.Fill
                                            Padding := margin{Left := 2.0, Top := 2.0, Right := 2.0, Bottom := 2.0}
                                            Widget := CreateProgressUI()
                                        stack_box_slot:
                                            HorizontalAlignment := horizontal_alignment.Fill
                                            VerticalAlignment := vertical_alignment.Fill
                                            Padding := margin{Left := 40.0, Top := 40.0, Right := 40.0, Bottom := 40.0}
                                            Distribution := option{1.0}
                                            Widget := CreateInfoUI()

        if (HasButton?):
            SelectButton.OnClick().Subscribe(OnClickSelect)
            StackBox.AddWidget of stack_box_slot:
                HorizontalAlignment := horizontal_alignment.Fill
                VerticalAlignment := vertical_alignment.Fill
                Padding := margin{Left := 0.0, Top := 10.0, Right := 0.0, Bottom := 0.0}
                Widget := SelectButton

        return StackBox

    CreateProgressUI():widget =
        overlay:
            Slots := array:
                overlay_slot:
                    HorizontalAlignment := horizontal_alignment.Fill
                    VerticalAlignment := vertical_alignment.Fill
                    Widget := color_block:
                        DefaultColor := NamedColors.White
                        DefaultOpacity := 0.1
                        DefaultDesiredSize := vector2{X := 0.0, Y := 600.0}
                overlay_slot:
                    HorizontalAlignment := horizontal_alignment.Center
                    VerticalAlignment := vertical_alignment.Top
                    Padding := margin{Left := 10.0, Top := 20.0, Right := 10.0, Bottom := 20.0}
                    Widget := set ProgressUI = stack_box:
                        Orientation := orientation.Vertical
                        Slots := for (Index := 1..MaxWeaponLevel):
                            IsTierUp := logic{Index = 1 or Mod[Index - 1, WeaponEvolutionInterval] = 0}
                            
                            BoxHeight := IsTierUp? and 30.0 or 15.0
                            BoxUI := MakeBoxUI(?DefaultWidth := 30.0, ?DefaultHeight := BoxHeight, ?DefaultColor := ThemeColors.Disabled, ?DefaultOpacity := 1.0, ?DefaultBorderColor := NamedColors.Black, ?DefaultBorderOpacity := 0.5)
                            set BoxUIs += array{BoxUI}

                            stack_box_slot:
                                HorizontalAlignment := horizontal_alignment.Fill
                                VerticalAlignment := vertical_alignment.Top
                                Padding := margin{Left := 0.0, Top := IsTierUp? and 8.0 or 0.0, Right := 0.0, Bottom := IsTierUp? and 8.0 or 2.0}
                                Widget := if (IsTierUp?):
                                    overlay:
                                        Slots := array:
                                            overlay_slot:
                                                HorizontalAlignment := horizontal_alignment.Fill
                                                VerticalAlignment := vertical_alignment.Fill
                                                Widget := BoxUI.GetWidget()
                                            overlay_slot:
                                                HorizontalAlignment := horizontal_alignment.Center
                                                VerticalAlignment := vertical_alignment.Center
                                                Widget := texture_block:
                                                    DefaultImage := UserInterface.Icons.Star_Flat_64
                                                    DefaultDesiredSize := vector2{X := 25.0, Y := 25.0}
                                                    DefaultTint := NamedColors.Black
                                else:
                                    BoxUI.GetWidget()

    CreateInfoUI():widget =
        stack_box:
            Orientation := orientation.Vertical
            Slots := array:
                stack_box_slot:
                    HorizontalAlignment := horizontal_alignment.Fill
                    Widget := color_block:
                        DefaultOpacity := 0.0
                        DefaultDesiredSize := vector2{X := 450.0, Y := 0.0}
                stack_box_slot:
                    HorizontalAlignment := horizontal_alignment.Left
                    VerticalAlignment := vertical_alignment.Top
                    Padding := margin{Left := 0.0, Top := 0.0, Right := 0.0, Bottom := 20.0}
                    Widget := overlay:
                        Slots := array:
                            overlay_slot:
                                HorizontalAlignment := horizontal_alignment.Left
                                VerticalAlignment := vertical_alignment.Top
                                Padding := margin{Left := 0.0, Top := 0.0, Right := 64.0, Bottom := 0.0}
                                Widget := Icon
                            overlay_slot:
                                HorizontalAlignment := horizontal_alignment.Right
                                VerticalAlignment := vertical_alignment.Center
                                Widget := StatIcon
                stack_box_slot:
                    HorizontalAlignment := horizontal_alignment.Left
                    VerticalAlignment := vertical_alignment.Fill
                    Widget := stack_box:
                        Orientation := orientation.Horizontal
                        Slots := array:
                            stack_box_slot:
                                HorizontalAlignment := horizontal_alignment.Fill
                                VerticalAlignment := vertical_alignment.Fill
                                Widget := TierText
                            stack_box_slot:
                                HorizontalAlignment := horizontal_alignment.Fill
                                VerticalAlignment := vertical_alignment.Fill
                                Padding := margin{Left := 20.0, Top := 0.0, Right := 0.0, Bottom := 0.0}
                                Widget := stack_box:
                                    Orientation := orientation.Horizontal
                                    Slots := array:
                                        stack_box_slot:
                                            HorizontalAlignment := horizontal_alignment.Fill
                                            VerticalAlignment := vertical_alignment.Fill
                                            Widget := text_block:
                                                DefaultText := MakeMessage("▶")
                                                DefaultTextColor := ThemeColors.PrimaryText
                                        stack_box_slot:
                                            HorizontalAlignment := horizontal_alignment.Fill
                                            VerticalAlignment := vertical_alignment.Fill
                                            Padding := margin{Left := 20.0, Top := 0.0, Right := 0.0, Bottom := 0.0}
                                            Widget := TierProjectionText
                stack_box_slot:
                    HorizontalAlignment := horizontal_alignment.Left
                    VerticalAlignment := vertical_alignment.Fill
                    Widget := NameText
                stack_box_slot:
                    HorizontalAlignment := horizontal_alignment.Left
                    VerticalAlignment := vertical_alignment.Fill
                    Distribution := option{1.0}
                    Widget := DescriptionText
                stack_box_slot:
                    VerticalAlignment := vertical_alignment.Fill
                    Widget := set EvolutionStackBox = stack_box:
                        Orientation := orientation.Horizontal
                        Slots := array:
                            stack_box_slot:
                                HorizontalAlignment := horizontal_alignment.Left
                                Widget := texture_block:
                                    DefaultImage := UserInterface.Icons.NewsEntryType_NewFeature_64
                                    DefaultDesiredSize := vector2{X := 40.0, Y := 40.0}
                            stack_box_slot:
                                HorizontalAlignment := horizontal_alignment.Fill
                                VerticalAlignment := vertical_alignment.Fill
                                Padding := margin{Left := 10.0, Top := 0.0, Right := 0.0, Bottom := 0.0}
                                Widget := text_block:
                                    DefaultText := MakeMessage("Evolution!")
                                    DefaultTextColor := ThemeColors.Success
                                    DefaultShadowColor := ThemeColors.Success * 0.125
                                    DefaultShadowOffset := option{vector2{X := 2.0, Y := 2.0}}
                                    DefaultShadowOpacity := 1.0
                stack_box_slot:
                    VerticalAlignment := vertical_alignment.Fill
                    Widget := BanishIcon
                stack_box_slot:
                    VerticalAlignment := vertical_alignment.Fill
                    Widget := stack_box:
                        Orientation := orientation.Horizontal
                        Slots := array:
                            stack_box_slot:
                                HorizontalAlignment := horizontal_alignment.Fill
                                Widget := StatText
                            stack_box_slot:
                                HorizontalAlignment := horizontal_alignment.Fill
                                Widget := StatDeltaText

    <# ----------------------------------------------------------------------------------------------------
        Get Functions
    ---------------------------------------------------------------------------------------------------- #>
    GetActiveAbilityInfo()<transacts>:?ability_info =
        MaybeAbilityInfo

    GetActiveWeaponInfo()<transacts>:?weapon_info =
        MaybeWeaponInfo

    GetTier()<transacts>:int =
        Tier

    <# ----------------------------------------------------------------------------------------------------
        Update Functions
    ---------------------------------------------------------------------------------------------------- #>
    Hide<override>():void =
        UI.SetVisibility(widget_visibility.Hidden)

    Update(AbilityInfo:ability_info, InTier:int, IsRoll:logic, ?AbilitySelectMode:ability_select_mode = ability_select_mode.Default):void =
        set ActiveAbilitySelectMode = AbilitySelectMode
        if (IsRoll?):
            set MaybeAbilityInfo = option{AbilityInfo}
            set MaybeWeaponInfo = false
            set Tier = InTier

            CancelAnimationEvent.Signal()
            InfoOverlay.SetVisibility(widget_visibility.Visible)

            if (WeaponId := AbilityInfo.MaybeWeaponId?):
                WeaponLevel := Player.GetWeaponLevel(WeaponId)
                ToggleProgress(true)
                if (WeaponLevel < MaxWeaponLevel):
                    spawn{AnimateProjectionBoxUI()}
            else:
                ToggleProgress(false)
                EvolutionStackBox.SetVisibility(widget_visibility.Hidden)

            Icon.SetImage(AbilityInfo.Icon)
            Icon.SetTint(AbilityInfo.IconColor)
            
            if (AbilityStatIcon := AbilityInfo.MaybeStatIcon?):
                StatIcon.SetImage(AbilityStatIcon)
                StatIcon.SetVisibility(widget_visibility.Visible)
                StatIcon.SetTint(AbilityInfo.StatIconColor)
            else:
                StatIcon.SetVisibility(widget_visibility.Hidden)

            TierText.SetText(MakeMessage("{Tier.ToTierString().ToUpper()}"))
            TierText.SetTextColor(Tier.ToTierColor() * 2.0)
            TierText.SetShadowColor(Tier.ToTierColor() * 0.25)
            NameText.SetText(MakeMessage("{AbilityInfo.Name}"))
            DescriptionText.SetText(MakeMessage("{AbilityInfo.Description.WrapText(450.0)}"))

            if (ParentWidget := StatText.GetParentWidget[]):
                ParentWidget.SetVisibility(widget_visibility.Visible)

        if (ParentWidget := StatText.GetParentWidget[]):
            BanishIcon.SetVisibility(AbilitySelectMode = ability_select_mode.Banish and widget_visibility.Visible or widget_visibility.Collapsed)
            ParentWidget.SetVisibility(AbilitySelectMode = ability_select_mode.Banish and widget_visibility.Collapsed or widget_visibility.Visible)

        if (MaybeAbilityInfo?, ParentWidget := TierProjectionText.GetParentWidget[]):
            BoostValue := (ActiveAbilitySelectMode = ability_select_mode.Boost) and 1 or 0
            CanBoost := if (AbilitySelectMode = ability_select_mode.Boost, AbilityInfo.TierValue[Tier - 1 + BoostValue]?):
                TierProjection := Tier + 1
                TierProjectionText.SetText(MakeMessage("{TierProjection.ToTierString().ToUpper()}"))
                TierProjectionText.SetTextColor(TierProjection.ToTierColor() * 2.0)
                TierProjectionText.SetShadowColor(TierProjection.ToTierColor() * 0.25)
                ParentWidget.SetVisibility(widget_visibility.Visible)
                UpdateStats(TierProjection)
                true
            else:
                UpdateStats(Tier)
                ParentWidget.SetVisibility(widget_visibility.Collapsed)
                false

            if (WeaponId := AbilityInfo.MaybeWeaponId?):
                WeaponLevel := Player.GetWeaponLevel(WeaponId)

                SetProgress(WeaponLevel)
                set ProjectionBoxUIs = for:
                    Index -> BoxUI : BoxUIs
                    ProjectionLevel := MaxWeaponLevel - (Index + 1)
                    ProjectionLevel = WeaponLevel or 
                    (
                        AbilityInfo.Id = ability_id.WeaponLevel
                        TierValue := Floor[AbilityInfo.TierValue[Tier - 1 + BoostValue]? or AbilityInfo.TierValue[Tier - 1]?]
                        ProjectionLevel > WeaponLevel and ProjectionLevel <= WeaponLevel + TierValue
                    )
                do:
                    BoxUI

                LevelUpAmount := if (AbilityInfo.Id = ability_id.WeaponLevel, TierValue := Floor[AbilityInfo.TierValue[Tier - 1 + BoostValue]?]):
                    TierValue + 1
                else:
                    1
                
                CurrentWeaponEvolution := CalculateWeaponEvolutionLevel(WeaponLevel)
                UpdatedWeaponEvolution := CalculateWeaponEvolutionLevel(WeaponLevel + LevelUpAmount)
                set IsEvolution = logic{UpdatedWeaponEvolution > CurrentWeaponEvolution}
                EvolutionStackBox.SetVisibility(IsEvolution? and widget_visibility.Visible or widget_visibility.Hidden)

            SelectButton.SetText of:
                MakeMessage of case (AbilitySelectMode):
                    ability_select_mode.Default => "Level Up"
                    ability_select_mode.Banish => "Banish"
                    ability_select_mode.Boost => CanBoost? and "Boost" or "Level Up"
        else:
            SelectButton.SetText(MakeMessage("Banished!"))

        if (BoxUI := MaybeBoxUI?):
            if (AbilitySelectMode = ability_select_mode.Banish):
                BoxUI.SetBorderColor(ThemeColors.Failure)
                BoxUI.SetBorderOpacity(1.0)
            else:
                BoxUI.SetBorderColor(NamedColors.White)
                BoxUI.SetBorderOpacity(0.1)

    Clear():void =
        set MaybeAbilityInfo = false
        set MaybeWeaponInfo = false
        
    UpdateStats(InTier:int):void =
        if (AbilityInfo := MaybeAbilityInfo?):
            CurrentValue := Player.GetTotalAbilityValue(AbilityInfo)
            TierValue := if (PowerUpInfo := GetPowerUpInfo[AbilityInfo.Id], PowerUpInfo.StatType = power_up_stat_type.Multiply):
                (AbilityInfo.TierValue[InTier - 1]? or 0.0) * (1.0 + Player.GetPowerUpValue(PowerUpInfo.Id))
            else:
                AbilityInfo.TierValue[InTier - 1]? or 0.0

            UpdatedValue := CurrentValue + TierValue
            StatText.SetText(MakeMessage("{AbilityInfo.FormatValue(CurrentValue)} ▶ "))
            StatDeltaText.SetText(MakeMessage("{AbilityInfo.FormatValue(UpdatedValue)} ({TierValue > 0.0 and "+" or ""}{AbilityInfo.FormatValue(TierValue)})"))

    ShowWeapon(WeaponInfo:weapon_info):void =
        set MaybeWeaponInfo = option{WeaponInfo}

        InfoOverlay.SetVisibility(widget_visibility.Visible)    
        Icon.SetImage(WeaponInfo.Icon)
        Icon.SetTint(NamedColors.White)
        NameText.SetText(MakeMessage("{WeaponInfo.Name}"))
        DescriptionText.SetText(MakeMessage("{WeaponInfo.Description.WrapText(450.0)}"))
        EvolutionStackBox.SetVisibility(widget_visibility.Hidden)

        TierText.SetText(MakeMessage("NEW WEAPON!"))
        TierText.SetTextColor(ThemeColors.Failure)
        TierText.SetShadowColor(ThemeColors.Failure * 0.125)

        ToggleProgress(false)
        StatIcon.SetVisibility(widget_visibility.Hidden)
        if (ParentWidget := StatText.GetParentWidget[]):
            ParentWidget.SetVisibility(widget_visibility.Hidden)

        if (ParentWidget := TierProjectionText.GetParentWidget[]):
            ParentWidget.SetVisibility(widget_visibility.Collapsed)

        SelectButton.SetText(MakeMessage("Select"))

    ToggleProgress(IsShow:logic):void =
        ProgressUI.SetVisibility(IsShow? and widget_visibility.Visible or widget_visibility.Hidden)
    
    SetProgress(WeaponLevel:int):void =
        for (Index -> BoxUI : BoxUIs):
            if (MaxWeaponLevel - (Index + 1) = WeaponLevel):
                BoxUI.SetColor(ThemeColors.Success)
            else if (MaxWeaponLevel - (Index + 2) >= WeaponLevel):
                BoxUI.SetColor(ThemeColors.Disabled)
            else:
                BoxUI.SetColor(ThemeColors.Highlight)

    AnimateProjectionBoxUI()<suspends>:void =
        race:
            CancelAnimationEvent.Await()
            loop:
                Sleep(0.0)
                for (ProjectionBoxUI : ProjectionBoxUIs):
                    ProjectionBoxUI.SetColor(ThemeColors.Success * (Sin(GetSecondsSinceEpoch() * 2) * 0.7 + 0.8))

    <# ----------------------------------------------------------------------------------------------------
        Event handlers
    ---------------------------------------------------------------------------------------------------- #>
    OnClickSelect(Message:widget_message):void =           
        if (WeaponInfo := MaybeWeaponInfo?):
            SelectWeaponEvent.Signal(WeaponInfo)
            return

        if (AbilityInfo := MaybeAbilityInfo?):
            if (ActiveAbilitySelectMode = ability_select_mode.Banish):
                InfoOverlay.SetVisibility(widget_visibility.Hidden)
                set MaybeAbilityInfo = false

            SelectEvent.Signal(AbilityInfo, Tier, IsEvolution)
        else:
            Player.PlayAudio(audio_id.Error)