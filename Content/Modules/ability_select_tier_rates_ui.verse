using { /Fortnite.com/UI }
using { /Verse.org/Colors }
using { /UnrealEngine.com/Temporary/SpatialMath }
using { /UnrealEngine.com/Temporary/UI }

<# ----------------------------------------------------------------------------------------------------
    ability_select_tier_rates_ui
---------------------------------------------------------------------------------------------------- #>
ability_select_tier_rates_ui := class(ui_component):
    InfoText:text_block = text_block:
        DefaultTextColor := ThemeColors.Failure       
        DefaultShadowColor := ThemeColors.Failure * 0.125
        DefaultShadowOffset := option{vector2{X := 2.0, Y := 2.0}}
        DefaultShadowOpacity := 1.0

    var TierTexts:[]text_block = array{}
        
    <# ----------------------------------------------------------------------------------------------------
        Initialization
    ---------------------------------------------------------------------------------------------------- #>
    CreateUI<override>():widget =
        InfoText.SetText(MakeMessage("Luck does not apply in the Item Shop."))

        overlay:
            Slots := array:
                overlay_slot:
                    HorizontalAlignment := horizontal_alignment.Fill
                    VerticalAlignment := vertical_alignment.Fill
                    Widget := color_block:
                        DefaultColor := NamedColors.White
                        DefaultOpacity := 0.3
                overlay_slot:
                    HorizontalAlignment := horizontal_alignment.Fill
                    VerticalAlignment := vertical_alignment.Fill
                    Padding := margin{Left := 4.0, Top := 4.0, Right := 4.0, Bottom := 4.0}
                    Widget := color_block:
                        DefaultColor := ThemeColors.Background
                overlay_slot:
                    HorizontalAlignment := horizontal_alignment.Fill
                    VerticalAlignment := vertical_alignment.Fill
                    Padding := margin{Left := 40.0, Top := 40.0, Right := 40.0, Bottom := 40.0}
                    Widget := stack_box:
                        Orientation := orientation.Vertical
                        Slots := array:
                            stack_box_slot:
                                HorizontalAlignment := horizontal_alignment.Center
                                VerticalAlignment := vertical_alignment.Fill
                                Padding := margin{Left := 0.0, Top := 0.0, Right := 0.0, Bottom := 10.0}
                                Widget := text_block:
                                    DefaultText := MakeMessage("Tier Rates")
                                    DefaultTextColor := ThemeColors.PrimaryText
                            stack_box_slot:
                                HorizontalAlignment := horizontal_alignment.Fill
                                VerticalAlignment := vertical_alignment.Top
                                Padding := margin{Left := 0.0, Top := 0.0, Right := 0.0, Bottom := 40.0}
                                Widget := color_block:
                                    DefaultColor := NamedColors.White
                                    DefaultOpacity := 0.1
                                    DefaultDesiredSize := vector2{X := 0.0, Y := 1.0}
                            stack_box_slot:
                                HorizontalAlignment := horizontal_alignment.Fill
                                VerticalAlignment := vertical_alignment.Fill
                                Padding := margin{Left := 40.0, Top := 0.0, Right := 40.0, Bottom := 20.0}
                                Widget := InfoText
                            stack_box_slot:
                                HorizontalAlignment := horizontal_alignment.Fill
                                VerticalAlignment := vertical_alignment.Fill
                                Padding := margin{Left := 0.0, Top := 0.0, Right := 0.0, Bottom := 20.0}
                                Widget := overlay:
                                    Slots := array:
                                        overlay_slot:
                                            HorizontalAlignment := horizontal_alignment.Fill
                                            VerticalAlignment := vertical_alignment.Fill
                                            Widget := MakeBox()
                                        overlay_slot:
                                            HorizontalAlignment := horizontal_alignment.Fill
                                            VerticalAlignment := vertical_alignment.Fill
                                            Padding := margin{Left := 40.0, Top := 40.0, Right := 40.0, Bottom := 40.0}
                                            Widget := stack_box:
                                                Orientation := orientation.Vertical
                                                Slots := for (Index := 1..5):
                                                    TierText := text_block:
                                                        DefaultText := MakeMessage("0.00%")
                                                        DefaultTextColor := ThemeColors.Success
                                                        DefaultShadowColor := ThemeColors.Success * 0.125
                                                        DefaultShadowOffset := option{vector2{X := 2.0, Y := 2.0}}
                                                        DefaultShadowOpacity := 1.0

                                                    set TierTexts += array{TierText}

                                                    stack_box_slot:
                                                        HorizontalAlignment := horizontal_alignment.Fill
                                                        VerticalAlignment := vertical_alignment.Fill
                                                        Widget := stack_box:
                                                            Orientation := orientation.Horizontal
                                                            Slots := array:
                                                                stack_box_slot:
                                                                    HorizontalAlignment := horizontal_alignment.Left
                                                                    VerticalAlignment := vertical_alignment.Fill
                                                                    Widget := text_block:
                                                                        DefaultText := MakeMessage(Index.ToTierString().ToUpper())
                                                                        DefaultTextColor := Index.ToTierColor() * 2.0
                                                                        DefaultShadowColor := Index.ToTierColor() * 0.25
                                                                        DefaultShadowOffset := option{vector2{X := 2.0, Y := 2.0}}
                                                                        DefaultShadowOpacity := 1.0
                                                                stack_box_slot:
                                                                    HorizontalAlignment := horizontal_alignment.Right
                                                                    VerticalAlignment := vertical_alignment.Fill
                                                                    Distribution := option{1.0}
                                                                    Padding := margin{Left := 100.0, Top := 0.0, Right := 0.0, Bottom := 0.0}
                                                                    Widget := TierText

    <# ----------------------------------------------------------------------------------------------------
        Update Functions
    ---------------------------------------------------------------------------------------------------- #>
    Update<override>(?LuckMultiplier:float = 0.0, ?IsIgnoreLuck:logic = false):void =
        PlayerAbilityTierWeights := IsIgnoreLuck? and BaseAbilityTierWeights or GetPlayerAbilityTierWeights(Player, ?LuckMultiplier := LuckMultiplier)
        TotalWeight := Sum(PlayerAbilityTierWeights)

        InfoText.SetVisibility(IsIgnoreLuck? and widget_visibility.Visible or widget_visibility.Collapsed)

        for (Index -> TierText : TierTexts, PlayerAbilityTierWeight := PlayerAbilityTierWeights[Index]):
            Percent := (PlayerAbilityTierWeight * 1.0) / (TotalWeight * 1.0)
            TierText.SetText(MakeMessage("{FormatAsPercent(Percent, ?Decimals := 2)}"))