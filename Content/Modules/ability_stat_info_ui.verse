using { /Fortnite.com/UI }
using { /Verse.org/Colors }
using { /UnrealEngine.com/Temporary/SpatialMath }
using { /UnrealEngine.com/Temporary/UI }

<# ----------------------------------------------------------------------------------------------------
    ability_stat_info_ui
---------------------------------------------------------------------------------------------------- #>
ability_stat_info_ui := class(ui_component):
    BackgroundOpacity:type{_X:float where 0.000000 <= _X, _X <= 1.000000} = 0.8

    TeamExperienceMultiplierText:text_block = text_block:
        DefaultTextColor := ThemeColors.Success
        DefaultShadowColor := ThemeColors.Success * 0.125
        DefaultShadowOffset := option{vector2{X := 2.0, Y := 2.0}}
        DefaultShadowOpacity := 1.0

    var AbilityTexts:[ability_id]text_block = map{}
    var AbilityTotalTexts:[ability_id]text_block = map{}
    var GameResultWeaponEntryUIs:[]game_result_weapon_entry_ui = array{}
    var BoonIcons:[]texture_block = array{}
    var BoonTexts:[]text_block = array{}

    <# ----------------------------------------------------------------------------------------------------
        Initialization
    ---------------------------------------------------------------------------------------------------- #>
    CreateUI<override>():widget =
        overlay:
            Slots := array:
                overlay_slot:
                    HorizontalAlignment := horizontal_alignment.Fill
                    VerticalAlignment := vertical_alignment.Fill
                    Widget := color_block:
                        DefaultColor := NamedColors.White
                        DefaultOpacity := 0.2
                overlay_slot:
                    HorizontalAlignment := horizontal_alignment.Fill
                    VerticalAlignment := vertical_alignment.Fill
                    Padding := margin{Left := 4.0, Top := 4.0, Right := 4.0, Bottom := 4.0}
                    Widget := color_block:
                        DefaultColor := ThemeColors.Background
                        DefaultOpacity := BackgroundOpacity
                overlay_slot:
                    HorizontalAlignment := horizontal_alignment.Fill
                    VerticalAlignment := vertical_alignment.Fill
                    Padding := margin{Left := 40.0, Top := 40.0, Right := 40.0, Bottom := 40.0}
                    Widget := CreatePlayerUI()

    CreatePlayerUI():widget =
        stack_box:
            Orientation := orientation.Horizontal
            Slots := array:
                stack_box_slot:
                    HorizontalAlignment := horizontal_alignment.Fill
                    VerticalAlignment := vertical_alignment.Fill
                    Distribution := option{1.0}
                    Widget := CreateAbilitiesUI()
                stack_box_slot:
                    HorizontalAlignment := horizontal_alignment.Fill
                    VerticalAlignment := vertical_alignment.Fill
                    Distribution := option{1.2}
                    Padding := margin{Left := 20.0, Top := 0.0, Right := 0.0, Bottom := 0.0}
                    Widget := stack_box:
                        Orientation := orientation.Vertical
                        Slots := array:
                            stack_box_slot:
                                HorizontalAlignment := horizontal_alignment.Fill
                                VerticalAlignment := vertical_alignment.Fill
                                Distribution := option{1.0}
                                Padding := margin{Left := 0.0, Top := 0.0, Right := 0.0, Bottom := 20.0}
                                Widget := CreateWeaponsUI()
                            stack_box_slot:
                                HorizontalAlignment := horizontal_alignment.Fill
                                VerticalAlignment := vertical_alignment.Fill
                                Widget := CreateBoonsUI()

    CreateAbilitiesUI():widget =
        AbilitiesStackBox := stack_box:
            Orientation := orientation.Vertical
            Slots := array:
                stack_box_slot:
                    HorizontalAlignment := horizontal_alignment.Fill
                    VerticalAlignment := vertical_alignment.Fill
                    Padding := margin{Left := 0.0, Top := 0.0, Right := 0.0, Bottom := 20.0}
                    Widget := text_block:
                        DefaultText := MakeMessage("Abilities")
                        DefaultTextColor := ThemeColors.PrimaryText
                        DefaultTextOpacity := 0.5
                stack_box_slot:
                    HorizontalAlignment := horizontal_alignment.Fill
                    VerticalAlignment := vertical_alignment.Fill
                    Widget := stack_box:
                        Orientation := orientation.Vertical
                        Slots := 
                            var Index:int = 0
                            for (AbilityInfo : GetAbilityInfo()):
                                AbilityText := text_block:
                                    DefaultShadowOffset := option{vector2{X := 2.0, Y := 2.0}}
                                    DefaultShadowColor := ThemeColors.Success * 0.125

                                AbilityTotalText := text_block:
                                    DefaultTextColor := ThemeColors.Success
                                    DefaultShadowOpacity := 1.0
                                    DefaultShadowOffset := option{vector2{X := 2.0, Y := 2.0}}
                                    DefaultShadowColor := ThemeColors.Success * 0.125
                    
                                if (set AbilityTexts[AbilityInfo.Id] = AbilityText) {}
                                if (set AbilityTotalTexts[AbilityInfo.Id] = AbilityTotalText) {}
                                set Index += 1
                
                                stack_box_slot:
                                    HorizontalAlignment := horizontal_alignment.Fill
                                    Padding := margin{Left := 0.0, Top := (Index = 1) and 0.0 or 4.0, Right := 0.0, Bottom := 0.0}
                                    Widget := stack_box:
                                        Orientation := orientation.Horizontal
                                        Slots := array:
                                            stack_box_slot:
                                                HorizontalAlignment := horizontal_alignment.Left
                                                VerticalAlignment := vertical_alignment.Center
                                                Padding := margin{Left := 0.0, Top := 0.0, Right := 10.0, Bottom := 0.0}
                                                Widget := texture_block:
                                                    DefaultImage := AbilityInfo.IconSmall
                                                    DefaultDesiredSize := vector2{X := 40.0, Y := 40.0}
                                            stack_box_slot:
                                                HorizontalAlignment := horizontal_alignment.Left
                                                Distribution := option{0.5}
                                                Widget := text_block:
                                                    DefaultText := MakeMessage("{AbilityInfo.ShortName}")
                                                    DefaultTextColor := ThemeColors.PrimaryText
                                            stack_box_slot:
                                                HorizontalAlignment := horizontal_alignment.Right
                                                Distribution := option{1.0}
                                                Widget := AbilityTotalText
                                            stack_box_slot:
                                                HorizontalAlignment := horizontal_alignment.Right
                                                Distribution := option{1.0}
                                                Widget := AbilityText

        AbilitiesStackBox.AddWidget of stack_box_slot:
            HorizontalAlignment := horizontal_alignment.Fill
            VerticalAlignment := vertical_alignment.Bottom
            Distribution := option{1.0}
            Widget := stack_box:
                Orientation := orientation.Vertical
                Slots := array:
                    stack_box_slot:
                        HorizontalAlignment := horizontal_alignment.Fill
                        VerticalAlignment := vertical_alignment.Bottom
                        Distribution := option{1.0}
                        Padding := margin{Left := 0.0, Top := 20.0, Right := 0.0, Bottom := 20.0}
                        Widget := color_block:
                            DefaultColor := NamedColors.White
                            DefaultOpacity := 0.1
                            DefaultDesiredSize := vector2{X := 0.0, Y := 1.0}
                    stack_box_slot:
                        HorizontalAlignment := horizontal_alignment.Fill
                        VerticalAlignment := vertical_alignment.Fill
                        Widget := stack_box:
                            Orientation := orientation.Horizontal
                            Slots := array:
                                stack_box_slot:
                                    HorizontalAlignment := horizontal_alignment.Left
                                    Padding := margin{Left := 0.0, Top := 0.0, Right := 10.0, Bottom := 0.0}
                                    Widget := texture_block:
                                        DefaultImage := UserInterface.Icons.ExperienceMultiplier_64
                                        DefaultDesiredSize := vector2{X := 40.0, Y := 40.0}
                                stack_box_slot:
                                    HorizontalAlignment := horizontal_alignment.Fill
                                    Widget := text_block:
                                        DefaultText := MakeMessage("Team Multiplier")
                                        DefaultTextColor := ThemeColors.PrimaryText
                                stack_box_slot:
                                    HorizontalAlignment := horizontal_alignment.Right
                                    Distribution := option{1.0}
                                    Widget := TeamExperienceMultiplierText

        overlay:
            Slots := array:
                overlay_slot:
                    HorizontalAlignment := horizontal_alignment.Fill
                    VerticalAlignment := vertical_alignment.Fill
                    Widget := MakeBox(?Width := 600.0)
                overlay_slot:
                    HorizontalAlignment := horizontal_alignment.Fill
                    VerticalAlignment := vertical_alignment.Fill
                    Padding := margin{Left := 40.0, Top := 40.0, Right := 40.0, Bottom := 40.0}
                    Widget := AbilitiesStackBox

    CreateWeaponsUI():widget =
        overlay:
            Slots := array:
                overlay_slot:
                    HorizontalAlignment := horizontal_alignment.Fill
                    VerticalAlignment := vertical_alignment.Fill
                    Widget := MakeBox(?Width := 600.0)
                overlay_slot:
                    HorizontalAlignment := horizontal_alignment.Fill
                    VerticalAlignment := vertical_alignment.Fill
                    Padding := margin{Left := 40.0, Top := 40.0, Right := 40.0, Bottom := 40.0}
                    Widget := stack_box:
                        Orientation := orientation.Vertical
                        Slots := array:
                            stack_box_slot:
                                HorizontalAlignment := horizontal_alignment.Fill
                                VerticalAlignment := vertical_alignment.Fill
                                Padding := margin{Left := 0.0, Top := 0.0, Right := 0.0, Bottom := 20.0}
                                Widget := stack_box:
                                    Orientation := orientation.Horizontal
                                    Slots := array:
                                        stack_box_slot:
                                            HorizontalAlignment := horizontal_alignment.Left
                                            Distribution := option{2.0}
                                            Widget := text_block:
                                                DefaultText := MakeMessage("Weapons")
                                                DefaultTextColor := ThemeColors.PrimaryText
                                                DefaultTextOpacity := 0.5
                                        stack_box_slot:
                                            Padding := margin{Left := 21.0, Top := 0.0, Right := 0.0, Bottom := 0.0}
                                            Widget := color_block:
                                                DefaultOpacity := 0.0
                                        stack_box_slot:
                                            HorizontalAlignment := horizontal_alignment.Left
                                            Distribution := option{1.0}
                                            Widget := texture_block:
                                                DefaultImage := UserInterface.Icons.ProjectileDamage_64
                                                DefaultTint := ThemeColors.WeaponDamage
                                                DefaultDesiredSize := vector2{X := 40.0, Y := 40.0}
                                        stack_box_slot:
                                            HorizontalAlignment := horizontal_alignment.Left
                                            Distribution := option{1.0}
                                            Widget := texture_block:
                                                DefaultImage := UserInterface.Icons.AbilityCooldown_64
                                                DefaultTint := ThemeColors.WeaponCooldown
                                                DefaultDesiredSize := vector2{X := 40.0, Y := 40.0}
                                        stack_box_slot:
                                            HorizontalAlignment := horizontal_alignment.Left
                                            Distribution := option{1.0}
                                            Widget := texture_block:
                                                DefaultImage := UserInterface.Icons.AbilitySize_64
                                                DefaultTint := ThemeColors.WeaponSize
                                                DefaultDesiredSize := vector2{X := 40.0, Y := 40.0}
                            stack_box_slot:
                                HorizontalAlignment := horizontal_alignment.Fill
                                VerticalAlignment := vertical_alignment.Fill

                                Widget := stack_box:
                                    Orientation := orientation.Vertical
                                    Slots := for (Index := 1..4):
                                        GameResultWeaponEntryUI := game_result_weapon_entry_ui{Player := Player}
                                        GameResultWeaponEntryUI.Initialize()
                                        set GameResultWeaponEntryUIs += array{GameResultWeaponEntryUI}

                                        stack_box_slot:
                                            HorizontalAlignment := horizontal_alignment.Fill
                                            VerticalAlignment := vertical_alignment.Fill
                                            Padding := margin{Left := 0.0, Top := (Index = 1) and 0.0 or 40.0, Right := 0.0, Bottom := 0.0}
                                            Widget := GameResultWeaponEntryUI.GetUI()
    CreateBoonsUI():widget =
        overlay:
            Slots := array:
                overlay_slot:
                    HorizontalAlignment := horizontal_alignment.Fill
                    VerticalAlignment := vertical_alignment.Fill
                    Widget := MakeBox()
                overlay_slot:
                    HorizontalAlignment := horizontal_alignment.Fill
                    VerticalAlignment := vertical_alignment.Fill
                    Padding := margin{Left := 40.0, Top := 20.0, Right := 40.0, Bottom := 20.0}
                    Widget := stack_box:
                        Orientation := orientation.Horizontal
                        Slots := array:
                            stack_box_slot:
                                HorizontalAlignment := horizontal_alignment.Fill
                                VerticalAlignment := vertical_alignment.Top
                                Distribution := option{1.5}
                                Widget := text_block:
                                    DefaultText := MakeMessage("Boons")
                                    DefaultTextColor := ThemeColors.PrimaryText
                                    DefaultTextOpacity := 0.5
                            stack_box_slot:
                                HorizontalAlignment := horizontal_alignment.Fill
                                VerticalAlignment := vertical_alignment.Fill
                                Distribution := option{5.0}
                                Padding := margin{Left := 0.0, Top := 0.0, Right := 0.0, Bottom := 0.0}
                                Widget := stack_box:
                                    Orientation := orientation.Horizontal
                                    Slots := for (Index := 1..5):
                                        BoonIcon := texture_block:
                                            DefaultImage := UserInterface.Icons.Boon_64
                                            DefaultDesiredSize := vector2{X := 64.0, Y := 64.0}

                                        BoonText := text_block:
                                            DefaultTextColor := ThemeColors.Success
                                            DefaultShadowColor := ThemeColors.Success * 0.125
                                            DefaultShadowOffset := option{vector2{X := 2.0, Y := 2.0}}
                                            DefaultShadowOpacity := 1.0

                                        set BoonIcons += array{BoonIcon}
                                        set BoonTexts += array{BoonText}

                                        stack_box_slot:
                                            Distribution := option{1.0}
                                            Widget := stack_box:
                                                Orientation := orientation.Vertical
                                                Slots := array:
                                                    stack_box_slot:
                                                        Widget := BoonIcon
                                                    stack_box_slot:
                                                        Widget := BoonText

    <# ----------------------------------------------------------------------------------------------------
        Update Functions
    ---------------------------------------------------------------------------------------------------- #>
    Update<override>():void =
        if (AbilityLevels := Player.GetAbilityLevels(), PlayerStats := Player.GetStats[]):
            BanishedIds := for (BanishedAbility : Player.GetBanishedAbilities(), not BanishedAbility.MaybeWeaponId?) {BanishedAbility.Id}
            for (Id -> AbilityText : AbilityTexts, AbilityTotalText := AbilityTotalTexts[Id], AbilityInfo := GetAbilityInfo[Id]):
                IsBanished := logic{BanishedIds.Find[Id]}
                AbilityTotalText.SetText(MakeMessage("{AbilityInfo.FormatValue(Player.GetTotalAbilityValue(AbilityInfo))}"))

                if (AbilityLevel := AbilityLevels[Id]):
                    AbilityValue := PlayerStats.CalculateAbilityValue(Id, AbilityLevel, ?IncludeBaseValue := false)
                    AbilityText.SetText(MakeMessage("+{AbilityInfo.FormatValue(AbilityValue)}"))
                    AbilityText.SetColor(IsBanished? and ThemeColors.Failure or ThemeColors.Success)
                else:
                    AbilityText.SetText(MakeMessage("0"))
                    if (IsBanished?):
                        AbilityText.SetColor(ThemeColors.Failure)
                    else:
                        AbilityText.SetTextColor(ThemeColors.Inactive)
                        AbilityText.SetShadowOpacity(0.0)

        PlayerWeaponLevels := Player.GetWeaponLevels()
        WeaponIds := for (WeaponId -> WeaponLevel : PlayerWeaponLevels) {WeaponId}
        for (Index -> GameResultWeaponEntryUI : GameResultWeaponEntryUIs):
            if (WeaponId := WeaponIds[Index], WeaponLevel := PlayerWeaponLevels[WeaponId]):
                GameResultWeaponEntryUI.Update(Player, WeaponId, WeaponLevel)
                GameResultWeaponEntryUI.Show()
            else:
                GameResultWeaponEntryUI.SetInactive()
        
        TeamExperienceMultiplierText.SetText(MakeMessage("+{FormatAsPercent(GetTeamExperienceMultiplier())}"))

        PlayerBoons := Player.GetBoons()
        for (Index -> BoonIcon : BoonIcons, BoonText := BoonTexts[Index]):
            if (BoonId := PlayerBoons[Index], BoonInfo := GetBoonInfo[BoonId]):
                if (GetValue := BoonInfo.MaybeGetValue?):
                    Value := GetValue(Player)
                    BoonText.SetText(MakeMessage("{Value > 0.0 and "+" or ""}{BoonInfo.FormatValue(GetValue(Player))}"))

                    if (Value > 0.0):
                        BoonText.SetColor(ThemeColors.Success)
                    else:                        
                        BoonText.SetTextColor(ThemeColors.Inactive)
                        BoonText.SetShadowOpacity(0.0)
                else:
                    BoonText.SetText(MakeMessage(""))

                BoonText.SetVisibility(widget_visibility.Visible)
                BoonIcon.SetImage(BoonInfo.IconSmall)
                BoonIcon.SetTint(BoonInfo.Tier.ToTierColor() * 2)
            else:
                BoonText.SetVisibility(widget_visibility.Collapsed)
                BoonIcon.SetImage(UserInterface.Icons.Lock_64)
                BoonIcon.SetTint(ThemeColors.Inactive)

    <# ----------------------------------------------------------------------------------------------------
        Event Handlers
    ---------------------------------------------------------------------------------------------------- #>
    