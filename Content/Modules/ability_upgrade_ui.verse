using { /Fortnite.com/UI }
using { /Verse.org/Colors }
using { /Verse.org/Random }
using { /Verse.org/Simulation }
using { /UnrealEngine.com/Temporary/SpatialMath }
using { /UnrealEngine.com/Temporary/UI }

ability_upgrade_ui := class(panel_ui):
    Title<override>:string = "Treasure Chest"
    HasCloseButton<override>:logic = false
    BackgroundOpacity<override>:type {_X:float where 0.000000 <= _X, _X <= 1.000000} = 0.7
    
    SkippedEvent:events(tuple()) = events(tuple()){}

    TierText:text_block = text_block:
        DefaultShadowOffset := option{vector2{X := 2.0, Y := 2.0}}
        DefaultShadowOpacity := 1.0

    ProgressIcon:texture_block = texture_block:
        DefaultImage := UserInterface.Icons.Random_128
        DefaultDesiredSize := vector2{X := 128.0, Y := 128.0}

    Button:button_loud = button_loud{}

    var MaybeAbilitySelectEntryUI:?ability_select_entry_ui = false
    var MaybeAbilityInfo:?ability_info = false
    var Tier:int = 0
    var CanSkipAnimation:logic = false
    var AnimationWidget:widget = canvas{}

    <# ----------------------------------------------------------------------------------------------------
        Initialization
    ---------------------------------------------------------------------------------------------------- #>
    CreateContent<override>():widget =
        ClosedEvent.Subscribe(OnClosed)
        Button.OnClick().Subscribe(OnClickButton)

        if (GameManager := GetGameManager[]):
            GameManager.ResumeCountdownStartEvent.Subscribe(OnResumeCountdownStart)
            GameManager.ResumeCountdownCompleteEvent.Subscribe(OnResumeCountdownComplete)
            GameManager.ResumeCountdownEvent.Subscribe(OnResumeCountdown)
            GameManager.GameResumedEvent.Subscribe(OnGameResumed)

        AbilitySelectEntryUI := ability_select_entry_ui{Player := Player, HasButton := false}
        AbilitySelectEntryUI.Initialize()
        set MaybeAbilitySelectEntryUI = option{AbilitySelectEntryUI}

        stack_box:
            Orientation := orientation.Vertical
            Slots := array:
                stack_box_slot:
                    HorizontalAlignment := horizontal_alignment.Fill
                    VerticalAlignment := vertical_alignment.Fill
                    Padding := margin{Left := 0.0, Top := 0.0, Right := 0.0, Bottom := 20.0}
                    Widget := overlay:
                        Slots := array:
                            overlay_slot:
                                HorizontalAlignment := horizontal_alignment.Fill
                                VerticalAlignment := vertical_alignment.Fill
                                Widget := set AnimationWidget = stack_box:
                                    Orientation := orientation.Vertical
                                    Slots := array:
                                        stack_box_slot:
                                            HorizontalAlignment := horizontal_alignment.Fill
                                            VerticalAlignment := vertical_alignment.Fill
                                            Widget := overlay:
                                                Slots := array:
                                                    overlay_slot:
                                                        HorizontalAlignment := horizontal_alignment.Fill
                                                        VerticalAlignment := vertical_alignment.Fill
                                                        Widget := MakeBox()
                                                    overlay_slot:
                                                        Padding := margin{Left := 10.0, Top := 10.0, Right := 10.0, Bottom := 10.0}
                                                        Widget := TierText
                                        stack_box_slot:
                                            HorizontalAlignment := horizontal_alignment.Fill
                                            VerticalAlignment := vertical_alignment.Fill
                                            Padding := margin{Left := 0.0, Top := 20.0, Right := 0.0, Bottom := 0.0}
                                            Distribution := option{1.0}
                                            Widget := overlay:
                                                Slots := array:
                                                    overlay_slot:
                                                        HorizontalAlignment := horizontal_alignment.Fill
                                                        VerticalAlignment := vertical_alignment.Fill
                                                        Widget := MakeBox()
                                                    overlay_slot:
                                                        Padding := margin{Left := 100.0, Top := 100.0, Right := 100.0, Bottom := 100.0}
                                                        Widget := ProgressIcon
                            overlay_slot:
                                HorizontalAlignment := horizontal_alignment.Center
                                VerticalAlignment := vertical_alignment.Center
                                Padding := margin{Left := 1.0, Top := 1.0, Right := 1.0, Bottom := 1.0}
                                Widget := AbilitySelectEntryUI.GetUI()
                stack_box_slot:
                    HorizontalAlignment := horizontal_alignment.Fill
                    Widget := Button

    <# ----------------------------------------------------------------------------------------------------
        Get Functions
    ---------------------------------------------------------------------------------------------------- #>
    IsWeaponEvolution()<transacts>:logic = 
        if (AbilityInfo := MaybeAbilityInfo?, WeaponId := AbilityInfo.MaybeWeaponId?):
            LevelUpAmount := if (AbilityInfo.Id = ability_id.WeaponLevel, TierValue := Floor[AbilityInfo.TierValue[Tier - 1]?]). TierValue else 0
            CurrentWeaponLevel := Player.GetWeaponLevel(WeaponId)
            CurrentEvolutionLevel := CalculateWeaponEvolutionLevel(CurrentWeaponLevel)
            UpdatedEvolutionLevel := CalculateWeaponEvolutionLevel(CurrentWeaponLevel + 1 + LevelUpAmount)
            logic{UpdatedEvolutionLevel > CurrentEvolutionLevel}
        else:
            false

    <# ----------------------------------------------------------------------------------------------------
        Update Functions
    ---------------------------------------------------------------------------------------------------- #>
    Update(AbilityInfo:ability_info, InTier:int):void =
        spawn{Animate(AbilityInfo, InTier)}

    Animate(AbilityInfo:ability_info, InTier:int)<suspends>:void =
        if (AbilitySelectEntryUI := MaybeAbilitySelectEntryUI?):
            set MaybeAbilityInfo = option{AbilityInfo}
            set Tier = InTier
            set CanSkipAnimation = true
            AbilitySelectEntryUI.Update(AbilityInfo, Tier, true)
            Button.SetText(MakeMessage("Skip"))
            AnimationWidget.SetVisibility(widget_visibility.Visible)
            ProgressIcon.SetImage(UserInterface.Icons.Random_128)
            AbilitySelectEntryUI.Hide()
            UpdateTitle()
            AudioPlayer := Player.PlayAudio(audio_id.UpgradeProgress)

            race:
                SkippedEvent.Await()
                block:
                    SleepTimeDilation(0.1)
                    var PreviousIndex:int = 0
                    for (Index := 1..6):
                        RandomRarity := if (Index = 1):
                            GetRandomInt(1, 5)
                        else:
                            Shuffle(for (RandomIndex := 1..5, RandomIndex <> PreviousIndex) {RandomIndex})[0] or 1

                        set PreviousIndex = RandomRarity

                        ProgressIcon.SetTint(RandomRarity.ToTierColor() * 2.0)
                        TierText.SetTextColor(RandomRarity.ToTierColor() * 2.0)
                        TierText.SetText(MakeMessage("{RandomRarity.ToTierString()}"))
                        TierText.SetShadowColor(RandomRarity.ToTierColor() * 0.25)
                        SleepTimeDilation(0.5)

            set CanSkipAnimation = false
            AudioPlayer.Stop(Player)
            Button.SetText(MakeMessage("Done"))
            AbilitySelectEntryUI.Show()
            AnimationWidget.SetVisibility(widget_visibility.Hidden)
            Player.PlayAudio(Tier >= 3 and audio_id.UpgradeRare or audio_id.Upgrade)

    UpdateTitle()<suspends>:void =
        Sleep(0.0)
        PauseCount := Player.GetPauseCount(pause_type.AbilityUpgrade)
        TitleText.SetText(MakeMessage("{Title} ({PauseCount})"))
        TitleText.SetTextColor(ThemeColors.PrimaryText)
        TitleText.SetShadowOpacity(0.0)

    <# ----------------------------------------------------------------------------------------------------
        Event Handlers
    ---------------------------------------------------------------------------------------------------- #>
    OnClosed(InPlayer:player):void =
        ResumeGame(Player)
        if (AbilitySelectEntryUI := MaybeAbilitySelectEntryUI?):
            AbilitySelectEntryUI.CancelAnimationEvent.Signal()

    OnClickButton(Message:widget_message):void =
        if (CanSkipAnimation?):
            SkippedEvent.Signal()
        else if (AbilityInfo := MaybeAbilityInfo?):
            if (IsWeaponEvolution()?, WeaponId := MaybeAbilityInfo?.MaybeWeaponId?):
                Hide(?IsSignal := false)
                Player.AddAbility(AbilityInfo, ?Amount := Tier)
                Player.ShowWeaponEvolutionUI(WeaponId)
                Player.PlayAudio(audio_id.Boost)
            else:
                if (NextPause := Player.GetNextPause[], NextPause(0) = pause_type.AbilityUpgrade):
                    if (RandomAbility := GetTreasureChestRandomAbility(Player)?):
                        Update(RandomAbility)
                    
                    ResumeGame(Player, ?IsShow := false)
                else:
                    Hide()

                Player.AddAbility(AbilityInfo, ?Amount := Tier)

    OnResumeCountdownStart():void =
        if (not IsShowing?) then return
        TitleText.SetText(MakeMessage("{GameResumeCountdownTimeAbilityUpgrade}s"))
        TitleText.SetTextColor(ThemeColors.Failure)
        TitleText.SetShadowColor(ThemeColors.Failure * 0.125)
        TitleText.SetShadowOffset(option{vector2{X := 2.0, Y := 2.0}})
        TitleText.SetShadowOpacity(1.0)
        
    OnResumeCountdownComplete():void =
        if (not IsShowing?) then return
        SkippedEvent.Signal()
        Hide()

    OnResumeCountdown(TimeRemaining:int):void =
        if (not IsShowing?) then return
        if (Player.GetIsAFK()?):
            Hide()
            return
            
        TitleText.SetText(MakeMessage("{TimeRemaining}s"))

    OnGameResumed():void =
        if (not IsShowing?) then return
        TitleText.SetText(MakeMessage("{Title}"))
        TitleText.SetTextColor(ThemeColors.PrimaryText)
        TitleText.SetShadowOpacity(0.0)