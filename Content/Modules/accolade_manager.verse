using { /Fortnite.com/Devices }
using { /UnrealEngine.com/Temporary/SpatialMath }
using { /Verse.org/Simulation }

<# ----------------------------------------------------------------------------------------------------
    Data definitions
---------------------------------------------------------------------------------------------------- #>
accolade_id := enum:
    NewWeapon
    StageCompleted
    SupplyDropSecured
    SurvivedTime15 # [Unused]
    CompleteGame
    Victory
    CompleteObjective
    AbilityEvolved
    AbilityUpgrade
    BonusItem
    SurvivedWave
    PetSlime
    ItemShop

accolade_count_id := enum:
    EliminateSlimes
    LevelUp
    GatherTree
    EliminateSlimeMelee
    EliminateSlimeMeleeElite
    EliminateSlimeMeleeGeneral
    EliminateSlimeMiniboss
    EliminateSlimeBoss

ToString(Id:accolade_count_id)<transacts>:string =
    case (Id):
        accolade_count_id.EliminateSlimes => "EliminateSlimes"
        accolade_count_id.LevelUp => "LevelUp"
        accolade_count_id.GatherTree => "GatherTree"
        accolade_count_id.EliminateSlimeMelee => "EliminateSlimeMelee"
        accolade_count_id.EliminateSlimeMeleeElite => "EliminateSlimeMeleeElite"
        accolade_count_id.EliminateSlimeMeleeGeneral => "EliminateSlimeMeleeGeneral"
        accolade_count_id.EliminateSlimeMiniboss => "EliminateSlimeMiniboss"
        accolade_count_id.EliminateSlimeBoss => "EliminateSlimeBoss"

accolade := class<concrete>:
    @editable
    Id:accolade_id = accolade_id.Victory
    @editable
    Device:accolades_device = accolades_device{}
    @editable
    MaybeCooldown:?float = false

accolade_count := class<concrete>:
    @editable
    Id:accolade_count_id = accolade_count_id.EliminateSlimes
    @editable
    Device:accolades_device = accolades_device{}
    @editable
    Count:int = 1

<# ----------------------------------------------------------------------------------------------------
    Global functions
---------------------------------------------------------------------------------------------------- #>
var GlobalAccoladeManager:weak_map(session, accolade_manager) = map{}

GetAccoladeManager()<decides><transacts>:accolade_manager =
    GlobalAccoladeManager[GetSession()]

(Player:player).AwardAccolade(Id:accolade_id):void =
    if (AccoladeManager := GetAccoladeManager[]):
        AccoladeManager.Award(Player, Id)

AwardAccolade(Id:accolade_id, Players:[]player):void =
    if (AccoladeManager := GetAccoladeManager[]):
        AccoladeManager.Award(Id, Players)

(Player:player).AwardAccolade(Id:accolade_count_id):void =
    if (AccoladeManager := GetAccoladeManager[]):
        AccoladeManager.Award(Player, Id)

AwardAccolade(Id:accolade_count_id, Players:[]player):void =
    if (AccoladeManager := GetAccoladeManager[]):
        AccoladeManager.Award(Id, Players)

(Player:player).IsAccoladeReady(Id:accolade_id)<transacts>:logic =
    if (AccoladeManager := GetAccoladeManager[]):
        AccoladeManager.IsAccoladeReady(Player, Id)
    else:
        false

<# ----------------------------------------------------------------------------------------------------
    accolade_manager
---------------------------------------------------------------------------------------------------- #>
accolade_manager := class(creative_device):    
    @editable
    AccoladeDevices:[]accolade = array{}

    @editable
    AccoladeCountDevices:[]accolade_count = array{}

    @editable
    PetSlimeButton:button_device = button_device{}

    @editable
    PetSlimeSwitch:switch_device = switch_device{}

    @editable
    PetSlimeTimer:timer_device = timer_device{}
    
    var MappedAccolades:[accolade_id]accolade = map{}
    var MappedAccoladeCounts:[accolade_count_id][]accolade_count = map{}
    var PlayerLastAwardTime:[player][accolade_id]float = map{}
    var PlayerAwardCount:[player][accolade_count_id]int = map{}

    <# ----------------------------------------------------------------------------------------------------
        Initialization
    ---------------------------------------------------------------------------------------------------- #>
    OnBegin<override>()<suspends>:void =
        if (set GlobalAccoladeManager[GetSession()] = Self) {}

        for (AccoladeDevice : AccoladeDevices):
            if (set MappedAccolades[AccoladeDevice.Id] = AccoladeDevice) {}

        for (AccoladeCountDevice : AccoladeCountDevices):
            if:
                not MappedAccoladeCounts[AccoladeCountDevice.Id]
                set MappedAccoladeCounts[AccoladeCountDevice.Id] = array{AccoladeCountDevice}
            else if:
                set MappedAccoladeCounts[AccoladeCountDevice.Id] += array{AccoladeCountDevice}
        
        GetPlayspace().PlayerAddedEvent().Subscribe(OnPlayerAdded)
        GetPlayspace().PlayerRemovedEvent().Subscribe(OnPlayerRemoved)
        PetSlimeButton.InteractedWithEvent.Subscribe(OnInteractedWith)
        #PetSlimeButton.SetInteractionText(MakeMessage("Pet (" + GetRoundedString(GetCooldown(accolade_id.PetSlime) / 60.0, 0) + " minutes cooldown)"))
        PetSlimeTimer.SuccessEvent.Subscribe(OnTimerSuccess)

        for (Player : GetPlayspace().GetPlayers()):
            Initialize(Player)

        <#
        WaitForGameManager()
        if (GameManager := GetGameManager[]):
            GameManager.GameEndedEvent.Subscribe(OnGameEnded)
        #>

    Initialize(Player:player):void =
        if:
            not PlayerLastAwardTime[Player]
            set PlayerLastAwardTime[Player] = map{}

        if:
            not PlayerAwardCount[Player]
            set PlayerAwardCount[Player] = map{}

        PetSlimeTimer.SetActiveDuration(GetCooldown(accolade_id.PetSlime), Player)
        #PetSlimeTimer.Disable(Player)

    <# ----------------------------------------------------------------------------------------------------
        Get functions
    ---------------------------------------------------------------------------------------------------- #>
    GetDevice(Id:accolade_id)<decides><transacts>:accolade =
        MappedAccolades[Id]

    GetDevices(Id:accolade_count_id)<decides><transacts>:[]accolade_count =
        MappedAccoladeCounts[Id]
        
    GetCooldown(Id:accolade_id)<transacts>:float =
        GetDevice[accolade_id.PetSlime].MaybeCooldown? or 0.0

    IsAccoladeReady(Player:player, Id:accolade_id)<transacts>:logic =
        if:
            Accolade := GetDevice[Id]
            Cooldown := Accolade.MaybeCooldown?
            Cooldown > 0.0
        then:
            if:
                LastAwardTime := PlayerLastAwardTime[Player][Id]
                ElapsedTime := GetSimulationElapsedTime() - LastAwardTime
                ElapsedTime < Cooldown
            then:
                false
            else:
                true
        else:
            true

    <# ----------------------------------------------------------------------------------------------------
        Update functions
    ---------------------------------------------------------------------------------------------------- #>
    Award(Player:player, Id:accolade_id):void =
        if (Accolade := GetDevice[Id]):
            if (Cooldown := Accolade.MaybeCooldown?, Cooldown > 0.0):
                if (LastAwardTime := PlayerLastAwardTime[Player][Id]):
                    ElapsedTime := GetSimulationElapsedTime() - LastAwardTime

                    if (ElapsedTime >= Cooldown):
                        Accolade.Device.Award(Player)
                        if (set PlayerLastAwardTime[Player][Id] = GetSimulationElapsedTime()) {}
                else:
                    Accolade.Device.Award(Player)
                    if (set PlayerLastAwardTime[Player][Id] = GetSimulationElapsedTime()) {}
            else:
                Accolade.Device.Award(Player)

    Award(Id:accolade_id, Players:[]player):void =
        for (Player : Players):
            Award(Player, Id)

    Award(Player:player, Id:accolade_count_id):void =
        if (set PlayerAwardCount[Player][Id] = PlayerAwardCount[Player][Id] + 1 or 1) {}
        
        for (Accolade : GetDevices[Id], Mod[PlayerAwardCount[Player][Id], Accolade.Count] = 0):
            Accolade.Device.Award(Player)

    Award(Id:accolade_count_id, Players:[]player):void =
        for (Player : Players):
            Award(Player, Id)

    <# ----------------------------------------------------------------------------------------------------
        Event handlers
    ---------------------------------------------------------------------------------------------------- #>
    OnPlayerAdded(Player:player):void =
        Initialize(Player)

    OnPlayerRemoved(Player:player):void =
        set PlayerLastAwardTime = RemoveKeyFromMap(PlayerLastAwardTime, Player)
        set PlayerAwardCount = RemoveKeyFromMap(PlayerAwardCount, Player)

    OnInteractedWith(Agent:agent):void =
        if (Player := player[Agent]):
            if (IsAccoladeReady(Player, accolade_id.PetSlime)?):
                Player.PlayAudio(audio_id.PetSlime)
                Player.AwardAccolade(accolade_id.PetSlime)
                SpawnVFX(vfx_id.Hearts, PetSlimeSwitch.GetTransform().Translation + vector3{Z := 100.0})
                PetSlimeSwitch.TurnOff(Player)
                PetSlimeTimer.Enable(Player)
                PetSlimeTimer.SetActiveDuration(GetCooldown(accolade_id.PetSlime), Player)
                PetSlimeTimer.Start(Player)
<#
    OnGameEnded():void =
        for (Player : GetAllPlayers(), IsAccoladeReady(Player, accolade_id.PetSlime)?):
            PetSlimeSwitch.TurnOn(Player)
            PetSlimeTimer.Disable(Player)
#>
    OnTimerSuccess(MaybeAgent:?agent):void =
        if (Agent := MaybeAgent?):
            PetSlimeSwitch.TurnOn(Agent)
            PetSlimeTimer.Disable(Agent)