using { /Fortnite.com/Devices }
using { /Verse.org/Colors }
using { /Verse.org/Random }
using { /Verse.org/Simulation }
using { /UnrealEngine.com/Temporary/SpatialMath }
using { Utility }

<# ----------------------------------------------------------------------------------------------------
    Constants
---------------------------------------------------------------------------------------------------- #>
LoadArenaDuration:float = 2.5
DefaultArenaSize:float = 4400.0
DefaultArenaOrigin:vector3 := vector3{}
MaxDifficultyLevel:int = 9

<# ----------------------------------------------------------------------------------------------------
    Data Definitions
---------------------------------------------------------------------------------------------------- #>
arena := class<concrete>:
    @editable
    Level:int = 1
    @editable
    Name:string = ""
    @editable
    Difficulty:string = "Normal"
    @editable
    HealthMultiplier:float = 1.0
    @editable
    StageCount:int = 3
    @editable
    StageMultiplier:float = 0.25
    @editable
    DamageMultiplier:float = 1.0
    @editable
    MovespeedMultiplier:float = 1.0
    @editable
    GoldMultiplier:float = 1.0
    @editable
    Size:float = DefaultArenaSize
    @editable
    MapController:map_controller_device = map_controller_device{}
    @editable
    GatherableColor:color = NamedColors.White
    @editable
    BossId:npc_id = npc_id.Boss
    @editable
    MinibossMultiplier:int = 1
    @editable
    Minibosses:[]npc_id = array{}
    @editable
    MusicId:music_id = music_id.CombatLow

<# ----------------------------------------------------------------------------------------------------
    Global Functions
---------------------------------------------------------------------------------------------------- #>
var GlobalArenaManager:weak_map(session, arena_manager) = map{}

GetArenaManager()<decides><transacts>:arena_manager = 
    GlobalArenaManager[GetSession()]

GetArenas()<transacts>:[]arena =
    GetArenaManager[].GetArenas() or array{}

GetActiveArena()<decides><transacts>:arena =
    GetArenaManager[].GetActiveArena[]

GetActiveArenaIndex()<transacts>:?int =
    GetArenaManager[].GetActiveArenaIndex() or false

GetActiveArenaSize<public>()<transacts>:float =
    GetArenaManager[].GetActiveArenaSize[] or DefaultArenaSize

CalculateArenaMovespeedMultiplier(Arena:arena)<transacts>:float =
    ((GetGameDifficulty() > 1) and 1.0 or Arena.MovespeedMultiplier) + (GetGameSpeed() - 1.0) * 0.5

GetArenaMovespeedMultiplier()<transacts>:float =
    CalculateArenaMovespeedMultiplier(GetArenaManager[].GetActiveArena[]) or 0.0

GetArenaStageCount()<transacts>:int =
    Min(10, GetGameDifficulty() + 1)

GetArenaStageMultiplier()<transacts>:float =
    0.5

CalculateArenaLevel(Arena:arena)<transacts>:int =
    CalculateArenaLevel(Arena.Level, GetGameDifficulty())

CalculateArenaLevel(ArenaLevel:int, Difficulty:int)<transacts>:int =
    GetArenaCount() * (Difficulty - 1) + ArenaLevel

CalculateArenaHealthMultiplier(Arena:arena)<transacts>:float =
    (CalculateArenaLevel(Arena) - 1) * 1.0

GetArenaHealthMultiplier()<transacts>:float =
    CalculateArenaHealthMultiplier(GetArenaManager[].GetActiveArena[]) or 0.0

CalculateArenaDamageMultiplier(Arena:arena)<transacts>:float =
    (CalculateArenaLevel(Arena) - 1) * 0.5

GetArenaDamageMultiplier()<transacts>:float =
    CalculateArenaDamageMultiplier(GetArenaManager[].GetActiveArena[]) or 0.0

CalculateArenaGoldMultiplier(Arena:arena)<transacts>:float =
    CalculateArenaGoldMultiplier(CalculateArenaLevel(Arena))

CalculateArenaGoldMultiplier(ArenaLevel:int)<transacts>:float =
    (ArenaLevel - 1) * 0.15

GetArenaGoldMultiplier()<transacts>:float =
    CalculateArenaGoldMultiplier(GetArenaManager[].GetActiveArena[]) or 0.0    

GetRandomPositionInArena()<transacts>:vector3 =
    GetArenaManager[].GetRandomPositionInArena() or vector3{}

LoadArena(Index:int):void =
    if (ArenaManager := GetArenaManager[]):
        spawn{ArenaManager.LoadArena(Index)}

(Position:vector3).IsInArena()<transacts>:logic =
    logic{DistanceSquared(Position, vector3{}) <= Pow(GetActiveArenaSize(), 2.0)}

GetArenaCount()<transacts>:int =
    GetArenaManager[].GetArenaCount() or 9

GetMaxArenaLevel()<transacts>:int =
    GetArenaCount() * MaxDifficultyLevel

GetGatherables()<transacts>:[]gatherable =
    GetArenaManager[].Gatherables or array{}

<# ----------------------------------------------------------------------------------------------------
    arena_manager
---------------------------------------------------------------------------------------------------- #>
arena_manager := class(creative_device):
    @editable
    Arenas:[]arena = array{}

    @editable
    ArenaPreviewCamera:gameplay_camera_fixed_point_device = gameplay_camera_fixed_point_device{}
    
    @editable
    CinematicSequenceLoadArenaDataLayer:cinematic_sequence_device = cinematic_sequence_device{}   
    
    @editable
    CinematicSequenceUnloadArenaDataLayer:cinematic_sequence_device = cinematic_sequence_device{}

    @editable
    CinematicSequenceMinimapBoss:cinematic_sequence_device = cinematic_sequence_device{}

    @editable
    Button:button_device = button_device{}

    @editable
    Gatherables:[]gatherable = array{}
    
    GatherableMaterialInstance:Materials.M_Tree_Inst_material = Materials.M_Tree_Inst_material{}

    var PlayerGatheredEvent:[player]events(tuple()) = map{}

    var PlayerArenaSelectUI:[player]arena_select_ui = map{}
    var IsLoading:logic = false
    var IsStarting:logic = false
    var MaybePlayerSelectingArena:?player = false
    var MaybeActiveArenaIndex:?int = false
    var MaybeNextArenaIndex:?int = false
    var MaybeDeltaTimer:?delta_timer = false
    var ArenaCount:int = 9
    
    <# ----------------------------------------------------------------------------------------------------
        Initialization
    ---------------------------------------------------------------------------------------------------- #>
    OnBegin<override>()<suspends>:void =
        if (set GlobalArenaManager[GetSession()] = Self) {}

        set MaybeDeltaTimer = option{delta_timer{}}
        set ArenaCount = Arenas.Length

        GetPlayspace().PlayerAddedEvent().Subscribe(OnPlayerAdded)
        GetPlayspace().PlayerRemovedEvent().Subscribe(OnPlayerRemoved)

        for (Player : GetPlayspace().GetPlayers()):
            Initialize(Player)

        WaitForGameManager()
        for (Gatherable : Gatherables):
            Gatherable.OnBegin()
            Gatherable.Prop.SetMaterial(GatherableMaterialInstance, ?Index := 1)
            Gatherable.GatheredEvent.Subscribe(OnGathered)

        (arena_manager:)LoadArena(0)

        Button.InteractedWithEvent.Subscribe(OnInteractedWith)

    Initialize(Player:player):void =
        if:
            not PlayerGatheredEvent[Player]
            set PlayerGatheredEvent[Player] = events(tuple()){}

    <# ----------------------------------------------------------------------------------------------------
        Get Functions
    ---------------------------------------------------------------------------------------------------- #>
    (arena_manager:)GetArenas()<transacts>:[]arena =
        Arenas

    GetArena(Index:int)<decides><transacts>:arena =
        Arenas[Index]

    (arena_manager:)GetArenaCount()<transacts>:int =
        ArenaCount

    (arena_manager:)GetActiveArena()<decides><transacts>:arena =
        Arenas[MaybeActiveArenaIndex?]
    
    (arena_manager:)GetActiveArenaIndex()<transacts>:?int =
        MaybeActiveArenaIndex

    GetLoadTimeRemaining()<transacts>:float =
        if (DeltaTimer := MaybeDeltaTimer?):
            if (MaybeNextArenaIndex?):
                LoadArenaDuration * 2.0 - DeltaTimer.GetElapsedTime()
            else:
                LoadArenaDuration - DeltaTimer.GetElapsedTime()
        else:
            LoadArenaDuration

    (arena_manager:)GetActiveArenaSize()<decides><transacts>:float =
        (arena_manager:)GetActiveArena[].Size

    (arena_manager:)GetRandomPositionInArena()<transacts>:vector3 =
        ArenaSize := (arena_manager:)GetActiveArenaSize[] or 2000.0
        GetRandomPositionInCircle(ArenaSize, ?Origin := DefaultArenaOrigin)

    <# ----------------------------------------------------------------------------------------------------
        Update Functions
    ---------------------------------------------------------------------------------------------------- #>
    LoadGatherables(Index:int):void =
        spawn{AsyncLoadGatherables(Index)}

    AsyncLoadGatherables(ArenaIndex:int)<suspends>:void =
        if (Arena := Arenas[ArenaIndex]):
            GatherablePositions := Shuffle(PoissonDiscSamplingGeneratePoints(800.0, Arena.Size * 0.8))
            TotalGatherables := Gatherables.Length

            var GatherableIndex:int = 0
            for (GatherablePosition : GatherablePositions, GatherableIndex <= TotalGatherables - 1):
                RaycastPosition := GetRaycast(GatherablePosition)

                if (RaycastPosition.Z <= 25.0, Gatherable := Gatherables[GatherableIndex]):
                    set GatherableIndex += 1
                    Gatherable.Load of transform:
                        Translation := vector3{X := GatherablePosition.X, Y := GatherablePosition.Y}
                        Rotation := IdentityRotation()
                        Scale := UnitVector

    (arena_manager:)LoadArena(Index:int)<suspends>:void =
        if (IsLoading?):
            set MaybeNextArenaIndex = option{Index}
            return

        if (ActiveArenaIndex := MaybeActiveArenaIndex?, Index = ActiveArenaIndex):
            Print("Arena is already loaded ...")
            return

        set IsLoading = true
        if (DeltaTimer := MaybeDeltaTimer?):
            DeltaTimer.Reset()

        UnloadArena()

        if (Arena := Arenas[Index]):
            CinematicSequenceLoadArenaDataLayer.SetPlaybackTime(Index * 1.0)
            CinematicSequenceLoadArenaDataLayer.Play()
            Sleep(0.25)
            CinematicSequenceLoadArenaDataLayer.Stop()

            CinematicSequenceMinimapBoss.Play()
            CinematicSequenceMinimapBoss.SetPlaybackTime(Index * 1.0)
            CinematicSequenceMinimapBoss.Pause()

            Arena.MapController.Activate()
            
            set GatherableMaterialInstance.Color = Arena.GatherableColor
            set MaybeActiveArenaIndex = option{Index}

        Sleep(LoadArenaDuration - 0.5)
        set IsLoading = false
        
        if (NextArenaIndex := MaybeNextArenaIndex?):
            set MaybeNextArenaIndex = false
            (arena_manager:)LoadArena(NextArenaIndex)
            
    UnloadArena()<suspends>:void =
        if (ActiveArenaIndex := MaybeActiveArenaIndex?, ActiveArena := GetArena[ActiveArenaIndex]):
            CinematicSequenceUnloadArenaDataLayer.SetPlaybackTime(ActiveArenaIndex * 1.0)
            CinematicSequenceUnloadArenaDataLayer.Play()
            Sleep(0.25)
            CinematicSequenceUnloadArenaDataLayer.Stop()

            ActiveArena.MapController.Deactivate()
            set MaybeActiveArenaIndex = false

    <# ----------------------------------------------------------------------------------------------------
        Event Handlers
    ---------------------------------------------------------------------------------------------------- #>
    OnPlayerAdded(Player:player):void =
        Initialize(Player)

    OnPlayerRemoved(Player:player):void =
        if (PlayerSelectingArena := MaybePlayerSelectingArena?, Player = PlayerSelectingArena):
            set MaybePlayerSelectingArena = false

        set PlayerArenaSelectUI = RemoveKeyFromMap(PlayerArenaSelectUI, Player)
        set PlayerGatheredEvent = RemoveKeyFromMap(PlayerGatheredEvent, Player)

    OnInteractedWith(Agent:agent):void =
        if (PlayerSelectingArena := MaybePlayerSelectingArena?):
            Agent.PlayAudio(audio_id.Error)
            Agent.ShowMessage(message_id.MessageSmall, MakeMessage(PlayerSelectingArena) + MakeMessage(" is already selecting an arena ..."))
            return
        
        if (IsGameInProgress()?):
            Agent.PlayAudio(audio_id.Error)
            Agent.ShowMessage(message_id.MessageSmall, "Game is already in progress ...")
            return

        if (Player := player[Agent]):
            UI := if (ArenaSelectUI := PlayerArenaSelectUI[Player]):
                ArenaSelectUI
            else:
                ArenaSelectUI := arena_select_ui{Player := Player}
                ArenaSelectUI.ClosedEvent.Subscribe(OnArenaSelectUIClosed)
                if (set PlayerArenaSelectUI[Player] = ArenaSelectUI) {}
                ArenaSelectUI

            ArenaPreviewCamera.AddTo(Player)
            UI.Show()
            spawn{UI.Update()}
            set MaybePlayerSelectingArena = option{Player}
        
    OnArenaSelectUIClosed(Player:player):void =
        ArenaPreviewCamera.RemoveFrom(Player)
        set MaybePlayerSelectingArena = false

    OnGathered(Player:player):void =
        if (GatheredEvent := PlayerGatheredEvent[Player]):
            GatheredEvent.Signal()