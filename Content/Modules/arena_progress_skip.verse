using { /Fortnite.com/Devices }
using { /Verse.org/Simulation }
using { /UnrealEngine.com/Temporary/SpatialMath }

<# ----------------------------------------------------------------------------------------------------
    arena_progress_skip
---------------------------------------------------------------------------------------------------- #>
arena_progress_skip := class(creative_device):
    @editable
    Button:button_device = button_device{}

    @editable
    PopupDialog:popup_dialog_device = popup_dialog_device{}

    <# ----------------------------------------------------------------------------------------------------
        Initialization
    ---------------------------------------------------------------------------------------------------- #>
    OnBegin<override>()<suspends>:void =
        Button.InteractedWithEvent.Subscribe(OnInteractedWith)
        PopupDialog.ShownEvent.Subscribe(OnShownEvent)
        PopupDialog.RespondingButtonEvent.Subscribe(OnRespondingButton)

    <# ----------------------------------------------------------------------------------------------------
        Get Functions
    ---------------------------------------------------------------------------------------------------- #>
    GetTargetArenaProgress(Player:player)<transacts>:int =
        var MaxWeaponProgress:int = 0
        if (WeaponProgress := Player.GetData[].WeaponProgress):
            for (WeaponIdInt -> Progress : WeaponProgress, Progress > MaxWeaponProgress):
                set MaxWeaponProgress = Progress

        Min(18, MaxWeaponProgress * 2)

    <# ----------------------------------------------------------------------------------------------------
        Update Functions
    ---------------------------------------------------------------------------------------------------- #>

    <# ----------------------------------------------------------------------------------------------------
        Event Handlers
    ---------------------------------------------------------------------------------------------------- #>
    OnInteractedWith(Agent:agent):void =
        if (Player := player[Agent]):
            if (Player.GetArenaProgress() >= GetTargetArenaProgress(Player)):
                Player.PlayAudio(audio_id.Error)
                Player.ShowMessage(message_id.MessageSmall, "Your arena progress is already at a higher level than your weapon progress ...")
                return
            
            PopupDialog.SetDescriptionText(MakeMessage("{GetTargetArenaProgress(Player)}"))
            PopupDialog.Show(Agent)

    OnShownEvent(Agent:agent):void =
        if (Player := player[Agent]):
            PopupDialog.SetDescriptionText(MakeMessage("{GetTargetArenaProgress(Player)}"))

    OnRespondingButton(Agent:agent, Index:int):void =
        if (Index = 0, Player := player[Agent]):
            Player.SetArenaProgress(GetTargetArenaProgress(Player))
            Player.ShowMessage(Player.GetRankUpMessage())