using { /Fortnite.com/UI }
using { /Verse.org/Colors }
using { /Verse.org/Simulation }
using { /UnrealEngine.com/Temporary/SpatialMath }
using { /UnrealEngine.com/Temporary/UI }

arena_select_player_info_entry_ui := class(ui_component):
    HighlightColorBlock:color_block = color_block:
        DefaultColor := NamedColors.White
        DefaultOpacity := 0.1

    RankIcon:texture_block = texture_block:
        DefaultImage := UserInterface.Icons.RankIcon_01_64
        DefaultDesiredSize := vector2{X := 40.0, Y := 40.0}
        
    NameText:text_block = text_block:
        DefaultTextColor := ThemeColors.PrimaryText
        DefaultShadowOffset := option{vector2{X := 2.0, Y := 2.0}}
        DefaultShadowOpacity := 0.5

    LevelText:text_block = text_block:
        DefaultTextColor := ThemeColors.PrimaryText
        DefaultShadowOffset := option{vector2{X := 2.0, Y := 2.0}}
        DefaultShadowOpacity := 0.5  

    WeaponIcon:texture_block = texture_block:
        DefaultImage := UserInterface.Icons.WeaponBow_64
        DefaultDesiredSize := vector2{X := 40.0, Y := 40.0}

    WeaponText:text_block = text_block:
        DefaultTextColor := ThemeColors.PrimaryText
        DefaultShadowOffset := option{vector2{X := 2.0, Y := 2.0}}
        DefaultShadowOpacity := 0.5

    WeaponLevelText:text_block = text_block:
        DefaultTextColor := ThemeColors.PrimaryText
        DefaultShadowOffset := option{vector2{X := 2.0, Y := 2.0}}
        DefaultShadowOpacity := 0.5

    <# ----------------------------------------------------------------------------------------------------
        Initialization
    ---------------------------------------------------------------------------------------------------- #>
    CreateUI<override>():widget =
        overlay:
            Slots := array:
                overlay_slot:
                    HorizontalAlignment := horizontal_alignment.Fill
                    VerticalAlignment := vertical_alignment.Fill
                    Widget := HighlightColorBlock
                overlay_slot:
                    HorizontalAlignment := horizontal_alignment.Fill
                    VerticalAlignment := vertical_alignment.Fill
                    Padding := margin{Left := 20.0, Top := 20.0, Right := 20.0, Bottom := 20.0}
                    Widget := stack_box:
                        Orientation := orientation.Vertical
                        Slots := array:
                            stack_box_slot:
                                HorizontalAlignment := horizontal_alignment.Fill
                                VerticalAlignment := vertical_alignment.Fill
                                Widget := stack_box:
                                    Orientation := orientation.Horizontal
                                    Slots := array:
                                        stack_box_slot:
                                            HorizontalAlignment := horizontal_alignment.Left
                                            Widget := RankIcon
                                        stack_box_slot:
                                            HorizontalAlignment := horizontal_alignment.Left
                                            Distribution := option{1.0}
                                            Padding := margin{Left := 10.0, Top := 0.0, Right := 0.0, Bottom := 0.0}
                                            Widget := NameText
                                        stack_box_slot:
                                            HorizontalAlignment := horizontal_alignment.Fill
                                            Widget := text_block:
                                                DefaultText := MakeMessage("Lv. ")
                                                DefaultTextColor := ThemeColors.Highlight
                                                DefaultShadowColor := ThemeColors.Highlight * 0.125
                                                DefaultShadowOffset := option{vector2{X := 2.0, Y := 2.0}}
                                                DefaultShadowOpacity := 1.0
                                        stack_box_slot:
                                            HorizontalAlignment := horizontal_alignment.Fill
                                            Widget := LevelText
                                        
                            stack_box_slot:
                                HorizontalAlignment := horizontal_alignment.Fill
                                VerticalAlignment := vertical_alignment.Fill
                                Widget := stack_box:
                                    Orientation := orientation.Horizontal
                                    Slots := array:
                                        stack_box_slot:
                                            HorizontalAlignment := horizontal_alignment.Left
                                            Widget := WeaponIcon
                                        stack_box_slot:
                                            HorizontalAlignment := horizontal_alignment.Left
                                            Distribution := option{1.0}
                                            Padding := margin{Left := 10.0, Top := 0.0, Right := 0.0, Bottom := 0.0}
                                            Widget := WeaponText
                                        stack_box_slot:
                                            HorizontalAlignment := horizontal_alignment.Fill
                                            Widget := text_block:
                                                DefaultText := MakeMessage("Lv. ")
                                                DefaultTextColor := ThemeColors.Highlight
                                                DefaultShadowColor := ThemeColors.Highlight * 0.125
                                                DefaultShadowOffset := option{vector2{X := 2.0, Y := 2.0}}
                                                DefaultShadowOpacity := 1.0
                                        stack_box_slot:
                                            HorizontalAlignment := horizontal_alignment.Fill
                                            Widget := WeaponLevelText

    <# ----------------------------------------------------------------------------------------------------
        Update Functions
    ---------------------------------------------------------------------------------------------------- #>
    Update(InPlayer:player, IsHighlight:logic):void =
        RankIcon.SetImage(InPlayer.GetRank().ToRankIcon())
        NameText.SetText(MakeMessage(InPlayer))
        LevelText.SetText(MakeMessage("{InPlayer.GetArenaProgress()}"))

        if (EquippedWeapon := InPlayer.GetEquippedWeapon()?, WeaponInfo := GetWeaponInfo[EquippedWeapon]):
            WeaponIcon.SetImage(WeaponInfo.IconSmall)
            WeaponText.SetText(MakeMessage("{WeaponInfo.Name}"))
            
            WeaponLevel := InPlayer.GetData[].WeaponProgress[EquippedWeapon.ToInt()] or 0
            WeaponLevelText.SetText(MakeMessage("{WeaponLevel}"))

        HighlightColorBlock.SetVisibility(IsHighlight? and widget_visibility.Visible or widget_visibility.Hidden)