using { /Fortnite.com/Devices }
using { /Fortnite.com/UI }
using { /Verse.org/Colors }
using { /Verse.org/Concurrency }
using { /Verse.org/Simulation }
using { /UnrealEngine.com/Temporary/SpatialMath }
using { /UnrealEngine.com/Temporary/UI }

<# ----------------------------------------------------------------------------------------------------
    arena_select_ui
---------------------------------------------------------------------------------------------------- #>
arena_select_ui := class(panel_ui):
    Title<override>:string = "Arena Select"
    BackgroundOpacity<override>:type {_X:float where 0.000000 <= _X, _X <= 1.000000} = 0.0
    HasCloseButton<override>:logic = false
    LabelPadding:float = 200.0

    LoadArenaCancelledEvent:events(tuple()) = events(tuple()){}
    TimeoutCancelEvent:events(tuple()) = events(tuple()){}
    TimeoutDuration:float = 15.0
    TimeoutDeltaTimer:delta_timer = delta_timer{}

    NameText:text_block = text_block:
        DefaultTextColor := ThemeColors.PrimaryText
        DefaultShadowOpacity := 0.5
        DefaultShadowOffset := option{vector2{X := 2.0, Y := 2.0}}

    LevelText:text_block = text_block:
        DefaultTextColor := ThemeColors.PrimaryText
        DefaultShadowOpacity := 0.5
        DefaultShadowOffset := option{vector2{X := 2.0, Y := 2.0}}

    DifficultyText:text_block = text_block:
        DefaultTextColor := ThemeColors.PrimaryText
        DefaultShadowOpacity := 0.5
        DefaultShadowOffset := option{vector2{X := 2.0, Y := 2.0}}

    StageText:text_block = text_block:
        DefaultTextColor := ThemeColors.PrimaryText
        DefaultShadowOpacity := 0.5
        DefaultShadowOffset := option{vector2{X := 2.0, Y := 2.0}}

    GoldMultiplierText:text_block = text_block:
        DefaultTextColor := ThemeColors.Success
        DefaultShadowColor := ThemeColors.Success * 0.125
        DefaultShadowOpacity := 1.0
        DefaultShadowOffset := option{vector2{X := 2.0, Y := 2.0}}

    HealthMultiplierText:text_block = text_block:
        DefaultTextColor := ThemeColors.Failure
        DefaultShadowColor := ThemeColors.Failure * 0.125
        DefaultShadowOpacity := 1.0
        DefaultShadowOffset := option{vector2{X := 2.0, Y := 2.0}}

    DamageMultiplierText:text_block = text_block:
        DefaultTextColor := ThemeColors.Failure
        DefaultShadowColor := ThemeColors.Failure * 0.125
        DefaultShadowOpacity := 1.0
        DefaultShadowOffset := option{vector2{X := 2.0, Y := 2.0}}

    MovespeedMultiplierText:text_block = text_block:
        DefaultTextColor := ThemeColors.Failure
        DefaultShadowColor := ThemeColors.Failure * 0.125
        DefaultShadowOpacity := 1.0
        DefaultShadowOffset := option{vector2{X := 2.0, Y := 2.0}}

    StageMultiplierText:text_block = text_block:
        DefaultTextColor := ThemeColors.Failure
        DefaultShadowColor := ThemeColors.Failure * 0.125
        DefaultShadowOpacity := 1.0
        DefaultShadowOffset := option{vector2{X := 2.0, Y := 2.0}}

    WeaponIcon:texture_block = texture_block:
        DefaultImage := UserInterface.Icons.WeaponBow_64
        DefaultDesiredSize := vector2{X := 40.0, Y := 40.0}

    PlayersReadyText:text_block = text_block{}
    
    TeamArenaLevelText:text_block = text_block:
        DefaultTextColor := ThemeColors.PrimaryText
        DefaultShadowOpacity := 0.5
        DefaultShadowOffset := option{vector2{X := 2.0, Y := 2.0}}

    StartButton:button_regular = button_regular{}
    InfoButton:button_quiet = button_quiet{}

    var MaybeLoadingTask:?task(void) = false

    var CurrentArenaIndex:int = 0
    var CurrentDifficulty:int = 1
    var CurrentGameSpeed:int = 1
    
    var IsPlayerInfo:logic = false
    var ArenaInfoUI:widget = canvas{}
    var PlayerInfoUI:widget = canvas{}
    var PlayerInfoEntryUIs:[]arena_select_player_info_entry_ui = array{}

    var ArenaColorBlocks:[]color_block = array{}
    var ArenaColorButtons:[]text_button_base = array{}
    var DifficultyColorBlocks:[]color_block = array{}
    var DifficultyColorButtons:[]text_button_base = array{}
    var GameSpeedColorBlocks:[]color_block = array{}
    var GameSpeedColorButtons:[]text_button_base = array{}

    <# ----------------------------------------------------------------------------------------------------
        Initialization
    ---------------------------------------------------------------------------------------------------- #>
    CreateContent<override>():widget =
        ClosedEvent.Subscribe(OnClosed)

        if (GameManager := GetGameManager[]):
            GameManager.PlayerReadiedEvent.Subscribe(OnPlayerReadied)
            GameManager.PlayerUnreadiedEvent.Subscribe(OnPlayerUnreadied)
            GameManager.GameStartedEvent.Subscribe(OnGameStarted)
            GameManager.GetPlayspace().PlayerAddedEvent().Subscribe(OnPlayerAdded)
            GameManager.GetPlayspace().PlayerRemovedEvent().Subscribe(OnPlayerRemoved)

        stack_box:
            Orientation := orientation.Horizontal
            Slots := array:               
                stack_box_slot:
                    HorizontalAlignment := horizontal_alignment.Fill
                    VerticalAlignment := vertical_alignment.Fill
                    Padding := margin{Left := 0.0, Top := 0.0, Right := 20.0, Bottom := 0.0}
                    Widget := CreateMenuUI()
                stack_box_slot:
                    HorizontalAlignment := horizontal_alignment.Fill
                    VerticalAlignment := vertical_alignment.Fill
                    Widget := overlay:
                        Slots := array:
                            overlay_slot:
                                HorizontalAlignment := horizontal_alignment.Fill
                                VerticalAlignment := vertical_alignment.Fill
                                Widget := MakeBox(?Width := 850.0)
                            overlay_slot:
                                HorizontalAlignment := horizontal_alignment.Fill
                                VerticalAlignment := vertical_alignment.Fill
                                Padding := margin{Left := 40.0, Top := 40.0, Right := 40.0, Bottom := 40.0}
                                Widget := set ArenaInfoUI = CreateArenaInfoUI()
                            overlay_slot:
                                HorizontalAlignment := horizontal_alignment.Fill
                                VerticalAlignment := vertical_alignment.Fill
                                Padding := margin{Left := 40.0, Top := 40.0, Right := 40.0, Bottom := 40.0}
                                Widget := set PlayerInfoUI = CreatePlayerInfoUI()

    CreateMenuUI():widget =
        StartButton.SetText(MakeMessage("Start"))
        StartButton.OnClick().Subscribe(OnClickStart)

        InfoButton.SetText(MakeMessage("Player Info"))
        InfoButton.OnClick().Subscribe(OnClickInfo)

        CloseButton := button_loud{DefaultText := MakeMessage("Close")}
        CloseButton.OnClick().Subscribe(OnClickClose)

        overlay:
            Slots := array:
                overlay_slot:
                    HorizontalAlignment := horizontal_alignment.Fill
                    VerticalAlignment := vertical_alignment.Fill
                    Widget := MakeBox(?Width := 900.0)
                overlay_slot:
                    HorizontalAlignment := horizontal_alignment.Fill
                    VerticalAlignment := vertical_alignment.Fill
                    Padding := margin{Left := 40.0, Top := 40.0, Right := 40.0, Bottom := 40.0}
                    Widget := stack_box:
                        Orientation := orientation.Vertical
                        Slots := array:
                            stack_box_slot:
                                HorizontalAlignment := horizontal_alignment.Fill
                                VerticalAlignment := vertical_alignment.Fill
                                Padding := margin{Left := 0.0, Top := 0.0, Right := 0.0, Bottom := 10.0}
                                Widget := text_block:
                                    DefaultText := MakeMessage("Arena")
                                    DefaultTextColor := ThemeColors.PrimaryText
                                    DefaultShadowOffset := option{vector2{X := 2.0, Y := 2.0}}
                                    DefaultShadowOpacity := 0.5
                            stack_box_slot:
                                HorizontalAlignment := horizontal_alignment.Fill
                                Widget := CreateArenaSelectUI()
                            stack_box_slot:
                                HorizontalAlignment := horizontal_alignment.Fill
                                VerticalAlignment := vertical_alignment.Fill
                                Padding := margin{Left := 0.0, Top := 40.0, Right := 0.0, Bottom := 10.0}
                                Widget := text_block:
                                    DefaultText := MakeMessage("Difficulty")
                                    DefaultTextColor := ThemeColors.PrimaryText
                                    DefaultShadowOffset := option{vector2{X := 2.0, Y := 2.0}}
                                    DefaultShadowOpacity := 0.5
                            stack_box_slot:
                                HorizontalAlignment := horizontal_alignment.Fill
                                Widget := CreateDifficultySelectUI()
                            stack_box_slot:
                                HorizontalAlignment := horizontal_alignment.Fill
                                VerticalAlignment := vertical_alignment.Fill
                                Padding := margin{Left := 0.0, Top := 40.0, Right := 0.0, Bottom := 10.0}
                                Widget := text_block:
                                    DefaultText := MakeMessage("Game Speed")
                                    DefaultTextColor := ThemeColors.PrimaryText
                                    DefaultShadowOffset := option{vector2{X := 2.0, Y := 2.0}}
                                    DefaultShadowOpacity := 0.5
                            stack_box_slot:
                                HorizontalAlignment := horizontal_alignment.Fill
                                Widget := CreateGameSpeedSelectUI()
                            stack_box_slot:
                                HorizontalAlignment := horizontal_alignment.Fill
                                Padding := margin{Left := 0.0, Top := 40.0, Right := 0.0, Bottom := 40.0}
                                Widget := color_block:
                                    DefaultColor := NamedColors.White
                                    DefaultOpacity := 0.1
                                    DefaultDesiredSize := vector2{X := 0.0, Y := 1.0}
                            stack_box_slot:
                                HorizontalAlignment := horizontal_alignment.Fill
                                VerticalAlignment := vertical_alignment.Bottom
                                Distribution := option{1.0}
                                Padding := margin{Left := 0.0, Top := 0.0, Right := 0.0, Bottom := 10.0}
                                Widget := InfoButton
                            stack_box_slot:
                                HorizontalAlignment := horizontal_alignment.Fill
                                VerticalAlignment := vertical_alignment.Fill
                                Padding := margin{Left := 0.0, Top := 0.0, Right := 0.0, Bottom := 10.0}
                                Widget := StartButton
                            stack_box_slot:
                                HorizontalAlignment := horizontal_alignment.Fill
                                VerticalAlignment := vertical_alignment.Fill
                                Widget := CloseButton

    CreateArenaSelectUI():widget =
        stack_box:
            Orientation := orientation.Horizontal
            Slots := for (Index := 1..GetArenaCount()):
                ColorBlock := color_block{DefaultColor := ThemeColors.Loud, DefaultDesiredSize := vector2{Y := 5.0}}
                ColorBlock.SetVisibility(widget_visibility.Hidden)
                Button := button_quiet{DefaultText := MakeMessage("{Index}")}
                spawn{AwaitArenaClicked(Button.OnClick(), Index - 1)}

                set ArenaColorBlocks += array{ColorBlock}
                set ArenaColorButtons += array{Button}

                stack_box_slot:
                    HorizontalAlignment := horizontal_alignment.Fill
                    VerticalAlignment := vertical_alignment.Fill
                    Distribution := option{1.0}
                    Padding := margin{Left := (Index = 1) and 0.0 or 10.0, Top := 0.0, Right := 0.0, Bottom := 0.0}
                    Widget := overlay:
                        Slots := array:
                            overlay_slot:
                                HorizontalAlignment := horizontal_alignment.Fill
                                VerticalAlignment := vertical_alignment.Bottom
                                Padding := margin{Left := 10.0, Top := 0.0, Right := 10.0, Bottom := -10.0}
                                Widget := ColorBlock
                            overlay_slot:
                                HorizontalAlignment := horizontal_alignment.Fill
                                VerticalAlignment := vertical_alignment.Fill
                                Widget := Button

    CreateDifficultySelectUI():widget =
        stack_box:
            Orientation := orientation.Horizontal
            Slots := for (Index := 1..MaxDifficultyLevel):
                ColorBlock := color_block{DefaultColor := ThemeColors.Loud, DefaultDesiredSize := vector2{Y := 5.0}}
                ColorBlock.SetVisibility(widget_visibility.Hidden)
                Button := button_quiet{DefaultText := MakeMessage("{Index}")}
                spawn{AwaitDifficultyClicked(Button.OnClick(), Index)}

                set DifficultyColorBlocks += array{ColorBlock}
                set DifficultyColorButtons += array{Button}

                stack_box_slot:
                    HorizontalAlignment := horizontal_alignment.Fill
                    VerticalAlignment := vertical_alignment.Fill
                    Distribution := option{1.0}
                    Padding := margin{Left := (Index = 1) and 0.0 or 10.0, Top := 0.0, Right := 0.0, Bottom := 0.0}
                    Widget := overlay:
                        Slots := array:
                            overlay_slot:
                                HorizontalAlignment := horizontal_alignment.Fill
                                VerticalAlignment := vertical_alignment.Bottom
                                Padding := margin{Left := 10.0, Top := 0.0, Right := 10.0, Bottom := -10.0}
                                Widget := ColorBlock
                            overlay_slot:
                                HorizontalAlignment := horizontal_alignment.Fill
                                VerticalAlignment := vertical_alignment.Fill
                                Widget := Button

    CreateGameSpeedSelectUI():widget =
        stack_box:
            Orientation := orientation.Horizontal
            Slots := for (Index := 1..3):
                ColorBlock := color_block{DefaultColor := ThemeColors.Loud, DefaultDesiredSize := vector2{Y := 5.0}}
                ColorBlock.SetVisibility(widget_visibility.Hidden)
                Button := button_quiet{DefaultText := MakeMessage("x{Index}")}
                spawn{AwaitGameSpeedClicked(Button.OnClick(), Index)}

                set GameSpeedColorBlocks += array{ColorBlock}
                set GameSpeedColorButtons += array{Button}

                stack_box_slot:
                    HorizontalAlignment := horizontal_alignment.Fill
                    VerticalAlignment := vertical_alignment.Fill
                    Distribution := option{1.0}
                    Padding := margin{Left := (Index = 1) and 0.0 or 10.0, Top := 0.0, Right := 0.0, Bottom := 0.0}
                    Widget := overlay:
                        Slots := array:
                            overlay_slot:
                                HorizontalAlignment := horizontal_alignment.Fill
                                VerticalAlignment := vertical_alignment.Bottom
                                Padding := margin{Left := 10.0, Top := 0.0, Right := 10.0, Bottom := -10.0}
                                Widget := ColorBlock
                            overlay_slot:
                                HorizontalAlignment := horizontal_alignment.Fill
                                VerticalAlignment := vertical_alignment.Fill
                                Widget := Button

    CreateArenaInfoUI():widget =
        stack_box:
            Orientation := orientation.Vertical
            Slots := array:
                stack_box_slot:
                    HorizontalAlignment := horizontal_alignment.Fill
                    Widget := NameText
                stack_box_slot:
                    HorizontalAlignment := horizontal_alignment.Fill
                    Widget := stack_box:
                        Orientation := orientation.Horizontal
                        Slots := array:
                            stack_box_slot:
                                HorizontalAlignment := horizontal_alignment.Fill
                                VerticalAlignment := vertical_alignment.Fill
                                Widget := text_block:
                                    DefaultText := MakeMessage("Lv. ")
                                    DefaultTextColor := ThemeColors.Highlight
                                    DefaultShadowColor := ThemeColors.Highlight * 0.125
                                    DefaultShadowOffset := option{vector2{X := 2.0, Y := 2.0}}
                                    DefaultShadowOpacity := 1.0
                            stack_box_slot:
                                HorizontalAlignment := horizontal_alignment.Fill
                                VerticalAlignment := vertical_alignment.Fill
                                Widget := LevelText

                stack_box_slot:
                    HorizontalAlignment := horizontal_alignment.Fill
                    VerticalAlignment := vertical_alignment.Top
                    Padding := margin{Left := 0.0, Top := 40.0, Right := 0.0, Bottom := 40.0}
                    Widget := color_block:
                        DefaultColor := NamedColors.White
                        DefaultOpacity := 0.1
                        DefaultDesiredSize := vector2{X := 0.0, Y := 1.0}
                stack_box_slot:
                    HorizontalAlignment := horizontal_alignment.Fill
                    Padding := margin{Left := 0.0, Top := 0.0, Right := 0.0, Bottom := 10.0}
                    Widget := text_block:
                        DefaultText := MakeMessage("Difficulty")
                        DefaultTextColor := ThemeColors.PrimaryText
                        DefaultTextOpacity := 0.5
                stack_box_slot:
                    HorizontalAlignment := horizontal_alignment.Fill
                    Widget := CreateStatUI(HealthMultiplierText, "Health", ?Padding := LabelPadding)
                stack_box_slot:
                    HorizontalAlignment := horizontal_alignment.Fill
                    Widget := CreateStatUI(DamageMultiplierText, "Damage", ?Padding := LabelPadding)
                stack_box_slot:
                    HorizontalAlignment := horizontal_alignment.Fill
                    Widget := CreateStatUI(MovespeedMultiplierText, "Movespeed", ?Padding := LabelPadding)
                stack_box_slot:
                    HorizontalAlignment := horizontal_alignment.Fill
                    Widget := CreateStatUI(StageText, "Stages", ?Padding := LabelPadding)
                stack_box_slot:
                    HorizontalAlignment := horizontal_alignment.Fill
                    Widget := CreateStatUI(StageMultiplierText, "Stage (HP/DMG)", ?Padding := LabelPadding)

                stack_box_slot:
                    HorizontalAlignment := horizontal_alignment.Fill
                    VerticalAlignment := vertical_alignment.Top
                    Padding := margin{Left := 0.0, Top := 40.0, Right := 0.0, Bottom := 40.0}
                    Widget := color_block:
                        DefaultColor := NamedColors.White
                        DefaultOpacity := 0.1
                        DefaultDesiredSize := vector2{X := 0.0, Y := 1.0}
                stack_box_slot:
                    HorizontalAlignment := horizontal_alignment.Fill
                    Padding := margin{Left := 0.0, Top := 0.0, Right := 0.0, Bottom := 10.0}
                    Widget := text_block:
                        DefaultText := MakeMessage("Rewards")
                        DefaultTextColor := ThemeColors.PrimaryText
                        DefaultTextOpacity := 0.5
                stack_box_slot:
                    HorizontalAlignment := horizontal_alignment.Fill
                    Widget := CreateStatUI(GoldMultiplierText, "Gold Multiplier", ?Padding := LabelPadding)

    CreatePlayerInfoUI():widget =
        stack_box:
            Orientation := orientation.Vertical
            Slots := array:
                stack_box_slot:
                    HorizontalAlignment := horizontal_alignment.Fill
                    Widget := CreateStatUI(PlayersReadyText, "Players Ready", ?Padding := LabelPadding)
                stack_box_slot:
                    HorizontalAlignment := horizontal_alignment.Fill
                    Widget := stack_box:
                        Orientation := orientation.Horizontal
                        Slots := array:
                             stack_box_slot:
                                HorizontalAlignment := horizontal_alignment.Left
                                Distribution := option{1.0}
                                Padding := margin{Left := 0.0, Top := 0.0, Right := LabelPadding, Bottom := 0.0}
                                Widget := text_block:
                                    DefaultText := MakeMessage("Team Arena Level:")
                                    DefaultTextColor := ThemeColors.PrimaryText
                            stack_box_slot:
                                HorizontalAlignment := horizontal_alignment.Right
                                Widget := stack_box:
                                    Orientation := orientation.Horizontal
                                    Slots := array:
                                        stack_box_slot:
                                            HorizontalAlignment := horizontal_alignment.Fill
                                            VerticalAlignment := vertical_alignment.Fill
                                            Widget := text_block:
                                                DefaultText := MakeMessage("Lv. ")
                                                DefaultTextColor := ThemeColors.Highlight
                                                DefaultShadowColor := ThemeColors.Highlight * 0.125
                                                DefaultShadowOffset := option{vector2{X := 2.0, Y := 2.0}}
                                                DefaultShadowOpacity := 1.0
                                        stack_box_slot:
                                            HorizontalAlignment := horizontal_alignment.Fill
                                            VerticalAlignment := vertical_alignment.Fill
                                            Widget := TeamArenaLevelText
                stack_box_slot:
                    HorizontalAlignment := horizontal_alignment.Fill
                    Padding := margin{Left := 0.0, Top := 40.0, Right := 0.0, Bottom := 40.0}
                    Widget := color_block:
                        DefaultColor := NamedColors.White
                        DefaultOpacity := 0.1
                        DefaultDesiredSize := vector2{X := 0.0, Y := 1.0}

                stack_box_slot:
                    HorizontalAlignment := horizontal_alignment.Fill
                    VerticalAlignment := vertical_alignment.Fill
                    Widget := stack_box:
                        Orientation := orientation.Vertical
                        Slots := for (Index := 1..MaxPlayers):
                            PlayerInfoEntryUI := arena_select_player_info_entry_ui{Player := Player}
                            PlayerInfoEntryUI.Initialize()
                            set PlayerInfoEntryUIs += array{PlayerInfoEntryUI}

                            stack_box_slot:
                                HorizontalAlignment := horizontal_alignment.Fill
                                VerticalAlignment := vertical_alignment.Fill
                                Padding := margin{Left := 0.0, Top := (Index = 1) and 0.0 or 10.0, Right := 0.0, Bottom := 0.0}
                                Widget := PlayerInfoEntryUI.GetUI()
    
    <# ----------------------------------------------------------------------------------------------------
        Get Functions
    ---------------------------------------------------------------------------------------------------- #>
    IsArenaUnlocked(ArenaLevel:int, Difficulty:int)<transacts>:logic =
        var UnlockedCount:int = 0
        var PlayerCount:int = 0
        for (AllPlayer : GetAllPlayers()):
            set PlayerCount += 1
            if (AllPlayer.GetArenaProgress() + 1 >= CalculateArenaLevel(ArenaLevel, Difficulty)):
                set UnlockedCount += 1

        logic{UnlockedCount >= PlayerCount}

    IsDifficultyUnlocked(Difficulty:int)<transacts>:logic =
        var UnlockedCount:int = 0
        var PlayerCount:int = 0
        for (AllPlayer : GetAllPlayers()):
            set PlayerCount += 1
            if (AllPlayer.GetArenaProgress() >= GetArenaCount() * (Difficulty - 1)):
                set UnlockedCount += 1

        logic{UnlockedCount >= PlayerCount}

    IsGameSpeedUnlocked(GameSpeed:int)<transacts>:logic =
        var UnlockedCount:int = 0
        var PlayerCount:int = 0
        for (AllPlayer : GetAllPlayers()):
            set PlayerCount += 1
            if ((GameSpeed = 1) or (AllPlayer.GetArenaProgress() >= GetGameSpeedUnlockLevel(GameSpeed))):
                set UnlockedCount += 1

        logic{UnlockedCount >= PlayerCount}

    GetGameSpeedUnlockLevel(GameSpeed:int)<transacts>:int =
        (GameSpeed - 1) * GetArenaCount() * 2

    <# ----------------------------------------------------------------------------------------------------
        Update Functions
    ---------------------------------------------------------------------------------------------------- #>
    Update()<suspends>:void =
        if (LoadingTask := MaybeLoadingTask?):
            LoadingTask.Await()
            return

        if (GameSettings := GetGameSettings[]):
            set CurrentArenaIndex = GameSettings.Arena.Level - 1
            ArenaInfoUI.SetVisibility(IsPlayerInfo? and widget_visibility.Hidden or widget_visibility.Visible)
            PlayerInfoUI.SetVisibility(IsPlayerInfo? and widget_visibility.Visible or widget_visibility.Hidden)
            UpdateInfo(GameSettings.Arena)
            UpdateDifficulty(GameSettings.Difficulty)
            UpdateGameSpeed(GameSettings.GameSpeed)
            UpdatePlayerReady()
            UpdatePlayerInfo()
        else:
            LoadingTask := spawn{LoadArena()}
            set MaybeLoadingTask = option{LoadingTask}

    UpdateInfo(Arena:arena):void =
        NameText.SetText(MakeMessage(Arena.Name))
        LevelText.SetText(MakeMessage("{CalculateArenaLevel(Arena)}"))

        DifficultyText.SetText(MakeMessage("{Arena.Difficulty}"))

        StageText.SetText(MakeMessage("{GetArenaStageCount()}"))
        GoldMultiplierText.SetText(MakeMessage("+{FormatAsPercent(CalculateArenaGoldMultiplier(Arena))}"))
        
        ArenaHealthMultiplier := CalculateArenaHealthMultiplier(Arena)
        HealthMultiplierText.SetText(MakeMessage("+{FormatAsPercent(ArenaHealthMultiplier)}"))
        HealthMultiplierText.SetColor(ArenaHealthMultiplier > 0.0 and ThemeColors.Failure or ThemeColors.PrimaryText)
        
        ArenaDamageMultiplier := CalculateArenaDamageMultiplier(Arena)
        DamageMultiplierText.SetText(MakeMessage("+{FormatAsPercent(ArenaDamageMultiplier)}"))
        DamageMultiplierText.SetColor(ArenaDamageMultiplier > 0.0 and ThemeColors.Failure or ThemeColors.PrimaryText)

        ArenaMovespeedMultiplier := CalculateArenaMovespeedMultiplier(Arena)
        MovespeedMultiplierText.SetText(MakeMessage("+{FormatAsPercent(ArenaMovespeedMultiplier)}"))
        MovespeedMultiplierText.SetColor(ArenaMovespeedMultiplier > 0.0 and ThemeColors.Failure or ThemeColors.PrimaryText)
        
        ArenaStageMultiplier := GetArenaStageMultiplier()
        StageMultiplierText.SetText(MakeMessage("+{FormatAsPercent(ArenaStageMultiplier)}/Stage"))
        StageMultiplierText.SetColor(ArenaStageMultiplier > 0.0 and ThemeColors.Failure or ThemeColors.PrimaryText)

        GameDifficulty := GetGameDifficulty()
        for (Index -> ArenaColorButton : ArenaColorButtons):
            ArenaLevel := Index + 1
            if (IsArenaUnlocked(ArenaLevel, GameDifficulty)?):
                ArenaColorButton.SetText(MakeMessage("{ArenaLevel}"))
            else:
                ArenaColorButton.SetText(MakeMessage("🔒"))
        
        for (Index -> DifficultyColorButton : DifficultyColorButtons):
            Difficulty := Index + 1
            if (IsDifficultyUnlocked(Difficulty)?):
                DifficultyColorButton.SetText(MakeMessage("{Difficulty}"))
            else:
                DifficultyColorButton.SetText(MakeMessage("🔒"))

        for (Index -> GameSpeedColorButton : GameSpeedColorButtons):
            GameSpeed := Index + 1
            if (IsGameSpeedUnlocked(GameSpeed)?):
                GameSpeedColorButton.SetText(MakeMessage("x{GameSpeed}"))
            else:
                GameSpeedColorButton.SetText(MakeMessage("🔒 Lv. {GetGameSpeedUnlockLevel(GameSpeed)}"))

        for (Index -> ColorBlock : ArenaColorBlocks):
            ColorBlock.SetVisibility(Index = Arena.Level - 1 and widget_visibility.Visible or widget_visibility.Hidden)

    LoadArena()<suspends>:void =
        race:           
            LoadArenaCancelledEvent.Await()
            block:
                if (Arena := GetArenas()[CurrentArenaIndex], ArenaManager := GetArenaManager[]):
                    if:
                        UpdatedGameSettings := game_settings:
                            MakeGameSettings<constructor>(GetGameSettings[])
                                Arena := Arena
                    then:
                        SetGameSettings(UpdatedGameSettings)
                        
                    UpdateInfo(Arena)
                    StartButton.SetText(MakeMessage("Loading ..."))
                    spawn{ArenaManager.LoadArena(CurrentArenaIndex)}
                    FadeBackground(ArenaManager.GetLoadTimeRemaining())

                UpdatePlayerReady()

        set MaybeLoadingTask = false

    FadeBackground(Duration:float)<suspends>:void =
        defer:
            SetBackgroundOpacity(0.0)

        race:
            Sleep(Duration)
            block:
                SetBackgroundOpacity(1.0)
                Sleep(Duration - 1.0)
                DeltaTimer := delta_timer{}
                loop:
                    if (Progress := type{_X:float where 0.000000 <= _X, _X <= 1.000000}[Pow(1.0 - DeltaTimer.GetElapsedTime() / 1.0, 1.0 / 3.0)]):
                        SetBackgroundOpacity(Progress)

                    Sleep(0.0)

    UpdatePlayerReady():void =
        PlayersReadyText.SetText(MakeMessage("{GetReadyPlayerCount()}/{GetAllPlayers().Length}"))
        PlayersReadyText.SetColor(IsAllPlayerReady()? and ThemeColors.Success or ThemeColors.Failure)

        if (IsGameCountdownInProgress()?):
            StartButton.SetText(MakeMessage("Cancel Countdown"))
        else:
            StartButton.SetText(MakeMessage(IsAllPlayerReady()? and "Start" or "Start Countdown"))

    GetTeamArenaLevel()<transacts>:tuple(int, int) =
        var MinArenaProgressIndex:int = 0
        var MinArenaProgress:int = 0
        for (Index -> AllPlayer : GetAllPlayers(), PlayerArenaProgress := AllPlayer.GetArenaProgress(), Index = 0 or PlayerArenaProgress < MinArenaProgress):
            set MinArenaProgressIndex = Index
            set MinArenaProgress = PlayerArenaProgress

        (MinArenaProgressIndex, MinArenaProgress)

    UpdatePlayerInfo():void =
        AllPlayers := GetAllPlayers()
        TeamArenaLevel := GetTeamArenaLevel()

        TeamArenaLevelText.SetText(MakeMessage("{TeamArenaLevel(1)}"))
        
        for (Index -> PlayerInfoEntryUI : PlayerInfoEntryUIs):
            if (AllPlayer := AllPlayers[Index]):
                PlayerInfoEntryUI.Show()
                PlayerInfoEntryUI.Update(AllPlayer, logic{Index = TeamArenaLevel(0)})
            else:
                PlayerInfoEntryUI.Hide()

    StartGame()<suspends>:void =
        Sleep(TickTime * 2)

        if (GameManager := GetGameManager[]):
            GameManager.Start(CurrentArenaIndex)

    UpdateArena(UpdatedIndex:int):void =
        if (IsGameInProgress()?):
            return

        LoadArenaCancelledEvent.Signal()
        set CurrentArenaIndex = UpdatedIndex

        LoadingTask := spawn{LoadArena()}
        set MaybeLoadingTask = option{LoadingTask}

    UpdateDifficulty(Difficulty:int):void =
        set CurrentDifficulty = Difficulty

        if:
            UpdatedGameSettings := game_settings:
                MakeGameSettings<constructor>(GetGameSettings[])
                    Difficulty := Difficulty
        then:
            SetGameSettings(UpdatedGameSettings)

        for (Index -> ColorBlock : DifficultyColorBlocks):
            ColorBlock.SetVisibility(Index = Difficulty - 1 and widget_visibility.Visible or widget_visibility.Hidden)

        if (Arena := GetArenas()[CurrentArenaIndex]):
            UpdateInfo(Arena)

        if (not IsArenaUnlocked(CurrentArenaIndex + 1, Difficulty)?):
            UpdateArena(0)

    UpdateGameSpeed(GameSpeed:int):void =
        set CurrentGameSpeed = GameSpeed

        if:
            UpdatedGameSettings := game_settings:
                MakeGameSettings<constructor>(GetGameSettings[])
                    GameSpeed := GameSpeed
        then:
            SetGameSettings(UpdatedGameSettings)

        for (Index -> ColorBlock : GameSpeedColorBlocks):
            ColorBlock.SetVisibility(Index = GameSpeed - 1 and widget_visibility.Visible or widget_visibility.Hidden)

        if (Arena := GetArenas()[CurrentArenaIndex]):
            UpdateInfo(Arena)

    Timeout()<suspends>:void =
        TimeoutDeltaTimer.Reset()

        race:
            TimeoutCancelEvent.Await()
            loop:
                Sleep(1.0)
                if (TimeoutDeltaTimer.GetElapsedTime() >= TimeoutDuration):
                    Hide()
                    break

    <# ----------------------------------------------------------------------------------------------------
        Event Handlers
    ---------------------------------------------------------------------------------------------------- #>
    AwaitArenaClicked(Event:listenable(widget_message), ArenaIndex:int)<suspends>:void =
        loop:
            Event.Await()
            TimeoutDeltaTimer.Reset()

            if (IsGameCountdownInProgress()? or not IsArenaUnlocked(ArenaIndex + 1, GetGameDifficulty())?):
                Player.PlayAudio(audio_id.Error)
            else if (ArenaIndex <> CurrentArenaIndex):
                UpdateArena(ArenaIndex)

    AwaitDifficultyClicked(Event:listenable(widget_message), Difficulty:int)<suspends>:void =
        loop:
            Event.Await()
            TimeoutDeltaTimer.Reset()

            if (IsGameCountdownInProgress()? or not IsDifficultyUnlocked(Difficulty)?):
                Player.PlayAudio(audio_id.Error)
            else if (Difficulty <> CurrentDifficulty):
                UpdateDifficulty(Difficulty)

    AwaitGameSpeedClicked(Event:listenable(widget_message), GameSpeed:int)<suspends>:void =
        loop:
            Event.Await()
            TimeoutDeltaTimer.Reset()

            if (IsGameCountdownInProgress()? or not IsGameSpeedUnlocked(GameSpeed)?):
                Player.PlayAudio(audio_id.Error)
            else if (GameSpeed <> CurrentGameSpeed):
                UpdateGameSpeed(GameSpeed)

    OnClickStart(Message:widget_message):void =
        TimeoutDeltaTimer.Reset()

        if (IsGameInProgress()? or MaybeLoadingTask?):
            Player.PlayAudio(audio_id.Error)
            return

        if (IsGameCountdownInProgress()?):
            StopGameCountdown()
            spawn{Update()}
            return

        TimeoutCancelEvent.Signal()
        Hide()

        if (IsAllPlayerReady()?):
            spawn{StartGame()}
        else:
            StartGameCountdown(CurrentArenaIndex)

    OnClickInfo(Message:widget_message):void =
        set IsPlayerInfo = logic{not IsPlayerInfo?}
        InfoButton.SetText(MakeMessage("{IsPlayerInfo? and "Arena" or "Player"} Info"))
        ArenaInfoUI.SetVisibility(IsPlayerInfo? and widget_visibility.Hidden or widget_visibility.Visible)
        PlayerInfoUI.SetVisibility(IsPlayerInfo? and widget_visibility.Visible or widget_visibility.Hidden)

    OnPlayerReadied():void =
        if (not MaybeLoadingTask?):
            UpdatePlayerReady()

    OnPlayerUnreadied():void =
        if (not MaybeLoadingTask?):
            UpdatePlayerReady()

    OnShow<override>():void =
        if (GetAllPlayers().Length > 1):
            spawn{Timeout()}

    OnClosed(InPlayer:player):void =
        TimeoutCancelEvent.Signal()

    OnGameStarted():void =
        TimeoutCancelEvent.Signal()

    OnPlayerAdded(InPlayer:player):void =
        if (not IsShowing?) then return
        spawn{HandlePlayerAdded(InPlayer)}

    HandlePlayerAdded(InPlayer:player)<suspends>:void =
        TimeoutCancelEvent.Signal()
        spawn{Timeout()}

        WaitForPlayerDataManager(InPlayer)

        UpdatePlayerInfo()
        UpdatePlayerReady()

        var IsUpdated:logic = false
        if (not IsArenaUnlocked(CurrentArenaIndex + 1, GetGameDifficulty())?):
            UpdateArena(0)
            set IsUpdated = true

        if (not IsDifficultyUnlocked(CurrentDifficulty)?):
            UpdateDifficulty(1)
            set IsUpdated = true

        if (not IsGameSpeedUnlocked(CurrentGameSpeed)?):
            UpdateGameSpeed(1)
            set IsUpdated = true

        if (not IsUpdated?):
            UpdateGameSpeed(CurrentGameSpeed)

    OnPlayerRemoved(InPlayer:player):void =
        if (not IsShowing?) then return

        UpdatePlayerInfo()
        UpdatePlayerReady()
        UpdateGameSpeed(CurrentGameSpeed)