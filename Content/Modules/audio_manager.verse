using { /Fortnite.com/Devices }
using { /Verse.org/Simulation }

<# ----------------------------------------------------------------------------------------------------
    Data definitions
---------------------------------------------------------------------------------------------------- #>
audio_id := enum:
    Damage
    Impulse
    Projectile
    LevelUp
    Experience
    UpgradeProgress
    Upgrade
    CriticalHit
    Reroll
    UpgradeRare
    Victory
    Defeat
    Error
    Heal
    Magnet
    Warning
    Start
    Activate
    Surprise
    BossEliminated
    Ready
    Banish
    Notification
    Purchase
    TimerWarning
    WeaponExpire
    Boost
    Capture
    ObjectiveSuccess
    ObjectiveFailure
    PowerUp
    Trail
    MinibossEliminated
    WeaponEvolution
    PetSlime
    ProjectileEvolved
    Shield
    TimerCountdown
    OpenTreasureChest
    Reroll2
    Reroll3
    Reroll4
    Reroll5
    SupplyDrop

audio_player := class<concrete>:
    @editable
    Id:audio_id = audio_id.Damage
    @editable
    Device:audio_player_device = audio_player_device{}

music_id := enum:
    Default
    CombatLow
    CombatHigh
    BossFight
    DancingLights
    GetBackUp
    Lilly
    StareDown
    Chimera
    ABC
    RainBoots
    GoodEnding
    Darling

music_player := class<concrete>:
    @editable
    Id:music_id = music_id.CombatLow
    @editable
    Devices:[]audio_player_device = array{}
    
<# ----------------------------------------------------------------------------------------------------
    Global Functions
---------------------------------------------------------------------------------------------------- #>
var GlobalAudioPlayerManager:weak_map(session, audio_manager) = map{}

(Agent:agent).PlayAudio(Id:audio_id):audio_player_device =
    if (AudioPlayerManager := GlobalAudioPlayerManager[GetSession()]):
        AudioPlayerManager.Play(Agent, Id)
    else:
        audio_player_device{}

(Agent:agent).StopAudio(Id:audio_id):void =
    if (AudioPlayerManager := GlobalAudioPlayerManager[GetSession()]):
        AudioPlayerManager.Stop(Agent, Id)

(Agent:agent).PlayMusic(Id:music_id):void =
    if (AudioPlayerManager := GlobalAudioPlayerManager[GetSession()]):
        AudioPlayerManager.Play(Agent, Id)

(Agent:agent).ResumeMusic():void =
    if (AudioPlayerManager := GlobalAudioPlayerManager[GetSession()]):
        AudioPlayerManager.ResumeMusic(Agent)

(Agent:agent).ClearMusic():void =
    if (AudioPlayerManager := GlobalAudioPlayerManager[GetSession()]):
        AudioPlayerManager.ClearMusic(Agent)

(Player:player).ToggleMusic():void =
    if:
        CurrentPlayerData := Player.GetData[]
        UpdatedIsMusicEnabled := logic{not CurrentPlayerData.Settings.Music?}

        set PlayerDataMap[Player] = player_data:
            MakePlayerData<constructor>(CurrentPlayerData)
            Settings := player_settings:
                MakePlayerSettings<constructor>(CurrentPlayerData.Settings)
                    Music := UpdatedIsMusicEnabled
    then:
        if (UpdatedIsMusicEnabled?):
            Player.ResumeMusic()
        else:
            Player.ClearMusic()

<# ----------------------------------------------------------------------------------------------------
    audio_manager
---------------------------------------------------------------------------------------------------- #>
audio_manager := class(creative_device):
    @editable
    AudioPlayers:[]audio_player = array{}

    @editable
    MusicPlayers:[]music_player = array{}
    
    DefaultMusicId:music_id = music_id.Default
    DefaultAudioPlayer:audio_player_device = audio_player_device{}

    var MappedAudioPlayers:[audio_id]audio_player_device = map{}
    var MappedMusicPlayers:[music_id]music_player = map{}
    var PlayerActiveMusic:[agent]?audio_player_device = map{}

    <# ----------------------------------------------------------------------------------------------------
        Initialization
    ---------------------------------------------------------------------------------------------------- #>
    OnBegin<override>()<suspends>:void =
        if (set GlobalAudioPlayerManager[GetSession()] = Self) {}

        for (AudioPlayer : AudioPlayers, set MappedAudioPlayers[AudioPlayer.Id] = AudioPlayer.Device) {}
        for (MusicPlayer : MusicPlayers, set MappedMusicPlayers[MusicPlayer.Id] = MusicPlayer) {}

        for (Player : GetPlayspace().GetPlayers()):
            Initialize(Player)

        GetPlayspace().PlayerAddedEvent().Subscribe(OnPlayerAdded)
        GetPlayspace().PlayerRemovedEvent().Subscribe(OnPlayerRemoved)

    Initialize(Player:player):void =
        Play(Player, DefaultMusicId)

    <# ----------------------------------------------------------------------------------------------------
        Get Functions
    ---------------------------------------------------------------------------------------------------- #>
    GetAudioPlayer(Id:audio_id)<decides><transacts>:audio_player_device =
        MappedAudioPlayers[Id]

    GetMusicPlayer(Id:music_id)<decides><transacts>:music_player =
        MappedMusicPlayers[Id]

    <# ----------------------------------------------------------------------------------------------------
        Get Functions
    ---------------------------------------------------------------------------------------------------- #>
    Play(Agent:agent, Id:audio_id):audio_player_device =
        if (AudioPlayer := GetAudioPlayer[Id]):
            AudioPlayer.Play(Agent)
            AudioPlayer
        else:
            DefaultAudioPlayer

    Play(Agent:agent, Id:music_id):void =
        ClearMusic(Agent)
        if (MusicPlayer := GetMusicPlayer[Id]):
            if (Player := player[Agent], Player.GetData[].Settings.Music?, AudioPlayer := MusicPlayer.Devices[Floor[GetGameSpeed() - 1.0]] or MusicPlayer.Devices[0]):
                AudioPlayer.Play(Agent)
                if (set PlayerActiveMusic[Agent] = option{AudioPlayer}) {}

    Stop(Agent:agent, Id:audio_id):void =
        if (AudioPlayer := GetAudioPlayer[Id]):
            AudioPlayer.Stop(Agent)

    ResumeMusic(Agent:agent):void =
        if (AudioPlayer := PlayerActiveMusic[Agent]?):
            AudioPlayer.Play(Agent)

    ClearMusic(Agent:agent):void =
        if (AudioPlayer := PlayerActiveMusic[Agent]?):
            AudioPlayer.Stop(Agent)

    <# ----------------------------------------------------------------------------------------------------
        Event Handlers
    ---------------------------------------------------------------------------------------------------- #>
    OnPlayerAdded(Player:player):void =
        Initialize(Player)

    OnPlayerRemoved(Player:player):void =
        ClearMusic(Player)
        set PlayerActiveMusic = RemoveKeyFromMap(PlayerActiveMusic, Player)