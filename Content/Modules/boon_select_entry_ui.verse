using { /Fortnite.com/UI }
using { /Verse.org/Colors }
using { /Verse.org/Simulation }
using { /UnrealEngine.com/Temporary/SpatialMath }
using { /UnrealEngine.com/Temporary/UI }

<# ----------------------------------------------------------------------------------------------------
    Constants
---------------------------------------------------------------------------------------------------- #>



<# ----------------------------------------------------------------------------------------------------
    Data Definitions
---------------------------------------------------------------------------------------------------- #>



<# ----------------------------------------------------------------------------------------------------
    boon_select_entry_ui
---------------------------------------------------------------------------------------------------- #>
boon_select_entry_ui := class(ui_component):
    SelectEvent:events(boon_info) = events(boon_info){}

    Icon:texture_block = texture_block:
        DefaultImage := UserInterface.Icons.WeaponSword_128
        DefaultDesiredSize := vector2{X := 128.0, Y := 128.0}

    NameText:text_block = text_block:
        DefaultTextColor := ThemeColors.PrimaryText
        DefaultShadowOffset := option{vector2{X := 2.0, Y := 2.0}}
        DefaultShadowOpacity := 0.5

    DescriptionText:text_block = text_block:
        DefaultTextColor := ThemeColors.PrimaryText
        DefaultTextOpacity := 0.5

    TierText:text_block = text_block:
        DefaultShadowOffset := option{vector2{X := 2.0, Y := 2.0}}
        DefaultShadowOpacity := 1.0

    TypeText:text_block = text_block:
        DefaultTextColor := ThemeColors.Success
        DefaultShadowOffset := option{vector2{X := 2.0, Y := 2.0}}
        DefaultShadowColor := ThemeColors.Success * 0.125
        DefaultShadowOpacity := 0.5

    SelectButton:button_quiet = button_quiet{}

    var InfoOverlay:overlay = overlay{}
    var MaybeBoonInfo:?boon_info = false
        
    <# ----------------------------------------------------------------------------------------------------
        Create UI
    ---------------------------------------------------------------------------------------------------- #>
    CreateUI<override>():widget =
        SelectButton.SetText(MakeMessage("Select"))
        SelectButton.OnClick().Subscribe(OnClickSelect)

        stack_box:
            Orientation := orientation.Vertical
            Slots := array:
                stack_box_slot:
                    HorizontalAlignment := horizontal_alignment.Fill
                    VerticalAlignment := vertical_alignment.Fill
                    Distribution := option{1.0}
                    Padding := margin{Left := 0.0, Top := 0.0, Right := 0.0, Bottom := 0.0}
                    Widget := set InfoOverlay = overlay:
                        Slots := array:
                            overlay_slot:
                                HorizontalAlignment := horizontal_alignment.Fill
                                VerticalAlignment := vertical_alignment.Fill
                                Widget := MakeBox(?Height := 600.0)
                            overlay_slot:
                                HorizontalAlignment := horizontal_alignment.Fill
                                VerticalAlignment := vertical_alignment.Fill
                                Padding := margin{Left := 40.0, Top := 40.0, Right := 40.0, Bottom := 40.0}
                                Widget := CreateInfoUI()
                stack_box_slot:
                    HorizontalAlignment := horizontal_alignment.Fill
                    VerticalAlignment := vertical_alignment.Fill
                    Padding := margin{Left := 0.0, Top := 10.0, Right := 0.0, Bottom := 0.0}
                    Widget := SelectButton

    CreateInfoUI():widget =
        stack_box:
            Orientation := orientation.Vertical
            Slots := array:
                stack_box_slot:
                    HorizontalAlignment := horizontal_alignment.Fill
                    VerticalAlignment := vertical_alignment.Top
                    Widget := color_block:
                        DefaultOpacity := 0.0
                        DefaultDesiredSize := vector2{X := 450.0, Y := 0.0}
                stack_box_slot:
                    HorizontalAlignment := horizontal_alignment.Left
                    VerticalAlignment := vertical_alignment.Top
                    Padding := margin{Left := 0.0, Top := 0.0, Right := 0.0, Bottom := 20.0}
                    Widget := Icon
                stack_box_slot:
                    HorizontalAlignment := horizontal_alignment.Left
                    VerticalAlignment := vertical_alignment.Fill
                    Widget := TierText
                stack_box_slot:
                    HorizontalAlignment := horizontal_alignment.Left
                    VerticalAlignment := vertical_alignment.Fill
                    Widget := NameText
                stack_box_slot:
                    HorizontalAlignment := horizontal_alignment.Left
                    VerticalAlignment := vertical_alignment.Top
                    Padding := margin{Left := 0.0, Top := 40.0, Right := 0.0, Bottom := 0.0}
                    Widget := DescriptionText
                stack_box_slot:
                    HorizontalAlignment := horizontal_alignment.Center
                    VerticalAlignment := vertical_alignment.Bottom
                    Distribution := option{1.0}
                    Widget := TypeText

    <# ----------------------------------------------------------------------------------------------------
        Get Functions
    ---------------------------------------------------------------------------------------------------- #>
    GetActiveBoonInfo()<transacts>:?boon_info =
        MaybeBoonInfo

    <# ----------------------------------------------------------------------------------------------------
        Update Functions
    ---------------------------------------------------------------------------------------------------- #>
    Update(BoonInfo:boon_info):void =
        set MaybeBoonInfo = option{BoonInfo}
        InfoOverlay.SetVisibility(widget_visibility.Visible)
        Icon.SetImage(BoonInfo.Icon)
        Icon.SetTint(BoonInfo.IconColor)
        TierText.SetText(MakeMessage("{BoonInfo.Tier.ToTierString().ToUpper()}"))
        TierText.SetTextColor(BoonInfo.Tier.ToTierColor() * 2.0)
        TierText.SetShadowColor(BoonInfo.Tier.ToTierColor() * 0.25)
        NameText.SetText(MakeMessage("{BoonInfo.Name}"))
        DescriptionText.SetText(MakeMessage("{BoonInfo.Description.WrapText(450.0)}"))
        TypeText.SetText(MakeMessage("{BoonInfo.Type}"))
        SelectButton.SetText(MakeMessage("Select"))
    
    Clear():void =
        set MaybeBoonInfo = false

    <# ----------------------------------------------------------------------------------------------------
        Event handlers
    ---------------------------------------------------------------------------------------------------- #>
    OnClickSelect(Message:widget_message):void =           
        if (BoonInfo := MaybeBoonInfo?):
            SelectEvent.Signal(BoonInfo)
        else:
            Player.PlayAudio(audio_id.Error)