using { /Fortnite.com/Devices }
using { /Verse.org/Simulation }

<# ----------------------------------------------------------------------------------------------------
    character
---------------------------------------------------------------------------------------------------- #>
character := class(creative_device):
    @editable
    Character:character_device = character_device{}

    @editable
    Button:button_device = button_device{}

    @editable
    PanelId:?panel_ui_id = false

    <# ----------------------------------------------------------------------------------------------------
        Initialization
    ---------------------------------------------------------------------------------------------------- #>
    OnBegin<override>()<suspends>:void =
        Button.InteractedWithEvent.Subscribe(OnInteractedWith)
    
    <# ----------------------------------------------------------------------------------------------------
        Event handlers
    ---------------------------------------------------------------------------------------------------- #>
    OnInteractedWith(Agent:agent):void =
        Character.PlayEmote()
    
        if (Id := PanelId?, Player := player[Agent]):
            Player.ShowPanelUI(Id)
            Player.PlayAudio(audio_id.Activate)