using { /Verse.org/Simulation }

<# ----------------------------------------------------------------------------------------------------
    Global Functions
---------------------------------------------------------------------------------------------------- #>
FormatCurrency(Value:float)<transacts>:string =
    var FormattedCurrency:string = ""
    var FormattedValue:float = Value
    var SuffixIndex:int = 0

    loop:
        if (FormattedValue < 1000.0):
            break

        set FormattedValue = FormattedValue / 1000.0
        set SuffixIndex += 1

    Decimals := SuffixIndex = 0 and 0 or if (FormattedValue / 10.0 >= 10.0):
        0
    else if (FormattedValue / 10.0 >= 1.0):
        1
    else:
        2
    
    Suffix := SuffixIndex = 0 and "" or NumberSuffixes[SuffixIndex] or ""
    GetRoundedString(FormattedValue, Decimals) + Suffix

NumberSuffixes:[]string = array:
    "-", "K", "M", "B", "T", "Qa", "Qt", "Sx", "Sp", "Oc", "No", "Dc",
    "UDc", "DDc", "TDc", "QaDc", "QtDc", "SxDc", "SpDc", "ODc", "NDc", "Vg",
    "UVg", "DVg", "TVg", "QaVg", "QtVg", "SxVg", "SpVg", "OVg", "NVg", "Tg",
    "UTg", "DTg", "TTg", "QaTg", "QtTg", "SxTg", "SpTg", "OTg", "NTg", "Qd",
    "UQd", "DQd", "TQd", "QaQd", "QtQd", "SxQd", "SpQd", "OQd", "NQd", "Qi",
    "UQi", "DQi", "TQi", "QaQi", "QtQi", "SxQi", "SpQi", "OQi", "NQi", "Se",
    "USe", "DSe", "TSe", "QaSe", "QtSe", "SxSe", "SpSe", "OSe", "NSe", "St",
    "USt", "DSt", "TSt", "QaSt", "QtSt", "SxSt", "SpSt", "OSt", "NSt", "Og",
    "UOg", "DOg", "TOg", "QaOg", "QtOg", "SxOg", "SpOg", "OOg", "NOg", "Nn",
    "UNn", "DNn", "TNn", "QaNn", "QtNn", "SxNn", "SpNn", "ONn", "NNn", "Ce"

<# ----------------------------------------------------------------------------------------------------
    currency_hud
---------------------------------------------------------------------------------------------------- #>
currency_hud := class:
    Player:player

    var Subscriptions:[]events_cancelable = array{}

    <# ----------------------------------------------------------------------------------------------------
        Initialization
    ---------------------------------------------------------------------------------------------------- #>
    Initialize(PlayerRemovedEvent:listenable(player)):void =
        if (GameManager := GetGameManager[], PlayerCurrencyChangedEvent := GetCurrencyManager[].PlayerCurrencyChangedEvent[Player]):
            set Subscriptions = array:
                GameManager.GameStartedEvent.Subscribe(OnGameStarted)
                GameManager.GameEndedEvent.Subscribe(OnGameEnded)
                GameManager.GameCurrencyChangedEvent.Subscribe(OnGameCurrencyChanged)
                GameManager.EliminationEvent.Subscribe(OnElimination)
                PlayerCurrencyChangedEvent.Subscribe(OnPlayerCurrencyChanged)

        PlayerRemovedEvent.Subscribe(OnPlayerRemoved)
        UpdatePlayerCurrency(currency_id.Gold)
        spawn{InitializeHUDMessage()}

    # [Hotfix] If HUD Message is shown for the first time Hide() fails if called within a short time frame.
    InitializeHUDMessage()<suspends>:void =
        Player.ShowMessage(message_id.CurrencyTreasureChest, "{FormatCurrency(0.0)}", ?DisplayTime := 0.0)
        Sleep(1.0)
        Player.HideMessage(message_id.CurrencyTreasureChest)

    <# ----------------------------------------------------------------------------------------------------
        Update Functions
    ---------------------------------------------------------------------------------------------------- #>
    Show():void =
        UpdateGameCurrency(game_currency_id.Wood)
        UpdateElimination()

    Hide():void =
        Player.HideMessage(message_id.CurrencyWood)
        Player.HideMessage(message_id.CurrencyElimination)
        Player.HideMessage(message_id.CurrencyTreasureChest)

    UpdateGameCurrency(Id:game_currency_id):void =
        case (Id):
            game_currency_id.Wood =>
                Player.ShowMessage(message_id.CurrencyWood, "{FormatCurrency(Player.GetGameCurrency(game_currency_id.Wood))}", ?DisplayTime := 0.0)

            game_currency_id.TreasureChest =>
                Amount := Player.GetGameCurrency(game_currency_id.TreasureChest)
                if (Amount > 0.0):
                    Player.ShowMessage(message_id.CurrencyTreasureChest, "{FormatCurrency(Amount)}", ?DisplayTime := 0.0)
                else:
                    Player.HideMessage(message_id.CurrencyTreasureChest)

    UpdateElimination():void =
        Player.ShowMessage(message_id.CurrencyElimination, "{FormatCurrency(GetEnemiesEliminated() * 1.0)}", ?DisplayTime := 0.0)

    UpdatePlayerCurrency(Id:currency_id):void =
        if (Id = currency_id.Gold):
            Player.ShowMessage(message_id.CurrencyGold, "{FormatCurrency(Player.GetCurrencyAvailable(currency_id.Gold))}", ?DisplayTime := 0.0)
            Player.UpdatePowerUpIndicator()

    <# ----------------------------------------------------------------------------------------------------
        Event Handlers
    ---------------------------------------------------------------------------------------------------- #>
    OnGameStarted():void =
        Show()
        Player.HideMessage(message_id.CurrencyGold)
        Player.RegisterInput(input_id.OpenTreasureChest)

    OnGameEnded():void =
        Hide()
        Player.UnregisterInput(input_id.OpenTreasureChest)

    OnGameCurrencyChanged(Id:game_currency_id):void =
        UpdateGameCurrency(Id)

    OnElimination():void =
        UpdateElimination()

    OnPlayerRemoved(InPlayer:player):void =
        if (InPlayer <> Player) then return

        for (Subscription : Subscriptions):
            Subscription.Cancel()

        set Subscriptions = array{}

    OnPlayerCurrencyChanged(Id:currency_id):void =
        UpdatePlayerCurrency(Id)
