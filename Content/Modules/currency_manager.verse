using { /Fortnite.com/Devices }
using { /Verse.org/Assets }
using { /Verse.org/Simulation }

<# ----------------------------------------------------------------------------------------------------
    Data definitions
---------------------------------------------------------------------------------------------------- #>
currency_id := enum:
    Gold
    Wood

ToString(Id:currency_id)<transacts>:string =
    case (Id):
        currency_id.Gold => "Gold"
        currency_id.Wood => "Wood"

(Id:currency_id).ToInt()<transacts>:int =
    case (Id):
        currency_id.Gold => 0
        currency_id.Wood => 1
        
(Id:int).ToCurrencyId()<transacts>:currency_id =
    case (Id):
        0 => currency_id.Gold
        1 => currency_id.Wood
        _ => currency_id.Gold

(Id:currency_id).ToIcon()<transacts>:texture =
    case (Id):
        currency_id.Gold => UserInterface.Icons.CurrencyGold_64
        currency_id.Wood => UserInterface.Icons.CurrencyWood_64

<# ----------------------------------------------------------------------------------------------------
    Global Functions
---------------------------------------------------------------------------------------------------- #>
var GlobalCurrencyManager:weak_map(session, currency_manager) = map{}

GetCurrencyManager()<decides><transacts>:currency_manager =
    GlobalCurrencyManager[GetSession()]

(Player:player).GetCurrency(Id:currency_id)<transacts>:float =
    GetCurrencyManager[].GetCurrency(Player, Id) or 0.0

(Player:player).AddCurrency(Id:currency_id, Amount:float):void =
    if (CurrencyManager := GetCurrencyManager[]):
        CurrencyManager.AddCurrency(Player, Id, Amount)

(Player:player).RemoveCurrency(Id:currency_id, Amount:float):void =
    if (CurrencyManager := GetCurrencyManager[]):
        CurrencyManager.RemoveCurrency(Player, Id, Amount)

(Player:player).GetCurrencyChangedEvent()<decides><transacts>:events(currency_id) =
    GetCurrencyManager[].GetCurrencyChangedEvent[Player]

(Player:player).UpdateGoldCurrencyHUD():void =
    if (CurrencyManager := GetCurrencyManager[], CurrencyHUD := CurrencyManager.PlayerCurrencyHUD[Player]):
        CurrencyHUD.UpdatePlayerCurrency(currency_id.Gold)

<# ----------------------------------------------------------------------------------------------------
    currency_manager
---------------------------------------------------------------------------------------------------- #>
currency_manager := class(creative_device):

    var PlayerCurrencyChangedEvent:[player]events(currency_id) = map{}
    var PlayerCurrencyHUD:[player]currency_hud = map{}

    <# ----------------------------------------------------------------------------------------------------
        Initialization
    ---------------------------------------------------------------------------------------------------- #>
    OnBegin<override>()<suspends>:void =
        if (set GlobalCurrencyManager[GetSession()] = Self) {}

        for (Player : GetPlayspace().GetPlayers()):
            spawn{Initialize(Player)}

        GetPlayspace().PlayerAddedEvent().Subscribe(OnPlayerAdded)
        GetPlayspace().PlayerRemovedEvent().Subscribe(OnPlayerRemoved)

    Initialize(Player:player)<suspends>:void =
        if (set PlayerCurrencyChangedEvent[Player] = events(currency_id){}) {}
        
        WaitForGameManager()
        WaitForUIManager(Player)

        CurrencyHUD := currency_hud{Player := Player}
        CurrencyHUD.Initialize(GetPlayspace().PlayerRemovedEvent())
        if (set PlayerCurrencyHUD[Player] = CurrencyHUD) {}

    <# ----------------------------------------------------------------------------------------------------
        Get Functions
    ---------------------------------------------------------------------------------------------------- #>
    GetCurrency(Player:player, Id:currency_id)<transacts>:float =
        if (CurrentPlayerData := Player.GetData[], CurrencyAmount := CurrentPlayerData.Currencies[Id.ToInt()]):
            CurrencyAmount
        else:
            0.0

    CanAfford(Player:player, Id:currency_id, Amount:float)<transacts>:logic =
        logic{Player.GetData[].Currencies[Id.ToInt()] >= Amount}

    GetCurrencyChangedEvent(Player:player)<decides><transacts>:events(currency_id) =
        PlayerCurrencyChangedEvent[Player]

    <# ----------------------------------------------------------------------------------------------------
        Update Functions
    ---------------------------------------------------------------------------------------------------- #>
    AddCurrency(Player:player, Id:currency_id, Amount:float):void =
        if:
            CurrentPlayerData := Player.GetData[]
            var UpdatedCurrencies:[int]float = CurrentPlayerData.Currencies
            set UpdatedCurrencies[Id.ToInt()] = UpdatedCurrencies[Id.ToInt()] + Amount or Amount

            set PlayerDataMap[Player] = player_data:
                MakePlayerData<constructor>(CurrentPlayerData)
                Currencies := UpdatedCurrencies

        if (CurrencyChangedEvent := GetCurrencyChangedEvent[Player]):
            CurrencyChangedEvent.Signal(Id)

    RemoveCurrency(Player:player, Id:currency_id, Amount:float):void =
        if:
            CurrentPlayerData := Player.GetData[]
            var UpdatedCurrencies:[int]float = CurrentPlayerData.Currencies
            set UpdatedCurrencies[Id.ToInt()] -= Amount

            set PlayerDataMap[Player] = player_data:
                MakePlayerData<constructor>(CurrentPlayerData)
                Currencies := UpdatedCurrencies
        
        if (CurrencyChangedEvent := GetCurrencyChangedEvent[Player]):
            CurrencyChangedEvent.Signal(Id)

    <# ----------------------------------------------------------------------------------------------------
        Event Handlers
    ---------------------------------------------------------------------------------------------------- #>
    OnPlayerAdded(Player:player):void =
        spawn{Initialize(Player)}

    OnPlayerRemoved(Player:player):void =
        set PlayerCurrencyHUD = RemoveKeyFromMap(PlayerCurrencyHUD, Player)