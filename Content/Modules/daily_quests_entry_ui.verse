using { /Fortnite.com/UI }
using { /Verse.org/Colors }
using { /UnrealEngine.com/Temporary/SpatialMath }
using { /UnrealEngine.com/Temporary/UI }
using { RealTime }

daily_quests_entry_ui := class(ui_component):
    Index:int
    ClaimEvent:events(tuple()) = events(tuple()){}
    RerollEvent:events(tuple()) = events(tuple()){}

    RewardText:text_block = text_block:
        DefaultTextColor := ThemeColors.Success
        DefaultShadowColor := ThemeColors.Success * 0.125
        DefaultShadowOffset := option{vector2{X := 2.0, Y := 2.0}}
        DefaultShadowOpacity := 1.0

    Icon:texture_block = texture_block:
        DefaultImage := UserInterface.Icons.WeaponBow_128
        DefaultDesiredSize := vector2{X := 128.0, Y := 128.0}

    TierText:text_block = text_block:
        DefaultShadowOffset := option{vector2{X := 2.0, Y := 2.0}}
        DefaultShadowOpacity := 1.0

    DescriptionText:text_block = text_block:
        DefaultTextColor := ThemeColors.PrimaryText
        DefaultJustification := text_justification.Center

    ProgressText:text_block = text_block:
        DefaultTextColor := ThemeColors.Success
        DefaultShadowColor := ThemeColors.Success * 0.125
        DefaultShadowOffset := option{vector2{X := 2.0, Y := 2.0}}
        DefaultShadowOpacity := 1.0

    ClaimRewardsButton:button_quiet = button_quiet{}

    NextQuestText:text_block = text_block:
        DefaultTextColor := ThemeColors.PrimaryText
        DefaultJustification := text_justification.Center
        DefaultShadowColor := NamedColors.Black
        DefaultShadowOffset := option{vector2{X := 2.0, Y := 2.0}}
        DefaultShadowOpacity := 0.5

    var InfoStackBox:stack_box = stack_box:
        Orientation := orientation.Vertical

    var MaybeCurrentRewardAmount:?float = false
    var MaybeActiveQuest:?daily_quest = false

    <# ----------------------------------------------------------------------------------------------------
        Initialization
    ---------------------------------------------------------------------------------------------------- #>
    CreateUI<override>():widget =
        ClaimRewardsButton.SetText(MakeMessage("Claim Rewards"))
        ClaimRewardsButton.OnClick().Subscribe(OnClickClaimRewards)

        stack_box:
            Orientation := orientation.Vertical
            Slots := array:
                stack_box_slot:
                    HorizontalAlignment := horizontal_alignment.Fill
                    VerticalAlignment := vertical_alignment.Fill
                    Distribution := option{1.0}
                    Widget := overlay:
                        Slots := array:
                            overlay_slot:
                                HorizontalAlignment := horizontal_alignment.Fill
                                VerticalAlignment := vertical_alignment.Fill
                                Widget := MakeBox(?Width := 400.0, ?Height := 600.0, ?BorderRadius := 4.0)
                            overlay_slot:                   
                                Widget := NextQuestText
                            overlay_slot:
                                HorizontalAlignment := horizontal_alignment.Fill
                                VerticalAlignment := vertical_alignment.Fill
                                Padding := margin{Left := 20.0, Top := 0.0, Right := 20.0, Bottom := 20.0}
                                Widget := CreateInfoUI()
                stack_box_slot:
                    HorizontalAlignment := horizontal_alignment.Fill
                    VerticalAlignment := vertical_alignment.Fill
                    Padding := margin{Left := 0.0, Top := 20.0, Right := 0.0, Bottom := 0.0}
                    Widget := ClaimRewardsButton

    CreateInfoUI():widget =
        set InfoStackBox = stack_box:
            Orientation := orientation.Vertical
            Slots := array:
                stack_box_slot:
                    VerticalAlignment := vertical_alignment.Fill
                    Padding := margin{Left := 0.0, Top := 4.0, Right := 0.0, Bottom := 0.0}
                    Widget := overlay:
                        Slots := array:
                            overlay_slot:
                                HorizontalAlignment := horizontal_alignment.Fill
                                VerticalAlignment := vertical_alignment.Fill
                                Padding := margin{Left := 0.0, Top := 0.0, Right := 0.0, Bottom := 0.0}
                                Widget := color_block:
                                    DefaultColor := NamedColors.White
                                    DefaultOpacity := 0.1
                            overlay_slot:
                                HorizontalAlignment := horizontal_alignment.Fill
                                VerticalAlignment := vertical_alignment.Fill
                                Padding := margin{Left := 40.0, Top := 10.0, Right := 40.0, Bottom := 10.0}
                                Widget := stack_box:
                                    Orientation := orientation.Horizontal
                                    Slots := array:
                                        stack_box_slot:
                                            HorizontalAlignment := horizontal_alignment.Left
                                            Widget := texture_block:
                                                DefaultImage := currency_id.Gold.ToIcon()
                                                DefaultDesiredSize := vector2{X := 40.0, Y := 40.0}
                                        stack_box_slot:
                                            HorizontalAlignment := horizontal_alignment.Fill
                                            Padding := margin{Left := 10.0, Top := 0.0, Right := 0.0, Bottom := 0.0}
                                            Widget := RewardText
                stack_box_slot:
                    Padding := margin{Left := 0.0, Top := 40.0, Right := 0.0, Bottom := 40.0}
                    Widget := Icon
                stack_box_slot:
                    Widget := TierText
                stack_box_slot:
                    HorizontalAlignment := horizontal_alignment.Fill
                    VerticalAlignment := vertical_alignment.Fill
                    Padding := margin{Left := 20.0, Top := 0.0, Right := 20.0, Bottom := 80.0}
                    Widget := DescriptionText
                stack_box_slot:
                    HorizontalAlignment := horizontal_alignment.Fill
                    VerticalAlignment := vertical_alignment.Bottom
                    Distribution := option{1.0}
                    Padding := margin{Left := 0.0, Top := 0.0, Right := 0.0, Bottom := 0.0}
                    Widget := overlay:
                        Slots := array:
                            overlay_slot:
                                HorizontalAlignment := horizontal_alignment.Fill
                                VerticalAlignment := vertical_alignment.Fill
                                Widget := MakeBox()
                            overlay_slot:
                                Padding := margin{Left := 40.0, Top := 10.0, Right := 40.0, Bottom := 10.0}
                                Widget := ProgressText        

    <# ----------------------------------------------------------------------------------------------------
        Get Functions
    ---------------------------------------------------------------------------------------------------- #>
    IsRerollAvailable()<transacts>:logic =
        if:
            ActiveQuest := MaybeActiveQuest?
            Info := GetDailyQuestsInfo[ActiveQuest.Id.ToDailyQuestId()]
            not Info.IsTutorial?
            Previous := MakeDateAndTime(Player.GetData[].DailyQuests.LastRerollTime)
            Current := GetDailyQuestsManager[].RealTimeDevice.GetDateAndTime()
            IsDateRolledOver[Previous, Current]
        then:
            true
        else:
            false

    <# ----------------------------------------------------------------------------------------------------
        Update Functions
    ---------------------------------------------------------------------------------------------------- #>
    SetActive(ActiveQuest:daily_quest):void =
        set MaybeActiveQuest = option{ActiveQuest}
        Id := ActiveQuest.Id.ToDailyQuestId()
        if (QuestsInfo := GetDailyQuestsInfo[Id]):
            CurrentProgress := Floor[ActiveQuest.Progress] or 0
            TargetProgress := Floor[QuestsInfo.ProgressTargets[ActiveQuest.Tier]] or 0
            RewardAmount := (QuestsInfo.RewardAmounts[ActiveQuest.Tier] or 0.0) * GetDailyQuestsRankMultiplier(Player)

            Tier := (ActiveQuest.Tier + 1) * 2 - 1
            TierText.SetColor(Tier.ToTierColor() * 2)
            TierText.SetText(MakeMessage(Tier.ToTierString()))

            RewardText.SetText(MakeMessage("{FormatCurrency(RewardAmount)}"))
            
            if (WeaponId := ActiveQuest.MaybeWeaponId?.ToWeaponId(), WeaponIcon := GetWeaponInfo[WeaponId].Icon):
                Icon.SetImage(WeaponIcon)
                DescriptionText.SetText(MakeMessage("{QuestsInfo.Description.Replace("TargetProgress", "{TargetProgress}").Replace("Weapon", "{WeaponId}").WrapText(320.0)}"))
            else:
                Icon.SetImage(QuestsInfo.Icon)
                DescriptionText.SetText(MakeMessage("{QuestsInfo.Description.Replace("TargetProgress", "{TargetProgress}").WrapText(320.0)}"))
            
            if (CurrentProgress >= TargetProgress):
                ClaimRewardsButton.SetText(MakeMessage("Claim Rewards"))
                ProgressText.SetText(MakeMessage("Complete!"))
                set MaybeCurrentRewardAmount = option{RewardAmount}
            else:
                ClaimRewardsButton.SetText(MakeMessage("{IsRerollAvailable()? and "Reroll (1)" or ""}"))
                ProgressText.SetText(MakeMessage("{CurrentProgress}/{TargetProgress}"))
                set MaybeCurrentRewardAmount = false

            InfoStackBox.SetVisibility(widget_visibility.Visible)
            NextQuestText.SetVisibility(widget_visibility.Hidden)

    SetPending():void =
        set MaybeActiveQuest = false
        if (RealTimeDevice := GetDailyQuestsManager[].RealTimeDevice):
            CurrentDateTime := RealTimeDevice.GetDateAndTime()
            TimeInterval := time_interval{Days := 1}
            NextDayDateTime := CurrentDateTime.Add(TimeInterval)
            NextQuestAvailableDateTime := MakeDateAndTime(NextDayDateTime.Years, NextDayDateTime.Months, NextDayDateTime.Days, 0, 0, 0.0)
            TimeIntervalRemaining := NextQuestAvailableDateTime.Subtract(CurrentDateTime)

            if (IsDailyQuestTutorialComplete(Player)?):
                NextQuestText.SetText(MakeMessage("New quests in {TimeIntervalRemaining.ToTruncatedString()} hours.".WrapText(320.0))) 
            else:
                NextQuestText.SetText(MakeMessage("New quests after completing the tutorial quests.".WrapText(320.0))) 

            InfoStackBox.SetVisibility(widget_visibility.Hidden)
            NextQuestText.SetVisibility(widget_visibility.Visible)
            ClaimRewardsButton.SetText(MakeMessage(""))

    SetInactive():void =
        set MaybeActiveQuest = false
        InfoStackBox.SetVisibility(widget_visibility.Hidden)
        NextQuestText.SetVisibility(widget_visibility.Hidden)
        ClaimRewardsButton.SetText(MakeMessage(""))

    Reroll():void =
        if:
            DailyQuestManager := GetDailyQuestsManager[]
            NewDailyQuest := DailyQuestManager.GetNewDailyQuest[Player]

            CurrentPlayerData := Player.GetData[]
            CurrentDateTime := GetDailyQuestsManager[].RealTimeDevice.GetDateAndTime()

            var UpdatedActiveQuests:[]daily_quest = CurrentPlayerData.DailyQuests.ActiveQuests
            set UpdatedActiveQuests[Index] = NewDailyQuest
            
            set PlayerDataMap[Player] = player_data:
                MakePlayerData<constructor>(CurrentPlayerData)
                DailyQuests := player_daily_quests:
                    MakePlayerDailyQuests<constructor>(CurrentPlayerData.DailyQuests)
                    LastRerollTime := CurrentDateTime.GetTotalSeconds()
                    ActiveQuests := UpdatedActiveQuests

        then:
            Player.PlayAudio(audio_id.Reroll)
            RerollEvent.Signal()

    <# ----------------------------------------------------------------------------------------------------
        Event Handlers
    ---------------------------------------------------------------------------------------------------- #>
    OnClickClaimRewards(Message:widget_message):void =
        if:
            CurrentRewardAmount := MaybeCurrentRewardAmount?
            CurrentPlayerData := Player.GetData[]
            PlayerDailyQuests := CurrentPlayerData.DailyQuests            
            UpdatedActiveQuests := PlayerDailyQuests.ActiveQuests.RemoveElement[Index]
            
            set PlayerDataMap[Player] = player_data:
                MakePlayerData<constructor>(CurrentPlayerData)
                DailyQuests := player_daily_quests:
                    MakePlayerDailyQuests<constructor>(PlayerDailyQuests)
                    ActiveQuests := UpdatedActiveQuests
        then:
            Player.AddCurrency(currency_id.Gold, CurrentRewardAmount)
            ClaimRewardsButton.SetText(MakeMessage("Claimed!"))
            Player.PlayAudio(audio_id.PowerUp)
            set MaybeCurrentRewardAmount = false
            ClaimEvent.Signal()

        else if (IsRerollAvailable()?):
            Reroll()

        else:
            Player.PlayAudio(audio_id.Error)