using { /Verse.org/Assets }
using { /Verse.org/Simulation }

<# ----------------------------------------------------------------------------------------------------
    Constants
---------------------------------------------------------------------------------------------------- #>
MaxDailyQuests:int = 3
DailyQuestsRankRewardMultiplier:float = 0.5
DailyQuestsTierWeights:[]int = array{50, 30, 15}

<# ----------------------------------------------------------------------------------------------------
    Data Definitions
---------------------------------------------------------------------------------------------------- #>
daily_quest_id := enum:
    Tutorial01
    Tutorial02
    Tutorial03
    GatherTree
    Objectives
    ItemShop
    EliminateEnemies
    Victory
    WeaponEvolution
    WeaponVictory

(Id:daily_quest_id).ToInt()<transacts>:int =
    case (Id):
        daily_quest_id.Tutorial01 => 1
        daily_quest_id.Tutorial02 => 2
        daily_quest_id.Tutorial03 => 3
        daily_quest_id.GatherTree => 4
        daily_quest_id.Objectives => 5
        daily_quest_id.ItemShop => 6
        daily_quest_id.EliminateEnemies => 7
        daily_quest_id.Victory => 8
        daily_quest_id.WeaponEvolution => 9
        daily_quest_id.WeaponVictory => 10

(Id:int).ToDailyQuestId()<transacts>:daily_quest_id =
    case (Id):
        1 => daily_quest_id.Tutorial01
        2 => daily_quest_id.Tutorial02
        3 => daily_quest_id.Tutorial03
        4 => daily_quest_id.GatherTree
        5 => daily_quest_id.Objectives
        6 => daily_quest_id.ItemShop
        7 => daily_quest_id.EliminateEnemies
        8 => daily_quest_id.Victory
        9 => daily_quest_id.WeaponEvolution
        10 => daily_quest_id.WeaponVictory
        _ => daily_quest_id.GatherTree

daily_quest_info := struct<computes>:
    Id:daily_quest_id
    Name:string
    Description:string
    Icon:texture
    RewardAmounts:[]float = array{}
    ProgressTargets:[]float = array{}
    IsTutorial:logic = false

<# ----------------------------------------------------------------------------------------------------
    Daily Quest Info
---------------------------------------------------------------------------------------------------- #>
DailyQuestsInfo:[daily_quest_id]daily_quest_info = map:
    daily_quest_id.Tutorial01 => daily_quest_info:
        Id := daily_quest_id.Tutorial01
        Name := "Tutorial 01 - Leveling Up"
        Description := "Eliminate slimes & pickup XP orbs to level up."
        Icon := UserInterface.Icons.ExperienceMultiplier_128
        RewardAmounts := array{1000.0}
        ProgressTargets := array{5.0}
        IsTutorial := true
    daily_quest_id.Tutorial02 => daily_quest_info:
        Id := daily_quest_id.Tutorial02
        Name := "Tutorial 02 - Gathering Wood"
        Description := "Gather wood from trees to use in the Item Shop."
        Icon := UserInterface.Icons.Tree_128
        RewardAmounts := array{1000.0}
        ProgressTargets := array{5.0}
        IsTutorial := true
    daily_quest_id.Tutorial03 => daily_quest_info:
        Id := daily_quest_id.Tutorial03
        Name := "Tutorial 03 - First Run"
        Description := "Complete your first game."
        Icon := UserInterface.Icons.Victory_128
        RewardAmounts := array{1000.0}
        ProgressTargets := array{1.0}
        IsTutorial := true
    daily_quest_id.GatherTree => daily_quest_info:
        Id := daily_quest_id.GatherTree
        Name := "The Lumberjack"
        Description := "Gather [TargetProgress] trees."
        Icon := UserInterface.Icons.Tree_128
        RewardAmounts := array{1000.0, 1500.0, 2000.0}
        ProgressTargets := array{50.0, 100.0, 150.0}
   
    daily_quest_id.Objectives => daily_quest_info:
        Id := daily_quest_id.Objectives
        Name := "Mission Critical"
        Description := "Complete [TargetProgress] missions."
        Icon := UserInterface.Icons.Objective_128
        RewardAmounts := array{1000.0, 1500.0, 2000.0}
        ProgressTargets := array{3.0, 6.0, 9.0}
   
    daily_quest_id.ItemShop => daily_quest_info:
        Id := daily_quest_id.ItemShop
        Name := "Shopping Spree"
        Description := "Purchase [TargetProgress] items from the item shop."
        Icon := UserInterface.Icons.ItemShop_128
        RewardAmounts := array{1000.0, 1500.0, 2000.0}
        ProgressTargets := array{10.0, 15.0, 20.0}
   
    daily_quest_id.EliminateEnemies => daily_quest_info:
        Id := daily_quest_id.EliminateEnemies
        Name := "Monster Hunter"
        Description := "Eliminate [TargetProgress] enemies."
        Icon := UserInterface.Icons.Elimination_128
        RewardAmounts := array{1000.0, 1500.0, 2000.0}
        ProgressTargets := array{5000.0, 7500.0, 10000.0}
   
    daily_quest_id.Victory => daily_quest_info:
        Id := daily_quest_id.Victory
        Name := "The Champion"
        Description := "Achieve [TargetProgress] victories."
        Icon := UserInterface.Icons.Victory_128
        RewardAmounts := array{1000.0, 1500.0, 2000.0}
        ProgressTargets := array{1.0, 2.0, 3.0}
    daily_quest_id.WeaponEvolution => daily_quest_info:
        Id := daily_quest_id.WeaponEvolution
        Name := "Weapon Mastery"
        Description := "Evolve a weapon [TargetProgress] times."
        Icon := UserInterface.Icons.WeaponEvolution_128
        RewardAmounts := array{1000.0, 1500.0, 2000.0}
        ProgressTargets := array{6.0, 12.0, 18.0}
    daily_quest_id.WeaponVictory => daily_quest_info:
        Id := daily_quest_id.WeaponVictory
        Name := "Weapon Expert"
        Description := "Achieve [TargetProgress] victories starting with [Weapon]."
        Icon := UserInterface.Icons.WeaponEvolution_128
        RewardAmounts := array{1500.0, 2000.0}
        ProgressTargets := array{1.0, 2.0}

<# ----------------------------------------------------------------------------------------------------
    Get Functions
---------------------------------------------------------------------------------------------------- #>
GetDailyQuestsInfo()<transacts>:[daily_quest_id]daily_quest_info =
    DailyQuestsInfo

GetDailyQuestsInfo(Id:daily_quest_id)<decides><transacts>:daily_quest_info =
    DailyQuestsInfo[Id]

GetDailyQuestsRankMultiplier(Player:player)<transacts>:float =
    1.0 + CalculateArenaGoldMultiplier(Player.GetArenaProgress())

GetStartingDailyQuests()<transacts>:[]daily_quest =
    array:
        daily_quest{Id := daily_quest_id.Tutorial01.ToInt()}
        daily_quest{Id := daily_quest_id.Tutorial02.ToInt()}
        daily_quest{Id := daily_quest_id.Tutorial03.ToInt()}