using { /Fortnite.com/Devices }
using { /Verse.org/Random }
using { /Verse.org/Simulation }
using { RealTime }

<# ----------------------------------------------------------------------------------------------------
    Global Functions
---------------------------------------------------------------------------------------------------- #>
var GlobalDailyQuestsManager:weak_map(session, daily_quests_manager) = map{}

GetDailyQuestsManager()<decides><transacts>:daily_quests_manager =
    GlobalDailyQuestsManager[GetSession()]

IsDailyQuestTutorialComplete(Player:player)<transacts>:logic =
    var IsComplete:logic = true
    for (ActiveQuest : Player.GetData[].DailyQuests.ActiveQuests, IsComplete?, Info := GetDailyQuestsInfo[ActiveQuest.Id.ToDailyQuestId()], Info.IsTutorial?):
        set IsComplete = false
    IsComplete

daily_quest_tracker := class<concrete>:
    var AssignedPlayer:?player = false
    @editable
    Device:tracker_device = tracker_device{}
    @editable
    CinematicSequence:cinematic_sequence_device = cinematic_sequence_device{}

daily_quest_hud_message := class<concrete>:
    @editable
    Id:daily_quest_id = daily_quest_id.Tutorial01
    @editable
    Device:hud_message_device = hud_message_device{}

<# ----------------------------------------------------------------------------------------------------
    daily_quests_manager
---------------------------------------------------------------------------------------------------- #>
daily_quests_manager := class(creative_device):
    @editable
    RealTimeDevice:real_time = real_time{}

    @editable
    var DailyQuestTrackers:[]daily_quest_tracker = array{}

    @editable
    Beacon:beacon_device = beacon_device{}

    @editable
    BeaconNew:beacon_device = beacon_device{}

    @editable
    DailyQuestHUDMessages:[]daily_quest_hud_message = array{}
    
    @editable
    ClearDailyQuestHUDMessageChannel:channel_device = channel_device{}

    var PlayerEventSubscriptions:[player][]events(tuple()) = map{}
    var PlayerTracker:[player]daily_quest_tracker = map{}
    var PlayerIsShowingQuestComplete:[player]logic = map{}
    var PlayerHUDMessage:[player]hud_message_device = map{}
    var MappedDailyQuestHUDMessage:[daily_quest_id]hud_message_device = map{}

    <# ----------------------------------------------------------------------------------------------------
        Initialization
    ---------------------------------------------------------------------------------------------------- #>
    OnBegin<override>()<suspends>:void =
        if (set GlobalDailyQuestsManager[GetSession()] = Self) {}

        for (DailyQuestHUDMessage : DailyQuestHUDMessages):
            if (set MappedDailyQuestHUDMessage[DailyQuestHUDMessage.Id] = DailyQuestHUDMessage.Device) {}

        for (Player : GetPlayspace().GetPlayers()):
            spawn{Initialize(Player)}

        GetPlayspace().PlayerAddedEvent().Subscribe(OnPlayerAdded)
        GetPlayspace().PlayerRemovedEvent().Subscribe(OnPlayerRemoved)

        WaitForGameManager()
        if (GameManager := GetGameManager[]):
            GameManager.GameStartedEvent.Subscribe(OnGameStarted)
            GameManager.GameEndedEvent.Subscribe(OnGameEnded)

    Initialize(Player:player)<suspends>:void =
        WaitForPlayerDataManager(Player)

        if:
            CurrentPlayerData := Player.GetData[]
            CurrentPlayerData.DailyQuests.LastQuestAssignedTime = 0.0

            set PlayerDataMap[Player] = player_data:
                MakePlayerData<constructor>(CurrentPlayerData)
                DailyQuests := player_daily_quests:
                    LastQuestAssignedTime := RealTimeDevice.GetDateAndTime().Subtract(MakeTimeInterval(1, 0, 0, 0.0)).GetTotalSeconds()
                    ActiveQuests := GetStartingDailyQuests()
        else:
            AssignNewQuests(Player)

        UpdateBeacon(Player)
        AssignTracker(Player)

    <# ----------------------------------------------------------------------------------------------------
        Get Functions
    ---------------------------------------------------------------------------------------------------- #>
    GetNewDailyQuest(Player:player)<decides><transacts>:daily_quest =
        GetNewDailyQuests[Player, 1][0]

    GetHUDMessage(Id:daily_quest_id)<decides><transacts>:hud_message_device =
        MappedDailyQuestHUDMessage[Id]

    GetNewDailyQuests(Player:player, Count:int)<decides><transacts>:[]daily_quest =
        ActiveQuestIds := for (ActiveQuest : Player.GetData[].DailyQuests.ActiveQuests) {ActiveQuest.Id.ToDailyQuestId()}
        Infos := for (Info : GetDailyQuestsInfo(), not Info.IsTutorial?, not ActiveQuestIds.Find[Info.Id]) {Info}
        ShuffledInfos := Shuffle(Infos)

        for (Index := 0..(Count - 1), Info := ShuffledInfos[Index]):
            MaxTier := Info.RewardAmounts.Length
            TierWeights := DailyQuestsTierWeights.Slice[0, MaxTier]
            daily_quest:
                Id := Info.Id.ToInt()
                Tier := Min(MaxTier - 1, GetWeightedRandom(TierWeights))
                Progress := 0.0
                MaybeWeaponId := (Info.Id = daily_quest_id.WeaponVictory) and option{GetRandomWeaponId[].ToInt()} or false

    GetDailyQuestEvent(Player:player, Id:daily_quest_id)<decides><transacts>:events(tuple()) =
        case (Id):
            daily_quest_id.Tutorial01 => GetPlayerManager[].LeveledUpEvent
            daily_quest_id.Tutorial02 => GetArenaManager[].PlayerGatheredEvent[Player]
            daily_quest_id.Tutorial03 => GetGameManager[].GameEndedEvent

            daily_quest_id.GatherTree => GetArenaManager[].PlayerGatheredEvent[Player]
            daily_quest_id.Objectives => GetObjectiveManager[].ObjectiveCompleteEvent
            daily_quest_id.ItemShop => GetItemShopManager[].PlayerItemShopPurchasedEvents[Player]
            daily_quest_id.EliminateEnemies => GetGameManager[].PlayerEliminationEvents[Player]
            daily_quest_id.Victory => GetGameManager[].VictoryEvent
            daily_quest_id.WeaponEvolution => Player.GetStats[].WeaponEvolvedEvent
            daily_quest_id.WeaponVictory => GetGameManager[].VictoryEvent

    IsComplete(Player:player, DailyQuest:daily_quest)<transacts>:logic =
        if (Info := GetDailyQuestsInfo[DailyQuest.Id.ToDailyQuestId()], Target := Info.ProgressTargets[DailyQuest.Tier]):
            logic{DailyQuest.Progress >= Target}
        else:
            false

    <# ----------------------------------------------------------------------------------------------------
        Update Functions
    ---------------------------------------------------------------------------------------------------- #>
    AssignTracker(Player:player):void =
        for (Index -> DailyQuestTracker : DailyQuestTrackers, not DailyQuestTracker.AssignedPlayer?):
            UpdateDailyQuestTracker := daily_quest_tracker:
                AssignedPlayer := option{Player}
                Device := DailyQuestTracker.Device
                CinematicSequence := DailyQuestTracker.CinematicSequence

            DailyQuestTracker.CinematicSequence.Play(Player)
            DailyQuestTracker.CinematicSequence.Pause(Player)

            if:
                set DailyQuestTrackers[Index] = UpdateDailyQuestTracker
                set PlayerTracker[Player] = DailyQuestTracker

            return
            
    UnassignTracker(Player:player):void =
        for (Index -> DailyQuestTracker : DailyQuestTrackers, Player = DailyQuestTracker.AssignedPlayer?):
            UpdateDailyQuestTracker := daily_quest_tracker:
                AssignedPlayer := false
                Device := DailyQuestTracker.Device
                CinematicSequence := DailyQuestTracker.CinematicSequence

            DailyQuestTracker.CinematicSequence.Stop(Player)

            if (set DailyQuestTrackers[Index] = UpdateDailyQuestTracker) {}
            set PlayerTracker = RemoveKeyFromMap(PlayerTracker, Player)
            return

    AssignNewQuests(Player:player):void =
        if (CurrentPlayerData := Player.GetData[]):
            UpdatedLastQuestAssignedTime := RealTimeDevice.GetDateAndTime().GetTotalSeconds()

            if:    
                PlayerDailyQuests := CurrentPlayerData.DailyQuests
                ActiveQuestsDelta := MaxDailyQuests - PlayerDailyQuests.ActiveQuests.Length
                ActiveQuestsDelta > 0

                IsDailyQuestTutorialComplete(Player)?
                
                Previous := MakeDateAndTime(PlayerDailyQuests.LastQuestAssignedTime)
                Current := RealTimeDevice.GetDateAndTime()
                IsDateRolledOver[Previous, Current]

                var UpdatedActiveQuests:[]daily_quest = PlayerDailyQuests.ActiveQuests
                NewDailyQuests := GetNewDailyQuests[Player, ActiveQuestsDelta]
                set UpdatedActiveQuests = Concatenate(UpdatedActiveQuests, NewDailyQuests)

                set PlayerDataMap[Player] = player_data:
                    MakePlayerData<constructor>(CurrentPlayerData)
                    DailyQuests := player_daily_quests:
                        MakePlayerDailyQuests<constructor>(CurrentPlayerData.DailyQuests)
                        LastQuestAssignedTime := UpdatedLastQuestAssignedTime
                        ActiveQuests := UpdatedActiveQuests

    UnsubscribeFromActiveQuests(Player:player):void =
        for (PlayerEventSubscription : PlayerEventSubscriptions[Player]):
            PlayerEventSubscription.Signal()

        if (set PlayerEventSubscriptions[Player] = array{}) {}

    SubscribeToActiveQuests(Player:player):void =
        UnsubscribeFromActiveQuests(Player)
        for (Index -> ActiveQuest : Player.GetData[].DailyQuests.ActiveQuests):
            CancelEvent := events(tuple()){}
            if (set PlayerEventSubscriptions[Player] += array{CancelEvent}) {}

            if:
                not IsComplete(Player, ActiveQuest)?
                Id := ActiveQuest.Id.ToDailyQuestId()
                DailyQuestEvent := GetDailyQuestEvent[Player, Id]
            then:
                spawn{AwaitDailyQuestEvent(DailyQuestEvent, CancelEvent, Player, Index, option{ActiveQuest.MaybeWeaponId?.ToWeaponId()})}

    TrackQuest(Player:player)<suspends>:void =
        var IsTracking:logic = false
        for (Index -> ActiveQuest : Player.GetData[].DailyQuests.ActiveQuests, not IsTracking?):
            if (not IsComplete(Player, ActiveQuest)?):
                UpdateTracker(Player, Index)
                set IsTracking = true

        if (not IsTracking?, HUDMessage := PlayerHUDMessage[Player]):
            HUDMessage.Hide(Player)
                
        <#
        if (not IsTracking?, Tracker := PlayerTracker[Player]):
            Tracker.Remove(Player)
        #>

    UpdateTracker(Player:player, Index:int)<suspends>:void =
        if (not IsGameInProgress()?) then return
        if (HUDMessage := PlayerHUDMessage[Player]):
            HUDMessage.Hide(Player)

        if:
            ActiveQuest := Player.GetData[].DailyQuests.ActiveQuests[Index]
            Info := GetDailyQuestsInfo[ActiveQuest.Id.ToDailyQuestId()]
            TargetProgress := Floor[Info.ProgressTargets[ActiveQuest.Tier]]
            Progress := Min(TargetProgress, Floor[ActiveQuest.Progress] or 0)
            Tracker := PlayerTracker[Player]
            HUDMessage := GetHUDMessage[Info.Id]
            set PlayerHUDMessage[Player] = HUDMessage
        then:
            <#
            DescriptionText := if (WeaponId := ActiveQuest.MaybeWeaponId?.ToWeaponId()):
                Info.Description.Replace("TargetProgress", "{TargetProgress}").Replace("Weapon", "{WeaponId}")
            else:
                Info.Description.Replace("TargetProgress", "{TargetProgress}")

            Tracker.Assign(Player)
            Tracker.SetTitleText(MakeMessage("{Info.IsTutorial? and "" or "Daily Quest - "}{Info.Name}"))
            Tracker.SetDescriptionText(MakeMessage("{DescriptionText}"))
            Tracker.SetTarget(TargetProgress)
            Sleep(0.0)
            Tracker.SetValue(Player, Progress)
            #>

            Tracker.CinematicSequence.SetPlaybackTime((Progress * 1.0)/(TargetProgress * 1.0))

            if (Progress >= TargetProgress, CancelDailyQuestEvent := PlayerEventSubscriptions[Player][Index]):
                if (set PlayerIsShowingQuestComplete[Player] = true) {}
                CancelDailyQuestEvent.Signal()
                Player.PlayAudio(audio_id.ObjectiveSuccess)
                Player.ShowMessage(message_id.DailyQuestComplete)
                HUDMessage.Show(Player, MakeMessage("Complete!"), ?DisplayTime := 0.0)
                Sleep(3.0)
                if (set PlayerIsShowingQuestComplete[Player] = false) {}
                HUDMessage.Hide(Player)
                TrackQuest(Player)
            else if (not PlayerIsShowingQuestComplete[Player]?):
                HUDMessage.Show(Player, MakeMessage("{Progress}/{TargetProgress}"), ?DisplayTime := 0.0)

    AddProgress(Player:player, Index:int, Amount:float):void =
        if:
            CurrentPlayerData := Player.GetData[]

            var UpdatedActiveQuests:[]daily_quest = CurrentPlayerData.DailyQuests.ActiveQuests
            UpdatedProgress := UpdatedActiveQuests[Index].Progress + Amount
            Progress := Floor[UpdatedProgress]
            UpdatedQuest := daily_quest:
                MakeDailyQuest<constructor>(UpdatedActiveQuests[Index])
                Progress := UpdatedProgress

            set UpdatedActiveQuests[Index] = UpdatedQuest
            set PlayerDataMap[Player] = player_data:
                MakePlayerData<constructor>(CurrentPlayerData)
                DailyQuests := player_daily_quests:
                    MakePlayerDailyQuests<constructor>(CurrentPlayerData.DailyQuests)
                    ActiveQuests := UpdatedActiveQuests
        then:
            spawn{UpdateTracker(Player, Index)}

    UpdateBeacon(Player:player):void =
        if (Player.GetData[].DailyQuests.ActiveQuests.Length > 0):
            Beacon.RemoveFromShowList(Player)
            BeaconNew.AddToShowList(Player)
        else:
            Beacon.AddToShowList(Player)
            BeaconNew.RemoveFromShowList(Player)

    HandleGameEnded()<suspends>:void =
        Sleep(TickTime * 2)
        <#
        for (Player : GetAllPlayers(), Tracker := PlayerTracker[Player]):
            Tracker.Device.Remove(Player)
        #>

        for (Player : GetAllPlayers(), HUDMessage := PlayerHUDMessage[Player]):
            HUDMessage.Hide(Player)
            UnsubscribeFromActiveQuests(Player)
            
    <# ----------------------------------------------------------------------------------------------------
        Event Handlers
    ---------------------------------------------------------------------------------------------------- #>
    OnPlayerAdded(Player:player):void =
        spawn{Initialize(Player)}

    OnPlayerRemoved(Player:player):void =
        UnassignTracker(Player)
        set PlayerEventSubscriptions = RemoveKeyFromMap(PlayerEventSubscriptions, Player)
        set PlayerIsShowingQuestComplete = RemoveKeyFromMap(PlayerIsShowingQuestComplete, Player)
        set PlayerHUDMessage = RemoveKeyFromMap(PlayerHUDMessage, Player)

    AwaitDailyQuestEvent(Event:events(tuple()), CancelEvent:events(tuple()), Player:player, Index:int, MaybeWeaponId:?weapon_id)<suspends>:void =
        race:
            CancelEvent.Await()
            loop:
                Event.Await()
                if (WeaponId := MaybeWeaponId?):
                    if (WeaponId = Player.GetEquippedWeapon()?):
                        AddProgress(Player, Index, 1.0)
                else:
                    AddProgress(Player, Index, 1.0)

    OnGameStarted():void =
        for (Player : GetActivePlayers()):
            spawn{TrackQuest(Player)}
            SubscribeToActiveQuests(Player)

    OnGameEnded():void =
        for (Player : GetAllPlayers()):
            UpdateBeacon(Player)

        spawn{HandleGameEnded()}