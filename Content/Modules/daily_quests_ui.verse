using { /Fortnite.com/UI }
using { /Verse.org/Colors }
using { /Verse.org/Simulation }
using { /UnrealEngine.com/Temporary/SpatialMath }
using { /UnrealEngine.com/Temporary/UI }
using { RealTime }

daily_quests_ui := class(panel_ui):
    Title<override>:string = "Daily Quests"
    BackgroundOpacity<override>:type {_X:float where 0.000000 <= _X, _X <= 1.000000} = 0.7
    
    RankIcon:texture_block = texture_block:
        DefaultImage := UserInterface.Icons.RankIcon_01_64
        DefaultDesiredSize := vector2{X := 40.0, Y := 40.0}

    RankMultiplier:text_block = text_block:
        DefaultTextColor := ThemeColors.Success
        DefaultShadowColor := ThemeColors.Success * 0.125
        DefaultShadowOffset := option{vector2{X := 2.0, Y := 2.0}}
        DefaultShadowOpacity := 1.0
    
    var DailyQuestsEntryUIs:[]daily_quests_entry_ui = array{}

    <# ----------------------------------------------------------------------------------------------------
        Initialization
    ---------------------------------------------------------------------------------------------------- #>
    CreateContent<override>():widget =
        ClosedEvent.Subscribe(OnClosed)

        stack_box:
            Orientation := orientation.Vertical
            Slots := array:
                stack_box_slot:
                    VerticalAlignment := vertical_alignment.Fill
                    Padding := margin{Left := 0.0, Top := 0.0, Right := 0.0, Bottom := 20.0}
                    Widget := overlay:
                        Slots := array:
                            overlay_slot:
                                HorizontalAlignment := horizontal_alignment.Fill
                                VerticalAlignment := vertical_alignment.Fill
                                Widget := MakeBox()
                            overlay_slot:
                                Padding := margin{Left := 40.0, Top := 10.0, Right := 40.0, Bottom := 10.0}
                                Widget := stack_box:
                                    Orientation := orientation.Horizontal
                                    Slots := array:
                                        stack_box_slot:
                                            HorizontalAlignment := horizontal_alignment.Left
                                            Padding := margin{Left := 0.0, Top := 0.0, Right := 10.0, Bottom := 0.0}
                                            Widget := RankIcon
                                        stack_box_slot:
                                            HorizontalAlignment := horizontal_alignment.Fill
                                            VerticalAlignment := vertical_alignment.Fill
                                            Widget := RankMultiplier
                stack_box_slot:
                    HorizontalAlignment := horizontal_alignment.Fill
                    VerticalAlignment := vertical_alignment.Fill
                    Widget := stack_box:
                        Orientation := orientation.Horizontal
                        Slots := for (Index := 1..MaxDailyQuests):
                            DailyQuestsEntryUI := daily_quests_entry_ui{Player := Player, Index := Index - 1}
                            DailyQuestsEntryUI.Initialize()
                            DailyQuestsEntryUI.RerollEvent.Subscribe(OnReroll)
                            DailyQuestsEntryUI.ClaimEvent.Subscribe(OnClaim)
                            
                            set DailyQuestsEntryUIs += array{DailyQuestsEntryUI}
                            
                            stack_box_slot:
                                HorizontalAlignment := horizontal_alignment.Fill
                                VerticalAlignment := vertical_alignment.Fill
                                Padding := margin{Left := 20.0, Top := 0.0, Right := 0.0, Bottom := 0.0}
                                Widget := DailyQuestsEntryUI.GetUI()

    <# ----------------------------------------------------------------------------------------------------
        Update Functions
    ---------------------------------------------------------------------------------------------------- #>
    Update():void =
        RankIcon.SetImage(Player.GetRank().ToRankIcon())
        RankMultiplier.SetText(MakeMessage("+{FormatAsPercent(GetDailyQuestsRankMultiplier(Player))}"))

        if (PlayerActiveQuests := Player.GetData[].DailyQuests.ActiveQuests):
            var HasSetPending:logic = false
            for (Index -> DailyQuestsEntryUI : DailyQuestsEntryUIs):
                if (PlayerActiveQuest := PlayerActiveQuests[Index]):
                    DailyQuestsEntryUI.SetActive(PlayerActiveQuest)
                else if (not HasSetPending?):
                    DailyQuestsEntryUI.SetPending()
                    set HasSetPending = true
                else:
                    DailyQuestsEntryUI.SetInactive()

    <# ----------------------------------------------------------------------------------------------------
        Event Handlers
    ---------------------------------------------------------------------------------------------------- #>
    OnShow<override>():void =
        if (DailyQuestsManager := GetDailyQuestsManager[]):
            DailyQuestsManager.AssignNewQuests(Player)

        Update()

    OnClosed(InPlayer:player):void =
        if (DailyQuestsManager := GetDailyQuestsManager[]):
            DailyQuestsManager.UpdateBeacon(Player)

    OnReroll():void =
        Update()

    OnClaim():void =
        if (DailyQuestsManager := GetDailyQuestsManager[]):
            DailyQuestsManager.AssignNewQuests(Player)
            DailyQuestsManager.UpdateBeacon(Player)
            Update()
        