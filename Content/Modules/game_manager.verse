using { /Fortnite.com/AI }
using { /Fortnite.com/Characters }
using { /Fortnite.com/Devices }
using { /Verse.org/Random }
using { /Verse.org/Simulation }
using { /UnrealEngine.com/Temporary }
using { /UnrealEngine.com/Temporary/SpatialMath }
using { /UnrealEngine.com/Temporary/UI }

<# ----------------------------------------------------------------------------------------------------
    Constants
---------------------------------------------------------------------------------------------------- #>
TeleportGroupSpacing:float = 100.0
TeleportToArenaPosition:vector3 = vector3{X := 0.0, Y := 0.0, Z := 1300.0}
GameStartCountdownTime:int = 30
GameResumeCountdownTimeAbilitySelect:int = 15
GameResumeCountdownTimeAbilityUpgrade:int = 5
GameResumeCountdownTimeItemShop:int = 30
VictoryBonusGoldMultplier:float = 0.5
PlayerReviveDuration:float = 2.0
ArenaStageDuration:float = 360.0

<# ----------------------------------------------------------------------------------------------------
    Data Definitions
---------------------------------------------------------------------------------------------------- #>
pause_type := enum:
    LevelUp
    AbilityUpgrade
    ItemShop
    Boon

result_info := struct:
    IsVictory:logic = false
    IsRankUp:logic = false
    Arena:arena = arena{}
    GameSettings:game_settings = game_settings{}
    EndTime:float = 0.0
    TimeSurvived:float = 0.0
    EnemiesEliminated:int = 0
    LevelReached:int = 1
    WoodCollected:float = 0.0
    GoldFromTimeSurvived:float = 0.0
    GoldFromWoodCollected:float = 0.0
    GoldFromObjectivesCompleted:float = 0.0
    GoldFromStagesCompleted:float = 0.0
    GoldFromVictory:float = 0.0
    GoldEarned:float = 0.0
    GoldMultiplier:float = 0.0
    GoldFromMultiplier:float = 0.0
    ObjectivesCompleted:int = 0
    StagesCompleted:int = 0
    ActivePlayers:[]player = array{}
    PlayerActiveAbilities:[player][weapon_id]ability = map{}
    PlayerEnemiesEliminated:[player]int = map{}
    PlayerStats:[player]player_stat = map{}
    TeamExperienceMultiplier:float = 0.0

game_currency_id := enum:
    Wood
    TreasureChest

ToString(Id:game_currency_id)<transacts>:string =
    case (Id):
        game_currency_id.Wood => "Wood"
        game_currency_id.TreasureChest => "TreasureChest"

game_currency := class:
    Id:game_currency_id
    Amount:float = 0.0
    TotalAmount:float = 0.0

MakeGameCurrency<constructor>(UpdatedGameCurrency:game_currency)<transacts> := game_currency:
    Id := UpdatedGameCurrency.Id
    Amount := UpdatedGameCurrency.Amount
    TotalAmount := UpdatedGameCurrency.TotalAmount

game_result_title_id := enum:
    PlayersRevived
    DamageDealt
    WoodCollected
    TimesDowned

game_result_title := class<concrete>:
    @editable
    Id:game_result_title_id = game_result_title_id.PlayersRevived
    @editable
    Billboard:billboard_device = billboard_device{}
    @editable
    PlayerReference:player_reference_device = player_reference_device{}

<# ----------------------------------------------------------------------------------------------------
    Game Settings
---------------------------------------------------------------------------------------------------- #>
game_settings := class<concrete>:
    Arena:arena = arena{}
    Difficulty:int = 1
    GameSpeed:int = 1

MakeGameSettings<constructor>(UpdatedGameSettings:game_settings)<transacts> := game_settings:
    Arena := UpdatedGameSettings.Arena
    Difficulty := UpdatedGameSettings.Difficulty
    GameSpeed := UpdatedGameSettings.GameSpeed

GetGameSettings()<decides><transacts>:game_settings =
    GetGameManager[].GameSettings

GetGameSpeed()<transacts>:float =
    (GetGameSettings[].GameSpeed * 1.0) or 1.0

GetGameDifficulty()<transacts>:int =
    GetGameSettings[].Difficulty or 1

SetGameSettings(GameSettings:game_settings):void =
    if (GameManager := GetGameManager[]):
        set GameManager.GameSettings = GameSettings
        GameManager.BillboardGameSettings.SetText(MakeMessage("{GameSettings.Arena.Name}\nArena Level: {CalculateArenaLevel(GameSettings.Arena)}\nGame Speed: x{Floor[GetGameSpeed()] or 1}"))

<# ----------------------------------------------------------------------------------------------------
    Game Stats
---------------------------------------------------------------------------------------------------- #>
game_stat_id := enum:
    EliminatedMiniboss

<# ----------------------------------------------------------------------------------------------------
    Global Functions
---------------------------------------------------------------------------------------------------- #>
var GlobalGameManager:weak_map(session, game_manager) = map{}

GetGameManager()<decides><transacts>:game_manager =
    GlobalGameManager[GetSession()]

WaitForGameManager()<suspends>:void =
    loop:
        if (GetGameManager[].GetIsInitialized()) {break}
        Sleep(0.3)

PauseGame():void =
    if (GameManager := GetGameManager[]):
        GameManager.Pause()

PauseGame(Type:pause_type, ?Level:int = 0):logic =
    if (GameManager := GetGameManager[]):
        GameManager.Pause(Type, Level)
    else:
        false

PauseGame(Player:player, Type:pause_type, Level:int):logic =
    if (GameManager := GetGameManager[]):
        GameManager.Pause(Player, Type, Level)
    else:
        false

ResumeGame(?IsInstant:logic = false):void =
    if (GameManager := GetGameManager[]):
        GameManager.Resume(?IsInstant := IsInstant)

ResumeGame(Player:player, ?IsShow:logic = true):void =
    if (GameManager := GetGameManager[]):
        GameManager.Resume(Player, ?IsShow := IsShow)

(Player:player).GetNextPause()<decides><transacts>:tuple(pause_type, int) =
    GetGameManager[].PlayerPauseQueue[Player].Elements[1]

(Player:player).GetPauseCount(Type:pause_type)<transacts>:int =
    var Count:int = 0
    for (Element : GetGameManager[].PlayerPauseQueue[Player].Elements):
        if (Element(0) = Type):
            set Count += 1
        else:
            return Count

    return Count

(Player:player).IsPlayerPaused()<transacts>:logic =
    GetGameManager[].IsPlayerPaused(Player) or false
    
IsGamePaused()<transacts>:logic =
    GetGameManager[].IsPaused() or false

GetReadyPlayerCount()<transacts>:int =
    GetGameManager[].GetReadyPlayerCount() or 0

IsAllPlayerReady()<transacts>:logic =
    GetGameManager[].IsAllPlayerReady() or false

IsGameInProgress()<transacts>:logic =
    GetGameManager[].IsGameInProgress() or false

GetActivePlayers()<transacts>:[]player =
    GetGameManager[].GetActivePlayers() or array{}

GetActivePlayersNotAFK()<transacts>:[]player =
    GetGameManager[].GetActivePlayersNotAFK() or array{}

GetAlivePlayers()<transacts>:[]player =
    GetGameManager[].GetAlivePlayers() or array{}

GetVulnerablePlayers()<transacts>:[]player =
    GetGameManager[].GetVulnerablePlayers() or array{}

GetRandomAlivePlayer()<transacts>:?player =
    GetGameManager[].GetRandomAlivePlayer() or false

GetDownButNotOut()<decides><transacts>:down_but_not_out_device =
    GetGameManager[].GetDownButNotOut()

EndGame(IsVictory:logic):void =
    if (GameManager := GetGameManager[]):
        GameManager.Stop(IsVictory)

(Player:player).GetGameCurrency(Id:game_currency_id)<transacts>:float =
    GetGameManager[].GetGameCurrency(Player, Id) or 0.0

(Player:player).CanAfford(Id:game_currency_id, Amount:float)<transacts>:logic =
    logic{Player.GetGameCurrency(Id) >= Amount or IsAlmostEqual[Player.GetGameCurrency(Id), Amount, 0.5]}

AddGameCurrency(Id:game_currency_id, Amount:float):void =
    if (GameManager := GetGameManager[]):
        GameManager.AddGameCurrency(Id, Amount)

(Player:player).AddGameCurrency(Id:game_currency_id, Amount:float):void =
    if (GameManager := GetGameManager[]):
        GameManager.AddGameCurrency(Player, Id, Amount)

(Player:player).RemoveGameCurrency(Id:game_currency_id, Amount:float):void =
    if (GameManager := GetGameManager[]):
        GameManager.RemoveGameCurrency(Player, Id, Amount)

AddEnemiesEliminated(Player:player):void =
    if (GameManager := GetGameManager[]):
        GameManager.AddEnemiesEliminated(Player)

AddObjectivesComplete():void =
    if (GameManager := GetGameManager[]):
        GameManager.AddObjectivesComplete()

GetEnemiesEliminated()<transacts>:int =
    GetGameManager[].GetEnemiesEliminated() or 0

GetEnemiesEliminated(Player:player)<transacts>:int =
    GetGameManager[].GetEnemiesEliminated(Player) or 0

SetBossHealthBarProgress(Progress:float):void =
    if (GameManager := GetGameManager[]):
        GameManager.CinematicSequenceBossHealthBar.SetPlaybackTime(Progress)

StartGameCountdown(ArenaIndex:int):void =
    if (GameManager := GetGameManager[]):
        if (not GameManager.MaybeCountdownStartTime?):
            spawn{GameManager.StartCountdown(ArenaIndex)}

StopGameCountdown():void =
    if (GameManager := GetGameManager[]):
        GameManager.CancelCountdownEvent.Signal()

IsGameCountdownInProgress()<transacts>:logic =
    logic{GetGameManager[].IsGameCountdownInProgress()?}

(Player:player).AutoRevive()<suspends>:void =
    if (GameManager := GetGameManager[]):
        GameManager.AutoRevive(Player)

GetCurrentStage()<transacts>:int =
    GetGameManager[].CurrentStage or 1

GetCurrentMinibossEvent()<transacts>:int =
    GetGameManager[].CurrentMinibossEvent or 0

(Player:player).IsInStageVolume()<transacts>:logic =
    GetGameManager[].IsPlayerInStageVolume(Player) or false

(Player:player).ShowGameResultsUI():void =
    if (GameManager := GetGameManager[]):
        GameManager.ShowGameResultsUI(Player)

AddGameStat(Id:game_stat_id, Amount:float):void =
    if (GameManager := GetGameManager[]):
        GameManager.AddGameStat(Id, Amount)

ResetGameStats():void =
    if (GameManager := GetGameManager[]):
        GameManager.ResetGameStats()

GetGameStat(Id:game_stat_id)<transacts>:float =
    GetGameManager[].GetGameStat(Id) or 0.0

<# ----------------------------------------------------------------------------------------------------
    game_manager
---------------------------------------------------------------------------------------------------- #>
game_manager := class(creative_device):
    @editable
    CinematicSequenceTimeDilation:cinematic_sequence_device = cinematic_sequence_device{}

    @editable
    CinematicSequenceBossHealthBar:cinematic_sequence_device = cinematic_sequence_device{}

    @editable
    GameplayControlsThirdPerson:gameplay_controls_third_person_device = gameplay_controls_third_person_device{}

    @editable
    GameplayCameraFixedAngle:gameplay_camera_fixed_angle_device = gameplay_camera_fixed_angle_device{}

    @editable
    GameplayCameraFixedAngleIntro:gameplay_camera_fixed_angle_device = gameplay_camera_fixed_angle_device{}
    
    @editable
    CinematicSequenceIntro:cinematic_sequence_device = cinematic_sequence_device{}

    @editable
    CinematicSequenceStageEntry:cinematic_sequence_device = cinematic_sequence_device{}

    @editable
    CinematicSequenceStageExit:cinematic_sequence_device = cinematic_sequence_device{}

    @editable
    CinematicSequenceBossFight:cinematic_sequence_device = cinematic_sequence_device{}

    @editable
    MapIndicator:map_indicator_device = map_indicator_device{}

    @editable
    PlayerCounterVolume:volume_device = volume_device{}

    @editable
    PlayerCounterSwitch:switch_device = switch_device{}

    @editable
    DownButNotOut:down_but_not_out_device = down_but_not_out_device{}

    @editable
    RoundSettings:round_settings_device = round_settings_device{}

    @editable
    ChannelChangingBoothEnable:channel_device = channel_device{}

    @editable
    ChannelChangingBoothDisable:channel_device = channel_device{}

    @editable
    GameTitleResults:[]game_result_title = array{}
    
    @editable
    GameEndPositionLocator:creative_prop = creative_prop{}

    @editable
    VolumeStage:volume_device = volume_device{}
    
    @editable
    BarrierStage:barrier_device = barrier_device{}

    @editable
    BasicStormController:basic_storm_controller_device = basic_storm_controller_device{}
    
    @editable
    AdvancedStormController:advanced_storm_controller_device = advanced_storm_controller_device{}

    @editable
    BillboardGameSettings:billboard_device = billboard_device{}

    GameStartedEvent<public>:events(tuple()) = events(tuple()){}
    GameEndedEvent<public>:events(tuple()) = events(tuple()){}
    VictoryEvent<public>:events(tuple()) = events(tuple()){}
    StageEndedEvent<public>:events(tuple()) = events(tuple()){}
    StopTimerEvent<public>:events(tuple()) = events(tuple()){}
    PlayerReadiedEvent<public>:events(tuple()) = events(tuple()){}
    PlayerUnreadiedEvent<public>:events(tuple()) = events(tuple()){}
    GameResumedEvent<public>:events(tuple()) = events(tuple()){}
    PlayerResumedEvent<public>:events(logic) = events(logic){}
    GameCurrencyChangedEvent<public>:events(game_currency_id) = events(game_currency_id){}
    BossFightStartedEvent<public>:events(tuple()) = events(tuple()){}
    EliminationEvent<public>:events(tuple()) = events(tuple()){}
    CancelCountdownEvent:events(tuple()) = events(tuple()){}
    CancelResumeCountdownEvent:events(tuple()) = events(tuple()){}
    ResumeCountdownStartEvent:events(tuple()) = events(tuple()){}
    ResumeCountdownCompleteEvent:events(tuple()) = events(tuple()){}
    ResumeCountdownEvent:events(int) = events(int){}
    MinibossEliminatedEvent<public>:events(tuple()) = events(tuple()){}

    var PlayerEliminationEvents<public>:[player]events(tuple()) = map{}

    var PauseQueue:queue(tuple(pause_type, int)) = queue(tuple(pause_type, int)){}
    var PlayerPauseQueue:[player]queue(tuple(pause_type, int)) = map{}

    var IsInitialized:logic = false
    var PlayerGameResultUI:[player]game_result_ui = map{}
    var PlayerGameProgressBarUI:[player]game_progress_bar_ui = map{}
    var ReadyPlayers:[]player = array{}
    var ActivePlayers:[]player = array{}
    var PlayersInStageVolume:[]player = array{}
    var MappedGameTitleResults:[game_result_title_id]game_result_title  = map{}

    var GameSettings:game_settings = game_settings{}
    var GameTime:float = 0.0
    var GameTotalTime:float = 0.0
    var EnemiesEliminated:int = 0
    var ObjectivesCompleted:int = 0
    var PlayerGameCurrencies:[player][game_currency_id]game_currency = map{}
    var PlayerEnemiesEliminated:[player]int = map{}
    var MaybeCountdownStartTime:?float = false
    var MaybeResumeCountdownStartTime:?float = false
    var CurrentStage:int = 1
    var CurrentMinibossEvent:int = 1
    var IsStart:logic = false
    var IsDropPodActive:logic = false
    var GameStats:[game_stat_id]float = map{}

    <# ----------------------------------------------------------------------------------------------------
        Initialization
    ---------------------------------------------------------------------------------------------------- #>
    OnBegin<override>()<suspends>:void =
        if (set GlobalGameManager[GetSession()] = Self) {}

        for (Player : GetPlayspace().GetPlayers()):
            spawn{Initialize(Player)}

        for (GameTitleResult : GameTitleResults):
            if (set MappedGameTitleResults[GameTitleResult.Id] = GameTitleResult) {}

        if (DefaultGameSettings := game_settings{Arena := GetArenas()[0]}):
            SetGameSettings(DefaultGameSettings)

        if (MapIndicator.TeleportTo[MapIndicatorDisabledPosition, IdentityRotation()]) {}
        GetPlayspace().PlayerAddedEvent().Subscribe(OnPlayerAdded)
        GetPlayspace().PlayerRemovedEvent().Subscribe(OnPlayerRemoved)
        PlayerCounterVolume.AgentEntersEvent.Subscribe(OnAgentEntersPlayerCounter)
        PlayerCounterVolume.AgentExitsEvent.Subscribe(OnAgentExitsPlayerCounter)
        DownButNotOut.AgentDownedEvent.Subscribe(OnAgentDowned)
        DownButNotOut.AgentRevivedEvent.Subscribe(OnAgentRevived)
        VolumeStage.AgentEntersEvent.Subscribe(OnAgentEntersStageVolume)
        VolumeStage.AgentExitsEvent.Subscribe(OnAgentExitsStageVolume)
        BasicStormController.DestroyStorm()
        set IsInitialized = true

    Initialize(Player:player)<suspends>:void =
        if (set PlayerEliminationEvents[Player] = events(tuple()){}) {}

    AddToGame(Player:player)<suspends>:void =
        if ((game_manager:)IsGameInProgress()?, Character := Player.GetFortCharacter[]):
            set ActivePlayers += array{Player}
            Player.ResetStats()
            Player.CloseActivePanel(?IsSignal := true)
            TeleportGroup(Player, 0, TeleportToArenaPosition, MakeRotation(vector3{Z := 1.0}, PiFloat))
            Player.PlayMusic(music_id.CombatLow)
            Player.PlayAudio(audio_id.Start)
            Player.ActivateAbility()
            Player.RegisterInput(input_id.AbilityStats)
            Player.AssignIndex()
            Character.SetVulnerability(false)
            Sleep(5.0)
            Character.SetVulnerability(true)
    
    <# ----------------------------------------------------------------------------------------------------
        Get Functions
    ---------------------------------------------------------------------------------------------------- #>
    GetIsInitialized()<transacts>:logic =
        IsInitialized

    GetTotalAlivePlayersPaused()<transacts>:int =
        Sum of for (Player -> Queue : PlayerPauseQueue, not Player.GetFortCharacter[].IsDownButNotOut[]):
            Queue.IsEmpty()? and 0 or 1

    IsPaused()<transacts>:logic =       
        PauseCount := Sum of for:
            Player -> Queue : PlayerPauseQueue
            Character := Player.GetFortCharacter[]
            not Queue.IsEmpty()? and (not Character.IsDownButNotOut[] or Player.GetPlayerStats(player_stat_id.Revives) > 0)
        do:
            1

        logic{PauseCount = (game_manager:)GetAlivePlayers().Length}

    (game_manager:)IsPlayerPaused(Player:player)<transacts>:logic =
        logic{PlayerPauseQueue[Player] and not PlayerPauseQueue[Player].IsEmpty()?}

    (game_manager:)IsGameInProgress()<transacts>:logic =
        logic{ActivePlayers.Length > 0}
        
    (game_manager:)GetReadyPlayerCount()<transacts>:int =
        ReadyPlayers.Length

    (game_manager:)IsAllPlayerReady()<transacts>:logic =
        logic{ReadyPlayers.Length = GetPlayspace().GetPlayers().Length}

    (game_manager:)GetActivePlayers()<transacts>:[]player =
        ActivePlayers

    (game_manager:)GetActivePlayersNotAFK()<transacts>:[]player =
        for (ActivePlayer : ActivePlayers, not ActivePlayer.GetIsAFK()?) {ActivePlayer}

    IsAlive(Player:player)<decides><transacts>:void =
        Character := Player.GetFortCharacter[]
        Character.IsActive[]
        not Character.IsDownButNotOut[] or Player.GetPlayerStats(player_stat_id.Revives) > 0

    (game_manager:)GetAlivePlayers()<transacts>:[]player =
        for (ActivePlayer : ActivePlayers, IsAlive[ActivePlayer]):
            ActivePlayer

    (game_manager:)GetVulnerablePlayers()<transacts>:[]player =
        for (ActivePlayer : ActivePlayers, Character := ActivePlayer.GetFortCharacter[], Character.IsActive[], not Character.IsDownButNotOut[], Character.IsVulnerable[]):
            ActivePlayer

    (game_manager:)GetRandomAlivePlayer()<transacts>:?player =
        AlivePlayers := (game_manager:)GetAlivePlayers()
        if (RandomAlivePlayer := AlivePlayers[GetRandomInt(0, AlivePlayers.Length - 1)]):
            option{RandomAlivePlayer}
        else:
            false

    (game_manager:)GetDownButNotOut()<transacts>:down_but_not_out_device =
        DownButNotOut

    (game_manager:)GetGameCurrency(Player:player, Id:game_currency_id)<transacts>:float =
        PlayerGameCurrencies[Player][Id].Amount or 0.0

    GetTimeSurvived()<transacts>:float =
        Min(GameTotalTime, GameTime)

    GetResultInfo(Player:player, IsVictory:logic):result_info =
        TimeSurvived := GetTimeSurvived()
        ActiveArena := GetActiveArena[] or arena{}
        TotalWoodCollected := PlayerGameCurrencies[Player][game_currency_id.Wood].TotalAmount or 0.0
        
        var PlayerStats:[player]player_stat = map{}
        for (StatPlayer -> PlayerStat : GetPlayerStats()):
            if (set PlayerStats[StatPlayer] = MakePlayerStats(PlayerStat)) {}

        ArenaGoldMultiplier := GetArenaGoldMultiplier()
        GoldFromTimeSurvived := TimeSurvived * (1.0 + ArenaGoldMultiplier)
        GoldFromWoodCollected := TotalWoodCollected * 0.1 * (1.0 + ArenaGoldMultiplier)
        GoldFromObjectivesCompleted := ObjectivesCompleted * 100.0 * (1.0 + ArenaGoldMultiplier)
        GoldFromStagesCompleted := CurrentStage * 100.0 * (1.0 + ArenaGoldMultiplier)
        
        GoldEarned := GoldFromTimeSurvived + GoldFromWoodCollected + GoldFromObjectivesCompleted
        GoldFromVictory := GoldEarned * (IsVictory? and VictoryBonusGoldMultplier or 0.0)
        GoldMultiplier := Player.GetPowerUpValue(power_up_id.GoldMultiplier)
        GoldFromMultiplier := (GoldEarned + GoldFromVictory) * GoldMultiplier
        TotalGoldEarned := (Floor[GoldEarned + GoldFromVictory + GoldFromMultiplier] or 0) * 1.0
        
        ArenaLevel := CalculateArenaLevel(GameSettings.Arena.Level, GameSettings.Difficulty)
        ArenaProgress := Player.GetArenaProgress()
        IsRankUp := logic{IsVictory?, ArenaLevel > ArenaProgress, Mod[ArenaLevel, GetArenaCount()] = 0}

        result_info:
            IsVictory := IsVictory
            IsRankUp := IsRankUp
            Arena := ActiveArena
            GameSettings := GameSettings
            EndTime := GetSimulationElapsedTime()
            TimeSurvived := TimeSurvived
            EnemiesEliminated := EnemiesEliminated
            LevelReached := GetLevel()
            WoodCollected := TotalWoodCollected
            GoldFromTimeSurvived := GoldFromTimeSurvived
            GoldFromWoodCollected := GoldFromWoodCollected
            GoldFromObjectivesCompleted := GoldFromObjectivesCompleted
            GoldFromStagesCompleted := GoldFromStagesCompleted
            GoldFromVictory := GoldFromVictory
            GoldEarned := TotalGoldEarned
            GoldMultiplier := GoldMultiplier
            GoldFromMultiplier := GoldFromMultiplier
            ObjectivesCompleted := ObjectivesCompleted
            StagesCompleted := CurrentStage
            ActivePlayers := (game_manager:)GetActivePlayers()
            PlayerStats := PlayerStats
            PlayerActiveAbilities := GetPlayerActiveAbilities()
            PlayerEnemiesEliminated := PlayerEnemiesEliminated
            TeamExperienceMultiplier := GetTeamExperienceMultiplier()

    (game_manager:)GetEnemiesEliminated()<transacts>:int =
        EnemiesEliminated

    (game_manager:)GetEnemiesEliminated(Player:player)<transacts>:int =
        PlayerEnemiesEliminated[Player] or 0

    GetCountdownTimeRemaining()<transacts>:int =
        (GameStartCountdownTime - Floor[GetSimulationElapsedTime() - MaybeCountdownStartTime?]) or 0

    (game_manager:)IsGameCountdownInProgress()<transacts>:logic =
        logic{MaybeCountdownStartTime?}

    GetResumeCountdownTimeRemaining()<transacts>:int =
        GameResumeCountdownTimeAbilitySelect - Floor[GetSecondsSinceEpoch() - MaybeResumeCountdownStartTime?] or 0

    IsPlayerInStageVolume(Player:player)<transacts>:logic =
        logic{PlayersInStageVolume.Find[Player]}

    (game_manager:)GetGameStat(Id:game_stat_id)<transacts>:float =
        GameStats[Id] or 0.0

    <# ----------------------------------------------------------------------------------------------------
        Update Functions
    ---------------------------------------------------------------------------------------------------- #>
    Pause():void =
        CinematicSequenceTimeDilation.SetPlaybackTime(0.0)
        CinematicSequenceTimeDilation.Play()

    Pause(Type:pause_type, Level:int):logic =
        if (not (game_manager:)IsGameInProgress()?):
            return false
            
        for (Player : (game_manager:)GetActivePlayers()):
            if:
                not PlayerPauseQueue[Player]
                set PlayerPauseQueue[Player] = queue(tuple(pause_type, int)){}

            if (PlayerPauseQueue[Player].IsEmpty()?):
                ShowPauseUI(Player, Type, Level)
                Player.DeactivateAbility(?IsPause := true)
                Player.SetVulnerability(false)
            
            if:
                UpdatedPauseQueue := PlayerPauseQueue[Player].Enqueue((Type, Level))
                set PlayerPauseQueue[Player] = UpdatedPauseQueue
        
        Pause()
        return true

    Pause(Player:player, Type:pause_type, Level:int):logic =
        if (not (game_manager:)IsGameInProgress()?):
            return false

        if:
            not PlayerPauseQueue[Player]
            set PlayerPauseQueue[Player] = queue(tuple(pause_type, int)){}
        
        if (PlayerPauseQueue[Player].IsEmpty()?):
            ShowPauseUI(Player, Type, Level)
            Player.DeactivateAbility(?IsPause := true)
            Player.SetVulnerability(false)
        
        if:
            UpdatedPauseQueue := PlayerPauseQueue[Player].Enqueue((Type, Level))
            set PlayerPauseQueue[Player] = UpdatedPauseQueue

        if (GetTotalAlivePlayersPaused() = (game_manager:)GetAlivePlayers().Length):
            Pause()
            return true
        else:
            return false

    Resume(Player:player, ?IsShow:logic = true):void =
        var IsGameResumed:logic = logic{IsPaused()? and (not Player.GetFortCharacter[].IsDownButNotOut[] or Player.GetPlayerStats(player_stat_id.Revives) > 0)}

        if (Dequeue := PlayerPauseQueue[Player].Dequeue[], UpdatedQueue := Dequeue(0)):
            if (QueueFront := UpdatedQueue.Front[]):
                if (IsShow?):
                    ShowPauseUI(Player, QueueFront(0), QueueFront(1))

                if (set PlayerPauseQueue[Player] = UpdatedQueue) {}
                return
            else:
                if (Dequeue(1)(0) = pause_type.ItemShop):
                    if (GetTotalAlivePlayersPaused() > 1):
                        set IsGameResumed = false
                    else:
                        set IsGameResumed = true

                if (IsGameResumed?):
                    Resume()
                    CancelResumeCountdownEvent.Signal()
                else if (IsPaused()?):
                    Player.ShowGamePausedUI()
                    if (not MaybeResumeCountdownStartTime?):
                        spawn{ResumeCountdown()}

                if (set PlayerPauseQueue[Player] = UpdatedQueue) {}

        Player.SetVulnerability(true, ?BufferTime := DefaultVulnerabilityBufferTime)
        Player.ReactivateAbility()
        KnockbackNPC(Player)
        PlayerResumedEvent.Signal(IsGameResumed)

    ShowPauseUI(Player:player, Type:pause_type, Level:int):void =
        Player.CloseActivePanel()
        case (Type):
            pause_type.LevelUp => Player.ShowAbilitySelectUI(Level)
            pause_type.AbilityUpgrade => Player.GrantAbilityUpgrade()
            pause_type.ItemShop => Player.ShowItemShopUI()
            pause_type.Boon => Player.ShowBoonSelectUI()

    Resume(?IsInstant:logic = false):void =
        CinematicSequenceTimeDilation.SetPlaybackTime((not IsPaused()? or IsInstant?) and 1.5 or 1.1)
        CinematicSequenceTimeDilation.Play()
        GameResumedEvent.Signal()

    ResumeCountdown()<suspends>:void =
        defer:
            set MaybeResumeCountdownStartTime = false

        set MaybeResumeCountdownStartTime = option{GetSecondsSinceEpoch()}
        ResumeCountdownStartEvent.Signal()

        race:
            CancelResumeCountdownEvent.Await()
            loop:
                SleepTimeDilation(1.0)
                TimeRemaining := GetResumeCountdownTimeRemaining()
                ResumeCountdownEvent.Signal(TimeRemaining)

                if (TimeRemaining <= 0):
                    AutoCompletePauseQueue()
                    ResumeCountdownCompleteEvent.Signal()
                    break

    AutoCompletePauseQueue():void =
        Resume()
        for (Player : (game_manager:)GetAlivePlayers(), Character := Player.GetFortCharacter[], not Character.IsDownButNotOut[], IsPlayerPaused(Player)?):
            var AutoSelectCount:int = 0
            loop:
                if (Dequeue := PlayerPauseQueue[Player].Dequeue[], set PlayerPauseQueue[Player] = Dequeue(0)):
                    case (Dequeue(1)(0)):
                        pause_type.LevelUp =>
                            set AutoSelectCount += 1
                            if (IsWeaponUnlockLevel(Dequeue(1)(1))?, WeaponInfo := GetRandomWeapons(Player, 1)[0]):
                                Player.ActivateNewAbility(WeaponInfo.Id)
                                Player.EquipNewWeapon(WeaponInfo.Id)

                            else if (AbilityInfo := GetRandomAbility(Player)?):
                                Player.AddAbility(AbilityInfo(0), ?Amount := AbilityInfo(1))

                        pause_type.AbilityUpgrade => if (RandomAbility := GetTreasureChestRandomAbility(Player)?):
                            set AutoSelectCount += 1
                            Player.AddAbility(RandomAbility(0), ?Amount := RandomAbility(1))

                        pause_type.ItemShop => {}

                        pause_type.Boon => if (RandomBoonInfo := GetRandomBoonInfo[Player]):
                            Player.AddBoon(RandomBoonInfo)

                else:
                    WeaponEvolutionInfos := Player.GetWeaponEvolutions()
                    for (WeaponId -> WeaponLevel : Player.GetWeaponLevels(), WeaponInfo := GetWeaponInfo[WeaponId]):
                        Current := WeaponEvolutionInfos[WeaponId].Length or 0
                        Target := Player.GetWeaponEvolutionLevel(WeaponId)
                        Delta := Target - Current
        
                        for (Index := 1..Delta, EvolutionInfos := WeaponInfo.Evolution[Current + (Index - 1)], WeaponEvolutionInfo := EvolutionInfos[GetRandomInt(0, EvolutionInfos.Length - 1)]):
                            Player.AddWeaponEvolution(WeaponEvolutionInfo)

                    Player.CloseActivePanel()
                    Player.ShowMessage(message_id.AutoSelectedAbilities, "{AutoSelectCount}")
                    Resume(Player)
                    break

        PlayerResumedEvent.Signal(true)
        
    ForceStart(Index:int):void =
        spawn{Start(Index, ?IsSkipIntro := true)}

    Start(ArenaIndex:int, ?IsSkipIntro:logic = false)<suspends>:void =
        if (IsStart?) then return
        set IsStart = true
        set GameTime = 0.0
        set ActivePlayers = GetAllPlayers()
        ResetStats()
        (game_manager:)ResetGameStats()

        ToggleMatchmaking(false)

        if (ArenaManager := GetArenaManager[], Arena := ArenaManager.GetActiveArena[], NPCManager := GetNPCManager[]):
            StageCount := GetArenaStageCount()
            set GameTotalTime = StageCount * ArenaStageDuration
            ArenaManager.LoadGatherables(ArenaIndex)
            GameStartedEvent.Signal()

            for (Index -> Player : GetPlayspace().GetPlayers(), Character := Player.GetFortCharacter[]):
                BarrierStage.AddToIgnoreList(Player)
                Player.CloseActivePanel(?IsSignal := true)
                Player.SetVulnerability(true)
                Player.ClearMusic()
                TeleportGroup(Player, Index, TeleportToArenaPosition, MakeRotation(vector3{Z := 1.0}, PiFloat))

                if (not IsSkipIntro?, Player.GetActiveCamera() <> camera_type.Orthographic):
                    GameplayCameraFixedAngleIntro.AddTo(Player)

                if (Player.GetIsAFK()?):
                    Player.ActivateVFXPowerup(vfx_powerup_id.AFK)
                else:
                    Player.DeactivateVFXPowerup of Player.GetRankVFXPowerUp()

                if (set PlayerGameCurrencies[Player] = map{}) {}

            if (not IsSkipIntro?):
                CinematicSequenceIntro.Play()
                Sleep(4.5)
                GameplayCameraFixedAngleIntro.RemoveFromAll()
                ActivateCameraShake(DefaultArenaOrigin)

            ArenaMusic := ArenaManager.GetActiveArena[].MusicId or music_id.CombatLow
            for (Index -> Player : GetPlayspace().GetPlayers()):
                Player.RefreshControls()
                Player.ActivateAbility()
                Player.RegisterInput(input_id.AbilityStats)
                Player.PlayAudio(audio_id.Start)
                Player.PlayMusic(ArenaMusic)
            
            race:
                StopTimerEvent.Await()
                block:
                    for (StageIndex := 1..StageCount):
                        set CurrentStage = StageIndex
                        ShowMessage(message_id.StageIndicator, "{StageIndex}/{StageCount}", ?DisplayTime := 0.0)
                        ShowMessage(message_id.Message, (StageIndex = StageCount) and "Final Stage!" or "Stage {StageIndex}")
                        
                        for (Player : (game_manager:)GetActivePlayers()):
                            Player.PlayAudio(audio_id.Start)

                            if (StageIndex = StageCount):
                                Player.PlayMusic(music_id.CombatHigh)

                        spawn{NPCManager.StartWaves(ArenaIndex)}
                        StartTimer(ArenaStageDuration, logic{StageIndex = StageCount})
                        
                        if (not IsMinibossEventCompleted()?):
                            for (Gatherable : GetGatherables()):
                                Gatherable.DisableRespawn()

                            loop:
                                EliminatedMiniboss := Floor[(game_manager:)GetGameStat(game_stat_id.EliminatedMiniboss)] or 0
                                ShowMessage(message_id.MinibossEliminatedCounter, "{EliminatedMiniboss}/{GetMinibossEliminationTarget()}", ?DisplayTime := 0.0)
                                MinibossEliminatedEvent.Await()
                                if (IsMinibossEventCompleted()?):
                                    break

                        if (StageIndex < StageCount):
                            # Generate storm & station entry cinematic sequence
                            set IsDropPodActive = true
                            ShowMessage(message_id.PlayerCounterShip, "0/{(game_manager:)GetActivePlayers().Length}", ?DisplayTime := 0.0)
                            ToggleItemShop(true)
                            CinematicSequenceStageEntry.Play()
                            BarrierStage.Enable()
                            for (Player : (game_manager:)GetActivePlayers()):
                                BarrierStage.AddToIgnoreList(Player)

                            # Await all players to enter driller
                            Sleep(5.0)
                            AdvancedStormController.GenerateStorm()
                            BasicStormController.GenerateStorm()
                            AdvancedStormController.DestroyStorm()

                            var ShipTimeRemaining:float = 30.0
                            race:
                                block:
                                    Sleep(15.0)
                                    for (Player : (game_manager:)GetActivePlayers()) {Player.PlayAudio(audio_id.TimerWarning)}
                                    for (Index := 1..15):
                                        Sleep(1.0)
                                        for (Player : (game_manager:)GetActivePlayers()) {Player.PlayAudio(audio_id.TimerCountdown)}
                                    Sleep(Inf)

                                block:
                                    loop:
                                        if (ShipTimeRemaining <= 0.0):
                                            ShowMessage(message_id.TimerShip, "00:00", ?DisplayTime := 0.0)
                                            break
                                        
                                        ShowMessage(message_id.TimerShip, "{ShipTimeRemaining.FormatAsTime()}", ?DisplayTime := 0.0)
                                        Sleep(1.0)
                                        set ShipTimeRemaining -= 1.0
                                        
                                    Sleep(1.0)
                                    HideMessage(message_id.TimerShip)
                                    Stop(false)

                                loop:
                                    if (PlayersInStageVolume.Length >= (game_manager:)GetActivePlayers().Length):
                                        break

                                    VolumeStage.AgentEntersEvent.Await()

                            set IsDropPodActive = false
                            for (Player : (game_manager:)GetActivePlayers()):
                                BarrierStage.RemoveFromIgnoreList(Player)

                            Sleep(1.0)
                            AwardAccolade(accolade_id.StageCompleted, (game_manager:)GetActivePlayers())
                            HideMessage(message_id.TimerShip)
                            StageEndedEvent.Signal()

                            for (Player : (game_manager:)GetActivePlayers(), GameProgressBarUI := PlayerGameProgressBarUI[Player]):
                                GameProgressBarUI.StageEnded()

                            for (NPCIndex -> NPCBehavior : GetAllNPCs()):
                                spawn{NPCBehavior.Dance()}

                            # Driller exit cinematic sequence
                            CinematicSequenceStageEntry.Stop()
                            CinematicSequenceStageExit.Play()
                            Sleep(6.5)
                            BarrierStage.Disable()

                            # Item shop
                            NPCManager.StopSpawners()
                            BasicStormController.DestroyStorm()
                            ToggleItemShop(false)
                            ClearItemDrops()
                            ArenaManager.LoadGatherables(ArenaIndex)
                            PauseGame(pause_type.ItemShop)
                            TeleportGroup((game_manager:)GetActivePlayers(), TeleportToArenaPosition, MakeRotation(vector3{Z := 1.0}, PiFloat))
                            for (Player : (game_manager:)GetActivePlayers()):
                                Player.PutInStasis(?IsAllowFalling := false)

                            # Next stage entry cinematic sequence
                            TeleportGroup((game_manager:)GetActivePlayers(), TeleportToArenaPosition, MakeRotation(vector3{Z := 1.0}, PiFloat))
                            for (Player : (game_manager:)GetActivePlayers()):
                                Player.ReleaseFromStasis()

                            CinematicSequenceIntro.Play()
                            Sleep(4.5)
                            ActivateCameraShake(DefaultArenaOrigin)

                            set PlayersInStageVolume = array{}
                            for (Player : (game_manager:)GetActivePlayers()):
                                Player.SetVulnerability(true)
                                Player.ReactivateAbility()

                            CinematicSequenceIntro.StoppedEvent.Await()
                            for (Player : (game_manager:)GetActivePlayers()):
                                Player.ReactivateAbility()

                    spawn{StartBossFight()}

    StartTimer(Duration:float, IsFinalStage:logic)<suspends>:void =
        var StageTime:float = 0.0
        set CurrentMinibossEvent = 0

        GameEvents := IsFinalStage? and GameEventSequenceFinalStage or GameEventSequence
        GameEventCount := GameEvents.Length
        GameSpeed := GetGameSpeed()
        
        defer:
            for (Player : (game_manager:)GetActivePlayers(), GameProgressBarUI := PlayerGameProgressBarUI[Player]):
                GameProgressBarUI.UpdateEventState(GameEventCount - 1, game_event_state.Active)

        for (Player : (game_manager:)GetActivePlayers()):
            ShowGameProgressBarUI(Player)
            if (GameProgressBarUI := PlayerGameProgressBarUI[Player]):
                GameProgressBarUI.InitializeEvents(GameEvents)
        
        race:
            StopTimerEvent.Await()
            block:
                var EventIndex:int = 0
                loop:
                    Sleep(Duration / (GameEventCount * 1.0) / GameSpeed)
                    if (GameEvent := GameEvents[EventIndex]):
                        case (GameEvent):
                            game_events.Miniboss =>
                                set CurrentMinibossEvent += 1
                                SpawnRandomMiniboss()
                                ShowMessage(message_id.Message, "Miniboss Spawned!")

                            game_events.Mission =>
                                StartRandomObjective()

                            game_events.SupplyDrop =>
                                SpawnSupplyDrop()
                                ShowMessage(message_id.SupplyDropNotification, "")

                            _ => {}

                    for (Player : (game_manager:)GetActivePlayers(), GameProgressBarUI := PlayerGameProgressBarUI[Player]):
                        GameProgressBarUI.UpdateEventState(EventIndex, game_event_state.Active)

                    if (EventIndex >= GameEventCount - 1):
                        break

                    set EventIndex += 1

                Sleep(Inf)

            loop:
                TimeRemaining := Duration - StageTime
                Progress := 1.0 - TimeRemaining / Duration
                
                for (Player : (game_manager:)GetActivePlayers(), GameProgressBarUI := PlayerGameProgressBarUI[Player]):
                    Player.ShowMessage(message_id.Timer, TimeRemaining.FormatAsTime(), ?DisplayTime := 0.0)
                    GameProgressBarUI.UpdateProgress(Progress)

                if (StageTime >= Duration):
                    break
                else:
                    Sleep(1.0)
                    set StageTime += GameSpeed
                    set GameTime += GameSpeed

    ShowGameProgressBarUI(Player:player):void =
        ProgressBarUI := if (GameProgressBarUI := PlayerGameProgressBarUI[Player]):
            GameProgressBarUI
        else:
            GameProgressBarUI := game_progress_bar_ui{Player := Player}
            GameProgressBarUI.Initialize()
            if (set PlayerGameProgressBarUI[Player] = GameProgressBarUI) {}
            
            Player.AddWidgetToHUD of canvas_slot:
                Anchors := anchors{Minimum := vector2{X := 0.5, Y := 0.0}, Maximum := vector2{X := 0.5, Y := 0.0}}
                Offsets := margin{Left := 0.0, Top := 20.0, Right := 0.0, Bottom := 0.0}
                Alignment := vector2{X := 0.5, Y := 0.0}
                Widget := GameProgressBarUI.GetUI()

            GameProgressBarUI

        ProgressBarUI.Show()
                    
    TeleportGroup(Players:[]player, Position:vector3, Rotation:rotation):void =
        for (Index -> Player : Players):
            TeleportGroup(Player, Index, Position, Rotation)

    TeleportGroup(Player:player, Index:int, Position:vector3, Rotation:rotation):void =
        YOffset := Index * TeleportGroupSpacing - (GetAllPlayers().Length - 1) * 0.5 * TeleportGroupSpacing
        if :
            Character := Player.GetFortCharacter[]
            Character.TeleportTo[vector3{X := Position.X, Y := Position.Y + YOffset, Z:= Position.Z}, Rotation]

    StartBossFight()<suspends>:void =
        BossFightStartedEvent.Signal()

        CinematicSequenceBossHealthBar.Play()
        CinematicSequenceBossHealthBar.Pause()
        Sleep(0.0)
        CinematicSequenceBossHealthBar.SetPlaybackTime(1.0)

        for (Index -> Player : (game_manager:)GetActivePlayers()):
            Player.HideMessage(message_id.Timer)
            HideMessage(message_id.PlayerCounterShip)
            HideMessage(message_id.TimerShip)
            HideMessage(message_id.MinibossEliminatedCounter)
            Player.HideMessage(message_id.StageIndicator)
            Player.ShowMessage(message_id.BossWarning)
            Player.PlayAudio(audio_id.Warning)

            if (GameProgressBarUI := PlayerGameProgressBarUI[Player]):
                GameProgressBarUI.Hide()

        MaybeNPCSpawner:?npc_spawner_device = if (TargetPlayer := (game_manager:)GetRandomAlivePlayer()?, Arena := GetActiveArena[]): # [TODO]
            SpawnNPC(Arena.BossId, TargetPlayer, ?MaybePosition := option{vector3{Z := 200.0}})
        else:
            false

        MaybeAgent:?agent = if (NPCSpawner := MaybeNPCSpawner?):
            Agent := NPCSpawner.SpawnedEvent.Await()
            option{Agent}
        else:
            false

        for (Player : (game_manager:)GetActivePlayers()):
            Player.PlayMusic(music_id.BossFight)
          
        if (Agent := MaybeAgent?, NPCBehaviorBase := npc_behavior_base[Agent.GetNPCBehavior[]]):
            spawn{NPCBehaviorBase.Start()}

        if:
            Agent := MaybeAgent?
            Character := Agent.GetFortCharacter[]
        then:
            race:
                StopTimerEvent.Await()
                loop:
                    MapIndicator.MoveTo(Character.GetTransform(), TickTime * 3)
                block:
                    Character.EliminatedEvent().Await()
                    if (MapIndicator.TeleportTo[MapIndicatorDisabledPosition, IdentityRotation()]) {}
                    DespawnNPC()
                    ClearItemDrops()

                    for (Player : GetAllPlayers()):
                        Player.PlayAudio(audio_id.BossEliminated)

                    Sleep(5.0)
                    Stop(true)

    ShowGameResultsUI(Player:player, ?MaybeResultInfo:?result_info := false):void =
        ResultUI := if (GameResultUI := PlayerGameResultUI[Player]):
            GameResultUI
        else:
            GameResultUI := game_result_ui{Player := Player}
            if (set PlayerGameResultUI[Player] = GameResultUI) {}
            GameResultUI

        ResultUI.Show()

        if (ResultInfo := MaybeResultInfo?):
            ResultUI.Update(ResultInfo)

    ShowResults(Players:[]player, IsVictory:logic)<suspends>:void =
        var PlayerResultInfo:[player]result_info = map{}
        for (Player : Players):
            ResultInfo := GetResultInfo(Player, IsVictory)
            if (set PlayerResultInfo[Player] = ResultInfo) {}
            
        Sleep(4.0)

        for (Player : Players, ResultInfo := PlayerResultInfo[Player]):
            ShowGameResultsUI(Player, ?MaybeResultInfo := option{ResultInfo})
            Player.AddCurrency(currency_id.Gold, ResultInfo.GoldEarned)
            Player.ReleaseFromStasis()

        PlayerPlayersRevived := for (Player : Players, PlayersRevived := PlayerResultInfo[Player].PlayerStats[Player].GetGameStat(player_game_stat_id.PlayersRevived)) {(Player, PlayersRevived)}
        PlayerDamageDealt := for (Player : Players, DamageDealt := PlayerResultInfo[Player].PlayerStats[Player].GetGameStat(player_game_stat_id.DamageDealt)) {(Player, DamageDealt)}
        PlayerWoodCollected := for (Player : Players, WoodCollected := PlayerResultInfo[Player].PlayerStats[Player].GetGameStat(player_game_stat_id.WoodCollected)) {(Player, WoodCollected)}       
        PlayerTimesDowned := for (Player : Players, TimesDowned := PlayerResultInfo[Player].PlayerStats[Player].GetGameStat(player_game_stat_id.TimesDowned)) {(Player, TimesDowned)}
        
        SortedPlayerPlayersRevived := SortBy(PlayerPlayersRevived, LessPlayerFloat)
        SortedPlayerDamageDealt := SortBy(PlayerDamageDealt, LessPlayerFloat)
        SortedPlayerWoodCollected := SortBy(PlayerWoodCollected, LessPlayerFloat)
        SortedPlayerTimesDowned := SortBy(PlayerTimesDowned, LessPlayerFloat)

        if (PlayerMostPlayersRevived := SortedPlayerPlayersRevived[0], GameTitleResult := MappedGameTitleResults[game_result_title_id.PlayersRevived]):
            GameTitleResult.PlayerReference.Register(PlayerMostPlayersRevived(0))
            GameTitleResult.Billboard.SetText(MakeMessage("Players Revived: {FormatCurrency(PlayerMostPlayersRevived(1))}"))
            
        if (PlayerMostDamageDealt := SortedPlayerDamageDealt[0], GameTitleResult := MappedGameTitleResults[game_result_title_id.DamageDealt]):
            GameTitleResult.PlayerReference.Register(PlayerMostDamageDealt(0))
            GameTitleResult.Billboard.SetText(MakeMessage("Damage Dealt: {FormatCurrency(PlayerMostDamageDealt(1))}"))

        if (PlayerMostWoodCollected := SortedPlayerWoodCollected[0], GameTitleResult := MappedGameTitleResults[game_result_title_id.WoodCollected]):
            GameTitleResult.PlayerReference.Register(PlayerMostWoodCollected(0))
            GameTitleResult.Billboard.SetText(MakeMessage("Wood Collected: {FormatCurrency(PlayerMostWoodCollected(1))}"))

        if (PlayerMostTimesDowned := SortedPlayerTimesDowned[0], GameTitleResult := MappedGameTitleResults[game_result_title_id.TimesDowned]):
            GameTitleResult.PlayerReference.Register(PlayerMostTimesDowned(0))
            GameTitleResult.Billboard.SetText(MakeMessage("Times Downed: {FormatCurrency(PlayerMostTimesDowned(1))}"))

        ResetStats()

    Stop(IsVictory:logic):void =
        set PlayerPauseQueue = map{}
        spawn{ShowResults(ActivePlayers, IsVictory)}

        StopTimerEvent.Signal()
        GameEndedEvent.Signal()

        DespawnNPC()
        ClearItemDrops()
        BasicStormController.DestroyStorm()
        if (MapIndicator.TeleportTo[MapIndicatorDisabledPosition, IdentityRotation()]) {}

        for (Index -> Player : (game_manager:)GetActivePlayers(), Character := Player.GetFortCharacter[]):
            Player.CloseActivePanel()
            Player.PlayMusic(music_id.Default)
            Player.PlayAudio(IsVictory? and audio_id.Victory or audio_id.Defeat)
            Player.ShowMessage(IsVictory? and message_id.Victory or message_id.Defeat)
            Player.HideMessage(message_id.Message)
            Player.HideMessage(message_id.ExperienceBar)
            Player.HideMessage(message_id.ExperienceBarLevel)
            Player.HideMessage(message_id.Timer)
            HideMessage(message_id.PlayerCounterShip)
            HideMessage(message_id.TimerShip)
            HideMessage(message_id.MinibossEliminatedCounter)
            Player.HideMessage(message_id.StageIndicator)
            Player.UnregisterInput(input_id.AbilityStats)
            Player.PutInStasis()

            if (GameProgressBarUI := PlayerGameProgressBarUI[Player]):
                GameProgressBarUI.Hide()
            
            if (IsVictory?):
                if (ActiveArenaIndex := GetActiveArenaIndex()?, WeaponId := Player.GetEquippedWeapon()?):
                    ArenaLevel := CalculateArenaLevel(GameSettings.Arena.Level, GameSettings.Difficulty)
                    ArenaProgress := Player.GetArenaProgress()
                    
                    if:
                        CurrentPlayerData := Player.GetData[]
                        CurrentValue := CurrentPlayerData.WeaponProgress[WeaponId.ToInt()] or 0
                        ArenaLevel > CurrentValue

                        var UpdatedWeaponProgress:[int]int = CurrentPlayerData.WeaponProgress
                        set UpdatedWeaponProgress[WeaponId.ToInt()] = ArenaLevel

                        set PlayerDataMap[Player] = player_data:
                            MakePlayerData<constructor>(CurrentPlayerData)
                            WeaponProgress := UpdatedWeaponProgress

                    if (ArenaLevel > ArenaProgress):
                        Player.SetArenaProgress(ArenaProgress + 1)
                        
            if (Character.IsDownButNotOut[]):
                DownButNotOut.Revive(Player)

            TeleportGroup(Player, Index, GameEndPositionLocator.GetTransform().Translation + vector3{Z := 200.0}, GameEndPositionLocator.GetTransform().Rotation)
            Player.DeactivateVFXPowerup(vfx_powerup_id.Invulnerability)
            Player.ActivateVFXPowerup of Player.GetRankVFXPowerUp()

        Resume(?IsInstant := true)
        ToggleMatchmaking(true)
        set PlayerGameCurrencies = map{}
        set EnemiesEliminated = 0
        set ObjectivesCompleted = 0
        set PlayerEnemiesEliminated = map{}
        set ActivePlayers = array{}
        set IsStart = false
        
        if (IsVictory?):
            VictoryEvent.Signal()

    (game_manager:)AddGameCurrency(Id:game_currency_id, Amount:float):void =
        for (ActivePlayer : (game_manager:)GetActivePlayers()):
            (game_manager:)AddGameCurrency(ActivePlayer, Id, Amount, ?IsSignal := false)

        GameCurrencyChangedEvent.Signal(Id)

    (game_manager:)AddGameCurrency(Player:player, Id:game_currency_id, Amount:float, ?IsSignal:logic = true):void =
        if:
            not PlayerGameCurrencies[Player][Id]
            set PlayerGameCurrencies[Player][Id] = game_currency{Id := Id}

        if:
            CurrentAmount := PlayerGameCurrencies[Player][Id]
            set PlayerGameCurrencies[Player][Id] = game_currency:
                MakeGameCurrency<constructor>(CurrentAmount)
                Amount := CurrentAmount.Amount + Amount
                TotalAmount := CurrentAmount.TotalAmount + Amount

        if (IsSignal?):
            GameCurrencyChangedEvent.Signal(Id)

    (game_manager:)RemoveGameCurrency(Player:player, Id:game_currency_id, Amount:float):void =
        if:
            CurrentAmount := PlayerGameCurrencies[Player][Id]
            set PlayerGameCurrencies[Player][Id] = game_currency:
                MakeGameCurrency<constructor>(CurrentAmount)
                Amount := CurrentAmount.Amount - Amount

        GameCurrencyChangedEvent.Signal(Id)

    (game_manager:)AddEnemiesEliminated(Player:player):void =
        set EnemiesEliminated += 1
        if (set PlayerEnemiesEliminated[Player] = PlayerEnemiesEliminated[Player] + 1 or 1) {}

        EliminationEvent.Signal()
        if (PlayerEliminationEvent := PlayerEliminationEvents[Player]):
            PlayerEliminationEvent.Signal()

    (game_manager:)AddObjectivesComplete():void =
        set ObjectivesCompleted += 1

    UpdatePlayerCounter():void =
        for (Player : ReadyPlayers):
            Player.ShowMessage(message_id.PlayerCounter, "{ReadyPlayers.Length}/{GetPlayspace().GetPlayers().Length}", ?DisplayTime := 0.0)

    StartCountdown(ArenaIndex:int)<suspends>:void =
        defer:
            HideMessage(message_id.Countdown)
            ChannelChangingBoothEnable.Transmit(false)
            set MaybeCountdownStartTime = false

        set MaybeCountdownStartTime = option{GetSimulationElapsedTime()}
        ShowMessage(message_id.Countdown, "{GameStartCountdownTime}", ?DisplayTime := 1.0)
        ToggleMatchmaking(false)
        ChannelChangingBoothDisable.Transmit(false)

        race:
            CancelCountdownEvent.Await()
            loop:
                Sleep(1.0)
                TimeRemaining := GetCountdownTimeRemaining()
                ShowMessage(message_id.Countdown, "{TimeRemaining}", ?DisplayTime := 1.0)
                
                for (Player : GetAllPlayers()):
                    Player.PlayAudio(audio_id.TimerCountdown)

                if (TimeRemaining <= 0):                   
                    Start(ArenaIndex)
                    break

    ToggleMatchmaking(IsEnable:logic):void =
        if (IsEnable?):
            RoundSettings.EnableMatchmaking()
        else:
            RoundSettings.DisableMatchmaking()

    (game_manager:)AddGameStat(Id:game_stat_id, Amount:float):void =
        if (not GameStats[Id], set GameStats[Id] = Amount):
            return

        if (set GameStats[Id] += Amount) {}

    (game_manager:)ResetGameStats():void =
        set GameStats = map{}

    <# ----------------------------------------------------------------------------------------------------
        Event Handlers
    ---------------------------------------------------------------------------------------------------- #>
    OnPlayerAdded(Player:player):void =
        spawn{Initialize(Player)}
        if ((game_manager:)IsGameCountdownInProgress()?):
            return

        UpdatePlayerCounter()
        
    OnPlayerRemoved(Player:player):void =
        set ActivePlayers = ActivePlayers.RemoveAllElements(Player)
        set PlayerGameResultUI = RemoveKeyFromMap(PlayerGameResultUI, Player)
        set PlayerGameCurrencies = RemoveKeyFromMap(PlayerGameCurrencies, Player)
        set PlayerEnemiesEliminated = RemoveKeyFromMap(PlayerEnemiesEliminated, Player)
        set PlayerEliminationEvents = RemoveKeyFromMap(PlayerEliminationEvents, Player)
        set PlayerPauseQueue = RemoveKeyFromMap(PlayerPauseQueue, Player)
        set PlayerGameProgressBarUI = RemoveKeyFromMap(PlayerGameProgressBarUI, Player)

        if ((game_manager:)GetAlivePlayers().Length = 0):
            Stop(false)

        if ((game_manager:)IsGameInProgress()?, GetTotalAlivePlayersPaused() = (game_manager:)GetAlivePlayers().Length):
            Pause()

    OnAgentEntersPlayerCounter(Agent:agent):void =
        if (Player := player[Agent]):
            set ReadyPlayers += array{Player}
            Player.SetAFK(false)
            Player.PlayAudio(audio_id.Ready)
            PlayerCounterSwitch.TurnOn(Player)

        if ((game_manager:)IsGameCountdownInProgress()?):
            if ((game_manager:)IsAllPlayerReady()?):
                StopGameCountdown()
                spawn{Start(GetActiveArenaIndex()? or 0)}

            return

        UpdatePlayerCounter()
        PlayerReadiedEvent.Signal()

    OnAgentExitsPlayerCounter(Agent:agent):void =
        if (Player := player[Agent]):
            set ReadyPlayers = ReadyPlayers.RemoveAllElements(Player)
            Player.HideMessage(message_id.PlayerCounter)
            PlayerCounterSwitch.TurnOff(Player)

        if ((game_manager:)IsGameCountdownInProgress()?):
            return

        UpdatePlayerCounter()
        PlayerUnreadiedEvent.Signal()

    AutoRevive(Player:player)<suspends>:void =
        if (PlayerStat := Player.GetStats[], Character := Player.GetFortCharacter[], ReviveVolumeActivator := Player.GetReviveVolumeActivator[]):
            if (PlayerStat.IsAutoReviving?) then return
            set PlayerStat.IsAutoReviving = true
            
            ReviveVolumeActivator.Disable()
            Player.SetVulnerability(false)
            Player.DeactivateAbility()
            Sleep(PlayerReviveDuration)
            Player.RemovePlayerStats(player_stat_id.Revives, 1)

            MaybePanel := option{Player.GetActivePanel[]}
            if (Panel := MaybePanel?):
                Panel.Hide(?IsSignal := false)

            DownButNotOut.Revive(Player)
            Player.AddGameStat(player_game_stat_id.AutoRevived, 1.0)
            Player.UpdateMaxHealth()

            if (Panel := MaybePanel?):
                Panel.Show(?IsSignal := false)
            
            if (ReviveBlastValue := Player.GetPowerUpValue(power_up_id.ReviveBlast), ReviveBlastValue > 0.0):
                ReviveBlastSize := 250.0 * (1.0 + ReviveBlastValue)
                ReviveBlastDamage := GetNPCMaxHealth(NPCBaseMaxHealth) * (0.5 + ReviveBlastValue)
                RevivePosition := Character.GetTransform().Translation.Make2D()
                SpawnVFX(vfx_id.ReviveBlast, RevivePosition, ?Scale := vector3{X := ReviveBlastSize * 0.1, Y := ReviveBlastSize * 0.1, Z := 1.0}, ?Duration := 3.0)
                Sleep(1.0)

                for:
                    NPCBehaviorBase : GetAllNPCs()
                    NPCCharacter := NPCBehaviorBase.GetAgent[].GetFortCharacter[]
                    NPCPosition := NPCCharacter.GetTransform().Translation.Make2D()
                    DistanceSquared(NPCPosition, RevivePosition) <= Pow(ReviveBlastSize, 2.0)
                do:
                    DamageArgs := ability_damage_args:
                        Target := NPCBehaviorBase
                        Amount := ReviveBlastDamage
                        IsCriticalHit := false

                    NPCBehaviorBase.Damage(Player, DamageArgs, ?IsKnockback := false)

            else:
                SpawnVFX(vfx_id.Revive, Character.GetTransform().Translation.Make2D(), ?Duration := 3.0)

            KnockbackNPC(Player, ?KnockbackDistanceSquared := DefaultNPCKnockbackDistanceSquaredRevive, ?Duration := 3.0)

            set PlayerStat.IsAutoReviving = false

    OnAgentDowned(Agent:agent):void =
        if (Player := player[Agent]):
            if (Player.GetPlayerStats(player_stat_id.Revives) > 0):
                spawn{AutoRevive(Player)}
                return

            if ((game_manager:)GetAlivePlayers().Length = 0):
                Stop(false)

            else if (Character := Player.GetFortCharacter[]):
                Player.SetVulnerability(false)
                Player.DeactivateAbility()

                if (GetTotalAlivePlayersPaused() = (game_manager:)GetAlivePlayers().Length):
                    Pause()
                    for (ActivePlayer : (game_manager:)GetActivePlayers(), not IsPlayerPaused(ActivePlayer)? and ActivePlayer.GetFortCharacter[].IsDownButNotOut[]):
                        ActivePlayer.ShowGamePausedUI()

                    if (not MaybeResumeCountdownStartTime?):
                        spawn{ResumeCountdown()}

    OnAgentRevived(Agent:agent):void =
        if (Player := player[Agent], (game_manager:)IsGameInProgress()?):
            if (Character := Player.GetFortCharacter[]):
                Character.SetHealth(Character.GetMaxHealth() * 0.5)

                if (IsPlayerPaused(Player)?):
                    Player.SetVulnerability(false)
                else:
                    Player.SetVulnerability(false)
                    Player.SetVulnerability(true, ?BufferTime := 3.0)
                    Player.ReactivateAbility()

        spawn{ReregisterInput(Agent)}

    ReregisterInput(Agent:agent)<suspends>:void =
        Agent.UnregisterInput(input_id.Dash)
        Sleep(TickTime * 2)
        Agent.RegisterInput(input_id.Dash)

    OnAgentEntersStageVolume(Agent:agent):void =
        if (not IsDropPodActive?) then return
        if (Player := player[Agent]):
            set PlayersInStageVolume += array{Player}
            ShowMessage(message_id.PlayerCounterShip, "{PlayersInStageVolume.Length}/{(game_manager:)GetActivePlayers().Length}", ?DisplayTime := 0.0)
            Player.SetVulnerability(false)
            Player.DeactivateAbility()

    OnAgentExitsStageVolume(Agent:agent):void =
        if (not IsDropPodActive?) then return
        if (Player := player[Agent]):
            set PlayersInStageVolume = for (PlayerInStageVolume : PlayersInStageVolume, PlayerInStageVolume <> Player) {PlayerInStageVolume}
            ShowMessage(message_id.PlayerCounterShip, "{PlayersInStageVolume.Length}/{(game_manager:)GetActivePlayers().Length}", ?DisplayTime := 0.0)
            Player.SetVulnerability(true)
            Player.ReactivateAbility()