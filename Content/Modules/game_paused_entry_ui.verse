using { /Fortnite.com/UI }
using { /Verse.org/Colors }
using { /Verse.org/Simulation }
using { /UnrealEngine.com/Temporary/SpatialMath }
using { /UnrealEngine.com/Temporary/UI }

game_paused_entry_ui := class(ui_component):
    RankIcon:texture_block = texture_block:
        DefaultImage := UserInterface.Icons.RankIcon_01_64
        DefaultDesiredSize := vector2{X := 64.0, Y := 64.0}
        
    NameText:text_block = text_block:
        DefaultTextColor := ThemeColors.PrimaryText
        DefaultShadowOffset := option{vector2{X := 2.0, Y := 2.0}}
        DefaultShadowOpacity := 0.5  

    EliminationText:text_block = text_block:
        DefaultTextColor := ThemeColors.PrimaryText
        DefaultShadowOffset := option{vector2{X := 2.0, Y := 2.0}}
        DefaultShadowOpacity := 0.5

    CheckmarkIcon:texture_block = texture_block:
        DefaultImage := UserInterface.Icons.Checkmark_64
        DefaultDesiredSize := vector2{X := 40.0, Y := 40.0}

    var WeaponWidgets:[]tuple(texture_block, text_block) = array{}

    <# ----------------------------------------------------------------------------------------------------
        Initialization
    ---------------------------------------------------------------------------------------------------- #>
    CreateUI<override>():widget =
        stack_box:
            Orientation := orientation.Horizontal
            Slots := array:
                stack_box_slot:
                    HorizontalAlignment := horizontal_alignment.Left
                    Widget := RankIcon
                stack_box_slot:
                    HorizontalAlignment := horizontal_alignment.Left
                    Padding := margin{Left := 20.0, Top := 0.0, Right := 100.0, Bottom := 0.0}
                    Distribution := option{10.0}
                    Widget := NameText
                stack_box_slot:
                    HorizontalAlignment := horizontal_alignment.Fill
                    VerticalAlignment := vertical_alignment.Fill
                    Padding := margin{Left := 0.0, Top := 0.0, Right := 100.0, Bottom := 0.0}
                    Widget := stack_box:
                        Orientation := orientation.Horizontal
                        Slots := for (Index := 1..4):
                            WeaponIcon :=  texture_block:
                                DefaultImage := UserInterface.Icons.Lock_64
                                DefaultDesiredSize := vector2{X := 40.0, Y := 40.0}
                                
                            WeaponLevelText := text_block:
                                DefaultTextColor := ThemeColors.Highlight * 2.0
                                DefaultShadowOffset := option{vector2{X := 2.0, Y := 2.0}}
                                DefaultShadowOpacity := 0.5

                            set WeaponWidgets += array{(WeaponIcon, WeaponLevelText)}

                            stack_box_slot:
                                HorizontalAlignment := horizontal_alignment.Fill
                                VerticalAlignment := vertical_alignment.Fill
                                Padding := margin{Left := (Index = 1) and 0.0 or 20.0, Top := 0.0, Right := 0.0, Bottom := 0.0}
                                Widget := stack_box:
                                    Orientation := orientation.Vertical
                                    Slots := array:
                                        stack_box_slot:
                                            VerticalAlignment := vertical_alignment.Top
                                            Widget := WeaponIcon
                                        stack_box_slot:
                                            VerticalAlignment := vertical_alignment.Bottom
                                            Widget := WeaponLevelText

                stack_box_slot:
                    HorizontalAlignment := horizontal_alignment.Left
                    Widget := texture_block:
                        DefaultImage := UserInterface.Icons.Elimination_Flat_64
                        DefaultDesiredSize := vector2{X := 40.0, Y := 40.0}
                stack_box_slot:
                    HorizontalAlignment := horizontal_alignment.Left
                    Padding := margin{Left := 10.0, Top := 0.0, Right := 100.0, Bottom := 0.0}
                    Widget := EliminationText
                stack_box_slot:
                    HorizontalAlignment := horizontal_alignment.Right
                    Widget := CheckmarkIcon

    <# ----------------------------------------------------------------------------------------------------
        Update Functions
    ---------------------------------------------------------------------------------------------------- #>
    Update(InPlayer:player):void =
        RankIcon.SetImage(InPlayer.GetRank().ToRankIcon())
        NameText.SetText(MakeMessage(InPlayer))
        EliminationText.SetText(MakeMessage("{FormatCurrency(GetEnemiesEliminated(InPlayer) * 1.0)}"))

        WeaponIds := for (WeaponId -> WeaponLevel : InPlayer.GetWeaponLevels()) {WeaponId}
        for (Index -> WeaponWidget : WeaponWidgets):
            if (WeaponId := WeaponIds[Index], WeaponInfo := GetWeaponInfo[WeaponId], Level := InPlayer.GetWeaponLevel(WeaponId)):
                WeaponWidget(0).SetImage(WeaponInfo.Icon)
                WeaponWidget(0).SetTint(NamedColors.White)
                WeaponWidget(1).SetText(MakeMessage("{Level}"))
            else:
                WeaponWidget(0).SetImage(UserInterface.Icons.Lock_64)
                WeaponWidget(0).SetTint(ThemeColors.Inactive)
                WeaponWidget(1).SetText(MakeMessage(""))

        CheckmarkIcon.SetTint(InPlayer.IsPlayerPaused()? and ThemeColors.Inactive or ThemeColors.Success)