using { /Fortnite.com/UI }
using { /Verse.org/Colors }
using { /Verse.org/Simulation }
using { /UnrealEngine.com/Temporary/SpatialMath }
using { /UnrealEngine.com/Temporary/UI }

game_result_entry_ui := class(ui_component):
    Index:int
    PlayerClickedEvent:events(tuple(player, int)) = events(tuple(player, int)){}

    PlayerWeaponIcon:texture_block = texture_block:
        DefaultImage := UserInterface.Icons.WeaponBow_64
        DefaultDesiredSize := vector2{X := 40.0, Y := 40.0}

    PlayerButton:button_quiet = button_quiet{}

    var MaybeActivePlayer:?player = false
    
    <# ----------------------------------------------------------------------------------------------------
        Initialization
    ---------------------------------------------------------------------------------------------------- #>
    CreateUI<override>():widget =
        PlayerButton.OnClick().Subscribe(OnClickPlayer)

        stack_box:
            Orientation := orientation.Horizontal
            Slots := array:
                stack_box_slot:
                    HorizontalAlignment := horizontal_alignment.Left
                    Padding := margin{Left := 0.0, Top := 0.0, Right := 10.0, Bottom := 0.0}
                    Widget := PlayerWeaponIcon
                stack_box_slot:
                    HorizontalAlignment := horizontal_alignment.Fill
                    Distribution := option{1.0}
                    Widget := PlayerButton

    <# ----------------------------------------------------------------------------------------------------
        Update Functions
    ---------------------------------------------------------------------------------------------------- #>
    Update(InPlayer:player):void =
        if (WeaponId := InPlayer.GetEquippedWeapon()?, WeaponInfo := GetWeaponInfo[WeaponId]):
            PlayerWeaponIcon.SetImage(WeaponInfo.IconSmall)

        PlayerButton.SetText(MakeMessage(InPlayer))
        set MaybeActivePlayer = option{InPlayer}

    SetActive():void =
        PlayerWeaponIcon.SetTint(ThemeColors.Highlight * 2)

    SetInactive():void =
        PlayerWeaponIcon.SetTint(NamedColors.White)

    <# ----------------------------------------------------------------------------------------------------
        Event Handlers
    ---------------------------------------------------------------------------------------------------- #>
    OnClickPlayer(Message:widget_message):void =
        if (ActivePlayer := MaybeActivePlayer?):
            PlayerClickedEvent.Signal(ActivePlayer, Index)