using { /Fortnite.com/UI }
using { /Verse.org/Colors }
using { /Verse.org/Simulation }
using { /UnrealEngine.com/Temporary }
using { /UnrealEngine.com/Temporary/SpatialMath }
using { /UnrealEngine.com/Temporary/UI }

<# ----------------------------------------------------------------------------------------------------
    Data Definitions
---------------------------------------------------------------------------------------------------- #>
game_result_type := enum:
    game_result_type.Game
    game_result_type.Team
    game_result_type.Damage
    game_result_type.Player

<# ----------------------------------------------------------------------------------------------------
    game_result_ui
---------------------------------------------------------------------------------------------------- #>
game_result_ui := class(panel_ui):
    Title<override>:string = ""
    BackgroundOpacity<override>:type {_X:float where 0.000000 <= _X, _X <= 1.000000} = 0.7
    Anchors<override>:anchors = anchors{Minimum := vector2{X := 0.0, Y := 0.0}, Maximum := vector2{X := 1.0, Y := 1.0}}
    
    CancelAnimationEvent:events(tuple()) = events(tuple()){}

    ArenaText:text_block = text_block:
        DefaultTextColor := ThemeColors.PrimaryText

    var GameUI:widget = canvas{}
    var TeamUI:widget = canvas{}
    var DamageUI:widget = canvas{}
    var PlayerUI:widget = canvas{}
    var GameResultEntryUIs:[]game_result_entry_ui = array{}

    var ActiveGameResultType:game_result_type = game_result_type.Game
    var ActiveResultsInfo:result_info = result_info{}
    var IsClaimed:logic = true
    var IsRankUp:logic = false
    
    <# ----------------------------------------------------------------------------------------------------
        Initialization
    ---------------------------------------------------------------------------------------------------- #>
    CreateContent<override>():widget =
        ClosedEvent.Subscribe(OnClosed)

        StackBox := stack_box:
            Orientation := orientation.Vertical
            Slots := array:
                
                stack_box_slot:
                    HorizontalAlignment := horizontal_alignment.Fill
                    VerticalAlignment := vertical_alignment.Fill
                    Distribution := option{1.0}
                    Widget := stack_box:
                        Orientation := orientation.Horizontal
                        Slots := array:
                            stack_box_slot:
                                HorizontalAlignment := horizontal_alignment.Fill
                                VerticalAlignment := vertical_alignment.Fill
                                Distribution := option{0.25}
                                Widget := stack_box:
                                    Orientation := orientation.Vertical
                                    Slots := array:
                                        stack_box_slot:
                                            HorizontalAlignment := horizontal_alignment.Fill
                                            VerticalAlignment := vertical_alignment.Fill
                                            Padding := margin{Left := 0.0, Top := 0.0, Right := 0.0, Bottom := 20.0}
                                            Widget := overlay:
                                                Slots := array:
                                                    overlay_slot:
                                                        HorizontalAlignment := horizontal_alignment.Fill
                                                        VerticalAlignment := vertical_alignment.Fill
                                                        Widget := MakeBox()
                                                    overlay_slot:
                                                        Padding := margin{Left := 20.0, Top := 20.0, Right := 20.0, Bottom := 20.0}
                                                        Widget := ArenaText
                                        stack_box_slot:
                                            HorizontalAlignment := horizontal_alignment.Fill
                                            VerticalAlignment := vertical_alignment.Fill
                                            Distribution := option{1.0}
                                            Widget := CreateMenuUI()
                            stack_box_slot:
                                HorizontalAlignment := horizontal_alignment.Fill
                                VerticalAlignment := vertical_alignment.Fill
                                Distribution := option{1.0}
                                Padding := margin{Left := 20.0, Top := 0.0, Right := 0.0, Bottom := 0.0}
                                Widget := overlay:
                                    Slots := array:
                                        overlay_slot:
                                            HorizontalAlignment := horizontal_alignment.Fill
                                            VerticalAlignment := vertical_alignment.Fill
                                            Widget := set GameUI = CreateGameUI()
                                        overlay_slot:
                                            HorizontalAlignment := horizontal_alignment.Fill
                                            VerticalAlignment := vertical_alignment.Fill
                                            Widget := set TeamUI = CreateTeamUI()
                                        overlay_slot:
                                            HorizontalAlignment := horizontal_alignment.Fill
                                            VerticalAlignment := vertical_alignment.Fill
                                            Widget := set DamageUI = CreateDamageUI()
                                        overlay_slot:
                                            HorizontalAlignment := horizontal_alignment.Fill
                                            VerticalAlignment := vertical_alignment.Fill
                                            Widget := set PlayerUI = CreatePlayerUI()
        
        TeamUI.SetVisibility(widget_visibility.Collapsed)
        PlayerUI.SetVisibility(widget_visibility.Collapsed)
        DamageUI.SetVisibility(widget_visibility.Collapsed)

        return StackBox
               
    <# ----------------------------------------------------------------------------------------------------
        Menu UI
    ---------------------------------------------------------------------------------------------------- #>
    GameIcon:texture_block = texture_block:
        DefaultImage := UserInterface.Icons.AbilityCooldown_64
        DefaultDesiredSize := vector2{X := 40.0, Y := 40.0}

    TeamIcon:texture_block = texture_block:
        DefaultImage := UserInterface.Icons.Team_64
        DefaultDesiredSize := vector2{X := 40.0, Y := 40.0}

    DamageIcon:texture_block = texture_block:
        DefaultImage := UserInterface.Icons.ProjectileDamage_64
        DefaultDesiredSize := vector2{X := 40.0, Y := 40.0}

    CreateMenuUI():widget =
        GameButton := button_quiet{DefaultText := MakeMessage("Game")}
        GameButton.OnClick().Subscribe(OnClickGame)
        TeamButton := button_quiet{DefaultText := MakeMessage("Team")}
        TeamButton.OnClick().Subscribe(OnClickTeam)
        DamageButton := button_quiet{DefaultText := MakeMessage("Damage")}
        DamageButton.OnClick().Subscribe(OnClickDamage)
        
        overlay:
            Slots := array:
                overlay_slot:
                    HorizontalAlignment := horizontal_alignment.Fill
                    VerticalAlignment := vertical_alignment.Fill
                    Widget := MakeBox()
                overlay_slot:
                    HorizontalAlignment := horizontal_alignment.Fill
                    VerticalAlignment := vertical_alignment.Fill
                    Padding := margin{Left := 40.0, Top := 40.0, Right := 40.0, Bottom := 40.0}
                    Widget := stack_box:
                        Orientation := orientation.Vertical
                        Slots := array:
                            stack_box_slot:
                                HorizontalAlignment := horizontal_alignment.Fill
                                VerticalAlignment := vertical_alignment.Fill
                                Widget := stack_box:
                                    Orientation := orientation.Horizontal
                                    Slots := array:
                                        stack_box_slot:
                                            HorizontalAlignment := horizontal_alignment.Left
                                            Padding := margin{Left := 0.0, Top := 0.0, Right := 10.0, Bottom := 0.0}
                                            Widget := GameIcon
                                        stack_box_slot:
                                            HorizontalAlignment := horizontal_alignment.Fill
                                            Distribution := option{1.0}
                                            Widget := GameButton
                            stack_box_slot:
                                HorizontalAlignment := horizontal_alignment.Fill
                                VerticalAlignment := vertical_alignment.Fill
                                Padding := margin{Left := 0.0, Top := 10.0, Right := 0.0, Bottom := 0.0}
                                Widget := stack_box:
                                    Orientation := orientation.Horizontal
                                    Slots := array:
                                        stack_box_slot:
                                            HorizontalAlignment := horizontal_alignment.Left
                                            Padding := margin{Left := 0.0, Top := 0.0, Right := 10.0, Bottom := 0.0}
                                            Widget := TeamIcon
                                        stack_box_slot:
                                            HorizontalAlignment := horizontal_alignment.Fill
                                            Distribution := option{1.0}
                                            Widget := TeamButton
                            stack_box_slot:
                                HorizontalAlignment := horizontal_alignment.Fill
                                VerticalAlignment := vertical_alignment.Fill
                                Padding := margin{Left := 0.0, Top := 10.0, Right := 0.0, Bottom := 0.0}
                                Widget := stack_box:
                                    Orientation := orientation.Horizontal
                                    Slots := array:
                                        stack_box_slot:
                                            HorizontalAlignment := horizontal_alignment.Left
                                            Padding := margin{Left := 0.0, Top := 0.0, Right := 10.0, Bottom := 0.0}
                                            Widget := DamageIcon
                                        stack_box_slot:
                                            HorizontalAlignment := horizontal_alignment.Fill
                                            Distribution := option{1.0}
                                            Widget := DamageButton
                            stack_box_slot:
                                HorizontalAlignment := horizontal_alignment.Fill
                                VerticalAlignment := vertical_alignment.Fill
                                Padding := margin{Left := 0.0, Top := 10.0, Right := 0.0, Bottom := 0.0}
                                Widget := stack_box:
                                    Orientation := orientation.Vertical
                                    Slots := for (Index := 0..MaxPlayers - 1):
                                        GameResultEntryUI := game_result_entry_ui{Index := Index, Player := Player}
                                        GameResultEntryUI.Initialize()
                                        GameResultEntryUI.PlayerClickedEvent.Subscribe(OnPlayerClicked)
                                        set GameResultEntryUIs += array{GameResultEntryUI}
                        
                                        stack_box_slot:
                                            HorizontalAlignment := horizontal_alignment.Fill
                                            VerticalAlignment := vertical_alignment.Fill
                                            Padding := margin{Left := 0.0, Top := (Index = 0) and 0.0 or 10.0, Right := 0.0, Bottom := 0.0}
                                            Widget := GameResultEntryUI.GetUI()

    <# ----------------------------------------------------------------------------------------------------
        Game UI
    ---------------------------------------------------------------------------------------------------- #>
    TimeSurvivedText:text_block = text_block:
        DefaultTextColor := ThemeColors.Success
        DefaultShadowOffset := option{vector2{X := 2.0, Y := 2.0}}
        DefaultShadowOpacity := 1.0
        DefaultShadowColor := ThemeColors.Success * 0.125
        
    EnemiesEliminatedText:text_block = text_block:
        DefaultTextColor := ThemeColors.Success
        DefaultShadowOffset := option{vector2{X := 2.0, Y := 2.0}}
        DefaultShadowOpacity := 1.0
        DefaultShadowColor := ThemeColors.Success * 0.125
        
    LevelReachedText:text_block = text_block:
        DefaultTextColor := ThemeColors.Success
        DefaultShadowOffset := option{vector2{X := 2.0, Y := 2.0}}
        DefaultShadowOpacity := 1.0
        DefaultShadowColor := ThemeColors.Success * 0.125
        
    WoodCollectedText:text_block = text_block:
        DefaultTextColor := ThemeColors.Success
        DefaultShadowOffset := option{vector2{X := 2.0, Y := 2.0}}
        DefaultShadowOpacity := 1.0
        DefaultShadowColor := ThemeColors.Success * 0.125
       
    GoldFromWoodCollectedText:text_block = text_block:
        DefaultTextColor := ThemeColors.Success
        DefaultShadowOffset := option{vector2{X := 2.0, Y := 2.0}}
        DefaultShadowOpacity := 1.0
        DefaultShadowColor := ThemeColors.Success * 0.125

    GoldFromTimeSurvivedText:text_block = text_block:
        DefaultTextColor := ThemeColors.Success
        DefaultShadowOffset := option{vector2{X := 2.0, Y := 2.0}}
        DefaultShadowOpacity := 1.0
        DefaultShadowColor := ThemeColors.Success * 0.125

    GoldFromVictoryText:text_block = text_block:
        DefaultTextColor := ThemeColors.Success
        DefaultShadowOffset := option{vector2{X := 2.0, Y := 2.0}}
        DefaultShadowOpacity := 1.0
        DefaultShadowColor := ThemeColors.Success * 0.125

    GoldFromObjectivesCompletedText:text_block = text_block:
        DefaultTextColor := ThemeColors.Success
        DefaultShadowOffset := option{vector2{X := 2.0, Y := 2.0}}
        DefaultShadowOpacity := 1.0
        DefaultShadowColor := ThemeColors.Success * 0.125

    GoldFromStagesCompletedText:text_block = text_block:
        DefaultTextColor := ThemeColors.Success
        DefaultShadowOffset := option{vector2{X := 2.0, Y := 2.0}}
        DefaultShadowOpacity := 1.0
        DefaultShadowColor := ThemeColors.Success * 0.125

    GoldMultiplierText:text_block = text_block:
        DefaultTextColor := ThemeColors.Success
        DefaultShadowOffset := option{vector2{X := 2.0, Y := 2.0}}
        DefaultShadowOpacity := 1.0
        DefaultShadowColor := ThemeColors.Success * 0.125
    
    GoldEarnedText:text_block = text_block:
        DefaultTextColor := ThemeColors.Success
        DefaultShadowOffset := option{vector2{X := 2.0, Y := 2.0}}
        DefaultShadowOpacity := 1.0
        DefaultShadowColor := ThemeColors.Success * 0.125

    ObjectivesCompletedText:text_block = text_block:
        DefaultTextColor := ThemeColors.Success
        DefaultShadowOffset := option{vector2{X := 2.0, Y := 2.0}}
        DefaultShadowOpacity := 1.0
        DefaultShadowColor := ThemeColors.Success * 0.125

    StagesCompletedText:text_block = text_block:
        DefaultTextColor := ThemeColors.Success
        DefaultShadowOffset := option{vector2{X := 2.0, Y := 2.0}}
        DefaultShadowOpacity := 1.0
        DefaultShadowColor := ThemeColors.Success * 0.125

    CreateGameUI():widget =
        overlay:
            Slots := array:
                overlay_slot:
                    HorizontalAlignment := horizontal_alignment.Fill
                    VerticalAlignment := vertical_alignment.Fill
                    Widget := MakeBox()
                overlay_slot:
                    HorizontalAlignment := horizontal_alignment.Fill
                    VerticalAlignment := vertical_alignment.Fill
                    Padding := margin{Left := 40.0, Top := 40.0, Right := 40.0, Bottom := 40.0}
                    Widget := stack_box:
                        Orientation := orientation.Vertical
                        Slots := array:
                            stack_box_slot:
                                HorizontalAlignment := horizontal_alignment.Fill
                                VerticalAlignment := vertical_alignment.Fill
                                Padding := margin{Left := 20.0, Top := 0.0, Right := 20.0, Bottom := 20.0}
                                Widget := stack_box:
                                    Orientation := orientation.Horizontal
                                    Slots := array:
                                        stack_box_slot:
                                            HorizontalAlignment := horizontal_alignment.Left
                                            Distribution := option{2.0}
                                            Padding := margin{Left := 0.0, Top := 0.0, Right := 100.0, Bottom := 0.0}
                                            Widget := text_block:
                                                DefaultText := MakeMessage("Results")
                                                DefaultTextColor := ThemeColors.PrimaryText
                                                DefaultTextOpacity := 0.5
                                        stack_box_slot:
                                            HorizontalAlignment := horizontal_alignment.Right
                                            Distribution := option{1.0}
                                            Widget := text_block:
                                                DefaultText := MakeMessage("Score")
                                                DefaultTextColor := ThemeColors.PrimaryText
                                                DefaultTextOpacity := 0.5
                                        stack_box_slot:
                                            HorizontalAlignment := horizontal_alignment.Right
                                            Distribution := option{1.0}
                                            Widget := text_block:
                                                DefaultText := MakeMessage("Gold")
                                                DefaultTextColor := ThemeColors.PrimaryText
                                                DefaultTextOpacity := 0.5
                            stack_box_slot{HorizontalAlignment := horizontal_alignment.Fill, Padding := margin{Left := 20.0, Right := 20.0}, Widget := CreateStatUI(TimeSurvivedText, GoldFromTimeSurvivedText, "Time Survived")}
                            stack_box_slot{HorizontalAlignment := horizontal_alignment.Fill, Padding := margin{Left := 20.0, Right := 20.0}, Widget := CreateStatUI(EnemiesEliminatedText, text_block{DefaultText := MakeMessage("")}, "Enemies Eliminated")}
                            stack_box_slot{HorizontalAlignment := horizontal_alignment.Fill, Padding := margin{Left := 20.0, Right := 20.0}, Widget := CreateStatUI(LevelReachedText, text_block{DefaultText := MakeMessage("")}, "Level Reached")}
                            stack_box_slot{HorizontalAlignment := horizontal_alignment.Fill, Padding := margin{Left := 20.0, Right := 20.0}, Widget := CreateStatUI(WoodCollectedText, GoldFromWoodCollectedText, "Wood Collected")}
                            stack_box_slot{HorizontalAlignment := horizontal_alignment.Fill, Padding := margin{Left := 20.0, Right := 20.0}, Widget := CreateStatUI(ObjectivesCompletedText, GoldFromObjectivesCompletedText, "Missions Completed")}
                            stack_box_slot{HorizontalAlignment := horizontal_alignment.Fill, Padding := margin{Left := 20.0, Right := 20.0}, Widget := CreateStatUI(StagesCompletedText, GoldFromStagesCompletedText, "Stages Completed")}
                            

                            stack_box_slot:
                                HorizontalAlignment := horizontal_alignment.Fill
                                VerticalAlignment := vertical_alignment.Bottom
                                Distribution := option{1.0}
                                Padding := margin{Left := 20.0, Right := 20.0}
                                Widget := CreateStatUI(GoldFromVictoryText, "Victory Bonus")
                            stack_box_slot{HorizontalAlignment := horizontal_alignment.Fill, Padding := margin{Left := 20.0, Right := 20.0}, Widget := CreateStatUI(GoldMultiplierText, "Gold Multiplier")}
                            stack_box_slot:
                                HorizontalAlignment := horizontal_alignment.Fill
                                Padding := margin{Top := 20.0}
                                Widget := overlay:
                                    Slots := array:
                                        overlay_slot:
                                            HorizontalAlignment := horizontal_alignment.Fill
                                            VerticalAlignment := vertical_alignment.Fill
                                            Widget := color_block:
                                                DefaultColor := NamedColors.White
                                                DefaultOpacity := 0.1
                                        overlay_slot:
                                            HorizontalAlignment := horizontal_alignment.Fill
                                            VerticalAlignment := vertical_alignment.Fill
                                            Padding := margin{Left := 20.0, Top := 20.0, Right := 20.0, Bottom := 20.0}
                                            Widget := CreateStatUI(GoldEarnedText, "Gold Earned", ?Icon := option{currency_id.Gold.ToIcon()})

    <# ----------------------------------------------------------------------------------------------------
        Team UI
    ---------------------------------------------------------------------------------------------------- #>
    var GameResultTeamEntryUIs:[]game_result_team_entry_ui = array{}

    CreateTeamUI():widget =
        overlay:
            Slots := array:
                overlay_slot:
                    HorizontalAlignment := horizontal_alignment.Fill
                    VerticalAlignment := vertical_alignment.Fill
                    Widget := MakeBox()
                overlay_slot:
                    HorizontalAlignment := horizontal_alignment.Fill
                    VerticalAlignment := vertical_alignment.Fill
                    Padding := margin{Left := 40.0, Top := 20.0, Right := 40.0, Bottom := 20.0}
                    Widget := stack_box:
                        Orientation := orientation.Horizontal
                        Slots := array:
                            stack_box_slot:
                                HorizontalAlignment := horizontal_alignment.Fill
                                VerticalAlignment := vertical_alignment.Fill
                                Padding := margin{Left := 0.0, Top := 40.0, Right := 20.0, Bottom := 40.0}
                                Widget := stack_box:
                                    Orientation := orientation.Vertical
                                    Slots := array:
                                        stack_box_slot:
                                            HorizontalAlignment := horizontal_alignment.Left
                                            VerticalAlignment := vertical_alignment.Fill
                                            Distribution := option{0.6}
                                            Widget := text_block:
                                                DefaultText := MakeMessage("Player")
                                                DefaultTextColor := ThemeColors.PrimaryText
                                                DefaultTextOpacity := 0.5
                                        stack_box_slot:
                                            HorizontalAlignment := horizontal_alignment.Left
                                            VerticalAlignment := vertical_alignment.Fill
                                            Distribution := option{1.0}
                                            Widget := text_block:
                                                DefaultText := MakeMessage("Weapons")
                                                DefaultTextColor := ThemeColors.PrimaryText
                                                DefaultTextOpacity := 0.5
                                        stack_box_slot:
                                            HorizontalAlignment := horizontal_alignment.Left
                                            VerticalAlignment := vertical_alignment.Fill
                                            Widget := text_block:
                                                DefaultText := MakeMessage("Eliminate")
                                                DefaultTextColor := ThemeColors.PrimaryText
                                                DefaultTextOpacity := 0.5
                                        stack_box_slot:
                                            HorizontalAlignment := horizontal_alignment.Left
                                            VerticalAlignment := vertical_alignment.Fill
                                            Widget := text_block:
                                                DefaultText := MakeMessage("Damage")
                                                DefaultTextColor := ThemeColors.PrimaryText
                                                DefaultTextOpacity := 0.5
                                        stack_box_slot:
                                            HorizontalAlignment := horizontal_alignment.Left
                                            VerticalAlignment := vertical_alignment.Fill
                                            Widget := text_block:
                                                DefaultText := MakeMessage("Wood")
                                                DefaultTextColor := ThemeColors.PrimaryText
                                                DefaultTextOpacity := 0.5
                                        stack_box_slot:
                                            HorizontalAlignment := horizontal_alignment.Left
                                            VerticalAlignment := vertical_alignment.Fill
                                            Widget := text_block:
                                                DefaultText := MakeMessage("Revived")
                                                DefaultTextColor := ThemeColors.PrimaryText
                                                DefaultTextOpacity := 0.5
                                        stack_box_slot:
                                            HorizontalAlignment := horizontal_alignment.Left
                                            VerticalAlignment := vertical_alignment.Fill
                                            Widget := text_block:
                                                DefaultText := MakeMessage("Downed")
                                                DefaultTextColor := ThemeColors.PrimaryText
                                                DefaultTextOpacity := 0.5
                                                
                            stack_box_slot:
                                HorizontalAlignment := horizontal_alignment.Fill
                                VerticalAlignment := vertical_alignment.Fill
                                Distribution := option{1.0}
                                Padding := margin{Left := 0.0, Top := 0.0, Right := 0.0, Bottom := 0.0}
                                Widget := stack_box:
                                    Orientation := orientation.Horizontal
                                    Slots := for (Index := 0..MaxPlayers - 1):
                                        GameResultTeamEntryUI := game_result_team_entry_ui{Player := Player}
                                        GameResultTeamEntryUI.Initialize()
                                        set GameResultTeamEntryUIs += array{GameResultTeamEntryUI}

                                        stack_box_slot:
                                            HorizontalAlignment := horizontal_alignment.Fill
                                            VerticalAlignment := vertical_alignment.Fill
                                            Distribution := option{1.0}
                                            Padding := margin{Left := (Index = 0) and 0.0 or 20.0, Top := 0.0, Right := 0.0, Bottom := 0.0}
                                            Widget := overlay:
                                                Slots := array:
                                                    overlay_slot:
                                                        HorizontalAlignment := horizontal_alignment.Fill
                                                        VerticalAlignment := vertical_alignment.Fill
                                                        Widget := MakeBox(?BorderRadius := 4.0)
                                                    overlay_slot:
                                                        HorizontalAlignment := horizontal_alignment.Fill
                                                        VerticalAlignment := vertical_alignment.Fill
                                                        Padding := margin{Left := 20.0, Top := 40.0, Right := 20.0, Bottom := 40.0}
                                                        Widget := GameResultTeamEntryUI.GetUI()

    <# ----------------------------------------------------------------------------------------------------
        Damage UI
    ---------------------------------------------------------------------------------------------------- #>
    var GameResultDamageEntryUIs:[]game_result_damage_entry_ui = array{}

    CreateDamageUI():widget =
        overlay:
            Slots := array:
                overlay_slot:
                    HorizontalAlignment := horizontal_alignment.Fill
                    VerticalAlignment := vertical_alignment.Fill
                    Widget := MakeBox()
                overlay_slot:
                    HorizontalAlignment := horizontal_alignment.Fill
                    VerticalAlignment := vertical_alignment.Fill
                    Padding := margin{Left := 40.0, Top := 40.0, Right := 40.0, Bottom := 40.0}
                    Widget := stack_box:
                        Orientation := orientation.Vertical
                        Slots := array:
                            stack_box_slot:
                                HorizontalAlignment := horizontal_alignment.Fill
                                VerticalAlignment := vertical_alignment.Fill
                                Padding := margin{Left := 0.0, Top := 0.0, Right := 0.0, Bottom := 20.0}
                                Widget := stack_box:
                                    Orientation := orientation.Horizontal
                                    Slots := array:
                                        stack_box_slot:
                                            HorizontalAlignment := horizontal_alignment.Left
                                            VerticalAlignment := vertical_alignment.Fill
                                            Distribution := option{1.0}
                                            Widget := text_block:
                                                DefaultText := MakeMessage("Player")
                                                DefaultTextColor := ThemeColors.PrimaryText
                                                DefaultTextOpacity := 0.5
                                        stack_box_slot:
                                            HorizontalAlignment := horizontal_alignment.Center
                                            VerticalAlignment := vertical_alignment.Fill
                                            Distribution := option{0.5}
                                            Widget := text_block:
                                                DefaultText := MakeMessage("Total Damage")
                                                DefaultTextColor := ThemeColors.PrimaryText
                                                DefaultTextOpacity := 0.5
                                        stack_box_slot:
                                            HorizontalAlignment := horizontal_alignment.Center
                                            VerticalAlignment := vertical_alignment.Fill
                                            Distribution := option{0.5}
                                            Widget := text_block:
                                                DefaultText := MakeMessage("Miniboss")
                                                DefaultTextColor := ThemeColors.PrimaryText
                                                DefaultTextOpacity := 0.5
                                        stack_box_slot:
                                            HorizontalAlignment := horizontal_alignment.Center
                                            VerticalAlignment := vertical_alignment.Fill
                                            Distribution := option{0.5}
                                            Widget := text_block:
                                                DefaultText := MakeMessage("Boss")
                                                DefaultTextColor := ThemeColors.PrimaryText
                                                DefaultTextOpacity := 0.5
                                                
                            stack_box_slot:
                                HorizontalAlignment := horizontal_alignment.Fill
                                VerticalAlignment := vertical_alignment.Fill
                                Distribution := option{1.0}
                                Padding := margin{Left := 0.0, Top := 0.0, Right := 0.0, Bottom := 0.0}
                                Widget := stack_box:
                                    Orientation := orientation.Vertical
                                    Slots := for (Index := 0..MaxPlayers - 1):
                                        GameResultDamageEntryUI := game_result_damage_entry_ui{Player := Player}
                                        GameResultDamageEntryUI.Initialize()
                                        set GameResultDamageEntryUIs += array{GameResultDamageEntryUI}

                                        stack_box_slot:
                                            HorizontalAlignment := horizontal_alignment.Fill
                                            VerticalAlignment := vertical_alignment.Fill
                                            Padding := if (Index = 0):
                                                margin{Left := 0.0, Top := 0.0, Right := 0.0, Bottom := 0.0}
                                            else:
                                                margin{Left := 20.0, Top := 20.0, Right := 20.0, Bottom := 0.0}

                                            Widget := if (Index = 0):
                                                overlay:
                                                    Slots := array:
                                                        overlay_slot:
                                                            HorizontalAlignment := horizontal_alignment.Fill
                                                            VerticalAlignment := vertical_alignment.Fill
                                                            Widget := color_block:
                                                                DefaultColor := NamedColors.White
                                                                DefaultOpacity := 0.1
                                                        overlay_slot:
                                                            HorizontalAlignment := horizontal_alignment.Fill
                                                            VerticalAlignment := vertical_alignment.Fill
                                                            Padding := margin{Left := 20.0, Top := 20.0, Right := 20.0, Bottom := 20.0}
                                                            Widget := GameResultDamageEntryUI.GetUI()
                                            else:
                                                GameResultDamageEntryUI.GetUI()

    <# ----------------------------------------------------------------------------------------------------
        Player UI
    ---------------------------------------------------------------------------------------------------- #>
    var AbilityTexts:[ability_id]text_block = map{}
    var AbilityTotalTexts:[ability_id]text_block = map{}
    var GameResultWeaponEntryUIs:[]game_result_weapon_entry_ui = array{}
    var BoonIcons:[]texture_block = array{}

    TeamExperienceMultiplierText:text_block = text_block:
        DefaultTextColor := ThemeColors.Success
        DefaultShadowColor := ThemeColors.Success * 0.125
        DefaultShadowOffset := option{vector2{X := 2.0, Y := 2.0}}
        DefaultShadowOpacity := 1.0

    CreatePlayerUI():widget =
        stack_box:
            Orientation := orientation.Horizontal
            Slots := array:
                stack_box_slot:
                    HorizontalAlignment := horizontal_alignment.Fill
                    VerticalAlignment := vertical_alignment.Fill
                    Distribution := option{1.0}
                    Widget := CreateAbilitiesUI()
                stack_box_slot:
                    HorizontalAlignment := horizontal_alignment.Fill
                    VerticalAlignment := vertical_alignment.Fill
                    Distribution := option{1.2}
                    Padding := margin{Left := 20.0, Top := 0.0, Right := 0.0, Bottom := 0.0}
                    Widget := stack_box:
                        Orientation := orientation.Vertical
                        Slots := array:
                            stack_box_slot:
                                HorizontalAlignment := horizontal_alignment.Fill
                                VerticalAlignment := vertical_alignment.Fill
                                Distribution := option{1.0}
                                Padding := margin{Left := 0.0, Top := 0.0, Right := 0.0, Bottom := 20.0}
                                Widget := CreateWeaponsUI()
                            stack_box_slot:
                                HorizontalAlignment := horizontal_alignment.Fill
                                VerticalAlignment := vertical_alignment.Fill
                                Widget := CreateBoonsUI()

    CreateAbilitiesUI():widget =
        AbilitiesStackBox := stack_box:
            Orientation := orientation.Vertical
            Slots := array:
                stack_box_slot:
                    HorizontalAlignment := horizontal_alignment.Fill
                    VerticalAlignment := vertical_alignment.Fill
                    Padding := margin{Left := 0.0, Top := 0.0, Right := 0.0, Bottom := 20.0}
                    Widget := text_block:
                        DefaultText := MakeMessage("Abilities")
                        DefaultTextColor := ThemeColors.PrimaryText
                        DefaultTextOpacity := 0.5
                stack_box_slot:
                    HorizontalAlignment := horizontal_alignment.Fill
                    VerticalAlignment := vertical_alignment.Fill
                    Widget := stack_box:
                        Orientation := orientation.Vertical
                        Slots := 
                            var Index:int = 0
                            for (AbilityInfo : GetAbilityInfo()):
                                AbilityText := text_block:
                                    DefaultShadowOffset := option{vector2{X := 2.0, Y := 2.0}}
                                    DefaultShadowColor := ThemeColors.Success * 0.125

                                AbilityTotalText := text_block:
                                    DefaultTextColor := ThemeColors.Success
                                    DefaultShadowOpacity := 1.0
                                    DefaultShadowOffset := option{vector2{X := 2.0, Y := 2.0}}
                                    DefaultShadowColor := ThemeColors.Success * 0.125
                    
                                if (set AbilityTexts[AbilityInfo.Id] = AbilityText) {}
                                if (set AbilityTotalTexts[AbilityInfo.Id] = AbilityTotalText) {}
                                set Index += 1
                
                                stack_box_slot:
                                    HorizontalAlignment := horizontal_alignment.Fill
                                    Padding := margin{Left := 0.0, Top := (Index = 1) and 0.0 or 4.0, Right := 0.0, Bottom := 0.0}
                                    Widget := stack_box:
                                        Orientation := orientation.Horizontal
                                        Slots := array:
                                            stack_box_slot:
                                                HorizontalAlignment := horizontal_alignment.Left
                                                VerticalAlignment := vertical_alignment.Center
                                                Padding := margin{Left := 0.0, Top := 0.0, Right := 10.0, Bottom := 0.0}
                                                Widget := texture_block:
                                                    DefaultImage := AbilityInfo.IconSmall
                                                    DefaultDesiredSize := vector2{X := 40.0, Y := 40.0}
                                            stack_box_slot:
                                                HorizontalAlignment := horizontal_alignment.Left
                                                Distribution := option{0.5}
                                                Widget := text_block:
                                                    DefaultText := MakeMessage("{AbilityInfo.ShortName}")
                                                    DefaultTextColor := ThemeColors.PrimaryText
                                            stack_box_slot:
                                                HorizontalAlignment := horizontal_alignment.Right
                                                Distribution := option{1.0}
                                                Widget := AbilityTotalText
                                            stack_box_slot:
                                                HorizontalAlignment := horizontal_alignment.Right
                                                Distribution := option{1.0}
                                                Widget := AbilityText

        AbilitiesStackBox.AddWidget of stack_box_slot:
            HorizontalAlignment := horizontal_alignment.Fill
            VerticalAlignment := vertical_alignment.Bottom
            Distribution := option{1.0}
            Widget := stack_box:
                Orientation := orientation.Vertical
                Slots := array:
                    stack_box_slot:
                        HorizontalAlignment := horizontal_alignment.Fill
                        VerticalAlignment := vertical_alignment.Bottom
                        Distribution := option{1.0}
                        Padding := margin{Left := 0.0, Top := 20.0, Right := 0.0, Bottom := 20.0}
                        Widget := color_block:
                            DefaultColor := NamedColors.White
                            DefaultOpacity := 0.1
                            DefaultDesiredSize := vector2{X := 0.0, Y := 1.0}
                    stack_box_slot:
                        HorizontalAlignment := horizontal_alignment.Fill
                        VerticalAlignment := vertical_alignment.Fill
                        Widget := stack_box:
                            Orientation := orientation.Horizontal
                            Slots := array:
                                stack_box_slot:
                                    HorizontalAlignment := horizontal_alignment.Left
                                    Padding := margin{Left := 0.0, Top := 0.0, Right := 10.0, Bottom := 0.0}
                                    Widget := texture_block:
                                        DefaultImage := UserInterface.Icons.ExperienceMultiplier_64
                                        DefaultDesiredSize := vector2{X := 40.0, Y := 40.0}
                                stack_box_slot:
                                    HorizontalAlignment := horizontal_alignment.Fill
                                    Widget := text_block:
                                        DefaultText := MakeMessage("Team Multiplier")
                                        DefaultTextColor := ThemeColors.PrimaryText
                                stack_box_slot:
                                    HorizontalAlignment := horizontal_alignment.Right
                                    Distribution := option{1.0}
                                    Widget := TeamExperienceMultiplierText

        overlay:
            Slots := array:
                overlay_slot:
                    HorizontalAlignment := horizontal_alignment.Fill
                    VerticalAlignment := vertical_alignment.Fill
                    Widget := MakeBox(?Width := 600.0)
                overlay_slot:
                    HorizontalAlignment := horizontal_alignment.Fill
                    VerticalAlignment := vertical_alignment.Fill
                    Padding := margin{Left := 40.0, Top := 40.0, Right := 40.0, Bottom := 40.0}
                    Widget := AbilitiesStackBox

    CreateWeaponsUI():widget =
        overlay:
            Slots := array:
                overlay_slot:
                    HorizontalAlignment := horizontal_alignment.Fill
                    VerticalAlignment := vertical_alignment.Fill
                    Widget := MakeBox()
                overlay_slot:
                    HorizontalAlignment := horizontal_alignment.Fill
                    VerticalAlignment := vertical_alignment.Fill
                    Padding := margin{Left := 40.0, Top := 40.0, Right := 40.0, Bottom := 40.0}
                    Widget := stack_box:
                        Orientation := orientation.Vertical
                        Slots := array:
                            stack_box_slot:
                                HorizontalAlignment := horizontal_alignment.Fill
                                VerticalAlignment := vertical_alignment.Fill
                                Padding := margin{Left := 0.0, Top := 0.0, Right := 0.0, Bottom := 20.0}
                                Widget := stack_box:
                                    Orientation := orientation.Horizontal
                                    Slots := array:
                                        stack_box_slot:
                                            HorizontalAlignment := horizontal_alignment.Left
                                            Distribution := option{2.0}
                                            Widget := text_block:
                                                DefaultText := MakeMessage("Weapons")
                                                DefaultTextColor := ThemeColors.PrimaryText
                                                DefaultTextOpacity := 0.5
                                        stack_box_slot:
                                            Padding := margin{Left := 21.0, Top := 0.0, Right := 0.0, Bottom := 0.0}
                                            Widget := color_block:
                                                DefaultOpacity := 0.0
                                        stack_box_slot:
                                            HorizontalAlignment := horizontal_alignment.Left
                                            Distribution := option{1.0}
                                            Widget := texture_block:
                                                DefaultImage := UserInterface.Icons.ProjectileDamage_64
                                                DefaultTint := ThemeColors.WeaponDamage
                                                DefaultDesiredSize := vector2{X := 40.0, Y := 40.0}
                                        stack_box_slot:
                                            HorizontalAlignment := horizontal_alignment.Left
                                            Distribution := option{1.0}
                                            Widget := texture_block:
                                                DefaultImage := UserInterface.Icons.AbilityCooldown_64
                                                DefaultTint := ThemeColors.WeaponCooldown
                                                DefaultDesiredSize := vector2{X := 40.0, Y := 40.0}
                                        stack_box_slot:
                                            HorizontalAlignment := horizontal_alignment.Left
                                            Distribution := option{1.0}
                                            Widget := texture_block:
                                                DefaultImage := UserInterface.Icons.AbilitySize_64
                                                DefaultTint := ThemeColors.WeaponSize
                                                DefaultDesiredSize := vector2{X := 40.0, Y := 40.0}
                            stack_box_slot:
                                HorizontalAlignment := horizontal_alignment.Fill
                                VerticalAlignment := vertical_alignment.Fill

                                Widget := stack_box:
                                    Orientation := orientation.Vertical
                                    Slots := for (Index := 1..4):
                                        GameResultWeaponEntryUI := game_result_weapon_entry_ui{Player := Player}
                                        GameResultWeaponEntryUI.Initialize()
                                        set GameResultWeaponEntryUIs += array{GameResultWeaponEntryUI}

                                        stack_box_slot:
                                            HorizontalAlignment := horizontal_alignment.Fill
                                            VerticalAlignment := vertical_alignment.Fill
                                            Padding := margin{Left := 0.0, Top := (Index = 1) and 0.0 or 30.0, Right := 0.0, Bottom := 0.0}
                                            Widget := GameResultWeaponEntryUI.GetUI()

    CreateBoonsUI():widget =
        overlay:
            Slots := array:
                overlay_slot:
                    HorizontalAlignment := horizontal_alignment.Fill
                    VerticalAlignment := vertical_alignment.Fill
                    Widget := MakeBox()
                overlay_slot:
                    HorizontalAlignment := horizontal_alignment.Fill
                    VerticalAlignment := vertical_alignment.Fill
                    Padding := margin{Left := 40.0, Top := 20.0, Right := 40.0, Bottom := 20.0}
                    Widget := stack_box:
                        Orientation := orientation.Horizontal
                        Slots := array:
                            stack_box_slot:
                                HorizontalAlignment := horizontal_alignment.Fill
                                VerticalAlignment := vertical_alignment.Center
                                Distribution := option{1.5}
                                Widget := text_block:
                                    DefaultText := MakeMessage("Boons")
                                    DefaultTextColor := ThemeColors.PrimaryText
                                    DefaultTextOpacity := 0.5
                            stack_box_slot:
                                HorizontalAlignment := horizontal_alignment.Fill
                                VerticalAlignment := vertical_alignment.Fill
                                Distribution := option{5.0}
                                Padding := margin{Left := 0.0, Top := 0.0, Right := 0.0, Bottom := 0.0}
                                Widget := stack_box:
                                    Orientation := orientation.Horizontal
                                    Slots := for (Index := 1..5):
                                        BoonIcon := texture_block:
                                            DefaultImage := UserInterface.Icons.Boon_64
                                            DefaultDesiredSize := vector2{X := 64.0, Y := 64.0}

                                        set BoonIcons += array{BoonIcon}

                                        stack_box_slot:
                                            Distribution := option{1.0}
                                            Widget := BoonIcon

    <# ----------------------------------------------------------------------------------------------------
        Get Functions
    ---------------------------------------------------------------------------------------------------- #>
    GetUI(GameResultType:game_result_type):widget =
        case (GameResultType):
            game_result_type.Game => GameUI
            game_result_type.Team => TeamUI
            game_result_type.Damage => DamageUI
            game_result_type.Player => PlayerUI

    <# ----------------------------------------------------------------------------------------------------
        Update Functions
    ---------------------------------------------------------------------------------------------------- #>
    AnimateReward(Text:text_block, Value:float, Duration:float, AudioId:audio_id, ?OptionalText:string = "")<suspends>:void =
        if (Value <= 0.0):
            Text.SetText(MakeMessage("{FormatCurrency(Value)}{OptionalText}"))
            Sleep(Duration)
            return

        DeltaTimer := delta_timer{}

        loop:
            ElapsedTime := DeltaTimer.GetElapsedTime()
            Progress := ElapsedTime / Duration
            CurrentAmount := Lerp(0.0, Value, Progress)
            Text.SetText(MakeMessage("{FormatCurrency(CurrentAmount)}{OptionalText}"))
            Sleep(0.0)
            if (Progress >= 1.0):
                Text.SetText(MakeMessage("{FormatCurrency(Value)}{OptionalText}"))
                Player.PlayAudio(AudioId)
                break

    AnimateRewards(ResultsInfo:result_info)<suspends>:void =
        defer:
            GoldFromTimeSurvivedText.SetText(MakeMessage("{FormatCurrency(ResultsInfo.GoldFromTimeSurvived)}"))
            GoldFromWoodCollectedText.SetText(MakeMessage("{FormatCurrency(ResultsInfo.GoldFromWoodCollected)}"))
            GoldFromObjectivesCompletedText.SetText(MakeMessage("{FormatCurrency(ResultsInfo.GoldFromObjectivesCompleted)}"))
            GoldFromStagesCompletedText.SetText(MakeMessage("{FormatCurrency(ResultsInfo.GoldFromStagesCompleted)}"))
            GoldFromVictoryText.SetText(MakeMessage("{FormatCurrency(ResultsInfo.GoldFromVictory)} (+{FormatAsPercent(VictoryBonusGoldMultplier)})"))
            GoldMultiplierText.SetText(MakeMessage("{FormatCurrency(ResultsInfo.GoldFromMultiplier)} (+{FormatAsPercent(ResultsInfo.GoldMultiplier)})"))
            GoldEarnedText.SetText(MakeMessage("{FormatCurrency(ResultsInfo.GoldEarned)}"))

        GoldFromTimeSurvivedText.SetText(MakeMessage(""))
        GoldFromWoodCollectedText.SetText(MakeMessage(""))
        GoldFromObjectivesCompletedText.SetText(MakeMessage(""))
        GoldFromStagesCompletedText.SetText(MakeMessage(""))
        GoldFromVictoryText.SetText(MakeMessage(""))
        GoldMultiplierText.SetText(MakeMessage(""))
        GoldEarnedText.SetText(MakeMessage(""))

        race:
            CancelAnimationEvent.Await()
            block:
                AnimateReward(GoldFromTimeSurvivedText, ResultsInfo.GoldFromTimeSurvived, 0.5, audio_id.Purchase)
                AnimateReward(GoldFromWoodCollectedText, ResultsInfo.GoldFromWoodCollected, 0.5, audio_id.Purchase)
                AnimateReward(GoldFromObjectivesCompletedText, ResultsInfo.GoldFromObjectivesCompleted, 0.5, audio_id.Purchase)
                AnimateReward(GoldFromStagesCompletedText, ResultsInfo.GoldFromStagesCompleted, 0.5, audio_id.Purchase)
                AnimateReward(GoldFromVictoryText, ResultsInfo.GoldFromVictory, 0.5, audio_id.Purchase, ?OptionalText := " (+{FormatAsPercent(VictoryBonusGoldMultplier)})")
                AnimateReward(GoldMultiplierText, ResultsInfo.GoldFromMultiplier, 0.5, audio_id.Purchase, ?OptionalText := " (+{FormatAsPercent(ResultsInfo.GoldMultiplier)})")
                AnimateReward(GoldEarnedText, ResultsInfo.GoldEarned, 1.0, audio_id.UpgradeRare)

    Update(ResultsInfo:result_info):void =
        set ActiveResultsInfo = ResultsInfo
        set IsClaimed = false
        set IsRankUp = ResultsInfo.IsRankUp
        SetType(game_result_type.Game)
        
        ArenaLevel := CalculateArenaLevel(ResultsInfo.GameSettings.Arena.Level, ResultsInfo.GameSettings.Difficulty)
        SetTitle("{ResultsInfo.Arena.Name} (Lv. {ArenaLevel}) (x{ResultsInfo.GameSettings.GameSpeed})")
        ArenaText.SetText(MakeMessage("{ResultsInfo.IsVictory? and "Victory" or "Defeat" + "!"}"))
        ArenaText.SetColor(ResultsInfo.IsVictory? and ThemeColors.Success or ThemeColors.Failure)

        # Menu UI
        for (Index -> GameResultEntryUI : GameResultEntryUIs):
            if (ActivePlayer := ResultsInfo.ActivePlayers[Index]):
                GameResultEntryUI.Update(ActivePlayer)
                GameResultEntryUI.Show()
            else:
                GameResultEntryUI.Hide()

        # Game UI
        TimeSurvivedText.SetText(MakeMessage(ResultsInfo.TimeSurvived.FormatAsTime()))
        EnemiesEliminatedText.SetText(MakeMessage("{ResultsInfo.EnemiesEliminated}"))
        LevelReachedText.SetText(MakeMessage("{ResultsInfo.LevelReached}"))
        WoodCollectedText.SetText(MakeMessage("{FormatCurrency(ResultsInfo.WoodCollected)}"))
        ObjectivesCompletedText.SetText(MakeMessage("{ResultsInfo.ObjectivesCompleted}/{ResultsInfo.Arena.StageCount}"))
        StagesCompletedText.SetText(MakeMessage("{ResultsInfo.StagesCompleted}/{ResultsInfo.Arena.StageCount}"))
        spawn{AnimateRewards(ResultsInfo)}

        # Team UI
        PlayerDamageDealt := for (ActivePlayer : ResultsInfo.ActivePlayers, DamageDealt := ResultsInfo.PlayerStats[ActivePlayer].GetGameStat(player_game_stat_id.DamageDealt)) {(ActivePlayer, DamageDealt)}
        SortedPlayerDamageDealt := SortBy(PlayerDamageDealt, LessPlayerFloat)
        for (Index -> GameResultTeamEntryUI : GameResultTeamEntryUIs):
            if (ActivePlayer := SortedPlayerDamageDealt[Index], PlayerStats := ActiveResultsInfo.PlayerStats[ActivePlayer(0)]):
                EnemiesEliminated := ActiveResultsInfo.PlayerEnemiesEliminated[ActivePlayer(0)] or 0
                DamageDealt := PlayerStats.GetGameStat(player_game_stat_id.DamageDealt)
                WeaponLevels := PlayerStats.GetWeaponLevels()
                WoodCollected := PlayerStats.GetGameStat(player_game_stat_id.WoodCollected)
                PlayersRevived := PlayerStats.GetGameStat(player_game_stat_id.PlayersRevived)
                TimesDowned := PlayerStats.GetGameStat(player_game_stat_id.TimesDowned)
                GameResultTeamEntryUI.Update(ActivePlayer(0), EnemiesEliminated, DamageDealt, WeaponLevels, WoodCollected, PlayersRevived, TimesDowned)
                GameResultTeamEntryUI.Show()
            else:
                GameResultTeamEntryUI.Hide()

        # Damage UI
        for (Index -> GameResultDamageEntryUI : GameResultDamageEntryUIs):
            if (ActivePlayer := SortedPlayerDamageDealt[Index], PlayerStats := ActiveResultsInfo.PlayerStats[ActivePlayer(0)]):
                DamageDealt := PlayerStats.GetGameStat(player_game_stat_id.DamageDealt)
                DamageMinibossDealt := PlayerStats.GetGameStat(player_game_stat_id.DamageDealtMiniboss)
                DamageBossDealt := PlayerStats.GetGameStat(player_game_stat_id.DamageDealtBoss)
                GameResultDamageEntryUI.Update(ActivePlayer(0), DamageDealt, DamageMinibossDealt, DamageBossDealt)
                GameResultDamageEntryUI.Show()
            else:
                GameResultDamageEntryUI.Hide()

    UpdatePlayerUI(InPlayer:player):void =
        if (PlayerStats := ActiveResultsInfo.PlayerStats[InPlayer]):
            AbilityLevels := PlayerStats.GetAbilityLevels()
            BanishedIds := for (BanishedAbility : PlayerStats.GetBanishedAbilities(), not BanishedAbility.MaybeWeaponId?) {BanishedAbility.Id}

            for (Id -> AbilityText : AbilityTexts, AbilityTotalText := AbilityTotalTexts[Id], AbilityInfo := GetAbilityInfo[Id]):
                IsBanished := logic{BanishedIds.Find[Id]}

                if (AbilityLevel := AbilityLevels[Id]):
                    AbilityValue := PlayerStats.CalculateAbilityValue(Id, AbilityLevel, ?IncludeBaseValue := false)
                    AbilityText.SetText(MakeMessage("+{AbilityInfo.FormatValue(AbilityValue)}"))
                    AbilityText.SetColor(IsBanished? and ThemeColors.Failure or ThemeColors.Success)
                    AbilityTotalText.SetText(MakeMessage("{AbilityInfo.FormatValue(InPlayer.CalculateTotalAbilityValue(AbilityInfo, PlayerStats.CalculateAbilityValue(Id, AbilityLevel)))}"))
                else:
                    AbilityText.SetText(MakeMessage("0"))
                    AbilityTotalText.SetText(MakeMessage("{AbilityInfo.FormatValue(InPlayer.CalculateTotalAbilityValue(AbilityInfo, PlayerStats.CalculateAbilityValue(Id, array{})))}"))
                    if (IsBanished?):
                        AbilityText.SetColor(ThemeColors.Failure)
                    else:
                        AbilityText.SetTextColor(ThemeColors.Inactive)
                        AbilityText.SetShadowOpacity(0.0)

            WeaponIds := for (WeaponId -> WeaponLevel : PlayerStats.GetWeaponLevels()) {WeaponId}
            for (Index -> GameResultWeaponEntryUI : GameResultWeaponEntryUIs):
                if (WeaponId := WeaponIds[Index], WeaponLevel := PlayerStats.GetWeaponLevels()[WeaponId]):
                    GameResultWeaponEntryUI.Update(InPlayer, WeaponId, WeaponLevel, ?MaybeResultsInfo := option{ActiveResultsInfo})
                    GameResultWeaponEntryUI.Show()
                else:
                    GameResultWeaponEntryUI.SetInactive()

            TeamExperienceMultiplierText.SetText(MakeMessage("+{FormatAsPercent(ActiveResultsInfo.TeamExperienceMultiplier)}"))

            for (Index -> BoonIcon : BoonIcons):
                if (BoonId := PlayerStats.GetBoons()[Index], BoonInfo := GetBoonInfo[BoonId]):
                    BoonIcon.SetImage(BoonInfo.IconSmall)
                    BoonIcon.SetTint(BoonInfo.Tier.ToTierColor() * 2)
                else:
                    BoonIcon.SetImage(UserInterface.Icons.Lock_64)
                    BoonIcon.SetTint(ThemeColors.Inactive)

    SetType(GameResultType:game_result_type, ?MaybePlayerIndex:?int = false):void =
        CancelAnimationEvent.Signal()
        GameIcon.SetTint(NamedColors.White)
        TeamIcon.SetTint(NamedColors.White)
        DamageIcon.SetTint(NamedColors.White)
        for (Index -> GameResultEntryUI : GameResultEntryUIs):
            GameResultEntryUI.SetInactive()

        case (GameResultType):
            game_result_type.Game => GameIcon.SetTint(ThemeColors.Highlight * 2)
            game_result_type.Team => TeamIcon.SetTint(ThemeColors.Highlight * 2)
            game_result_type.Damage => DamageIcon.SetTint(ThemeColors.Highlight * 2)
            game_result_type.Player => if (PlayerIndex := MaybePlayerIndex?):
                for (Index -> GameResultEntryUI : GameResultEntryUIs):
                    if (Index = PlayerIndex):
                        GameResultEntryUI.SetActive()
                    else:
                        GameResultEntryUI.SetInactive()

        GetUI(ActiveGameResultType).SetVisibility(widget_visibility.Collapsed)
        GetUI(GameResultType).SetVisibility(widget_visibility.Visible)

        set ActiveGameResultType = GameResultType

    <# ----------------------------------------------------------------------------------------------------
        Event Handlers
    ---------------------------------------------------------------------------------------------------- #>
    OnClosed(InPlayer:player):void =
        if (not IsClaimed?):
            InPlayer.AwardAccolade(ActiveResultsInfo.IsVictory? and accolade_id.Victory or accolade_id.CompleteGame)
            set IsClaimed = true

            if (IsRankUp?):
                InPlayer.ShowMessage(InPlayer.GetRankUpMessage())
            
        CancelAnimationEvent.Signal()

    OnPlayerClicked(InPlayer:player, PlayerIndex:int):void =
        SetType(game_result_type.Player, ?MaybePlayerIndex := option{PlayerIndex})
        UpdatePlayerUI(InPlayer)
        
    OnClickGame(Message:widget_message):void =
        SetType(game_result_type.Game)       

    OnClickTeam(Message:widget_message):void =
        SetType(game_result_type.Team)

    OnClickDamage(Message:widget_message):void =
        SetType(game_result_type.Damage)