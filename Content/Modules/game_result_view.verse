using { /Verse.org/Simulation }
using { /Fortnite.com/Devices }

game_result_view := class:
    @editable
    Volume:volume_device = volume_device{}
    
    @editable
    Camera:gameplay_camera_fixed_point_device = gameplay_camera_fixed_point_device{}

    @editable
    PopupDialog:popup_dialog_device = popup_dialog_device{}
    
    <# ----------------------------------------------------------------------------------------------------
        Initialization
    ---------------------------------------------------------------------------------------------------- #>
    OnBegin()<suspends>:void =
        Volume.AgentEntersEvent.Subscribe(OnAgentEnters)
        PopupDialog.DismissedEvent.Subscribe(OnDismissed)
        PopupDialog.RespondingButtonEvent.Subscribe(OnRespondingButton)
        
        WaitForGameManager()
        if (GameManager := GetGameManager[]):
            GameManager.GameStartedEvent.Subscribe(OnGameStarted)

    <# ----------------------------------------------------------------------------------------------------
        Update Functions
    ---------------------------------------------------------------------------------------------------- #>

    <# ----------------------------------------------------------------------------------------------------
        Event Handlers
    ---------------------------------------------------------------------------------------------------- #>
    OnAgentEnters(Agent:agent):void =
        Agent.PlayAudio(audio_id.Activate)
        Camera.AddTo(Agent)
        PopupDialog.Show(Agent)
        if (Player := player[Agent]):
            Player.ToggleHUD(false)

    OnDismissed(Agent:agent):void =
        Camera.RemoveFrom(Agent)
        if (Player := player[Agent]):
            Player.ToggleHUD(true)

    OnGameStarted():void =
        PopupDialog.Hide()

    OnRespondingButton(Agent:agent, Index:int):void =
        if (Index = 0, Player := player[Agent]):
            if (GetGameManager[].PlayerGameResultUI[Player]):
                Player.ShowGameResultsUI()
            else:
                Player.PlayAudio(audio_id.Error)