using { /Fortnite.com/UI }
using { /Verse.org/Colors }
using { /Verse.org/Simulation }
using { /UnrealEngine.com/Temporary/SpatialMath }
using { /UnrealEngine.com/Temporary/UI }

game_result_weapon_entry_ui := class(ui_component):
    PlayerWeaponIcon:texture_block = texture_block:
        DefaultImage := UserInterface.Icons.WeaponBow_64
        DefaultDesiredSize := vector2{X := 64.0, Y := 64.0}

    LevelText:text_block = text_block:
        DefaultTextColor := ThemeColors.PrimaryText
        DefaultShadowOpacity := 0.5
        DefaultShadowOffset := option{vector2{X := 2.0, Y := 2.0}}

    DamageText:text_block = text_block:
        DefaultTextColor := ThemeColors.Inactive
        DefaultShadowOffset := option{vector2{X := 2.0, Y := 2.0}}
        DefaultShadowColor := ThemeColors.Success * 0.125

    CooldownText:text_block = text_block:
        DefaultTextColor := ThemeColors.Inactive
        DefaultShadowOffset := option{vector2{X := 2.0, Y := 2.0}}
        DefaultShadowColor := ThemeColors.Success * 0.125

    SizeText:text_block = text_block:
        DefaultTextColor := ThemeColors.Inactive
        DefaultShadowOffset := option{vector2{X := 2.0, Y := 2.0}}
        DefaultShadowColor := ThemeColors.Success * 0.125

    DPSText:text_block = text_block:
        DefaultTextColor := ThemeColors.Success
        DefaultShadowOpacity := 1.0
        DefaultShadowOffset := option{vector2{X := 2.0, Y := 2.0}}
        DefaultShadowColor := ThemeColors.Success * 0.125

    DPSDamageText:text_block = text_block:
        DefaultTextColor := ThemeColors.Success
        DefaultShadowOpacity := 1.0
        DefaultShadowOffset := option{vector2{X := 2.0, Y := 2.0}}
        DefaultShadowColor := ThemeColors.Success * 0.125

    DPSDurationText:text_block = text_block:
        DefaultTextColor := ThemeColors.Success
        DefaultShadowOpacity := 1.0
        DefaultShadowOffset := option{vector2{X := 2.0, Y := 2.0}}
        DefaultShadowColor := ThemeColors.Success * 0.125

    var EvolutionIcons:[]texture_block = array{}
    
    <# ----------------------------------------------------------------------------------------------------
        Initialization
    ---------------------------------------------------------------------------------------------------- #>
    CreateUI<override>():widget =
        stack_box:
            Orientation := orientation.Vertical
            Slots := array:
                stack_box_slot:
                    HorizontalAlignment := horizontal_alignment.Fill
                    VerticalAlignment := vertical_alignment.Fill
                    Widget := stack_box:
                        Orientation := orientation.Horizontal
                        Slots := array:
                            stack_box_slot:
                                HorizontalAlignment := horizontal_alignment.Left
                                Padding := margin{Left := 21.0, Top := 0.0, Right := 0.0, Bottom := 0.0}
                                Distribution := option{1.0}
                                Widget := PlayerWeaponIcon
                            stack_box_slot:
                                HorizontalAlignment := horizontal_alignment.Left
                                VerticalAlignment := vertical_alignment.Bottom
                                Distribution := option{1.0}
                                Widget := stack_box:
                                    Orientation := orientation.Horizontal
                                    Slots := array:
                                        stack_box_slot:
                                            HorizontalAlignment := horizontal_alignment.Fill
                                            VerticalAlignment := vertical_alignment.Fill
                                            Padding := margin{Left := 0.0, Top := 0.0, Right := 10.0, Bottom := 0.0}
                                            Widget := text_block:
                                                DefaultText := MakeMessage("Lv.")
                                                DefaultTextColor := ThemeColors.Highlight * 2
                                                DefaultShadowColor := ThemeColors.Highlight * 0.25
                                                DefaultShadowOffset := option{vector2{X := 2.0, Y := 2.0}}
                                                DefaultShadowOpacity := 1.0
                                        stack_box_slot:
                                            HorizontalAlignment := horizontal_alignment.Fill
                                            VerticalAlignment := vertical_alignment.Fill
                                            Widget := LevelText
                            stack_box_slot:
                                HorizontalAlignment := horizontal_alignment.Left
                                VerticalAlignment := vertical_alignment.Bottom
                                Distribution := option{1.0}
                                Widget := DamageText
                            stack_box_slot:
                                HorizontalAlignment := horizontal_alignment.Left
                                VerticalAlignment := vertical_alignment.Bottom
                                Distribution := option{1.0}
                                Widget := CooldownText
                            stack_box_slot:
                                HorizontalAlignment := horizontal_alignment.Left
                                VerticalAlignment := vertical_alignment.Bottom
                                Distribution := option{1.0}
                                Widget := SizeText
                stack_box_slot:
                    HorizontalAlignment := horizontal_alignment.Fill
                    VerticalAlignment := vertical_alignment.Fill
                    Widget := stack_box:
                        Orientation := orientation.Horizontal
                        Slots := array:
                            stack_box_slot:
                                HorizontalAlignment := horizontal_alignment.Left
                                Distribution := option{1.0}
                                Widget := stack_box:
                                    Orientation := orientation.Horizontal
                                    Slots := for (Index := 1..3):
                                        EvolutionIcon:= texture_block:
                                            DefaultImage := UserInterface.Icons.Lock_64
                                            DefaultDesiredSize := vector2{X := 32.0, Y := 32.0}

                                        set EvolutionIcons += array{EvolutionIcon}

                                        stack_box_slot:
                                            Padding := margin{Left := (Index = 1) and 0.0 or 5.0, Top := 0.0, Right := 0.0, Bottom := 0.0}
                                            Widget := EvolutionIcon
                                                
                            stack_box_slot:
                                HorizontalAlignment := horizontal_alignment.Left
                                Distribution := option{1.0}
                                Padding := margin{Left := 21.0, Top := 0.0, Right := 0.0, Bottom := 0.0}
                                Widget := text_block:
                                    DefaultText := MakeMessage("DPS")
                                    DefaultTextColor := ThemeColors.PrimaryText
                                    DefaultTextOpacity := 0.5
                            stack_box_slot:
                                HorizontalAlignment := horizontal_alignment.Left
                                Distribution := option{1.0}
                                Widget := DPSText
                            stack_box_slot:
                                HorizontalAlignment := horizontal_alignment.Left
                                Distribution := option{1.0}
                                Widget := DPSDamageText
                            stack_box_slot:
                                HorizontalAlignment := horizontal_alignment.Left
                                Distribution := option{1.0}
                                Widget := DPSDurationText

    <# ----------------------------------------------------------------------------------------------------
        Update Functions
    ---------------------------------------------------------------------------------------------------- #>
    Update(InPlayer:player, WeaponId:weapon_id, WeaponLevels:[]tuple(ability_id, int), ?MaybeResultsInfo:?result_info = false):void =
        if (WeaponInfo := GetWeaponInfo[WeaponId]):
            PlayerWeaponIcon.SetImage(WeaponInfo.IconSmall)
            PlayerWeaponIcon.SetTint(NamedColors.White)
            LevelText.SetText(MakeMessage("{CalculateWeaponLevel(WeaponId, WeaponLevels)}"))

            if (ParentWidget := LevelText.GetParentWidget[]):
                ParentWidget.SetVisibility(widget_visibility.Visible)

            if (PlayerStats := InPlayer.GetStats[]):
                Damage := PlayerStats.CalculateAbilityValue(WeaponId, ability_id.Damage, WeaponLevels)
                Cooldown := PlayerStats.CalculateAbilityValue(WeaponId, ability_id.Cooldown, WeaponLevels)
                Size := PlayerStats.CalculateAbilityValue(WeaponId, ability_id.Size, WeaponLevels)

                WeaponEvolutionInfos := MaybeResultsInfo?.PlayerStats[InPlayer].GetWeaponEvolution[WeaponId] or InPlayer.GetWeaponEvolution(WeaponId)
                for (Index -> EvolutionIcon : EvolutionIcons):
                    if (EvolutionInfo := WeaponEvolutionInfos[Index]):
                        EvolutionIcon.SetImage(EvolutionInfo.IconSmall)
                        EvolutionIcon.SetTint(EvolutionInfo.IconColor)
                    else:
                        EvolutionIcon.SetImage(UserInterface.Icons.Lock_64)
                        EvolutionIcon.SetTint(ThemeColors.Inactive)
                        
                    EvolutionIcon.SetVisibility(widget_visibility.Visible)

                BanishedAbilities := MaybeResultsInfo?.PlayerStats[InPlayer].GetBanishedAbilities() or InPlayer.GetBanishedAbilities()
                BanishedIds := for (BanishedAbility : BanishedAbilities, BanishedAbility.MaybeWeaponId? = WeaponId) {BanishedAbility.Id}
                
                DamageText.SetText(MakeMessage("+{FormatAsPercent(Damage)}"))
                if (BanishedIds.Find[ability_id.Damage]):
                    DamageText.SetColor(ThemeColors.Failure)
                else:
                    DamageText.SetTextColor(Damage > 0.0 and ThemeColors.Success or ThemeColors.Inactive)
                    DamageText.SetShadowOpacity(Damage > 0.0 and 1.0 or 0.0)
                
                CooldownText.SetText(MakeMessage("+{FormatAsPercent(Cooldown)}"))
                if (BanishedIds.Find[ability_id.Cooldown]):
                    CooldownText.SetColor(ThemeColors.Failure)
                else:
                    CooldownText.SetTextColor(Cooldown > 0.0 and ThemeColors.Success or ThemeColors.Inactive)
                    CooldownText.SetShadowOpacity(Cooldown > 0.0 and 1.0 or 0.0)

                SizeText.SetText(MakeMessage("+{FormatAsPercent(Size)}"))
                if (BanishedIds.Find[ability_id.Size]):
                    SizeText.SetColor(ThemeColors.Failure)
                else:
                    SizeText.SetTextColor(Size > 0.0 and ThemeColors.Success or ThemeColors.Inactive)
                    SizeText.SetShadowOpacity(Size > 0.0 and 1.0 or 0.0)

            if (ResultsInfo := MaybeResultsInfo?, Ability := ResultsInfo.PlayerActiveAbilities[InPlayer][WeaponId]):
                DPSText.SetText(MakeMessage("{FormatCurrency(Ability.GetDPS(?EndTime := option{ResultsInfo.EndTime}))}/s"))
                DPSDamageText.SetText(MakeMessage("{FormatCurrency(Ability.GetTotalDamage())}"))
                DPSDurationText.SetText(MakeMessage("/ {Ability.GetDuration(?EndTime := option{ResultsInfo.EndTime}).FormatAsTime()}"))
            else:
                DPSText.SetText(MakeMessage("{FormatCurrency(InPlayer.GetWeaponDPS(WeaponId))}/s"))
                DPSDamageText.SetText(MakeMessage("{FormatCurrency(InPlayer.GetWeaponTotalDamage(WeaponId))}"))
                DPSDurationText.SetText(MakeMessage("/ {InPlayer.GetWeaponDuration(WeaponId).FormatAsTime()}"))

            if (ParentWidget := DPSText.GetParentWidget[]):
                ParentWidget.SetVisibility(widget_visibility.Visible)

    SetInactive():void =
        PlayerWeaponIcon.SetImage(UserInterface.Icons.Lock_64)
        PlayerWeaponIcon.SetTint(ThemeColors.Inactive)
        DamageText.SetText(MakeMessage(""))
        CooldownText.SetText(MakeMessage(""))
        SizeText.SetText(MakeMessage(""))

        if (ParentWidget := LevelText.GetParentWidget[]):
            ParentWidget.SetVisibility(widget_visibility.Hidden)

        for (EvolutionIcon : EvolutionIcons):
            EvolutionIcon.SetVisibility(widget_visibility.Hidden)

        if (ParentWidget := DPSText.GetParentWidget[]):
            ParentWidget.SetVisibility(widget_visibility.Hidden)