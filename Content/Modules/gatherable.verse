using { /Fortnite.com/Characters }
using { /Fortnite.com/Devices }
using { /Fortnite.com/Devices/CreativeAnimation }
using { /Verse.org/Random }
using { /Verse.org/Simulation }
using { /UnrealEngine.com/Temporary/SpatialMath }

<# ----------------------------------------------------------------------------------------------------
    Constants
---------------------------------------------------------------------------------------------------- #>
DefaultGatherDuration:float = 5.0
DefaultGatherableRespawnDuration:float = 90.0
DefaultGatherableItemDropRate:float = 0.1

<# ----------------------------------------------------------------------------------------------------
    Data Definitions
---------------------------------------------------------------------------------------------------- #>
gatherable_state := enum:
    Enabled
    Disabled
    Available
    Gathering
    Respawning
    
<# ----------------------------------------------------------------------------------------------------
    gatherable
---------------------------------------------------------------------------------------------------- #>
gatherable := class<concrete>:
    @editable
    Prop:creative_prop = creative_prop{}

    @editable
    VolumeDevice:volume_device = volume_device{}
    
    @editable
    MapIndicator:map_indicator_device = map_indicator_device{}

    GatheredEvent:events(player) = events(player){}
    GatherStoppedEvent:events(player) = events(player){}
    GatherCancelledEvent:events(tuple()) = events(tuple()){}
    RespawnCancelledEvent:events(tuple()) = events(tuple()){}
    
    var PlayersGathering:[]player = array{}
    var State:gatherable_state = gatherable_state.Available
    var InitialTransform:transform = transform{}
    var InitialScale:vector3 = vector3{}
    var MaybeSpawnedItemDrop:?item_drop = false
    var Progress:float = 0.0
    var IsRespawn:logic = true
  
    OnBegin()<suspends>:void =
        set InitialScale = Prop.GetTransform().Scale

        VolumeDevice.AgentEntersEvent.Subscribe(OnAgentEnters)
        VolumeDevice.AgentExitsEvent.Subscribe(OnAgentExits)

        if (GameManager := GetGameManager[]):
            GameManager.GameEndedEvent.Subscribe(OnGameEndedEvent)
            GameManager.BossFightStartedEvent.Subscribe(OnBossFightStartedEvent)
            
    <# ----------------------------------------------------------------------------------------------------
        Update Functions
    ---------------------------------------------------------------------------------------------------- #>
    Load(Transform:transform):void =
        if:
            Prop.TeleportTo[Transform.Translation, Transform.Rotation]
            VolumeDevice.TeleportTo[Transform.Translation, Transform.Rotation]
        then:
            set InitialTransform = Prop.GetTransform()
            Enable()
        else:
            Print("Gatherable failed to teleport!", ?Color := ThemeColors.Failure)

    StartGather(Player:player)<suspends>:void =
        set State = gatherable_state.Gathering

        GatherStartTime := GetSimulationElapsedTime()

        var HasChainsaw:logic = false
        var GatherRateMultiplier:float = 0.0
        for (PlayerGathering : PlayersGathering):
            set GatherRateMultiplier += (PlayerGathering.GetAbilityValue(ability_id.GatherRate) + PlayerGathering.GetPowerUpValue(power_up_id.GatherRate)) * 2.0
            if (PlayerGathering.HasBoon(boon_id.Chainsaw)?):
                set HasChainsaw = true

        InitialProgress := Progress
        GatherRateBoonMultiplier := HasChainsaw? and 2.0 or 1.0
        BaseGatherDuration := DefaultGatherDuration / GetGameSpeed()
        GatherDurationFactor := BaseGatherDuration / DefaultGatherDuration
        GatherDuration := BaseGatherDuration / (1.0 + GatherRateMultiplier) / GatherRateBoonMultiplier / (PlayersGathering.Length * 1.0)
        GatherFactor := GatherDuration / BaseGatherDuration
        GatherTier := if (GatherDuration <= GatherDurationFactor * 1.0) then 5
        else if (GatherDuration <= GatherDurationFactor * 2.0) then 4
        else if (GatherDuration <= GatherDurationFactor * 3.0) then 3
        else if (GatherDuration <= GatherDurationFactor * 4.0) then 2
        else 1

        GatherDurationRemaining := GatherDuration * (1.0 - Progress)
        ProgressPerHit := 0.2 * (1.0 / GatherDurationFactor)
        VFXId := HasChainsaw? and vfx_id.GatheringChainsaw or vfx_id.Gathering
        SpawnedVFX := SpawnVFX(VFXId, Prop.GetTransform().Translation, ?Scale := vector3{X := 20.0 + (GatherTier * 0.001), Y := 20.0, Z := GatherFactor * 3.0}, ?Duration := 0.0)

        defer:
            if (VFX := SpawnedVFX?):
                VFX.Dispose()

        race:
            block:
                Sleep(GatherDurationRemaining)
                OnGathered(Player)
            block:
                GatherCancelledEvent.Await()
                set Progress = InitialProgress + (GetSimulationElapsedTime() - GatherStartTime) / GatherDuration
                set State = gatherable_state.Available
                if (Prop.TeleportTo[transform{Translation := Prop.GetTransform().Translation, Rotation := InitialTransform.Rotation, Scale := Prop.GetTransform().Scale}]) {}
            block:
                loop:
                    if (AnimationController := Prop.GetAnimationController[], Character := Player.GetFortCharacter[]):
                        Direction := Prop.GetTransform().Translation.Make2D() - Character.GetTransform().Translation.Make2D()
                        RotationVector := MakeRotationFromYawPitchRollDegrees(90.0, 0.0, 0.0).RotateVector(Direction)
                        CurrentScale := Prop.GetTransform().Scale
                        TargetScale:vector3 = InitialScale - (InitialScale * 0.6) * (Progress + ProgressPerHit)
                        DeltaRotation := MakeRotation(RotationVector, DegreesToRadians(15.0))
                        DeltaScale01 := UnitVector - (CurrentScale - TargetScale) * 0.25 / CurrentScale
                        DeltaScale02 := UnitVector - (CurrentScale - TargetScale) * 0.75 / (CurrentScale * DeltaScale01)
                        
                        KeyFrames := array:
                            keyframe_delta:
                                DeltaLocation := vector3{X := 0.0, Y := 0.0, Z := 0.0}
                                DeltaRotation := DeltaRotation
                                DeltaScale := DeltaScale01
                                Interpolation := InterpolationTypes.EaseOut
                                Time := 0.075 * GatherFactor
                            keyframe_delta:
                                DeltaLocation := vector3{X := 0.0, Y := 0.0, Z := 0.0}
                                DeltaRotation := DeltaRotation.Invert()
                                DeltaScale := DeltaScale02
                                Interpolation := InterpolationTypes.EaseOut
                                Time := 0.225 * GatherFactor
            
                        AnimationController.SetAnimation(KeyFrames, ?Mode := animation_mode.OneShot)
                        AnimationController.Play()
                        AnimationController.MovementCompleteEvent.Await()
                        Sleep(0.7 * GatherFactor)
                        set Progress += ProgressPerHit

    RefreshGather(Player:player):void =
        GatherCancelledEvent.Signal()
        spawn{StartGather(Player)}

    Respawn()<suspends>:void =
        defer:
            Reset()

        race:
            RespawnCancelledEvent.Await()
            block:
                Prop.Hide()
                set State = gatherable_state.Respawning
                
                Sleep(DefaultGatherableRespawnDuration / (Max(1, GetActivePlayers().Length) * 1.0) / GetGameSpeed())
                if (Prop.TeleportTo[transform{Translation := Prop.GetTransform().Translation, Rotation := Prop.GetTransform().Rotation, Scale := InitialScale * 0.6}]) {}
                Prop.Show()

                if (SpawnedItemDrop := MaybeSpawnedItemDrop?):
                    SpawnedItemDrop.Reposition()
                    set MaybeSpawnedItemDrop = false

                if (PlayAnimationController := Prop.GetAnimationController[]):
                    DeltaScale := InitialScale / Prop.GetTransform().Scale
                    KeyFrames := array:
                        keyframe_delta:
                            DeltaLocation := vector3{}
                            DeltaRotation := rotation{}
                            DeltaScale := DeltaScale
                            Time := 1.0
                            Interpolation := InterpolationTypes.EaseIn

                    PlayAnimationController.SetAnimation(KeyFrames, ?Mode := animation_mode.OneShot)
                    PlayAnimationController.Play()
                    PlayAnimationController.MovementCompleteEvent.Await()
  
    Reset():void =
        if (State = gatherable_state.Disabled) then return
        RespawnCancelledEvent.Signal()

        set State = gatherable_state.Available
        set Progress = 0.0

        TargetTransform := transform{Translation := InitialTransform.Translation, Rotation := InitialTransform.Rotation, Scale := InitialScale}
        if (Prop.TeleportTo[TargetTransform], MapIndicator.TeleportTo[TargetTransform]):
            Prop.Show()

        if (Player := PlayersGathering[0]):
            spawn{StartGather(Player)}

    Enable():void =
        set State = gatherable_state.Enabled
        set IsRespawn = true
        Reset()

    Disable():void =
        set State = gatherable_state.Disabled
        GatherCancelledEvent.Signal()
        RespawnCancelledEvent.Signal()

        if (MapIndicator.TeleportTo[MapIndicatorDisabledPosition, IdentityRotation()]) {}
        Prop.Hide()
        
    DisableRespawn():void =
        set IsRespawn = false
        RespawnCancelledEvent.Signal()

    <# ----------------------------------------------------------------------------------------------------
        Event Handlers
    ---------------------------------------------------------------------------------------------------- #>
    OnAgentEnters(Agent:agent):void =
        if (Player := player[Agent]):
            if (Player.GetIsAFK()?) then return
            if (Character := Player.GetFortCharacter[], Character.IsDownButNotOut[]) then return

            set PlayersGathering += array{Player}
            if (State = gatherable_state.Available):
                spawn{StartGather(Player)}
            else if (State = gatherable_state.Gathering):
                RefreshGather(Player)

            spawn{AwaitAbilityChangedEvent(Player)}
                
    OnAgentExits(Agent:agent):void =
        if:
            Player := player[Agent]
            set PlayersGathering = PlayersGathering.RemoveAllElements(Player)
        then:
            GatherStoppedEvent.Signal(Player)
            if (PlayersGathering.Length = 0):
                GatherCancelledEvent.Signal()
            else if (State = gatherable_state.Gathering):
                RefreshGather(Player)

    OnGathered(Player:player):void =
        var GatherMultiplier:float = 0.0
        for (PlayerGathering : PlayersGathering):
            PlayerGatherMultiplier := PlayerGathering.GetAbilityValue(ability_id.GatherRate) + PlayerGathering.GetPowerUpValue(power_up_id.GatherRate)
            if (PlayerGatherMultiplier > GatherMultiplier):
                set GatherMultiplier = PlayerGatherMultiplier

            GatherStoppedEvent.Signal(PlayerGathering)

        if (GatherAmount := Floor[GetRandomFloat(30.0, 60.0) * (1.0 + GatherMultiplier) / (Max(1, GetActivePlayers().Length) * 1.0)] * 1.0):
            AddGameCurrency(game_currency_id.Wood, GatherAmount)
            SpawnVFX(vfx_id.Gather, Prop.GetTransform().Translation, ?Scale := vector3{X := 1.0, Y := 1.0, Z := GatherAmount}, ?Duration := 2.0)

            if (PlayerGathering := PlayersGathering[0]):
                PlayerGathering.AddGameStat(player_game_stat_id.WoodCollected, GatherAmount)

        IsDrop := GetPsuedoRandomResult(pseudo_random_id.HealthDrop)
        if (IsDrop?):
            set MaybeSpawnedItemDrop = SpawnItemDrop(item_drop_id.Health, Prop.GetTransform().Translation)
            SpawnExperience(GetRandomPositionOnCircle(150.0, ?Origin := Prop.GetTransform().Translation), 20)

            for (PlayerGathering : PlayersGathering):
                PlayerGathering.PlayAudio(audio_id.Surprise)
        else:
            SpawnExperience(Prop.GetTransform().Translation, 20)

        AwardAccolade(accolade_count_id.GatherTree, GetActivePlayers())
        if (MapIndicator.TeleportTo[MapIndicatorDisabledPosition, IdentityRotation()]) {}
        if (IsRespawn?):
            spawn{Respawn()}
        else:
            Disable()

        GatheredEvent.Signal(Player)

    OnGameEndedEvent():void =
        Disable()

    OnBossFightStartedEvent():void =
        Disable()

    AwaitAbilityChangedEvent(Player:player)<suspends>:void =
        if (AbilityChangedEvent := Player.GetStats[].AbilityChangedEvent):
            race:
                RespawnCancelledEvent.Await()

                loop:
                    PlayerStopped := GatherStoppedEvent.Await()
                    if (PlayerStopped = Player):
                        break

                loop:
                    AbilityId := AbilityChangedEvent.Await()
                    if (AbilityId = ability_id.GatherRate):
                        RefreshGather(Player)