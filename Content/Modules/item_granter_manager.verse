using { /Fortnite.com/Devices }
using { /Verse.org/Simulation }

<# ----------------------------------------------------------------------------------------------------
    Deta Definitions
---------------------------------------------------------------------------------------------------- #>
item_granter_id := enum:
    ChugSplash
    BananaOfTheGods
    AirStrike
    KineticBlade
    NitroFists
    ShockwaveHammer

item_granter := class<concrete>:
    @editable
    Id:item_granter_id = item_granter_id.ChugSplash
    @editable
    GrantDevice:item_granter_device = item_granter_device{}
    @editable
    MaybeRemoveDevice:?item_remover_device = false

<# ----------------------------------------------------------------------------------------------------
    Global Functions
---------------------------------------------------------------------------------------------------- #>
var GlobalItemGranterManager:weak_map(session, item_granter_manager) = map{}

GetItemGranterManager()<decides><transacts>:item_granter_manager =
    GlobalItemGranterManager[GetSession()]

(Player:player).GrantItem(Id:item_granter_id):void =
    if (ItemGranterManager := GetItemGranterManager[]):
        ItemGranterManager.GrantItem(Player, Id)

(Player:player).RemoveItem(Id:item_granter_id):void =
    if (ItemGranterManager := GetItemGranterManager[]):
        ItemGranterManager.RemoveItem(Player, Id)

(Player:player).RemoveAllItems():void =
    if (ItemGranterManager := GetItemGranterManager[]):
        ItemGranterManager.RemoveAllItems(Player)

<# ----------------------------------------------------------------------------------------------------
    item_granter_manager
---------------------------------------------------------------------------------------------------- #>
item_granter_manager := class(creative_device):
    @editable
    ItemGranters:[]item_granter = array{}

    @editable
    ItemRemoverAll:item_remover_device = item_remover_device{}

    var MappedItemGranters:[item_granter_id]item_granter = map{}

    OnBegin<override>()<suspends>:void =
        if (set GlobalItemGranterManager[GetSession()] = Self) {}

        for (ItemGranter : ItemGranters, set MappedItemGranters[ItemGranter.Id] = ItemGranter) {}

    <# ----------------------------------------------------------------------------------------------------
        Get Functions
    ---------------------------------------------------------------------------------------------------- #>
    GetItemGranter(Id:item_granter_id)<decides><transacts>:item_granter = 
        MappedItemGranters[Id]

    <# ----------------------------------------------------------------------------------------------------
        Update Functions
    ---------------------------------------------------------------------------------------------------- #>               
    GrantItem(Player:player, Id:item_granter_id):void =
        if (ItemGranter := GetItemGranter[Id]):
            ItemGranter.GrantDevice.GrantItem(Player)

    RemoveItem(Player:player, Id:item_granter_id):void =
        if (ItemGranter := GetItemGranter[Id], ItemRemover := ItemGranter.MaybeRemoveDevice?):
            ItemRemover.Remove(Player)

    RemoveAllItems(Player:player):void =
        ItemRemoverAll.Remove(Player)

    <# ----------------------------------------------------------------------------------------------------
        Event Handlers
    ---------------------------------------------------------------------------------------------------- #>
