using { /Fortnite.com/Devices }
using { /Verse.org/Simulation }
using { /UnrealEngine.com/Temporary/SpatialMath }
using { /UnrealEngine.com/Temporary/UI }

<# ----------------------------------------------------------------------------------------------------
    Constants
---------------------------------------------------------------------------------------------------- #>
ItemShopSpawnInterval:float = 360.0
ItemShopExpiryTime:float = 60.0
ItemShopTimeRemainingThreshold:float = 60.0

#ItemShopSpawnInterval:float = 30.0
#ItemShopExpiryTime:float = 15.0
#ItemShopTimeRemainingThreshold:float = 15.0

<# ----------------------------------------------------------------------------------------------------
    Global Functions
---------------------------------------------------------------------------------------------------- #>

var GlobalItemShopManager:weak_map(session, item_shop_manager) = map{}

GetItemShopManager()<decides><transacts>:item_shop_manager =
    GlobalItemShopManager[GetSession()]

ToggleItemShop(IsEnable:logic):void =
    if (ItemShopManager := GetItemShopManager[]):
        if (IsEnable?):
            ItemShopManager.Enable()
        else:
            ItemShopManager.Disable()

EnableItemShop():void =
    if (ItemShopManager := GetItemShopManager[]):
        ItemShopManager.Enable()
        spawn{ItemShopManager.Expire()}

(Player:player).ShowItemShopUI():void =
    Player.ShowItemShopUI(?IsResume := false)

(Player:player).ShowItemShopUI(?IsResume:logic = false):void =
    if (ItemShopManager := GetItemShopManager[]):
        ItemShopManager.ShowItemShopUI(Player, ?IsResume := IsResume)

<# ----------------------------------------------------------------------------------------------------
    item_shop_manager
---------------------------------------------------------------------------------------------------- #>
item_shop_manager := class(creative_device):
    @editable
    Character:character_device = character_device{}

    @editable
    Beacon:beacon_device = beacon_device{}

    @editable
    MapIndicator:map_indicator_device = map_indicator_device{}
    
    @editable
    VolumeActivator:volume_activator = volume_activator{}

    StoppedEvent:event(void) = event(void){}
    ActivatedEvent:event(void) = event(void){}

    var PlayerItemShopPurchasedEvents:[player]events(tuple()) = map{}

    var IsActive:logic = false
    var PlayerItemShopUI:[player]item_shop_ui = map{}
    var PlayerItemShopTimerUI:[player]item_shop_timer_ui = map{}

    <# ----------------------------------------------------------------------------------------------------
        Initialization
    ---------------------------------------------------------------------------------------------------- #>
    OnBegin<override>()<suspends>:void =
        if (set GlobalItemShopManager[GetSession()] = Self) {}

        for (Player : GetPlayspace().GetPlayers()):
            spawn{Initialize(Player)}

        WaitForGameManager()
        if (GameManager := GetGameManager[]):
            GameManager.GameEndedEvent.Subscribe(OnGameEnded)
            #GameManager.GameStartedEvent.Subscribe(OnGameStarted)
            #GameManager.BossFightStartedEvent.Subscribe(OnBossFightStarted)

        if (MapIndicator.TeleportTo[MapIndicatorDisabledPosition, IdentityRotation()]) {}
        VolumeActivator.OnBegin()
        Disable()
        VolumeActivator.ActivatedEvent.Subscribe(OnActivated)
        GetPlayspace().PlayerAddedEvent().Subscribe(OnPlayerAdded)
        GetPlayspace().PlayerRemovedEvent().Subscribe(OnPlayerRemoved)

    Initialize(Player:player)<suspends>:void =
        WaitForUIManager(Player)

        ItemShopTimerUI := item_shop_timer_ui{Player := Player}
        ItemShopTimerUI.Initialize()
        Player.AddWidgetToHUD of canvas_slot:
            Anchors := anchors{Minimum := vector2{X := 0.5, Y := 1.0}, Maximum := vector2{X := 0.5, Y := 1.0}}
            Offsets := margin{Left := 0.0, Top := -250.0, Right := 0.0, Bottom := 0.0}
            Alignment := vector2{X := 0.5, Y := 1.0}
            Widget := ItemShopTimerUI.GetUI()

        if (set PlayerItemShopTimerUI[Player] = ItemShopTimerUI) {}
        if (set PlayerItemShopPurchasedEvents[Player] = events(tuple()){}) {}
        
        Sleep(0.0)
        ItemShopTimerUI.Hide()

    <# ----------------------------------------------------------------------------------------------------
        Update Functions
    ---------------------------------------------------------------------------------------------------- #>
    <#
    Start()<suspends>:void =
        race:
            StoppedEvent.Await()
            loop:
                Sleep(ItemShopSpawnInterval)

                if (GetGameTimeRemaining() < ItemShopTimeRemainingThreshold):
                    break

                Enable()
                spawn{Expire()}
    #>

    Stop():void =
        StoppedEvent.Signal()
        Disable()

    Expire()<suspends>:void =
        defer:
            for (ItemShopTimerUI : PlayerItemShopTimerUI):
                ItemShopTimerUI.Hide()

        StartTime:float = GetSimulationElapsedTime()

        for (ItemShopTimerUI : PlayerItemShopTimerUI):
            ItemShopTimerUI.SetIcon(UserInterface.Icons.Badge_ItemShop_64)
            ItemShopTimerUI.Show()

        race:
            StoppedEvent.Await()
            ActivatedEvent.Await()
            block:
                Sleep(ItemShopExpiryTime)
                Disable()
                
            loop:
                ElapsedTime := GetSimulationElapsedTime() - StartTime
                TimeRemaining := ItemShopExpiryTime - ElapsedTime
                Progress :=  ElapsedTime / ItemShopExpiryTime

                for (ItemShopTimerUI : PlayerItemShopTimerUI):
                    ItemShopTimerUI.Update(TimeRemaining, Progress)

                Sleep(0.0)

    Enable():void =
        #VolumeActivator.Enable()
        Beacon.Enable()
        if (MapIndicator.TeleportTo[MapIndicatorEnabledPosition, IdentityRotation()]) {}
        Character.Show()

        for (Player : GetActivePlayers()):
            Player.ShowMessage(message_id.ItemShopNotification)
            Player.PlayAudio(audio_id.Warning)
            for (Index := 1..2):
                MapIndicator.ActivateObjectivePulse(Player)

    Disable():void =
        #VolumeActivator.Disable()

        for (Player : GetAllPlayers()):
            MapIndicator.DeactivateObjectivePulse(Player)

        Beacon.Disable()
        if (MapIndicator.TeleportTo[MapIndicatorDisabledPosition, IdentityRotation()]) {}
        Character.Hide()
        set IsActive = false

    ShowItemShopUI(Player:player, ?IsResume:logic = false):void =
        UI := if (ItemShopUI := PlayerItemShopUI[Player]):
            ItemShopUI
        else:
            ItemShopUI := item_shop_ui{Player := Player}
            ItemShopUI.ClosedEvent.Subscribe(OnCloseItemShopUI)
            ItemShopUI.ItemPurchasedEvent.Subscribe(OnItemPurchased)
            if (set PlayerItemShopUI[Player] = ItemShopUI) {}
            ItemShopUI

        UI.Show()
        if (not IsResume?):
            UI.OnShow()

    <# ----------------------------------------------------------------------------------------------------
        Event Handler
    ---------------------------------------------------------------------------------------------------- #>
    OnPlayerAdded(Player:player):void =
        spawn{Initialize(Player)}

    OnPlayerRemoved(Player:player):void =
        set PlayerItemShopUI = RemoveKeyFromMap(PlayerItemShopUI, Player)
        set PlayerItemShopTimerUI = RemoveKeyFromMap(PlayerItemShopTimerUI, Player)

    <#
    OnGameStarted():void =
        spawn{Start()}

    OnBossFightStarted():void =
        Stop()
    #>

    OnGameEnded():void =
        Stop()

    OnCloseItemShopUI(Player:player):void =
        ResumeGame(Player)

    OnActivated(InPlayer:player):void =
        for (Player : GetActivePlayers()):
            Player.AwardAccolade(accolade_id.ItemShop)
            
        PauseGame(pause_type.ItemShop)
        Disable()
        ActivatedEvent.Signal()

    OnItemPurchased(Player:player):void =
        if (ItemShopPurchasedEvent := PlayerItemShopPurchasedEvents[Player]):
            ItemShopPurchasedEvent.Signal()