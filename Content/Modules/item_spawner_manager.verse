using { /Fortnite.com/Characters }
using { /Fortnite.com/Devices }
using { /Verse.org/Simulation }
using { /UnrealEngine.com/Temporary/SpatialMath }

<# ----------------------------------------------------------------------------------------------------
    Global Functions
---------------------------------------------------------------------------------------------------- #>
var GlobalItemSpawnerManager:weak_map(session, item_spawner_manager) = map{}

GetItemSpawnerManager()<decides><transacts>:item_spawner_manager =
    GlobalItemSpawnerManager[GetSession()]

SpawnItemDrop(Id:item_drop_id, Position:vector3):?item_drop =
    if (ItemSpawnerManager := GetItemSpawnerManager[]):
        return ItemSpawnerManager.Spawn(Id, Position)

    return false

ClearItemDrops():void =
    if (ItemSpawnerManager := GetItemSpawnerManager[]):
        ItemSpawnerManager.DisableAll()

(Agent:agent).SpawnItemDrop(Id:item_drop_id):void =
    if (ItemSpawnerManager := GetItemSpawnerManager[], Position := Agent.GetFortCharacter[].GetTransform().Translation - vector3{Z := CharacterHeightOffset}):
        ItemSpawnerManager.Spawn(Id, Position)

(Agent:agent).SpawnExperience(Amount:int):void =
    if (ItemSpawnerManager := GetItemSpawnerManager[], Position := Agent.GetFortCharacter[].GetTransform().Translation - vector3{Z := CharacterHeightOffset}):
        ItemSpawnerManager.SpawnExperience(Position, Amount)

SpawnExperience(Position:vector3, Amount:int):void =
    if (ItemSpawnerManager := GetItemSpawnerManager[]):
        ItemSpawnerManager.SpawnExperience(Position, Amount)

(Player:player).PickupAllExperience():void =
    if (ItemSpawnerManager := GetItemSpawnerManager[]):
        ItemSpawnerManager.PickupAllExperience(Player)

<# ----------------------------------------------------------------------------------------------------
    item_spawner_manager
---------------------------------------------------------------------------------------------------- #>
item_spawner_manager := class(creative_device):
    @editable
    ItemDropDefinitions:[]item_drop_definition = array{}
    
    ItemDropFactory:item_drop_factory = item_drop_factory{}

    MaxExperienceOrb:int = 100

    var MappedItemDropDefinitions<private>:[item_drop_id]item_drop_definition = map{}
    var ActiveItemDrops:[item_drop_id][]item_drop = map{}
    var ItemDropObjectPoolByType<private>:[item_drop_id]queue(item_drop) = map{}
    var ItemDropMapIndicatorQueue<private>:[item_drop_id]queue(map_indicator_device) = map{}

    OnBegin<override>()<suspends>:void =
        if (set GlobalItemSpawnerManager[GetSession()] = Self) {}
        
        for (ItemDropDefinition : ItemDropDefinitions, set MappedItemDropDefinitions[ItemDropDefinition.Id] = ItemDropDefinition):
            if:
                ItemDropMapIndicators := ItemDropDefinition.MaybeMapIndicators?
                set ItemDropMapIndicatorQueue[ItemDropDefinition.Id] = queue(map_indicator_device){Elements := ItemDropMapIndicators}
            then:
                for (ItemDropMapIndicator : ItemDropMapIndicators):
                    if (ItemDropMapIndicator.TeleportTo[MapIndicatorDisabledPosition, IdentityRotation()]) {}

    <# ----------------------------------------------------------------------------------------------------
        Get Functions
    ---------------------------------------------------------------------------------------------------- #>
    GetItemDropDefinition(Id:item_drop_id)<decides><transacts>:item_drop_definition =
        MappedItemDropDefinitions[Id]

    GetPooledItemDrop(Id:item_drop_id)<decides><transacts>:item_drop =
        if (not ItemDropObjectPoolByType[Id]):
            set ItemDropObjectPoolByType[Id] = queue(item_drop){}

        ItemDropObjectPool := ItemDropObjectPoolByType[Id]
        if (ItemDropObjectPool.IsEmpty()?):
            ItemDropFactory.GetItemDrop(Self, Id)
        else:
            DequeuedObject := ItemDropObjectPool.Dequeue[]
            set ItemDropObjectPoolByType[Id] = DequeuedObject(0)
            DequeuedObject(1)

    GetMapIndicator(Id:item_drop_id)<decides><transacts>:map_indicator_device =
        DequeuedObject := ItemDropMapIndicatorQueue[Id].Dequeue[]
        set ItemDropMapIndicatorQueue[Id] = DequeuedObject(0)
        DequeuedObject(1)

    ReturnMapIndicator(Id:item_drop_id, MapIndicator:map_indicator_device):void =
        if:
            UpdatedItemDropMapIndicatorQueue := ItemDropMapIndicatorQueue[Id].Enqueue(MapIndicator)
            set ItemDropMapIndicatorQueue[Id] = UpdatedItemDropMapIndicatorQueue
            MapIndicator.TeleportTo[MapIndicatorDisabledPosition, IdentityRotation()]

    <# ----------------------------------------------------------------------------------------------------
        Update Functions
    ---------------------------------------------------------------------------------------------------- #>               
    Spawn(Id:item_drop_id, Position:vector3):?item_drop =
        if (ItemDrop := GetPooledItemDrop[Id], ItemDropDefinition := GetItemDropDefinition[Id]):
            ItemDrop.OnBegin(transform{Translation := Position, Rotation := IdentityRotation().ApplyYaw(DegreesToRadians(90.0))}, ItemDropDefinition)
            if (set ActiveItemDrops[Id] = ActiveItemDrops[Id] + array{ItemDrop} or array{ItemDrop}) {}
            return option{ItemDrop}

        return false

    (item_spawner_manager:)SpawnExperience(Position:vector3, Amount:int):void =
        if (ActiveItemDrops[item_drop_id.Experience].Length >= MaxExperienceOrb):
            var MaybeNearestOrb:?item_drop_experience = false
            var NearestDistenceSquared:float = 0.0
            
            for:
                ActiveItemDrop : ActiveItemDrops[item_drop_id.Experience]
                ActiveItemDrop.IsActive?
                ActiveExperienceOrb := item_drop_experience[ActiveItemDrop]
                DistanceSquaredFromTarget := DistanceSquared(Position, ActiveItemDrop.GetPosition())
                not MaybeNearestOrb? or DistanceSquaredFromTarget < NearestDistenceSquared
            do:
                set MaybeNearestOrb = option{ActiveExperienceOrb}
                set NearestDistenceSquared = DistanceSquaredFromTarget

            if (NearestOrb := MaybeNearestOrb?):
                PreviousAmount := NearestOrb.XPAmount
                NearestOrb.Add(Amount)

            return

        if (ItemDrop := item_drop_experience[GetPooledItemDrop[item_drop_id.Experience]], ItemDropDefinition := GetItemDropDefinition[item_drop_id.Experience]):
            ItemDrop.OnBegin(transform{Translation := Position}, ItemDropDefinition)
            ItemDrop.Update(Amount)
            if (set ActiveItemDrops[item_drop_id.Experience] = ActiveItemDrops[item_drop_id.Experience] + array{ItemDrop} or array{ItemDrop}) {}

    PickupAllExperience(Player:player):void =
        for (ActiveItemDrop : ActiveItemDrops[item_drop_id.Experience], ActiveItemDrop.Id = item_drop_id.Experience):
            spawn{ActiveItemDrop.Pickup(Player)}

        if (set ActiveItemDrops[item_drop_id.Experience] = array{}) {}

    DisableAll():void =
        for (ActiveItemDrop : ActiveItemDrops):
            for (ItemDrop : ActiveItemDrop):
                ItemDrop.Disable()

    ReturnToObjectPool(ItemDrop:item_drop):void =
        if:
            set ActiveItemDrops[ItemDrop.Id] = for (ActiveItemDrop : ActiveItemDrops[ItemDrop.Id], ActiveItemDrop <> ItemDrop) {ActiveItemDrop}

        if:
            ItemDropObjectPool := ItemDropObjectPoolByType[ItemDrop.Id]
            set ItemDropObjectPoolByType[ItemDrop.Id] = ItemDropObjectPool.Enqueue(ItemDrop)

    <# ----------------------------------------------------------------------------------------------------
        Event Handlers
    ---------------------------------------------------------------------------------------------------- #>
