using { /Fortnite.com/Devices }
using { /Verse.org/Simulation }
using { /UnrealEngine.com/Temporary/UI }

<# ----------------------------------------------------------------------------------------------------
    Data Definitions
---------------------------------------------------------------------------------------------------- #>
message_id := enum:
    Message
    Victory
    Defeat
    BossWarning
    Timer
    PlayerCounter
    MessageSmall
    ExperienceBar
    ExperienceBarLevel
    CurrencyWood
    CurrencyElimination
    BossHealthBar
    Countdown
    CurrencyGold
    AutoSelectedAbilities
    CurrencyTreasureChest
    MissionComplete
    DailyQuestComplete
    Revives
    AutoAim
    TreasureChestCollect
    PlayerDowned
    StageIndicator
    ItemShopNotification
    PlayerCounterShip
    SupplyDropNotification
    TimerShip
    RankUp01
    RankUp02
    RankUp03
    RankUp04
    RankUp05
    RankUp06
    RankUp07
    RankUp08
    RankUp09
    MinibossEliminatedCounter

ToString(Id:message_id)<transacts>:string =
    case (Id):
        message_id.Message => "Message"
        message_id.Victory => "Victory"
        message_id.Defeat => "Defeat"
        message_id.BossWarning => "BossWarning"
        message_id.Timer => "Timer"
        message_id.PlayerCounter => "PlayerCounter"
        message_id.MessageSmall => "MessageSmall"
        message_id.ExperienceBar => "ExperienceBar"
        message_id.ExperienceBarLevel => "ExperienceBarLevel"
        message_id.CurrencyWood => "CurrencyWood"
        message_id.CurrencyElimination => "CurrencyElimination"
        message_id.BossHealthBar => "BossHealthBar"
        message_id.Countdown => "Countdown"
        message_id.CurrencyGold => "CurrencyGold"
        message_id.AutoSelectedAbilities => "AutoSelectedAbilities"
        message_id.CurrencyTreasureChest => "CurrencyTreasureChest"
        message_id.MissionComplete => "MissionComplete"
        message_id.DailyQuestComplete => "DailyQuestComplete"
        message_id.Revives => "Revives"
        message_id.AutoAim => "AutoAim"
        message_id.TreasureChestCollect => "TreasureChestCollect"
        message_id.PlayerDowned => "PlayerDowned"
        message_id.StageIndicator => "StageIndicator"
        message_id.ItemShopNotification => "ItemShopNotification"
        message_id.PlayerCounterShip => "PlayerCounterShip"
        message_id.SupplyDropNotification => "SupplyDropNotification"
        message_id.TimerShip => "TimerShip"
        message_id.RankUp01 => "RankUp01"
        message_id.RankUp02 => "RankUp02"
        message_id.RankUp03 => "RankUp03"
        message_id.RankUp04 => "RankUp04"
        message_id.RankUp05 => "RankUp05"
        message_id.RankUp06 => "RankUp06"
        message_id.RankUp07 => "RankUp07"
        message_id.RankUp08 => "RankUp08"
        message_id.RankUp09 => "RankUp09"
        message_id.MinibossEliminatedCounter => "MinibossEliminatedCounter"
        
hud_message := class<concrete>:
    @editable
    Id:message_id = message_id.Message
    @editable
    Device:hud_message_device = hud_message_device{}
    @editable
    IsConsumeInput:logic = false

<# ----------------------------------------------------------------------------------------------------
    Global Functions
---------------------------------------------------------------------------------------------------- #>
var GlobalMessageManager:weak_map(session, message_manager) = map{}

GetMessageManager()<decides><transacts>:message_manager =
    GlobalMessageManager[GetSession()]

(Agent:agent).ShowMessage(Id:message_id, Message:message, ?DisplayTime:float = 5.0):void =
    if (MessageManager := GetMessageManager[]):
        MessageManager.Show(Agent, Id, Message, ?DisplayTime := DisplayTime)

(Agent:agent).ShowMessage(Id:message_id, Message:string, ?DisplayTime:float = 5.0):void =
    if (MessageManager := GetMessageManager[]):
        MessageManager.Show(Agent, Id, MakeMessage(Message), ?DisplayTime := DisplayTime)

(Agent:agent).ShowMessage(Id:message_id):void =
    if (MessageManager := GetMessageManager[]):
        MessageManager.Show(Agent, Id)

(Agent:agent).HideMessage(Id:message_id):void =
    if (MessageManager := GetMessageManager[]):
        MessageManager.Hide(Agent, Id)

ShowMessage(Id:message_id, Message:string, ?DisplayTime:float = 5.0):void =
    if (MessageManager := GetMessageManager[]):
        MessageManager.Show(Id, MakeMessage(Message), ?DisplayTime := DisplayTime)

HideMessage(Id:message_id):void =
    if (MessageManager := GetMessageManager[]):
        for (Player : GetAllPlayers()):
            MessageManager.Hide(Player, Id)

<# ----------------------------------------------------------------------------------------------------
    message_manager
---------------------------------------------------------------------------------------------------- #>
message_manager := class(creative_device):
    @editable
    Messages:[]hud_message = array{}

    var MappedMessages:[message_id]hud_message = map{}

    <# ----------------------------------------------------------------------------------------------------
        Initialization
    ---------------------------------------------------------------------------------------------------- #>
    OnBegin<override>()<suspends>:void =
        if (set GlobalMessageManager[GetSession()] = Self) {}
        
        for (Message : Messages, set MappedMessages[Message.Id] = Message) {}

    <# ----------------------------------------------------------------------------------------------------
        Get Functions
    ---------------------------------------------------------------------------------------------------- #>
    GetMessage(Id:message_id)<decides><transacts>:hud_message =
        MappedMessages[Id]

    <# ----------------------------------------------------------------------------------------------------
        Update Functions
    ---------------------------------------------------------------------------------------------------- #>
    Show(Id:message_id, Message:message, ?DisplayTime:float = 5.0):void =
        if (HUDMessage := GetMessage[Id]):
            HUDMessage.Device.Show(Message, ?DisplayTime := DisplayTime)

    Show(Agent:agent, Id:message_id, Message:message, ?DisplayTime:float = 5.0):void =
        if (HUDMessage := GetMessage[Id]):
            HUDMessage.Device.Show(Agent, Message, ?DisplayTime := DisplayTime)
            if (HUDMessage.IsConsumeInput?):
                spawn{ConsumeInput(Agent, DisplayTime)}

    Show(Agent:agent, Id:message_id):void =
        if (HUDMessage := GetMessage[Id]):
            HUDMessage.Device.Show(Agent)

    Hide(Id:message_id):void =
        if (HUDMessage := GetMessage[Id]):
            HUDMessage.Device.Hide()

    Hide(Agent:agent, Id:message_id):void =
        if (HUDMessage := GetMessage[Id]):
            HUDMessage.Device.Hide(Agent)

    ConsumeInput(Agent:agent, Duration:float)<suspends>:void =
        if (Player := player[Agent], PlayerUI := GetPlayerUI[Player]):
            UI := canvas{}
            PlayerUI.AddWidget(UI, Slot := player_ui_slot{InputMode := ui_input_mode.All})
            Sleep(Duration)
            PlayerUI.RemoveWidget(UI)