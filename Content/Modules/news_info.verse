using { /Verse.org/Assets }
using { /Verse.org/Colors }

news_info := struct<concrete>:
    Date:string = ""
    Title:string = ""
    Version:float = 0.0
    MaybeImage:?texture = false
    Entries:[]news_entry_info = array{}
    
news_entry_info := struct<concrete>:
    Type:news_entry_type = news_entry_type.NewFeatures
    Notes:string = ""

news_entry_type := enum:
    None
    NewFeatures
    BugFixes
    PerformanceOptimizations
    GameplayChanges

(Type:news_entry_type).ToIcon()<transacts>:texture =
    case (Type):
        news_entry_type.None => UserInterface.Icons.NewsEntryType_NewFeature_64
        news_entry_type.NewFeatures => UserInterface.Icons.NewsEntryType_NewFeature_64
        news_entry_type.BugFixes => UserInterface.Icons.NewsEntryType_BugFix_64
        news_entry_type.PerformanceOptimizations => UserInterface.Icons.NewsEntryType_PerformanceOptimization_64
        news_entry_type.GameplayChanges => UserInterface.Icons.NewsEntryType_GameplayChange_64

(Type:news_entry_type).ToColor()<transacts>:color =
    case (Type):
        news_entry_type.None => MakeColorFromHex("FF3C47")
        news_entry_type.NewFeatures => MakeColorFromHex("F7FF19")
        news_entry_type.BugFixes => MakeColorFromHex("FF3C47")
        news_entry_type.PerformanceOptimizations => MakeColorFromHex("A7FF90")
        news_entry_type.GameplayChanges => MakeColorFromHex("3CBFFF")
        
GetNewsInfo()<transacts>:[]news_info =
    array:
        news_info:
            Date := "March 08, 2025"
            Title := "Update 22 - Arena difficulty levels, game speed modes, new arenas & power ups and more."
            Version := 0.700
            MaybeImage := option{UserInterface.News.NewsUpdateHighlights_01}
            Entries := array:
                news_entry_info{Type := news_entry_type.NewFeatures, Notes := "Added Arena Difficulty levels - re-challenge arenas with higher multipliers."}
                news_entry_info{Type := news_entry_type.NewFeatures, Notes := "Implemented Arena Unlock progression."}
                news_entry_info{Type := news_entry_type.None, Notes := "Arenas now need to be unlocked by completing the previous arena level."}
                news_entry_info{Type := news_entry_type.None, Notes := "In multiplayer, the lowest unlocked arena level between all players will be available."}
                news_entry_info{Type := news_entry_type.NewFeatures, Notes := "Added arena Game Speed modes (x1/x2/x3)."}
                news_entry_info{Type := news_entry_type.None, Notes := "Timer counts down faster, enemies move faster (+0/50/100% movespeed)."}
                news_entry_info{Type := news_entry_type.None, Notes := "Missions, supply drop & gather speed scales with the game speed."}
                news_entry_info{Type := news_entry_type.NewFeatures, Notes := "Reworked Arenas:"}
                news_entry_info{Type := news_entry_type.None, Notes := "Arena 4 - Reality Falls (New!)"}
                news_entry_info{Type := news_entry_type.None, Notes := "Arena 7 - The Dojo (Reordered)"}
                news_entry_info{Type := news_entry_type.None, Notes := "Arena 8 - Bushido Dojo (Duplicates Removed)"}
                news_entry_info{Type := news_entry_type.None, Notes := "Arena 9 - Flooded Frog Temple (New!)"}
                news_entry_info{Type := news_entry_type.NewFeatures, Notes := "Added Treasure Chest Tier power up - increase the minimum tier granted."}
                news_entry_info{Type := news_entry_type.NewFeatures, Notes := "Added Revive Blast power up - deal damage to enemies around you on auto revive."}
                news_entry_info{Type := news_entry_type.NewFeatures, Notes := "Updated supply drop minimap icon with a pulsing circle animation."}
                news_entry_info{Type := news_entry_type.NewFeatures, Notes := "Added an info panel to the lobby to show the arena settings."}
                news_entry_info{Type := news_entry_type.NewFeatures, Notes := "Added player rank up animations."}
                news_entry_info{Type := news_entry_type.NewFeatures, Notes := "Updated the player progress UI:"}
                news_entry_info{Type := news_entry_type.None, Notes := "Changed stars icons to progress bars to support higher arena levels."}
                news_entry_info{Type := news_entry_type.None, Notes := "Added Rank Info tab to show all available ranks and required arena levels."}
                news_entry_info{Type := news_entry_type.None, Notes := "Added power up progress and amount of gold invested in power ups."}
                news_entry_info{Type := news_entry_type.NewFeatures, Notes := "Added option to reopen the Game Result UI (via the podium view)."}
                news_entry_info{Type := news_entry_type.NewFeatures, Notes := "Added Player Info tab to the Arena Select UI."}
                news_entry_info{Type := news_entry_type.NewFeatures, Notes := "Updated Arena Select UI timeout now only applies in multiplayer."}
                news_entry_info{Type := news_entry_type.NewFeatures, Notes := "Updated Ability Select UI banish mode to show cross icon and red outlines."}

                news_entry_info{Type := news_entry_type.GameplayChanges, Notes := "Arena stage multiplier is now a fixed +50% Health/Damage for all arenas."}
                news_entry_info{Type := news_entry_type.GameplayChanges, Notes := "Arena stage count now scales with difficulty level instead of the arena."}
                news_entry_info{Type := news_entry_type.GameplayChanges, Notes := "Health Drop rate changed to use pseudo random chance."}
                news_entry_info{Type := news_entry_type.None, Notes := "(10% drop rate > more likely to drop around every 10th trees)"}
                news_entry_info{Type := news_entry_type.GameplayChanges, Notes := "Changed player ranks to be based on arena progress instead of weapon progress."}
                news_entry_info{Type := news_entry_type.None, Notes := "Rank up after completing a difficulty level on all arenas (every 9th level)."}
                news_entry_info{Type := news_entry_type.GameplayChanges, Notes := "Renamed player ranks to: Genin I/II/III, Chunin I/II/III, Jonin I/II/III and Ninja."}
                news_entry_info{Type := news_entry_type.GameplayChanges, Notes := "Changed Daily Quest multiplier to be based on arena progress."}
                news_entry_info{Type := news_entry_type.GameplayChanges, Notes := "All minibosses now need to be eliminated before continuing to the next stage."}

                news_entry_info{Type := news_entry_type.BugFixes, Notes := "Fixed - Tree gathering duration no longer scale with player count."}
                news_entry_info{Type := news_entry_type.BugFixes, Notes := "Fixed - Stage multiplier was being applied from stage 1 instead of stage 2 onwards."}

        news_info:
            Date := "February 26, 2025"
            Title := "Update 21 - Rebranding to \"Ninja Swarm\", small gameplay changes & bug fixes."
            Version := 0.610
            MaybeImage := option{UserInterface.News.NewsUpdateHighlights_01}
            Entries := array:
                news_entry_info{Type := news_entry_type.NewFeatures, Notes := "Added a Select Buffer Time (0-3s) to settings menu to prevent accidental selection."}
                news_entry_info{Type := news_entry_type.NewFeatures, Notes := "Added a cooldown timer for petting the slime."}
                news_entry_info{Type := news_entry_type.NewFeatures, Notes := "Updated game result podium view to be exited using UI button."}

                news_entry_info{Type := news_entry_type.GameplayChanges, Notes := "Reduced the Mission Critical daily quest task amounts from 5/10/15 to 3/6/9."}
                news_entry_info{Type := news_entry_type.GameplayChanges, Notes := "Added separate player damaged buffer time for miniboss/boss slimes."}
                news_entry_info{Type := news_entry_type.None, Notes := "Players can now be damaged by miniboss/boss even if damaged by a normal slime within 1s."}
                news_entry_info{Type := news_entry_type.GameplayChanges, Notes := "Changed boss fight barrage attack to be focused around the player position."}
                news_entry_info{Type := news_entry_type.None, Notes := "Previously was targeting a random position around the entire arena."}

                news_entry_info{Type := news_entry_type.BugFixes, Notes := "Fixed - Supply drop was not spawning near the edge of the arena (reduced the spawn radius)."}
                news_entry_info{Type := news_entry_type.BugFixes, Notes := "Fixed - Orbit/Orbit Free camera glitched after game intro sequence completed."}
                news_entry_info{Type := news_entry_type.BugFixes, Notes := "Fixed - Stuck in pause if player opens the menu & minimap at the same time."}
                news_entry_info{Type := news_entry_type.None, Notes := "(press menu key again to unpause)"}
                news_entry_info{Type := news_entry_type.BugFixes, Notes := "Fixed - Music was muffled when player is at the center of the arena."}
                news_entry_info{Type := news_entry_type.BugFixes, Notes := "Fixed - Petting slime was only working for a single player."}
                news_entry_info{Type := news_entry_type.BugFixes, Notes := "Fixed - Accolades for supply drop & stage complete was not granting."}
                news_entry_info{Type := news_entry_type.BugFixes, Notes := "Fixed - Game progress bar UI was showing failed status for Item Shop on stage complete."}
                news_entry_info{Type := news_entry_type.BugFixes, Notes := "Fixed - Capture area minimap icon was always showing."}

        news_info:
            Date := "February 16, 2025"
            Title := "Update 20 - Arena Stages, Supply Drops, Performance Optimizations, Game Progress UI & more!"
            Version := 0.600
            MaybeImage := option{UserInterface.News.NewsUpdateHighlights_01}
            Entries := array:
                news_entry_info{Type := news_entry_type.NewFeatures, Notes := "Implemented Arena Stage mechanic."}
                news_entry_info{Type := news_entry_type.None, Notes := "Arenas now have a varying number of stages to complete (3-5)."}
                news_entry_info{Type := news_entry_type.None, Notes := "A Storm Circle is spawned at the end of each stage."}
                news_entry_info{Type := news_entry_type.None, Notes := "All players must evacuate onto the ship before the storm closes (resulting in defeat)."}
                news_entry_info{Type := news_entry_type.None, Notes := "Completing a stage grants bonus gold & access to the Item Shop."}
                news_entry_info{Type := news_entry_type.NewFeatures, Notes := "Added Game Progress Bar UI to indicate upcoming game events (e.g. miniboss, missions)."}
                news_entry_info{Type := news_entry_type.NewFeatures, Notes := "Added new Supply Drop game event/mechanic."}
                news_entry_info{Type := news_entry_type.None, Notes := "Supply Drop spawns once every stage, secure the Supply Drop (similar to capture mission)."}
                news_entry_info{Type := news_entry_type.None, Notes := "Supply Drop grants bonus experience orbs & a guaranteed experience magnet"}
                news_entry_info{Type := news_entry_type.NewFeatures, Notes := "Added 7 background music, a unique music for each arena and the lobby area."}
                news_entry_info{Type := news_entry_type.NewFeatures, Notes := "Added minimap icon for Experience Magnet."}
                news_entry_info{Type := news_entry_type.NewFeatures, Notes := "Updated the Experience Magnet item drop."}
                news_entry_info{Type := news_entry_type.NewFeatures, Notes := "Updated the Equipped Weapon HUD UI to show for all players."}
                news_entry_info{Type := news_entry_type.NewFeatures, Notes := "Added the ability to push downed players using the dash ability."}
                news_entry_info{Type := news_entry_type.NewFeatures, Notes := "Added info panel slideshow to the lobby area to show gameplay hints."}
                news_entry_info{Type := news_entry_type.NewFeatures, Notes := "Updated the projectile warning VFX to fade out along the projectile's path."}
                news_entry_info{Type := news_entry_type.NewFeatures, Notes := "Updated the health orb minimap icon (brighter & outlined to improve visibility)."}
                news_entry_info{Type := news_entry_type.NewFeatures, Notes := "Updated Daily Quest HUD to all have a fixed width."}
                news_entry_info{Type := news_entry_type.NewFeatures, Notes := "Updated several HUD UI to align font size & style."}

                news_entry_info{Type := news_entry_type.GameplayChanges, Notes := "Changed tree gathering to scale with player count (1/2/3/4 players)."}
                news_entry_info{Type := news_entry_type.None, Notes := "Tree average base wood amount changed from 45/45/45/45 to 45/22.5/15/11.25 wood."}
                news_entry_info{Type := news_entry_type.None, Notes := "Tree respawn time changed from 90/90/90/90 to 90/45/30/22.5s."}
                news_entry_info{Type := news_entry_type.None, Notes := "Tree item drop rate changed from 20/20/20/20 to 10/7.5/5/2.5%."}
                news_entry_info{Type := news_entry_type.GameplayChanges, Notes := "Multiple players can now gather from the same tree to speed up gather speed."}
                news_entry_info{Type := news_entry_type.GameplayChanges, Notes := "Enabled transparent minimap while in the arena."}
                news_entry_info{Type := news_entry_type.GameplayChanges, Notes := "Experience Magnet no longer drops from trees (now obtained through supply drops)."}
                news_entry_info{Type := news_entry_type.GameplayChanges, Notes := "Purple Snow Miniboss - projectile count reduced 8 > 4, range increased 300 > 450."}
                news_entry_info{Type := news_entry_type.GameplayChanges, Notes := "Increased NPC movementspeed for Arenas 1/2 from +0/25% > +50/50%."}

                news_entry_info{Type := news_entry_type.PerformanceOptimizations, Notes := "Lobby area is now unloaded while players are in the arena."}
                news_entry_info{Type := news_entry_type.PerformanceOptimizations, Notes := "Removed tree gathering progress indicator VFX (tree size already indicates progress)."}
                news_entry_info{Type := news_entry_type.PerformanceOptimizations, Notes := "Disabled movement speed trail VFX."}
                news_entry_info{Type := news_entry_type.PerformanceOptimizations, Notes := "Switched VFX with high particle count from CPU to GPU simulation target."}
                news_entry_info{Type := news_entry_type.PerformanceOptimizations, Notes := "Reduced texture resolution for various VFX textures & UI icons."}
                news_entry_info{Type := news_entry_type.PerformanceOptimizations, Notes := "Combined damage numbers & hit impact into a single VFX."}
                news_entry_info{Type := news_entry_type.PerformanceOptimizations, Notes := "Optimized the damage number & hit impact VFX."}
                news_entry_info{Type := news_entry_type.PerformanceOptimizations, Notes := "Optimized the NPC circle attack indicator VFX."}
                news_entry_info{Type := news_entry_type.PerformanceOptimizations, Notes := "Updated VFX to be disposed immediately instead of using default 5s lifetime."}
                news_entry_info{Type := news_entry_type.PerformanceOptimizations, Notes := "Added a damage number & hit impact VFX spawn rate limit 0.1s per NPC."}

                news_entry_info{Type := news_entry_type.BugFixes, Notes := "Fixed - Game could be started multiple times causing 2 bosses & timer ticking 2/s."}
                news_entry_info{Type := news_entry_type.BugFixes, Notes := "Fixed - NPC attack VFX sometimes disappearing before completing."}
                news_entry_info{Type := news_entry_type.BugFixes, Notes := "Fixed - Treasure Chest HUD sometimes not hiding when reaching 0 amount."}
                news_entry_info{Type := news_entry_type.BugFixes, Notes := "Fixed - Trees sometimes respawn with an incorrect size."}
                news_entry_info{Type := news_entry_type.BugFixes, Notes := "Fixed - Arena Select UI Lv. not updated when viewed by another player."}
                news_entry_info{Type := news_entry_type.BugFixes, Notes := "Fixed - Boss health bar overlapping with Daily Quest HUD (reduced width)."}
                news_entry_info{Type := news_entry_type.BugFixes, Notes := "Fixed - Tree minimap icons were visible while the arena was still loading."}
                news_entry_info{Type := news_entry_type.BugFixes, Notes := "Fixed - Projectile warning arrow pattern was repeating at higher widths (for boss)."}
                news_entry_info{Type := news_entry_type.BugFixes, Notes := "Fixed - Lightning focus fire was not working on the boss in higher level arenas."}

        news_info:
            Date := "February 07, 2025"
            Title := "Update 19 - Boons codex, post game accolades"
            Version := 0.510
            MaybeImage := option{UserInterface.News.NewsUpdateHighlights_01}
            Entries := array:               
                news_entry_info{Type := news_entry_type.NewFeatures, Notes := "Added Weapon Evolution info to the Weapon Select UI."}
                news_entry_info{Type := news_entry_type.NewFeatures, Notes := "Added Boon Codex to the hub area (UI to show boon info)."}
                news_entry_info{Type := news_entry_type.NewFeatures, Notes := "Added type info to boon select UI (Combat, Survivability, Utility)."}
                news_entry_info{Type := news_entry_type.NewFeatures, Notes := "Added Post Game Accolades to the hub area (player with most damage/revives/wood collected/times downed)."}
                news_entry_info{Type := news_entry_type.NewFeatures, Notes := "Updated Menu UI to pause the game when only 1 player is in the game."}
                news_entry_info{Type := news_entry_type.NewFeatures, Notes := "Updated the dash input control switch to show icon instead of info text."}
                news_entry_info{Type := news_entry_type.NewFeatures, Notes := "Added Matchmaking Portal to the hub area."}
                news_entry_info{Type := news_entry_type.NewFeatures, Notes := "Added player's Gold amount to the player progress UI."}
                news_entry_info{Type := news_entry_type.NewFeatures, Notes := "Added HUD & audio notification for when a player is downed."}

                news_entry_info{Type := news_entry_type.GameplayChanges, Notes := "Boomerang Weapon changed to return to the players location instead of cast position."}
                news_entry_info{Type := news_entry_type.GameplayChanges, Notes := "Boomerang Weapon base range increased from 500 > 600."}
                news_entry_info{Type := news_entry_type.GameplayChanges, Notes := "Lightning Weapon focus fire reworked so all projectiles will target the highest health."}
                news_entry_info{Type := news_entry_type.GameplayChanges, Notes := "Updated Trees to always drop an experience orb."}
                news_entry_info{Type := news_entry_type.GameplayChanges, Notes := "NPCs now have 0.5s invulnerability time after spawning (to prevent lag from big Aura)."}
                news_entry_info{Type := news_entry_type.GameplayChanges, Notes := "Increased game start countdown time from 15 > 30s."}
                news_entry_info{Type := news_entry_type.GameplayChanges, Notes := "Reduced player revive decay rate from 10% > 5%/s."}
                news_entry_info{Type := news_entry_type.GameplayChanges, Notes := "Boss Siphon boon now grants bonus HP on assists (if player has dealt at least 10% damage)."}
                news_entry_info{Type := news_entry_type.GameplayChanges, Notes := "Damage to Golden Slime is now always a percentage of its max health."}
                news_entry_info{Type := news_entry_type.None, Notes := "Percentage scales with the number of players (i.e. 1.00/0.50/0.33/0.25% with 4/3/2/1 players)."}
                news_entry_info{Type := news_entry_type.GameplayChanges, Notes := "Changed Golden Pyramids collectibles to Scrolls (visual change only)."}

                news_entry_info{Type := news_entry_type.BugFixes, Notes := "Fixed - Wooden Club/Arrow boons were not being capped to max value (50%)."}
                news_entry_info{Type := news_entry_type.BugFixes, Notes := "Fixed - Wooden Arrow was granting 0.5% instead of 1.0% per 100 wood."}
                news_entry_info{Type := news_entry_type.BugFixes, Notes := "Fixed - Game Result UI total ability values were always showing the local player's power up amounts."}
                news_entry_info{Type := news_entry_type.BugFixes, Notes := "Fixed - Chainsaw was not applying immediately if player is already gathering."}
                news_entry_info{Type := news_entry_type.BugFixes, Notes := "Fixed - Player was not being revived when getting Immortal Heart while downed (causing respawn into hub area)."}
                news_entry_info{Type := news_entry_type.BugFixes, Notes := "Fixed - Power Up UI was still showing cost when power up is already maxed (now shows N/A)."}

        news_info:
            Date := "February 01, 2025"
            Title := "Update 18 - New Boons mechanic, game start cinematic, UI improvements & bug fixes"
            Version := 0.500
            MaybeImage := option{UserInterface.News.NewsUpdateHighlights_01}
            Entries := array:
                news_entry_info{Type := news_entry_type.NewFeatures, Notes := "Implemented Boons mechanic - complete missions to earn a random Boon granting a unique bonus."}
                news_entry_info{Type := news_entry_type.None, Notes := "Chainsaw - Doubles your tree gathering speed."}
                news_entry_info{Type := news_entry_type.None, Notes := "Discount Coupon - 20% off cost in the Item Shop."}
                news_entry_info{Type := news_entry_type.None, Notes := "Immortal Heart - Grants 1 revive. +25 Max Health each time an auto revive is used."}
                news_entry_info{Type := news_entry_type.None, Notes := "Boss Hunter - Deal bonus +50% damage to bosses & minibosses."}
                news_entry_info{Type := news_entry_type.None, Notes := "Boss Hunter - Eliminating minibosses heals & grants +25 max health."}
                news_entry_info{Type := news_entry_type.None, Notes := "Boss Siphon - Your starting weapon gains +50% damage. All other weapons receive -25% damage."}
                news_entry_info{Type := news_entry_type.None, Notes := "All For One - +1% damage for every weapon level (max +12% per weapon)."}
                news_entry_info{Type := news_entry_type.None, Notes := "Multi Tool - The first ability reroll is free on level ups."}
                news_entry_info{Type := news_entry_type.None, Notes := "Spare Dice - +25% Luck every time you reroll. Bonus luck resets once you select an ability."}
                news_entry_info{Type := news_entry_type.None, Notes := "Lucky Dice - Grants 3 Treasure Chests."}
                news_entry_info{Type := news_entry_type.None, Notes := "Dashing Wings - Increase Dash height and distance (+100% upwards impulse)."}
                news_entry_info{Type := news_entry_type.None, Notes := "Wooden Club - +1% Damage for every 100 Wood you are holding (max +50% at 5000 Wood)."}
                news_entry_info{Type := news_entry_type.None, Notes := "Wooden Arrow - +1% Critical Rate for every 100 Wood you are holding (max +50% at 5000 Wood)."}
                news_entry_info{Type := news_entry_type.None, Notes := "Critical Overdrive - Critical Rate above 100% is converted to +2% Critical Damage (e.g. 125% Critical Rate converts to +50% Critical Damage)."}
                news_entry_info{Type := news_entry_type.None, Notes := "Ninja Run - +1% Dodge Chance for every 10% Movement Speed you have (can still be stunned on dodge)."}
                news_entry_info{Type := news_entry_type.NewFeatures, Notes := "Added Boon Reroll power up - increase amount of times boons can be rerolled."}
                news_entry_info{Type := news_entry_type.NewFeatures, Notes := "Added a game start cinematic sequence."}
                news_entry_info{Type := news_entry_type.NewFeatures, Notes := "Added the option to purchase additional rerolls in the Item Shop."}
                news_entry_info{Type := news_entry_type.NewFeatures, Notes := "Updated the Item Shop UI layout."}
                news_entry_info{Type := news_entry_type.NewFeatures, Notes := "Moved the Team Experience stat into the stat info UI."}
                news_entry_info{Type := news_entry_type.NewFeatures, Notes := "Added shorted stat names to the Stat Info UI."}
                news_entry_info{Type := news_entry_type.NewFeatures, Notes := "Added Info Stats & Tier Rates to the Item Shop UI."}
                news_entry_info{Type := news_entry_type.NewFeatures, Notes := "Added Boon info to the Stats Info UI."}
                news_entry_info{Type := news_entry_type.NewFeatures, Notes := "Added Treasure Chest collected HUD animation."}
                news_entry_info{Type := news_entry_type.NewFeatures, Notes := "Added Mission started HUD animation."}
                news_entry_info{Type := news_entry_type.NewFeatures, Notes := "Added VFX when revive is regenerated."}
                
                news_entry_info{Type := news_entry_type.GameplayChanges, Notes := "Updated missions to grant a Boon instead of a Treasure Chest."}
                news_entry_info{Type := news_entry_type.GameplayChanges, Notes := "Re-ordered the abilities in the stat info UI."}
                news_entry_info{Type := news_entry_type.GameplayChanges, Notes := "Updated gathering VFX color to be based on gather duration 5/4/3/2/1s."}
                news_entry_info{Type := news_entry_type.GameplayChanges, Notes := "Updated slime spawn/teleport to face target player and be slightly above the ground."}
                
                news_entry_info{Type := news_entry_type.BugFixes, Notes := "Fixed - Camera gets stuck in arena preview if player is in Arena Select UI while countdown completes."}
                news_entry_info{Type := news_entry_type.BugFixes, Notes := "Fixed - Ability Select/Item Shop was filtering out abilities from the previous level."}
                news_entry_info{Type := news_entry_type.BugFixes, Notes := "Fixed - Item Shop currency rounding error (cannot afford when wood is equal to cost)."}
                news_entry_info{Type := news_entry_type.BugFixes, Notes := "Fixed - Revive regeneration starting when a game stat (i.e. Reroll) was less than max revive charge."}

        news_info:
            Date := "January 26, 2025"
            Title := "Update 17 - New power ups, more survivability, accurate stats & improved NPCs"
            Version := 0.430
            MaybeImage := option{UserInterface.News.NewsUpdateHighlights_01}
            Entries := array:
                news_entry_info{Type := news_entry_type.NewFeatures, Notes := "Added input trigger to toggle ON/OFF the Auto Aim setting."}
                news_entry_info{Type := news_entry_type.NewFeatures, Notes := "Added Treasure Chest Focus power up - to grant only abilities the player has selected."}
                news_entry_info{Type := news_entry_type.NewFeatures, Notes := "Added Air Dash power up - enables ability to activate an extra dash while airborne."}
                news_entry_info{Type := news_entry_type.NewFeatures, Notes := "Added Revive Regeneration power up - regain revive charges after a 3 minute cooldown."}
                news_entry_info{Type := news_entry_type.NewFeatures, Notes := "Updated Power Up UI to show stat type (additive/multiplicative)."}
                news_entry_info{Type := news_entry_type.NewFeatures, Notes := "Added current ability stat value to the Item Shop UI."}
                news_entry_info{Type := news_entry_type.NewFeatures, Notes := "Added a player damaged buffer - player can only receive damage every 1s."}
                news_entry_info{Type := news_entry_type.NewFeatures, Notes := "Updated Stats Info UI to show total stats (including power up amounts)."}
                news_entry_info{Type := news_entry_type.NewFeatures, Notes := "Updated Level Up/Treasure Chest/Item Shop UI to show total stats (including power up amounts)."}

                news_entry_info{Type := news_entry_type.GameplayChanges, Notes := "Updated NPCs to no longer spawn/teleport to the edge of the arena when players are nearby."}
                news_entry_info{Type := news_entry_type.GameplayChanges, Notes := "Player can no longer be stunned while invulnerable."}
                news_entry_info{Type := news_entry_type.GameplayChanges, Notes := "Reduced Snow Miniboss projectile speed from 1000 > 750 and from range 2000 > 1500."}
                news_entry_info{Type := news_entry_type.GameplayChanges, Notes := "Increased Health Regeneration power up from 5% > 10% per level."}
                news_entry_info{Type := news_entry_type.GameplayChanges, Notes := "Increased the Treasure Luck power up from 10% > 20% per level."}
                news_entry_info{Type := news_entry_type.GameplayChanges, Notes := "Increased the size of beach ball & goal area."}
                news_entry_info{Type := news_entry_type.GameplayChanges, Notes := "Increased the knockback range 400 > 800 and stun duration 1s > 3s on player revive."}
                news_entry_info{Type := news_entry_type.GameplayChanges, Notes := "Updated the projectile warning VFX to only show for projectile path & removed VFX fade."}
                news_entry_info{Type := news_entry_type.GameplayChanges, Notes := "Updated Team Experience UI to show multiplier amount (e.g. +10% instead of 110%)."}

                news_entry_info{Type := news_entry_type.BugFixes, Notes := "Fixed - NPC sometimes getting stuck navigating/targeting the incorrect position."}
                news_entry_info{Type := news_entry_type.BugFixes, Notes := "Fixed - Stat Info UI showing incorrect value for Max Health."}
                news_entry_info{Type := news_entry_type.BugFixes, Notes := "Fixed - Treasure Chest UI showing default icons/values for 1 frame."}
                news_entry_info{Type := news_entry_type.BugFixes, Notes := "Fixed - Arena Select UI unable to start game if the currently loaded arena is selected."}

        news_info:
            Date := "January 21, 2025"
            Title := "Update 16 - New settings, improved random positioning, VFX reworks & various bug fixes"
            Version := 0.420
            MaybeImage := option{UserInterface.News.NewsUpdateHighlights_01}
            Entries := array:
                news_entry_info{Type := news_entry_type.NewFeatures, Notes := "Added Auto Aim setting - weapons with aiming will auto target the nearest enemy."}
                news_entry_info{Type := news_entry_type.NewFeatures, Notes := "Added Camera Distance setting - change camera distance for Twin Stick/Orthographic."}
                news_entry_info{Type := news_entry_type.NewFeatures, Notes := "Updated minimap icons for minibosses to show variant."}
                news_entry_info{Type := news_entry_type.NewFeatures, Notes := "Added minimap icons for health pickups."}
                news_entry_info{Type := news_entry_type.NewFeatures, Notes := "Updated VFX for the Trail weapon (no gameplay changes)."}
                news_entry_info{Type := news_entry_type.NewFeatures, Notes := "Updated the gathering VFX - axe color now changes with gather rate."}
                news_entry_info{Type := news_entry_type.None, Notes := "0/25/50/75/100% Gather Rate > Common/Uncommon/Rare/Epic/Legendary color."}
                news_entry_info{Type := news_entry_type.NewFeatures, Notes := "Added Gather Rate power up."}
                news_entry_info{Type := news_entry_type.NewFeatures, Notes := "Implemented collision detection for random positioning to avoid object overlapping:"}
                news_entry_info{Type := news_entry_type.None, Notes := "Trees no longer overlap/spawn too close to arena environment."}
                news_entry_info{Type := news_entry_type.None, Notes := "Collectibles no longer overlap/spawn too close to arena environment."}
                news_entry_info{Type := news_entry_type.NewFeatures, Notes := "Updated the Game Result - Team UI to include player wood collected, revives & downed count."}
                news_entry_info{Type := news_entry_type.NewFeatures, Notes := "Updated the enemy projectile warning VFX:"}
                news_entry_info{Type := news_entry_type.None, Notes := "VFX now remains and indicates the path of the projectile."}
                news_entry_info{Type := news_entry_type.None, Notes := "Reduced the fade amount of the VFX to increase visibility at the furthest range."}
                news_entry_info{Type := news_entry_type.NewFeatures, Notes := "Updated item shop text & icons to be centered."}
                
                news_entry_info{Type := news_entry_type.GameplayChanges, Notes := "Pickup Radius ability changed from 0/25/35/50/65 > 0/0/50/100/200%."}
                news_entry_info{Type := news_entry_type.GameplayChanges, Notes := "Dash Cooldown ability changed from 25/40/55/75/100 > 0/50/75/100/150%."}

                news_entry_info{Type := news_entry_type.BugFixes, Notes := "Fixed - Player enters inactive state when auto revived while in a pause UI."}
                news_entry_info{Type := news_entry_type.BugFixes, Notes := "Fixed - Player enters inactive state when revived while in item shop weapon evolution UI."}
                news_entry_info{Type := news_entry_type.BugFixes, Notes := "Fixed - Beach Ball should no longer spawn on top of arena environment."}
                news_entry_info{Type := news_entry_type.BugFixes, Notes := "Fixed - Low health overlay was blocking buttons on mobile devices."}
                news_entry_info{Type := news_entry_type.BugFixes, Notes := "Fixed - Give up now removes all remaining revives before eliminating the player."}
                news_entry_info{Type := news_entry_type.BugFixes, Notes := "Fixed - Gathering VFX was not in sync with the tree animations."}
                news_entry_info{Type := news_entry_type.BugFixes, Notes := "Fixed - Arena UI showing incorrect movement speed for Bushido arenas (150% > 100%)"}
                news_entry_info{Type := news_entry_type.BugFixes, Notes := "Fixed - NPC can randomly receive movement speed buff."}

        news_info:
            Date := "January 17, 2025"
            Title := "Update 15 - Bug fixes, improving clarity for player damaged & small QoL improvements"
            Version := 0.410
            MaybeImage := option{UserInterface.News.NewsUpdateHighlights_01}
            Entries := array:
                news_entry_info{Type := news_entry_type.NewFeatures, Notes := "Improved clarity for when player is damaged:"}
                news_entry_info{Type := news_entry_type.None, Notes := "Added screen overlay effect when player is at a low health."}
                news_entry_info{Type := news_entry_type.None, Notes := "Updated player damage numbers with flash effect and increased size."}
                news_entry_info{Type := news_entry_type.None, Notes := "Increased glow intensity when player is damaged."}
                news_entry_info{Type := news_entry_type.None, Notes := "Increased volume & adjusted pitch of player damaged sound."}
                news_entry_info{Type := news_entry_type.NewFeatures, Notes := "Added player healed VFX when picking up health."}
                news_entry_info{Type := news_entry_type.NewFeatures, Notes := "Added HUD to display the currently equipped weapon."}

                news_entry_info{Type := news_entry_type.GameplayChanges, Notes := "Updated health pickup to activate immediately instead of after moving to player."}

                news_entry_info{Type := news_entry_type.BugFixes, Notes := "Fixed - Invulnerability was not working after player is revived."}
                news_entry_info{Type := news_entry_type.BugFixes, Notes := "Fixed - NPC could not be knocked back if it's target player is invulnerable."}
                news_entry_info{Type := news_entry_type.BugFixes, Notes := "Fixed - Health pickup was sometimes not applying."}
                news_entry_info{Type := news_entry_type.BugFixes, Notes := "Fixed - Orthographic camera view distance was incorrect."}
                news_entry_info{Type := news_entry_type.BugFixes, Notes := "Fixed - Tree was not disabled when boss fight starts if player was gathering."}

        news_info:
            Date := "January 15, 2025"
            Title := "Update 14 - New weapons & power ups, item shop rework & many QoL/UI improvements"
            Version := 0.400
            MaybeImage := option{UserInterface.News.NewsUpdateHighlights_01}
            Entries := array:
                news_entry_info{Type := news_entry_type.NewFeatures, Notes := "New Weapon - Chain Lightning - targets and bounces to the nearest enemies."}
                news_entry_info{Type := news_entry_type.NewFeatures, Notes := "New Weapon - Energy Bomb - a projectile that explodes after a short time."}
                news_entry_info{Type := news_entry_type.NewFeatures, Notes := "New Weapon - Boomerang - piercing projectile that returns to the player."}
                news_entry_info{Type := news_entry_type.NewFeatures, Notes := "Updated Weapon Select UI to allow purchasing/unlocking the new weapons (using gold)."}
                news_entry_info{Type := news_entry_type.NewFeatures, Notes := "Reworked Item Shop:"}
                news_entry_info{Type := news_entry_type.None, Notes := "Increased amount of items offered per roll from 3 > 5."}
                news_entry_info{Type := news_entry_type.None, Notes := "Added option to purchase revives (cost 1500)."}
                news_entry_info{Type := news_entry_type.None, Notes := "Removed free reroll every odd roll."}
                news_entry_info{Type := news_entry_type.NewFeatures, Notes := "Added Revive power up - revives the player when they are downed."}
                news_entry_info{Type := news_entry_type.NewFeatures, Notes := "Added Gold Multiplier power up increases gold earned from completing a game."}
                news_entry_info{Type := news_entry_type.NewFeatures, Notes := "Added Weapon Reroll power up - increases amount of reroll for weapon select."}
                news_entry_info{Type := news_entry_type.NewFeatures, Notes := "Added Treasure Chest Luck power up - increases luck for opening treasure chests."}
                news_entry_info{Type := news_entry_type.NewFeatures, Notes := "Updated Power Up UI:"}
                news_entry_info{Type := news_entry_type.None, Notes := "Added pulsing animation to highlight affordable power ups."}
                news_entry_info{Type := news_entry_type.None, Notes := "Added refund (-1) button to refund a single level."}
                news_entry_info{Type := news_entry_type.None, Notes := "Added buy text when player can afford the selected power up."}
                news_entry_info{Type := news_entry_type.NewFeatures, Notes := "Updated Game Result UI to show an animation & breakdown of gold earned."}
                news_entry_info{Type := news_entry_type.NewFeatures, Notes := "Updated Ability Select/Treasure Chest UI:"}
                news_entry_info{Type := news_entry_type.None, Notes := "Updated to not reset the cursor on multi-level ups."}
                news_entry_info{Type := news_entry_type.None, Notes := "Added to show count remaining on multi-level ups."}
                news_entry_info{Type := news_entry_type.NewFeatures, Notes := "Reworked Arena Select UI:"}
                news_entry_info{Type := news_entry_type.None, Notes := "Replaced arrow icons with number pad to select arena level."}
                news_entry_info{Type := news_entry_type.None, Notes := "Added progress indicator for the currently equipped weapon."}
                news_entry_info{Type := news_entry_type.None, Notes := "Updated the Arena Info layout to improve readability."}
                news_entry_info{Type := news_entry_type.NewFeatures, Notes := "Updated Treasure Chest UI to have a fixed panel size (button won't move around)."}
                news_entry_info{Type := news_entry_type.NewFeatures, Notes := "Added world text to show the currently selected arena."}
                news_entry_info{Type := news_entry_type.NewFeatures, Notes := "Updated Daily Quest UI:."}
                news_entry_info{Type := news_entry_type.None, Notes := "Added info to show Daily Quest rewards multiplier from player rank."}
                news_entry_info{Type := news_entry_type.None, Notes := "Added Daily Quest tier text."}
                news_entry_info{Type := news_entry_type.NewFeatures, Notes := "Updated Player Progress UI to show total weapon star level."}
                news_entry_info{Type := news_entry_type.NewFeatures, Notes := "Added HUD message for completing Daily Quests & Missions."}

                news_entry_info{Type := news_entry_type.GameplayChanges, Notes := "Added bonus 100 gold reward from completing missions (increased by multipliers)."}
                news_entry_info{Type := news_entry_type.GameplayChanges, Notes := "Increased invulnerability time after revive from 1.5s > 3.0s."}
                news_entry_info{Type := news_entry_type.GameplayChanges, Notes := "Updated health orbs to heal all players within a 300 radius."}
                news_entry_info{Type := news_entry_type.GameplayChanges, Notes := "Updated health/chest/magnet pickup distance to not be affected by pickup radius stat."}
                news_entry_info{Type := news_entry_type.GameplayChanges, Notes := "Reroll for Ability & Weapon are no longer shared and now have separate amounts."}
                news_entry_info{Type := news_entry_type.GameplayChanges, Notes := "Item Drops changed to wait for spawn animation (0.4s) to complete before it can be picked up."}              
             
                news_entry_info{Type := news_entry_type.BugFixes, Notes := "Fixed - Player movement getting stopped whenever treasure chest is collected or opened."}
                news_entry_info{Type := news_entry_type.BugFixes, Notes := "Fixed - Ability Select/Evolution UI can lose focus on controller with rate/stat info opened."}
                news_entry_info{Type := news_entry_type.BugFixes, Notes := "Fixed - Player joining a game that is already in progress does not get assigned revive zone."}
                news_entry_info{Type := news_entry_type.BugFixes, Notes := "Fixed - Desert arena walls having no collisions."}
                news_entry_info{Type := news_entry_type.BugFixes, Notes := "Fixed - Item drops were sometimes spawned floating in the air."}

        news_info:
            Date := "January 6, 2025"
            Title := "Hotfix"
            Version := 0.321
            MaybeImage := option{UserInterface.News.NewsUpdateHighlights_01}
            Entries := array:
                news_entry_info{Type := news_entry_type.NewFeatures, Notes := "Added custom background music."}

                news_entry_info{Type := news_entry_type.GameplayChanges, Notes := "Added invisible walls to block fountains in the Mega City arena."}
             
                news_entry_info{Type := news_entry_type.BugFixes, Notes := "Fixed - Revive zones remained active if downed player leaves the game."}
                news_entry_info{Type := news_entry_type.BugFixes, Notes := "Fixed - Revive progress getting reset if player is reviving player is downed."}
                news_entry_info{Type := news_entry_type.BugFixes, Notes := "Fixed - Revive not starting if player enters the zone while invulnerable."}
                news_entry_info{Type := news_entry_type.BugFixes, Notes := "Fixed - Ability/Dash HUD sometimes not loading when player joins the game."}
                news_entry_info{Type := news_entry_type.BugFixes, Notes := "Fixed - NPC repositioning had an unintentional 1s delay."}

        news_info:
            Date := "January 1, 2025"
            Title := "Update 13 - Player suggestions"
            Version := 0.320
            MaybeImage := option{UserInterface.News.NewsUpdateHighlights_01}
            Entries := array:
                news_entry_info{Type := news_entry_type.NewFeatures, Notes := "Reworked random upgrades to be manually activated by the player."}
                news_entry_info{Type := news_entry_type.None, Notes := "Changed random upgrade red orbs to treasure chests."}
                news_entry_info{Type := news_entry_type.None, Notes := "Added Treasure Chest UI to the HUD."}
                news_entry_info{Type := news_entry_type.None, Notes := "Completing missions now rewards a treasure chest."}
                news_entry_info{Type := news_entry_type.None, Notes := "Treasure Chest can be opened by pressing the Interact key."}
                news_entry_info{Type := news_entry_type.NewFeatures, Notes := "Added a warning VFX for enemy projectiles."}
                news_entry_info{Type := news_entry_type.NewFeatures, Notes := "Added player damage distribution screen to the Game Result UI."}
                news_entry_info{Type := news_entry_type.NewFeatures, Notes := "Added 15s inactivity timeout to Arena Select UI."}

                news_entry_info{Type := news_entry_type.GameplayChanges, Notes := "Added 1.5s invulnerability time when resuming game from a pause menu."}
             
                news_entry_info{Type := news_entry_type.BugFixes, Notes := "Fixed - Player can give up when they were already downed."}
                news_entry_info{Type := news_entry_type.BugFixes, Notes := "Fixed - Revive zone was being linked to the incorrect players."}
                news_entry_info{Type := news_entry_type.None, Notes := "Fixes revive zone sometimes not reviving the downed player."}
                news_entry_info{Type := news_entry_type.None, Notes := "Fixes revive zone resetting when other player is downed."}
                news_entry_info{Type := news_entry_type.BugFixes, Notes := "Fixed - Game can get stuck in pause when player leaves the game."}

        news_info:
            Date := "December 30, 2024"
            Title := "Update 12 - New Arena, more difficulty, more mini-bosses & more trees"
            Version := 0.310
            MaybeImage := option{UserInterface.News.NewsUpdateHighlights_01}
            Entries := array:
                news_entry_info{Type := news_entry_type.NewFeatures, Notes := "Added a new arena Bushido Dojo I/II/III with Unreal difficulty."}
                news_entry_info{Type := news_entry_type.NewFeatures, Notes := "Added new player ranks Ancient/Divine/Immortal."}
                news_entry_info{Type := news_entry_type.NewFeatures, Notes := "Added NPC wave multiplier info to the Arena Select UI."}
                news_entry_info{Type := news_entry_type.NewFeatures, Notes := "Added audio for the game starting countdown."}
                
                news_entry_info{Type := news_entry_type.GameplayChanges, Notes := "Updated movement speed tuning from BR > Ballistics (smoother movement)."}
                news_entry_info{Type := news_entry_type.GameplayChanges, Notes := "Mini-bosses can now spawn multiple times per wave (from Arena 4 onwards)."}
                news_entry_info{Type := news_entry_type.GameplayChanges, Notes := "Mini-bosses now only drops one random upgrade per wave."}
                news_entry_info{Type := news_entry_type.GameplayChanges, Notes := "Mini-boss will drop large XP orb if random upgrade has already dropped."}
                news_entry_info{Type := news_entry_type.GameplayChanges, Notes := "Updated trees to have randomly generated positions."}
                news_entry_info{Type := news_entry_type.GameplayChanges, Notes := "Tree count increased from 10 > 20."}
                news_entry_info{Type := news_entry_type.GameplayChanges, Notes := "Tree wood amount reduced from 60-120 > 30-60."}
                news_entry_info{Type := news_entry_type.GameplayChanges, Notes := "Tree bonus item drop rate reduced from 30% > 20%."}
                news_entry_info{Type := news_entry_type.GameplayChanges, Notes := "Trees now drop small experience orb (20 XP) if no bonus item is dropped."}
                news_entry_info{Type := news_entry_type.GameplayChanges, Notes := "Updated Gather Rate upgrade to immediately apply if player is already gathering."}
                news_entry_info{Type := news_entry_type.GameplayChanges, Notes := "Adjusted NPC wave multiplier values (health & damage increase per wave)."}
                news_entry_info{Type := news_entry_type.None, Notes := "Changed from 25/25/25/25/25/25/25/25/25% > 20/25/30/35/40/45/50/55/60%"}
                news_entry_info{Type := news_entry_type.GameplayChanges, Notes := "Updated Collectible mission positions to have minimum distance between each other."}

                news_entry_info{Type := news_entry_type.BugFixes, Notes := "Fixed - Player revive zone could be linked to the wrong player."}
                news_entry_info{Type := news_entry_type.BugFixes, Notes := "Fixed - Mission UI not showing for player joining a game in progress."}
                news_entry_info{Type := news_entry_type.BugFixes, Notes := "Fixed - Daily Quests UI not showing for player joining a game in progress."}
                news_entry_info{Type := news_entry_type.BugFixes, Notes := "Fixed - Mini-map icons were sometimes not visible."}
                news_entry_info{Type := news_entry_type.BugFixes, Notes := "Fixed - Boss health bar was showing incorrect value until first hit."}
                news_entry_info{Type := news_entry_type.BugFixes, Notes := "Fixed - Team weapon level wave incorrect in the Game Result UI."}
                news_entry_info{Type := news_entry_type.BugFixes, Notes := "Fixed - Weapon Evolution UI was showing incorrect Sword base angle."}
                news_entry_info{Type := news_entry_type.BugFixes, Notes := "Fixed - Aura VFX was duplicating effect overtime (slowly gets darker)."}

        news_info:
            Date := "December 18, 2024"
            Title := "Update 11 - Weapon Evolution customization, new weapon level upgrade & random ability rework"
            Version := 0.300
            MaybeImage := option{UserInterface.News.NewsUpdateHighlights_01}
            Entries := array:
                news_entry_info{Type := news_entry_type.NewFeatures, Notes := "Added Weapon Evolution options - customize weapon evolution upgrades."}
                news_entry_info{Type := news_entry_type.None, Notes := "Sword - Projectile Count/Slash Angle."}
                news_entry_info{Type := news_entry_type.None, Notes := "Sword - Precision Strike - half size & double base damage."}
                news_entry_info{Type := news_entry_type.None, Notes := "Bow - Projectile Count/Multicast/Projectile Piercing."}
                news_entry_info{Type := news_entry_type.None, Notes := "Bow - Supercharged Arrow - combines all projectiles into a single arrow."}
                news_entry_info{Type := news_entry_type.None, Notes := "Shuriken - Projectile Count/Weighted Shuriken."}
                news_entry_info{Type := news_entry_type.None, Notes := "Shuriken - Shuriken Mastery - deals damage every 0.5s."}
                news_entry_info{Type := news_entry_type.None, Notes := "Sentry - Projectile Count/Fire Rate."}
                news_entry_info{Type := news_entry_type.None, Notes := "Sentry - Heavy-Duty Sentry - double damage, range & duration."}
                news_entry_info{Type := news_entry_type.None, Notes := "Aura - Supercharged Duration/Supercharged Power."}
                news_entry_info{Type := news_entry_type.None, Notes := "Aura - Knockback - applies knockback to enemies while supercharged."}
                news_entry_info{Type := news_entry_type.None, Notes := "Torrent - Projectile Count/Torrent Split."}
                news_entry_info{Type := news_entry_type.None, Notes := "Torrent - Travelling Torrent - torrent moves towards the player position."}
                news_entry_info{Type := news_entry_type.None, Notes := "Lightning - Projectile Count/Size Multiplier."}
                news_entry_info{Type := news_entry_type.None, Notes := "Lightning - Focus Fire - targets the enemy with the highest health."}
                news_entry_info{Type := news_entry_type.None, Notes := "Trail - Supercharged Trail Chance/Supercharged Trail Power."}
                news_entry_info{Type := news_entry_type.None, Notes := "Trail - Supercharged Trail Bomb - explodes on expire applying knockback."}
                news_entry_info{Type := news_entry_type.None, Notes := "Shield - Max Shield/Shield Regeneration."}
                news_entry_info{Type := news_entry_type.None, Notes := "Shield - Cracked Shield Blast - applies knockback when shield is cracked."}
                
                news_entry_info{Type := news_entry_type.NewFeatures, Notes := "Updated stats info/game result UI to show weapon evolution options."}
                news_entry_info{Type := news_entry_type.NewFeatures, Notes := "Added new Weapon Level upgrade that grants addition level progress."}
                news_entry_info{Type := news_entry_type.None, Notes := "Only Rare/Epic tiers granting +1/2 weapon level."}
                news_entry_info{Type := news_entry_type.None, Notes := "Only available from level 6 onwards (after unlock 2nd weapon)."}
                news_entry_info{Type := news_entry_type.NewFeatures, Notes := "Added item drop (orbs) spawn animation."}
                news_entry_info{Type := news_entry_type.NewFeatures, Notes := "Updated weapon HUD to show the weapon level."}             
                news_entry_info{Type := news_entry_type.NewFeatures, Notes := "Added a glow effect to the slime in the hub area to indicate accolade available."}
                news_entry_info{Type := news_entry_type.NewFeatures, Notes := "Added/updated various sound effects (shield, weapon evolution, mini-boss ...)."}

                news_entry_info{Type := news_entry_type.GameplayChanges, Notes := "Increased the chance of weapon ability upgrades to always be 50% chance."}
                news_entry_info{Type := news_entry_type.None, Notes := "50/50 chance to be either a weapon or general ability."}
                news_entry_info{Type := news_entry_type.None, Notes := "Affects all level up, random upgrade & item shop."}
                news_entry_info{Type := news_entry_type.GameplayChanges, Notes := "Reworked item shop reroll cost."}
                news_entry_info{Type := news_entry_type.None, Notes := "Reroll is now free initially and after every reroll purchase."}
                news_entry_info{Type := news_entry_type.None, Notes := "Reroll now cost 100 and increases by 100 after every reroll purchased."}
                news_entry_info{Type := news_entry_type.GameplayChanges, Notes := "Removed the Common tier for Pickup Range, Experience Multiplier & Gather Rate abilities."}
                news_entry_info{Type := news_entry_type.GameplayChanges, Notes := "Removed the Common/Uncommon tier for Luck ability."}
                news_entry_info{Type := news_entry_type.GameplayChanges, Notes := "Shuriken base damage increased from 8 > 10."}
                news_entry_info{Type := news_entry_type.GameplayChanges, Notes := "Gather Rate multiplier is doubled for gathering speed."}
                news_entry_info{Type := news_entry_type.GameplayChanges, Notes := "Added 15s delay to the spawn time of the first mini-boss."}
                
                news_entry_info{Type := news_entry_type.GameplayChanges, Notes := "Reduced Victory/Defeat screen duration from 5s > 3s."}
                news_entry_info{Type := news_entry_type.GameplayChanges, Notes := "Changed tree gathering to no longer reset progress."}
                news_entry_info{Type := news_entry_type.GameplayChanges, Notes := "Increased the base time to gather tree from 3s > 5s."}
                news_entry_info{Type := news_entry_type.GameplayChanges, Notes := "Updated the collectible mission to have randomized position."}
                news_entry_info{Type := news_entry_type.GameplayChanges, Notes := "Increased Orthographic camera view distance of from 2500 > 3000."}
                news_entry_info{Type := news_entry_type.GameplayChanges, Notes := "Increased Orbit Free camera vertical view angle 45° > 15° (lower angle)."}
                news_entry_info{Type := news_entry_type.GameplayChanges, Notes := "Increased the NPC spawn/teleport distance from player 1200 > 1300-1400."}
                news_entry_info{Type := news_entry_type.GameplayChanges, Notes := "Reduced Monster Hunter daily quest 5000/10000/15000 > 5000/7500/10000."}
                news_entry_info{Type := news_entry_type.GameplayChanges, Notes := "Increase Torrent range from 500 > 700."}
                
                news_entry_info{Type := news_entry_type.PerformanceOptimizations, Notes := "Performance optimization for weapon stat calculations."}

                news_entry_info{Type := news_entry_type.BugFixes, Notes := "Fixed - General fix for cooldown resetting on game resume for all weapons."}
                news_entry_info{Type := news_entry_type.BugFixes, Notes := "Fixed - Multicast increasing cooldown time for Sentry/Torrent/Bow."}
                news_entry_info{Type := news_entry_type.BugFixes, Notes := "Fixed - Team Info HUD layout to support longer usernames."}
                news_entry_info{Type := news_entry_type.BugFixes, Notes := "Fixed - NPC sometimes fail to teleport to next spawn position when eliminated."}
                news_entry_info{Type := news_entry_type.BugFixes, Notes := "Fixed - DPS stats showing 0 amounts for other players in game result UI."}
                news_entry_info{Type := news_entry_type.BugFixes, Notes := "Fixed - Wood currency HUD showing in hub area if tree gathered on game end."}
                news_entry_info{Type := news_entry_type.BugFixes, Notes := "Fixed - Weapon stats not applying immediately after auto selected."}
                news_entry_info{Type := news_entry_type.BugFixes, Notes := "Fixed - Weapon level progress indicator pulsing animation on multi-level up."}

        news_info:
            Date := "December 7, 2024"
            Title := "Update 10 - Reworked NPC teleporting"
            Version := 0.220
            MaybeImage := option{UserInterface.News.NewsUpdateHighlights_01}
            Entries := array:
                news_entry_info{Type := news_entry_type.NewFeatures, Notes := "Added smooth resizing animation for Aura weapon."}
                news_entry_info{Type := news_entry_type.NewFeatures, Notes := "Updated stats info/game results to highlight banished abilities (red color)."}
                news_entry_info{Type := news_entry_type.NewFeatures, Notes := "Updated the Lightning weapon VFX/sound."}
                news_entry_info{Type := news_entry_type.NewFeatures, Notes := "Updated the Trail weapon sound."}
                news_entry_info{Type := news_entry_type.NewFeatures, Notes := "Updated minimap shape to circular while in the arena."}

                news_entry_info{Type := news_entry_type.GameplayChanges, Notes := "Updated NPC teleporting behavior in multiplayer scenario."}
                news_entry_info{Type := news_entry_type.None, Notes := "NPC no teleports if it is within range of any player."}
                news_entry_info{Type := news_entry_type.None, Notes := "NPC no longer teleports to a position that is in range of another player."}
                news_entry_info{Type := news_entry_type.None, Notes := "NPC no longer teleports to a position within range of a player respawn zone."}
                news_entry_info{Type := news_entry_type.GameplayChanges, Notes := "Updated player revive to decay progress 10.0%/s when player leaves the zone."}
                news_entry_info{Type := news_entry_type.GameplayChanges, Notes := "Added NPC knockback when player is revived."}
                news_entry_info{Type := news_entry_type.GameplayChanges, Notes := "Updated Shuriken weapon to increase projectile speed +33.3% per evolution."}
                news_entry_info{Type := news_entry_type.GameplayChanges, Notes := "Renamed Elimination accolades to 'Eliminate Slime' to reduce screen clutter."}
                news_entry_info{Type := news_entry_type.GameplayChanges, Notes := "Reduced range of Lightning weapon from 1000 > 900."}

                news_entry_info{Type := news_entry_type.BugFixes, Notes := "Fixed - Lightning weapon multicast was increasing the cooldown time."}
                news_entry_info{Type := news_entry_type.BugFixes, Notes := "Fixed - Aura/Shield weapon resetting when game pauses/resumes."}
                news_entry_info{Type := news_entry_type.BugFixes, Notes := "Fixed - Daily quest HUD not hiding after reaching the target."}

        news_info:
            Date := "November 30, 2024"
            Title := "Bug fixes"
            Version := 0.211
            MaybeImage := option{UserInterface.News.NewsUpdateHighlights_01}
            Entries := array:
                news_entry_info{Type := news_entry_type.NewFeatures, Notes := "Added accolade for Eliminating Melee Slime General."}
                news_entry_info{Type := news_entry_type.BugFixes, Notes := "Fixed - Weapon DPS stats was showing 0 amount in game result UI."}
                news_entry_info{Type := news_entry_type.BugFixes, Notes := "Fixed - Gather tree daily quest was sometimes not updating progress."}
                news_entry_info{Type := news_entry_type.BugFixes, Notes := "Fixed - Game could get stuck in paused state if player left the game while in hub area."}
                news_entry_info{Type := news_entry_type.BugFixes, Notes := "Fixed - Tree respawning during boss fight if player was gathering when fight starts."}

        news_info:
            Date := "November 29, 2024"
            Title := "Update 9 - Pause system overhaul, custom revive & updated snow slimes"
            Version := 0.210
            MaybeImage := option{UserInterface.News.NewsUpdateHighlights_01}
            Entries := array:
                news_entry_info{Type := news_entry_type.NewFeatures, Notes := "Pause system complete overhaul."}
                news_entry_info{Type := news_entry_type.None, Notes := "Game now only pauses when all players are in a pause UI (level up/upgrade orb/item shop)."}
                news_entry_info{Type := news_entry_type.None, Notes := "Players will be invulnerable while in a pause UI."}
                news_entry_info{Type := news_entry_type.None, Notes := "Pause countdown will only start if a downed player resumes the game."}
                news_entry_info{Type := news_entry_type.None, Notes := "Upon countdown complete all pause UI queued will be auto selected."}
                news_entry_info{Type := news_entry_type.NewFeatures, Notes := "Implemented a custom player revive mechanic."}
                news_entry_info{Type := news_entry_type.None, Notes := "Revive activation area will spawn at the place the player is eliminated."}
                news_entry_info{Type := news_entry_type.None, Notes := "Player will have to stand within the activation area for 5s to revive."}
                news_entry_info{Type := news_entry_type.None, Notes := "Revived player will be teleported back to the revive position."}
                news_entry_info{Type := news_entry_type.NewFeatures, Notes := "Added HUD UI to show count of auto selected abilities."}
                news_entry_info{Type := news_entry_type.NewFeatures, Notes := "Updated boss/miniboss minimap icon to update image according to arena."}
                
                news_entry_info{Type := news_entry_type.GameplayChanges, Notes := "Updated arena 4-6 NPC wave sequence to the new sequence."}
                news_entry_info{Type := news_entry_type.GameplayChanges, Notes := "Reworked the Boss Fight system."}
                news_entry_info{Type := news_entry_type.None, Notes := "Removed the Boss Fight intro cinematic (warning message still shows)."}
                news_entry_info{Type := news_entry_type.None, Notes := "NPC waves will now continue to spawn during the boss fight."}
                news_entry_info{Type := news_entry_type.GameplayChanges, Notes := "Boss/miniboss attacks can now stun the player."}
                news_entry_info{Type := news_entry_type.None, Notes := "Red/Orange/Yellow slime attack stuns for 1s."}
                news_entry_info{Type := news_entry_type.None, Notes := "Purple slime/Boss Forward Jump attack stuns for 0.5s."}
                news_entry_info{Type := news_entry_type.GameplayChanges, Notes := "Increased the size of boss slime (2x)."}
                news_entry_info{Type := news_entry_type.GameplayChanges, Notes := "Removed NPC HP scaling with player count."}
                news_entry_info{Type := news_entry_type.GameplayChanges, Notes := "Updated activation area to always be visible for item shop/storm mission."}
                news_entry_info{Type := news_entry_type.GameplayChanges, Notes := "Health orb disables pickup if player is at full health."}

                news_entry_info{Type := news_entry_type.BugFixes, Notes := "Fixed - All UI stops showing is player is revived while UI is opened."}
                news_entry_info{Type := news_entry_type.BugFixes, Notes := "Fixed - Victory daily quests was counting progress for each player."}
                news_entry_info{Type := news_entry_type.BugFixes, Notes := "Fixed - Scenario where weapons to remained active after a game ends."}
                news_entry_info{Type := news_entry_type.BugFixes, Notes := "Fixed - Weapon DPS stats was not counting if player received weapon while downed."}
                news_entry_info{Type := news_entry_type.BugFixes, Notes := "Fixed - Ability upgrade animation duration was not consistent."}

        news_info:
            Date := "November 25, 2024"
            Title := "Bug fixes"
            Version := 0.201
            MaybeImage := option{UserInterface.News.NewsUpdateHighlights_01}
            Entries := array:               
                news_entry_info{Type := news_entry_type.NewFeatures, Notes := "Updated XP orbs to increase in size based on the amount of XP."}
                news_entry_info{Type := news_entry_type.BugFixes, Notes := "Fixed - Aura VFX was not showing for all players."}
                news_entry_info{Type := news_entry_type.BugFixes, Notes := "Fixed - Item shop timer UI was sometimes not showing."}
                news_entry_info{Type := news_entry_type.BugFixes, Notes := "Fixed - Weapons HUD UI was sometimes not showing."}               
                news_entry_info{Type := news_entry_type.BugFixes, Notes := "Fixed - Yellow miniboss was not dropping ability upgrade orb."}
                news_entry_info{Type := news_entry_type.BugFixes, Notes := "Fixed - Snow minibosses was not showing minimap icon."}
                news_entry_info{Type := news_entry_type.BugFixes, Notes := "Fixed - Blurry player rank icons."}
                news_entry_info{Type := news_entry_type.BugFixes, Notes := "Fixed - Incorrect button interaction text (Power Up/Daily Quests)."}
                news_entry_info{Type := news_entry_type.BugFixes, Notes := "Fixed - Area in the hub where player can get stuck."}

        news_info:
            Date := "November 23, 2024"
            Title := "Update 8 - Daily quests, miniboss variety & quality of life"
            Version := 0.200
            MaybeImage := option{UserInterface.News.NewsUpdateHighlights_01}
            Entries := array:               
                news_entry_info{Type := news_entry_type.NewFeatures, Notes := "Added a Daily Quests system - complete to earn bonus gold."}
                news_entry_info{Type := news_entry_type.None, Notes := "Players will start with 3 basic tutorial quests."}
                news_entry_info{Type := news_entry_type.None, Notes := "New quests will awarded every day (00:00 UTC time)."}
                news_entry_info{Type := news_entry_type.None, Notes := "Quests progress is saved and incomplete quests will be remain the next day."}
                news_entry_info{Type := news_entry_type.None, Notes := "One quest can be rerolled once per day."}
                news_entry_info{Type := news_entry_type.NewFeatures, Notes := "Updated NPC wave sequence (only for arena 1-3, 4-6 remains unchanged for comparison)."}
                news_entry_info{Type := news_entry_type.None, Notes := "Removed the projectile, dash, barrage & charge slimes."}
                news_entry_info{Type := news_entry_type.None, Notes := "Added Miniboss slime variants (projectile, dash, barrage, charge)."}
                news_entry_info{Type := news_entry_type.None, Notes := "Added Big Melee Elite slime (that starts spawns on wave 5-6)."}
                news_entry_info{Type := news_entry_type.NewFeatures, Notes := "Added ability/weapon stats info to the ability select UI."}
                news_entry_info{Type := news_entry_type.NewFeatures, Notes := "Added a new mission - accept the challenge & survive 4 phases of storm."}
                news_entry_info{Type := news_entry_type.NewFeatures, Notes := "Updated the hub area layout."}
                news_entry_info{Type := news_entry_type.NewFeatures, Notes := "Updated the Sword weapon VFX (added slash trail)."}
                news_entry_info{Type := news_entry_type.NewFeatures, Notes := "Updated the Aura weapon VFX."}
                news_entry_info{Type := news_entry_type.NewFeatures, Notes := "Updated the Aura weapon VFX to change color while buff from evolution is active."}
                
                news_entry_info{Type := news_entry_type.NewFeatures, Notes := "Added minimap icon for the Golden Slime."}
                news_entry_info{Type := news_entry_type.NewFeatures, Notes := "Added minimap icon for Miniboss Slimes."}
                news_entry_info{Type := news_entry_type.NewFeatures, Notes := "Updated the minimap icon for Boss Fight."}
                
                news_entry_info{Type := news_entry_type.GameplayChanges, Notes := "Aura weapon buffed - base cooldown 0.75 > 0.5 and base damage 6 > 8."}
                news_entry_info{Type := news_entry_type.GameplayChanges, Notes := "Reworked new power up available indicator to always show icon on screen."}
                news_entry_info{Type := news_entry_type.GameplayChanges, Notes := "Renamed objectives to missions."}
                
                news_entry_info{Type := news_entry_type.BugFixes, Notes := "Fixed - aura weapon VFX showing incorrect size in multiplayer."}
                news_entry_info{Type := news_entry_type.BugFixes, Notes := "Fixed - weapon DPS stats incorrect for other players in game results UI."}
                news_entry_info{Type := news_entry_type.BugFixes, Notes := "Fixed - new power up available indicator no longer changes players team."}

        news_info:
            Date := "November 16, 2024"
            Title := "Hotfix"
            Version := 0.131
            MaybeImage := option{UserInterface.News.NewsUpdateHighlights_01}
            Entries := array:
                news_entry_info{Type := news_entry_type.GameplayChanges, Notes := "Removed luck ability from applying in item shops."}
                news_entry_info{Type := news_entry_type.GameplayChanges, Notes := "Changed the level up audio."}
                news_entry_info{Type := news_entry_type.BugFixes, Notes := "Fixed - players were not set to same team causing missing team info & cannot revive."}
                news_entry_info{Type := news_entry_type.BugFixes, Notes := "Fixed - game results total weapon damage not counting when eliminations in a single hit."}
                news_entry_info{Type := news_entry_type.BugFixes, Notes := "Fixed - game results weapon DPS/Duration kept counting after the game ended."}

        news_info:
            Date := "November 15, 2024"
            Title := "Update 7 - New Luck ability, testing a new cameras & improvements"
            Version := 0.13
            MaybeImage := option{UserInterface.News.NewsUpdateHighlights_01}
            Entries := array:
                news_entry_info{Type := news_entry_type.NewFeatures, Notes := "Added Luck ability - increase the chance of higher tier abilities."}
                news_entry_info{Type := news_entry_type.NewFeatures, Notes := "Added Luck power up - increase the chance of higher tier abilities."}
                news_entry_info{Type := news_entry_type.NewFeatures, Notes := "Added ability Tier Rates info to ability select UI (i.e. chance of Legendary)."}
                news_entry_info{Type := news_entry_type.NewFeatures, Notes := "Added Dash Count power up - increase count of dash charges."}
                news_entry_info{Type := news_entry_type.NewFeatures, Notes := "Added Team Experience Multiplier info to the ability select UI."}
                news_entry_info{Type := news_entry_type.NewFeatures, Notes := "Added Weapon DPS (Damage/Duration) stats to ability stats & game result UI."}
                news_entry_info{Type := news_entry_type.NewFeatures, Notes := "Added Power Up new upgrade available indicator icon (minimap/beacon)."}
                news_entry_info{Type := news_entry_type.NewFeatures, Notes := "Added Gold currency HUD (visible while in hub area)."}
                news_entry_info{Type := news_entry_type.NewFeatures, Notes := "Added Player Rank icons (visible while in hub area)."}
                news_entry_info{Type := news_entry_type.NewFeatures, Notes := "Updated tree minimap icon colors to match arena tree colors."}
                news_entry_info{Type := news_entry_type.NewFeatures, Notes := "Added level up animation VFX & audio cue."}
                news_entry_info{Type := news_entry_type.NewFeatures, Notes := "Added changing booth - disabled when countdown timer is started."}
                news_entry_info{Type := news_entry_type.NewFeatures, Notes := "Added music toggle to the player menu UI & removed switch from hub area."}
                news_entry_info{Type := news_entry_type.NewFeatures, Notes := "Added minimap icons for capture area/beach ball goal objectives."}
                news_entry_info{Type := news_entry_type.NewFeatures, Notes := "Added button in the hub area to open the settings menu."}
                news_entry_info{Type := news_entry_type.NewFeatures, Notes := "Added AFK player detection - skips pause screen if remaining player is AFK."}
                news_entry_info{Type := news_entry_type.NewFeatures, Notes := "Added a Give Up button to the menu to eliminate yourself."}
                news_entry_info{Type := news_entry_type.NewFeatures, Notes := "Updated the NPC hit impact VFX."}
                news_entry_info{Type := news_entry_type.NewFeatures, Notes := "Updated camera settings to be persistent."}
                news_entry_info{Type := news_entry_type.NewFeatures, Notes := "Added Orthographic camera."}
                news_entry_info{Type := news_entry_type.NewFeatures, Notes := "Added Orbit Free camera (camera can move vertically)."}
                
                news_entry_info{Type := news_entry_type.GameplayChanges, Notes := "Changed default camera to Orbit Free."}
                news_entry_info{Type := news_entry_type.GameplayChanges, Notes := "Adjusted orbit camera to a higher angle & further distance."}
                news_entry_info{Type := news_entry_type.GameplayChanges, Notes := "Increased the Sword weapon attack angle from 90° > 120°."}
                news_entry_info{Type := news_entry_type.GameplayChanges, Notes := "Banish power up max level from 5 > 10, base cost 500 > 750, cost exponent 3 > 2."}
                news_entry_info{Type := news_entry_type.GameplayChanges, Notes := "Reroll power up max level from 5 > 10, base cost 250 > 500, cost exponent 3 > 2."}
                
                news_entry_info{Type := news_entry_type.BugFixes, Notes := "Fixed - Sword multicast was increasing the weapon's cooldown time."}
                news_entry_info{Type := news_entry_type.BugFixes, Notes := "Fixed - VFX not showing/disappearing when there are too many VFX active."}
                news_entry_info{Type := news_entry_type.BugFixes, Notes := "Fixed - Aura VFX flickering when overlapping with tree gathering VFX."}
                news_entry_info{Type := news_entry_type.BugFixes, Notes := "Fixed - Player progress UI showing the incorrect progress."}
                news_entry_info{Type := news_entry_type.BugFixes, Notes := "Fixed - Tree size was not resetting to the default size."}
                news_entry_info{Type := news_entry_type.BugFixes, Notes := "Fixed - Server time was not counting paused time."}

        news_info:
            Date := "November 6, 2024"
            Title := "Update 6 - Improving multiplayer experience"
            Version := 0.12
            MaybeImage := option{UserInterface.News.NewsUpdateHighlights_01}
            Entries := array:
                news_entry_info{Type := news_entry_type.NewFeatures, Notes := "Added Ability Stats UI (hold Aim button during game to show ability/weapon stats)."}
                news_entry_info{Type := news_entry_type.NewFeatures, Notes := "Added option to start countdown even if all players are not ready."}
                news_entry_info{Type := news_entry_type.NewFeatures, Notes := "Added countdown to game paused screen to automatically resume the game."}
                news_entry_info{Type := news_entry_type.None, Notes := "Ability will be auto selected for players still in ability select."}
                news_entry_info{Type := news_entry_type.NewFeatures, Notes := "Updated Numbered Tiles objective to randomize tile positions."}
                news_entry_info{Type := news_entry_type.NewFeatures, Notes := "Updated Item Shop to be activated using volume device instead of button interaction."}
                news_entry_info{Type := news_entry_type.None, Notes := "Same mechanism as tree gathering - stand next to NPC for 1s to activate."}
                news_entry_info{Type := news_entry_type.NewFeatures, Notes := "Added warning animation to objective timer UI (number will pulse red at <= 10s)."}
                news_entry_info{Type := news_entry_type.NewFeatures, Notes := "Added progress animation to objective UI."}

                news_entry_info{Type := news_entry_type.GameplayChanges, Notes := "Arena Gold reward increased from 0%/10%/20%/30%/40%/50% > 0%/20%/40%/60%/80%/100%."}
                news_entry_info{Type := news_entry_type.GameplayChanges, Notes := "Shield Weapon - reduced shield amount from 10%/40%/70%/100% > 12.5%/25%/37.5%/50%."}
                news_entry_info{Type := news_entry_type.GameplayChanges, Notes := "Removed matchmaking switch (matchmaking is disabled when a game is in progress)."}
                
                news_entry_info{Type := news_entry_type.BugFixes, Notes := "Fixed - Player joining a game in progress gets stuck in the hub area."}
                news_entry_info{Type := news_entry_type.None, Notes := "Player will now be teleported into the arena with 5s invulnerability."}
                news_entry_info{Type := news_entry_type.BugFixes, Notes := "Fixed - Player leaves game during game paused screen resumes the game."}
                news_entry_info{Type := news_entry_type.BugFixes, Notes := "Fixed - Downed player gets stuck in pause state if other player leaves the game."}
                news_entry_info{Type := news_entry_type.BugFixes, Notes := "Fixed - NPC stops moving if target player leaves the game."}
                news_entry_info{Type := news_entry_type.BugFixes, Notes := "Fixed - NPC health not adjusting to player count when player leaves the game."}
                news_entry_info{Type := news_entry_type.None, Notes := "NPC health only adjusts when they have been eliminated once and respawns."}
                news_entry_info{Type := news_entry_type.BugFixes, Notes := "Fixed - Resume game could be triggered multiple times by the same player."}
                news_entry_info{Type := news_entry_type.None, Notes := "Resulting in game resuming when other players are still selecting."}
                news_entry_info{Type := news_entry_type.BugFixes, Notes := "Fixed - XP bar was showing on top of defeat/victory screen."}
                news_entry_info{Type := news_entry_type.BugFixes, Notes := "Fixed - Boss health bar was showing on top of boss fight warning screen."}
                news_entry_info{Type := news_entry_type.BugFixes, Notes := "Fixed - Boss could be damaged during intro (boss is now invulnerable until the fight starts)."}
                news_entry_info{Type := news_entry_type.BugFixes, Notes := "Fixed - Lighting was too bright."}
                news_entry_info{Type := news_entry_type.BugFixes, Notes := "Fixed - Weapon level progress indicator pulsing animation was too slow."}
                
        news_info:
            Date := "October 29, 2024"
            Title := "Update 5 - Bug fixing & minor changes"
            Version := 0.11
            MaybeImage := option{UserInterface.News.NewsUpdateHighlights_01}
            Entries := array:
                news_entry_info{Type := news_entry_type.NewFeatures, Notes := "Added 1s buffer to ability select UI to prevent accidental selection."}
                news_entry_info{Type := news_entry_type.NewFeatures, Notes := "Added switch to manually enable/disable matchmaking (allow players to join game)."}

                news_entry_info{Type := news_entry_type.GameplayChanges, Notes := "Matchmaking is no longer automatically enable when a game ends."}
                news_entry_info{Type := news_entry_type.GameplayChanges, Notes := "Reduced revive time from 10s > 1s (temporary solution)."}
                news_entry_info{Type := news_entry_type.GameplayChanges, Notes := "Improved visibility of collectibles (increased size/brightness)."}

                news_entry_info{Type := news_entry_type.BugFixes, Notes := "Fixed - Multi-level up granting players weapon select multiple times."}
                news_entry_info{Type := news_entry_type.BugFixes, Notes := "Fixed - Multi-level up causing weapon select to be skipped."}
                news_entry_info{Type := news_entry_type.BugFixes, Notes := "Fixed - Banishing an ability before weapon select causing weapon to be banished."}
                news_entry_info{Type := news_entry_type.BugFixes, Notes := "Fixed - Player becomes invulnerable when downed and starting a new game."}
                news_entry_info{Type := news_entry_type.BugFixes, Notes := "Fixed - Player remains downed if victory is achieved."}
                news_entry_info{Type := news_entry_type.BugFixes, Notes := "Fixed - Health regeneration upgrade was being applied to only 1 player."}
                news_entry_info{Type := news_entry_type.BugFixes, Notes := "Fixed - Slimes on Arena 4 - Wave 4 health multiplier was set too high (x150 > x1.5)."}
                news_entry_info{Type := news_entry_type.BugFixes, Notes := "Fixed - Slime spawn area set too small on Arena 2/3 (2000 > 4000)."}
                news_entry_info{Type := news_entry_type.BugFixes, Notes := "Fixed - Small holes in the arena floor near the edge of the arenas."}
                news_entry_info{Type := news_entry_type.BugFixes, Notes := "Fixed - Defeat is triggered if a player leaves the game and 1 player remains."}
                news_entry_info{Type := news_entry_type.BugFixes, Notes := "Fixed - Wood counter HUD stops working if a player leaves the game."}
                news_entry_info{Type := news_entry_type.BugFixes, Notes := "Fixed - Shuriken range indicator overriding player direction arrow indicator."}

        news_info:
            Date := "October 25, 2024"
            Title := "Update 4 - Ability system complete overhaul, multi-weapon & objectives mechanic"
            Version := 0.10
            MaybeImage := option{UserInterface.News.NewsUpdateHighlights_01}
            Entries := array:
                news_entry_info{Type := news_entry_type.NewFeatures, Notes := "Added Multi-Weapon system - unlock a new weapon at levels 5/15/25."}
                news_entry_info{Type := news_entry_type.NewFeatures, Notes := "New Weapon - Lightning - randomly strikes an enemy dealing damage in a small area."}
                news_entry_info{Type := news_entry_type.NewFeatures, Notes := "New Weapon - Torrent - spawns a pool at a random position that deals damages over time."}
                news_entry_info{Type := news_entry_type.NewFeatures, Notes := "New Weapon - Aura - deals damage at a high rate in a small area around the player."}
                news_entry_info{Type := news_entry_type.NewFeatures, Notes := "New Weapon - Shield - generates a shield that deals damage around the player when hit."}
                news_entry_info{Type := news_entry_type.NewFeatures, Notes := "New Weapon - Trail - spawns a trail in the player's path that deals damage over time."}
                news_entry_info{Type := news_entry_type.NewFeatures, Notes := "Added Weapon Evolution system - at weapon level 6/12/18 weapon unlocks a bonus ability."}
                news_entry_info{Type := news_entry_type.NewFeatures, Notes := "Added Ability Tier system - higher tiers grant higher stat values."}
                news_entry_info{Type := news_entry_type.NewFeatures, Notes := "Updated Ability system to have general stats & weapon specific stats."}
                news_entry_info{Type := news_entry_type.NewFeatures, Notes := "Updated Ability Upgrades to align with new ability system."}
                news_entry_info{Type := news_entry_type.NewFeatures, Notes := "Added Weapons HUD - shows active weapons and level required to unlocked a new weapon."}
                news_entry_info{Type := news_entry_type.NewFeatures, Notes := "Updated Ability select UI."}
                news_entry_info{Type := news_entry_type.NewFeatures, Notes := "Updated Weapon select UI to show description & weapon base stats."}
                news_entry_info{Type := news_entry_type.NewFeatures, Notes := "Added a pulsing animation to ability level progress indicator."}
                news_entry_info{Type := news_entry_type.NewFeatures, Notes := "Added Objectives mechanic - complete a challenge to earn a random ability upgrade."}
                news_entry_info{Type := news_entry_type.None, Notes := "Beach Ball - push the Beach Ball(s) into the goal."}
                news_entry_info{Type := news_entry_type.None, Notes := "Collectibes - find & collect all the Blue Pyramids."}
                news_entry_info{Type := news_entry_type.None, Notes := "Golden Slime - eliminate the Golden Slime."}
                news_entry_info{Type := news_entry_type.None, Notes := "Capture Point - stay inside the capture zone."}
                news_entry_info{Type := news_entry_type.None, Notes := "Numbered Tiles - stand on the tiles in the numbered order."}
                news_entry_info{Type := news_entry_type.NewFeatures, Notes := "Updated Game Paused UI to include player rank icon, weapon levels & elimination count."}
                news_entry_info{Type := news_entry_type.NewFeatures, Notes := "Updated Game Results UI - objectives completed, team stats & player weapon stats."}
                news_entry_info{Type := news_entry_type.NewFeatures, Notes := "Updated the environment of the HUB area."}
                news_entry_info{Type := news_entry_type.NewFeatures, Notes := "Updated Item Shop mechanic to offer abilities/heal/shield instead of Fortnite items."}
                news_entry_info{Type := news_entry_type.NewFeatures, Notes := "Updated Item Shop wood consumed is no longer shared between players (gathering is still shared)."}
                news_entry_info{Type := news_entry_type.NewFeatures, Notes := "Improved responsiveness of dash controls using input trigger."}
                news_entry_info{Type := news_entry_type.None, Notes := "Added switch to toggle back to old dash control for Mobile devices."}                
                
                news_entry_info{Type := news_entry_type.GameplayChanges, Notes := "Player base max health increased 100 > 200."}
                news_entry_info{Type := news_entry_type.GameplayChanges, Notes := "Player base health regeneration increased 0.0 > 0.5."}
                news_entry_info{Type := news_entry_type.GameplayChanges, Notes := "Dash Count ability changed to Dash Cooldown - reduce cooldown of dash ability"}
                news_entry_info{Type := news_entry_type.GameplayChanges, Notes := "Ability Boost changed from level up ability 2x > increase the tier of ability."}
                news_entry_info{Type := news_entry_type.GameplayChanges, Notes := "Ability Banished no longer consumes ability select - only banish once per level."}
                news_entry_info{Type := news_entry_type.GameplayChanges, Notes := "Updated Arenas 1/2/3 to increase size to match the size of Arenas 4/5/6."}
                news_entry_info{Type := news_entry_type.GameplayChanges, Notes := "Arena 1 duration increased from 12 > 15."}
                news_entry_info{Type := news_entry_type.GameplayChanges, Notes := "Sentry weapon now spawns at a random position & small distance around the player."}
                news_entry_info{Type := news_entry_type.GameplayChanges, Notes := "Item drops from trees will now move when the tree respawns to prevent overlapping."}
                news_entry_info{Type := news_entry_type.GameplayChanges, Notes := "Adjusted Item Shop spawn interval from every 5m > 6m."}
                news_entry_info{Type := news_entry_type.GameplayChanges, Notes := "Accolade changed from Ability Maxed > Ability Evolved."}
                news_entry_info{Type := news_entry_type.GameplayChanges, Notes := "Accolade replaced from Ability Banish > New Weapon."}
                news_entry_info{Type := news_entry_type.GameplayChanges, Notes := "Accolade removed Revive Player (replaced with Complete Objective accolade)"}
                news_entry_info{Type := news_entry_type.GameplayChanges, Notes := "Accolade for Complete Game/Victory is now awarded when game results UI is closed."}
                news_entry_info{Type := news_entry_type.GameplayChanges, Notes := "Multicast/Projectile Count ability removed (replaced with weapon evolution)."}
                news_entry_info{Type := news_entry_type.GameplayChanges, Notes := "Close button in News UI moved to the top left for quick access for controllers."}

                news_entry_info{Type := news_entry_type.BugFixes, Notes := "Fixed - Sentry weapon projectile count damaging respawning slimes."}
                news_entry_info{Type := news_entry_type.BugFixes, Notes := "Fixed - Sentry weapon range indicator not visible in desert arena (added shadow)."}
                news_entry_info{Type := news_entry_type.BugFixes, Notes := "Fixed - Sentry weapon animation not in sync with the damage timing."}
                news_entry_info{Type := news_entry_type.BugFixes, Notes := "Fixed - Banish ability was not consuming any charges (infinite banish)."}
                news_entry_info{Type := news_entry_type.BugFixes, Notes := "Fixed - Invisible collision in the top-right area of Arenas 4/5/6."}
                news_entry_info{Type := news_entry_type.BugFixes, Notes := "Fixed - Weapons activating for DBNO player when boss fight starts."}
                news_entry_info{Type := news_entry_type.BugFixes, Notes := "Fixed - Time dilation reduced from 0.0001 > 0.00001 (time progresses slower during pause)."}

        news_info:
            Date := "October 08, 2024"
            Title := "Update 3 - Bug fixing"
            Version := 0.03
            MaybeImage := option{UserInterface.News.NewsUpdateHighlights_01}
            Entries := array:
                news_entry_info{Type := news_entry_type.BugFixes, Notes := "Fixed - Player progress not updating/saving."}
                news_entry_info{Type := news_entry_type.BugFixes, Notes := "Fixed - Controls broken when player joined a game in progress (on mobile & controller)."}
                news_entry_info{Type := news_entry_type.BugFixes, Notes := "Fixed - Player direction arrow disappearing after being respawned."}
                news_entry_info{Type := news_entry_type.BugFixes, Notes := "Fixed - Slime animations not playing when starting from outside of the player's view."}
                news_entry_info{Type := news_entry_type.BugFixes, Notes := "Fixed - Boss slime dash/jump animations facing the incorrect direction."}
                news_entry_info{Type := news_entry_type.BugFixes, Notes := "Fixed - Boss health bar remains showing if the player is defeated."}
                news_entry_info{Type := news_entry_type.BugFixes, Notes := "Fixed - Boss map indicator was still showing after boss fight ended."}
                news_entry_info{Type := news_entry_type.BugFixes, Notes := "Fixed - Item shop spawning at the same time as boss fight starting."}
                news_entry_info{Type := news_entry_type.BugFixes, Notes := "Fixed - Items not being removed after the game ended."}
                news_entry_info{Type := news_entry_type.BugFixes, Notes := "Fixed - Shield was not resetting after the game ended."}
                news_entry_info{Type := news_entry_type.BugFixes, Notes := "Fixed - Squad player info HUD overlapping with other UI (at 4 players)."}

        news_info:
            Date := "October 06, 2024"
            Title := "Update 2 - New Arenas, Power Up, NPC Improvements & HUD UI update"
            Version := 0.02
            MaybeImage := option{UserInterface.News.NewsUpdateHighlights_01}
            Entries := array:
                news_entry_info{Type := news_entry_type.NewFeatures, Notes := "Added Gold currency. Awarded at game end based on time survived, wood collected & victory."}
                news_entry_info{Type := news_entry_type.NewFeatures, Notes := "Added Power Ups mechanic. Use Gold to permanently upgrade ability stats/ability selection."}
                news_entry_info{Type := news_entry_type.None, Notes := "Ability Buff - Damage, Rate, Radius, Health, Regen, Pickup Distance etc ..."}
                news_entry_info{Type := news_entry_type.None, Notes := "Ability Select - Rerolls, Banish & Boost."}
                news_entry_info{Type := news_entry_type.NewFeatures, Notes := "Added Arena 4 - The Dojo."}
                news_entry_info{Type := news_entry_type.NewFeatures, Notes := "Added Arena 5 - Ice Castle."}
                news_entry_info{Type := news_entry_type.NewFeatures, Notes := "Added Arena 6 - Mega City."}
                news_entry_info{Type := news_entry_type.NewFeatures, Notes := "Added Snow Slime variants - with more difficult attack parameters (e.g. faster attack rate)."}
                news_entry_info{Type := news_entry_type.NewFeatures, Notes := "Added custom health bar for boss fight."}
                news_entry_info{Type := news_entry_type.NewFeatures, Notes := "Added ability select options to grant health/shield/gold when all abilities are maxed."}
                news_entry_info{Type := news_entry_type.NewFeatures, Notes := "Reworked NPC wave spawning to limit slime types during a wave (e.g. max 5 projectile slimes)."}      
                news_entry_info{Type := news_entry_type.NewFeatures, Notes := "Added custom player info & squad info HUD."}
                news_entry_info{Type := news_entry_type.NewFeatures, Notes := "Added elimination counter to the HUD."}
                news_entry_info{Type := news_entry_type.NewFeatures, Notes := "Updated wood counter HUD."}
                news_entry_info{Type := news_entry_type.NewFeatures, Notes := "Updated XP bar/level HUD."}
                news_entry_info{Type := news_entry_type.NewFeatures, Notes := "Added new Ability Boost mechanic that levels up an ability x2 (available from Power Ups)."}
                news_entry_info{Type := news_entry_type.NewFeatures, Notes := "Updated ability select UI to highlight star icons to project the resulting level."}
                news_entry_info{Type := news_entry_type.NewFeatures, Notes := "Enabled public matchmaking - matchmaking is disabled while a game is in progress."}

                news_entry_info{Type := news_entry_type.PerformanceOptimizations, Notes := "Implemented NPC pooling system to allow NPCs to spawn at a faster rate."}
                news_entry_info{Type := news_entry_type.PerformanceOptimizations, Notes := "Implemented XP orb count limit (100). If limit is reach XP will be merged into the nearest orb."}

                news_entry_info{Type := news_entry_type.GameplayChanges, Notes := "Adjusted XP curve to require more XP at higher levels (40+), align with increase NPC count."}
                news_entry_info{Type := news_entry_type.GameplayChanges, Notes := "Melee Slimes attack is triggered from a further distance."}
                news_entry_info{Type := news_entry_type.GameplayChanges, Notes := "Melee Elite Slime attack area increased by 50%."}
                news_entry_info{Type := news_entry_type.GameplayChanges, Notes := "Tree respawn time increased 30s > 60s."}
                news_entry_info{Type := news_entry_type.GameplayChanges, Notes := "Tree base gather amount increased 30 - 60 > 60 - 120."}
                news_entry_info{Type := news_entry_type.GameplayChanges, Notes := "Tree item drop chance increase 20% > 30%."}
                news_entry_info{Type := news_entry_type.GameplayChanges, Notes := "Accolades updated to award larger amounts at a lower frequency."}
                news_entry_info{Type := news_entry_type.GameplayChanges, Notes := "Arena 1 - removed vines (invisible collision) & stone floor blocking VFX."}
                news_entry_info{Type := news_entry_type.GameplayChanges, Notes := "Arena 1 - removed last wave (total duration is now 10 minutes)."}

                news_entry_info{Type := news_entry_type.BugFixes, Notes := "Fixed - Health multipliers was not applying to bosses."}
                news_entry_info{Type := news_entry_type.BugFixes, Notes := "Fixed - Abilities could be rerolled when no new options are available."}
                news_entry_info{Type := news_entry_type.BugFixes, Notes := "Fixed - Shimmering/aliasing in the minimap icons."}
                news_entry_info{Type := news_entry_type.BugFixes, Notes := "Fixed - Multi-level up was not granting ability select immediately."}
                news_entry_info{Type := news_entry_type.BugFixes, Notes := "Fixed - NPC navigating to the incorrect position after teleporting."}
                news_entry_info{Type := news_entry_type.BugFixes, Notes := "Fixed - Down but not out device settings was not applying correctly."}
                news_entry_info{Type := news_entry_type.BugFixes, Notes := "Fixed - Boss fight music was not playing."}
                news_entry_info{Type := news_entry_type.BugFixes, Notes := "Fixed - Tree gather VFX was always showing the base amount (was granting correct amount)."}
                news_entry_info{Type := news_entry_type.BugFixes, Notes := "Fixed - DBNO will be stuck in the game if other players leave the server."}
                
        news_info:
            Date := "September 20, 2024"
            Title := "Update 1 - Item shop, ability banish & gameplay balancing"
            Version := 0.01
            MaybeImage := option{UserInterface.News.NewsUpdateHighlights_01}
            Entries := array:               
                news_entry_info{Type := news_entry_type.NewFeatures, Notes := "Added Banish mechanic to ability selection, stops ability from showing for the rest of the run."}
                news_entry_info{Type := news_entry_type.NewFeatures, Notes := "Added Experience Multiplier ability, increases experience earned from orbs."}
                news_entry_info{Type := news_entry_type.NewFeatures, Notes := "Added Item Shop mechanic. Spawns every 5 minutes, use wood to purchase items."}
                news_entry_info{Type := news_entry_type.None, Notes := "Consumables - Chug Splash, Bananas of the Gods and Air Strike."}
                news_entry_info{Type := news_entry_type.None, Notes := "Temporary Weapons - Shockwave Hammer, Nitro Fists and Vampire Katana (lasts 60s)."}
                news_entry_info{Type := news_entry_type.NewFeatures, Notes := "Added News UI."}
                news_entry_info{Type := news_entry_type.NewFeatures, Notes := "Added Menu UI."}
                news_entry_info{Type := news_entry_type.NewFeatures, Notes := "Added Orbit camera option to menu UI."}
                news_entry_info{Type := news_entry_type.NewFeatures, Notes := "Added server time to menu UI."}
                news_entry_info{Type := news_entry_type.NewFeatures, Notes := "Added accolades level up (1/10/100), gather tree (1/10/100), ability banish/maxed, item shop."}

                news_entry_info{Type := news_entry_type.GameplayChanges, Notes := "Sword Weapon - increased base range by 20%."}
                news_entry_info{Type := news_entry_type.GameplayChanges, Notes := "Sword Weapon - increased attack angle from 60° to 90°."}
                news_entry_info{Type := news_entry_type.GameplayChanges, Notes := "Sword Weapon - changed to follow player position."}
                news_entry_info{Type := news_entry_type.GameplayChanges, Notes := "Sentry Weapon - reduced duration 6s > 5s."}
                news_entry_info{Type := news_entry_type.GameplayChanges, Notes := "Sentry Weapon - reduced fire rate 0.67/s > 1.00/s."}
                news_entry_info{Type := news_entry_type.GameplayChanges, Notes := "Sentry Weapon - base range increased from 125 > 400, no longer scales with projectile radius."}
                news_entry_info{Type := news_entry_type.GameplayChanges, Notes := "Sentry Weapon - now does AoE damage at target position, AoE scales with projectile radius."}
                news_entry_info{Type := news_entry_type.GameplayChanges, Notes := "Shuriken Weapon - increased base damage by 50%."}
                news_entry_info{Type := news_entry_type.GameplayChanges, Notes := "Shuriken Weapon - ability base cooldown reduced by 1.5s."}
                news_entry_info{Type := news_entry_type.GameplayChanges, Notes := "Shuriken Weapon - projectile rate upgrade changed to increase speed (+10% per level)."}
                news_entry_info{Type := news_entry_type.GameplayChanges, Notes := "Gather rate ability buffed to reduce gather time from 0.2s > 0.25s per level."}
                news_entry_info{Type := news_entry_type.GameplayChanges, Notes := "Gather rate ability also grants gather amount multiplier +10% per level."}
                news_entry_info{Type := news_entry_type.GameplayChanges, Notes := "Increased knockback distance by 33% and amount by 50%."}
                news_entry_info{Type := news_entry_type.GameplayChanges, Notes := "Changed knockback to interrupt NPC attacks (boss/miniboss excepted)."}
                news_entry_info{Type := news_entry_type.GameplayChanges, Notes := "Updated dash ability to apply knockback to NPCs in the dash direction."}
                news_entry_info{Type := news_entry_type.GameplayChanges, Notes := "Reworked yellow slimes attack to have pause before charging and attack animation."}
                news_entry_info{Type := news_entry_type.GameplayChanges, Notes := "Increased health orb heal amount from 25% > 50%."}
                news_entry_info{Type := news_entry_type.GameplayChanges, Notes := "Added clarification text for projectile count upgrade (+1 every odd level)."}

                news_entry_info{Type := news_entry_type.BugFixes, Notes := "Fixed - VFX not being in sync with damage at higher difficulty levels."}
                news_entry_info{Type := news_entry_type.BugFixes, Notes := "Fixed - Player counter not updating on player join."}
                news_entry_info{Type := news_entry_type.BugFixes, Notes := "Fixed - Arena select UI showed text rewards instead of difficulty."}
                news_entry_info{Type := news_entry_type.BugFixes, Notes := "Fixed - Player can be stuck on defeat message if leveled up before elimination."}
                news_entry_info{Type := news_entry_type.BugFixes, Notes := "Fixed - Shield VFX getting duplicated after the boss fight intro."}
                news_entry_info{Type := news_entry_type.BugFixes, Notes := "Fixed - HUD can disappear when the player opens Fortnite menu."}
                news_entry_info{Type := news_entry_type.BugFixes, Notes := "Fixed - Arena can be changed after game has started."}