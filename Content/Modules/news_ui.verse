using { /Fortnite.com/UI }
using { /UnrealEngine.com/Temporary/SpatialMath }
using { /UnrealEngine.com/Temporary/UI }
using { /Verse.org/Colors }

<# ----------------------------------------------------------------------------------------------------
    news_ui
---------------------------------------------------------------------------------------------------- #>
news_ui := class(panel_ui):
    Title<override>:string = "News"
    Anchors<override>:anchors = anchors{Minimum := vector2{X := 0.0, Y := 0.0}, Maximum := vector2{X := 1.0, Y := 1.0}}
    HasCloseButton<override>:logic = false

    MaxVersionButtons:int = 10
    MaxNotesPerPage:int = 10

    HighlightImage:texture_block = texture_block:
        DefaultImage := UserInterface.News.NewsUpdateHighlights_01
        DefaultDesiredSize := vector2{X := 1536.0, Y := 128.0}

    VersionStackBox:stack_box = stack_box:
        Orientation := orientation.Vertical

    DateText:text_block = text_block:
        DefaultTextColor := ThemeColors.Highlight
        DefaultShadowColor := ThemeColors.Highlight * 0.12
        DefaultShadowOffset := option{vector2{X := 2.0, Y := 2.0}}
        DefaultShadowOpacity := 1.0

    NewsTitleText:text_block = text_block:
        DefaultTextColor := ThemeColors.PrimaryText
        DefaultShadowOffset := option{vector2{X := 2.0, Y := 2.0}}
        DefaultShadowOpacity := 0.5

    VersionText:text_block = text_block:
        DefaultTextColor := ThemeColors.PrimaryText
        DefaultShadowOffset := option{vector2{X := 2.0, Y := 2.0}}
        DefaultShadowOpacity := 0.5

    PageNumberText:text_block = text_block:
        DefaultTextColor := ThemeColors.PrimaryText

    var ActiveNewsInfo:news_info = news_info{}
    var CurrentVersionButtonsPage:int = 1
    var CurrentNotesPage:int = 1
    var VersionButtons:[]text_button_base = array{}
    var NotesEntries:[]tuple(texture_block, text_block) = array{}
    
    <# ----------------------------------------------------------------------------------------------------
        Create UI
    ---------------------------------------------------------------------------------------------------- #>
    CreateContent<override>():widget =
        NextButton:button_quiet = button_quiet{DefaultText := MakeMessage("▶")}
        NextButton.OnClick().Subscribe(OnClickNext)
        PreviousButton:button_quiet = button_quiet{DefaultText := MakeMessage("◀")}
        PreviousButton.OnClick().Subscribe(OnClickPrevious)

        NextVersionsButton:button_quiet = button_quiet{DefaultText := MakeMessage("▼")}
        NextVersionsButton.OnClick().Subscribe(OnClickNextVersions)
        PreviousVersionsButton:button_quiet = button_quiet{DefaultText := MakeMessage("▲")}
        PreviousVersionsButton.OnClick().Subscribe(OnClickPreviousVersions)

        CloseButton:button_loud = button_loud{DefaultText := MakeMessage("Close")}
        CloseButton.OnClick().Subscribe(OnClickClose)

        stack_box:
            Orientation := orientation.Horizontal
            Slots := array:
                stack_box_slot:
                    HorizontalAlignment := horizontal_alignment.Fill
                    VerticalAlignment := vertical_alignment.Fill
                    Widget := stack_box:
                        Orientation := orientation.Vertical
                        Slots := array:
                            stack_box_slot:
                                HorizontalAlignment := horizontal_alignment.Fill
                                VerticalAlignment := vertical_alignment.Top
                                Padding := margin{Left := 0.0, Top := 0.0, Right := 0.0, Bottom := 10.0}
                                Widget := CloseButton
                            stack_box_slot:
                                HorizontalAlignment := horizontal_alignment.Fill
                                VerticalAlignment := vertical_alignment.Top
                                Widget := PreviousVersionsButton
                            stack_box_slot:
                                HorizontalAlignment := horizontal_alignment.Fill
                                VerticalAlignment := vertical_alignment.Center
                                Distribution := option{1.0}
                                Padding := margin{Left := 0.0, Top := 20.0, Right := 0.0, Bottom := 20.0}
                                Widget := stack_box:
                                    Orientation := orientation.Vertical
                                    Slots := for (Index := 1..MaxVersionButtons):
                                        VersionButton := button_quiet{}
                                        spawn{AwaitOnClick(VersionButton, Index - 1)}
                                        set VersionButtons += array{VersionButton}
                                        stack_box_slot:
                                            HorizontalAlignment := horizontal_alignment.Fill
                                            VerticalAlignment := vertical_alignment.Fill
                                            Padding := margin{Top := if (Index = 1) then 0.0 else 10.0, Left := 0.0, Right := 0.0, Bottom := 0.0}
                                            Widget := VersionButton
                            stack_box_slot:
                                HorizontalAlignment := horizontal_alignment.Fill
                                VerticalAlignment := vertical_alignment.Bottom
                                Widget := NextVersionsButton
                            stack_box_slot:
                                HorizontalAlignment := horizontal_alignment.Fill
                                Widget := color_block:
                                    DefaultOpacity := 0.0
                                    DefaultDesiredSize := vector2{X := 200.0, Y := 0.0}
                stack_box_slot:
                    HorizontalAlignment := horizontal_alignment.Fill
                    VerticalAlignment := vertical_alignment.Fill
                    Padding := margin{Top := 0.0, Left := 20.0, Right := 0.0, Bottom := 0.0}
                    Distribution := option{1.0}
                    Widget := stack_box:
                        Orientation := orientation.Vertical
                        Slots := array:
                            stack_box_slot:
                                HorizontalAlignment := horizontal_alignment.Fill
                                VerticalAlignment := vertical_alignment.Fill
                                Distribution := option{1.0}
                                Widget := overlay:
                                    Slots := array:
                                        overlay_slot:
                                            HorizontalAlignment := horizontal_alignment.Fill
                                            VerticalAlignment := vertical_alignment.Fill
                                            Widget := MakeBox()
                                        overlay_slot:
                                            HorizontalAlignment := horizontal_alignment.Fill
                                            VerticalAlignment := vertical_alignment.Fill
                                            Padding := margin{Left := 40.0, Top := 20.0, Right := 40.0, Bottom := 20.0}
                                            Widget := stack_box:
                                                Orientation := orientation.Vertical
                                                Slots := array:
                                                    stack_box_slot:
                                                        HorizontalAlignment := horizontal_alignment.Left
                                                        VerticalAlignment := vertical_alignment.Top
                                                        Padding := margin{Top := 0.0, Left := 0.0, Right := 0.0, Bottom := 20.0}
                                                        Widget := HighlightImage
                                                    stack_box_slot:
                                                        HorizontalAlignment := horizontal_alignment.Left
                                                        VerticalAlignment := vertical_alignment.Fill
                                                        Widget := DateText
                                                    stack_box_slot:
                                                        HorizontalAlignment := horizontal_alignment.Left
                                                        VerticalAlignment := vertical_alignment.Fill
                                                        Widget := NewsTitleText
                                                    stack_box_slot:
                                                        HorizontalAlignment := horizontal_alignment.Left
                                                        VerticalAlignment := vertical_alignment.Fill
                                                        Padding := margin{Top := 0.0, Left := 0.0, Right := 0.0, Bottom := 20.0}
                                                        Widget := VersionText
                                                    stack_box_slot:
                                                        HorizontalAlignment := horizontal_alignment.Left
                                                        VerticalAlignment := vertical_alignment.Fill
                                                        Widget := stack_box:
                                                            Orientation := orientation.Vertical
                                                            Slots := for (Index := 1..MaxNotesPerPage):
                                                                NotesIcon := texture_block:
                                                                    DefaultImage := UserInterface.Icons.NewsEntryType_NewFeature_64
                                                                    DefaultDesiredSize := vector2{X := 40.0, Y := 40.0}

                                                                NotesText := text_block:
                                                                    DefaultTextColor := ThemeColors.PrimaryText

                                                                NotesEntry := (NotesIcon, NotesText)
                                                                set NotesEntries += array{NotesEntry}

                                                                stack_box_slot:
                                                                    HorizontalAlignment := horizontal_alignment.Fill
                                                                    Widget := stack_box:
                                                                        Orientation := orientation.Horizontal
                                                                        Slots := array:
                                                                            stack_box_slot:
                                                                                HorizontalAlignment := horizontal_alignment.Left
                                                                                VerticalAlignment := vertical_alignment.Top
                                                                                Widget := NotesIcon
                                                                            stack_box_slot:
                                                                                HorizontalAlignment := horizontal_alignment.Fill
                                                                                VerticalAlignment := vertical_alignment.Fill
                                                                                Padding := margin{Left := 20.0, Top := 0.0, Right := 0.0, Bottom := 0.0}
                                                                                Widget := NotesText
                            stack_box_slot:
                                HorizontalAlignment := horizontal_alignment.Fill
                                VerticalAlignment := vertical_alignment.Fill
                                Padding := margin{Left := 0.0, Top := 20.0, Right := 0.0, Bottom := 0.0}
                                Widget := stack_box:
                                    Orientation := orientation.Horizontal
                                    Slots := array:
                                        stack_box_slot:
                                            HorizontalAlignment := horizontal_alignment.Fill
                                            Distribution := option{1.0}
                                            Widget := PreviousButton
                                        stack_box_slot:
                                            HorizontalAlignment := horizontal_alignment.Fill
                                            Padding := margin{Left := 20.0, Top := 0.0, Right := 20.0, Bottom := 0.0}
                                            Widget := PageNumberText
                                        stack_box_slot:
                                            HorizontalAlignment := horizontal_alignment.Fill
                                            Distribution := option{1.0}
                                            Widget := NextButton

    <# ----------------------------------------------------------------------------------------------------
        Get functions
    ---------------------------------------------------------------------------------------------------- #>
    GetIndex(Index:int)<transacts>:int =
        (CurrentVersionButtonsPage - 1) * MaxVersionButtons + Index

    GetMaxPages()<transacts>:int =
        Ceil[(ActiveNewsInfo.Entries.Length * 1.0) / (MaxNotesPerPage * 1.0)] or 1
        
    GetMaxVersionButtonPages()<transacts>:int =
        Ceil[(GetNewsInfo().Length * 1.0) / (MaxVersionButtons * 1.0)] or 1

    <# ----------------------------------------------------------------------------------------------------
        Update functions
    ---------------------------------------------------------------------------------------------------- #>
    Update(NewsInfo:news_info):void =
        set ActiveNewsInfo = NewsInfo

        DateText.SetText(MakeMessage("{NewsInfo.Date}"))
        NewsTitleText.SetText(MakeMessage("{NewsInfo.Title}"))
        VersionText.SetText(MakeMessage("v{GetRoundedString(NewsInfo.Version, 3)}"))

        for (Index -> NotesEntry : NotesEntries, ParentWidget := NotesEntry(0).GetParentWidget[]):
            if (NewsEntry := NewsInfo.Entries[(CurrentNotesPage - 1) * MaxNotesPerPage + Index]):
                if (NewsEntry.Type <> news_entry_type.None):
                    NotesEntry(0).SetImage(NewsEntry.Type.ToIcon())
                    NotesEntry(0).SetTint(NewsEntry.Type.ToColor())
                    NotesEntry(0).SetVisibility(widget_visibility.Visible)
                    NotesEntry(1).SetText(MakeMessage("{NewsEntry.Notes.WrapText(1450.0)}"))
                else:
                    NotesEntry(0).SetVisibility(widget_visibility.Hidden)
                    NotesEntry(1).SetText(MakeMessage("- {NewsEntry.Notes.WrapText(1450.0)}"))

                ParentWidget.SetVisibility(widget_visibility.Visible)
            else:
                ParentWidget.SetVisibility(widget_visibility.Collapsed)

        if (Image := NewsInfo.MaybeImage?):
            HighlightImage.SetImage(Image)
            HighlightImage.SetVisibility(widget_visibility.Visible)
        else:
            HighlightImage.SetVisibility(widget_visibility.Collapsed)

        if (PaginationStackBox := PageNumberText.GetParentWidget[]):
            if (GetMaxPages() > 1):
                PaginationStackBox.SetVisibility(widget_visibility.Visible)
                UpdatePageNumber()
            else:
                PaginationStackBox.SetVisibility(widget_visibility.Collapsed)
    
    UpdatePageNumber():void =
        PageNumberText.SetText(MakeMessage("{CurrentNotesPage}/{GetMaxPages()}"))

    UpdateVersionButtons():void =
        for (Index -> VersionButton : VersionButtons):
            if (NewsInfo := GetNewsInfo()[GetIndex(Index)]):
                VersionButton.SetText(MakeMessage("v{GetRoundedString(NewsInfo.Version, 3)}"))
                VersionButton.SetVisibility(widget_visibility.Visible)
            else:
                VersionButton.SetVisibility(widget_visibility.Hidden)

    <# ----------------------------------------------------------------------------------------------------
        Event handlers
    ---------------------------------------------------------------------------------------------------- #>
    OnShow<override>():void =
        set CurrentNotesPage = 1
        set CurrentVersionButtonsPage = 1

        UpdateVersionButtons()

        if (NewsInfo := GetNewsInfo()[0]):
            Update(NewsInfo)

    OnClickNext(Message:widget_message):void =
        set CurrentNotesPage = (CurrentNotesPage >= GetMaxPages()) and 1 or (CurrentNotesPage + 1)
        Update(ActiveNewsInfo)

    OnClickPrevious(Message:widget_message):void =
        set CurrentNotesPage = (CurrentNotesPage = 1) and GetMaxPages() or (CurrentNotesPage - 1)
        Update(ActiveNewsInfo)

    OnClickNextVersions(Message:widget_message):void =
        set CurrentVersionButtonsPage = (CurrentVersionButtonsPage = 1) and GetMaxVersionButtonPages() or (CurrentVersionButtonsPage - 1)
        UpdateVersionButtons()

    OnClickPreviousVersions(Message:widget_message):void =
        set CurrentVersionButtonsPage = (CurrentVersionButtonsPage = GetMaxVersionButtonPages()) and 1 or (CurrentVersionButtonsPage + 1)
        UpdateVersionButtons()
    
    AwaitOnClick(VersionButton:text_button_base, Index:int)<suspends>:void =
        loop:
            VersionButton.OnClick().Await()
            if (NewsInfo := GetNewsInfo()[GetIndex(Index)]):
                set CurrentNotesPage = 1
                Update(NewsInfo)