using { /Fortnite.com/AI }
using { /Fortnite.com/Animation/PlayAnimation }
using { /Fortnite.com/Characters }
using { /Fortnite.com/Game }
using { /Fortnite.com/Devices }
using { /Verse.org/Simulation }
using { /UnrealEngine.com/Temporary/SpatialMath }
using { WeaponAbilities }

<# ----------------------------------------------------------------------------------------------------
    Data Definitions
---------------------------------------------------------------------------------------------------- #>
npc_attack_type := enum:
    Melee
    Jump
    Projectile
    Area
    Charge
    Dash
    ForwardJump
    Barrage
    Explode

ToString(NPCAttackType:npc_attack_type)<transacts>:string =
    case (NPCAttackType):
        npc_attack_type.Melee => "Melee"
        npc_attack_type.Jump => "Jump"
        npc_attack_type.Projectile => "Projectile"
        npc_attack_type.Area => "Area"
        npc_attack_type.Charge => "Charge"
        npc_attack_type.Dash => "Dash"
        npc_attack_type.ForwardJump => "Forward Jump"
        npc_attack_type.Barrage => "Barrage"
        npc_attack_type.Explode => "Explode"

npc_attack_factory := class<concrete>:
    MeleeAttack:npc_attack_melee = npc_attack_melee{}
    JumpAttack:npc_attack_jump = npc_attack_jump{}
    ProjectileAttack:npc_attack_projectile = npc_attack_projectile{}
    AreaAttack:npc_attack_area = npc_attack_area{}
    ChargeAttack:npc_attack_charge = npc_attack_charge{}
    DashAttack:npc_attack_dash = npc_attack_dash{}
    ForwardJumpAttack:npc_attack_forward_jump = npc_attack_forward_jump{}
    BarrageAttack:npc_attack_barrage = npc_attack_barrage{}
    ExplodeAttack:npc_attack_explode = npc_attack_explode{}

    (npc_attack_factory:)GetNPCAttack(Type:npc_attack_type)<transacts>:npc_attack =
        case (Type):
            npc_attack_type.Melee => MeleeAttack
            npc_attack_type.Jump => JumpAttack
            npc_attack_type.Projectile => ProjectileAttack
            npc_attack_type.Area => AreaAttack
            npc_attack_type.Charge => ChargeAttack
            npc_attack_type.Dash => DashAttack
            npc_attack_type.ForwardJump => ForwardJumpAttack
            npc_attack_type.Barrage => BarrageAttack
            npc_attack_type.Explode => ExplodeAttack

<# ----------------------------------------------------------------------------------------------------
    Global Functions
---------------------------------------------------------------------------------------------------- #>
NPCDamagePlayersWithinRange(TargetPosition:vector3, NPCAttack:npc_attack, NPCBehavior:npc_behavior_base)<suspends>:void =
    AttackProperties := NPCBehavior.AttackProperties
    for:
        Player : GetAlivePlayers()
        PlayerCharacter := Player.GetFortCharacter[]
        DistanceSquared(TargetPosition, PlayerCharacter.GetTransform().Translation) <= Pow(AttackProperties.AttackRange + PlayerCharacterSize, 2.0)
    do:
        Player.ApplyDamage(NPCAttack.GetAttackDamage(AttackProperties.AttackDamage), ?IsBoss := NPCBehavior.IsBoss)
        if (Duration := AttackProperties.StunDuration?):
            Player.Stun(Duration)

<# ----------------------------------------------------------------------------------------------------
    NPC Attack Base Class
---------------------------------------------------------------------------------------------------- #>
npc_attack := class<abstract>:
    Execute(NPCBehaviorBase:npc_behavior_base, AttackProperties:npc_attack_properties, TargetPlayer:player)<suspends>:void

    GetAttackDamage(BaseDamage:float)<transacts>:float =
        NPCWaveDamageMultiplier := 1.0 + (GetCurrentStage() - 1) * GetArenaStageMultiplier()
        NPCArenaDamageMultiplier := 1.0 + GetArenaDamageMultiplier()
        BaseDamage * NPCWaveDamageMultiplier * NPCArenaDamageMultiplier

<# ----------------------------------------------------------------------------------------------------
    npc_attack_melee
---------------------------------------------------------------------------------------------------- #>
npc_attack_melee := class(npc_attack):
    Execute<override>(NPCBehaviorBase:npc_behavior_base, AttackProperties:npc_attack_properties, TargetPlayer:player)<suspends>:void =
        if (Character := NPCBehaviorBase.GetAgent[].GetFortCharacter[], TargetPlayerCharacter := TargetPlayer.GetFortCharacter[]):
            ForwardDiraction := Character.GetTransform().Rotation.GetLocalForward()
            TargetPosition := Character.GetTransform().Translation + ForwardDiraction * (AttackProperties.AttackRange + NPCBehaviorBase.Size)

            if (PlayAnimationController := Character.GetPlayAnimationController[]):
                PlayAnimationController.PlayAndAwait(NPC.AnimationSequences.AS_Slime_Attack, ?PlayRate := 1.0 / AttackProperties.AttackTime)

            if (DistanceSquared(TargetPosition.Make2D(), TargetPlayerCharacter.GetTransform().Translation.Make2D()) <= Pow(AttackProperties.AttackRange + PlayerCharacterSize, 2.0)):
                TargetPlayer.ApplyDamage(GetAttackDamage(AttackProperties.AttackDamage))

<# ----------------------------------------------------------------------------------------------------
    npc_attack_jump
---------------------------------------------------------------------------------------------------- #>
npc_attack_jump := class(npc_attack):
    Execute<override>(NPCBehaviorBase:npc_behavior_base, AttackProperties:npc_attack_properties, TargetPlayer:player)<suspends>:void =
        var MaybeSpawnedVFX:?creative_prop = false

        defer:
            if (SpawnedVFX := MaybeSpawnedVFX?):
                SpawnedVFX.Dispose()

        if (Character := NPCBehaviorBase.GetAgent[].GetFortCharacter[]):
            TargetPosition := Character.GetTransform().Translation

            set MaybeSpawnedVFX = SpawnVFX(vfx_id.NPCJump, TargetPosition.Make2D(), ?Scale := vector3{X := AttackProperties.AttackRange * 0.1, Y := 1.0, Z := AttackProperties.AttackTime}, ?Duration := AttackProperties.AttackTime + 2.0)
            if (PlayAnimationController := Character.GetPlayAnimationController[]):
                PlayAnimationController.PlayAndAwait(NPC.AnimationSequences.AS_Slime_Jump, ?PlayRate := 1.0 / AttackProperties.AttackTime)
            set MaybeSpawnedVFX = false

            ActivateCameraShake(TargetPosition)
            spawn{NPCDamagePlayersWithinRange(TargetPosition, Self, NPCBehaviorBase)}

<# ----------------------------------------------------------------------------------------------------
    npc_attack_forward_jump
---------------------------------------------------------------------------------------------------- #>
npc_attack_forward_jump := class(npc_attack):
    Execute<override>(NPCBehaviorBase:npc_behavior_base, AttackProperties:npc_attack_properties, TargetPlayer:player)<suspends>:void =
        var MaybeSpawnedVFX:?creative_prop = false

        defer:
            if (SpawnedVFX := MaybeSpawnedVFX?):
                SpawnedVFX.Dispose()

        if:
            Character := NPCBehaviorBase.GetAgent[].GetFortCharacter[]
            FocusInterface := Character.GetFocusInterface[]
            PlayAnimationController := Character.GetPlayAnimationController[]
        then:
            race:
                FocusInterface.MaintainFocus(TargetPlayer)
                Sleep(0.5)

            ForwardDirection := Character.GetTransform().Rotation.GetLocalForward()
            TargetPosition := Character.GetTransform().Translation + ForwardDirection * 500.0

            set MaybeSpawnedVFX = SpawnVFX(vfx_id.NPCJump, TargetPosition.Make2D(), ?Scale := vector3{X := AttackProperties.AttackRange * 0.1, Y := 1.0, Z := AttackProperties.AttackTime}, ?Duration := AttackProperties.AttackTime + 2.0)
            PlayAnimationController.PlayAndAwait(NPC.AnimationSequences.AS_Slime_Forward_Jump, ?PlayRate := 1.0 / AttackProperties.AttackTime)
            set MaybeSpawnedVFX = false

            ActivateCameraShake(TargetPosition)
            spawn{NPCDamagePlayersWithinRange(TargetPosition, Self, NPCBehaviorBase)}
            if (TargetPosition.IsInArena()?, Character.TeleportTo[TargetPosition, rotation{}]) {}

<# ----------------------------------------------------------------------------------------------------
    npc_attack_projectile
---------------------------------------------------------------------------------------------------- #>
npc_attack_projectile := class(npc_attack):  
    Execute<override>(NPCBehaviorBase:npc_behavior_base, AttackProperties:npc_attack_properties, TargetPlayer:player)<suspends>:void =
        var MaybeSpawnedVFX:?creative_prop = false

        defer:
            if (SpawnedVFX := MaybeSpawnedVFX?):
                SpawnedVFX.Dispose()
                set MaybeSpawnedVFX = false

        if (Agent := NPCBehaviorBase.GetAgent[], Character := Agent.GetFortCharacter[], FocusInterface := Character.GetFocusInterface[], TargetPlayerCharacter := TargetPlayer.GetFortCharacter[]):
            CastPosition := Character.GetTransform().Translation.Make2D()
            ProjectileDuration := AttackProperties.AttackRange / AttackProperties.AttackProjectileSpeed? or 1000.0
            MaybeProjectileChargingVFX := SpawnVFX(vfx_id.NPCProjectileCharging, Character.GetTransform().Translation + vector3{Z := 100.0}, ?Scale := vector3{X := (AttackProperties.AttackProjectileRadius? or 50.0) * 0.1, Y := 1.0, Z := AttackProperties.AttackTime}, ?Duration := AttackProperties.AttackTime)          
            set MaybeSpawnedVFX = MaybeProjectileChargingVFX
            
            race:
                Sleep(AttackProperties.AttackTime)
                FocusInterface.MaintainFocus(TargetPlayer)

            SpawnVFX(vfx_id.NPCProjectileWarning, CastPosition, ?Rotation := Character.GetTransform().Rotation, ?Scale := vector3{X := AttackProperties.AttackRange * 0.1, Y := (AttackProperties.AttackProjectileRadius? or 100.0) * 0.1, Z := ProjectileDuration}, ?Duration := ProjectileDuration)

            ProjectileProperties := projectile_properties:
                Owner := Agent
                Rotation := Character.GetTransform().Rotation
                Speed := AttackProperties.AttackProjectileSpeed? or 1000.0
                Distance := AttackProperties.AttackRange
                Damage := GetAttackDamage(AttackProperties.AttackDamage)
                Radius := AttackProperties.AttackProjectileRadius? or 50.0
                VFX := vfx_id.NPCProjectile

            projectile{}.Activate(ProjectileProperties)

<# ----------------------------------------------------------------------------------------------------
    npc_attack_area
---------------------------------------------------------------------------------------------------- #>
npc_attack_area := class(npc_attack):
    Execute<override>(NPCBehaviorBase:npc_behavior_base, AttackProperties:npc_attack_properties, TargetPlayer:player)<suspends>:void =
        var MaybeSpawnedVFX:?creative_prop = false

        defer:
            if (SpawnedVFX := MaybeSpawnedVFX?):
                SpawnedVFX.Dispose()

        if (Character := NPCBehaviorBase.GetAgent[].GetFortCharacter[], TargetPlayerCharacter := TargetPlayer.GetFortCharacter[]):
            TargetPosition := GetRandomPositionInCircle(1000.0, ?Origin := TargetPlayerCharacter.GetTransform().Translation.Make2D())

            set MaybeSpawnedVFX = SpawnVFX(vfx_id.NPCArea, TargetPosition, ?Scale := vector3{X := AttackProperties.AttackRange * 0.1, Y := 1.0, Z := AttackProperties.AttackTime}, ?Duration := AttackProperties.AttackTime + 2.0)
            if (PlayAnimationController := Character.GetPlayAnimationController[]):
                PlayAnimationController.PlayAndAwait(NPC.AnimationSequences.AS_Slime_AreaAttack, ?PlayRate := 1.0 / AttackProperties.AttackTime)
            
            set MaybeSpawnedVFX = false
            spawn{NPCDamagePlayersWithinRange(TargetPosition, Self, NPCBehaviorBase)}

<# ----------------------------------------------------------------------------------------------------
    npc_attack_barrage
---------------------------------------------------------------------------------------------------- #>
npc_attack_barrage := class(npc_attack):
    Execute<override>(NPCBehaviorBase:npc_behavior_base, AttackProperties:npc_attack_properties, TargetPlayer:player)<suspends>:void =
        var MaybeSpawnedVFX:?creative_prop = false

        defer:
            if (SpawnedVFX := MaybeSpawnedVFX?):
                SpawnedVFX.Dispose()

        if (Character := NPCBehaviorBase.GetAgent[].GetFortCharacter[], TargetPlayerCharacter := TargetPlayer.GetFortCharacter[]):
            AttackCount := AttackProperties.AttackCount? or 1

            set MaybeSpawnedVFX = SpawnVFX(vfx_id.NPCProjectileCharging, Character.GetTransform().Translation + vector3{Z := 100.0}, ?Scale := vector3{X := (AttackProperties.AttackProjectileRadius? or 50.0) * 0.1, Y := 1.0, Z := AttackProperties.AttackTime}, ?Duration := AttackProperties.AttackTime)
            if (PlayAnimationController := Character.GetPlayAnimationController[]):
                PlayAnimationController.PlayAndAwait(NPC.AnimationSequences.AS_Slime_AreaAttack, ?PlayRate := 1.0)

            set MaybeSpawnedVFX = false
            for (Index := 1..AttackCount):
                TargetPosition := GetRandomPositionInCircle(1000.0, ?Origin := TargetPlayerCharacter.GetTransform().Translation.Make2D())
                spawn{Barrage(TargetPosition, NPCBehaviorBase)}
                Sleep(0.2)

    Barrage(TargetPosition:vector3, NPCBehaviorBase:npc_behavior_base)<suspends>:void =
        AttackProperties := NPCBehaviorBase.AttackProperties
        SpawnVFX(vfx_id.NPCArea, TargetPosition, ?Scale := vector3{X := AttackProperties.AttackRange * 0.1, Y := 1.0, Z := AttackProperties.AttackTime}, ?Duration := AttackProperties.AttackTime + 2.0)
        Sleep(AttackProperties.AttackTime)
        NPCDamagePlayersWithinRange(TargetPosition, Self, NPCBehaviorBase)

<# ----------------------------------------------------------------------------------------------------
    npc_attack_charge
---------------------------------------------------------------------------------------------------- #>
npc_attack_charge := class(npc_attack):
    Execute<override>(NPCBehaviorBase:npc_behavior_base, AttackProperties:npc_attack_properties, TargetPlayer:player)<suspends>:void =
        if (Character := NPCBehaviorBase.GetAgent[].GetFortCharacter[], Navigatable := Character.GetNavigatable[], TargetPlayerCharacter := TargetPlayer.GetFortCharacter[]):
            Navigatable.Wait(?Duration := 1.0)
            Navigatable.SetMovementSpeedMultiplier(2.0)
            NavigationTarget := MakeNavigationTarget(TargetPlayer)
                            
            race:
                Navigatable.NavigateTo(NavigationTarget, ?MovementType := movement_types.Running)
                Sleep(3.0)

            Navigatable.SetMovementSpeedMultiplier(1.0)
            SpawnVFX(vfx_id.NPCJump, Character.GetTransform().Translation.Make2D(), ?Scale := vector3{X := AttackProperties.AttackRange * 0.1, Y := 1.0, Z := AttackProperties.AttackTime}, ?Duration := AttackProperties.AttackTime + 2.0)
            if (PlayAnimationController := Character.GetPlayAnimationController[]):
                PlayAnimationController.Play(NPC.AnimationSequences.AS_Slime_Charge, ?PlayRate := 1.0 / AttackProperties.AttackTime)

            Sleep(AttackProperties.AttackTime)
            spawn{NPCDamagePlayersWithinRange(Character.GetTransform().Translation, Self, NPCBehaviorBase)}

<# ----------------------------------------------------------------------------------------------------
    npc_attack_dash
---------------------------------------------------------------------------------------------------- #>
npc_attack_dash := class(npc_attack):
    Execute<override>(NPCBehaviorBase:npc_behavior_base, AttackProperties:npc_attack_properties, TargetPlayer:player)<suspends>:void =
        var MaybeSpawnedVFX:?creative_prop = false

        defer:
            if (SpawnedVFX := MaybeSpawnedVFX?):
                SpawnedVFX.Dispose()

        if:
            Character := NPCBehaviorBase.GetAgent[].GetFortCharacter[]
            TargetPlayerCharacter := TargetPlayer.GetFortCharacter[]
            FocusInterface := Character.GetFocusInterface[]
            PlayAnimationController := Character.GetPlayAnimationController[]
        then:
            race:
                FocusInterface.MaintainFocus(TargetPlayer)
                Sleep(0.5)

            InitialPosition := Character.GetTransform().Translation.Make2D()
            ForwardDirection := Character.GetTransform().Rotation.GetLocalForward()
            RectangleWidth := 300.0
            RectangleLength := AttackProperties.AttackRange
            Rectangle := MakeRectangle(InitialPosition, Character.GetTransform().Rotation, RectangleWidth, RectangleLength)
            
            set MaybeSpawnedVFX = SpawnVFX(vfx_id.NPCDash, InitialPosition, ?Rotation := Character.GetTransform().Rotation, ?Scale := vector3{X := RectangleLength * 0.1, Y := RectangleWidth * 0.1, Z := AttackProperties.AttackTime}, ?Duration := AttackProperties.AttackTime)
            PlayAnimationInstance := PlayAnimationController.Play(NPC.AnimationSequences.AS_Slime_Dash)
            Sleep(AttackProperties.AttackTime)

            set MaybeSpawnedVFX = false

            spawn{DamagePlayers(Rectangle, AttackProperties)}

            Sleep(0.5)
            TargetPosition := Character.GetTransform().Translation + ForwardDirection * AttackProperties.AttackRange
            PlayAnimationInstance.Stop()
            if (TargetPosition.IsInArena()?, Character.TeleportTo[TargetPosition, rotation{}]) {}

    DamagePlayers(Rectangle:rectangle, AttackProperties:npc_attack_properties)<suspends>:void =
        for:
            Player : GetAlivePlayers()
            PlayerCharacter := Player.GetFortCharacter[]
            IsInRectangle(PlayerCharacter.GetTransform().Translation, Rectangle)?
        do:
            Player.ApplyDamage(GetAttackDamage(AttackProperties.AttackDamage))
            if (Duration := AttackProperties.StunDuration?):
                Player.Stun(Duration)

<# ----------------------------------------------------------------------------------------------------
    npc_attack_explode
---------------------------------------------------------------------------------------------------- #>
npc_attack_explode := class(npc_attack):
    Execute<override>(NPCBehaviorBase:npc_behavior_base, AttackProperties:npc_attack_properties, TargetPlayer:player)<suspends>:void =
        if (Character := NPCBehaviorBase.GetAgent[].GetFortCharacter[], Navigatable := Character.GetNavigatable[]):
            Sleep(0.1)            
            spawn{NPCDamagePlayersWithinRange(Character.GetTransform().Translation.Make2D(), Self, NPCBehaviorBase)}