using { /Fortnite.com/AI }
using { /Fortnite.com/Characters }
using { /Fortnite.com/Game }
using { /Verse.org/Random }
using { /Verse.org/Simulation }
using { /UnrealEngine.com/Temporary/SpatialMath }

<# ----------------------------------------------------------------------------------------------------
    Constants
---------------------------------------------------------------------------------------------------- #>
NPCMovementSpeedBuffRate:float = 0.01
NPCSpawnPositionHeightOffset:float = 180.0
NPCSpawnInvulnerabilityTime:float = 0.5
NPCDamagedVFXBufferTime:float = 0.1
NPCBaseMaxHealth:float = 30.0

<# ----------------------------------------------------------------------------------------------------
    Data Definitions
---------------------------------------------------------------------------------------------------- #>
npc_attack_properties := class<concrete>:
    @editable
    AttackType:npc_attack_type = npc_attack_type.Melee
    @editable
    AttackDistance:float = 150.0
    @editable
    AttackRange:float = 100.0
    @editable
    AttackDamage:float = 10.0
    @editable
    AttackCooldown:float = 1.0
    @editable
    AttackTime:float = 1.0
    @editable
    AttackProjectileRadius:?float = false
    @editable
    AttackProjectileSpeed:?float = false
    @editable
    AttackCount:?int = false
    @editable
    StunDuration:?float = false

MakeNPCAttackProperties<constructor>(UpdatedNPCAttackProperties:npc_attack_properties)<transacts> := npc_attack_properties:
    AttackType := UpdatedNPCAttackProperties.AttackType
    AttackDistance := UpdatedNPCAttackProperties.AttackDistance
    AttackRange := UpdatedNPCAttackProperties.AttackRange
    AttackDamage := UpdatedNPCAttackProperties.AttackDamage
    AttackCooldown := UpdatedNPCAttackProperties.AttackCooldown
    AttackTime := UpdatedNPCAttackProperties.AttackTime
    AttackProjectileRadius := UpdatedNPCAttackProperties.AttackProjectileRadius
    AttackProjectileSpeed := UpdatedNPCAttackProperties.AttackProjectileSpeed
    AttackCount := UpdatedNPCAttackProperties.AttackCount

ToString(NPCAttackProperties:npc_attack_properties)<transacts>:string =
    "{NPCAttackProperties.AttackType}\n" +
    "AttackDistance: {NPCAttackProperties.AttackDistance}\n" +
    "AttackRange: {NPCAttackProperties.AttackRange}\n" +
    "AttackDamage: {NPCAttackProperties.AttackDamage}\n" +
    "AttackCooldown: {NPCAttackProperties.AttackCooldown}\n" +
    "AttackTime: {NPCAttackProperties.AttackDistance}\n" +
    "AttackProjectileRadius: {"{NPCAttackProperties.AttackProjectileRadius?}" or "false"}\n" +
    "AttackProjectileSpeed: {"{NPCAttackProperties.AttackProjectileSpeed?}" or "false"}\n"
    "AttackCount: {"{NPCAttackProperties.AttackCount?}" or "false"}\n"

npc_behavior_message_module := module:
    SettingsCategory<public><localizes>:message = "Settings"
    BossSettingsCategory<public><localizes>:message = "Boss Settings"
    AttackSettingsCategory<public><localizes>:message = "Attack Settings"

<# ----------------------------------------------------------------------------------------------------
    Global Functions
---------------------------------------------------------------------------------------------------- #>
GetNPCMaxHealth(BaseHealth:float)<transacts>:float =
    BaseHealth * (1.0 + GetCurrentStage() * GetArenaStageMultiplier()) * (1.0 + GetArenaHealthMultiplier())

<# ----------------------------------------------------------------------------------------------------
    npc_behavior_base
---------------------------------------------------------------------------------------------------- #>
npc_behavior_base := class<unique>(npc_behavior):
    @editable:
        Categories := array{npc_behavior_message_module.SettingsCategory}
    Id:npc_id = npc_id.Melee

    @editable:
        Categories := array{npc_behavior_message_module.SettingsCategory}
    var AttackProperties:npc_attack_properties = npc_attack_properties{}

    @editable:
        Categories := array{npc_behavior_message_module.SettingsCategory}
    ItemDrop:?item_drop_id = false

    @editable:
        Categories := array{npc_behavior_message_module.SettingsCategory}
    ItemDropExperienceAmount:?int = false

    @editable:
        Categories := array{npc_behavior_message_module.SettingsCategory}
    Size:float = 25.0

    @editable:
        Categories := array{npc_behavior_message_module.SettingsCategory}
    IsBoss:logic = false

    @editable:
        Categories := array{npc_behavior_message_module.SettingsCategory}
    AccoladeId:accolade_count_id = accolade_count_id.EliminateSlimeMelee

    InterruptedEvent:event(npc_state) = event(npc_state){}
    StoppedEvent:event(tuple()) = event(tuple()){}
    TargetPlayerEvent:event(tuple(?player, player)) = event(tuple(?player, player)){}
    EliminatedEvent:event(tuple()) = event(tuple()){}

    var State:npc_state = npc_state_chase{}
    var MaybeTargetPlayer:?player = false
    var MaybeCooldownStartTime:?float = false
    var BaseHealth:float = 100.0
    var SpawnTime:float = 0.0
    var LastHitTime:float = 0.0
        
    <# ----------------------------------------------------------------------------------------------------
        Initialization
    ---------------------------------------------------------------------------------------------------- #>
    OnBegin<override>()<suspends>:void =
        Sleep(0.0)
        if (Agent := GetAgent[], Character := Agent.GetFortCharacter[]):
            if (NPCWave := GetCurrentNPCWave()?):
                set BaseHealth = Character.GetMaxHealth()
                set SpawnTime = GetSimulationElapsedTime()
                Character.Show()
                Character.SetMaxHealth(GetNPCMaxHealth(BaseHealth))
                ActivateMovementSpeedBuff()

            else:
                Despawn()
                return
            
            Character.DamagedEvent().Subscribe(OnDamaged)
            Character.EliminatedEvent().Subscribe(OnEliminated)
            Self.RegisterNPC()
            UpdateTarget()

            if (TargetPlayer := MaybeTargetPlayer?, TargetCharacter := TargetPlayer.GetFortCharacter[]):
                loop:
                    if:
                        SpawnPosition := GetSpawnPosition(TargetPlayer).Make2D()
                        SpawnRotation := MakeRotation(SpawnPosition, TargetCharacter.GetTransform().Translation.Make2D())
                        Character.TeleportTo[SpawnPosition + vector3{Z := NPCSpawnPositionHeightOffset}, SpawnRotation]
                    then:
                        break

                    Sleep(TickTime * 5)

                spawn{Start()}
            else:
                Print("[npc_behavior_base] OnBegin() TargetPlayer missing!", ?Color := ThemeColors.Failure)
                Despawn()

    <# ----------------------------------------------------------------------------------------------------
        Get Functions
    ---------------------------------------------------------------------------------------------------- #>        
    GetSize()<transacts>:float =
        Size

    GetIsBoss()<transacts>:logic =
        IsBoss

    GetIsInterruptible()<transacts>:logic =
        logic{not IsBoss?}

    IsDamagable()<transacts>:logic =
        logic{GetSimulationElapsedTime() - SpawnTime > NPCSpawnInvulnerabilityTime}

    <# ----------------------------------------------------------------------------------------------------
        Update Functions
    ---------------------------------------------------------------------------------------------------- #>
    Start()<suspends>:void =
        if (Character := GetAgent[].GetFortCharacter[], DownButNotOut := GetDownButNotOut[]):
            race:
                StoppedEvent.Await()
                Character.EliminatedEvent().Await()
                loop:
                    if (not Character.IsActive[]):
                        break

                    if (not npc_state_knockback[State], not npc_state_dance[State]):
                        RetargetPlayer()
                    
                    race:
                        set State = State.OnBegin(Self)
                        loop:
                            Agent := DownButNotOut.AgentDownedEvent.Await()
                            if (Agent = MaybeTargetPlayer?):
                                if (Player := player[Agent], Player.GetPlayerStats(player_stat_id.Revives) > 0):
                                    Sleep(PlayerReviveDuration + 0.1)
                                    
                                UpdateTarget()
                                break

                    Sleep(0.0)

    RetargetPlayer():void =
        if (TargetPlayer := MaybeTargetPlayer?, TargetPlayer.IsActive[]):
            if (TargetCharacter := TargetPlayer.GetFortCharacter[], TargetCharacter.IsDownButNotOut[] or (not IsGamePaused()? and not TargetCharacter.IsVulnerable[])):
                UpdateTarget()
        else:
            UpdateTarget()

    UpdateTarget(?MaybeNewTarget:?player = false):void =
        if (Agent := GetAgent[], NPCManager := GetNPCManager[], Player := MaybeNewTarget? or NPCManager.GetTarget()?):
            Agent.UpdateNPCTarget(MaybeTargetPlayer, Player)
            set MaybeTargetPlayer = option{Player}
            set State = GetNPCState(npc_state_type.Chase)
        else:
            Despawn()

    Knockback(Duration:float, Rotation:rotation)<suspends>:void =
        if (not GetIsInterruptible()?). return
        InterruptedEvent.Signal of npc_state_knockback:
            Duration := Duration
            Rotation := Rotation
            PreviousState := State

    Dance()<suspends>:void =
        InterruptedEvent.Signal of GetNPCState(npc_state_type.Dance)
        
    ShowDamageFeedback(?IsKnockback:logic = true)<suspends>:void =
        if (not GetIsInterruptible()?). return
        if (not IsKnockback?). return
        InterruptedEvent.Signal of npc_state_hit:
            PreviousState := State

    Damage(InstigatorAgent:agent, AbilityDamageArgs:ability_damage_args, ?IsKnockback:logic = true):void =
        if (not IsDamagable()?):
            return

        if (Character := GetAgent[].GetFortCharacter[]):
            #SpawnVFX(AbilityDamageArgs.IsCriticalHit? and vfx_id.DamageNumberCriticalHit or vfx_id.DamageNumber, Character.GetTransform().Translation + vector3{Z := 100.0}, ?Scale := vector3{X := 1.0, Y := 1.0, Z := AbilityDamageArgs.Amount}, ?Duration := 1.0)
            #SpawnVFX(vfx_id.Impact, Character.GetTransform().Translation + vector3{Z := (Size / 50.0 * 80.0) - CharacterHeightOffset}, ?Scale := vector3{X := 2.0, Y := 2.0, Z := 2.0}, ?Duration := 2.0)

            spawn{ShowDamageFeedback(?IsKnockback := IsKnockback)}

            if (GetSimulationElapsedTime() - LastHitTime >= NPCDamagedVFXBufferTime):
                SpawnVFX(AbilityDamageArgs.IsCriticalHit? and vfx_id.DamageNumberCriticalHit or vfx_id.DamageNumber, Character.GetTransform().Translation + vector3{Z := (Size / 50.0 * 80.0) - CharacterHeightOffset}, ?Scale := vector3{X := 2.0, Y := 2.0, Z := AbilityDamageArgs.Amount}, ?Duration := 1.0)
            
            if (NPCManager := GetNPCManager[], NPCManager.NPCSpawned[Id] <= NPCManager.NPCSpawnTarget[Id], Character.GetHealth() - AbilityDamageArgs.Amount <= 0.0):
                EliminatedEvent.Signal()
                Respawn(InstigatorAgent)

                if (Player := player[InstigatorAgent]):
                    HandleEliminated(Player)
            else:
                DamageArgs := damage_args:
                    Instigator := option{InstigatorAgent.GetInstigator()}
                    Amount := AbilityDamageArgs.Amount
                
                Character.Damage(DamageArgs)
                set LastHitTime = GetSimulationElapsedTime()

    Despawn():void =
        if (Character := GetAgent[].GetFortCharacter[], Character.IsActive[]):
            Character.Damage(Character.GetHealth())

    Respawn(InstigatorAgent:agent):void =
        if (Agent := GetAgent[], Character := Agent.GetFortCharacter[]):
            StoppedEvent.Signal()
            #Self.DeactivateMovementModulator()

            if (ExperienceAmount := ItemDropExperienceAmount?):
                Agent.SpawnExperience(ExperienceAmount)
            else if (Drop := ItemDrop?):
                Agent.SpawnItemDrop(Drop)

            if (Player := player[InstigatorAgent]):
                AddEnemiesEliminated(Player)
                Player.AwardAccolade(AccoladeId)
                Player.AwardAccolade(accolade_count_id.EliminateSlimes)

            UpdateTarget()

            if (TargetPlayer := MaybeTargetPlayer?, TargetCharacter := TargetPlayer.GetFortCharacter[]):
                var TeleportAttempts:int = 1

                loop:
                    if:
                        SpawnPosition := GetSpawnPosition(TargetPlayer).Make2D()
                        SpawnRotation := MakeRotation(SpawnPosition, TargetCharacter.GetTransform().Translation.Make2D())
                        Character.TeleportTo[SpawnPosition + vector3{Z := NPCSpawnPositionHeightOffset}, SpawnRotation]
                    then:
                        break

                    set TeleportAttempts += 1
                    if (TeleportAttempts >= 3):
                        break

            if (NPCWave := GetCurrentNPCWave()?):
                ActivateMovementSpeedBuff()
                Character.SetMaxHealth(GetNPCMaxHealth(BaseHealth))
                Character.SetHealth(Character.GetMaxHealth())
                set SpawnTime = GetSimulationElapsedTime()

            spawn{Start()}

    ActivateMovementSpeedBuff():void =
        if (GameSpeed := Floor[GetGameSpeed() - 1.0]):
            case (GameSpeed):
                2 => Self.ActivateMovementModulator(4)
                3 => Self.ActivateMovementModulator(9)
                _ => Self.DeactivateMovementModulator()

    HandleEliminated(EliminatingPlayer:player):void = {}

    <# ----------------------------------------------------------------------------------------------------
        Event handlers
    ---------------------------------------------------------------------------------------------------- #>
    OnDamaged(DamageResult:damage_result):void = {}
    
    OnEliminated(EliminationResult:elimination_result):void =
        EliminationResult.EliminatedCharacter.Hide()

        if (Character := EliminationResult.EliminatingCharacter?, Player := player[Character.GetAgent[]]):
            HandleEliminated(Player)
            AddEnemiesEliminated(Player)
            Player.AwardAccolade(AccoladeId)
            Player.AwardAccolade(accolade_count_id.EliminateSlimes)

        Self.UnregisterNPC(MaybeTargetPlayer)
        EliminatedEvent.Signal()