using { /Fortnite.com/Characters }
using { /Fortnite.com/Game }
using { /Verse.org/Random }
using { /Verse.org/Simulation }
using { /UnrealEngine.com/Temporary/SpatialMath }

<# ----------------------------------------------------------------------------------------------------
    Data Definitions
---------------------------------------------------------------------------------------------------- #>
boss_attack := class<concrete>:
    @editable:
        Categories := array{npc_behavior_message_module.AttackSettingsCategory}
    AttackProperties:npc_attack_properties = npc_attack_properties{}
    @editable:
        Categories := array{npc_behavior_message_module.AttackSettingsCategory}
    Weight:int = 0
    @editable:
        Categories := array{npc_behavior_message_module.AttackSettingsCategory}
    CastCount:int = 1

<# ----------------------------------------------------------------------------------------------------
    npc_behavior_boss
---------------------------------------------------------------------------------------------------- #>
npc_behavior_boss := class(npc_behavior_base):
    @editable:
        Categories := array{npc_behavior_message_module.BossSettingsCategory}
    BossAttacks:[]boss_attack = array{}

    <# ----------------------------------------------------------------------------------------------------
        Initialization
    ---------------------------------------------------------------------------------------------------- #>
    OnBegin<override>()<suspends>:void =
        Sleep(0.0)
        if (Agent := GetAgent[], Character := Agent.GetFortCharacter[]):
            set BaseHealth = Character.GetMaxHealth()
            Character.DamagedEvent().Subscribe(OnDamaged)
            Character.EliminatedEvent().Subscribe(OnEliminated)
            if (Character.TeleportTo[Character.GetTransform().Translation, Character.GetTransform().Rotation.ApplyYaw(PiFloat)]) {}
            Character.SetMaxHealth(Character.GetMaxHealth() * (1.0 + GetArenaHealthMultiplier()))
            Self.RegisterNPC()

    OnEnd<override>():void =
        HideMessage(message_id.BossHealthBar)

    <# ----------------------------------------------------------------------------------------------------
        Get Functions
    ---------------------------------------------------------------------------------------------------- #>        
    GetRandomBossAttack()<decides><transacts>:boss_attack =
        AttackWeights := for (BossAttack : BossAttacks) {BossAttack.Weight}
        RandomIndex := GetWeightedRandom(AttackWeights)
        BossAttacks[RandomIndex]

    <# ----------------------------------------------------------------------------------------------------
        Update Functions
    ---------------------------------------------------------------------------------------------------- #>
    Barrage()<suspends>:void =
        AttackTime := 2.0
        AttackRadius := 200.0
        
        TargetPosition := if (Player := GetRandomAlivePlayer()?, Character := Player.GetFortCharacter[]):
            OffsetAmount := 800.0
            PositionOffset := vector3{X := GetRandomFloat(-OffsetAmount, OffsetAmount), Y := GetRandomFloat(-OffsetAmount, OffsetAmount), Z := 0.0}
            Character.GetTransform().Translation.Make2D() + PositionOffset
        else:
            GetRandomPositionInCircle(GetActiveArenaSize()) 

        SpawnVFX(vfx_id.NPCArea, TargetPosition, ?Scale := vector3{X := AttackRadius * 0.1, Y := 1.0, Z := AttackTime}, ?Duration := AttackTime + 2.0)
        Sleep(AttackTime)

        NPCArenaDamageMultiplier := GetArenaDamageMultiplier() + 1.0
        for:
            Player : GetAlivePlayers()
            PlayerCharacter := Player.GetFortCharacter[]
            DistanceSquared(TargetPosition, PlayerCharacter.GetTransform().Translation) <= Pow(AttackRadius + PlayerCharacterSize, 2.0)
        do:
            PlayerCharacter.Damage(20.0 * NPCArenaDamageMultiplier)

    Start<override>()<suspends>:void =   
        if (Agent := GetAgent[], Character := Agent.GetFortCharacter[], DownButNotOut := GetDownButNotOut[]):
            Sleep(TickTime)
            ShowMessage(message_id.BossHealthBar, "{Floor[Max(0.0, Character.GetHealth())] or 0}/{Floor[Character.GetMaxHealth()] or 0}", ?DisplayTime := 0.0)
            SetBossHealthBarProgress(1.0)
            Sleep(1.0)
            
            race:
                Character.EliminatedEvent().Await()
                loop:
                    spawn{Barrage()}
                    Sleep(GetRandomFloat(0.25, 0.5))
                loop:
                    if (not Character.IsActive[]):
                        break

                    UpdateTarget()

                    race:
                        loop:
                            DownedAgent := DownButNotOut.AgentDownedEvent.Await()
                            if (DownedAgent = MaybeTargetPlayer?):
                                if (Player := player[Agent], Player.GetPlayerStats(player_stat_id.Revives) > 0):
                                    Sleep(PlayerReviveDuration + 0.1)

                                UpdateTarget()
                                break

                        if (Player := MaybeTargetPlayer?, BossAttack := GetRandomBossAttack[]):
                            set AttackProperties = BossAttack.AttackProperties
                            for (Index := 1..BossAttack.CastCount):
                                GetNPCState(npc_state_type.Focus).OnBegin(Self)
                                GetNPCState(npc_state_type.Attack).OnBegin(Self)

                            set AttackProperties = npc_attack_properties:
                                MakeNPCAttackProperties<constructor>(AttackProperties)
                                AttackDistance := 500.0

                            GetNPCState(npc_state_type.Cooldown).OnBegin(Self)

                    Sleep(0.0)

    UpdateTarget<override>(?MaybeNewTarget:?player = false):void =
        if (Agent := GetAgent[], Player := player[GetRandomAlivePlayer()?]):
            set MaybeTargetPlayer = option{Player}
        else:
            OnEnd()

    <# ----------------------------------------------------------------------------------------------------
        Event handlers
    ---------------------------------------------------------------------------------------------------- #>
    OnDamaged<override>(DamageResult:damage_result):void =
        if (Character := GetAgent[].GetFortCharacter[]):
            ShowMessage(message_id.BossHealthBar, "{Floor[Max(0.0, Character.GetHealth())] or 0}/{Floor[Character.GetMaxHealth()] or 0}", ?DisplayTime := 0.0)
            SetBossHealthBarProgress(Character.GetHealth() / Character.GetMaxHealth())
        
        if (Instigator := DamageResult.Instigator?, Agent := Instigator.GetInstigatorAgent[], Player := player[Agent]):
            Player.AddGameStat(player_game_stat_id.DamageDealtBoss, DamageResult.Amount)