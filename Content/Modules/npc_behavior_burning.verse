using { /Fortnite.com/Characters }
using { /Verse.org/Random }
using { /Verse.org/Simulation }
using { /UnrealEngine.com/Temporary/SpatialMath }

<# ----------------------------------------------------------------------------------------------------
    npc_behavior_burning_slime
---------------------------------------------------------------------------------------------------- #>
npc_behavior_burning_slime := class(npc_behavior_base):
    <# ----------------------------------------------------------------------------------------------------
        Update Functions
    ---------------------------------------------------------------------------------------------------- #>
    Start<override>()<suspends>:void =
        if (Character := GetAgent[].GetFortCharacter[], DownButNotOut := GetDownButNotOut[]):
            race:
                StoppedEvent.Await()
                Character.EliminatedEvent().Await()
                loop:
                    if (not Character.IsActive[]):
                        break

                    if (not npc_state_knockback[State], not npc_state_dance[State]):
                        RetargetPlayer()
                    
                    race:
                        GetNPCState(npc_state_type.Chase).OnBegin(Self)
                        loop:
                            Sleep(AttackProperties.AttackCooldown)
                            NPCDamagePlayersWithinRange(Character.GetTransform().Translation.Make2D(), GetNPCAttack(AttackProperties.AttackType), Self)
                        loop:
                            Agent := DownButNotOut.AgentDownedEvent.Await()
                            if (Agent = MaybeTargetPlayer?):
                                if (Player := player[Agent], Player.GetPlayerStats(player_stat_id.Revives) > 0):
                                    Sleep(PlayerReviveDuration + 0.1)
                                    
                                UpdateTarget()
                                break

                    Sleep(0.0)