using { /Fortnite.com/Characters }
using { /Verse.org/Random }
using { /Verse.org/Simulation }
using { /UnrealEngine.com/Temporary/SpatialMath }

<# ----------------------------------------------------------------------------------------------------
    npc_behavior_golden_slime
---------------------------------------------------------------------------------------------------- #>
npc_behavior_golden_slime := class(npc_behavior_base):
    <# ----------------------------------------------------------------------------------------------------
        Initialization
    ---------------------------------------------------------------------------------------------------- #>
    OnBegin<override>()<suspends>:void =
        Sleep(0.0)
        if (Agent := GetAgent[], Character := Agent.GetFortCharacter[]):
            set BaseHealth = Character.GetMaxHealth()
            Character.DamagedEvent().Subscribe(OnDamaged)
            Character.EliminatedEvent().Subscribe(OnEliminated)
            if (Character.TeleportTo[Character.GetTransform().Translation, Character.GetTransform().Rotation.ApplyYaw(PiFloat)]) {}
            Character.SetMaxHealth(GetNPCMaxHealth(Character.GetMaxHealth()))
            Self.RegisterNPC()
            Start()

    <# ----------------------------------------------------------------------------------------------------
        Update Functions
    ---------------------------------------------------------------------------------------------------- #>
    Start<override>()<suspends>:void =
        Sleep(1.0)
        if (Agent := GetAgent[], Character := Agent.GetFortCharacter[], DownButNotOut := GetDownButNotOut[]):
            race:
                Character.EliminatedEvent().Await()
                loop:
                    if (not Character.IsActive[]):
                        break

                    race:
                        GetNPCState(npc_state_type.Flee).OnBegin(Self)
                        Sleep(GetRandomFloat(5.0, 10.0))

                    Sleep(GetRandomFloat(2.0, 3.0))

    Damage<override>(InstigatorAgent:agent, AbilityDamageArgs:ability_damage_args, ?IsKnockback:logic = true):void =
        if (not IsDamagable()?):
            return

        if (Character := GetAgent[].GetFortCharacter[]):
            DamageAmountOverride := Character.GetMaxHealth() * 0.01 / (GetActivePlayers().Length * 1.0) * GetGameSpeed()
            #SpawnVFX(AbilityDamageArgs.IsCriticalHit? and vfx_id.DamageNumberCriticalHit or vfx_id.DamageNumber, Character.GetTransform().Translation + vector3{Z := 100.0}, ?Scale := vector3{X := 1.0, Y := 1.0, Z := DamageAmountOverride}, ?Duration := 1.0)
            #SpawnVFX(vfx_id.Impact, Character.GetTransform().Translation + vector3{Z := (Size / 50.0 * 80.0) - CharacterHeightOffset}, ?Scale := vector3{X := 2.0, Y := 2.0, Z := 2.0}, ?Duration := 2.0)
            if (GetSimulationElapsedTime() - LastHitTime >= NPCDamagedVFXBufferTime):
                SpawnVFX(AbilityDamageArgs.IsCriticalHit? and vfx_id.DamageNumberCriticalHit or vfx_id.DamageNumber, Character.GetTransform().Translation + vector3{Z := (Size / 50.0 * 80.0) - CharacterHeightOffset}, ?Scale := vector3{X := 2.0, Y := 2.0, Z := AbilityDamageArgs.Amount}, ?Duration := 1.0)

            spawn{ShowDamageFeedback(?IsKnockback := IsKnockback)}
            Character.Damage(DamageAmountOverride)
            set LastHitTime = GetSimulationElapsedTime()