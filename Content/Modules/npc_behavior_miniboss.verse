using { /Fortnite.com/Characters }
using { /Fortnite.com/Game }
using { /Verse.org/Simulation }
using { /UnrealEngine.com/Temporary/SpatialMath }

GetMinibossEliminationTarget()<transacts>:int =
    (GetArenaManager[].GetActiveArena[].MinibossMultiplier * GetCurrentStage() * 3) or 1

IsMinibossEventCompleted()<transacts>:logic =
    logic{GetGameStat(game_stat_id.EliminatedMiniboss) >= GetMinibossEliminationTarget() * 1.0}

<# ----------------------------------------------------------------------------------------------------
    npc_behavior_miniboss
---------------------------------------------------------------------------------------------------- #>
npc_behavior_miniboss := class(npc_behavior_base):
    @editable:
        Categories := array{npc_behavior_message_module.BossSettingsCategory}
    BossAttacks:[]boss_attack = array{}

    @editable:
        Categories := array{npc_behavior_message_module.BossSettingsCategory}
    MinimapIconIndex:int = 0

    var PlayerDamageReceived:[player]float = map{}

    <# ----------------------------------------------------------------------------------------------------
        Get Functions
    ---------------------------------------------------------------------------------------------------- #>        
    GetRandomBossAttack()<decides><transacts>:boss_attack =
        AttackWeights := for (BossAttack : BossAttacks) {BossAttack.Weight}
        RandomIndex := GetWeightedRandom(AttackWeights)
        BossAttacks[RandomIndex]

    <# ----------------------------------------------------------------------------------------------------
        Update Functions
    ---------------------------------------------------------------------------------------------------- #>
    Start<override>()<suspends>:void =   
        if (Agent := GetAgent[], Character := Agent.GetFortCharacter[], DownButNotOut := GetDownButNotOut[]):          
            race:
                Character.EliminatedEvent().Await()

                if (NPCManager := GetNPCManager[], MinibossMapIndicator := NPCManager.GetMinibossMapIndicator[]):
                    defer:
                        NPCManager.ReturnMinibossMapIndicator(MinibossMapIndicator)
                    
                    MinibossMapIndicator.CinematicSequence.Play()
                    MinibossMapIndicator.CinematicSequence.SetPlaybackTime(MinimapIconIndex * 1.0)
                    MinibossMapIndicator.CinematicSequence.Pause()

                    if (MinibossMapIndicator.MapIndicator.TeleportTo[Character.GetTransform()]) {}

                    loop:
                        MinibossMapIndicator.MapIndicator.MoveTo(Character.GetTransform(), TickTime * 3)
                else:
                    Sleep(Inf)

                loop:
                    if (not Character.IsActive[]):
                        break

                    if (not npc_state_dance[State]):
                        RetargetPlayer()

                    race:
                        loop:
                            DownedAgent := DownButNotOut.AgentDownedEvent.Await()
                            if (DownedAgent = MaybeTargetPlayer?):
                                if (Player := player[Agent], Player.GetPlayerStats(player_stat_id.Revives) > 0):
                                    Sleep(PlayerReviveDuration + 0.1)

                                UpdateTarget()
                                break

                        if (Player := MaybeTargetPlayer?, BossAttack := GetRandomBossAttack[]):
                            set AttackProperties = BossAttack.AttackProperties
                            set State = GetNPCState(npc_state_type.Chase)
                            loop:
                                set State = State.OnBegin(Self)
                                if (npc_state_attack[State]):
                                    break

                            for (Index := 1..BossAttack.CastCount):
                                GetNPCState(npc_state_type.Focus).OnBegin(Self)
                                GetNPCState(npc_state_type.Attack).OnBegin(Self)

                            GetNPCState(npc_state_type.Cooldown).OnBegin(Self)

                    Sleep(0.0)

    HandleEliminated<override>(EliminatingPlayer:player):void =
        AddGameStat(game_stat_id.EliminatedMiniboss, 1.0)

        for (Player : GetActivePlayers()):
            Player.PlayAudio(audio_id.MinibossEliminated)

        for (Player -> DamageReceived : PlayerDamageReceived):
            if:
                Player.HasBoon(boon_id.BossSiphon)?
                MaxHealth := GetAgent[].GetFortCharacter[].GetMaxHealth()
                DamagePercent := DamageReceived / MaxHealth
                DamagePercent >= 0.1
            then:
                Player.AddGameStat(player_game_stat_id.EliminatedMiniboss, 1.0)
                Player.UpdateMaxHealth()

        if (GameManager := GetGameManager[]):
            GameManager.MinibossEliminatedEvent.Signal()

    <# ----------------------------------------------------------------------------------------------------
        Event handlers
    ---------------------------------------------------------------------------------------------------- #>    
    OnDamaged<override>(DamageResult:damage_result):void =
        if (Instigator := DamageResult.Instigator?, Agent := Instigator.GetInstigatorAgent[], Player := player[Agent]):
            if (set PlayerDamageReceived[Player] = PlayerDamageReceived[Player] + DamageResult.Amount or DamageResult.Amount) {}
            Player.AddGameStat(player_game_stat_id.DamageDealtMiniboss, DamageResult.Amount)