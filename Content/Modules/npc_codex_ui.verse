using { /Fortnite.com/UI }
using { /Verse.org/Assets }
using { /Verse.org/Colors }
using { /UnrealEngine.com/Temporary/SpatialMath }
using { /UnrealEngine.com/Temporary/UI }

npc_codex_ui := class(panel_ui):
    Title<override>:string = "NPC Codex"
    BackgroundOpacity<override>:type {_X:float where 0.000000 <= _X, _X <= 1.000000} = 0.7
    Anchors<override>:anchors = anchors{Minimum := vector2{X := 0.05, Y := 0.0}, Maximum := vector2{X := 0.95, Y := 1.0}}

    NameText:text_block = text_block:
        DefaultTextColor := ThemeColors.PrimaryText
        DefaultShadowOffset := option{vector2{X := 2.0, Y := 2.0}}
        DefaultShadowOpacity := 0.5

    Icon:texture_block = texture_block:
        DefaultImage := UserInterface.Icons.WeaponBow_128
        DefaultDesiredSize := vector2{X := 128.0, Y := 128.0}

    DescriptionText:text_block = text_block:
        DefaultTextColor := ThemeColors.PrimaryText
        DefaultTextOpacity := 0.5

    DamageText:text_block = text_block:
        DefaultTextColor := ThemeColors.Success
        DefaultShadowColor := ThemeColors.Success * 0.125
        DefaultShadowOffset := option{vector2{X := 2.0, Y := 2.0}}
        DefaultShadowOpacity := 0.5

    CooldownText:text_block = text_block:
        DefaultTextColor := ThemeColors.Success
        DefaultShadowColor := ThemeColors.Success * 0.125
        DefaultShadowOffset := option{vector2{X := 2.0, Y := 2.0}}
        DefaultShadowOpacity := 0.5

    SizeText:text_block = text_block:
        DefaultTextColor := ThemeColors.Success
        DefaultShadowColor := ThemeColors.Success * 0.125
        DefaultShadowOffset := option{vector2{X := 2.0, Y := 2.0}}
        DefaultShadowOpacity := 0.5

    GridRow:int = 3
    GridColumn:int = 3

    var NPCCodexEntryUIs:[]npc_codex_entry_ui = array{}

    <# ----------------------------------------------------------------------------------------------------
        Initialization
    ---------------------------------------------------------------------------------------------------- #>
    CreateContent<override>():widget =
        if (WeaponEquippedEvent := Player.GetStats[].WeaponEquippedEvent):
            WeaponEquippedEvent.Subscribe(OnWeaponEquipped)

        stack_box:
            Orientation := orientation.Horizontal
            Slots := array:
                stack_box_slot:
                    HorizontalAlignment := horizontal_alignment.Fill
                    VerticalAlignment := vertical_alignment.Fill
                    Widget := CreateGrid()
                stack_box_slot:
                    HorizontalAlignment := horizontal_alignment.Fill
                    VerticalAlignment := vertical_alignment.Fill
                    Distribution := option{1.0}
                    Padding := margin{Left := 20.0, Top := 0.0, Right := 0.0, Bottom := 0.0}
                    Widget := CreateInfoUI()

    CreateGrid():widget =
        overlay:
            Slots := array:
                overlay_slot:
                    HorizontalAlignment := horizontal_alignment.Fill
                    VerticalAlignment := vertical_alignment.Fill
                    Widget := MakeBox(?Width := 900.0)
                overlay_slot:
                    HorizontalAlignment := horizontal_alignment.Fill
                    VerticalAlignment := vertical_alignment.Fill
                    Padding := margin{Left := 40.0, Top := 40.0, Right := 40.0, Bottom := 40.0}
                    Widget := stack_box:
                        Orientation := orientation.Vertical
                        Slots := for (RowIndex := 0..GridRow - 1):
                            stack_box_slot:
                                HorizontalAlignment := horizontal_alignment.Fill
                                VerticalAlignment := vertical_alignment.Fill
                                Distribution := option{1.0}
                                Padding := margin{Left := 0.0, Top := (RowIndex = 0) and 0.0 or 50.0, Right := 0.0, Bottom := 0.0}
                                Widget := stack_box:
                                    Orientation := orientation.Horizontal
                                    Slots := for (ColumnIndex := 0..GridColumn - 1):
                                        EntryIndex := GridRow * RowIndex + ColumnIndex
                                        NPCCodexEntryUI := npc_codex_entry_ui{Player := Player}
                                        NPCCodexEntryUI.Initialize()

                                        if (WeaponId := StartingWeapons[EntryIndex], WeaponInfo := GetWeaponInfo[WeaponId]):
                                            NPCCodexEntryUI.Set(WeaponInfo)
                                            spawn{AwaitButtonClicked(NPCCodexEntryUI.EquipButton.OnClick(), WeaponInfo)}

                                        set NPCCodexEntryUIs += array{NPCCodexEntryUI}

                                        stack_box_slot:
                                            HorizontalAlignment := horizontal_alignment.Fill
                                            VerticalAlignment := vertical_alignment.Fill
                                            Distribution := option{1.0}
                                            Padding := margin{Left := (ColumnIndex = 0) and 0.0 or 20.0, Top := 0.0, Right := 0.0, Bottom := 0.0}
                                            Widget := NPCCodexEntryUI.GetUI()

    CreateInfoUI():widget =
        overlay:
            Slots := array:
                overlay_slot:
                    HorizontalAlignment := horizontal_alignment.Fill
                    VerticalAlignment := vertical_alignment.Fill
                    Widget := MakeBox()
                overlay_slot:
                    HorizontalAlignment := horizontal_alignment.Fill
                    VerticalAlignment := vertical_alignment.Fill
                    Padding := margin{Left := 40.0, Top := 40.0, Right := 40.0, Bottom := 40.0}
                    Widget := stack_box:
                        Orientation := orientation.Vertical
                        Slots := array:
                            stack_box_slot:
                                HorizontalAlignment := horizontal_alignment.Center
                                VerticalAlignment := vertical_alignment.Top
                                Padding := margin{Left := 0.0, Top := 0.0, Right := 0.0, Bottom := 20.0}
                                Widget := NameText
                            stack_box_slot:
                                VerticalAlignment := vertical_alignment.Top
                                Widget := Icon
                            stack_box_slot:
                                HorizontalAlignment := horizontal_alignment.Fill
                                VerticalAlignment := vertical_alignment.Top
                                Padding := margin{Left := 0.0, Top := 20.0, Right := 0.0, Bottom := 40.0}
                                Widget := color_block:
                                    DefaultColor := NamedColors.White
                                    DefaultOpacity := 0.1
                                    DefaultDesiredSize := vector2{X := 0.0, Y := 1.0}
                            stack_box_slot:
                                HorizontalAlignment := horizontal_alignment.Left
                                VerticalAlignment := vertical_alignment.Top
                                Widget := DescriptionText
                            stack_box_slot:
                                HorizontalAlignment := horizontal_alignment.Fill
                                VerticalAlignment := vertical_alignment.Bottom
                                Padding := margin{Left := 0.0, Top := 20.0, Right := 0.0, Bottom := 40.0}
                                Distribution := option{1.0}
                                Widget := color_block:
                                    DefaultColor := NamedColors.White
                                    DefaultOpacity := 0.1
                                    DefaultDesiredSize := vector2{X := 0.0, Y := 1.0}
                            stack_box_slot:
                                HorizontalAlignment := horizontal_alignment.Center
                                VerticalAlignment := vertical_alignment.Top
                                Padding := margin{Left := 0.0, Top := 0.0, Right := 0.0, Bottom := 20.0}
                                Widget := text_block:
                                    DefaultText := MakeMessage("Base Stats")
                                    DefaultTextColor := ThemeColors.PrimaryText
                                    DefaultShadowOffset := option{vector2{X := 2.0, Y := 2.0}}
                                    DefaultShadowOpacity := 0.5
                            stack_box_slot:
                                HorizontalAlignment := horizontal_alignment.Fill
                                Widget := CreateStatEntry(UserInterface.Icons.ProjectileDamage_64, ThemeColors.WeaponDamage, "Damage", DamageText)
                            stack_box_slot:
                                HorizontalAlignment := horizontal_alignment.Fill
                                Padding := margin{Left := 0.0, Top := 5.0, Right := 0.0, Bottom := 0.0}
                                Widget := CreateStatEntry(UserInterface.Icons.AbilityCooldown_64, ThemeColors.WeaponCooldown, "Cooldown", CooldownText)
                            stack_box_slot:
                                HorizontalAlignment := horizontal_alignment.Fill
                                Padding := margin{Left := 0.0, Top := 5.0, Right := 0.0, Bottom := 0.0}
                                Widget := CreateStatEntry(UserInterface.Icons.AbilitySize_64, ThemeColors.WeaponSize, "Size", SizeText)

    CreateStatEntry(StatIcon:texture, StatIconColor:color, Label:string, Text:text_block):widget =
        stack_box:
            Orientation := orientation.Horizontal
            Slots := array:
                stack_box_slot:
                    HorizontalAlignment := horizontal_alignment.Fill
                    VerticalAlignment := vertical_alignment.Fill
                    Widget := texture_block:
                        DefaultImage := StatIcon
                        DefaultTint := StatIconColor
                        DefaultDesiredSize := vector2{X := 40.0, Y := 40.0}
                stack_box_slot:
                    HorizontalAlignment := horizontal_alignment.Fill
                    VerticalAlignment := vertical_alignment.Fill
                    Padding := margin{Left := 20.0, Top := 0.0, Right := 0.0, Bottom := 0.0}
                    Widget := text_block:
                        DefaultText := MakeMessage("{Label}")
                        DefaultTextColor := ThemeColors.PrimaryText
                        DefaultTextOpacity := 0.5
                stack_box_slot:
                    HorizontalAlignment := horizontal_alignment.Right
                    VerticalAlignment := vertical_alignment.Fill
                    Distribution := option{1.0}
                    Widget := Text

    <# ----------------------------------------------------------------------------------------------------
        Update Functions
    ---------------------------------------------------------------------------------------------------- #>
    Update():void =
        for (NPCCodexEntryUI : NPCCodexEntryUIs):
            NPCCodexEntryUI.Update()

    UpdateInfo(WeaponInfo:weapon_info):void =
        NameText.SetText(MakeMessage("{WeaponInfo.Name}"))
        Icon.SetImage(WeaponInfo.Icon)
        DescriptionText.SetText(MakeMessage("{WeaponInfo.Description.WrapText(600.0)}"))
        DamageText.SetText(MakeMessage("{GetRoundedString(WeaponInfo.Damage, 0)}"))
        CooldownText.SetText(MakeMessage("{GetRoundedString(WeaponInfo.Cooldown, 0)}s"))
        SizeText.SetText(MakeMessage("{GetRoundedString(WeaponInfo.Size, 0)}"))

    <# ----------------------------------------------------------------------------------------------------
        Event Handlers
    ---------------------------------------------------------------------------------------------------- #>
    OnWeaponEquipped():void =
        Update()
        Player.PlayAudio(audio_id.Activate)

    OnShow<override>():void =
        Update()
        if (EquippedWeapon := Player.GetEquippedWeapon()?, WeaponInfo := GetWeaponInfo[EquippedWeapon]):
            UpdateInfo(WeaponInfo)

    AwaitButtonClicked(Listenable:listenable(widget_message), WeaponInfo:weapon_info)<suspends>:void =
        loop:
            Listenable.Await()
            UpdateInfo(WeaponInfo)