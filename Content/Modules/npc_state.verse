<# ----------------------------------------------------------------------------------------------------
    Data Definitions
---------------------------------------------------------------------------------------------------- #>
npc_state_type := enum:
    Chase
    Attack
    Cooldown
    Focus
    Flee
    Reposition
    Dance

ToString(NPCStateType:npc_state_type)<transacts>:string =
    case (NPCStateType):
        npc_state_type.Chase => "Chase"
        npc_state_type.Attack => "Attack"
        npc_state_type.Cooldown => "Cooldown"
        npc_state_type.Focus => "Focus"
        npc_state_type.Flee => "Flee"
        npc_state_type.Reposition => "Reposition"
        npc_state_type.Dance => "Dance"

npc_state_factory := class<concrete>:
    ChaseState:npc_state_chase = npc_state_chase{}
    AttackState:npc_state_attack = npc_state_attack{}
    CooldownState:npc_state_cooldown = npc_state_cooldown{}
    FocusState:npc_state_focus = npc_state_focus{}
    FleeState:npc_state_flee = npc_state_flee{}
    RepositionState:npc_state_reposition = npc_state_reposition{}
    DanceState:npc_state_dance = npc_state_dance{}

    (npc_state_factory:)GetNPCState(Type:npc_state_type)<transacts>:npc_state =
        case (Type):
            npc_state_type.Chase => ChaseState
            npc_state_type.Attack => AttackState
            npc_state_type.Cooldown => CooldownState
            npc_state_type.Focus => FocusState
            npc_state_type.Flee => FleeState
            npc_state_type.Reposition => RepositionState
            npc_state_type.Dance => DanceState

npc_state := class<abstract>:
    OnBegin(NPCBehaviorBase:npc_behavior_base)<suspends>:npc_state
