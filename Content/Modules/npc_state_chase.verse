using { /Fortnite.com/AI }
using { /Fortnite.com/Characters }
using { /UnrealEngine.com/Temporary/SpatialMath }
using { /Verse.org/Simulation }

ToString(NavigationResult:navigation_result)<transacts>:string =
    case (NavigationResult):
        navigation_result.Reached => "Reached"
        navigation_result.PartiallyReached => "PartiallyReached"
        navigation_result.Interrupted => "Interrupted"
        navigation_result.Blocked => "Blocked"
        navigation_result.Unreachable => "Unreachable"

npc_state_chase := class(npc_state):

    OnBegin<override>(NPCBehaviorBase:npc_behavior_base)<suspends>:npc_state =
        race:
            return NPCBehaviorBase.InterruptedEvent.Await()
            if:
                Agent := NPCBehaviorBase.GetAgent[]
                Character := Agent.GetFortCharacter[]
                Navigatable := Character.GetNavigatable[]
                FocusInterface := Character.GetFocusInterface[]
            
                TargetPlayer := NPCBehaviorBase.MaybeTargetPlayer?
                TargetPlayerCharacter := TargetPlayer.GetFortCharacter[]
                TargetPlayerCharacter.IsActive[]
            then:
                NavigationTarget := MakeNavigationTarget(TargetPlayer)
                Navigatable.SetMovementSpeedMultiplier(1.0 + GetArenaMovespeedMultiplier())
                var NavigationCompleteCount:int = 0
                var IsRepositionRequired:logic = false

                race:
                    FocusInterface.MaintainFocus(TargetPlayer)

                    loop:
                        NavigationResult := Navigatable.NavigateTo(NavigationTarget, ?MovementType := movement_types.Running, ?ReachRadius := NPCBehaviorBase.AttackProperties.AttackDistance)

                        if (NavigationResult = navigation_result.Reached):
                            break
                        else:
                            set NavigationCompleteCount += 1

                            if (NavigationCompleteCount = 3):
                                Navigatable.Wait(?Duration := 0.5)
                                ReverseNavigationTarget := MakeNavigationTarget(Character.GetTransform().Translation + (Character.GetTransform().Rotation.GetLocalRight() - Character.GetTransform().Rotation.GetLocalForward()) * 150.0)
                                Navigatable.NavigateTo(ReverseNavigationTarget, ?MovementType := movement_types.Running)

                            else if (NavigationCompleteCount >= 5):
                                return GetNPCState(npc_state_type.Reposition)

                        Navigatable.Wait(?Duration := 0.5)

                    loop:
                        Sleep(3.0)                        
                        
                        set IsRepositionRequired = true
                        for (AlivePlayer : GetAlivePlayers(), AliveCharacter := AlivePlayer.GetFortCharacter[], IsRepositionRequired?):
                            DistanceSquaredFromAlivePlayer := DistanceSquared(Character.GetTransform().Translation, AliveCharacter.GetTransform().Translation)
                            if (DistanceSquaredFromAlivePlayer <= DefaultNPCRepositionDistanceSquared):
                                set IsRepositionRequired = false

                        if (IsRepositionRequired?):
                            return GetNPCState(npc_state_type.Reposition)

                    loop:
                        Sleep(3.0)
                        if (NearestAgent := GetNearestAgent(Agent, GetVulnerablePlayers())?, NearestPlayer := player[NearestAgent], NearestPlayer <> TargetPlayer):
                            NPCBehaviorBase.UpdateTarget(?MaybeNewTarget := option{NearestPlayer})
                            return Self

                if:
                    Direction := (TargetPlayerCharacter.GetTransform().Translation.Make2D() - Character.GetTransform().Translation.Make2D()).MakeUnitVector[]
                    RotationBetween := MakeShortestRotationBetween(Character.GetTransform().Rotation.GetLocalForward(), Direction)
                    Character.TeleportTo[Character.GetTransform().Translation, Character.GetTransform().Rotation.RotateBy(RotationBetween)]

                return GetNPCState(npc_state_type.Attack)

        return Self