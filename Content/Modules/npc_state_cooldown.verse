using { /Fortnite.com/AI }
using { /Fortnite.com/Characters }
using { /UnrealEngine.com/Temporary/SpatialMath }
using { /Verse.org/Simulation }

npc_state_cooldown := class(npc_state):
    var StartTime:float = 0.0

    OnBegin<override>(NPCBehaviorBase:npc_behavior_base)<suspends>:npc_state =
        if (CooldownStartTime := NPCBehaviorBase.MaybeCooldownStartTime?):
            set StartTime = CooldownStartTime
        else:
            set StartTime = GetSimulationElapsedTime()
            set NPCBehaviorBase.MaybeCooldownStartTime = option{GetSimulationElapsedTime()}

        race:
            return NPCBehaviorBase.InterruptedEvent.Await()

            if:
                Agent := NPCBehaviorBase.GetAgent[]
                Character := Agent.GetFortCharacter[]
                Navigatable := Character.GetNavigatable[]
                FocusInterface := Character.GetFocusInterface[]

                TargetPlayer := NPCBehaviorBase.MaybeTargetPlayer?
                PlayerCharacter := TargetPlayer.GetFortCharacter[]
            then:
                NavigationTarget := MakeNavigationTarget(TargetPlayer)

                race:
                    block:
                        CooldownTimeRemaining := NPCBehaviorBase.AttackProperties.AttackCooldown - (GetSimulationElapsedTime() - StartTime)
                        if (CooldownTimeRemaining > 0.0):
                            Sleep(CooldownTimeRemaining)

                    FocusInterface.MaintainFocus(TargetPlayer)

                    loop:
                        Navigatable.NavigateTo(NavigationTarget, ?MovementType := movement_types.Running, ?ReachRadius := NPCBehaviorBase.AttackProperties.AttackDistance)
                        Navigatable.Wait(?Duration := TickTime * 5)

        set NPCBehaviorBase.MaybeCooldownStartTime = false
        return GetNPCState(npc_state_type.Chase)