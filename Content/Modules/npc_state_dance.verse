using { /Fortnite.com/AI }
using { /Fortnite.com/Animation/PlayAnimation }
using { /Fortnite.com/Characters }
using { /Verse.org/Random }
using { /Verse.org/Simulation }

npc_state_dance := class(npc_state):
    OnBegin<override>(NPCBehaviorBase:npc_behavior_base)<suspends>:npc_state =
        if (Agent := NPCBehaviorBase.GetAgent[], Character := Agent.GetFortCharacter[], PlayAnimationController := Character.GetPlayAnimationController[], Navigatable := Character.GetNavigatable[]):
            Navigatable.StopNavigation()
            Sleep(GetRandomFloat(0.0, 2.0))
            loop:
                PlayAnimationController.PlayAndAwait(NPC.AnimationSequences.AS_Slime_Jump, ?PlayRate := 1.0)
                Sleep(GetRandomFloat(0.0, 0.3))
        
        return GetNPCState(npc_state_type.Chase)