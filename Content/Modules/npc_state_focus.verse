using { /Fortnite.com/AI }
using { /Fortnite.com/Characters }
using { /UnrealEngine.com/Temporary/SpatialMath }
using { /Verse.org/Simulation }

npc_state_focus := class(npc_state):
    OnBegin<override>(NPCBehaviorBase:npc_behavior_base)<suspends>:npc_state =
        race:
            return NPCBehaviorBase.InterruptedEvent.Await()
            if:
                Agent := NPCBehaviorBase.GetAgent[]
                Character := Agent.GetFortCharacter[]
                FocusInterface := Character.GetFocusInterface[]
            
                TargetPlayer := NPCBehaviorBase.MaybeTargetPlayer?
                TargetPlayerCharacter := TargetPlayer.GetFortCharacter[]
                TargetPlayerCharacter.IsActive[]
            then:
                race:
                    FocusInterface.MaintainFocus(TargetPlayer)
                    loop:
                        if (Direction := (TargetPlayerCharacter.GetTransform().Translation.Make2D() - Character.GetTransform().Translation.Make2D()).MakeUnitVector[]):
                            Angle := RadiansToDegrees(MakeShortestRotationBetween(Character.GetTransform().Rotation.GetLocalForward(), Direction).GetAngle())
                            if (Angle < 2.5):
                                if:
                                    RotationBetween := MakeShortestRotationBetween(Character.GetTransform().Rotation.GetLocalForward(), Direction)
                                    Character.TeleportTo[Character.GetTransform().Translation, Character.GetTransform().Rotation.RotateBy(RotationBetween)]

                                break

                        Sleep(0.0)

                return GetNPCState(npc_state_type.Attack)

        return Self