using { /Fortnite.com/Animation/PlayAnimation }
using { /Fortnite.com/Characters }

npc_state_hit := class(npc_state):
    PreviousState:npc_state

    OnBegin<override>(NPCBehaviorBase:npc_behavior_base)<suspends>:npc_state =
        if (PlayAnimationController := NPCBehaviorBase.GetAgent[].GetFortCharacter[].GetPlayAnimationController[]):
            PlayAnimationController.PlayAndAwait(NPC.AnimationSequences.AS_Slime_Hit)

        return PreviousState