using { /Fortnite.com/AI }
using { /Fortnite.com/Characters }
using { /UnrealEngine.com/Temporary/SpatialMath }
using { /Verse.org/Simulation }

npc_state_knockback := class(npc_state):
    Duration:float
    Rotation:rotation
    PreviousState:npc_state

    OnBegin<override>(NPCBehaviorBase:npc_behavior_base)<suspends>:npc_state =
        if (Agent := NPCBehaviorBase.GetAgent[], Character := Agent.GetFortCharacter[], Navigatable := Character.GetNavigatable[]):
            Navigatable.StopNavigation()
            Agent.ApplyImpulse(impulse_id.Knockback, ?Rotation := Rotation)
            Navigatable.Wait(?Duration := Duration)
        
        return PreviousState