using { /Fortnite.com/AI }
using { /Fortnite.com/Characters }
using { /Verse.org/Simulation }
using { /UnrealEngine.com/Temporary/SpatialMath }

npc_state_reposition := class(npc_state):
    
    OnBegin<override>(NPCBehaviorBase:npc_behavior_base)<suspends>:npc_state =
        race:
            return NPCBehaviorBase.InterruptedEvent.Await()
            if:
                Agent := NPCBehaviorBase.GetAgent[]
                Character := Agent.GetFortCharacter[]
                Navigatable := Character.GetNavigatable[]

                TargetPlayer := NPCBehaviorBase.MaybeTargetPlayer?
                TargetPlayerCharacter := TargetPlayer.GetFortCharacter[]
                TargetPlayerCharacter.IsActive[]
            then:
                Navigatable.StopNavigation()

                if (CurrentDestination := Navigatable.GetCurrentDestination[], Character.TeleportTo[Character.GetTransform().Translation, IdentityRotation()]):
                    Navigatable.Wait(?Duration := TickTime)
                    Navigatable.NavigateTo(MakeNavigationTarget(Character.GetTransform().Translation), ?MovementType := movement_types.Running, ?ReachRadius := DefaultArenaSize * 2, ?AllowPartialPath := true)

                var TeleportAttempt:int = 0
                loop:
                    if (TeleportAttempt >= 3):
                        break

                    if:
                        SpawnPosition := Agent.GetRepositionLocation(TargetPlayer).Make2D()
                        SpawnRotation := MakeRotation(SpawnPosition, TargetPlayerCharacter.GetTransform().Translation.Make2D())
                        Character.TeleportTo[SpawnPosition + vector3{Z := NPCSpawnPositionHeightOffset}, SpawnRotation]
                    then:
                        break
                    
                    set TeleportAttempt += 1
                    Sleep(0.0)

        return GetNPCState(npc_state_type.Chase)