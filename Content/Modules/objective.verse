using { /Fortnite.com/Characters }
using { /Fortnite.com/Devices }
using { /Verse.org/Simulation }
using { /UnrealEngine.com/Temporary/SpatialMath }

<# ----------------------------------------------------------------------------------------------------
    Constants
---------------------------------------------------------------------------------------------------- #>
ObjectiveTimerWarning:float = 10.0 # Requires updating the widget animation

<# ----------------------------------------------------------------------------------------------------
    Data Definitions
---------------------------------------------------------------------------------------------------- #>
tracker_info := class<concrete>:
    @editable
    Title:string = ""
    @editable
    Description:string = ""
    @editable
    Device:tracker_device = tracker_device{}

objective_id := enum:
    Collectible
    GoldenSlime
    BeachBall
    CapturePoint
    ColorChangingTiles
    Storm

ToString(Id:objective_id)<transacts>:string =
    case (Id):
        objective_id.Collectible => "Collectibles"
        objective_id.GoldenSlime => "Golden Slime"
        objective_id.BeachBall => "Beach Ball"
        objective_id.CapturePoint => "Capture Point"
        objective_id.ColorChangingTiles => "Numbered Tiles"
        objective_id.Storm => "Survive The Storm"

<# ----------------------------------------------------------------------------------------------------
    objective
---------------------------------------------------------------------------------------------------- #>
objective := class<unique><concrete>:
    Id:objective_id = objective_id.BeachBall

    @editable
    HUDMessage:hud_message_device = hud_message_device{}
    
    @editable
    CinematicSequence:cinematic_sequence_device = cinematic_sequence_device{}

    @editable
    TrackerInfo:tracker_info = tracker_info{}

    @editable
    TrackerTimer:tracker_device = tracker_device{}

    @editable
    TimeLimit:float = 60.0

    StartEvent:events(tuple()) = events(tuple()){}
    StopEvent:events(tuple()) = events(tuple()){}
    CompleteEvent:events(tuple()) = events(tuple()){}
    ChallengeAccepted:events(tuple()) = events(tuple()){}

    var HUDMessageIntro:hud_message_device = hud_message_device{}
    var HUDMessageTimer:hud_message_device = hud_message_device{}

    <# ----------------------------------------------------------------------------------------------------
        Initialization
    ---------------------------------------------------------------------------------------------------- #>
    OnBegin()<suspends>:void = {}

    <# ----------------------------------------------------------------------------------------------------
        Get Functions
    ---------------------------------------------------------------------------------------------------- #>
    GetTimeLimit()<transacts>:float =
        TimeLimit / GetGameSpeed()

    <# ----------------------------------------------------------------------------------------------------
        Update Functions
    ---------------------------------------------------------------------------------------------------- #>
    Start()<suspends>:void =
        defer:
            #TrackerInfo.Device.RemoveFromAll()
            #TrackerTimer.RemoveFromAll()
            HUDMessage.Hide()
            HUDMessageTimer.Hide()
            OnStop()

        #TrackerInfo.Device.AssignToAll()
        #TrackerInfo.Device.SetValue(0)
        #TrackerInfo.Device.SetTitleText(MakeMessage("Mission - {TrackerInfo.Title}"))
        #TrackerInfo.Device.SetDescriptionText(MakeMessage("{TrackerInfo.Description}"))

        CinematicSequence.Play()
        CinematicSequence.Pause()
        Sleep(0.0)
        CinematicSequence.SetPlaybackTime(0.0)

        #TrackerTimer.AssignToAll()
        #TrackerTimer.SetTarget(Floor[GetTimeLimit()] or 0)
        #TrackerTimer.SetValue(Floor[GetTimeLimit()] or 0)

        HUDMessageTimer.Show(MakeMessage("{Floor[GetTimeLimit()] or 0}"), ?DisplayTime := 0.0)
        HUDMessageIntro.Show(MakeMessage("{Id}"), ?DisplayTime := 1.8)

        for (Index -> Player : GetActivePlayers()):
            Player.PlayAudio(audio_id.Notification)

        StartEvent.Signal()

        race:
            StopEvent.Await()

            block:
                race:
                    ChallengeAccepted.Await()
                    block:
                        Sleep(GetTimeLimit() - ObjectiveTimerWarning)
                        for (Player : GetActivePlayers()):
                            Player.PlayAudio(audio_id.TimerWarning)
                Sleep(Inf)

            block:
                StartTime := GetSimulationElapsedTime()

                race:
                    block:
                        ChallengeAccepted.Await()
                        HUDMessageTimer.Hide()
                        #TrackerTimer.RemoveFromAll()
                        
                    loop:
                        Sleep(1.0)
                        ElapsedTime := GetSimulationElapsedTime() - StartTime
                        TimeRemaining := GetTimeLimit() - ElapsedTime
                        HUDMessageTimer.SetText(MakeMessage("{Floor[GetTimeLimit() - ElapsedTime] or 0}"))
                        #TrackerTimer.SetValue(Floor[GetTimeLimit() - ElapsedTime] or 0)

                        if (ElapsedTime >= GetTimeLimit()):
                            for (Player : GetActivePlayers()):
                                Player.PlayAudio(audio_id.ObjectiveFailure)

                                if (GameProgressBarUI := GetGameManager[].PlayerGameProgressBarUI[Player]):
                                    GameProgressBarUI.UpdateEventState(GameEventMissionIndex, game_event_state.Failed)

                            Stop()
                            break

                Sleep(Inf)

            block:
                CompleteEvent.Await()
                spawn{GrantRewards()}

    Stop():void =
        StopEvent.Signal()

    GrantRewards()<suspends>:void =
        AddObjectivesComplete()
    
        for (Player : GetActivePlayers()):
            Player.PlayAudio(audio_id.ObjectiveSuccess)
            Player.AwardAccolade(accolade_id.CompleteObjective)
            Player.ShowMessage(message_id.MissionComplete)

            if (GameProgressBarUI := GetGameManager[].PlayerGameProgressBarUI[Player]):
                GameProgressBarUI.UpdateEventState(GameEventMissionIndex, game_event_state.Success)

        Sleep(2.0)
        PauseGame(pause_type.Boon)

        <#
        AddGameCurrency(game_currency_id.TreasureChest, 1.0)
        for (Player : GetActivePlayers(), Character := Player.GetFortCharacter[]):
            Player.PlayAudio(audio_id.Purchase)
            Player.ActivateVFXPowerup(vfx_powerup_id.TreasureChestCollect, ?Duration := 1.0)
        #>

    <# ----------------------------------------------------------------------------------------------------
        Event Handlers
    ---------------------------------------------------------------------------------------------------- #>
    OnStop():void = {}