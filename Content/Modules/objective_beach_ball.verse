using { /Fortnite.com/Devices }
using { /Verse.org/Random }
using { /Verse.org/Simulation }
using { /UnrealEngine.com/Temporary/SpatialMath }

<# ----------------------------------------------------------------------------------------------------
    objective_beach_ball
---------------------------------------------------------------------------------------------------- #>
objective_beach_ball := class<concrete>(objective):
    Id<override>:objective_id = objective_id.BeachBall

    @editable
    CaptureArea:capture_area_device = capture_area_device{}

    @editable
    MapIndicator:map_indicator_device = map_indicator_device{}

    @editable
    ChannelEnable:channel_device = channel_device{}

    @editable
    ChannelDisable:channel_device = channel_device{}

    @editable
    BallSpawners:[]creative_prop = array{}

    var Score:int = 0

    <# ----------------------------------------------------------------------------------------------------
        Initialization
    ---------------------------------------------------------------------------------------------------- #>
    OnBegin<override>()<suspends>:void =
        StartEvent.Subscribe(OnStart)
        CaptureArea.Disable()
        if (MapIndicator.TeleportTo[MapIndicatorDisabledPosition, IdentityRotation()]) {}
        ChannelDisable.Transmit(false)
        CaptureArea.AreaIsScoredEvent.Subscribe(OnAreaIsScored)
    
    <# ----------------------------------------------------------------------------------------------------
        Get Functions
    ---------------------------------------------------------------------------------------------------- #>
    GetTarget()<transacts>:int =
        Max(1, GetActivePlayers().Length)

    <# ----------------------------------------------------------------------------------------------------
        Update Functions
    ---------------------------------------------------------------------------------------------------- #>
    RepositionBallSpawners()<suspends>:void =
        ArenaSize := GetActiveArenaSize()

        for (Index -> BallSpawner : BallSpawners):
            var Attempts:int = 1
            loop:
                Position := GetRandomPositionOnCircle(GetRandomFloat(ArenaSize * 0.5, ArenaSize * 0.9)) + vector3{Z := 1000.0}
                RaycastPosition := GetRaycast(Position)
                if (Attempts >= 30):
                    break
                
                if:
                    RaycastPosition.Z <= 10.0
                    BallSpawner.TeleportTo[Position, IdentityRotation()]
                then:
                    break

                set Attempts += 1
        
        ChannelEnable.Transmit(false)

    <# ----------------------------------------------------------------------------------------------------
        Event Handlers
    ---------------------------------------------------------------------------------------------------- #>
    OnStart():void =
        set Score = 0
        CaptureArea.Enable()
        CaptureArea.ActivateObjectivePulse()
        if (MapIndicator.TeleportTo[MapIndicatorEnabledPosition, IdentityRotation()]) {}
        spawn{RepositionBallSpawners()}
        #TrackerInfo.Device.SetTarget(GetTarget())
        HUDMessage.Show(MakeMessage("{Score}/{GetTarget()}"), ?DisplayTime := 0.0)

    OnStop<override>():void =
        CaptureArea.Disable()
        if (MapIndicator.TeleportTo[MapIndicatorDisabledPosition, IdentityRotation()]) {}
        CaptureArea.DeactivateObjectivePulse()
        ChannelDisable.Transmit(false)

    OnAreaIsScored(Agent:agent):void =
        set Score += 1
        #RepositionBallSpawners()
        #TrackerInfo.Device.SetValue(Score)
        HUDMessage.SetText(MakeMessage("{Score}/{GetTarget()}"))
        CinematicSequence.SetPlaybackTime((Score * 1.0) / (GetTarget() * 1.0))

        if (Score >= GetTarget()):
            CompleteEvent.Signal()