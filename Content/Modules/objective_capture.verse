using { /Fortnite.com/Characters }
using { /Fortnite.com/Devices }
using { /Verse.org/Simulation }
using { /UnrealEngine.com/Temporary/SpatialMath }

<# ----------------------------------------------------------------------------------------------------
    objective_capture
---------------------------------------------------------------------------------------------------- #>
objective_capture := class<concrete>(objective):
    Id<override>:objective_id = objective_id.CapturePoint

    @editable
    VolumeDevice:volume_device = volume_device{}
    
    @editable
    MapIndicator:map_indicator_device = map_indicator_device{}

    @editable
    CaptureArea:capture_area_device = capture_area_device{}

    @editable
    CaptureTime:float = 30.0

    @editable
    DecayRate:float = 0.25

    @editable
    Size:float = 52.0

    CancelObjectiveEvent:events(tuple()) = events(tuple()){}

    var MaybeSpawnedVFX:?creative_prop = false
    var PlayersInCaptureArea:[]player = array{}
    var Progress:float = MinScale

    <# ----------------------------------------------------------------------------------------------------
        Initialization
    ---------------------------------------------------------------------------------------------------- #>
    OnBegin<override>()<suspends>:void =
        StartEvent.Subscribe(OnStart)
        CaptureArea.Disable()
        if (MapIndicator.TeleportTo[MapIndicatorDisabledPosition, IdentityRotation()]) {}

        VolumeDevice.AgentEntersEvent.Subscribe(OnAgentEnters)
        VolumeDevice.AgentExitsEvent.Subscribe(OnAgentExits)
        if (VolumeDevice.TeleportTo[vector3{Z := -1000.0}, IdentityRotation()]) {}

    <# ----------------------------------------------------------------------------------------------------
        Get Functions
    ---------------------------------------------------------------------------------------------------- #>
    GetCaptureTime()<transacts>:float =
        CaptureTime / GetGameSpeed()
        
    GetCaptureTimeRemaining()<transacts>:float =
        GetCaptureTime() * (1.0 - Progress)

    GetDecayTime()<transacts>:float =
        (Floor[GetCaptureTime() / DecayRate * Progress] or 0) * 1.0

    GetIsCapturing()<transacts>:logic =
        var IsCapturing:logic = false
        for:
            Player : PlayersInCaptureArea
            not IsCapturing?
            not Player.GetFortCharacter[].IsDownButNotOut[]
        do:
            set IsCapturing = true

        IsCapturing

    <# ----------------------------------------------------------------------------------------------------
        Update Functions
    ---------------------------------------------------------------------------------------------------- #>
    StartObjective()<suspends>:void =
        if (MaybeSpawnedVFX?):
            return

        if (VolumeDevice.TeleportTo[DefaultArenaOrigin, IdentityRotation()]) {}
        set Progress = MinScale
        set MaybeSpawnedVFX = SpawnVFX(vfx_id.RingProgress, VolumeDevice.GetTransform().Translation, ?Scale := vector3{X := Size, Y := Size, Z := MinScale}, ?Duration := 0.0)
        #TrackerInfo.Device.SetTarget(Floor[CaptureTime] or 0)
        HUDMessage.Show(MakeMessage("{0}/{Floor[GetCaptureTime()] or 0}"), ?DisplayTime := 0.0)

        CaptureArea.Enable()
        CaptureArea.ActivateObjectivePulse()
        if (MapIndicator.TeleportTo[MapIndicatorEnabledPosition, IdentityRotation()]) {}

        if (SpawnedVFX := MaybeSpawnedVFX?):
            race:
                CancelObjectiveEvent.Await()
                
                loop:
                    UpdatedValue := Round[SpawnedVFX.GetTransform().Scale.Z * GetCaptureTime()] or 0
                    #TrackerInfo.Device.SetValue(UpdatedValue)
                    HUDMessage.SetText(MakeMessage("{UpdatedValue}/{Floor[GetCaptureTime()] or 0}"))
                    CinematicSequence.SetPlaybackTime((UpdatedValue * 1.0) / GetCaptureTime())
                    Sleep(1.0)

                loop:
                    IsCapturing := GetIsCapturing()

                    TargetTransform := transform:
                        Translation := VolumeDevice.GetTransform().Translation
                        Rotation := VolumeDevice.GetTransform().Rotation
                        Scale := vector3{X := Size, Y := Size, Z := IsCapturing? and 1.0 or MinScale}
                        
                    OverTime := IsCapturing? and GetCaptureTimeRemaining() or GetDecayTime()
                    
                    MoveToResult := race:
                        SpawnedVFX.MoveTo(TargetTransform, Max(MinScale, OverTime))

                        if (AgentDownedEvent := GetGameManager[].DownButNotOut.AgentDownedEvent):
                            AgentDownedEvent.Await()
                            move_to_result.WillNotReachDestination

                        block:
                            if (IsCapturing?):
                                VolumeDevice.AgentExitsEvent.Await()
                            else:
                                VolumeDevice.AgentEntersEvent.Await()

                            move_to_result.WillNotReachDestination
                        
                    if (IsCapturing?, MoveToResult = move_to_result.DestinationReached):
                        CompleteEvent.Signal()
                        break
                    else:
                        set Progress = SpawnedVFX.GetTransform().Scale.Z

                    Sleep(0.0)

    <# ----------------------------------------------------------------------------------------------------
        Event Handlers
    ---------------------------------------------------------------------------------------------------- #>
    OnStart():void =
        spawn{StartObjective()}
        
    OnStop<override>():void =
        CancelObjectiveEvent.Signal()
        CaptureArea.Disable()
        CaptureArea.DeactivateObjectivePulse()
        if (MapIndicator.TeleportTo[MapIndicatorDisabledPosition, IdentityRotation()]) {}
        if (VolumeDevice.TeleportTo[vector3{Z := -1000.0}, IdentityRotation()]) {}

        if (SpawnedVFX := MaybeSpawnedVFX?):
            SpawnedVFX.Dispose()
            set MaybeSpawnedVFX = false
        
    OnAgentEnters(Agent:agent):void =
        if (Player := player[Agent]):
            set PlayersInCaptureArea += array{Player}
            Player.PlayAudio(audio_id.Capture)

    OnAgentExits(Agent:agent):void =
        if (Player := player[Agent]):
            set PlayersInCaptureArea = PlayersInCaptureArea.RemoveAllElements(Player)
            Player.StopAudio(audio_id.Capture)

    ControlChange(Agent:agent):void =
        CompleteEvent.Signal()