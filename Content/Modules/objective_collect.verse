using { /Fortnite.com/Devices }
using { /Verse.org/Random }
using { /Verse.org/Simulation }
using { /UnrealEngine.com/Temporary/SpatialMath }
using { Utility }

<# ----------------------------------------------------------------------------------------------------
    objective_collect
---------------------------------------------------------------------------------------------------- #>
objective_collect := class<concrete>(objective):
    Id<override>:objective_id = objective_id.Collectible

    @editable
    CollectibleObjects:[]collectible_object_device = array{}
    
    var Progress:int = 0

    <# ----------------------------------------------------------------------------------------------------
        Initialization
    ---------------------------------------------------------------------------------------------------- #>
    OnBegin<override>()<suspends>:void =
        StartEvent.Subscribe(OnStart)

        for (CollectibleObject : CollectibleObjects):
            CollectibleObject.CollectedEvent.Subscribe(OnCollected)

    <# ----------------------------------------------------------------------------------------------------
        Get Functions
    ---------------------------------------------------------------------------------------------------- #>
    GetTarget()<transacts>:int =
        PlayerCount := GetActivePlayers().Length
        TargetCount := 6 + (PlayerCount - 1) * 3
        Floor[(TargetCount * 1.0) / GetGameSpeed()] or 6

    <# ----------------------------------------------------------------------------------------------------
        Update Functions
    ---------------------------------------------------------------------------------------------------- #>
    PositionCollectibles()<suspends>:void =
        MinDistanceBetween := 1400.0 - 200.0 * GetActivePlayers().Length
        RandomPositions := Shuffle(PoissonDiscSamplingGeneratePoints(MinDistanceBetween, GetActiveArenaSize() * 0.9, ?MinRadius := 0.0))
    
        var CollectibleIndex:int = 0
        for (RandomPosition : RandomPositions, CollectibleIndex < GetTarget() * 2 - 1):
            RaycastPosition := GetRaycast(RandomPosition)
            if (RaycastPosition.Z <= 25.0, CollectibleObject := CollectibleObjects[CollectibleIndex]):
                CollectibleObject.RespawnForAll()
                CollectibleObject.Show()
                if (CollectibleObject.TeleportTo[vector3{X := RandomPosition.X, Y := RandomPosition.Y, Z := 75.0}, IdentityRotation()]) {}
                set CollectibleIndex += 1

    <# ----------------------------------------------------------------------------------------------------
        Event Handlers
    ---------------------------------------------------------------------------------------------------- #>
    OnStart():void =
        set Progress = 0
        #TrackerInfo.Device.SetTarget(GetTarget())
        HUDMessage.Show(MakeMessage("{Progress}/{GetTarget()}"), ?DisplayTime := 0.0)
        spawn{PositionCollectibles()}

    OnStop<override>():void =
        for (CollectibleObject : CollectibleObjects):
            CollectibleObject.Hide()

    OnCollected(Agent:agent):void =
        set Progress += 1
        #TrackerInfo.Device.SetValue(Progress)
        HUDMessage.SetText(MakeMessage("{Progress}/{GetTarget()}"))
        CinematicSequence.SetPlaybackTime((Progress * 1.0) / (GetTarget() * 1.0))
        
        if (Progress >= GetTarget()):
            CompleteEvent.Signal()