using { /Fortnite.com/Devices }
using { /Verse.org/Random }
using { /Verse.org/Simulation }
using { /UnrealEngine.com/Temporary/SpatialMath }

<# ----------------------------------------------------------------------------------------------------
    objective_color_tiles
---------------------------------------------------------------------------------------------------- #>
objective_color_tiles := class<concrete>(objective):
    Id<override>:objective_id = objective_id.ColorChangingTiles

    @editable
    MapIndicator:map_indicator_device = map_indicator_device{}

    @editable
    ColorChangingTiles:[]color_changing_tiles_device = array{}

    var Sequence:int = 0
    var TileTransforms:[]transform = array{}

    <# ----------------------------------------------------------------------------------------------------
        Initialization
    ---------------------------------------------------------------------------------------------------- #>
    OnBegin<override>()<suspends>:void =
        StartEvent.Subscribe(OnStart)
        if (MapIndicator.TeleportTo[MapIndicatorDisabledPosition, IdentityRotation()]) {}
        
        set TileTransforms = for (Index -> ColorChangingTile : ColorChangingTiles):
            ColorChangingTile.Disable()
            ColorChangingTile.Hide()
            spawn{AwaitActivated(ColorChangingTile, Index)}
            ColorChangingTile.GetTransform()

    <# ----------------------------------------------------------------------------------------------------
        Update Functions
    ---------------------------------------------------------------------------------------------------- #>
    RandomizeTilePositions():void =
        RandomTransforms := Shuffle(TileTransforms)
        for (Index -> ColorChangingTile : ColorChangingTiles, Transform := RandomTransforms[Index]):
            if (ColorChangingTile.TeleportTo[Transform]) {}

    <# ----------------------------------------------------------------------------------------------------
        Event Handlers
    ---------------------------------------------------------------------------------------------------- #>
    OnStart():void =
        set Sequence = 0
        if (MapIndicator.TeleportTo[MapIndicatorEnabledPosition, IdentityRotation()]) {}

        for (Index -> Player : GetActivePlayers()):
            MapIndicator.ActivateObjectivePulse(Player)
            MapIndicator.ActivateObjectivePulse(Player)

        #TrackerInfo.Device.SetTarget(ColorChangingTiles.Length)
        HUDMessage.Show(MakeMessage("0/{ColorChangingTiles.Length}"), ?DisplayTime := 0.0)

        RandomizeTilePositions()
        for (ColorChangingTile : ColorChangingTiles):
            ColorChangingTile.Enable()
            ColorChangingTile.Reset()
            ColorChangingTile.Show()

    OnStop<override>():void =
        if (MapIndicator.TeleportTo[MapIndicatorDisabledPosition, IdentityRotation()]) {}
        
        for (Player : GetAllPlayers()):
            MapIndicator.DeactivateObjectivePulse(Player)
            MapIndicator.DeactivateObjectivePulse(Player)
        
        for (ColorChangingTile : ColorChangingTiles):
            ColorChangingTile.Disable()
            ColorChangingTile.Hide()

    AwaitActivated(InColorChangingTile:color_changing_tiles_device, Index:int)<suspends>:void =
        loop:
            InColorChangingTile.ActivatedEvent.Await()

            if (Index = Sequence):
                set Sequence += 1
                #TrackerInfo.Device.SetValue(Sequence)
                HUDMessage.SetText(MakeMessage("{Sequence}/{ColorChangingTiles.Length}"))
                CinematicSequence.SetPlaybackTime((Sequence * 1.0) / (ColorChangingTiles.Length * 1.0))

                if (Sequence = ColorChangingTiles.Length):
                    CompleteEvent.Signal()
                else:
                    for (Player : GetAllPlayers()):
                        Player.PlayAudio(audio_id.Activate)
            else:
                for (Player : GetAllPlayers()):
                    Player.PlayAudio(audio_id.Error)

                for (ColorChangingTile : ColorChangingTiles):
                    ColorChangingTile.Reset()

                HUDMessage.SetText(MakeMessage("0/{ColorChangingTiles.Length}"))
                CinematicSequence.SetPlaybackTime(0.0)
                set Sequence = 0