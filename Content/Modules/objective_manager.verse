using { /Fortnite.com/Devices }
using { /Verse.org/Random }
using { /Verse.org/Simulation }

<# ----------------------------------------------------------------------------------------------------
    Global Functions
---------------------------------------------------------------------------------------------------- #>
var GlobalObjectiveManager:weak_map(session, objective_manager) = map{}

GetObjectiveManager()<decides><transacts>:objective_manager = 
    GlobalObjectiveManager[GetSession()]

StartRandomObjective():void =
    if (ObjectiveManager := GetObjectiveManager[]):
        ObjectiveManager.StartObjective()
        
StartObjective(Id:objective_id):void =
    if (ObjectiveManager := GetObjectiveManager[]):
        ObjectiveManager.Start(Id)

StopObjective():void =
    if (ObjectiveManager := GetObjectiveManager[]):
        ObjectiveManager.Stop()

<# ----------------------------------------------------------------------------------------------------
    objective_manager
---------------------------------------------------------------------------------------------------- #>
objective_manager := class(creative_device):
    @editable
    HUDMessageIntro:hud_message_device = hud_message_device{}
    
    @editable
    HUDMessageTimer:hud_message_device = hud_message_device{}

    @editable
    ObjectiveBeachBall:objective_beach_ball = objective_beach_ball{}

    @editable
    ObjectiveCollect:objective_collect = objective_collect{}

    @editable
    ObjectiveGoldenSlime:objective_golden_slime = objective_golden_slime{}

    @editable
    ObjectiveCapture:objective_capture = objective_capture{}

    @editable
    ObjectiveColorChangingTiles:objective_color_tiles = objective_color_tiles{}
    
    @editable
    ObjectiveStorm:objective_storm = objective_storm{}

    ObjectiveCompleteEvent:events(tuple()) = events(tuple()){}
    StopEvent:event(void) = event(void){}

    var Objectives:[]objective = array{}
    var GameObjectives:[]objective = array{}
    var MaybeCurrentObjective:?objective = false
    var IsActive:logic = false

    <# ----------------------------------------------------------------------------------------------------
        Initialization
    ---------------------------------------------------------------------------------------------------- #>
    OnBegin<override>()<suspends>:void =
        if (set GlobalObjectiveManager[GetSession()] = Self) {}

        set Objectives = array:
            ObjectiveBeachBall
            ObjectiveCollect
            ObjectiveGoldenSlime
            ObjectiveCapture
            ObjectiveColorChangingTiles
            ObjectiveStorm

        for (Objective : Objectives):
            Objective.OnBegin()
            Objective.CompleteEvent.Subscribe(OnComplete)
            set Objective.HUDMessageIntro = HUDMessageIntro
            set Objective.HUDMessageTimer = HUDMessageTimer

        WaitForGameManager()
        if (GameManager := GetGameManager[]):
            GameManager.GameEndedEvent.Subscribe(Stop)
            GameManager.StageEndedEvent.Subscribe(Stop)
            GameManager.BossFightStartedEvent.Subscribe(Stop)

    <# ----------------------------------------------------------------------------------------------------
        Get Functions
    ---------------------------------------------------------------------------------------------------- #>
    GetRandomObjective()<transacts>:objective =
        var RandomObjective:objective = ObjectiveCollect        
        var IsFound:logic = false

        for (ShuffleIndex -> ShuffledObjective : Shuffle(Objectives), not IsFound?):
            var IsEncountered:logic = false
            
            for (GameObjectiveIndex -> GameObjective : GameObjectives, not IsEncountered?, GameObjective = ShuffledObjective):
                set IsEncountered = true

            if (not IsEncountered?):
                set RandomObjective = ShuffledObjective
                set IsFound = true

        if (not IsFound?, RandomIndex := GetRandomInt(0, Objectives.Length - 1), Objective := Objectives[RandomIndex]):
            set RandomObjective = Objective

        RandomObjective

    <# ----------------------------------------------------------------------------------------------------
        Update Functions
    ---------------------------------------------------------------------------------------------------- #>
    <#
    Start()<suspends>:void =
        if (IsActive?):
            Print("Objective Manager is already active!", ?Color := ThemeColors.Failure)
            return

        set IsActive = true
        race:
            StopEvent.Await()
            block:
                Sleep(90.0)
                (objective_manager:)StartObjective()

                loop:
                    Sleep(180.0)

                    if (GetGameTimeRemaining() < 60.0):
                        break

                    (objective_manager:)StartObjective()

        set IsActive = false
    #>

    StartObjective():void =
        RandomObjective := GetRandomObjective()
        spawn{RandomObjective.Start()}
        set MaybeCurrentObjective = option{RandomObjective}
        set GameObjectives += array{RandomObjective}

    Start(Id:objective_id):void =
        if (CurrentObjective := MaybeCurrentObjective?):
            Stop()

        Objective := case (Id):
            objective_id.Collectible => ObjectiveCollect
            objective_id.GoldenSlime => ObjectiveGoldenSlime
            objective_id.BeachBall => ObjectiveBeachBall
            objective_id.CapturePoint => ObjectiveCapture
            objective_id.ColorChangingTiles => ObjectiveColorChangingTiles
            objective_id.Storm => ObjectiveStorm

        spawn{Objective.Start()}
        set MaybeCurrentObjective = option{Objective}

    Stop():void =
        StopEvent.Signal()

        if (CurrentObjective := MaybeCurrentObjective?):
            CurrentObjective.Stop()
            set MaybeCurrentObjective = false

    <# ----------------------------------------------------------------------------------------------------
        Event Handlers
    ---------------------------------------------------------------------------------------------------- #>
    <#
    OnGameStarted():void =
        spawn{Start()}
        set GameObjectives = array{}
    #>

    OnComplete():void =
        ObjectiveCompleteEvent.Signal()