using { /Fortnite.com/Devices }
using { /Verse.org/Simulation }
using { /UnrealEngine.com/Temporary/SpatialMath }

<# ----------------------------------------------------------------------------------------------------
    objective_storm
---------------------------------------------------------------------------------------------------- #>
objective_storm := class<concrete>(objective):
    Id<override>:objective_id = objective_id.Storm

    @editable
    Character:character_device = character_device{}

    @editable
    MapIndicator:map_indicator_device = map_indicator_device{}

    @editable
    Beacon:beacon_device = beacon_device{}

    @editable
    AdvancedStormController:advanced_storm_controller_device = advanced_storm_controller_device{}
    
    @editable
    AdvancedStormBeacons:[]advanced_storm_beacon_device = array{}
    
    @editable
    VolumeActivator:volume_activator = volume_activator{}

    var TotalPhases:int = 5
    var CurrentPhase:int = 1
    var StartTime:float = 0.0

    <# ----------------------------------------------------------------------------------------------------
        Initialization
    ---------------------------------------------------------------------------------------------------- #>
    OnBegin<override>()<suspends>:void =
        set TotalPhases = AdvancedStormBeacons.Length
        Character.Hide()
        if (MapIndicator.TeleportTo[MapIndicatorDisabledPosition, IdentityRotation()]) {}
        Beacon.Disable()
        StartEvent.Subscribe(OnStart)
        AdvancedStormController.DestroyStorm()
        AdvancedStormController.PhaseEndedEvent.Subscribe(OnPhaseEnded)
        VolumeActivator.OnBegin()
        VolumeActivator.ActivatedEvent.Subscribe(OnActivated)

    <# ----------------------------------------------------------------------------------------------------
        Get Functions
    ---------------------------------------------------------------------------------------------------- #>
    GetTotalPhases()<transacts>:int =
        TotalPhases - Floor[GetGameSpeed() - 1.0] or 0

    <# ----------------------------------------------------------------------------------------------------
        Event Handlers
    ---------------------------------------------------------------------------------------------------- #>
    OnStart():void =
        set StartTime = GetSecondsSinceEpoch()
        set CurrentPhase = 1
        #TrackerInfo.Device.SetTarget(GetTotalPhases())
        HUDMessage.Show(MakeMessage("0/{GetTotalPhases()}"), ?DisplayTime := 0.0)
        VolumeActivator.Enable()
        Beacon.Enable()
        if (MapIndicator.TeleportTo[MapIndicatorEnabledPosition, IdentityRotation()]) {}

        for (Index -> Player : GetActivePlayers()):
            MapIndicator.ActivateObjectivePulse(Player)
            MapIndicator.ActivateObjectivePulse(Player)

        Character.Show()

    OnStop<override>():void =
        for (Player : GetAllPlayers()):
            MapIndicator.DeactivateObjectivePulse(Player)
            MapIndicator.DeactivateObjectivePulse(Player)

        if (MapIndicator.TeleportTo[MapIndicatorDisabledPosition, IdentityRotation()]) {}

        Beacon.Disable()
        VolumeActivator.Disable()
        Character.Hide()
        AdvancedStormController.DestroyStorm()
        
    OnActivated(InPlayer:player):void =
        for (Player : GetAllPlayers()):
            MapIndicator.DeactivateObjectivePulse(Player)
            MapIndicator.DeactivateObjectivePulse(Player)

        if (MapIndicator.TeleportTo[MapIndicatorDisabledPosition, IdentityRotation()]) {}
        Beacon.Disable()
        #TrackerInfo.Device.SetValue(CurrentPhase)
        HUDMessage.SetText(MakeMessage("{CurrentPhase}/{GetTotalPhases()}"))
        CinematicSequence.SetPlaybackTime((CurrentPhase * 1.0) / (GetTotalPhases() * 1.0))
        VolumeActivator.Disable()
        Character.Hide()
        AdvancedStormController.GenerateStorm()
        ChallengeAccepted.Signal()

    OnPhaseEnded():void =
        set CurrentPhase += 1
        #TrackerInfo.Device.SetValue(CurrentPhase)
        HUDMessage.SetText(MakeMessage("{CurrentPhase}/{GetTotalPhases()}"))
        CinematicSequence.SetPlaybackTime((CurrentPhase * 1.0) / (GetTotalPhases() * 1.0))
        if (CurrentPhase >= GetTotalPhases()):
            CompleteEvent.Signal()