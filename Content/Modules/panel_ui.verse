using { /Fortnite.com/UI }
using { /UnrealEngine.com/Temporary/SpatialMath }
using { /UnrealEngine.com/Temporary/UI }
using { /Verse.org/Colors }
using { /Verse.org/Simulation }

panel_ui := class:
    Player:player
    Title:string
    PanelBorder:float = 4.0
    Padding:float = 40.0
    Anchors:anchors = anchors{Minimum := vector2{X := 0.5, Y := 0.5}, Maximum := vector2{X := 0.5, Y := 0.5}}
    Offsets:margin = margin{Left := 0.0, Top := 0.0, Right := 0.0, Bottom := 0.0}
    Alignment:vector2 = vector2{X := 0.5, Y := 0.5}
    BackgroundOpacity:type {_X:float where 0.000000 <= _X, _X <= 1.000000} = 0.3
    HasCloseButton:logic = true

    TitleText:text_block = text_block{DefaultTextColor := NamedColors.White}
    ClosedEvent:events(player) = events(player){}

    Background:color_block = color_block:
        DefaultColor := NamedColors.Black

    var MaybeUI:?widget = false
    var InputMode:ui_input_mode = ui_input_mode.All
    var IsShowing:logic = false

    <# ----------------------------------------------------------------------------------------------------
        UI functions
    ---------------------------------------------------------------------------------------------------- #>
    Show(?IsSignal:logic = true):void =
        if (IsShowing?) then return
        if (PlayerUI := GetPlayerUI[Player]):
            UI := if (ExistingUI := MaybeUI?):
                ExistingUI
            else:
                NewUI := CreateUI()
                set MaybeUI = option{NewUI}
                NewUI

            PlayerUI.AddWidget(UI, Slot := player_ui_slot{InputMode := InputMode})
            Player.TogglePlayerHUD(false)
            Player.UpdateActivePanel(Self)
            set IsShowing = true

            if (IsSignal?):
                OnShow()
    
    Hide(?IsSignal:logic = true):void =
        if (not IsShowing?) then return
        if (PlayerUI := GetPlayerUI[Player], UI := MaybeUI?):
            PlayerUI.RemoveWidget(UI)
            Player.TogglePlayerHUD(true)
            Player.ClearActivePanel()
            set IsShowing = false

            if (IsSignal?):
                ClosedEvent.Signal(Player)

    CreateUI():widget =
        TitleText.SetText(MakeMessage(Title))
        Background.SetOpacity(BackgroundOpacity)

        StackBox := stack_box:
            Orientation := orientation.Vertical
            Slots := array:
                stack_box_slot:
                    HorizontalAlignment := horizontal_alignment.Center
                    Padding := margin{Left := 0.0, Top := 0.0, Right := 0.0, Bottom := 10.0}
                    Widget := TitleText
                stack_box_slot:
                    HorizontalAlignment := horizontal_alignment.Fill
                    Padding := margin{Left := 0.0, Top := 0.0, Right := 0.0, Bottom := 20.0}
                    Widget := color_block:
                        DefaultColor := NamedColors.White
                        DefaultOpacity := 0.1
                        DefaultDesiredSize := vector2{X := 0.0, Y := 1.0}
                stack_box_slot:
                    HorizontalAlignment := horizontal_alignment.Fill
                    VerticalAlignment := vertical_alignment.Fill
                    Distribution := option{1.0}
                    Widget := CreateContent()

        if (HasCloseButton?):
            CloseButton:button_loud = button_loud{DefaultText := MakeMessage("Close")}
            CloseButton.OnClick().Subscribe(OnClickClose)

            StackBox.AddWidget(
                stack_box_slot:
                    HorizontalAlignment := horizontal_alignment.Fill
                    Padding := margin{Left := 0.0, Top := 20.0, Right := 0.0, Bottom := 10.0}
                    Widget := CloseButton
            )

        canvas:
            Slots := array:
                canvas_slot:
                    Anchors := anchors{Minimum := vector2{X := 0.0, Y := 0.0}, Maximum := vector2{X := 1.0, Y := 1.0}}
                    Widget := Background
                canvas_slot:
                    Anchors := Anchors
                    Alignment := Alignment
                    Offsets := Offsets
                    Widget := overlay:
                        Slots := array:
                            overlay_slot:
                                HorizontalAlignment := horizontal_alignment.Fill
                                VerticalAlignment := vertical_alignment.Fill
                                Widget := color_block:
                                    DefaultColor := NamedColors.White
                                    DefaultOpacity := 0.2
                            overlay_slot:
                                HorizontalAlignment := horizontal_alignment.Fill
                                VerticalAlignment := vertical_alignment.Fill
                                Padding := margin{Left := PanelBorder, Top := PanelBorder, Right := PanelBorder, Bottom := PanelBorder}
                                Widget := color_block:
                                    DefaultColor := ThemeColors.Background                               
                            overlay_slot:
                                HorizontalAlignment := horizontal_alignment.Fill
                                VerticalAlignment := vertical_alignment.Fill
                                Padding := margin{Left := Padding, Top := Padding, Right := Padding, Bottom := Padding}
                                Widget := StackBox

    CreateContent():widget =
        canvas{}

    <# ----------------------------------------------------------------------------------------------------
        Update Functions
    ---------------------------------------------------------------------------------------------------- #>
    SetTitle(NewTitle:string):void =
        TitleText.SetText(MakeMessage(NewTitle))

    SetInputMode(NewInputMode:ui_input_mode):void =
        set InputMode = NewInputMode

    SetBackgroundOpacity(Opacity:type{_X:float where 0.000000 <= _X, _X <= 1.000000}):void =
        Background.SetOpacity(Opacity)
        
    <# ----------------------------------------------------------------------------------------------------
        Event Handlers
    ---------------------------------------------------------------------------------------------------- #>
    OnClickClose(Message:widget_message):void =
        Hide()

    OnShow():void = {}