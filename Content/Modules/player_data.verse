using { /Verse.org/Simulation }

<# ----------------------------------------------------------------------------------------------------
    Constants
---------------------------------------------------------------------------------------------------- #>
WeaponUnlockCostIncrementRate:float = 0.25
WeaponUnlockBaseCost:float = 10000.0

<# ----------------------------------------------------------------------------------------------------
    Global Functions
---------------------------------------------------------------------------------------------------- #>
var PlayerDataMap:weak_map(player, player_data) = map{}

(Player:player).GetData()<decides><transacts>:player_data =
    Player.IsActive[]
    PlayerDataMap[Player]

<# ----------------------------------------------------------------------------------------------------
    Power Ups
---------------------------------------------------------------------------------------------------- #>
(Player:player).GetPowerUpLevel(Id:power_up_id)<transacts>:int =
    Player.GetData[].PowerUps[Id.ToInt()] or 0

(Player:player).GetPowerUpValue(Id:power_up_id)<transacts>:float =
    if (PowerUpInfo := GetPowerUpInfo[Id]):
        PowerUpInfo.BaseValue + Player.GetPowerUpLevel(Id) * PowerUpInfo.IncrementValue
    else:
        0.0

(Player:player).GetPowerUpCost(Id:power_up_id, Level:int)<decides><transacts>:float =
    PowerUpInfo := GetPowerUpInfo[Id]
    Floor[Pow(Level * 1.0, PowerUpInfo.CostFactor * 1.0) * PowerUpInfo.BaseCost] * 1.0

(Player:player).GetTotalPowerUpCost()<transacts>:float =
    var TotalPowerUpCost:float = 0.0
    for (PowerUpId : GetPowerUpIds()):
        for (PowerUpLevel := 1..Player.GetPowerUpLevel(PowerUpId), PowerUpCost := Player.GetPowerUpCost[PowerUpId, PowerUpLevel]):
            set TotalPowerUpCost += PowerUpCost
    TotalPowerUpCost

(Player:player).AddPowerUp(Id:power_up_id):void =
    if:
        CurrentPlayerData := Player.GetData[]
        var UpdatedPowerUps:[int]int = CurrentPlayerData.PowerUps
        set UpdatedPowerUps[Id.ToInt()] = UpdatedPowerUps[Id.ToInt()] + 1 or 1
        set PlayerDataMap[Player] = player_data:
            MakePlayerData<constructor>(CurrentPlayerData)
            PowerUps := UpdatedPowerUps

(Player:player).RefundPowerUp(Id:power_up_id):void =
    if:
        CurrentPlayerData := Player.GetData[]
        
        var UpdatedPowerUps:[int]int = CurrentPlayerData.PowerUps
        set UpdatedPowerUps[Id.ToInt()] = 0

        set PlayerDataMap[Player] = player_data:
            MakePlayerData<constructor>(CurrentPlayerData)
            PowerUps := UpdatedPowerUps

(Player:player).RefundOnePowerUp(Id:power_up_id)<decides><transacts>:void =
    CurrentPlayerData := Player.GetData[]
    
    var UpdatedPowerUps:[int]int = CurrentPlayerData.PowerUps
    UpdatedPowerUps[Id.ToInt()] >= 1
    set UpdatedPowerUps[Id.ToInt()] -= 1

    set PlayerDataMap[Player] = player_data:
        MakePlayerData<constructor>(CurrentPlayerData)
        PowerUps := UpdatedPowerUps

(Player:player).RefundPowerUps(Type:power_up_type):void =
    if:
        CurrentPlayerData := Player.GetData[]
        
        var UpdatedPowerUps:[int]int = CurrentPlayerData.PowerUps
        for (PowerUpId : GetPowerUpIds(Type)):
            set UpdatedPowerUps[PowerUpId.ToInt()] = 0

        set PlayerDataMap[Player] = player_data:
            MakePlayerData<constructor>(CurrentPlayerData)
            PowerUps := UpdatedPowerUps 

(Player:player).GetCurrencyAvailable(Id:currency_id)<transacts>:float =
    Player.GetData[].Currencies[Id.ToInt()] - Player.GetTotalPowerUpCost() or 0.0

(Player:player).GetSelectBufferTime()<transacts>:float =
    Player.GetData[].Settings.SelectBufferTime or 0.0

<# ----------------------------------------------------------------------------------------------------
    Weapon Unlock
---------------------------------------------------------------------------------------------------- #>
(Player:player).IsWeaponUnlocked(Id:weapon_id)<transacts>:logic =
    if (StartingWeapons.Find[Id]):
        return true
    else:
        return logic{Player.GetData[].WeaponsUnlocked[Id.ToInt()]}

(Player:player).GetWeaponUnlockCost()<transacts>:float =
    (Floor[WeaponUnlockBaseCost * Pow((1.0 + WeaponUnlockCostIncrementRate), Player.GetWeaponsUnlocked() * 1.0)] or 0) * 1.0

(Player:player).GetWeaponsUnlocked()<transacts>:int =
    Player.GetData[].WeaponsUnlocked.Length or 0

(Player:player).CanAffordWeaponUnlock()<transacts>:logic =
    logic{Player.GetCurrencyAvailable(currency_id.Gold) >= Player.GetWeaponUnlockCost()}

(Player:player).UnlockWeapon(Id:weapon_id):void =
    if (not Player.CanAffordWeaponUnlock()?) then return
    Player.RemoveCurrency(currency_id.Gold, Player.GetWeaponUnlockCost())

    if:
        CurrentPlayerData := Player.GetData[]
        
        Player.GetWeaponUnlockCost()
        var UpdatedWeaponsUnlocked:[int]logic = CurrentPlayerData.WeaponsUnlocked
        set UpdatedWeaponsUnlocked[Id.ToInt()] = true

        set PlayerDataMap[Player] = player_data:
            MakePlayerData<constructor>(CurrentPlayerData)
            WeaponsUnlocked := UpdatedWeaponsUnlocked

<# ----------------------------------------------------------------------------------------------------
    Arena Progress
---------------------------------------------------------------------------------------------------- #>
(Player:player).SetArenaProgress(ArenaProgress:int):void =
    if:
        CurrentPlayerData := Player.GetData[]

        set PlayerDataMap[Player] = player_data:
            MakePlayerData<constructor>(CurrentPlayerData)
            ArenaProgress := ArenaProgress

(Player:player).GetArenaProgress()<transacts>:int =
    Player.GetData[].ArenaProgress or 0

<# ----------------------------------------------------------------------------------------------------
    player_daily_quests
---------------------------------------------------------------------------------------------------- #>
player_daily_quests := class<final><persistable><computes>:
    LastQuestAssignedTime:float = 0.0
    LastRerollTime:float = 0.0
    ActiveQuests:[]daily_quest = array{}

MakePlayerDailyQuests<constructor>(UpdatedPlayerDailyQuests:player_daily_quests)<transacts> := player_daily_quests:
    LastQuestAssignedTime := UpdatedPlayerDailyQuests.LastQuestAssignedTime
    LastRerollTime := UpdatedPlayerDailyQuests.LastRerollTime
    ActiveQuests := UpdatedPlayerDailyQuests.ActiveQuests

daily_quest := class<final><persistable><computes>:
    Id:int
    Tier:int = 0
    Progress:float = 0.0
    MaybeWeaponId:?int = false

MakeDailyQuest<constructor>(UpdatedMakeDailyQuest:daily_quest)<transacts> := daily_quest:
    Id := UpdatedMakeDailyQuest.Id
    Tier := UpdatedMakeDailyQuest.Tier
    Progress := UpdatedMakeDailyQuest.Progress
    MaybeWeaponId := UpdatedMakeDailyQuest.MaybeWeaponId

<# ----------------------------------------------------------------------------------------------------
    player_data
---------------------------------------------------------------------------------------------------- #>
player_data := class<final><persistable><computes>:
    Version:int = 0
    News:int = 0
    Settings:player_settings = player_settings{}
    (player_data:)PowerUps:[int]int = map{}
    Currencies:[int]float = map{}
    ArenaProgress:int = 0
    WeaponProgress:[int]int = map{}
    WeaponsUnlocked:[int]logic = map{}
    DailyQuests:player_daily_quests = player_daily_quests{}

MakePlayerData<constructor>(UpdatedPlayerData:player_data)<transacts> := player_data:
    Version := UpdatedPlayerData.Version
    News := UpdatedPlayerData.News
    Settings := UpdatedPlayerData.Settings
    PowerUps := UpdatedPlayerData.PowerUps
    Currencies := UpdatedPlayerData.Currencies
    ArenaProgress := UpdatedPlayerData.ArenaProgress
    WeaponProgress := UpdatedPlayerData.WeaponProgress
    WeaponsUnlocked := UpdatedPlayerData.WeaponsUnlocked
    DailyQuests := UpdatedPlayerData.DailyQuests