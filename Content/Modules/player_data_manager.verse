using { /Fortnite.com/Devices }
using { /Verse.org/Simulation }

<# ----------------------------------------------------------------------------------------------------
    Global Functions
---------------------------------------------------------------------------------------------------- #>
var GlobalPlayerDataManager:weak_map(session, player_data_manager) = map{}

WaitForPlayerDataManager(Player:player)<suspends>:void =
    loop:
        if (GetPlayerDataManager[].GetIsInitialized[Player]) {break}
        Sleep(0.3)

GetPlayerDataManager()<decides><transacts>:player_data_manager =
    GlobalPlayerDataManager[GetSession()]

<# ----------------------------------------------------------------------------------------------------
    player_data_manager
---------------------------------------------------------------------------------------------------- #>
player_data_manager := class(creative_device):

    var PlayerIsInitialized:[player]logic = map{}

    <# ----------------------------------------------------------------------------------------------------
        Initialization
    ---------------------------------------------------------------------------------------------------- #>
    OnBegin<override>()<suspends>:void =
        if (set GlobalPlayerDataManager[GetSession()] = Self) {}

        for (Player : GetPlayspace().GetPlayers()):
            spawn{Initialize(Player)}

        GetPlayspace().PlayerAddedEvent().Subscribe(OnPlayerAdded)
        GetPlayspace().PlayerRemovedEvent().Subscribe(OnPlayerRemoved)

    
    Initialize(Player:player)<suspends>:void =
        if:
            PlayerOverrideData := GetPlayerDataOverride()?
            set PlayerDataMap[Player] = PlayerOverrideData

        if:
            not PlayerDataMap[Player]
            set PlayerDataMap[Player] = player_data{}
        
        CurrentNews := GetNewsInfo().Length
        if:
            CurrentPlayerData := Player.GetData[]
            CurrentPlayerData.News < CurrentNews
            set PlayerDataMap[Player] = player_data:
                MakePlayerData<constructor>(CurrentPlayerData)
                News := CurrentNews
        then:
            WaitForUIManager(Player)
            if (UIManager := GetUIManager[], PlayerMenuUI := UIManager.PlayerMenuUIs[Player]):
                PlayerMenuUI.ShowPanelUI("News")

        if (set PlayerIsInitialized[Player] = true) {}

    <# ----------------------------------------------------------------------------------------------------
        Get Functions
    ---------------------------------------------------------------------------------------------------- #>
    GetIsInitialized(Player:player)<decides><transacts>:void =
        PlayerIsInitialized[Player]

    <# ----------------------------------------------------------------------------------------------------
        Update Functions
    ---------------------------------------------------------------------------------------------------- #>
    

    <# ----------------------------------------------------------------------------------------------------
        Event handlers
    ---------------------------------------------------------------------------------------------------- #>
    OnPlayerAdded(Player:player):void =
        spawn{Initialize(Player)}

    OnPlayerRemoved(Player:player):void =
        set PlayerIsInitialized = RemoveKeyFromMap(PlayerIsInitialized, Player)