using { /Fortnite.com/Characters }
using { /Verse.org/Simulation }
using { /UnrealEngine.com/Temporary/SpatialMath }

<# ----------------------------------------------------------------------------------------------------
    player_settings
---------------------------------------------------------------------------------------------------- #>
player_settings := class<final><persistable><computes>:
    Camera:int = 0
    CameraDistance:int = 2
    Music:logic = true
    AutoAim:logic = false
    SelectBufferTime:float = 0.0

MakePlayerSettings<constructor>(UpdatedPlayerSettings:player_settings)<transacts> := player_settings:
    Camera := UpdatedPlayerSettings.Camera
    CameraDistance := UpdatedPlayerSettings.CameraDistance
    Music := UpdatedPlayerSettings.Music
    AutoAim := UpdatedPlayerSettings.AutoAim
    SelectBufferTime := UpdatedPlayerSettings.SelectBufferTime

<# ----------------------------------------------------------------------------------------------------
    Global Functions
---------------------------------------------------------------------------------------------------- #>
(Player:player).GetAutoAimRotation()<transacts>:rotation =
    if (Character := Player.GetFortCharacter[]):
        if:
            Player.GetData[].Settings.AutoAim?
            NearestAgent := GetNearestAgent(Player, for(NPCBehavior : GetAllNPCs(), NPCAgent := NPCBehavior.GetAgent[]) {NPCAgent})?
            NearestCharacter := NearestAgent.GetFortCharacter[]
            Direction := (NearestCharacter.GetTransform().Translation.Make2D() - Character.GetTransform().Translation.Make2D()).MakeUnitVector[]
        then:
            MakeRotation(vector3{X := 0.0, Y := 0.0, Z := 1.0}, ArcTan(Direction.Y, Direction.X))
        else:
            Character.GetTransform().Rotation
    else:
        IdentityRotation()

(Player:player).ToggleAutoAim():void =
    if:
        CurrentPlayerData := Player.GetData[]
        UpdatedAutoAim := logic{not CurrentPlayerData.Settings.AutoAim?}

        set PlayerDataMap[Player] = player_data:
            MakePlayerData<constructor>(CurrentPlayerData)
            Settings := player_settings:
                MakePlayerSettings<constructor>(CurrentPlayerData.Settings)
                    AutoAim := UpdatedAutoAim

(Player:player).ToggleSelectBufferTime():void =
    if:
        CurrentPlayerData := Player.GetData[]
        UpdatedSelectBufferTime := (CurrentPlayerData.Settings.SelectBufferTime = 3.0) and 0.0 or (CurrentPlayerData.Settings.SelectBufferTime + 0.5)

        set PlayerDataMap[Player] = player_data:
            MakePlayerData<constructor>(CurrentPlayerData)
            Settings := player_settings:
                MakePlayerSettings<constructor>(CurrentPlayerData.Settings)
                    SelectBufferTime := UpdatedSelectBufferTime