using { /Fortnite.com/Characters }
using { /Fortnite.com/Devices }
using { /Fortnite.com/FortPlayerUtilities }
using { /Verse.org/Simulation }
using { /UnrealEngine.com/Temporary/SpatialMath }
using { /UnrealEngine.com/Temporary/UI }

<# ----------------------------------------------------------------------------------------------------
    Constants
---------------------------------------------------------------------------------------------------- #>
MaxPlayerLevel:int = 200
MaxPlayers:int = 4

<# ----------------------------------------------------------------------------------------------------
    Global Functions
---------------------------------------------------------------------------------------------------- #>
var GlobalPlayerManager:weak_map(session, player_manager) = map{}

WaitForPlayerManager(Player:player)<suspends>:void =
    loop:
        if (GetPlayerManager[].GetIsInitialized[Player]) {break}
        Sleep(0.3)

GetPlayerManager()<decides><transacts>:player_manager =
    GlobalPlayerManager[GetSession()]

GetAllPlayers()<transacts>:[]player =
    GetPlayerManager[].GetPlayers() or array{}

GetLevel()<transacts>:int =
    GetPlayerManager[].GetLevel() or 1

AddLevel():void =
    if (PlayerManager := GetPlayerManager[]):
        PlayerManager.AddLevel()

GetXP()<transacts>:int =
    GetPlayerManager[].GetXP() or 0

GetXPRequired()<decides><transacts>:int =
    GetPlayerManager[].GetXPRequired[]

AddXP(Amount:int):void =
    if (PlayerManager := GetPlayerManager[]):
        PlayerManager.AddXP(Amount)

GetPlayerStats()<transacts>:[player]player_stat =
    GetPlayerManager[].GetPlayerStat() or map{}

ResetStats():void =
    if (PlayerManager := GetPlayerManager[]):
        PlayerManager.ResetStats()

GetServerTime()<transacts>:float =
    GetPlayerManager[].GetServerTime() or 0.0

GetTeamExperienceMultiplier()<transacts>:float =
    GetPlayerManager[].GetExperienceMultiplier() or 0.0

GetPlayerIndex(InPlayer:player)<transacts>:int =
    GetPlayerManager[].GetPlayerIndex[InPlayer] or 0

(Agent:agent).GetRevivePosition()<decides><transacts>:vector3 =
    GetPlayerManager[].ReviveVolumeActivators[GetPlayerIndex(player[Agent])].VolumeDevice.GetTransform().Translation

(Player:player).AssignIndex():void =
    if (PlayerManager := GetPlayerManager[]):
        PlayerManager.AssignIndex(Player)

(Player:player).GetReviveVolumeActivator()<decides><transacts>:volume_activator =
    GetPlayerManager[].GetReviveVolumeActivator[Player]

(Player:player).GetJoinedTime()<transacts>:float =
    GetPlayerManager[].GetPlayerJoinedTime[Player] or 0.0

<# ----------------------------------------------------------------------------------------------------
    player_manager
---------------------------------------------------------------------------------------------------- #>
player_manager := class(creative_device):
    @editable
    CinematicSequenceExperienceBar:cinematic_sequence_device = cinematic_sequence_device{}

    @editable
    CinematicSequencePlayerRevives:[]cinematic_sequence_device = array{}

    @editable
    DownButNotOut:down_but_not_out_device = down_but_not_out_device{}

    @editable
    ReviveVolumeActivators:[]volume_activator = array{}
    
    LeveledUpEvent<public>:events(tuple()) = events(tuple()){}
    XPChangedEvent<public>:events(int) = events(int){}
    WeaponEquippedEvent<public>:events(player) = events(player){}

    var PlayerStats<private>:[player]player_stat = map{}
    var PlayerRevives<private>:[player]player_revives = map{}
    var PlayerJoinedTime<private>:[player]float = map{}
    var Level<private>:int = 1
    var XP<private>:int = 0
    var ServerStartTime<private>:float = 0.0
    var PlayerIndexes:[player]int = map{}
    var LevelUpCounter:int = 0

    <# ----------------------------------------------------------------------------------------------------
        Initialization
    ---------------------------------------------------------------------------------------------------- #>
    OnBegin<override>()<suspends>:void =
        if (set GlobalPlayerManager[GetSession()] = Self) {}
        set ServerStartTime = GetSecondsSinceEpoch()

        Playspace := GetPlayspace()
        Playspace.PlayerAddedEvent().Subscribe(OnPlayerAdded)
        Playspace.PlayerRemovedEvent().Subscribe(OnPlayerRemoved)
        DownButNotOut.AgentDownedEvent.Subscribe(OnAgentDowned)

        for (Player : Playspace.GetPlayers()):
            spawn{Initialize(Player)}

        for (Index -> ReviveVolumeActivator : ReviveVolumeActivators):
            ReviveVolumeActivator.OnBegin()
            spawn{AwaitReviveVolumeActivated(ReviveVolumeActivator, Index)}

        WaitForGameManager()
        if (GameManager := GetGameManager[]):
            GameManager.GameStartedEvent.Subscribe(OnGameStarted)
            GameManager.GameEndedEvent.Subscribe(OnGameEnded)
    
    Initialize(Player:player)<suspends>:void =
        if (PlayerJoinedTime[Player]) then return

        PlayerStat := player_stat{Player := Player}
        if (set PlayerStats[Player] = PlayerStat) {}

        PlayerXPBar := player_xp_bar{Player := Player, PlayerManager := Self}
        PlayerXPBar.Initialize()

        PlayerRevive := player_revives{Player := Player}
        PlayerRevive.Initialize()
        if (set PlayerRevives[Player] = PlayerRevive) {}

        if (Character := Player.GetFortCharacter[]):
            spawn{AwaitDamaged(Character)}

        if (set PlayerJoinedTime[Player] = GetSimulationElapsedTime()) {}

    <# ----------------------------------------------------------------------------------------------------
        Get Functions
    ---------------------------------------------------------------------------------------------------- #>
    GetIsInitialized(Player:player)<decides><transacts>:void =
        PlayerJoinedTime[Player]

    GetPlayerJoinedTime(Player:player)<decides><transacts>:float =
        PlayerJoinedTime[Player]

    GetPlayers()<transacts>:[]player =
        GetPlayspace().GetPlayers()

    GetPlayerStat()<transacts>:[player]player_stat =
        PlayerStats

    GetPlayerStat(Player:player)<decides><transacts>:player_stat =
        PlayerStats[Player]

    (player_manager:)GetLevel()<transacts>:int =
        Level

    (player_manager:)GetXP()<transacts>:int =
        XP

    (player_manager:)GetXPRequired()<decides><transacts>:int = 
        Level < MaxPlayerLevel
        Floor[10 * Pow(1.035, (player_manager:)GetLevel() * 1.0) * (player_manager:)GetLevel()]

    (player_manager:)GetServerTime()<transacts>:float =
        GetSecondsSinceEpoch() - ServerStartTime

    GetExperienceMultiplier()<transacts>:float =
        Average of for (Player : GetActivePlayers(), AbilityInfo := GetAbilityInfo[ability_id.ExperienceMultiplier]):
            Player.GetTotalAbilityValue(AbilityInfo)

    (player_manager:)GetPlayerIndex(Player:player)<decides><transacts>:int =
        PlayerIndexes[Player]

    (player_manager:)GetPlayerByIndex(Index:int)<decides><transacts>:player =
        for (Player -> PlayerIndex : PlayerIndexes, PlayerIndex = Index) {Player}[0]

    GetReviveVolumeActivator(Player:player)<decides><transacts>:volume_activator =
        ReviveVolumeActivators[(player_manager:)GetPlayerIndex[Player]]

    <# ----------------------------------------------------------------------------------------------------
        Update Functions
    ---------------------------------------------------------------------------------------------------- #>
    (player_manager:)AddXP(Amount:int):void =
        if (Level >= MaxPlayerLevel) then return

        TotalAmount := Floor[Amount * (1.0 + GetExperienceMultiplier())] or Amount
        set XP += TotalAmount

        loop:
            if (XPRequired := (player_manager:)GetXPRequired[], XP >= XPRequired):
                (player_manager:)AddLevel()

                if (Level >= MaxPlayerLevel):
                    set XP = 0
                else:
                    set XP = XP - XPRequired
            else:
                break

        XPChangedEvent.Signal(XP)

    (player_manager:)AddLevel():void =
        if (Level < MaxPlayerLevel):
            set Level = Level + 1
            if (LevelUpCounter > 0):
                set LevelUpCounter += 1
            else:
                spawn{LevelUp(Level)}

            for (Player : GetActivePlayers()):
                Player.ActivateVFXPowerup(vfx_powerup_id.LevelUp)
                Player.PlayAudio(audio_id.LevelUp)
                Player.AwardAccolade(accolade_count_id.LevelUp)

            LeveledUpEvent.Signal()

    LevelUp(InLevel:int)<suspends>:void =
        set LevelUpCounter = 1
        for (Player : GetActivePlayersNotAFK()):
            Player.ActivateVFXPowerup(vfx_powerup_id.LevelUpText, ?Duration := 3.0)
            
        Sleep(3.0)
        for (Index := 0..LevelUpCounter - 1):
            PauseGame(pause_type.LevelUp, ?Level := InLevel + Index)

        set LevelUpCounter = 0

    (player_manager:)ResetStats():void =
        set Level = 1
        set XP = 0
        for (Player : GetAllPlayers(), PlayerStat := GetPlayerStat[Player]):
            PlayerStat.ResetStats()

    AssignIndex(Player:player):void =
        if (set PlayerIndexes[Player] = PlayerIndexes.Length) {}

    <# ----------------------------------------------------------------------------------------------------
        Event handlers
    ---------------------------------------------------------------------------------------------------- #>
    OnPlayerAdded(Player:player):void =
        spawn{Initialize(Player)}

    OnPlayerRemoved(Player:player):void =
        if (ReviveVolumeActivator := GetReviveVolumeActivator[Player]):
            ReviveVolumeActivator.Disable()

        if (PlayerRevive := PlayerRevives[Player]):
            PlayerRevive.Reset()
        
        set PlayerStats = RemoveKeyFromMap(PlayerStats, Player)
        set PlayerJoinedTime = RemoveKeyFromMap(PlayerJoinedTime, Player)
        set PlayerRevives = RemoveKeyFromMap(PlayerRevives, Player)

    AwaitDamaged(Character:fort_character)<suspends>:void =
        loop:
            DamageResult := Character.DamagedEvent().Await()
            if (DamageResult.Amount > 0.0, Agent := Character.GetAgent[]):
                Agent.ActivateVFXPowerup(vfx_powerup_id.DamagedGlow, ?Duration := 0.3)
                SpawnVFX(vfx_id.DamageNumberPlayer, Character.GetTransform().Translation + vector3{Z := 100.0}, ?Scale := vector3{X := 1.0, Y := 1.0, Z := DamageResult.Amount}, ?Duration := 1.0)

    OnAgentDowned(Agent:agent):void =
        if (GetAlivePlayers().Length = 0) then return

        if:
            Player := player[Agent]
            Player.GetPlayerStats(player_stat_id.Revives) <= 0
            Character := Player.GetFortCharacter[]
            ReviveVolumeActivator := GetReviveVolumeActivator[Player]
        then:
            Player.AddGameStat(player_game_stat_id.TimesDowned, 1.0)
            ReviveVolumeActivator.Enable(?TargetPosition := option{Character.GetTransform().Translation})

            for (ActivePlayer : GetActivePlayers(), ActivePlayer <> Player):
                ActivePlayer.ShowMessage(message_id.PlayerDowned, MakeMessage(Player))

    AwaitReviveVolumeActivated(ReviveVolumeActivator:volume_activator, Index:int)<suspends>:void =
        loop:
            RevivingPlayer := ReviveVolumeActivator.ActivatedEvent.Await()
            RevivingPlayer.AddGameStat(player_game_stat_id.PlayersRevived, 1.0)

            if (Player := GetPlayerByIndex[Index], Character := Player.GetFortCharacter[]):
                MaybePanel := option{Player.GetActivePanel[]}
                if (Panel := MaybePanel?):
                    Panel.Hide(?IsSignal := false)

                DownButNotOut.Revive(Player)

                if (Panel := MaybePanel?):
                    Panel.Show(?IsSignal := false)
                
                if (Character.TeleportTo[ReviveVolumeActivator.InitialPosition.Make2D() + vector3{Z := 300.0}, IdentityRotation()]) {}
                KnockbackNPC(Player, ?KnockbackDistanceSquared := DefaultNPCKnockbackDistanceSquaredRevive, ?Duration := 3.0)
                ReviveVolumeActivator.Disable()

    OnGameStarted():void =
        for (Index -> Player : GetActivePlayers(), set PlayerIndexes[Player] = Index):
            if (PlayerRevive := PlayerRevives[Player], CinematicSequencePlayerRevive := CinematicSequencePlayerRevives[Index]):
                PlayerRevive.Assign(CinematicSequencePlayerRevive)

    OnGameEnded():void =
        set PlayerIndexes = map{}

        for (ReviveVolumeActivator : ReviveVolumeActivators):
            ReviveVolumeActivator.Disable()

        for (Player : GetAllPlayers()):
            Player.HideMessage(message_id.Revives)