using { /Fortnite.com/Characters }
using { /Fortnite.com/UI }
using { /Verse.org/Simulation }
using { /UnrealEngine.com/Temporary/SpatialMath }
using { /UnrealEngine.com/Temporary/UI }

player_menu_ui := class(panel_ui):
    Title<override>:string = "Menu"
    BackgroundOpacity<override>:type {_X:float where 0.000000 <= _X, _X <= 1.000000} = 0.7
    HasCloseButton<override>:logic = false

    ServerTimeText:text_block = text_block:
        DefaultTextColor := ThemeColors.PrimaryText
        DefaultShadowOffset := option{vector2{X := 2.0, Y := 2.0}}
        DefaultShadowOpacity := 0.5

    GiveUpButton:button_regular = button_regular{}

    Options:[]string = array:
        "News"
        "Camera"
        "Camera Distance"
        "Music"
        "Auto Aim"
        "Select Buffer Time"

    var AbilityLevelTexts:[ability_id]text_block = map{}
    var MaybeSpawnNPCUI:?test_tool_spawn_npc_ui = false
    var PanelUIs:[string]panel_ui = map{}
    var OptionButtons:[string]text_button_base = map{}
    var AbilityTexts:[ability_id]text_block = map{}
    var IsGiveUpConfirmation:logic = false

    <# ----------------------------------------------------------------------------------------------------
        Initialization
    ---------------------------------------------------------------------------------------------------- #>
    CreateContent<override>():widget =
        ClosedEvent.Subscribe(OnClosed)
        GiveUpButton.OnClick().Subscribe(OnClickGiveUp)

        CloseButton := button_loud{DefaultText := MakeMessage("Close")}
        CloseButton.OnClick().Subscribe(OnClickClose)

        AbilitiesStackBox := stack_box:
            Orientation := orientation.Vertical
            Slots := for (AbilityInfo : GetAbilityInfo()):
                AbilityText := text_block:
                    DefaultTextColor := ThemeColors.PrimaryText
    
                if (set AbilityTexts[AbilityInfo.Id] = AbilityText) {}

                stack_box_slot:
                    HorizontalAlignment := horizontal_alignment.Fill
                    Widget := CreateStatUI(AbilityText, "{AbilityInfo.Name}", ?Icon := option{AbilityInfo.IconSmall})

        stack_box:
            Orientation := orientation.Horizontal
            Slots := array:
                stack_box_slot:
                    HorizontalAlignment := horizontal_alignment.Fill
                    VerticalAlignment := vertical_alignment.Fill
                    Distribution := option{1.0}
                    Widget := stack_box:
                        Orientation := orientation.Vertical
                        Slots := array:
                            stack_box_slot:
                                HorizontalAlignment := horizontal_alignment.Fill
                                VerticalAlignment := vertical_alignment.Fill
                                Widget := CreateServerTimeUI()
                            stack_box_slot:
                                HorizontalAlignment := horizontal_alignment.Fill
                                VerticalAlignment := vertical_alignment.Fill
                                Padding := margin{Left := 0.0, Top := 20.0, Right := 0.0, Bottom := 0.0}
                                Widget := CloseButton
                            stack_box_slot:
                                HorizontalAlignment := horizontal_alignment.Fill
                                VerticalAlignment := vertical_alignment.Fill
                                Padding := margin{Left := 0.0, Top := 20.0, Right := 0.0, Bottom := 0.0}
                                Widget := CreateMenuButtonsUI()
                            stack_box_slot:
                                HorizontalAlignment := horizontal_alignment.Fill
                                VerticalAlignment := vertical_alignment.Bottom
                                Distribution := option{1.0}
                                Padding := margin{Left := 0.0, Top := 80.0, Right := 0.0, Bottom := 0.0}
                                Widget := GiveUpButton


    CreateMenuButtonsUI():widget =
        stack_box:
            Orientation := orientation.Vertical
            Slots := for (Index -> Option : Options):
                Button:button_quiet = button_quiet:
                    DefaultText := MakeMessage("{Option}")

                if (set OptionButtons[Option] = Button) {}

                case (Option):
                    "News" => Button.OnClick().Subscribe(OnClickNews)
                    "Camera" => Button.OnClick().Subscribe(OnClickCamera)
                    "Camera Distance" => Button.OnClick().Subscribe(OnClickCameraDistance)
                    "Music" => Button.OnClick().Subscribe(OnClickMusic)
                    "Auto Aim" => Button.OnClick().Subscribe(OnClickAutoAim)
                    "Select Buffer Time" => Button.OnClick().Subscribe(OnClickSelectBufferTime)
                    _ => {}

                stack_box_slot:
                    HorizontalAlignment := horizontal_alignment.Fill
                    VerticalAlignment := vertical_alignment.Fill
                    Padding := margin{Left := 0.0, Top := (Index = 0) and 0.0 or 10.0, Right := 0.0, Bottom := 0.0}
                    Widget := Button
                    
    CreateServerTimeUI():widget =
        overlay:
            Slots := array:
                overlay_slot:
                    HorizontalAlignment := horizontal_alignment.Fill
                    VerticalAlignment := vertical_alignment.Fill
                    Widget := MakeBox() 
                overlay_slot:
                    HorizontalAlignment := horizontal_alignment.Fill
                    VerticalAlignment := vertical_alignment.Fill
                    Padding := margin{Left := 40.0, Top := 20.0, Right := 40.0, Bottom := 20.0}
                    Widget := stack_box:
                        Orientation := orientation.Horizontal
                        Slots := array:
                            stack_box_slot:
                                HorizontalAlignment := horizontal_alignment.Left
                                Distribution := option{1.0}
                                Padding := margin{Left := 0.0, Top := 0.0, Right := 20.0, Bottom := 0.0}
                                Widget := text_block:
                                    DefaultText := MakeMessage("Server Time:")
                                    DefaultTextColor := ThemeColors.PrimaryText
                                    DefaultTextOpacity := 0.5
                            stack_box_slot:
                                HorizontalAlignment := horizontal_alignment.Right
                                Widget := ServerTimeText

    <# ----------------------------------------------------------------------------------------------------
        Get Functions
    ---------------------------------------------------------------------------------------------------- #>
    CanGiveUp()<transacts>:logic =
        logic:
            IsGameInProgress()? 
            GetGameManager[].GameTime >= 0.1
            not Player.GetFortCharacter[].IsDownButNotOut[]
            not Player.IsInStageVolume()?

    <# ----------------------------------------------------------------------------------------------------
        Update Functions
    ---------------------------------------------------------------------------------------------------- #>
    Update():void =
        ServerTimeText.SetText(MakeMessage("{GetServerTime().FormatAsTime(?IsHour := true)}"))
        
        GiveUpButton.SetText(MakeMessage(IsGiveUpConfirmation? and "Confirm" or "Give Up"))

        if (CanGiveUp()?):
            GiveUpButton.SetVisibility(widget_visibility.Visible)
        else:
            GiveUpButton.SetVisibility(widget_visibility.Collapsed)

        if (CameraButton := OptionButtons["Camera"], CameraDistanceButton := OptionButtons["Camera Distance"]):
            CameraButton.SetText(MakeMessage("Camera: {Player.GetActiveCamera()}"))

            if (Player.GetActiveCamera() = camera_type.Orthographic or Player.GetActiveCamera() = camera_type.TwinStick, CameraDistance := Player.GetData[].Settings.CameraDistance):
                CameraDistanceButton.SetText(MakeMessage("Camera Distance: {CameraDistance}"))
            else:
                CameraDistanceButton.SetText(MakeMessage(""))

        if (MusicButton := OptionButtons["Music"], IsMusicEnabled := Player.GetData[].Settings.Music):
            MusicButton.SetText(MakeMessage("Music: {IsMusicEnabled? and "ON" or "OFF"}"))

        if (AutoAimButton := OptionButtons["Auto Aim"], IsAutoAim := Player.GetData[].Settings.AutoAim):
            AutoAimButton.SetText(MakeMessage("Auto Aim: {IsAutoAim? and "ON" or "OFF"}"))

        if (SelectBufferTimeButton := OptionButtons["Select Buffer Time"], SelectBufferTime := Player.GetData[].Settings.SelectBufferTime):
            SelectBufferTimeButton.SetText(MakeMessage("Select Buffer Time: {FormatAsTime(SelectBufferTime)}"))

    ShowPanelUI(Panel:string):void =
        UI := if (PanelUI := PanelUIs[Panel]):
            PanelUI
        else:
            PanelUI := case (Panel):
                "News" => news_ui{Player := Player}
                _ => 
                    Print("ShowPanelUI({Panel}) is not a panel_ui.", ?Color := ThemeColors.Failure)
                    return

            if (set PanelUIs[Panel] = PanelUI) {}
            PanelUI

        UI.Show()

    <# ----------------------------------------------------------------------------------------------------
        Event Handlers
    ---------------------------------------------------------------------------------------------------- #>
    OnShow<override>():void =
        set IsGiveUpConfirmation = false
        Update()
        Player.PlayAudio(audio_id.Activate)
        if (IsGameInProgress()? and GetAllPlayers().Length = 1):
            PauseGame()

    OnClosed(InPlayer:player):void =
        if (IsGameInProgress()? and GetAllPlayers().Length = 1):
            ResumeGame(?IsInstant := true)

    OnClickNews(Message:widget_message):void =
        Hide()
        ShowPanelUI("News")

    OnClickCamera(Message:widget_message):void =
        Player.ToggleCamera()
        Update()

    OnClickCameraDistance(Message:widget_message):void =
        Player.ToggleCameraDistance()
        Player.RefreshControls()
        Update()

    OnClickMusic(Message:widget_message):void =
        Player.ToggleMusic()
        Update()

    OnClickGiveUp(Message:widget_message):void =
        if (Character := Player.GetFortCharacter[]):
            if (not CanGiveUp()?):
                Player.PlayAudio(audio_id.Error)
                return

            if (IsGiveUpConfirmation?):
                Player.SetVulnerability(true)
                Player.RemovePlayerStats(player_stat_id.Revives, Player.GetPlayerStats(player_stat_id.Revives))
                Player.HideMessage(message_id.Revives)
                Character.Damage(Character.GetMaxHealth() + Character.GetMaxShield())
                Hide()
            else:
                set IsGiveUpConfirmation = true
                Update()
                Player.PlayAudio(audio_id.Activate)

    OnClickAutoAim(Message:widget_message):void =
        Player.ToggleAutoAim()
        Player.RefreshControls()
        Update()

    OnClickSelectBufferTime(Message:widget_message):void =
        Player.ToggleSelectBufferTime()
        Update()