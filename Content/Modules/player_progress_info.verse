using { /Verse.org/Assets }
using { /Verse.org/Simulation }

<# ----------------------------------------------------------------------------------------------------
    Data Definitions
---------------------------------------------------------------------------------------------------- #>
Ranks:[]string = array:
    "Genin-I"
    "Genin-II"
    "Genin-III"
    "Chunin-I"
    "Chunin-II"
    "Chunin-III"
    "Jonin-I"
    "Jonin-II"
    "Jonin-III"
    "Ninja"

RanksIcons:[]texture = array:
    UserInterface.Icons.RankIcon_01_64
    UserInterface.Icons.RankIcon_02_64
    UserInterface.Icons.RankIcon_03_64
    UserInterface.Icons.RankIcon_04_64
    UserInterface.Icons.RankIcon_05_64
    UserInterface.Icons.RankIcon_06_64
    UserInterface.Icons.RankIcon_07_64
    UserInterface.Icons.RankIcon_08_64
    UserInterface.Icons.RankIcon_09_64
    UserInterface.Icons.RankIcon_10_64

<# ----------------------------------------------------------------------------------------------------
    Global Functions
---------------------------------------------------------------------------------------------------- #>
(Player:player).GetRank()<transacts>:int =
    Floor[(Player.GetArenaProgress() * 1.0) / (GetArenaCount() * 1.0)] or 1

(Player:player).GetTotalWeaponProgress()<transacts>:int =
    Sum of for (Progress : Player.GetData[].WeaponProgress) {Progress}

(Player:player).GetRankVFXPowerUp()<transacts>:vfx_powerup_id =
    case (Player.GetRank()):
        0 => vfx_powerup_id.Rank01
        1 => vfx_powerup_id.Rank02
        2 => vfx_powerup_id.Rank03
        3 => vfx_powerup_id.Rank04
        4 => vfx_powerup_id.Rank05
        5 => vfx_powerup_id.Rank06
        6 => vfx_powerup_id.Rank07
        7 => vfx_powerup_id.Rank08
        8 => vfx_powerup_id.Rank09
        9 => vfx_powerup_id.Rank10
        _ => vfx_powerup_id.Rank01

(Player:player).GetRankUpMessage()<transacts>:message_id =
    case (Player.GetRank()):
        1 => message_id.RankUp01
        2 => message_id.RankUp02
        3 => message_id.RankUp03
        4 => message_id.RankUp04
        5 => message_id.RankUp05
        6 => message_id.RankUp06
        7 => message_id.RankUp07
        8 => message_id.RankUp08
        9 => message_id.RankUp09
        _ => message_id.RankUp01

(Rank:int).ToRankString()<transacts>:string =
    Ranks[Rank] or "D-Rank"

(Rank:int).ToRankIcon()<transacts>:texture =
        RanksIcons[Rank] or UserInterface.Icons.RankIcon_01_64
