using { /Fortnite.com/UI }
using { /Verse.org/Colors }
using { /Verse.org/Simulation }
using { /UnrealEngine.com/Temporary/SpatialMath }
using { /UnrealEngine.com/Temporary/UI }

<# ----------------------------------------------------------------------------------------------------
    player_progress_ui
---------------------------------------------------------------------------------------------------- #>
player_progress_ui := class(panel_ui):
    Title<override>:string = "Player Progress"
    BackgroundOpacity<override>:type {_X:float where 0.000000 <= _X, _X <= 1.000000} = 0.7

    var PlayerButtons:[]button_quiet = array{}
    var MaybePlayerProgressStatUI:?player_progress_stat_ui = false
    var MaybePlayerProgressWeaponUI:?player_progress_weapon_ui = false
    var RankInfoUI:widget = canvas{}

    <# ----------------------------------------------------------------------------------------------------
        Initialization
    ---------------------------------------------------------------------------------------------------- #>
    CreateContent<override>():widget =
        PlayerProgressStatUI := player_progress_stat_ui{Player := Player}
        PlayerProgressStatUI.Initialize()
        set MaybePlayerProgressStatUI = option{PlayerProgressStatUI}

        PlayerProgressWeaponUI := player_progress_weapon_ui{Player := Player}
        PlayerProgressWeaponUI.Initialize()
        set MaybePlayerProgressWeaponUI = option{PlayerProgressWeaponUI}

        stack_box:
            Orientation := orientation.Horizontal
            Slots := array:
                stack_box_slot:
                    HorizontalAlignment := horizontal_alignment.Fill
                    VerticalAlignment := vertical_alignment.Fill
                    Widget := overlay:
                        Slots := array:
                            overlay_slot:
                                HorizontalAlignment := horizontal_alignment.Fill
                                VerticalAlignment := vertical_alignment.Fill
                                Widget := MakeBox()
                            overlay_slot:
                                HorizontalAlignment := horizontal_alignment.Fill
                                VerticalAlignment := vertical_alignment.Fill
                                Padding := margin{Left := 40.0, Top := 40.0, Right := 40.0, Bottom := 40.0}
                                Widget := CreateMenuUI()
                stack_box_slot:
                    HorizontalAlignment := horizontal_alignment.Fill
                    VerticalAlignment := vertical_alignment.Fill
                    Padding := margin{Left := 20.0, Top := 0.0, Right := 0.0, Bottom := 0.0}
                    Widget := overlay:
                        Slots := array:
                            overlay_slot:
                                HorizontalAlignment := horizontal_alignment.Fill
                                VerticalAlignment := vertical_alignment.Fill
                                Widget := stack_box:
                                    Orientation := orientation.Horizontal
                                    Slots := array:
                                        stack_box_slot:
                                            HorizontalAlignment := horizontal_alignment.Fill
                                            VerticalAlignment := vertical_alignment.Fill
                                            Widget := PlayerProgressStatUI.GetUI()
                                        stack_box_slot:
                                            HorizontalAlignment := horizontal_alignment.Fill
                                            VerticalAlignment := vertical_alignment.Fill
                                            Padding := margin{Left := 20.0, Top := 0.0, Right := 0.0, Bottom := 0.0}
                                            Widget := PlayerProgressWeaponUI.GetUI()
                            overlay_slot:
                                HorizontalAlignment := horizontal_alignment.Fill
                                VerticalAlignment := vertical_alignment.Fill
                                Widget := set RankInfoUI = CreateRankInfoUI()

    <# ----------------------------------------------------------------------------------------------------
        Menu UI
    ---------------------------------------------------------------------------------------------------- #>
    CreateMenuUI():widget =
        StackBox := stack_box:
            Orientation := orientation.Vertical
            Slots := array:
                stack_box_slot:
                    VerticalAlignment := vertical_alignment.Fill
                    Padding := margin{Left := 0.0, Top := 0.0, Right := 0.0, Bottom := 20.0}
                    Widget := text_block:
                        DefaultText := MakeMessage("Menu")
                        DefaultTextColor := ThemeColors.PrimaryText
                        DefaultTextOpacity := 0.5

        for (Index := 0..MaxPlayers - 1):
            PlayerButton := button_quiet{}
            spawn{AwaitPlayerButtonClicked(PlayerButton.OnClick(), Index)}
            set PlayerButtons += array{PlayerButton}
                
            StackBox.AddWidget of stack_box_slot:
                HorizontalAlignment := horizontal_alignment.Fill
                VerticalAlignment := vertical_alignment.Fill
                Padding := margin{Left := 0.0, Top := (Index = 0) and 0.0 or 10.0, Right := 0.0, Bottom := 0.0}
                Widget := PlayerButton

        RankInfoButton := button_quiet{DefaultText := MakeMessage("Rank Info")}
        RankInfoButton.OnClick().Subscribe(OnClickRankInfo)

        StackBox.AddWidget of stack_box_slot:
            HorizontalAlignment := horizontal_alignment.Fill
            VerticalAlignment := vertical_alignment.Bottom
            Distribution := option{1.0}
            Widget := RankInfoButton

        StackBox

    <# ----------------------------------------------------------------------------------------------------
        Rank Info UI
    ---------------------------------------------------------------------------------------------------- #>
    CreateRankInfoUI():widget =
        overlay:
            Slots := array:
                overlay_slot:
                    HorizontalAlignment := horizontal_alignment.Fill
                    VerticalAlignment := vertical_alignment.Fill
                    Widget := MakeBox(?Width := 1220.0)
                overlay_slot:
                    HorizontalAlignment := horizontal_alignment.Fill
                    VerticalAlignment := vertical_alignment.Fill
                    Padding := margin{Left := 40.0, Top := 40.0, Right := 40.0, Bottom := 40.0}
                    Widget := stack_box:
                        Orientation := orientation.Vertical
                        Slots := array:
                            stack_box_slot:
                                HorizontalAlignment := horizontal_alignment.Fill
                                VerticalAlignment := vertical_alignment.Fill
                                Widget := text_block:
                                    DefaultText := MakeMessage("Rank Info")
                                    DefaultTextColor := ThemeColors.PrimaryText
                                    DefaultShadowColor := NamedColors.Black
                                    DefaultShadowOffset := option{vector2{X := 2.0, Y := 2.0}}
                                    DefaultShadowOpacity := 0.5
                            stack_box_slot:
                                HorizontalAlignment := horizontal_alignment.Fill
                                VerticalAlignment := vertical_alignment.Fill
                                Padding := margin{Left := 0.0, Top := 20.0, Right := 0.0, Bottom := 0.0}
                                Widget := stack_box:
                                    Orientation := orientation.Horizontal
                                    Slots := array:
                                        stack_box_slot:
                                            HorizontalAlignment := horizontal_alignment.Left
                                            VerticalAlignment := vertical_alignment.Fill
                                            Distribution := option{1.0}
                                            Widget := text_block:
                                                DefaultText := MakeMessage("Rank")
                                                DefaultTextColor := ThemeColors.PrimaryText
                                                DefaultTextOpacity := 0.5
                                        stack_box_slot:
                                            HorizontalAlignment := horizontal_alignment.Left
                                            VerticalAlignment := vertical_alignment.Fill
                                            Distribution := option{1.0}
                                            Widget := text_block:
                                                DefaultText := MakeMessage("Arena Level")
                                                DefaultTextColor := ThemeColors.PrimaryText
                                                DefaultTextOpacity := 0.5
                            stack_box_slot:
                                HorizontalAlignment := horizontal_alignment.Fill
                                VerticalAlignment := vertical_alignment.Fill
                                Padding := margin{Left := 0.0, Top := 20.0, Right := 0.0, Bottom := 0.0}
                                Widget := stack_box:
                                    Orientation := orientation.Vertical
                                    Slots := for (RankIndex -> Rank : Ranks):
                                        stack_box_slot:
                                            HorizontalAlignment := horizontal_alignment.Fill
                                            VerticalAlignment := vertical_alignment.Fill
                                            Padding := margin{Left := 0.0, Top := (RankIndex = 0) and 0.0 or 10.0, Right := 0.0, Bottom := 0.0}
                                            Widget := stack_box:
                                                Orientation := orientation.Horizontal
                                                Slots := array:
                                                    stack_box_slot:
                                                        HorizontalAlignment := horizontal_alignment.Fill
                                                        VerticalAlignment := vertical_alignment.Fill
                                                        Distribution := option{1.0}
                                                        Widget := stack_box:
                                                            Orientation := orientation.Horizontal
                                                            Slots := array:
                                                                stack_box_slot:
                                                                    HorizontalAlignment := horizontal_alignment.Left
                                                                    Widget := texture_block:
                                                                        DefaultImage := RankIndex.ToRankIcon()
                                                                        DefaultDesiredSize := vector2{X := 40.0, Y := 40.0}
                                                                stack_box_slot:
                                                                    HorizontalAlignment := horizontal_alignment.Fill
                                                                    VerticalAlignment := vertical_alignment.Fill
                                                                    Padding := margin{Left := 10.0, Top := 0.0, Right := 0.0, Bottom := 0.0}
                                                                    Widget := text_block:
                                                                        DefaultText := MakeMessage("{RankIndex.ToRankString()}")
                                                                        DefaultTextColor := RankIndex.ToTierColor() * 2.0
                                                                        DefaultShadowColor := RankIndex.ToTierColor() * 0.25
                                                                        DefaultShadowOffset := option{vector2{X := 2.0, Y := 2.0}}
                                                                        DefaultShadowOpacity := 1.0
                                                    stack_box_slot:
                                                        HorizontalAlignment := horizontal_alignment.Fill
                                                        VerticalAlignment := vertical_alignment.Fill
                                                        Distribution := option{1.0}
                                                        Widget := stack_box:
                                                            Orientation := orientation.Horizontal
                                                            Slots := array:
                                                                stack_box_slot:
                                                                    HorizontalAlignment := horizontal_alignment.Fill
                                                                    VerticalAlignment := vertical_alignment.Fill
                                                                    Widget := text_block:
                                                                        DefaultText := MakeMessage("Lv.")
                                                                        DefaultTextColor := ThemeColors.Highlight
                                                                        DefaultShadowColor := ThemeColors.Highlight * 0.125
                                                                        DefaultShadowOffset := option{vector2{X := 2.0, Y := 2.0}}
                                                                        DefaultShadowOpacity := 1.0
                                                                stack_box_slot:
                                                                    HorizontalAlignment := horizontal_alignment.Fill
                                                                    VerticalAlignment := vertical_alignment.Fill
                                                                    Padding := margin{Left := 10.0, Top := 0.0, Right := 0.0, Bottom := 0.0}
                                                                    Widget := text_block:
                                                                        DefaultText := MakeMessage("{RankIndex * GetArenaCount()}")
                                                                        DefaultTextColor := ThemeColors.PrimaryText
                                                                        DefaultShadowColor := NamedColors.Black
                                                                        DefaultShadowOffset := option{vector2{X := 2.0, Y := 2.0}}
                                                                        DefaultShadowOpacity := 0.5

    <# ----------------------------------------------------------------------------------------------------
        Update Functions
    ---------------------------------------------------------------------------------------------------- #>
    Update():void =
        Players := GetAllPlayers()
        for (Index -> PlayerButton : PlayerButtons):
            if (InPlayer := Players[Index]):
                PlayerButton.SetText(MakeMessage(InPlayer))
                PlayerButton.SetVisibility(widget_visibility.Visible)
            else:
                PlayerButton.SetVisibility(widget_visibility.Hidden)

        if (InPlayer := Players[0]):
            SetPlayer(InPlayer)

    SetPlayer(InPlayer:player):void =
        RankInfoUI.SetVisibility(widget_visibility.Hidden)

        if (PlayerProgressStatUI := MaybePlayerProgressStatUI?):
            PlayerProgressStatUI.Update(InPlayer)
            PlayerProgressStatUI.Show()

        if (PlayerProgressWeaponUI := MaybePlayerProgressWeaponUI?):
            PlayerProgressWeaponUI.Update(InPlayer)
            PlayerProgressWeaponUI.Show()

    ShowRankInfo():void =
        RankInfoUI.SetVisibility(widget_visibility.Visible)

        if (PlayerProgressStatUI := MaybePlayerProgressStatUI?):
            PlayerProgressStatUI.Hide()

        if (PlayerProgressWeaponUI := MaybePlayerProgressWeaponUI?):
            PlayerProgressWeaponUI.Hide()

    <# ----------------------------------------------------------------------------------------------------
        Event Handlers
    ---------------------------------------------------------------------------------------------------- #>
    OnShow<override>():void =
        Update()

    AwaitPlayerButtonClicked(Listenable:listenable(widget_message), Index:int)<suspends>:void =
        loop:
            Listenable.Await()
            if (InPlayer := GetAllPlayers()[Index]):
                SetPlayer(InPlayer)

    OnClickRankInfo(Message:widget_message):void =
        ShowRankInfo()