using { /Fortnite.com/UI }
using { /Verse.org/Colors }
using { /Verse.org/Simulation }
using { /UnrealEngine.com/Temporary/SpatialMath }
using { /UnrealEngine.com/Temporary/UI }

player_progress_weapon_ui := class(ui_component):

    var PlayerProgressWeaponEntryUIs:[]player_progress_weapon_entry_ui = array{}

    <# ----------------------------------------------------------------------------------------------------
        Initialization
    ---------------------------------------------------------------------------------------------------- #>
    CreateUI<override>():widget =
        overlay:
            Slots := array:
                overlay_slot:
                    HorizontalAlignment := horizontal_alignment.Fill
                    VerticalAlignment := vertical_alignment.Fill
                    Widget := MakeBox(?Width := 600.0)
                overlay_slot:
                    HorizontalAlignment := horizontal_alignment.Fill
                    VerticalAlignment := vertical_alignment.Fill
                    Padding := margin{Left := 40.0, Top := 40.0, Right := 40.0, Bottom := 40.0}
                    Widget := stack_box:
                        Orientation := orientation.Vertical
                        Slots := array:
                            stack_box_slot:
                                VerticalAlignment := vertical_alignment.Fill
                                Padding := margin{Left := 0.0, Top := 0.0, Right := 0.0, Bottom := 20.0}
                                Widget := text_block:
                                    DefaultText := MakeMessage("Weapon Progress")
                                    DefaultTextColor := ThemeColors.PrimaryText
                                    DefaultShadowColor := NamedColors.Black
                                    DefaultShadowOffset := option{vector2{X := 2.0, Y := 2.0}}
                                    DefaultShadowOpacity := 0.5
                            stack_box_slot:
                                HorizontalAlignment := horizontal_alignment.Fill
                                Widget := stack_box:
                                    Orientation := orientation.Vertical
                                    Slots := for (Index -> WeaponId : GetWeaponIds()):
                                        PlayerProgressWeaponEntryUI := player_progress_weapon_entry_ui{Player := Player, WeaponId := WeaponId}
                                        PlayerProgressWeaponEntryUI.Initialize()
                                        set PlayerProgressWeaponEntryUIs += array{PlayerProgressWeaponEntryUI}

                                        stack_box_slot:
                                            HorizontalAlignment := horizontal_alignment.Fill
                                            Padding := margin{Left := 0.0, Top := (Index = 0) and 0.0 or 5.0, Right := 0.0, Bottom := 0.0}
                                            Widget := PlayerProgressWeaponEntryUI.GetUI()

    <# ----------------------------------------------------------------------------------------------------
        Update Functions
    ---------------------------------------------------------------------------------------------------- #>
    Update(InPlayer:player):void =
        if (WeaponProgress := InPlayer.GetData[].WeaponProgress):
            for (Index -> PlayerProgressWeaponEntryUI : PlayerProgressWeaponEntryUIs):
                Progress := WeaponProgress[PlayerProgressWeaponEntryUI.WeaponId.ToInt()] or 0
                PlayerProgressWeaponEntryUI.Update(InPlayer, Progress)