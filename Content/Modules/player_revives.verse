using { /Fortnite.com/Characters }
using { /Fortnite.com/Devices }
using { /Verse.org/Simulation }
using { /UnrealEngine.com/Temporary/SpatialMath }
using { /UnrealEngine.com/Temporary/UI }

<# ----------------------------------------------------------------------------------------------------
    Constants
---------------------------------------------------------------------------------------------------- #>
BaseReviveRegenerationTime:float = 180.0

<# ----------------------------------------------------------------------------------------------------
    player_revives
---------------------------------------------------------------------------------------------------- #>
player_revives := class:
    Player:player

    RechargeCancelledEvent:event(void) = event(void){}
    EndEvent:event(void) = event(void){}

    var IsRecharging:logic = false
    var CinematicSequencePlayerRevives:cinematic_sequence_device = cinematic_sequence_device{}

    <# ----------------------------------------------------------------------------------------------------
        Initialization
    ---------------------------------------------------------------------------------------------------- #>
    Initialize()<suspends>:void =
        WaitForGameManager()
        if (GameManager := GetGameManager[]):
            GameManager.GameStartedEvent.Subscribe(Reset)
            GameManager.GameEndedEvent.Subscribe(Reset)

        if (PlayerStats := Player.GetStats[]):
            PlayerStats.PlayerStatsChangedEvent.Subscribe(OnPlayerStatsChanged)

    <# ----------------------------------------------------------------------------------------------------
        Get Functions
    ---------------------------------------------------------------------------------------------------- #>
    GetMaxCharge()<transacts>:int =
        Floor[Player.GetPowerUpValue(power_up_id.Revive)] or 0

    <# ----------------------------------------------------------------------------------------------------
        Update Functions
    ---------------------------------------------------------------------------------------------------- #>
    Assign(CinematicSequence:cinematic_sequence_device):void =
        set CinematicSequencePlayerRevives = CinematicSequence
        CinematicSequencePlayerRevives.Stop()
        CinematicSequencePlayerRevives.SetPlaybackTime(0.0)

    Recharge()<suspends>:void =
        set IsRecharging = true
        defer:
            set IsRecharging = false

        race:
            RechargeCancelledEvent.Await()

            loop:
                CooldownTime := Player.GetPowerUpValue(power_up_id.ReviveRegeneration)
                CinematicSequencePlayerRevives.SetPlayRate(1.0 / CooldownTime)
                CinematicSequencePlayerRevives.SetPlaybackTime(0.0)
                CinematicSequencePlayerRevives.Play(Player)
                Sleep(CooldownTime)
                Player.AddPlayerStats(player_stat_id.Revives, 1)
                CinematicSequencePlayerRevives.Pause(Player)
                Sleep(TickTime)

                if (Character := Player.GetFortCharacter[]):
                    if (Character.IsDownButNotOut[]):
                        Player.AutoRevive()
                    else:
                        SpawnVFX(vfx_id.Revive, Character.GetTransform().Translation.Make2D(), ?Duration := 3.0)

                if (Player.GetPlayerStats(player_stat_id.Revives) >= GetMaxCharge()):
                    CinematicSequencePlayerRevives.Stop(Player)
                    CinematicSequencePlayerRevives.SetPlaybackTime(0.0)
                    break
                else:

    Reset():void =
        RechargeCancelledEvent.Signal()
        CinematicSequencePlayerRevives.Stop(Player)
        CinematicSequencePlayerRevives.SetPlaybackTime(0.0)

    <# ----------------------------------------------------------------------------------------------------
        Event Handlers
    ---------------------------------------------------------------------------------------------------- #>       
    OnPlayerStatsChanged(Id:player_stat_id, NewValue:int):void =
        if (Id = player_stat_id.Revives):
            Player.ShowMessage(message_id.Revives, "{Player.GetPlayerStats(player_stat_id.Revives)}", ?DisplayTime := 0.0)

            if (Player.GetPowerUpLevel(power_up_id.ReviveRegeneration) > 0, NewValue < GetMaxCharge(), not IsRecharging?):
                spawn{Recharge()}
            else if (Character := Player.GetFortCharacter[], Character.IsDownButNotOut[]):
                spawn{Player.AutoRevive()}