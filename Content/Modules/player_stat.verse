using { /Fortnite.com/Characters }
using { /Verse.org/Random }
using { /Verse.org/Simulation }

<# ----------------------------------------------------------------------------------------------------
    Constants
---------------------------------------------------------------------------------------------------- #>
BaseRerolls:int = 3
BaseWeaponRerolls:int = 1
BaseBoonRerolls:int = 1
DefaultVulnerabilityBufferTime:float = 1.5
PlayerDamagedBufferTime:float = 1.0

<# ----------------------------------------------------------------------------------------------------
    Data Definitions
---------------------------------------------------------------------------------------------------- #>        
player_game_stat_id := enum:
    EnemiesEliminated
    EliminatedMiniboss
    DamageDealt
    DamageDealtMiniboss
    DamageDealtBoss
    WoodCollected
    PlayersRevived
    AutoRevived
    TimesDowned

player_stat_id := enum:
    Reroll
    WeaponReroll
    BoonReroll
    Boost
    Revives

<# ----------------------------------------------------------------------------------------------------
    Global Functions
---------------------------------------------------------------------------------------------------- #>
(Player:player).GetStats()<decides><transacts>:player_stat =
    GetPlayerManager[].GetPlayerStat[Player]

(Player:player).GetAbilityLevels()<transacts>:[ability_id][]int =
    Player.GetStats[].GetAbilityLevels() or map{}

(Player:player).GetAbilityLevel(Id:ability_id)<transacts>:int =
    Player.GetStats[].GetAbilityLevel(Id) or 0

(Player:player).GetAbilityValue(AbilityInfo:ability_info)<transacts>:float =
    if (WeaponId := AbilityInfo.MaybeWeaponId?):
        Player.GetAbilityValue(WeaponId, AbilityInfo.Id)
    else:
        Player.GetAbilityValue(AbilityInfo.Id)

(Player:player).GetTotalAbilityValue(AbilityInfo:ability_info)<transacts>:float =
    var AbilityValue:float = Player.GetAbilityValue(AbilityInfo)

    if (not AbilityInfo.MaybeWeaponId?, PowerUpInfo := GetPowerUpInfo[AbilityInfo.Id], PowerUpInfo.StatType = power_up_stat_type.Add):
        set AbilityValue += Player.GetPowerUpValue(PowerUpInfo.Id)
    
    AbilityValue

(Player:player).CalculateTotalAbilityValue(AbilityInfo:ability_info, AbilityValue:float)<transacts>:float =
    var TotalAbilityValue:float = AbilityValue

    if (not AbilityInfo.MaybeWeaponId?, PowerUpInfo := GetPowerUpInfo[AbilityInfo.Id], PowerUpInfo.StatType = power_up_stat_type.Add):
        set TotalAbilityValue += Player.GetPowerUpValue(PowerUpInfo.Id)

    TotalAbilityValue

(Player:player).GetAbilityValue(Id:ability_id)<transacts>:float =
    Player.GetStats[].GetAbilityValue(Id) or 0.0

(Player:player).GetAbilityValue(WeaponId:weapon_id, AbilityId:ability_id)<transacts>:float =
    Player.GetStats[].GetAbilityValue(WeaponId, AbilityId) or 0.0

(Player:player).AddAbility(Id:ability_info, ?Amount:int = 1):void =
    if (PlayerStat := Player.GetStats[]):
        PlayerStat.AddAbility(Id, Amount)

(Player:player).SubtractAbility(Id:ability_id, ?Amount:int = 1):void =
    if (PlayerStat := Player.GetStats[]):
        PlayerStat.SubtractAbility(Id, Amount)

(Player:player).AddWeaponEvolution(AbilityInfo:ability_info):void =
    if (PlayerStat := Player.GetStats[]):
        PlayerStat.AddWeaponEvolution(AbilityInfo)

(Player:player).GetBanishedAbilities()<transacts>:[]ability_info =
    Player.GetStats[].GetBanishedAbilities() or array{}

(Player:player).BanishAbility(AbilityInfo:ability_info):void =
    if (PlayerStat := Player.GetStats[]):
        PlayerStat.BanishAbility(AbilityInfo)

(Player:player).AddGameStat(Id:player_game_stat_id, Amount:float):void =
    if (PlayerStat := Player.GetStats[]):
        PlayerStat.AddGameStat(Id, Amount)

(Player:player).GetGameStat(Id:player_game_stat_id)<transacts>:float =
    Player.GetStats[].GetGameStat(Id) or 0.0

(Player:player).ResetStats():void =
    if (PlayerStat := Player.GetStats[]):
        PlayerStat.ResetStats()

(Player:player).GetBanish()<transacts>:int =
    Player.GetStats[].GetBanish() or 0

(Player:player).GetPlayerStats(Id:player_stat_id)<transacts>:int =
    Player.GetStats[].GetPlayerStats[Id] or 0

(Player:player).AddPlayerStats(Id:player_stat_id, Amount:int):void =
    if (PlayerStats := Player.GetStats[]):
        PlayerStats.AddPlayerStats(Id, Amount)
    
(Player:player).RemovePlayerStats(Id:player_stat_id, Amount:int):void =
    if (PlayerStats := Player.GetStats[]):
        PlayerStats.RemovePlayerStats(Id, Amount)

(Player:player).PutInStasis():void =
    Player.PutInStasis(?IsAllowFalling := true)

(Player:player).PutInStasis(?IsAllowFalling:logic = true):void =
    if (PlayerStats := Player.GetStats[]):
        PlayerStats.PutInStasis(?IsAllowFalling := IsAllowFalling)

(Player:player).ReleaseFromStasis():void =
    if (PlayerStats := Player.GetStats[]):
        PlayerStats.ReleaseFromStasis()

(Player:player).IsInStasis()<transacts>:logic =
    Player.GetStats[].GetIsInStasis() or false

(Player:player).EquipWeapon(Id:weapon_id):void =
    if (PlayerStats := Player.GetStats[]):
        PlayerStats.EquipWeapon(Id)

(Player:player).EquipNewWeapon(Id:weapon_id):void =
    if (PlayerStats := Player.GetStats[]):
        PlayerStats.EquipNewWeapon(Id)

(Player:player).GetEquippedWeapon()<transacts>:?weapon_id =
    Player.GetStats[].GetEquippedWeapon() or option{DefaultWeapon}

(Player:player).GetWeaponLevels()<transacts>:[weapon_id][]tuple(ability_id, int) =
    Player.GetStats[].GetWeaponLevels() or map{}

(Player:player).GetWeaponLevel(Id:weapon_id)<transacts>:int =
    Player.GetStats[].GetWeaponLevel(Id) or 0

(Player:player).GetWeaponCount()<transacts>:int =
    Player.GetStats[].GetWeaponCount() or 1

(Player:player).HasWeapon(Id:weapon_id)<transacts>:logic =
    Player.GetStats[].HasWeapon(Id) or false

(Player:player).GetWeaponEvolutionLevel(Id:weapon_id)<transacts>:int =
    Player.GetStats[].GetWeaponEvolutionLevel(Id) or 0

(Player:player).GetWeaponEvolutions()<transacts>:[weapon_id][]ability_info =
    Player.GetStats[].GetWeaponEvolutions() or map{}

(Player:player).GetWeaponEvolution(Id:weapon_id)<transacts>:[]ability_info =
    Player.GetStats[].GetWeaponEvolution[Id] or array{}

(Player:player).GetWeaponEvolutionValue(WeaponId:weapon_id, Id:ability_id)<transacts>:float =
    var BaseValue:float = 0.0
    var IsFound:logic = false
    
    if (WeaponInfo := GetWeaponInfo[WeaponId]):
        for (EvolutionInfos : WeaponInfo.Evolution, not IsFound?):
            for (EvolutionInfo : EvolutionInfos, not IsFound?, EvolutionInfo.Id = Id):
                set BaseValue = EvolutionInfo.BaseValue
                set IsFound = true

    TotalTierValue := SumFloat of for (Index -> AbilityInfo : Player.GetWeaponEvolution(WeaponId), AbilityInfo.Id = Id, TierValue := AbilityInfo.TierValue[0]?) {TierValue}

    BaseValue + TotalTierValue

(Player:player).IsWeaponEvolutionAbilityActive(WeaponId:weapon_id, Id:ability_id)<transacts>:logic =
    logic{for (AbilityInfo : Player.GetWeaponEvolution(WeaponId), AbilityInfo.Id = Id) {AbilityInfo}[0]}

<#
(Player:player).IsBanished(AbilityInfo:ability_info)<transacts>:logic =
    Player.GetStats[].IsBanished(AbilityInfo) or false
#>

(Player:player).GetIsAFK()<transacts>:logic =
    Player.GetStats[].GetIsAFK() or false

(Player:player).SetAFK(IsAFK:logic):void =
    if (PlayerStats := Player.GetStats[]):
        PlayerStats.SetAFK(IsAFK)

(Player:player).Stun(Duration:float):void =
    if (Character := Player.GetFortCharacter[].IsVulnerable[], PlayerStats := Player.GetStats[]):
        spawn{PlayerStats.Stun(Duration)}

(Player:player).SetVulnerability(IsVulnerable:logic, ?BufferTime:float = 0.0):void =
    if (PlayerStats := Player.GetStats[]):
        PlayerStats.SetVulnerability(IsVulnerable, ?BufferTime := BufferTime)

CalculateWeaponLevel(Id:weapon_id, WeaponLevels:[]tuple(ability_id, int))<transacts>:int =
    var WeaponLevelFromAbilities:int = 0
    if (AbilityInfo := GetAbilityManager[].GetWeaponAbilityInfo(Id, ability_id.WeaponLevel)?):
        for:
            WeaponLevel : WeaponLevels
            WeaponLevel(0) = ability_id.WeaponLevel
            AbilityValue := Floor[AbilityInfo.TierValue[WeaponLevel(1) - 1]?]
        do:
            set WeaponLevelFromAbilities += AbilityValue

    WeaponLevels.Length + WeaponLevelFromAbilities

CalculateWeaponEvolutionLevel(WeaponLevel:int)<transacts>:int =
    if (WeaponLevel >= 18). 3
    else if (WeaponLevel >= 12). 2
    else if (WeaponLevel >= 6). 1
    else. 0

(Player:player).ApplyDamage(Amount:float, ?IsBoss:logic = false):void =
    if (PlayerStats := Player.GetStats[], Character := Player.GetFortCharacter[]):
        LastDamagedTime := IsBoss? and PlayerStats.LastDamagedTimeBoss or PlayerStats.LastDamagedTime
        ElapsedTime := GetSimulationElapsedTime() - LastDamagedTime
        if (ElapsedTime >= PlayerDamagedBufferTime):
            if (Player.HasBoon(boon_id.NinjaRun)?, DodgeChance := BoonGetValueNinjaRun(Player), GetRandomFloat(0.0, 1.0) <= DodgeChance):
                SpawnVFX(vfx_id.Dodged, Character.GetTransform().Translation.Make2D(), ?Duration := 1.0)
            else:
                Character.Damage(Amount)

            if (IsBoss?):
                set PlayerStats.LastDamagedTimeBoss = GetSimulationElapsedTime()
            else:
                set PlayerStats.LastDamagedTime = GetSimulationElapsedTime()

(Player:player).GetBoons()<transacts>:[]boon_id =
    Player.GetStats[].GetBoons() or array{}

(Player:player).HasBoon(Id:boon_id)<transacts>:logic =
    logic{Player.GetStats[].GetBoons().Find[Id]}

(Player:player).AddBoon(BoonInfo:boon_info):void =
    if (PlayerStats := Player.GetStats[]):
        PlayerStats.AddBoon(BoonInfo)

<# ----------------------------------------------------------------------------------------------------
    Constructor
---------------------------------------------------------------------------------------------------- #>
MakePlayerStats<constructor>(UpdatedPlayerStats:player_stat)<transacts> := player_stat:
    Player := UpdatedPlayerStats.Player
    AbilityLevels := UpdatedPlayerStats.AbilityLevels
    BanishedAbilities := UpdatedPlayerStats.BanishedAbilities
    MaybeEquippedWeapon := UpdatedPlayerStats.MaybeEquippedWeapon
    WeaponLevels := UpdatedPlayerStats.WeaponLevels
    WeaponEvolutions := UpdatedPlayerStats.WeaponEvolutions
    Boons := UpdatedPlayerStats.Boons
    GameStats := UpdatedPlayerStats.GameStats
    PlayerStats := UpdatedPlayerStats.PlayerStats

<# ----------------------------------------------------------------------------------------------------
    player_stat
---------------------------------------------------------------------------------------------------- #>
player_stat := class:
    Player:player

    AbilityChangedEvent<public>:events(ability_id) = events(ability_id){}
    StatsResetEvent<public>:events(tuple()) = events(tuple()){}
    WeaponEquippedEvent<public>:events(tuple()) = events(tuple()){}
    WeaponEvolvedEvent<public>:events(tuple()) = events(tuple()){}
    CancelUpdateVulnerabilityEvent:events(tuple()) = events(tuple()){}
    PlayerStatsChangedEvent<public>:events(tuple(player_stat_id, int)) = events(tuple(player_stat_id, int)){}

    # Temporary stats
    # [Reminder] Constructor needs to be updated if new stats are added.
    var AbilityLevels:[ability_id][]int = map{}
    var BanishedAbilities:[]ability_info = array{}
    var MaybeEquippedWeapon:?weapon_id = option{DefaultWeapon}
    var WeaponLevels:[weapon_id][]tuple(ability_id, int) = map{}
    var WeaponEvolutions:[weapon_id][]ability_info = map{}
    var Boons:[]boon_id = array{}
    var GameStats:[player_game_stat_id]float = map{}
    var PlayerStats:[player_stat_id]int = map{}

    # Statuses
    var IsInStasis<private>:logic = false
    var IsStunned<private>:logic = false
    var IsAutoReviving:logic = false
    var IsAFK<private>:logic = false
    var LastDamagedTime:float = 0.0
    var LastDamagedTimeBoss:float = 0.0

    <# ----------------------------------------------------------------------------------------------------
        Get Functions
    ---------------------------------------------------------------------------------------------------- #>
    GetPlayerStats(Id:player_stat_id)<decides><transacts>:int =
        PlayerStats[Id]

    GetBanish()<transacts>:int =
        (Floor[Player.GetPowerUpValue(power_up_id.AbilityBanish)] or 0) - BanishedAbilities.Length

    GetAbilityLevels()<transacts>:[ability_id][]int =
        AbilityLevels

    GetAbilityLevel(Id:ability_id)<transacts>:int =
        var AbilityLevel:int = 0
        for (Level : GetAbilityLevels()[Id]):
            set AbilityLevel += Level
        AbilityLevel

    GetAbilityValue(Id:ability_id)<transacts>:float =
        CalculateAbilityValue(Id, AbilityLevels[Id] or array{})

    CalculateAbilityValue(Id:ability_id, AbilityLevel:[]int, ?IncludeBaseValue:logic = true)<transacts>:float =
        var AbilityValue:float = 0.0
        
        if (AbilityInfo := GetAbilityInfo[Id]):
            if (IncludeBaseValue?):
                set AbilityValue += AbilityInfo.BaseValue

            for (Level : AbilityLevel, TierValue := AbilityInfo.TierValue[Level - 1]?):
                set AbilityValue += TierValue

        if (PowerUpInfo := GetPowerUpInfo[Id], PowerUpInfo.StatType = power_up_stat_type.Multiply):
            set AbilityValue *= 1.0 + Player.GetPowerUpValue(PowerUpInfo.Id)
                
        AbilityValue

    GetAbilityValue(WeaponId:weapon_id, AbilityId:ability_id)<transacts>:float =
        CalculateAbilityValue(WeaponId, AbilityId, WeaponLevels[WeaponId] or array{})

    GetAbilityEvolutionValue(WeaponId:weapon_id, AbilityId:ability_id)<transacts>:float =
        SumFloat of for (AbilityInfo : Player.GetWeaponEvolution(WeaponId), AbilityInfo.Id = AbilityId, TierValue := AbilityInfo.TierValue[0]?) {TierValue}

    CalculateAbilityValue(WeaponId:weapon_id, AbilityId:ability_id, InWeaponLevels:[]tuple(ability_id, int))<transacts>:float =
        var AbilityValue:float = 0.0
        
        if (AbilityManager := GetAbilityManager[], AbilityInfo := AbilityManager.GetWeaponAbilityInfo(WeaponId, AbilityId)?):
            set AbilityValue += AbilityInfo.BaseValue

            for (WeaponLevel : InWeaponLevels, WeaponLevel(0) = AbilityId, TierValue := AbilityInfo.TierValue[WeaponLevel(1) - 1]?):
                set AbilityValue += TierValue
                
        AbilityValue

    GetGameStat(Id:player_game_stat_id)<transacts>:float =
        GameStats[Id] or 0.0

    GetIsInStasis()<transacts>:logic =
        (player_stat:)IsInStasis

    GetEquippedWeapon()<transacts>:?weapon_id =
        MaybeEquippedWeapon

    GetBanishedAbilities()<transacts>:[]ability_info =
        BanishedAbilities

    GetWeaponLevels()<transacts>:[weapon_id][]tuple(ability_id, int) =
        WeaponLevels

    GetWeaponCount()<transacts>:int =
        WeaponLevels.Length

    GetWeaponLevel(Id:weapon_id)<transacts>:int =
        CalculateWeaponLevel(Id, WeaponLevels[Id] or array{})

    GetWeaponEvolutionLevel(Id:weapon_id)<transacts>:int =
        CalculateWeaponEvolutionLevel(GetWeaponLevel(Id))

    GetWeaponEvolutions()<transacts>:[weapon_id][]ability_info =
        WeaponEvolutions

    GetWeaponEvolution(Id:weapon_id)<decides><transacts>:[]ability_info =
        WeaponEvolutions[Id]

    HasWeapon(Id:weapon_id)<transacts>:logic =
        logic{WeaponLevels[Id]}

    IsBanished(AbilityInfo:ability_info)<transacts>:logic =
        var IsAbilityBanished:logic = false

        for:
            BanishedAbility : BanishedAbilities
            not IsAbilityBanished?
            AbilityInfo.Id = BanishedAbility.Id and AbilityInfo.MaybeWeaponId = BanishedAbility.MaybeWeaponId
        do:
            set IsAbilityBanished = true

        return IsAbilityBanished

    GetIsAFK()<transacts>:logic =
        IsAFK

    GetBoons()<transacts>:[]boon_id =
        Boons

    <# ----------------------------------------------------------------------------------------------------
        Update Functions
    ---------------------------------------------------------------------------------------------------- #>
    AddAbility(AbilityInfo:ability_info, Amount:int):void =
        if:
            WeaponId := AbilityInfo.MaybeWeaponId?
            set WeaponLevels[WeaponId] += array{(AbilityInfo.Id, Amount)}
        then:
            AbilityChangedEvent.Signal(AbilityInfo.Id)
            return

        if (not AbilityLevels[AbilityInfo.Id], set AbilityLevels[AbilityInfo.Id] = array{Amount}):
            AbilityChangedEvent.Signal(AbilityInfo.Id)
            return
        
        if:
            set AbilityLevels[AbilityInfo.Id] += array{Amount}
            AbilityLevel := AbilityLevels[AbilityInfo.Id]
        then:
            AbilityChangedEvent.Signal(AbilityInfo.Id)

    SubtractAbility(Id:ability_id, Amount:int):void = {}

    AddWeaponEvolution(AbilityInfo:ability_info):void =
        if (WeaponId := AbilityInfo.MaybeWeaponId?):
            if:
                not WeaponEvolutions[WeaponId]
                set WeaponEvolutions[WeaponId] = array{}

            if:
                set WeaponEvolutions[WeaponId] += array{AbilityInfo}
            then:
                Player.AwardAccolade(accolade_id.AbilityEvolved)
                Player.PlayAudio(audio_id.WeaponEvolution)
                WeaponEvolvedEvent.Signal()

    AddGameStat(Id:player_game_stat_id, Amount:float):void =
        if (not GameStats[Id], set GameStats[Id] = Amount):
            return

        if (set GameStats[Id] += Amount) {}
    
    # Called on game start/end.
    (player_stat:)ResetStats():void =
        set AbilityLevels = map{}
        set WeaponEvolutions = map{}
        set GameStats = map{}
        set Boons = array{}
        set PlayerStats = map:
            player_stat_id.Reroll => BaseRerolls + Floor[Player.GetPowerUpValue(power_up_id.AbilityReroll)] or 0
            player_stat_id.WeaponReroll => BaseWeaponRerolls + Floor[Player.GetPowerUpValue(power_up_id.WeaponReroll)] or 0
            player_stat_id.Boost => Floor[Player.GetPowerUpValue(power_up_id.AbilityBoost)] or 0
            player_stat_id.Revives => Floor[Player.GetPowerUpValue(power_up_id.Revive)] or 0
            player_stat_id.BoonReroll => BaseBoonRerolls + Floor[Player.GetPowerUpValue(power_up_id.BoonReroll)] or 0

        PlayerStatsChangedEvent.Signal(player_stat_id.Revives, Floor[Player.GetPowerUpValue(power_up_id.Revive)] or 0)
        set BanishedAbilities = array{}
        if (EquippedWeapon := MaybeEquippedWeapon?):
            set WeaponLevels = map{EquippedWeapon => array{}}

        StatsResetEvent.Signal()

    AddPlayerStats(Id:player_stat_id, Amount:int):void =
        if:
            UpdatedPlayerStats := PlayerStats[Id] + Amount
            set PlayerStats[Id] += Amount
        then:
            PlayerStatsChangedEvent.Signal(Id, UpdatedPlayerStats)

    RemovePlayerStats(Id:player_stat_id, Amount:int):void =
        if:
            UpdatedPlayerStats := Max(0, PlayerStats[Id] - Amount)
            set PlayerStats[Id] = UpdatedPlayerStats
        then:
            PlayerStatsChangedEvent.Signal(Id, UpdatedPlayerStats)

    PutInStasis(?IsAllowFalling:logic = true):void =
        if (Character := Player.GetFortCharacter[]):
            Character.PutInStasis(stasis_args{AllowTurning := false, AllowFalling := IsAllowFalling, AllowEmotes := false})
            set (player_stat:)IsInStasis = true

    ReleaseFromStasis():void =
        if (Character := Player.GetFortCharacter[]):
            Character.ReleaseFromStasis()
            set (player_stat:)IsInStasis = false

    EquipWeapon(Id:weapon_id):void =
        set MaybeEquippedWeapon = option{Id}
        WeaponEquippedEvent.Signal()
        if (ManagerWeaponEquippedEvent := GetPlayerManager[].WeaponEquippedEvent):
            ManagerWeaponEquippedEvent.Signal(Player)

    EquipNewWeapon(Id:weapon_id):void =
        if (set WeaponLevels[Id] = array{}) {}
        Player.AwardAccolade(accolade_id.NewWeapon)

    BanishAbility(AbilityInfo:ability_info):void =
        set BanishedAbilities += array{AbilityInfo}

    SetMaxAbilities():void =
        var MaxAbilityLevels:[ability_id][]int = map{}
        for (Id : GetAbilityIds()):
            if (set MaxAbilityLevels[Id] = array{5, 5, 5, 5, 5}) {}

        var MaxWeaponLevels:[weapon_id][]tuple(ability_id, int) = map{}
        for (Id -> WeaponLevel : WeaponLevels):
            UpdatedMaxWeaponLevels := array:
                (ability_id.Damage, 5)
                (ability_id.Damage, 5)
                (ability_id.Damage, 5)
                (ability_id.Cooldown, 5)
                (ability_id.Cooldown, 5)
                (ability_id.Cooldown, 5)
                (ability_id.Size, 5)
                (ability_id.Size, 5)
                (ability_id.Size, 5)

            if (set MaxWeaponLevels[Id] = UpdatedMaxWeaponLevels) {}

        set AbilityLevels = MaxAbilityLevels
        set WeaponLevels = MaxWeaponLevels

    SetAFK(AFK:logic):void =
        if (AFK = IsAFK) then return
        set IsAFK = AFK
        if (IsAFK?):
            Player.ActivateVFXPowerup(vfx_powerup_id.AFK)
        else:
            Player.DeactivateVFXPowerup(vfx_powerup_id.AFK)
        
    Stun(Duration:float)<suspends>:void =
        if (IsStunned?) then return
        set IsStunned = true

        if (Character := Player.GetFortCharacter[]):
            SpawnVFX(vfx_id.Stun, Character.GetTransform().Translation, ?Duration := Duration)
            Character.PutInStasis(stasis_args{AllowTurning := false, AllowFalling := true, AllowEmotes := false})
            Sleep(Duration)
            Character.ReleaseFromStasis()

        set IsStunned = false

    SetVulnerability(IsVulnerable:logic, ?BufferTime:float = 0.0):void =
        if (Player.IsInStageVolume()? and IsVulnerable?) then return

        CancelUpdateVulnerabilityEvent.Signal()
        if (BufferTime > 0.0):
            spawn{UpdateVulnerability(IsVulnerable, ?BufferTime := BufferTime)}
        else if (Character := Player.GetFortCharacter[]):
            Character.SetVulnerability(IsVulnerable)
            if (IsVulnerable?):
                Player.DeactivateVFXPowerup(vfx_powerup_id.Invulnerability)
            else:
                Player.ActivateVFXPowerup(vfx_powerup_id.Invulnerability)

    UpdateVulnerability(IsVulnerable:logic, ?BufferTime:float = 0.0)<suspends>:void =
        race:
            block:
                CancelUpdateVulnerabilityEvent.Await()

            if (Character := Player.GetFortCharacter[]):
                Sleep(BufferTime)
                Character.SetVulnerability(IsVulnerable)
                if (IsVulnerable?):
                    Player.DeactivateVFXPowerup(vfx_powerup_id.Invulnerability)
                else:
                    Player.ActivateVFXPowerup(vfx_powerup_id.Invulnerability)

    AddBoon(BoonInfo:boon_info):void =
        set Boons += array{BoonInfo.Id}
        if (OnGrant := BoonInfo.MaybeOnGrant?):
            OnGrant(Player)