using { /Verse.org/Simulation }

player_xp_bar := class:
    PlayerManager:player_manager
    Player:player

    <# ----------------------------------------------------------------------------------------------------
        Initialization
    ---------------------------------------------------------------------------------------------------- #>        
    Initialize():void =
        UpdateLevel()
        UpdateXP()

        PlayerManager.CinematicSequenceExperienceBar.Play()
        PlayerManager.CinematicSequenceExperienceBar.Pause()
        PlayerManager.CinematicSequenceExperienceBar.SetPlaybackTime(0.0)

        if (PlayerStats := Player.GetStats[]):
            PlayerManager.LeveledUpEvent.Subscribe(OnLeveledUp)
            PlayerManager.XPChangedEvent.Subscribe(OnXPChanged)
            PlayerStats.StatsResetEvent.Subscribe(OnStatsReset)

    <# ----------------------------------------------------------------------------------------------------
        Get Functions
    ---------------------------------------------------------------------------------------------------- #>

    <# ----------------------------------------------------------------------------------------------------
        Update Functions
    ---------------------------------------------------------------------------------------------------- #>
    UpdateLevel():void =
        Player.ShowMessage(message_id.ExperienceBarLevel, "{PlayerManager.GetLevel()}", ?DisplayTime := 0.0)

    UpdateXP():void =
        XPPercent := (PlayerManager.GetXP() * 1.0) / (PlayerManager.GetXPRequired[] * 1.0) or 0.0
        Player.ShowMessage(message_id.ExperienceBar, "{PlayerManager.GetXP()}/{PlayerManager.GetXPRequired[] or 0} ({FormatAsPercent(XPPercent)})", ?DisplayTime := 0.0)

        if (Progress := (PlayerManager.GetXP() * 1.0)/(PlayerManager.GetXPRequired[] * 1.0)):
            PlayerManager.CinematicSequenceExperienceBar.SetPlaybackTime(Progress)
        else:
            PlayerManager.CinematicSequenceExperienceBar.SetPlaybackTime(1.0)

    <# ----------------------------------------------------------------------------------------------------
        Event Handlers
    ---------------------------------------------------------------------------------------------------- #>
    OnLeveledUp():void =
        UpdateLevel()

    OnXPChanged(XP:int):void =
        UpdateXP()

    OnStatsReset():void =
        UpdateLevel()
        UpdateXP()