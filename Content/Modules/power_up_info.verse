using { /Verse.org/Assets }
using { /Verse.org/Simulation }

<# ----------------------------------------------------------------------------------------------------
    Data Definitions
---------------------------------------------------------------------------------------------------- #>
power_up_id := enum:
    Damage
    Cooldown
    Size
    PlayerMaxHealth
    PlayerHealthRegeneration
    PlayerItemPickupDistance
    PlayerCriticalRate
    PlayerCriticalDamage
    PlayerExperienceMultiplier
    AbilityReroll
    AbilityBanish
    AbilityBoost
    DashCount
    Luck
    GoldMultiplier
    WeaponReroll
    TreasureChestLuck
    Revive
    GatherRate
    TreasureChestFocus
    ReviveRegeneration
    AirDash
    BoonReroll
    TreasureChestTier
    ReviveBlast
    ShopkeepersMagnet
    NinjaMastery01
    NinjaMastery02
    NinjaMastery03
    NinjaMastery04
    NinjaMastery05
    NinjaMastery06
    NinjaMastery07
    NinjaMastery08
    NinjaMastery09

power_up_stat_type := enum:
    Add
    Multiply

ToString(Type:power_up_stat_type)<transacts>:string =
    case (Type):
        power_up_stat_type.Add => "Additive"
        power_up_stat_type.Multiply => "Multiplicative"

(Id:power_up_id).ToInt()<transacts>:int =
    case (Id):
        power_up_id.Damage => 0
        power_up_id.Cooldown => 1
        power_up_id.Size => 2
        power_up_id.PlayerMaxHealth => 3
        power_up_id.PlayerHealthRegeneration => 4
        power_up_id.PlayerItemPickupDistance => 5
        power_up_id.PlayerCriticalRate => 6
        power_up_id.PlayerCriticalDamage => 7
        power_up_id.PlayerExperienceMultiplier => 8
        power_up_id.AbilityReroll => 9
        power_up_id.AbilityBanish => 10
        power_up_id.AbilityBoost => 11
        power_up_id.DashCount => 12
        power_up_id.Luck => 13
        power_up_id.GoldMultiplier => 14
        power_up_id.WeaponReroll => 15
        power_up_id.TreasureChestLuck => 16
        power_up_id.Revive => 17
        power_up_id.GatherRate => 18
        power_up_id.TreasureChestFocus => 19
        power_up_id.ReviveRegeneration => 20
        power_up_id.AirDash => 21
        power_up_id.BoonReroll => 22
        power_up_id.TreasureChestTier => 23
        power_up_id.ReviveBlast => 24
        power_up_id.ShopkeepersMagnet => 25
        power_up_id.NinjaMastery01 => 26
        power_up_id.NinjaMastery02 => 27
        power_up_id.NinjaMastery03 => 28
        power_up_id.NinjaMastery04 => 29
        power_up_id.NinjaMastery05 => 30
        power_up_id.NinjaMastery06 => 31
        power_up_id.NinjaMastery07 => 32
        power_up_id.NinjaMastery08 => 33
        power_up_id.NinjaMastery09 => 34
        
(Id:int).ToPowerUpId()<transacts>:power_up_id =
    case (Id):
        0 => power_up_id.Damage
        1 => power_up_id.Cooldown
        2 => power_up_id.Size
        3 => power_up_id.PlayerMaxHealth
        4 => power_up_id.PlayerHealthRegeneration
        5 => power_up_id.PlayerItemPickupDistance
        6 => power_up_id.PlayerCriticalRate
        7 => power_up_id.PlayerCriticalDamage
        8 => power_up_id.PlayerExperienceMultiplier
        9 => power_up_id.AbilityReroll
        10 => power_up_id.AbilityBanish
        11 => power_up_id.AbilityBoost
        12 => power_up_id.DashCount
        13 => power_up_id.Luck
        14 => power_up_id.GoldMultiplier
        15 => power_up_id.WeaponReroll
        16 => power_up_id.TreasureChestLuck
        17 => power_up_id.Revive
        18 => power_up_id.GatherRate
        19 => power_up_id.TreasureChestFocus
        20 => power_up_id.ReviveRegeneration
        21 => power_up_id.AirDash
        22 => power_up_id.BoonReroll
        23 => power_up_id.TreasureChestTier
        24 => power_up_id.ReviveBlast
        25 => power_up_id.ShopkeepersMagnet
        26 => power_up_id.NinjaMastery01
        27 => power_up_id.NinjaMastery02
        28 => power_up_id.NinjaMastery03
        29 => power_up_id.NinjaMastery04
        30 => power_up_id.NinjaMastery05
        31 => power_up_id.NinjaMastery06
        32 => power_up_id.NinjaMastery07
        33 => power_up_id.NinjaMastery08
        34 => power_up_id.NinjaMastery09
        _ => power_up_id.Damage

power_up_info := struct<computes>:
    Id:power_up_id
    MaybeAbilityId:?ability_id = false
    Name:string
    Description:string
    Type:power_up_type
    Icon:texture
    IconSmall:texture
    BaseCost:float
    CostFactor:int = 2
    MaxLevel:int
    BaseValue:float = 0.0
    IncrementValue:float
    FormatValue:type{_(:float)<transacts>:string}
    IsDeltaValue:logic = true
    IsPermanent:logic = false
    StatType:power_up_stat_type = power_up_stat_type.Add

power_up_type := enum:
    AbilityBuff
    AbilitySelect
    Income
    Survival
    NinjaMastery
    
PowerUpTypes:[]power_up_type = array:
    power_up_type.AbilityBuff
    power_up_type.AbilitySelect
    power_up_type.Income
    power_up_type.Survival
    #power_up_type.NinjaMastery

ToString(PowerUpType:power_up_type)<transacts>:string =
    case (PowerUpType):
        power_up_type.AbilityBuff => "Ability Buff"
        power_up_type.AbilitySelect => "Ability Select"
        power_up_type.Income => "Income"
        power_up_type.Survival => "Survival"
        power_up_type.NinjaMastery => "Ninja Mastery"

<# ----------------------------------------------------------------------------------------------------
    PowerUps Info
---------------------------------------------------------------------------------------------------- #>
PowerUps:[power_up_id]power_up_info = map:
    power_up_id.Damage => power_up_info:
        Id := power_up_id.Damage
        MaybeAbilityId := option{ability_id.Damage}
        Name := "Damage"
        Description := "Increase base damage of all weapons."
        Type := power_up_type.AbilityBuff
        Icon := UserInterface.Icons.ProjectileDamage_128
        IconSmall := UserInterface.Icons.ProjectileDamage_64
        BaseCost := 100.0
        MaxLevel := 10
        IncrementValue := 0.1
        FormatValue := FormatAsPercent
    power_up_id.Cooldown => power_up_info:
        Id := power_up_id.Cooldown
        MaybeAbilityId := option{ability_id.Cooldown}
        Name := "Cooldown"
        Description := "Reduce the cooldown time of all weapons."
        Type := power_up_type.AbilityBuff
        Icon := UserInterface.Icons.AbilityCooldown_128
        IconSmall := UserInterface.Icons.AbilityCooldown_64
        BaseCost := 100.0
        MaxLevel := 10
        IncrementValue := 0.05
        FormatValue := FormatAsPercent
    power_up_id.Size => power_up_info:
        Id := power_up_id.Size
        MaybeAbilityId := option{ability_id.Size}
        Name := "Size"
        Description := "Increase the size of all weapons."
        Type := power_up_type.AbilityBuff
        Icon := UserInterface.Icons.AbilitySize_128
        IconSmall := UserInterface.Icons.AbilitySize_64
        BaseCost := 100.0
        MaxLevel := 10
        IncrementValue := 0.05
        FormatValue := FormatAsPercent
    power_up_id.PlayerMaxHealth => power_up_info:
        Id := power_up_id.PlayerMaxHealth
        MaybeAbilityId := option{ability_id.MaxHealth}
        Name := "Max Health"
        Description := "Increase max health."
        Type := power_up_type.AbilityBuff
        Icon := UserInterface.Icons.MaxHealth_128
        IconSmall := UserInterface.Icons.MaxHealth_64
        BaseCost := 100.0
        MaxLevel := 10
        IncrementValue := 0.1
        FormatValue := FormatAsPercent
        StatType := power_up_stat_type.Multiply
    power_up_id.PlayerHealthRegeneration => power_up_info:
        Id := power_up_id.PlayerHealthRegeneration
        MaybeAbilityId := option{ability_id.HealthRegeneration}
        Name := "Health Regeneration"
        Description := "Increase health regeneration rate."
        Type := power_up_type.AbilityBuff
        Icon := UserInterface.Icons.HealthRegeneration_128
        IconSmall := UserInterface.Icons.HealthRegeneration_64
        BaseCost := 100.0
        MaxLevel := 10
        IncrementValue := 0.1
        FormatValue := FormatAsPercent
        StatType := power_up_stat_type.Multiply
    power_up_id.PlayerItemPickupDistance => power_up_info:
        Id := power_up_id.PlayerItemPickupDistance
        MaybeAbilityId := option{ability_id.ItemPickupDistance}
        Name := "Pickup Range"
        Description := "Increase distance player can pickup EXP orbs."
        Type := power_up_type.AbilityBuff
        Icon := UserInterface.Icons.ItemPickupDistance_128
        IconSmall := UserInterface.Icons.ItemPickupDistance_64
        BaseCost := 100.0
        MaxLevel := 10
        IncrementValue := 0.1
        FormatValue := FormatAsPercent
    power_up_id.PlayerCriticalRate => power_up_info:
        Id := power_up_id.PlayerCriticalRate
        MaybeAbilityId := option{ability_id.CriticalRate}
        Name := "Critical Rate"
        Description := "Increase chance of critical hit."
        Type := power_up_type.AbilityBuff
        Icon := UserInterface.Icons.CriticalRate_128
        IconSmall := UserInterface.Icons.CriticalRate_64
        BaseCost := 100.0
        MaxLevel := 10
        IncrementValue := 0.05
        FormatValue := FormatAsPercent
    power_up_id.PlayerCriticalDamage => power_up_info:
        Id := power_up_id.PlayerCriticalDamage
        MaybeAbilityId := option{ability_id.CriticalDamage}
        Name := "Critical Damage"
        Description := "Increase critical damage multiplier."
        Type := power_up_type.AbilityBuff
        Icon := UserInterface.Icons.CriticalDamage_128
        IconSmall := UserInterface.Icons.CriticalDamage_64
        BaseCost := 100.0
        MaxLevel := 10
        IncrementValue := 0.05
        FormatValue := FormatAsPercent
    power_up_id.PlayerExperienceMultiplier => power_up_info:
        Id := power_up_id.PlayerExperienceMultiplier
        MaybeAbilityId := option{ability_id.ExperienceMultiplier}
        Name := "Experience Multiplier"
        Description := "Increase experience earned from orbs."
        Type := power_up_type.AbilityBuff
        Icon := UserInterface.Icons.ExperienceMultiplier_128
        IconSmall := UserInterface.Icons.ExperienceMultiplier_64
        BaseCost := 100.0
        MaxLevel := 10
        IncrementValue := 0.1
        FormatValue := FormatAsPercent

    power_up_id.AbilityReroll => power_up_info:
        Id := power_up_id.AbilityReroll
        Name := "Ability Reroll"
        Description := "Increase amount of rerolls per run. Reroll replaces the offered abilities with different ones."
        Type := power_up_type.AbilitySelect
        Icon := UserInterface.Icons.AbilityReroll_128
        IconSmall := UserInterface.Icons.AbilityReroll_64
        BaseCost := 500.0
        CostFactor := 2
        MaxLevel := 10
        IncrementValue := 1.0
        FormatValue := FormatAsInt
    power_up_id.AbilityBanish => power_up_info:
        Id := power_up_id.AbilityBanish
        Name := "Ability Banish"
        Description := "Increase amount of banishes per run. Banish stops the ability from being offered for the rest of the run."
        Type := power_up_type.AbilitySelect
        Icon := UserInterface.Icons.AbilityBanish_128
        IconSmall := UserInterface.Icons.AbilityBanish_64
        BaseCost := 750.0
        CostFactor := 2
        MaxLevel := 10
        IncrementValue := 1.0
        FormatValue := FormatAsInt
    power_up_id.AbilityBoost => power_up_info:
        Id := power_up_id.AbilityBoost
        Name := "Ability Boost"
        Description := "Increase amount of boosts per run. Boosts increases the tier of the selected ability (e.g. Epic > Legendary)."
        Type := power_up_type.AbilitySelect
        Icon := UserInterface.Icons.AbilityBoost_128
        IconSmall := UserInterface.Icons.AbilityBoost_64
        BaseCost := 1000.0
        CostFactor := 3
        MaxLevel := 5
        IncrementValue := 1.0
        FormatValue := FormatAsInt
    power_up_id.WeaponReroll => power_up_info:
        Id := power_up_id.WeaponReroll
        Name := "Weapon Reroll"
        Description := "Increase amount of weapon rerolls per run. Reroll replaces the offered weapons with different ones."
        Type := power_up_type.AbilitySelect
        Icon := UserInterface.Icons.RerollWeapon_128
        IconSmall := UserInterface.Icons.RerollWeapon_64
        BaseCost := 1500.0
        CostFactor := 3
        MaxLevel := 5
        IncrementValue := 1.0
        FormatValue := FormatAsInt
    power_up_id.BoonReroll => power_up_info:
        Id := power_up_id.BoonReroll
        Name := "Boon Reroll"
        Description := "Increase amount of boon rerolls per run. Reroll replaces the offered boons with different ones."
        Type := power_up_type.AbilitySelect
        Icon := UserInterface.Icons.BoonReroll_128
        IconSmall := UserInterface.Icons.BoonReroll_64
        BaseCost := 2000.0
        CostFactor := 3
        MaxLevel := 5
        IncrementValue := 1.0
        FormatValue := FormatAsInt
    power_up_id.Luck => power_up_info:
        Id := power_up_id.Luck
        MaybeAbilityId := option{ability_id.Luck}
        Name := "Luck"
        Description := "Increase chance of receiving higher tier abilities (chance of legendary is doubled at 50% luck). Doesn't apply to item shop."
        Type := power_up_type.AbilitySelect
        Icon := UserInterface.Icons.AbilityLuck_128
        IconSmall := UserInterface.Icons.AbilityLuck_64
        BaseCost := 750.0
        CostFactor := 2
        MaxLevel := 10
        IncrementValue := 0.1
        FormatValue := FormatAsPercent
    power_up_id.TreasureChestLuck => power_up_info:
        Id := power_up_id.TreasureChestLuck
        Name := "Treasure Luck"
        Description := "Increase chance of receiving higher tiers from Treasure Chests (stacks with Luck stats)."
        Type := power_up_type.AbilitySelect
        Icon := UserInterface.Icons.TreasureChestLuck_128
        IconSmall := UserInterface.Icons.TreasureChestLuck_64
        BaseCost := 500.0
        CostFactor := 2
        MaxLevel := 10
        IncrementValue := 0.2
        FormatValue := FormatAsPercent
    power_up_id.TreasureChestFocus => power_up_info:
        Id := power_up_id.TreasureChestFocus
        Name := "Treasure Focus"
        Description := "Treasure chests will only grants abilities that the player has selected during the game (from level up/item shop)."
        Type := power_up_type.AbilitySelect
        Icon := UserInterface.Icons.TreasureChestFocus_128
        IconSmall := UserInterface.Icons.TreasureChestFocus_64
        BaseCost := 25000.0
        CostFactor := 0
        MaxLevel := 1
        IncrementValue := 1.0
        FormatValue := FormatAsInt
    power_up_id.TreasureChestTier => power_up_info:
        Id := power_up_id.TreasureChestTier
        Name := "Treasure Tier"
        Description := "Increases the lowest tier that Treasure Chests grant (e.g. Treasure Chest will no longer grant Common tier at level 1)."
        Type := power_up_type.AbilitySelect
        Icon := UserInterface.Icons.TreasureChestTier_128
        IconSmall := UserInterface.Icons.TreasureChestTier_64
        BaseCost := 75000.0
        CostFactor := 3
        MaxLevel := 2
        IncrementValue := 1.0
        FormatValue := FormatAsTier
        IsDeltaValue := false

    power_up_id.DashCount => power_up_info:
        Id := power_up_id.DashCount
        Name := "Dash Count"
        Description := "Increase the max charge of dash ability."
        Type := power_up_type.Survival
        Icon := UserInterface.Icons.DashCooldown_128
        IconSmall := UserInterface.Icons.DashCooldown_64
        BaseCost := 1000.0
        CostFactor := 3
        MaxLevel := 4
        IncrementValue := 1.0
        FormatValue := FormatAsInt
    power_up_id.AirDash => power_up_info:
        Id := power_up_id.AirDash
        Name := "Air Dash"
        Description := "Enables the ability to activate an extra dash while airborne (can be used to jump over fences)."
        Type := power_up_type.Survival
        Icon := UserInterface.Icons.AirDash_128
        IconSmall := UserInterface.Icons.AirDash_64
        BaseCost := 5000.0
        CostFactor := 1
        MaxLevel := 1
        IncrementValue := 1.0
        FormatValue := FormatAsInt
    power_up_id.Revive => power_up_info:
        Id := power_up_id.Revive
        Name := "Revive"
        Description := "Automatically revive when you are downed."
        Type := power_up_type.Survival
        Icon := UserInterface.Icons.Revive_128
        IconSmall := UserInterface.Icons.Revive_64
        BaseCost := 50000.0
        CostFactor := 2
        MaxLevel := 2
        IncrementValue := 1.0
        FormatValue := FormatAsInt
    power_up_id.ReviveRegeneration => power_up_info:
        Id := power_up_id.ReviveRegeneration
        Name := "Revive Regeneration"
        Description := "Revives can regenerate after a cooldown (requires revive power up)."
        Type := power_up_type.Survival
        Icon := UserInterface.Icons.ReviveRegeneration_128
        IconSmall := UserInterface.Icons.ReviveRegeneration_64
        BaseCost := 1000.0
        CostFactor := 2
        MaxLevel := 10
        BaseValue := 180.0
        IncrementValue := -6.0
        FormatValue := FormatAsTime
    power_up_id.ReviveBlast => power_up_info:
        Id := power_up_id.ReviveBlast
        Name := "Revive Blast"
        Description := "Deals damage to enemies around you on auto revive. Damage and area increase with level."
        Type := power_up_type.Survival
        Icon := UserInterface.Icons.ReviveBlast_128
        IconSmall := UserInterface.Icons.ReviveBlast_64
        BaseCost := 2000.0
        CostFactor := 2
        MaxLevel := 10
        IncrementValue := 0.1
        FormatValue := FormatAsPercent

    power_up_id.GoldMultiplier => power_up_info:
        Id := power_up_id.GoldMultiplier
        Name := "Gold Multiplier"
        Description := "Increase amount of gold earned from completing a game."
        Type := power_up_type.Income
        Icon := UserInterface.Icons.Gold_128
        IconSmall := UserInterface.Icons.Gold_64
        BaseCost := 5000.0
        CostFactor := 2
        MaxLevel := 10
        IncrementValue := 0.2
        FormatValue := FormatAsPercent
    power_up_id.GatherRate => power_up_info:
        Id := power_up_id.GatherRate
        MaybeAbilityId := option{ability_id.GatherRate}
        Name := "Gather Rate"
        Description := "Increase wood gathering speed & amount (x2 effectiveness for gathering speed)."
        Type := power_up_type.Income
        Icon := UserInterface.Icons.GatherRate_128
        IconSmall := UserInterface.Icons.GatherRate_64
        BaseCost := 2000.0
        CostFactor := 2
        MaxLevel := 5
        IncrementValue := 0.05
        FormatValue := FormatAsPercent

    <#
    power_up_id.NinjaMastery01 => power_up_info:
        Id := power_up_id.NinjaMastery01
        Name := "Genin I"
        Description := "Increase Health & Damage."
        Type := power_up_type.NinjaMastery
        Icon := UserInterface.Icons.RankIcon_01_128
        IconSmall := UserInterface.Icons.RankIcon_01_64
        BaseCost := 2000.0
        CostFactor := 2
        MaxLevel := 10
        IncrementValue := 0.05
        FormatValue := FormatAsPercent
    power_up_id.NinjaMastery02 => power_up_info:
        Id := power_up_id.NinjaMastery02
        Name := "Genin II"
        Description := "Increase Health & Damage."
        Type := power_up_type.NinjaMastery
        Icon := UserInterface.Icons.RankIcon_02_128
        IconSmall := UserInterface.Icons.RankIcon_02_64
        BaseCost := 2000.0
        CostFactor := 2
        MaxLevel := 10
        IncrementValue := 0.05
        FormatValue := FormatAsPercent
    power_up_id.NinjaMastery03 => power_up_info:
        Id := power_up_id.NinjaMastery03
        Name := "Genin III"
        Description := "Increase Health & Damage."
        Type := power_up_type.NinjaMastery
        Icon := UserInterface.Icons.RankIcon_03_128
        IconSmall := UserInterface.Icons.RankIcon_03_64
        BaseCost := 2000.0
        CostFactor := 2
        MaxLevel := 10
        IncrementValue := 0.05
        FormatValue := FormatAsPercent
    power_up_id.NinjaMastery04 => power_up_info:
        Id := power_up_id.NinjaMastery04
        Name := "Chunin I"
        Description := "Increase Health & Damage."
        Type := power_up_type.NinjaMastery
        Icon := UserInterface.Icons.RankIcon_04_128
        IconSmall := UserInterface.Icons.RankIcon_04_64
        BaseCost := 2000.0
        CostFactor := 2
        MaxLevel := 10
        IncrementValue := 0.05
        FormatValue := FormatAsPercent
    power_up_id.NinjaMastery05 => power_up_info:
        Id := power_up_id.NinjaMastery05
        Name := "Chunin II"
        Description := "Increase Health & Damage."
        Type := power_up_type.NinjaMastery
        Icon := UserInterface.Icons.RankIcon_05_128
        IconSmall := UserInterface.Icons.RankIcon_05_64
        BaseCost := 2000.0
        CostFactor := 2
        MaxLevel := 10
        IncrementValue := 0.05
        FormatValue := FormatAsPercent
    power_up_id.NinjaMastery06 => power_up_info:
        Id := power_up_id.NinjaMastery06
        Name := "Chunin III"
        Description := "Increase Health & Damage."
        Type := power_up_type.NinjaMastery
        Icon := UserInterface.Icons.RankIcon_06_128
        IconSmall := UserInterface.Icons.RankIcon_06_64
        BaseCost := 2000.0
        CostFactor := 2
        MaxLevel := 10
        IncrementValue := 0.05
        FormatValue := FormatAsPercent
    power_up_id.NinjaMastery07 => power_up_info:
        Id := power_up_id.NinjaMastery07
        Name := "Jonin I"
        Description := "Increase Health & Damage."
        Type := power_up_type.NinjaMastery
        Icon := UserInterface.Icons.RankIcon_07_128
        IconSmall := UserInterface.Icons.RankIcon_07_64
        BaseCost := 2000.0
        CostFactor := 2
        MaxLevel := 10
        IncrementValue := 0.05
        FormatValue := FormatAsPercent
    power_up_id.NinjaMastery08 => power_up_info:
        Id := power_up_id.NinjaMastery08
        Name := "Jonin II"
        Description := "Increase Health & Damage."
        Type := power_up_type.NinjaMastery
        Icon := UserInterface.Icons.RankIcon_08_128
        IconSmall := UserInterface.Icons.RankIcon_08_64
        BaseCost := 2000.0
        CostFactor := 2
        MaxLevel := 10
        IncrementValue := 0.05
        FormatValue := FormatAsPercent
    power_up_id.NinjaMastery09 => power_up_info:
        Id := power_up_id.NinjaMastery09
        Name := "Jonin III"
        Description := "Increase Health & Damage."
        Type := power_up_type.NinjaMastery
        Icon := UserInterface.Icons.RankIcon_09_128
        IconSmall := UserInterface.Icons.RankIcon_09_64
        BaseCost := 2000.0
        CostFactor := 2
        MaxLevel := 10
        IncrementValue := 0.05
        FormatValue := FormatAsPercent
    #>
        
<# ----------------------------------------------------------------------------------------------------
    Global Functions
---------------------------------------------------------------------------------------------------- #>
GetPowerUpInfo()<transacts>:[power_up_id]power_up_info =
    PowerUps

GetPowerUpInfo(Id:power_up_id)<decides><transacts>:power_up_info =
    GetPowerUpInfo()[Id]

GetPowerUpIds()<transacts>:[]power_up_id =
    for (Id -> PowerUpInfo : GetPowerUpInfo()) {Id}

GetPowerUpIds(Type:power_up_type)<transacts>:[]power_up_id =
    for (Id -> PowerUpInfo : GetPowerUpInfo(), PowerUpInfo.Type = Type) {Id}

(Player:player).CanAffordPowerUp()<transacts>:logic = 
    var IsAffordPowerUp:logic = false
    for (PowerUpId -> PowerUpInfo : GetPowerUpInfo(), not IsAffordPowerUp?):
        if:
            Player.GetPowerUpLevel(PowerUpId) < PowerUpInfo.MaxLevel
            PowerUpCost := Player.GetPowerUpCost[PowerUpId, Player.GetPowerUpLevel(PowerUpId) + 1]
            Player.GetCurrencyAvailable(currency_id.Gold) >= PowerUpCost
        then:
            set IsAffordPowerUp = true

    IsAffordPowerUp

GetPowerUpInfo(InAbilityId:ability_id)<decides><transacts>:power_up_info =
    for (PowerUpInfo : GetPowerUpInfo(), AbilityId := PowerUpInfo.MaybeAbilityId?, AbilityId = InAbilityId) {PowerUpInfo} [0]

GetTotalPowerUpCost()<transacts>:float =
    var TotalPowerUpCost:float = 0.0
    for (PowerUpInfo : GetPowerUpInfo()):       
        for (PowerUpLevel := 1..PowerUpInfo.MaxLevel):
            set TotalPowerUpCost += (Floor[Pow(PowerUpLevel * 1.0, PowerUpInfo.CostFactor * 1.0) * PowerUpInfo.BaseCost] or 0) * 1.0
        
    TotalPowerUpCost