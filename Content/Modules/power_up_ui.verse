using { /Fortnite.com/UI }
using { /Verse.org/Colors }
using { /Verse.org/Simulation }
using { /UnrealEngine.com/Temporary/SpatialMath }
using { /UnrealEngine.com/Temporary/UI }

power_up_ui := class(panel_ui):
    Title<override>:string = "Power Ups"
    Anchors<override>:anchors = anchors{Minimum := vector2{X := 0.0, Y := 0.0}, Maximum := vector2{X := 1.0, Y := 1.0}}
    BackgroundOpacity<override>:type {_X:float where 0.000000 <= _X, _X <= 1.000000} = 0.7

    CurrencyText:text_block = text_block:
        DefaultTextColor := ThemeColors.PrimaryText

    NameText:text_block = text_block:
        DefaultTextColor := ThemeColors.PrimaryText

    Icon:texture_block = texture_block:
        DefaultImage := UserInterface.Icons.ProjectileDamage_128
        DefaultDesiredSize := vector2{X := 128.0, Y := 128.0}

    LevelLabelText:text_block = text_block:
        DefaultTextColor := ThemeColors.Highlight
        DefaultShadowColor := ThemeColors.Highlight * 0.125
        DefaultShadowOffset := option{vector2{X := 2.0, Y := 2.0}}
        DefaultShadowOpacity := 1.0

    LevelText:text_block = text_block:
        DefaultTextColor := ThemeColors.PrimaryText

    DescriptionText:text_block = text_block:
        DefaultTextColor := ThemeColors.PrimaryText
        DefaultTextOpacity := 0.5

    StatTypeText:text_block = text_block:
        DefaultTextColor := ThemeColors.Success
        DefaultShadowColor := ThemeColors.Success * 0.125
        DefaultShadowOffset := option{vector2{X := 2.0, Y := 2.0}}
        DefaultShadowOpacity := 1.0
            
    ValueText:text_block = text_block:
        DefaultTextColor := ThemeColors.PrimaryText
        DefaultShadowOffset := option{vector2{X := 2.0, Y := 2.0}}
        DefaultShadowOpacity := 0.5

    DeltaText:text_block = text_block:
        DefaultTextColor := ThemeColors.Success
        DefaultShadowColor := ThemeColors.Success * 0.125
        DefaultShadowOffset := option{vector2{X := 2.0, Y := 2.0}}
        DefaultShadowOpacity := 1.0

    CostText:text_block = text_block{}

    GridRow:int = 3
    GridColumn:int = 3

    var PowerUpEntryUIs:[]power_up_entry_ui = array{}
    var ActiveIndex:int = 0
    var ActiveType:power_up_type = power_up_type.AbilityBuff
            
    <# ----------------------------------------------------------------------------------------------------
        Initialization
    ---------------------------------------------------------------------------------------------------- #>
    CreateContent<override>():widget =
        ClosedEvent.Subscribe(OnClosed)

        RefundOneButton := button_quiet{DefaultText := MakeMessage("-1")}
        RefundOneButton.OnClick().Subscribe(OnClickRefundOne)

        RefundButton := button_quiet{DefaultText := MakeMessage("Refund Selected")}
        RefundButton.OnClick().Subscribe(OnClickRefund)

        RefundAllButton := button_quiet{DefaultText := MakeMessage("Refund All")}
        RefundAllButton.OnClick().Subscribe(OnClickRefundAll)

        stack_box:
            Orientation := orientation.Horizontal
            Slots := array:
                stack_box_slot:
                    HorizontalAlignment := horizontal_alignment.Fill
                    VerticalAlignment := vertical_alignment.Fill
                    Widget := stack_box:
                        Orientation := orientation.Vertical
                        Slots := array:
                            stack_box_slot:
                                Widget := color_block:
                                    DefaultColor := NamedColors.Black
                                    DefaultOpacity := 0.0
                                    DefaultDesiredSize := vector2{X := 240.0, Y := 1.0}
                            stack_box_slot:
                                HorizontalAlignment := horizontal_alignment.Fill
                                VerticalAlignment := vertical_alignment.Fill
                                Widget := overlay:
                                    Slots := array:
                                        overlay_slot:
                                            HorizontalAlignment := horizontal_alignment.Fill
                                            VerticalAlignment := vertical_alignment.Fill
                                            Padding := margin{Left := 0.0, Top := 0.0, Right := 0.0, Bottom := 0.0}
                                            Widget := MakeBox()
                                        overlay_slot:
                                            HorizontalAlignment := horizontal_alignment.Fill
                                            VerticalAlignment := vertical_alignment.Fill
                                            Padding := margin{Left := 20.0, Top := 10.0, Right := 20.0, Bottom := 10.0}
                                            Widget := stack_box:
                                                Orientation := orientation.Horizontal
                                                Slots := array:
                                                    stack_box_slot:
                                                        HorizontalAlignment := horizontal_alignment.Left
                                                        Widget := texture_block:
                                                            DefaultImage := currency_id.Gold.ToIcon()
                                                            DefaultDesiredSize := vector2{X := 40.0, Y := 40.0}
                                                    stack_box_slot:
                                                        HorizontalAlignment := horizontal_alignment.Left
                                                        Padding := margin{Left := 10.0, Top := 0.0, Right := 0.0, Bottom := 0.0}
                                                        Widget := CurrencyText
                            stack_box_slot:
                                HorizontalAlignment := horizontal_alignment.Fill
                                VerticalAlignment := vertical_alignment.Fill
                                Widget := stack_box:
                                    Orientation := orientation.Vertical
                                    Slots := for (PowerUpType : PowerUpTypes):
                                        MenuButton := button_quiet{DefaultText := MakeMessage("{PowerUpType}")}
                                        spawn{AwaitMenuButtonClick(MenuButton.OnClick(), PowerUpType)}
                                            
                                        stack_box_slot:
                                            HorizontalAlignment := horizontal_alignment.Fill
                                            VerticalAlignment := vertical_alignment.Fill
                                            Padding := margin{Left := 0.0, Top := 10.0, Right := 0.0, Bottom := 0.0}
                                            Widget := MenuButton
                stack_box_slot:
                    HorizontalAlignment := horizontal_alignment.Fill
                    VerticalAlignment := vertical_alignment.Fill
                    Padding := margin{Left := 20.0, Top := 0.0, Right := 0.0, Bottom := 0.0}
                    Widget := overlay:
                        Slots := array:
                            overlay_slot:
                                HorizontalAlignment := horizontal_alignment.Fill
                                VerticalAlignment := vertical_alignment.Fill
                                Widget := MakeBox()
                            overlay_slot:
                                HorizontalAlignment := horizontal_alignment.Fill
                                VerticalAlignment := vertical_alignment.Fill
                                Padding := margin{Left := 40.0, Top := 40.0, Right := 40.0, Bottom := 40.0}
                                Widget := stack_box:
                                    Orientation := orientation.Vertical
                                    Slots := array:
                                        stack_box_slot:
                                            HorizontalAlignment := horizontal_alignment.Fill
                                            VerticalAlignment := vertical_alignment.Fill
                                            Widget := CreateGrid()
                                        stack_box_slot:
                                            HorizontalAlignment := horizontal_alignment.Fill
                                            VerticalAlignment := vertical_alignment.Bottom
                                            Distribution := option{1.0}
                                            Widget := color_block:
                                                DefaultColor := NamedColors.White
                                                DefaultOpacity := 0.1
                                                DefaultDesiredSize := vector2{X := 0.0, Y := 1.0}
                                        stack_box_slot:
                                            HorizontalAlignment := horizontal_alignment.Fill
                                            VerticalAlignment := vertical_alignment.Bottom
                                            Padding := margin{Left := 0.0, Top := 20.0, Right := 0.0, Bottom := 0.0}
                                            Widget := stack_box:
                                                Orientation := orientation.Horizontal
                                                Slots := array:
                                                    stack_box_slot:
                                                        HorizontalAlignment := horizontal_alignment.Fill
                                                        VerticalAlignment := vertical_alignment.Fill
                                                        Distribution := option{1.0}
                                                        Padding := margin{Left := 0.0, Top := 0.0, Right := 10.0, Bottom := 0.0}
                                                        Widget := RefundButton
                                                    stack_box_slot:
                                                        HorizontalAlignment := horizontal_alignment.Fill
                                                        VerticalAlignment := vertical_alignment.Fill
                                                        Widget := RefundOneButton
                                        stack_box_slot:
                                            HorizontalAlignment := horizontal_alignment.Fill
                                            VerticalAlignment := vertical_alignment.Bottom
                                            Padding := margin{Left := 0.0, Top := 10.0, Right := 0.0, Bottom := 0.0}
                                            Widget := RefundAllButton

                stack_box_slot:
                    HorizontalAlignment := horizontal_alignment.Fill
                    VerticalAlignment := vertical_alignment.Fill
                    Distribution := option{1.0}
                    Padding := margin{Left := 20.0, Top := 0.0, Right := 0.0, Bottom := 0.0}
                    Widget := CreateInfoUI()

    CreateGrid():widget =
        stack_box:
            Orientation := orientation.Vertical
            Slots := for (RowIndex := 0..GridRow - 1):
                stack_box_slot:
                    HorizontalAlignment := horizontal_alignment.Fill
                    VerticalAlignment := vertical_alignment.Fill
                    Distribution := option{1.0}
                    Padding := margin{Left := 0.0, Top := (RowIndex = 0) and 0.0 or 10.0, Right := 0.0, Bottom := 0.0}
                    Widget := stack_box:
                        Orientation := orientation.Horizontal
                        Slots := for (ColumnIndex := 0..GridColumn - 1):
                            PowerUpEntryUI := power_up_entry_ui{Player := Player}
                            PowerUpEntryUI.Initialize()
                            spawn{AwaitButtonClicked(PowerUpEntryUI.ButtonClickedEvent, GridRow * RowIndex + ColumnIndex)}
                            set PowerUpEntryUIs += array{PowerUpEntryUI}
                            
                            stack_box_slot:
                                HorizontalAlignment := horizontal_alignment.Fill
                                VerticalAlignment := vertical_alignment.Fill
                                Distribution := option{1.0}
                                Padding := margin{Left := (ColumnIndex = 0) and 0.0 or 20.0, Top := 0.0, Right := 0.0, Bottom := 0.0}
                                Widget := PowerUpEntryUI.GetUI()
        
    CreateInfoUI():widget =
        LevelLabelText.SetText(MakeMessage("Lv."))

        overlay:
            Slots := array:
                overlay_slot:
                    HorizontalAlignment := horizontal_alignment.Fill
                    VerticalAlignment := vertical_alignment.Fill
                    Widget := MakeBox()
                overlay_slot:
                    HorizontalAlignment := horizontal_alignment.Fill
                    VerticalAlignment := vertical_alignment.Fill
                    Padding := margin{Left := 40.0, Top := 40.0, Right := 40.0, Bottom := 40.0}
                    Widget := stack_box:
                        Orientation := orientation.Vertical
                        Slots := array:
                            stack_box_slot:
                                HorizontalAlignment := horizontal_alignment.Center
                                VerticalAlignment := vertical_alignment.Top
                                Widget := NameText
                            stack_box_slot:
                                VerticalAlignment := vertical_alignment.Top
                                Padding := margin{Left := 0.0, Top := 20.0, Right := 0.0, Bottom := 0.0}
                                Widget := Icon
                            stack_box_slot:
                                HorizontalAlignment := horizontal_alignment.Center
                                VerticalAlignment := vertical_alignment.Top
                                Padding := margin{Left := 0.0, Top := 20.0, Right := 0.0, Bottom := 0.0}
                                Widget := stack_box:
                                    Orientation := orientation.Horizontal
                                    Slots := array:
                                        stack_box_slot:
                                            HorizontalAlignment := horizontal_alignment.Fill
                                            VerticalAlignment := vertical_alignment.Fill
                                            Padding := margin{Left := 0.0, Top := 0.0, Right := 10.0, Bottom := 0.0}
                                            Widget := LevelLabelText 
                                        stack_box_slot:
                                            HorizontalAlignment := horizontal_alignment.Fill
                                            VerticalAlignment := vertical_alignment.Fill                                            
                                            Widget := LevelText
                            stack_box_slot:
                                HorizontalAlignment := horizontal_alignment.Fill
                                VerticalAlignment := vertical_alignment.Top
                                Padding := margin{Left := 0.0, Top := 20.0, Right := 0.0, Bottom := 0.0}
                                Widget := color_block:
                                    DefaultColor := NamedColors.White
                                    DefaultOpacity := 0.1
                                    DefaultDesiredSize := vector2{X := 0.0, Y := 1.0}
                            stack_box_slot:
                                HorizontalAlignment := horizontal_alignment.Left
                                VerticalAlignment := vertical_alignment.Top
                                Padding := margin{Left := 0.0, Top := 40.0, Right := 0.0, Bottom := 0.0}
                                Widget := DescriptionText
                            stack_box_slot:
                                HorizontalAlignment := horizontal_alignment.Center
                                VerticalAlignment := vertical_alignment.Bottom
                                Distribution := option{1.0}
                                Padding := margin{Left := 0.0, Top := 20.0, Right := 0.0, Bottom := 0.0}
                                Widget := StatTypeText
                            stack_box_slot:
                                HorizontalAlignment := horizontal_alignment.Center
                                Padding := margin{Left := 0.0, Top := 0.0, Right := 0.0, Bottom := 0.0}
                                Widget := stack_box:
                                    Orientation := orientation.Horizontal
                                    Slots := array:
                                        stack_box_slot:
                                            HorizontalAlignment := horizontal_alignment.Fill
                                            VerticalAlignment := vertical_alignment.Fill
                                            Widget := ValueText
                                        stack_box_slot:
                                            HorizontalAlignment := horizontal_alignment.Fill
                                            VerticalAlignment := vertical_alignment.Fill
                                            Padding := margin{Left := 10.0, Top := 0.0, Right := 0.0, Bottom := 0.0}
                                            Widget := DeltaText
                            stack_box_slot:
                                HorizontalAlignment := horizontal_alignment.Fill
                                Padding := margin{Left := 0.0, Top := 20.0, Right := 0.0, Bottom := 20.0}
                                Widget := color_block:
                                    DefaultColor := NamedColors.White
                                    DefaultOpacity := 0.1
                                    DefaultDesiredSize := vector2{X := 0.0, Y := 1.0}
                            stack_box_slot:
                                HorizontalAlignment := horizontal_alignment.Center
                                VerticalAlignment := vertical_alignment.Fill
                                Widget := text_block:
                                    DefaultText := MakeMessage("Cost")
                                    DefaultTextColor := ThemeColors.PrimaryText
                                    DefaultTextOpacity := 0.5
                            stack_box_slot:
                                HorizontalAlignment := horizontal_alignment.Center
                                VerticalAlignment := vertical_alignment.Fill
                                Widget := stack_box:
                                    Orientation := orientation.Horizontal
                                    Slots := array:
                                        stack_box_slot:
                                            HorizontalAlignment := horizontal_alignment.Left
                                            Widget := texture_block:
                                                DefaultImage := currency_id.Gold.ToIcon()
                                                DefaultDesiredSize := vector2{X := 40.0, Y := 40.0}
                                        stack_box_slot:
                                            HorizontalAlignment := horizontal_alignment.Fill
                                            VerticalAlignment := vertical_alignment.Fill
                                            Padding := margin{Left := 10.0, Top := 0.0, Right := 0.0, Bottom := 0.0}
                                            Widget := CostText
        
    <# ----------------------------------------------------------------------------------------------------
        Get Functions
    ---------------------------------------------------------------------------------------------------- #>


    <# ----------------------------------------------------------------------------------------------------
        Update Functions
    ---------------------------------------------------------------------------------------------------- #>
    Update():void =
        CurrencyText.SetText(MakeMessage("{FormatCurrency(Player.GetCurrencyAvailable(currency_id.Gold))}"))

        PowerUpIds := GetPowerUpIds(ActiveType)
        for (Index -> PowerUpEntryUI : PowerUpEntryUIs):
            if (PowerUpId := PowerUpIds[Index], PowerUpInfo := GetPowerUpInfo[PowerUpId]):
                PowerUpEntryUI.Update(PowerUpInfo)
                PowerUpEntryUI.Show()
            else:
                PowerUpEntryUI.Hide()

            if (Index = ActiveIndex):
                PowerUpEntryUI.SetActive()
            
        UpdateInfo()

    UpdateInfo():void =
        if (PowerUpId := GetPowerUpIds(ActiveType)[ActiveIndex], PowerUpInfo := GetPowerUpInfo[PowerUpId]):
            NameText.SetText(MakeMessage("{PowerUpInfo.Name}"))
            Icon.SetImage(PowerUpInfo.Icon)
            DescriptionText.SetText(MakeMessage("{PowerUpInfo.Description.WrapText(800.0)}"))
            
            PowerUpValue := Player.GetPowerUpValue(PowerUpId)
            StatTypeText.SetText(MakeMessage("{PowerUpInfo.StatType}"))

            if (Player.GetPowerUpLevel(PowerUpId) >= PowerUpInfo.MaxLevel):
                LevelText.SetText(MakeMessage("Maxed!"))
                LevelText.SetColor(ThemeColors.Success)
                LevelLabelText.SetVisibility(widget_visibility.Collapsed)
                ValueText.SetText(MakeMessage("{PowerUpInfo.FormatValue(PowerUpValue)}"))
                DeltaText.SetVisibility(widget_visibility.Collapsed)

                CostText.SetText(MakeMessage("N/A"))
                CostText.SetColor(ThemeColors.Success)
            else:
                LevelText.SetText(MakeMessage("{Player.GetPowerUpLevel(PowerUpInfo.Id)}/{PowerUpInfo.MaxLevel}"))
                LevelText.SetTextColor(ThemeColors.PrimaryText)
                LevelLabelText.SetVisibility(widget_visibility.Visible)
                ValueText.SetText(MakeMessage("{PowerUpInfo.FormatValue(PowerUpValue)} ▶"))

                DeltaValue :=  PowerUpInfo.IsDeltaValue? and " ({PowerUpInfo.IncrementValue > 0.0 and "+" or ""}{PowerUpInfo.FormatValue(PowerUpInfo.IncrementValue)})" or ""
                DeltaText.SetText(MakeMessage("{PowerUpInfo.FormatValue(PowerUpValue + PowerUpInfo.IncrementValue)}" + DeltaValue))
                DeltaText.SetVisibility(widget_visibility.Visible)

                if (PowerUpCost := Player.GetPowerUpCost[PowerUpId, Player.GetPowerUpLevel(PowerUpId) + 1]):
                    CostText.SetText(MakeMessage("{FormatCurrency(PowerUpCost)}"))
                    CostText.SetColor(Player.GetCurrencyAvailable(currency_id.Gold) >= PowerUpCost and ThemeColors.Success or ThemeColors.Failure)

    HandleButtonClicked(Index:int):void =
        if (Index = ActiveIndex):
            Upgrade()
            return

        if (PowerUpEntryUI := PowerUpEntryUIs[ActiveIndex]):
            PowerUpEntryUI.ClearActive()

        if (PowerUpEntryUI := PowerUpEntryUIs[Index]):
            PowerUpEntryUI.SetActive()

        set ActiveIndex = Index
        UpdateInfo()
        Player.PlayAudio(audio_id.Activate)

    Upgrade():void =
        if (PowerUpId := GetPowerUpIds(ActiveType)[ActiveIndex], PowerUpInfo := GetPowerUpInfo[PowerUpId]):
            if (Player.GetPowerUpLevel(PowerUpId) >= PowerUpInfo.MaxLevel):
                Player.PlayAudio(audio_id.Error)
                return

            if (PowerUpCost := Player.GetPowerUpCost[PowerUpId, Player.GetPowerUpLevel(PowerUpId) + 1], Player.GetCurrencyAvailable(currency_id.Gold) >= PowerUpCost):
                Player.AddPowerUp(PowerUpId)
                Update()
                Player.PlayAudio(audio_id.PowerUp)
            else:
                Player.PlayAudio(audio_id.Error)

    SetType(Type:power_up_type):void =
        if (PowerUpEntryUI := PowerUpEntryUIs[ActiveIndex]):
            PowerUpEntryUI.ClearActive()

        set ActiveIndex = 0
        set ActiveType = Type
        Update()
        Player.PlayAudio(audio_id.Activate)
        
    <# ----------------------------------------------------------------------------------------------------
        Event Handlers
    ---------------------------------------------------------------------------------------------------- #>
    OnShow<override>():void =
        if (PowerUpEntryUI := PowerUpEntryUIs[ActiveIndex]):
            PowerUpEntryUI.ClearActive()

        set ActiveIndex = 0
        Update()

    OnClosed(InPlayer:player):void =
        for (PowerUpEntryUI : PowerUpEntryUIs):
            PowerUpEntryUI.CancelAnimationEvent.Signal()

        InPlayer.UpdateGoldCurrencyHUD()
        InPlayer.UpdatePowerUpIndicator()

    OnClickRefundOne(Message:widget_message):void =
        if (PowerUpId := GetPowerUpIds(ActiveType)[ActiveIndex], Player.RefundOnePowerUp[PowerUpId]):
            Update()
            Player.PlayAudio(audio_id.Banish)
        else:
            Player.PlayAudio(audio_id.Error)

    OnClickRefund(Message:widget_message):void =
        if (PowerUpId := GetPowerUpIds(ActiveType)[ActiveIndex]):
            Player.RefundPowerUp(PowerUpId)
            Update()
            Player.PlayAudio(audio_id.Banish)

    OnClickRefundAll(Message:widget_message):void =
        Player.RefundPowerUps(ActiveType)
        Update()
        Player.PlayAudio(audio_id.Banish)

    AwaitButtonClicked(Event:events(tuple()), Index:int)<suspends>:void =
        loop:
            Event.Await()
            HandleButtonClicked(Index)

    AwaitMenuButtonClick(Event:listenable(widget_message), Type:power_up_type)<suspends>:void =
        loop:
            Event.Await()
            SetType(Type)