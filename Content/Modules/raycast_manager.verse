using { /Fortnite.com/Devices }
using { /Verse.org/Simulation }
using { /UnrealEngine.com/Temporary/SpatialMath }

<# ----------------------------------------------------------------------------------------------------
    Global Functions
---------------------------------------------------------------------------------------------------- #>
var GlobalRaycastManager:weak_map(session, raycast_manager) = map{}

GetRaycastManager()<decides><transacts>:raycast_manager =
    GlobalRaycastManager[GetSession()]

GetRaycast<public>(Position:vector3)<suspends>:vector3 =
    if (RaycastManager := GetRaycastManager[]):
        RaycastManager.Raycast(Position)
    else:
        vector3{Z := 2000.0}

GetRaycast<public>(Position:vector2)<suspends>:vector3 =
    GetRaycast(vector3{X := Position.X, Y := Position.Y})

<# ----------------------------------------------------------------------------------------------------
    raycast_manager
---------------------------------------------------------------------------------------------------- #>
raycast_manager := class(creative_device):
    @editable
    RaycastRigProp:creative_prop = creative_prop{}

    @editable
    RaycastHitProp:creative_prop = creative_prop{}

    <# ----------------------------------------------------------------------------------------------------
        Initialization
    ---------------------------------------------------------------------------------------------------- #>
    OnBegin<override>()<suspends>:void =
        if (set GlobalRaycastManager[GetSession()] = Self) {}

    <# ----------------------------------------------------------------------------------------------------
        Get Functions
    ---------------------------------------------------------------------------------------------------- #>
    Raycast(Position:vector3)<suspends>:vector3 =
        if (RaycastRigProp.TeleportTo[vector3{X := Position.X, Y := Position.Y, Z := 2000.0}, MakeRotationFromYawPitchRollDegrees(0.0, 180.0, 0.0)]) {}
        Sleep(0.0)
        RaycastHitProp.GetTransform().Translation

    <# ----------------------------------------------------------------------------------------------------
        Update Functions
    ---------------------------------------------------------------------------------------------------- #>
    