using { /Fortnite.com/Characters }
using { /Fortnite.com/Devices }
using { /Verse.org/Random }
using { /Verse.org/Simulation }
using { /UnrealEngine.com/Temporary/SpatialMath }
using { Utility }

<# ----------------------------------------------------------------------------------------------------
    Constants
---------------------------------------------------------------------------------------------------- #>
SupplyDropDefaultActivationTime:float = 30.0

<# ----------------------------------------------------------------------------------------------------
    Global Functions
---------------------------------------------------------------------------------------------------- #>
var GlobalSupplyManager:weak_map(session, supply_drop_manager) = map{}

GetSupplyManager()<decides><transacts>:supply_drop_manager = 
    GlobalSupplyManager[GetSession()]

SpawnSupplyDrop():void =
    if (SupplyManager := GetSupplyManager[]):
        spawn{SupplyManager.Start()}

DestroySupplyDrop():void =
    if (SupplyManager := GetSupplyManager[]):
        SupplyManager.Stop()

<# ----------------------------------------------------------------------------------------------------
    supply_drop_manager
---------------------------------------------------------------------------------------------------- #>
supply_drop_manager := class(creative_device):
    @editable
    SupplyDropSpawner:supply_drop_spawner_device = supply_drop_spawner_device{}

    @editable
    VolumeActivator:volume_activator = volume_activator{}

    <# ----------------------------------------------------------------------------------------------------
        Initialization
    ---------------------------------------------------------------------------------------------------- #>
    OnBegin<override>()<suspends>:void =
        if (set GlobalSupplyManager[GetSession()] = Self) {}

        SupplyDropSpawner.LandingEvent.Subscribe(OnLanding)
        VolumeActivator.ActivatedEvent.Subscribe(OnActivated)
        VolumeActivator.OnBegin()

        WaitForGameManager()
        if (GameManager := GetGameManager[]):
            GameManager.GameEndedEvent.Subscribe(Stop)
            GameManager.BossFightStartedEvent.Subscribe(Stop)
            GameManager.StageEndedEvent.Subscribe(Stop)

    <# ----------------------------------------------------------------------------------------------------
        Get Functions
    ---------------------------------------------------------------------------------------------------- #>
 

    <# ----------------------------------------------------------------------------------------------------
        Update Functions
    ---------------------------------------------------------------------------------------------------- #>
    Start()<suspends>:void =
        RandomPositions := Shuffle(PoissonDiscSamplingGeneratePoints(1500.0, GetActiveArenaSize() - VolumeActivator.Size * 10.0, ?MinRadius := 600.0))
        
        var Attempts:int = 1
        loop:
            RandomPosition := GetRandomPositionInDonut(600.0, 3500.0, ?Origin := DefaultArenaOrigin)
            RaycastPosition := GetRaycast(RandomPosition)

            if (Attempts >= 30):
                break
            
            if (RaycastPosition.Z <= 10.0, SupplyDropSpawner.TeleportTo[RaycastPosition.Make2D(), IdentityRotation()]):
                SupplyDropSpawner.Spawn()
                VolumeActivator.UpdateMapIndicator(RaycastPosition.Make2D())
                break

            set Attempts += 1

        for (ActivePlayer : GetActivePlayers()):
            ActivePlayer.PlayAudio(audio_id.SupplyDrop)

    Stop():void =
        VolumeActivator.Disable()
        if (Player := GetAllPlayers()[0]):
            SupplyDropSpawner.Open(Player)

    <# ----------------------------------------------------------------------------------------------------
        Event Handlers
    ---------------------------------------------------------------------------------------------------- #>
    OnLanding():void =
        SupplyDropPosition := SupplyDropSpawner.GetTransform().Translation.Make2D()
        SupplyDropDamage := GetNPCMaxHealth(NPCBaseMaxHealth)

        if (RandomPlayer := GetRandomAlivePlayer()?):
            for:
                NPCBehaviorBase : GetAllNPCs()
                NPCCharacter := NPCBehaviorBase.GetAgent[].GetFortCharacter[]
                NPCPosition := NPCCharacter.GetTransform().Translation.Make2D()
                DistanceSquared(NPCPosition, SupplyDropPosition) <= Pow(VolumeActivator.Size * 10.0 + NPCBehaviorBase.GetSize(), 2.0)
            do:
                AbilityDamageArgs := ability_damage_args{Target := NPCBehaviorBase, Amount := SupplyDropDamage, IsCriticalHit := false}
                NPCBehaviorBase.Damage(RandomPlayer, AbilityDamageArgs)

        KnockbackNPC(SupplyDropPosition, ?KnockbackDistanceSquared := Pow(VolumeActivator.Size * 10.0, 2.0), ?Duration := 1.0)
        ActivateCameraShake(SupplyDropPosition)
        VolumeActivator.Enable(?TargetPosition := option{SupplyDropPosition}, ?OverrideActivationTime := option{SupplyDropDefaultActivationTime / GetGameSpeed()})
        SupplyDropSpawner.DestroyBalloon()

    OnActivated(Player:player):void =
        VolumeActivator.Disable()
        SupplyDropSpawner.Open(Player)

        for (ActivePlayer : GetActivePlayers(), GameProgressBarUI := GetGameManager[].PlayerGameProgressBarUI[ActivePlayer]):
            GameProgressBarUI.UpdateEventState(GameEventSupplyDropIndex, game_event_state.Success)
    
        for (Index := 1..GetRandomInt(10, 20)):
            SpawnExperience(GetRandomPositionInCircle(VolumeActivator.Size * 10.0, ?Origin := SupplyDropSpawner.GetTransform().Translation.Make2D()), GetRandomInt(10, 60))

        SpawnItemDrop(item_drop_id.Magnet, SupplyDropSpawner.GetTransform().Translation.Make2D())

        AwardAccolade(accolade_id.SupplyDropSecured, GetActivePlayers())