using { /Fortnite.com/AI }
using { /Fortnite.com/Characters }
using { /Fortnite.com/Devices }
using { /Fortnite.com/FortPlayerUtilities }
using { /Fortnite.com/Game }
using { /Verse.org/Colors }
using { /Verse.org/Concurrency }
using { /Verse.org/Random }
using { /Verse.org/Simulation }
using { /UnrealEngine.com/Temporary/Diagnostics }
using { /UnrealEngine.com/Temporary/SpatialMath }
using { /UnrealEngine.com/Temporary/UI }
using { RealTime }

<# ----------------------------------------------------------------------------------------------------
    Global Functions
---------------------------------------------------------------------------------------------------- #>
var GlobalTestTool:weak_map(session, test_tool) = map{}

GetTestTool()<decides><transacts>:test_tool =
    GlobalTestTool[GetSession()]

(Player:player).GetTestToolDebugValue()<transacts>:float =
    GetTestTool[].TestToolUIs[Player].DebugValue or 0.0

DebugDraw()<transacts>:debug_draw =
    GetTestTool[].GetDebugDraw() or debug_draw{Channel := debug_draw_global_channel}

debug_draw_global_channel := class(debug_draw_channel) {}

test_trigger := class<concrete>:
    @editable
    Trigger:trigger_device = trigger_device{}
    @editable
    Tracker:tracker_device = tracker_device{}

<# ----------------------------------------------------------------------------------------------------
    test_tool
---------------------------------------------------------------------------------------------------- #>
test_tool := class(creative_device):
    @editable
    ItemGranterSignalRemote:item_granter_device = item_granter_device{}

    @editable
    SignalRemoteManager:signal_remote_manager_device = signal_remote_manager_device{}
    
    @editable
    DownButNotOut:down_but_not_out_device = down_but_not_out_device{}

    @editable
    HUDController:hud_controller_device = hud_controller_device{}

    @editable
    ConditionalButton:conditional_button_device = conditional_button_device{}

    @editable
    VFXPowerUp:visual_effect_powerup_device = visual_effect_powerup_device{}

    @editable
    VFXPowerUpEvolved:visual_effect_powerup_device = visual_effect_powerup_device{}

    @editable
    VFXPowerUpDisable:visual_effect_powerup_device = visual_effect_powerup_device{}

    @editable
    VFXPowerUpDisableEvolved:visual_effect_powerup_device = visual_effect_powerup_device{}

    @editable
    Prop:creative_prop = creative_prop{}

    @editable
    TestTrigger1:test_trigger = test_trigger{}

    @editable
    TestTrigger2:test_trigger = test_trigger{}

    @editable
    TestTrigger3:test_trigger = test_trigger{}

    @editable
    TestTrigger4:test_trigger = test_trigger{}

    @editable
    Chair:chair_device = chair_device{}

    @editable
    CinematicSequence01:cinematic_sequence_device = cinematic_sequence_device{}
    @editable
    CinematicSequence02:cinematic_sequence_device = cinematic_sequence_device{}
    @editable
    CinematicSequence03:cinematic_sequence_device = cinematic_sequence_device{}

    var (test_tool:)DebugDraw:debug_draw = debug_draw{Channel := debug_draw_global_channel}
    var TestToolUIs:[player]test_tool_ui = map{}

    <# ----------------------------------------------------------------------------------------------------
        Initialization
    ---------------------------------------------------------------------------------------------------- #>
    OnBegin<override>()<suspends>:void =
        if (set GlobalTestTool[GetSession()] = Self) {}

        for (Player : GetPlayspace().GetPlayers()):
            spawn{Initialize(Player)}

        HUDController.Enable()
        SignalRemoteManager.PrimarySignalEvent.Subscribe(OnPrimarySignal)
        SignalRemoteManager.SecondarySignalEvent.Subscribe(OnSecondarySignal)

        <#
        spawn{AwaitTestTrigger1(TestTrigger1)}
        spawn{AwaitTestTrigger2(TestTrigger2)}
        spawn{AwaitTestTrigger3(TestTrigger3)}
        spawn{AwaitTestTrigger4(TestTrigger4)}
        #>

        GetPlayspace().PlayerAddedEvent().Subscribe(OnPlayerAdded)
        GetPlayspace().PlayerRemovedEvent().Subscribe(OnPlayerRemoved)
        
    Initialize(Player:player)<suspends>:void =
        ItemGranterSignalRemote.GrantItem(Player)

        WaitForInputManager()
        spawn{TrackIsHoldingSignalRemote(Player)}
        
    <# ----------------------------------------------------------------------------------------------------
        Get Functions
    ---------------------------------------------------------------------------------------------------- #>
    GetDebugDraw()<transacts>:debug_draw =
        (test_tool:)DebugDraw

    <# ----------------------------------------------------------------------------------------------------
        Update Functions
    ---------------------------------------------------------------------------------------------------- #>
    ShowTestTool(Agent:agent):void =
        if (Player := player[Agent]):
            TestToolUI := if (UI := TestToolUIs[Player]):
                UI
            else:
                UI := test_tool_ui{Player := Player}
                if (set TestToolUIs[Player] = UI) {}
                UI

            TestToolUI.Update()
            TestToolUI.Show()
            Agent.PlayAudio(audio_id.Activate)

    AsyncTest(Agent:agent)<suspends>:void =
        if (Player := player[Agent], Character := Agent.GetFortCharacter[]):
            <#
            DebugDrawInst := DebugDraw()
            loop:
                for (Index := 1..45):
                    DebugDrawInst.DrawPoint(GetSpawnPosition(Agent), ?Color := ThemeColors.Failure, ?Thickness := 25.0, ?DrawDurationPolicy := debug_draw_duration_policy.SingleFrame)

                Sleep(0.0)
            #>

    TrackIsHoldingSignalRemote(Player:player)<suspends>:void =
        loop:
            if (IsGameInProgress()?):
                if (ConditionalButton.IsHoldingItem[Player]):
                    Player.UnregisterInput(input_id.AbilityStats)
                else:
                    Player.RegisterInput(input_id.AbilityStats)

            Sleep(TickTime * 5)

    <# ----------------------------------------------------------------------------------------------------
        Event handlers
    ---------------------------------------------------------------------------------------------------- #>
    OnPlayerAdded(Player:player):void =
        spawn{Initialize(Player)}

    OnPlayerRemoved(Player:player):void =
        set TestToolUIs = RemoveKeyFromMap(TestToolUIs, Player)
        TestTrigger1.Tracker.Reset(Player)
        TestTrigger1.Tracker.Remove(Player)

    OnPrimarySignal(Agent:agent):void =
        ShowTestTool(Agent)
   
    Test(Player:player)<suspends>:void =
        if (Character := Player.GetFortCharacter[]):

    
    OnSecondarySignal(Agent:agent):void =
        if (Player := player[Agent], Character := Agent.GetFortCharacter[]):
            if (IsGameInProgress()?):
                
            else:
                if (Arena := GetArenas()[0]):
                    SetGameSettings of game_settings:
                        Arena := Arena
                        Difficulty := 1
                        GameSpeed := 5

                if (GameManager := GetGameManager[]):
                    GameManager.ForceStart(0)

    <#
    AwaitTestTrigger1(TestTrigger:test_trigger)<suspends>:void =
        loop:
            MaybeAgent := TestTrigger.Trigger.TriggeredEvent.Await()
            if (Agent := MaybeAgent?):
                TestTrigger.Tracker.AssignToAll()
                TestTrigger.Tracker.SetValue(0)
                TestTrigger.Tracker.SetTarget(10)

    AwaitTestTrigger2(TestTrigger:test_trigger)<suspends>:void =
        loop:
            MaybeAgent := TestTrigger.Trigger.TriggeredEvent.Await()
            if (Agent := MaybeAgent?):
                TestTrigger.Tracker.RemoveFromAll()

    AwaitTestTrigger3(TestTrigger:test_trigger)<suspends>:void =
        loop:
            MaybeAgent := TestTrigger.Trigger.TriggeredEvent.Await()
            for (Player : GetAllPlayers()):
                TestTrigger.Tracker.SetValue(TestTrigger.Tracker.GetValue() + 1)
        
    AwaitTestTrigger4(TestTrigger:test_trigger)<suspends>:void =
        loop:
            MaybeAgent := TestTrigger.Trigger.TriggeredEvent.Await()
            if (Agent := MaybeAgent?):
                for (Player : GetAllPlayers()):
                    TestTrigger.Tracker.Reset(Player)
    #>

<# ----------------------------------------------------------------------------------------------------
        Event handlers
---------------------------------------------------------------------------------------------------- #>

<#
# Activate objectives in sequential order.
var IsObjectiveStarted:logic = false
var ObjectiveIndex:int = 0

ObjectiveIds := array:
    objective_id.Collectible
    objective_id.GoldenSlime
    objective_id.BeachBall
    objective_id.CapturePoint
    objective_id.ColorChangingTiles
    objective_id.Storm

if (not IsObjectiveStarted?):
    if (ObjectiveId := ObjectiveIds[ObjectiveIndex]):
        StartObjective(ObjectiveId)
    set ObjectiveIndex += 1
    if (ObjectiveIndex > ObjectiveIds.Length - 1):
        set ObjectiveIndex = 0
else:
    StopObjective()

set IsObjectiveStarted = logic{not IsObjectiveStarted?}


# Start game instantly.
if (GameManager := GetGameManager[]):
    GameManager.ForceStart(0)

    
# Enables Item Shop
EnableItemShop()
AddGameCurrency(game_currency_id.Wood, 3000.0)
AddGameCurrency(game_currency_id.TreasureChest, 10.0)

# Grant random daily quests.
if:
    CurrentPlayerData := Player.GetData[]
    UpdatedLastQuestAssignedTime := GetDailyQuestsManager[].RealTimeDevice.GetDateAndTime().GetTotalSeconds()
    UpdatedActiveQuests := GetDailyQuestsManager[].GetNewDailyQuests[Player, 3]
    set PlayerDataMap[Player] = player_data:
        MakePlayerData<constructor>(CurrentPlayerData)
        DailyQuests := player_daily_quests:
            MakePlayerDailyQuests<constructor>(CurrentPlayerData.DailyQuests)
            LastQuestAssignedTime := UpdatedLastQuestAssignedTime
            ActiveQuests := UpdatedActiveQuests

# Spawn Item Drop in front of the player character.
Direction := Character.GetTransform().Rotation.GetLocalForward()
TargetPosition := Character.GetTransform().Translation + Direction * 300.0
SpawnItemDrop(item_drop_id.AbilityUpgrade, TargetPosition)

# Opens the Boon Select UI.
PauseGame(pause_type.Boon)

# Spawns Supply Drop.
var IsSpawned:logic = false
if (not IsSpawned?):
    SpawnSupplyDrop()
else:
    DestroySupplyDrop()
set IsSpawned = logic{not IsSpawned?}

# Show player rank up HUD Message.
var Rank:int = 1
case (Rank):
    1 => Player.ShowMessage(message_id.RankUp01)
    2 => Player.ShowMessage(message_id.RankUp02)
    3 => Player.ShowMessage(message_id.RankUp03)
    4 => Player.ShowMessage(message_id.RankUp04)
    5 => Player.ShowMessage(message_id.RankUp05)
    6 => Player.ShowMessage(message_id.RankUp06)
    7 => Player.ShowMessage(message_id.RankUp07)
    8 => Player.ShowMessage(message_id.RankUp08)
    9 => Player.ShowMessage(message_id.RankUp09)
    _ => Player.ShowMessage(message_id.RankUp01)

set Rank = (Rank = 9) and 1 or (Rank + 1)

#>