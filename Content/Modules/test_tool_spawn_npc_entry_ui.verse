using { /Fortnite.com/UI }
using { /UnrealEngine.com/Temporary/SpatialMath }
using { /UnrealEngine.com/Temporary/UI }

test_tool_spawn_npc_entry_ui := class:
    ClickedSpawnEvent:events(npc_id) = events(npc_id){}
    ClickedSpawnFrontEvent:events(npc_id) = events(npc_id){}

    Icon:texture_block = texture_block:
        DefaultImage := UserInterface.Icons.NPC_Melee_64
        DefaultDesiredSize := vector2{X := 64.0, Y := 64.0}

    NameText:text_block = text_block:
        DefaultTextColor := ThemeColors.PrimaryText

    var Id:npc_id
    var Widget:widget = canvas{}

    Initialize():void =
        set Widget = CreateUI()
        Update()

    CreateUI():widget =
        SpawnButton := button_quiet{DefaultText := MakeMessage("Spawn")}
        SpawnButton.OnClick().Subscribe(OnClickSpawn)

        SpawnFrontButton := button_quiet{DefaultText := MakeMessage("Front")}
        SpawnFrontButton.OnClick().Subscribe(OnClickSpawnFront)
                    
        stack_box:
            Orientation := orientation.Horizontal
            Slots := array:
                stack_box_slot:
                    HorizontalAlignment := horizontal_alignment.Fill
                    Widget := Icon
                stack_box_slot:
                    HorizontalAlignment := horizontal_alignment.Left
                    Padding := margin{Left := 10.0, Top := 0.0, Right := 0.0, Bottom := 0.0}
                    Distribution := option{1.0}
                    Widget := NameText 
                stack_box_slot:
                    HorizontalAlignment := horizontal_alignment.Fill
                    Padding := margin{Left := 100.0, Top := 0.0, Right := 0.0, Bottom := 0.0}
                    Widget := SpawnButton
                stack_box_slot:
                    HorizontalAlignment := horizontal_alignment.Fill
                    Padding := margin{Left := 10.0, Top := 0.0, Right := 0.0, Bottom := 0.0}
                    Widget := SpawnFrontButton

    <# ----------------------------------------------------------------------------------------------------
        Get Functions
    ---------------------------------------------------------------------------------------------------- #>
    GetUI():widget =
        Widget

    <# ----------------------------------------------------------------------------------------------------
        Update Functions
    ---------------------------------------------------------------------------------------------------- #>
    SetId(NewId:npc_id):void =
        set Id = NewId
        Widget.SetVisibility(widget_visibility.Visible)
        Update()

    Update():void =
        Icon.SetImage(Id.ToIcon())
        NameText.SetText(MakeMessage("{Id}"))

    Hide():void =
        Widget.SetVisibility(widget_visibility.Hidden)

    <# ----------------------------------------------------------------------------------------------------
        Event Handlers
    ---------------------------------------------------------------------------------------------------- #>
    OnClickSpawn(Message:widget_message):void =
        ClickedSpawnEvent.Signal(Id)

    OnClickSpawnFront(Message:widget_message):void =
        ClickedSpawnFrontEvent.Signal(Id)