using { /Fortnite.com/Characters }
using { /Fortnite.com/UI }
using { /Verse.org/Simulation }
using { /UnrealEngine.com/Temporary/SpatialMath }
using { /UnrealEngine.com/Temporary/UI }

test_tool_ui := class(panel_ui):
    Title<override>:string = "Test Tools"
    BackgroundOpacity<override>:type {_X:float where 0.000000 <= _X, _X <= 1.000000} = 0.7
    IncrementAmount:float = 1.0

    DebugValueText:text_block = text_block:
        DefaultTextColor := ThemeColors.PrimaryText
    
    Options:[]string = array:
        "Toggle Ability"
        "Max Abilities"
        "Reset Level"
        "Spawn NPC"
        "Game In Progress"
        "Start Bossfight"
        "Victory"
        "Teleport To Hub"
        "Teleport To Arena"

    var AbilityLevelTexts:[ability_id]text_block = map{}
    var MaybeSpawnNPCUI:?test_tool_spawn_npc_ui = false
    var DebugValue:float = 1.0

    <# ----------------------------------------------------------------------------------------------------
        Initialization
    ---------------------------------------------------------------------------------------------------- #>
    CreateContent<override>():widget =
        ButtonIncrement := button_quiet{DefaultText := MakeMessage("+")}
        ButtonIncrement.OnClick().Subscribe(OnClickIncrement)

        ButtonDecrement := button_quiet{DefaultText := MakeMessage("-")}
        ButtonDecrement.OnClick().Subscribe(OnClickDecrement)

        stack_box:
            Orientation := orientation.Horizontal
            Slots := array:
                stack_box_slot:
                    HorizontalAlignment := horizontal_alignment.Fill
                    VerticalAlignment := vertical_alignment.Fill
                    Padding := margin{Left := 0.0, Top := 0.0, Right := 0.0, Bottom := 0.0}
                    Widget := stack_box:
                        Orientation := orientation.Vertical
                        Slots := array:
                            stack_box_slot:
                                HorizontalAlignment := horizontal_alignment.Fill
                                VerticalAlignment := vertical_alignment.Fill
                                Widget := stack_box:
                                    Orientation := orientation.Vertical
                                    Slots := for (Index -> Option : Options):
                                        Button:button_quiet = button_quiet:
                                            DefaultText := MakeMessage("{Option}")
                            
                                        case (Option):
                                            "Toggle Ability" => Button.OnClick().Subscribe(OnClickToggleAbility)
                                            "Max Abilities" => Button.OnClick().Subscribe(OnClickMaxAbilities)
                                            "Reset Level" => Button.OnClick().Subscribe(OnClickResetLevel)
                                            "Game In Progress" => Button.OnClick().Subscribe(OnClickGameInProgress)
                                            "Spawn NPC" => Button.OnClick().Subscribe(OnClickSpawnNPC)
                                            "Start Bossfight" => Button.OnClick().Subscribe(OnClickStartBossFight)
                                            "Victory" => Button.OnClick().Subscribe(OnClickVictory)
                                            "Teleport To Hub" => Button.OnClick().Subscribe(OnClickTeleportToHub)
                                            "Teleport To Arena" => Button.OnClick().Subscribe(OnClickTeleportToArena)
                                            _ => {}
                        
                                        stack_box_slot:
                                            HorizontalAlignment := horizontal_alignment.Fill
                                            VerticalAlignment := vertical_alignment.Fill
                                            Padding := margin{Left := 0.0, Top := Index = 0 and 0.0 or 10.0, Right := 0.0, Bottom := 0.0}
                                            Widget := Button
                            stack_box_slot:
                                HorizontalAlignment := horizontal_alignment.Fill
                                VerticalAlignment := vertical_alignment.Top
                                Padding := margin{Left := 0.0, Top := 20.0, Right := 0.0, Bottom := 0.0}
                                Distribution := option{1.0}
                                Widget := stack_box:
                                    Orientation := orientation.Horizontal
                                    Slots := array:
                                        stack_box_slot:
                                            HorizontalAlignment := horizontal_alignment.Fill
                                            VerticalAlignment := vertical_alignment.Fill
                                            Distribution := option{1.0}
                                            Widget := ButtonDecrement
                                        stack_box_slot:
                                            HorizontalAlignment := horizontal_alignment.Fill
                                            Padding := margin{Left := 10.0, Top := 0.0, Right := 10.0, Bottom := 0.0}
                                            Widget := DebugValueText
                                        stack_box_slot:
                                            HorizontalAlignment := horizontal_alignment.Fill
                                            VerticalAlignment := vertical_alignment.Fill
                                            Distribution := option{1.0}
                                            Widget := ButtonIncrement
                stack_box_slot:
                    HorizontalAlignment := horizontal_alignment.Fill
                    VerticalAlignment := vertical_alignment.Fill
                    Padding := margin{Left := 20.0, Top := 0.0, Right := 0.0, Bottom := 0.0}
                    Widget := CreateAbilityUI()
    
    CreateAbilityUI():widget =
        AbilityUI := stack_box:
            Orientation := orientation.Vertical
            Slots := for (AbilityInfo : GetAbilityInfo()):
                ButtonSubtract := button_quiet{DefaultText := MakeMessage("-")}
                ButtonAdd := button_quiet{DefaultText := MakeMessage("+")}
                ButtonReset := button_quiet{DefaultText := MakeMessage("Reset")}
                ButtonMax := button_quiet{DefaultText := MakeMessage("Max")}
                spawn{OnAbilitySubtractClicked(ButtonSubtract, AbilityInfo.Id)}
                spawn{OnAbilityAddClicked(ButtonAdd, AbilityInfo.Id)}
                spawn{OnAbilityResetClicked(ButtonReset, AbilityInfo.Id)}
                spawn{OnAbilityMaxClicked(ButtonMax, AbilityInfo.Id)}
                

                AbilityLevelText := text_block:
                    DefaultText := MakeMessage("{Player.GetAbilityLevel(AbilityInfo.Id)}")
                    DefaultTextColor := ThemeColors.Highlight

                if (set AbilityLevelTexts[AbilityInfo.Id] = AbilityLevelText) {}

                stack_box_slot:
                    HorizontalAlignment := horizontal_alignment.Fill
                    VerticalAlignment := vertical_alignment.Fill
                    Distribution := option{1.0}
                    Padding := margin{Left := 0.0, Top := 5.0, Right := 0.0, Bottom := 0.0}
                    Widget := stack_box:
                        Orientation := orientation.Horizontal
                        Slots := array:
                            stack_box_slot:
                                Widget := texture_block:
                                    DefaultImage := AbilityInfo.IconSmall
                                    DefaultDesiredSize := vector2{X := 30.0, Y := 30.0}
                            stack_box_slot:
                                Padding := margin{Left := 10.0, Top := 0.0, Right := 0.0, Bottom := 0.0}
                                Widget := text_block:
                                    DefaultText := MakeMessage("{AbilityInfo.Name}")
                                    DefaultTextColor := ThemeColors.PrimaryText
                            stack_box_slot:
                                HorizontalAlignment := horizontal_alignment.Right
                                Distribution := option{1.0}
                                Padding := margin{Left := 20.0, Top := 0.0, Right := 0.0, Bottom := 0.0}
                                Widget := AbilityLevelText
                            stack_box_slot:
                                Padding := margin{Left := 10.0, Top := 0.0, Right := 0.0, Bottom := 0.0}
                                Widget := ButtonSubtract
                            stack_box_slot:
                                Padding := margin{Left := 10.0, Top := 0.0, Right := 0.0, Bottom := 0.0}
                                Widget := ButtonAdd
                            stack_box_slot:
                                Padding := margin{Left := 10.0, Top := 0.0, Right := 0.0, Bottom := 0.0}
                                Widget := ButtonReset
                            stack_box_slot:
                                Padding := margin{Left := 10.0, Top := 0.0, Right := 0.0, Bottom := 0.0}
                                Widget := ButtonMax

        AbilityUI

    <# ----------------------------------------------------------------------------------------------------
        Update Functions
    ---------------------------------------------------------------------------------------------------- #>
    Update():void =
        for (Id -> AbilityLevelText : AbilityLevelTexts):
            AbilityLevelText.SetText(MakeMessage("{Player.GetAbilityLevel(Id)}"))

        DebugValueText.SetText(MakeMessage("{GetRoundedString(DebugValue, 2)}"))

    UpdateAbilityText(Id:ability_id):void =
        if (AbilityLevelText := AbilityLevelTexts[Id]):
            AbilityLevelText.SetText(MakeMessage("{Player.GetAbilityLevel(Id)}"))

    <# ----------------------------------------------------------------------------------------------------
        Event Handlers
    ---------------------------------------------------------------------------------------------------- #>
    OnShow<override>():void =
        Update()

    OnClickResetLevel(Message:widget_message):void =
        Hide()
        Player.ResetStats()

    OnClickGameInProgress(Message:widget_message):void =
        if (GameManager := GetGameManager[]):
            if (IsGameInProgress()?):
                set GameManager.ActivePlayers = array{}
            else:
                set GameManager.ActivePlayers = GetAllPlayers()

        if (Button := text_button_base[Message.Source]):
            Button.SetText(MakeMessage("GameInProgress: {IsGameInProgress()? and "true" or "false"}"))

    OnAbilitySubtractClicked(Button:text_button_base, Id:ability_id)<suspends>:void =
        loop:
            Button.OnClick().Await()
            Player.SubtractAbility(Id)
            UpdateAbilityText(Id)

    OnAbilityAddClicked(Button:text_button_base, Id:ability_id)<suspends>:void =
        loop:
            Button.OnClick().Await()
            if (AbilityInfo := GetAbilityInfo[Id]):
                Player.AddAbility(AbilityInfo)
                UpdateAbilityText(Id)

    OnAbilityResetClicked(Button:text_button_base, Id:ability_id)<suspends>:void =
        loop:
            Button.OnClick().Await()
            Player.SubtractAbility(Id, ?Amount := MaxAbilityLevel)
            UpdateAbilityText(Id)

    OnAbilityMaxClicked(Button:text_button_base, Id:ability_id)<suspends>:void =
        loop:
            Button.OnClick().Await()
            if (AbilityInfo := GetAbilityInfo[Id]):
                Player.AddAbility(AbilityInfo, ?Amount := MaxAbilityLevel)
                UpdateAbilityText(Id)

    OnClickSpawnNPC(Message:widget_message):void =
        Hide()
        UI := if(SpawnNPCUI := MaybeSpawnNPCUI?):
            SpawnNPCUI
        else:
            SpawnNPCUI := test_tool_spawn_npc_ui{Player := Player}
            set MaybeSpawnNPCUI = option{SpawnNPCUI}
            SpawnNPCUI

        UI.Show()
            
    OnClickTeleportToArena(Message:widget_message):void =
        if:
            set GetGameManager[].ActivePlayers += array{Player}
            Character := Player.GetFortCharacter[]
            Character.TeleportTo[vector3{Z := 100.0}, rotation{}]

            if (NPCManager := GetNPCManager[], not NPCManager.CurrentNPCWave?):
                set NPCManager.CurrentNPCWave = option{npc_wave{}}

    OnClickTeleportToHub(Message:widget_message):void =
        if:
            GameManager := GetGameManager[]
            Locator := GameManager.GameEndPositionLocator
            Character := Player.GetFortCharacter[]
            Character.TeleportTo[Locator.GetTransform().Translation + vector3{Z := CharacterHeightOffset}, Locator.GetTransform().Rotation]
        then:
            Hide()
            Player.PlayAudio(audio_id.Activate)

    OnClickToggleAbility(Message:widget_message):void =
        Player.ToggleAbility()

    OnClickMaxAbilities(Message:widget_message):void =
        if (PlayerStats := Player.GetStats[]):
            PlayerStats.SetMaxAbilities()
        
        Update()
        Player.PlayAudio(audio_id.Activate)

    OnClickStartBossFight(Message:widget_message):void =
        if (NPCManager := GetNPCManager[], GameManager := GetGameManager[]):
            GameManager.StopTimerEvent.Signal()
            spawn{GameManager.StartBossFight()}
            Hide()

    OnClickVictory(Message:widget_message):void =
        if (GameManager := GetGameManager[]):
            GameManager.Stop(true)
            Hide()

    OnClickIncrement(Message:widget_message):void =
        set DebugValue += IncrementAmount
        Player.SetArenaProgress(Floor[DebugValue] or 1)
        Update()

    OnClickDecrement(Message:widget_message):void =
        set DebugValue -= IncrementAmount
        Player.SetArenaProgress(Floor[DebugValue] or 1)
        Update()