using { /Fortnite.com/Devices }
using { /Verse.org/Simulation }
using { /UnrealEngine.com/Temporary/UI }
using { /UnrealEngine.com/Temporary/SpatialMath }

<# ----------------------------------------------------------------------------------------------------
    Data Definitions
---------------------------------------------------------------------------------------------------- #>
panel_ui_id := enum:
    WeaponSelect
    DailyQuest
    PlayerProgress
    PlayerPowerUps
    BoonCodex

<# ----------------------------------------------------------------------------------------------------
    Global Functions
---------------------------------------------------------------------------------------------------- #>
var GlobalUIManager:weak_map(session, ui_manager) = map{}

GetUIManager()<decides><transacts>:ui_manager =
    GlobalUIManager[GetSession()]

WaitForUIManager(Player:player)<suspends>:void =
    loop:
        if (GetUIManager[].GetIsInitialized[Player]) {break}
        Sleep(0.3)

(Player:player).AddWidgetToHUD(CanvasSlot:canvas_slot):void =
    if (UIManager := GetUIManager[]):
        UIManager.AddWidgetToHUD(Player, CanvasSlot)

(Player:player).ToggleHUD(IsShow:logic):void =
    if (UIManager := GetUIManager[]):
        UIManager.ToggleHUD(Player, IsShow)

(Player:player).TogglePlayerHUD(IsShow:logic):void =
    if (UIManager := GetUIManager[]):
        UIManager.TogglePlayerHUD(Player, IsShow)

(Player:player).ShowGamePausedUI():void =
    if (UIManager := GetUIManager[]):
        UIManager.ShowGamePausedUI(Player)

(Player:player).GetActivePanel()<decides><transacts>:panel_ui =
    GetUIManager[].GetActivePanel[Player]

(Player:player).UpdateActivePanel(PanelUI:panel_ui):void =
    if (UIManager := GetUIManager[]):
        UIManager.UpdateActivePanel(Player, PanelUI)

(Player:player).ClearActivePanel():void =
    if (UIManager := GetUIManager[]):
        UIManager.ClearActivePanel(Player)

(Player:player).CloseActivePanel(?IsSignal:logic = false):void =
    if (UIManager := GetUIManager[]):
        UIManager.CloseActivePanel(Player, ?IsSignal := IsSignal)

(Player:player).CloseActivePanel():void =
    Player.CloseActivePanel(?IsSignal := false)

(Player:player).UpdatePowerUpIndicator():void =
    if (UIManager := GetUIManager[]):
        UIManager.UpdatePowerUpIndicator(Player)

(Player:player).ShowPanelUI(Id:panel_ui_id):void =
    if (UIManager := GetUIManager[]):
        UIManager.ShowPanelUI(Player, Id)

<# ----------------------------------------------------------------------------------------------------
    ui_manager
---------------------------------------------------------------------------------------------------- #>
ui_manager := class(creative_device):
    @editable
    HUDController:hud_controller_device = hud_controller_device{}

    @editable
    ButtonMenu:button_device = button_device{}
    
    @editable
    PowerUpBeacon:beacon_device = beacon_device{}
    
    @editable
    PowerUpBeaconNew:beacon_device = beacon_device{}

    @editable
    InputControlKBMControllerBeacon:beacon_device = beacon_device{}
    
    @editable
    InputControlMobileBeacon:beacon_device = beacon_device{}

    @editable
    InputControlSwitch:switch_device = switch_device{}

    @editable
    GameResultView:game_result_view = game_result_view{}

    var PlayerHUDs:[player]canvas = map{}
    var PlayerHUDIsShowing:[player]logic = map{}
    var PlayerIsInitialized<private>:[player]logic = map{}
    var PlayerGamePausedUI<private>:[player]game_paused_ui = map{}
    var PlayerMenuUIs:[player]player_menu_ui = map{}
    var PlayerAbilityStatInfoUI:[player]ability_stat_info_ui = map{}
    var PlayerActivePanel:[player]panel_ui = map{}
    var PlayerWeaponEquippedUI:[player]weapon_equipped_ui = map{}
    var PlayerPanelUIs:[player][panel_ui_id]panel_ui = map{}

    <# ----------------------------------------------------------------------------------------------------
        Initialization
    ---------------------------------------------------------------------------------------------------- #>
    OnBegin<override>()<suspends>:void =
        if (set GlobalUIManager[GetSession()] = Self) {}

        for (Player : GetPlayspace().GetPlayers()):
            spawn{Initialize(Player)}

        GetPlayspace().PlayerAddedEvent().Subscribe(OnPlayerAdded)
        GetPlayspace().PlayerRemovedEvent().Subscribe(OnPlayerRemoved)
        ButtonMenu.InteractedWithEvent.Subscribe(OnInteractedWithMenu)
        InputControlSwitch.TurnedOnEvent.Subscribe(OnTurnedOn)
        InputControlSwitch.TurnedOffEvent.Subscribe(OnTurnedOff)

        spawn{GameResultView.OnBegin()}

        WaitForInputManager()
        if (InputManager := GetInputManager[]):
            InputManager.InputUnregisteredEvent.Subscribe(OnInputUnregistered)
            
        if (MenuInputPressedEvent := GetInputPressedEvent[input_id.Menu]):
            MenuInputPressedEvent.Subscribe(OnMenuInputPressedEvent)

        if (AbilityStatsPressedEvent := GetInputPressedEvent[input_id.AbilityStats], AbilityStatsReleasedEvent := GetInputReleasedEvent[input_id.AbilityStats]):
            AbilityStatsPressedEvent.Subscribe(OnAbilityStatsInputPressedEvent)
            AbilityStatsReleasedEvent.Subscribe(OnAbilityStatsInputReleasedEvent)

    Initialize(Player:player)<suspends>:void =
        if (PlayerIsInitialized[Player]?) then return

        PlayerHUD := canvas{}
        if (PlayerUI := GetPlayerUI[Player], set PlayerHUDs[Player] = PlayerHUD):
            PlayerUI.AddWidget(PlayerHUD, player_ui_slot{InputMode := ui_input_mode.None})

        AbilityStatInfoUI := ability_stat_info_ui{Player := Player}
        AbilityStatInfoUI.Initialize()
        AbilityStatInfoUI.Hide()

        Player.AddWidgetToHUD of canvas_slot:
            Anchors := anchors{Minimum := vector2{X := 0.15, Y := 0.5}, Maximum := vector2{X := 0.85, Y := 0.5}}
            Offsets := margin{Left := 0.0, Top := 0.0, Right := 0.0, Bottom := 0.0}
            Alignment := vector2{X := 0.5, Y := 0.5}
            Widget := AbilityStatInfoUI.GetUI()

        WeaponEquippedUI := weapon_equipped_ui{Player := Player}
        WeaponEquippedUI.Initialize()
        Player.AddWidgetToHUD of canvas_slot:
            Anchors := anchors{Minimum := vector2{X := 1.0, Y := 0.0}, Maximum := vector2{X := 1.0, Y := 0.0}}
            Offsets := margin{Left := -210.0, Top := 404.0, Right := 0.0, Bottom := 0.0}
            Alignment := vector2{X := 1.0, Y := 0.0}
            Widget := WeaponEquippedUI.GetUI()

        if:
            set PlayerMenuUIs[Player] = player_menu_ui{Player := Player}
            set PlayerAbilityStatInfoUI[Player] = AbilityStatInfoUI
            set PlayerWeaponEquippedUI[Player] = WeaponEquippedUI
            set PlayerIsInitialized[Player] = true

        WaitForInputManager()
        Player.RegisterInput(input_id.Menu)
        Player.UpdatePowerUpIndicator()

        if (InputControlSwitch.GetCurrentState[Player]):
            InputControlKBMControllerBeacon.AddToShowList(Player)
        else:
            InputControlMobileBeacon.AddToShowList(Player)

    <# ----------------------------------------------------------------------------------------------------
        Get Functions
    ---------------------------------------------------------------------------------------------------- #>
    GetIsInitialized(Player:player)<decides><transacts>:void =
        PlayerIsInitialized[Player]

    GetActivePanel(Player:player)<decides><transacts>:panel_ui =
        PlayerActivePanel[Player]

    <# ----------------------------------------------------------------------------------------------------
        Update Functions
    ---------------------------------------------------------------------------------------------------- #>
    AddWidgetToHUD(Player:player, CanvasSlot:canvas_slot):void =
        if (HUD := PlayerHUDs[Player]):
            HUD.AddWidget(CanvasSlot)

    ToggleHUD(Player:player, IsShow:logic):void =
        TogglePlayerHUD(Player, IsShow)
        if (IsShow?) then HUDController.Disable() else HUDController.Enable()

    TogglePlayerHUD(Player:player, IsShow:logic):void =
        if (PlayerUI := GetPlayerUI[Player], PlayerHUD := PlayerHUDs[Player]):
            if (IsShow?):        
                PlayerUI.AddWidget(PlayerHUD, player_ui_slot{InputMode := ui_input_mode.None})
            else:
                PlayerUI.RemoveWidget(PlayerHUD)

    ShowGamePausedUI(Player:player):void =
        UI := if (GamePausedUI := PlayerGamePausedUI[Player]):
            GamePausedUI
        else:
            GamePausedUI := game_paused_ui{Player := Player}
            if (set PlayerGamePausedUI[Player] = GamePausedUI) {}
            GamePausedUI

        UI.Show()

    ShowPanelUI(Player:player, Id:panel_ui_id):void =
        if:
            not PlayerPanelUIs[Player]
            set PlayerPanelUIs[Player] = map{}
        
        UI := if (PanelUI := PlayerPanelUIs[Player][Id]):
            PanelUI
        else:
            PanelUI := case (Id):
                panel_ui_id.WeaponSelect => weapon_select_ui{Player := Player}
                panel_ui_id.PlayerProgress => player_progress_ui{Player := Player}
                panel_ui_id.PlayerPowerUps => power_up_ui{Player := Player}
                panel_ui_id.DailyQuest => daily_quests_ui{Player := Player}
                panel_ui_id.BoonCodex => boon_codex_ui{Player := Player}

            if (set PlayerPanelUIs[Player][Id] = PanelUI) {}
            PanelUI

        UI.Show()

    UpdateActivePanel(Player:player, PanelUI:panel_ui):void =
        if (set PlayerActivePanel[Player] = PanelUI) {}

    ClearActivePanel(Player:player):void =
        set PlayerActivePanel = RemoveKeyFromMap(PlayerActivePanel, Player)

    CloseActivePanel(Player:player, ?IsSignal:logic = false):void =
        if (ActivePanel := PlayerActivePanel[Player]):
            ActivePanel.Hide(?IsSignal := IsSignal)
            ClearActivePanel(Player)

    UpdatePowerUpIndicator(Player:player):void =
        if (Player.CanAffordPowerUp()?):
            PowerUpBeacon.RemoveFromShowList(Player)
            PowerUpBeaconNew.AddToShowList(Player)
        else:
            PowerUpBeacon.AddToShowList(Player)
            PowerUpBeaconNew.RemoveFromShowList(Player)

    <# ----------------------------------------------------------------------------------------------------
        Event Handlers
    ---------------------------------------------------------------------------------------------------- #>
    OnPlayerAdded(Player:player):void =
        spawn{Initialize(Player)}

    OnPlayerRemoved(Player:player):void =
        <#
        if (PlayerUI := GetPlayerUI[Player], PlayerHUD := PlayerHUDs[Player]):
            PlayerUI.RemoveWidget(PlayerHUD)
        #>

        set PlayerHUDs = RemoveKeyFromMap(PlayerHUDs, Player)
        set PlayerIsInitialized = RemoveKeyFromMap(PlayerIsInitialized, Player)
        set PlayerHUDIsShowing = RemoveKeyFromMap(PlayerHUDIsShowing, Player)
        set PlayerGamePausedUI = RemoveKeyFromMap(PlayerGamePausedUI, Player)
        set PlayerMenuUIs = RemoveKeyFromMap(PlayerMenuUIs, Player)
        set PlayerActivePanel = RemoveKeyFromMap(PlayerActivePanel, Player)
        set PlayerAbilityStatInfoUI = RemoveKeyFromMap(PlayerAbilityStatInfoUI, Player)
        set PlayerWeaponEquippedUI = RemoveKeyFromMap(PlayerWeaponEquippedUI, Player)
        set PlayerPanelUIs = RemoveKeyFromMap(PlayerPanelUIs, Player)

    OnMenuInputPressedEvent(Agent:agent):void =
        if (Player := player[Agent], PlayerMenuUI := PlayerMenuUIs[Player]):
            if (PlayerMenuUI.IsShowing?):
                PlayerMenuUI.Hide()
            else:
                PlayerMenuUI.Show()

    OnAbilityStatsInputPressedEvent(Agent:agent):void =
        if (Player := player[Agent], AbilityStatInfoUI := PlayerAbilityStatInfoUI[Player]):
            AbilityStatInfoUI.Show()
            AbilityStatInfoUI.Update()
        
    OnAbilityStatsInputReleasedEvent(Agent:agent, DurationHeld:float):void =
        if (Player := player[Agent], AbilityStatInfoUI := PlayerAbilityStatInfoUI[Player]):
            AbilityStatInfoUI.Hide()

    OnInputUnregistered(Player:player):void =
        if (AbilityStatInfoUI := PlayerAbilityStatInfoUI[Player]):
            AbilityStatInfoUI.Hide()

    OnInteractedWithMenu(Agent:agent):void =
        if (Player := player[Agent], MenuUI := PlayerMenuUIs[Player]):
            MenuUI.Show()

    OnTurnedOn(Agent:agent):void =
        InputControlMobileBeacon.RemoveFromShowList(Agent)
        InputControlKBMControllerBeacon.AddToShowList(Agent)

    OnTurnedOff(Agent:agent):void =
        InputControlKBMControllerBeacon.RemoveFromShowList(Agent)
        InputControlMobileBeacon.AddToShowList(Agent)