using { /Fortnite.com/Devices }
using { /Verse.org/Simulation }
using { /UnrealEngine.com/Temporary/SpatialMath }

<# ----------------------------------------------------------------------------------------------------
    Data Definitions
---------------------------------------------------------------------------------------------------- #>
vfx_id := enum:
    DamageNumber
    DamageNumberCriticalHit
    DamageNumberPlayer
    Gather
    Gathering
    Impact
    AbilityBow
    NPCProjectile
    NPCProjectileCharging
    NPCJump
    NPCArea
    NPCDash
    AbilityArea
    AbilityShuriken
    AbilitySword
    AbilitySentry
    Hearts
    ImpactArea
    DashAttack
    RingProgress
    AbilityLightning
    AbilityTorrent
    AbilityTrail
    AbilityShield
    AbilityTrailEvolution
    RingProgressRevive
    Stun
    AbilityLightningEvolved
    AbilitySwordEvolved
    AbilityBowEvolved
    AbilitySentryEvolved
    ImpactAreaEvolved
    AbilityShurikenEvolved
    AbilityTorrentEvolved
    AbilityShieldEvolved
    AbilityTrailBlast
    NPCProjectileWarning
    AbilityChainLightning
    AbilityChainLightningEvolved
    AbilityEnergyBomb
    AbilityEnergyBombEvolved
    AbilityBoomerang
    AbilityBoomerangEvolved
    Revive
    Dodged
    GatheringChainsaw
    NPCExplode
    ReviveBlast

vfx := class<concrete>:
    @editable
    Id:vfx_id = vfx_id.DamageNumber
    @editable
    Prop:creative_prop_asset = DefaultCreativePropAsset

vfx_powerup_id := enum:
    Base
    DamagedGlow
    Shield
    Arrow
    ArrowOrbit
    ShieldEvolved
    LevelUp
    LevelUpText
    Rank01
    Rank02
    Rank03
    Rank04
    Rank05
    Rank06
    Rank07
    AFK
    Invulnerability
    Rank08
    Rank09
    Rank10
    TreasureChestCollect
    TreasureChestOpen
    Heal
    ShurikenBase

vfx_powerup := class<concrete>:
    @editable
    Id:vfx_powerup_id = vfx_powerup_id.Base
    @editable
    Device:visual_effect_powerup_device = visual_effect_powerup_device{}
    @editable
    DisableDevice:?visual_effect_powerup_device = false

vfx_powerup_sequence_id := enum:
    AbilityAura
    AbilityAuraEvolved

vfx_powerup_sequence := class<concrete>:
    var AssignedPlayer:?player = false
    var MappedDevices:[vfx_powerup_sequence_id]vfx_powerup_sequence_devices = map{}
    @editable
    Devices:[]vfx_powerup_sequence_devices = array{}

MakeVFXPowerSequence<constructor>(VFXPowerSequence:vfx_powerup_sequence)<transacts> := vfx_powerup_sequence:
    AssignedPlayer := VFXPowerSequence.AssignedPlayer
    MappedDevices := VFXPowerSequence.MappedDevices
    Devices := VFXPowerSequence.Devices

vfx_powerup_sequence_devices := class<concrete>:
    @editable
    Id:vfx_powerup_sequence_id = vfx_powerup_sequence_id.AbilityAura
    @editable
    Device:visual_effect_powerup_device = visual_effect_powerup_device{}
    @editable
    DisableDevice:visual_effect_powerup_device = visual_effect_powerup_device{}
    @editable
    CinematicSequence:cinematic_sequence_device = cinematic_sequence_device{}

<# ----------------------------------------------------------------------------------------------------
    Global Functions
---------------------------------------------------------------------------------------------------- #>
var GlobalVFXManager:weak_map(session, vfx_manager) = map{}

WaitForVFXManager()<suspends>:void =
    loop:
        if (GetVFXManager[].GetIsInitialized()) {break}
        Sleep(0.3)

GetVFXManager()<decides><transacts>:vfx_manager =
    GlobalVFXManager[GetSession()]

SpawnVFX(Id:vfx_id, Transform:transform, ?Duration:float = 5.0):?creative_prop =
    if (VFXManager := GetVFXManager[], VFX := VFXManager.GetVFX[Id]):
        SpawnPropResult := SpawnProp(VFX, Transform)
        if (Duration > 0.0):
            spawn{DisposeProp(SpawnPropResult(0), Duration)}
        SpawnPropResult(0)
    else:
        false

SpawnVFX(Id:vfx_id, Position:vector3, ?Rotation:rotation = IdentityRotation(), ?Scale:vector3 = UnitVector, ?Duration:float = 5.0):?creative_prop =
    SpawnVFX(Id, transform{Translation := Position, Rotation := Rotation, Scale := Scale}, ?Duration := Duration)

(Agent:agent).ActivateVFXPowerup(Id:vfx_powerup_id, ?Duration:float = 0.0):void =
    if (VFXManager := GetVFXManager[]):
        VFXManager.ActivateVFXPowerup(Agent, Id, ?Duration := Duration)

(Agent:agent).DeactivateVFXPowerup(Id:vfx_powerup_id):void =
    if (VFXManager := GetVFXManager[]):
        VFXManager.DeactivateVFXPowerup(Agent, Id)

(Agent:agent).ActivateVFXPowerup(Id:vfx_powerup_sequence_id, ?Duration:float = 0.0):vfx_powerup_sequence_devices =
    if (VFXManager := GetVFXManager[]):
        VFXManager.ActivateVFXPowerup(Agent, Id, ?Duration := Duration)
    else:
        vfx_powerup_sequence_devices{}

(Agent:agent).DeactivateVFXPowerup(Id:vfx_powerup_sequence_id):void =
    if (VFXManager := GetVFXManager[]):
        VFXManager.DeactivateVFXPowerup(Agent, Id)

<# ----------------------------------------------------------------------------------------------------
    vfx_manager
---------------------------------------------------------------------------------------------------- #>
vfx_manager := class(creative_device):
    @editable
    VFXs:[]vfx = array{}

    @editable
    VFXPowerups:[]vfx_powerup = array{}

    @editable
    var VFXPowerUpSequences:[]vfx_powerup_sequence = array{}

    var MappedVFXs:[vfx_id]creative_prop_asset = map{}
    var MappedVFXPowerups:[vfx_powerup_id]vfx_powerup = map{}
    var PlayerVFXPowerupSequences:[player]vfx_powerup_sequence = map{}

    var IsInitialized:logic = false

    OnBegin<override>()<suspends>:void =
        if (set GlobalVFXManager[GetSession()] = Self) {}

        for (VFX : VFXs, set MappedVFXs[VFX.Id] = VFX.Prop) {}
        for (VFXPowerup : VFXPowerups, set MappedVFXPowerups[VFXPowerup.Id] = VFXPowerup) {}
        for (VFXPowerUpSequence : VFXPowerUpSequences):
            for (Devices : VFXPowerUpSequence.Devices, set VFXPowerUpSequence.MappedDevices[Devices.Id] = Devices) {}

        for (Player : GetPlayspace().GetPlayers()):
            Initialize(Player)
        
        GetPlayspace().PlayerAddedEvent().Subscribe(OnPlayerAdded)
        GetPlayspace().PlayerRemovedEvent().Subscribe(OnPlayerRemoved)
        
        set IsInitialized = true

    Initialize(Player:player):void =
        AssignVFXPowerupSequence(Player)

    <# ----------------------------------------------------------------------------------------------------
        Get Functions
    ---------------------------------------------------------------------------------------------------- #>
    GetIsInitialized()<transacts>:logic =
        IsInitialized

    GetVFX(Id:vfx_id)<decides><transacts>:creative_prop_asset =
        MappedVFXs[Id]

    GetVFXPowerup(Id:vfx_powerup_id)<decides><transacts>:visual_effect_powerup_device =
        MappedVFXPowerups[Id].Device

    <# ----------------------------------------------------------------------------------------------------
        Update Functions
    ---------------------------------------------------------------------------------------------------- #>
    ActivateVFXPowerup(Agent:agent, Id:vfx_powerup_id, ?Duration:float = 0.0):void =
        if (VFXPowerup := GetVFXPowerup[Id]):
            VFXPowerup.SetDuration(Duration)
            VFXPowerup.Pickup(Agent)

    DeactivateVFXPowerup(Agent:agent, Id:vfx_powerup_id):void =
        if (VFXPowerup := MappedVFXPowerups[Id].DisableDevice?):
            VFXPowerup.Pickup(Agent)
            return

        if (VFXPowerup := GetVFXPowerup[Id]):
            VFXPowerup.SetDuration(TickTime)
            VFXPowerup.Pickup(Agent)

    AssignVFXPowerupSequence(Player:player):void =
        if (PlayerVFXPowerupSequences[Player]) then return
        for (Index -> VFXPowerUpSequence : VFXPowerUpSequences, not VFXPowerUpSequence.AssignedPlayer?):
            if:
                set VFXPowerUpSequences[Index] = vfx_powerup_sequence:
                    MakeVFXPowerSequence<constructor>(VFXPowerUpSequence)
                    AssignedPlayer := option{Player}

                set PlayerVFXPowerupSequences[Player] = VFXPowerUpSequence

            return

    UnassignVFXPowerupSequence(Player:player):void =
        for (Index -> VFXPowerUpSequence : VFXPowerUpSequences, Player = VFXPowerUpSequence.AssignedPlayer?):
            if:
                set VFXPowerUpSequences[Index] = vfx_powerup_sequence:
                    MakeVFXPowerSequence<constructor>(VFXPowerUpSequence)
                    AssignedPlayer := false
            then:
                set PlayerVFXPowerupSequences = RemoveKeyFromMap(PlayerVFXPowerupSequences, Player)

            return

    ActivateVFXPowerup(Agent:agent, Id:vfx_powerup_sequence_id, ?Duration:float = 0.0):vfx_powerup_sequence_devices =
        if (Player := player[Agent], VFXPowerupSequence := PlayerVFXPowerupSequences[Player].MappedDevices[Id]):
            VFXPowerupSequence.Device.SetDuration(Duration)
            VFXPowerupSequence.Device.Pickup(Player)
            return VFXPowerupSequence

        return vfx_powerup_sequence_devices{}

    DeactivateVFXPowerup(Agent:agent, Id:vfx_powerup_sequence_id):void =
        if (Player := player[Agent], VFXPowerupSequence := PlayerVFXPowerupSequences[Player].MappedDevices[Id]):
            VFXPowerupSequence.DisableDevice.Pickup(Player)

    <# ----------------------------------------------------------------------------------------------------
        Event Handlers
    ---------------------------------------------------------------------------------------------------- #>
    OnPlayerAdded(Player:player):void =
        Initialize(Player)

    OnPlayerRemoved(Player:player):void =
        UnassignVFXPowerupSequence(Player)