using { /Fortnite.com/Characters }
using { /Fortnite.com/Devices }
using { /Verse.org/Simulation }
using { /UnrealEngine.com/Temporary/SpatialMath }

MinScale:float = 0.000001

<# ----------------------------------------------------------------------------------------------------
    volume_activator
---------------------------------------------------------------------------------------------------- #>
volume_activator := class<concrete>:
    @editable
    VolumeDevice:volume_device = volume_device{}

    @editable
    MapIndicatorDevice:?map_indicator_device = false

    @editable
    VfxId:vfx_id = vfx_id.RingProgress

    @editable
    IsProgressAlwaysVisible:logic = false

    @editable
    IsAudio:logic = false

    @editable
    var ActivationTime:float = 1.0
    
    @editable
    DecayRate:float = 0.0
    
    @editable
    Size:float = 20.0

    ActivatedEvent:events(player) = events(player){}
    ActivateCancelledEvent:events(tuple()) = events(tuple()){}

    var InitialPosition:vector3 = vector3{}
    var PlayersActivating:[]player = array{}
    var MaybeSpawnedVFX:?creative_prop = false

    <# ----------------------------------------------------------------------------------------------------
        Initialization
    ---------------------------------------------------------------------------------------------------- #>
    OnBegin()<suspends>:void =
        set InitialPosition = VolumeDevice.GetTransform().Translation
        
        if (MapIndicator := MapIndicatorDevice?):
            if (MapIndicator.TeleportTo[MapIndicatorDisabledPosition, IdentityRotation()]) {}

        Disable()
        VolumeDevice.AgentEntersEvent.Subscribe(OnAgentEnters)
        VolumeDevice.AgentExitsEvent.Subscribe(OnAgentExits)

    <# ----------------------------------------------------------------------------------------------------
        Get Functions
    ---------------------------------------------------------------------------------------------------- #>
    GetActivationTime()<transacts>:float =
        ActivationTime * (1.0 - GetProgress())
        
    GetDecayTime()<transacts>:float =
        (Floor[ActivationTime / DecayRate * GetProgress()] or 0) * 1.0

    GetProgress()<transacts>:float =
        MaybeSpawnedVFX?.GetTransform().Scale.Z or 0.0

    GetIsActivating()<transacts>:logic =
        var IsActivating:logic = false
        for:
            Player : PlayersActivating
            not IsActivating?
            not Player.GetFortCharacter[].IsDownButNotOut[]
        do:
            set IsActivating = true

        IsActivating

    <# ----------------------------------------------------------------------------------------------------
        Update Functions
    ---------------------------------------------------------------------------------------------------- #>
    Enable(?TargetPosition:?vector3 = false, ?OverrideActivationTime:?float = false):void =
        if (Position := TargetPosition?):
            set InitialPosition = Position.Make2D()

        if (OverrideTime := OverrideActivationTime?):
            set ActivationTime = OverrideTime

        if (IsProgressAlwaysVisible?):
            set MaybeSpawnedVFX = SpawnVFX(VfxId, InitialPosition, ?Scale := vector3{X := Size, Y := Size, Z := MinScale}, ?Duration := 0.0)

        if (VolumeDevice.TeleportTo[InitialPosition, IdentityRotation()]) {}

        UpdateMapIndicator(InitialPosition)

    UpdateMapIndicator(Position:vector3):void =
        if (MapIndicator := MapIndicatorDevice?):
            if (MapIndicator.TeleportTo[Position, IdentityRotation()]) {}

    Disable():void =
        ActivateCancelledEvent.Signal()

        if (VolumeDevice.TeleportTo[vector3{Z := -1000.0}, IdentityRotation()]) {}
        
        if (SpawnedVFX := MaybeSpawnedVFX?):
            SpawnedVFX.Dispose()
            set MaybeSpawnedVFX = false

        if (MapIndicator := MapIndicatorDevice?):
            if (MapIndicator.TeleportTo[MapIndicatorDisabledPosition, IdentityRotation()]) {}

    StartActivate()<suspends>:void =
        if (not MaybeSpawnedVFX?):
            set MaybeSpawnedVFX = SpawnVFX(VfxId, InitialPosition, ?Scale := vector3{X := Size, Y := Size, Z := MinScale}, ?Duration := ActivationTime)

        defer:
            set PlayersActivating = array{}

            if (SpawnedVFX := MaybeSpawnedVFX?):
                if (IsProgressAlwaysVisible?):
                    TargetTransform := transform:
                        Translation := SpawnedVFX.GetTransform().Translation
                        Rotation := SpawnedVFX.GetTransform().Rotation
                        Scale := vector3{X := SpawnedVFX.GetTransform().Scale.X, Y := SpawnedVFX.GetTransform().Scale.Y, Z := MinScale}

                    if (SpawnedVFX.TeleportTo[TargetTransform]) {}
                else:
                    SpawnedVFX.Dispose()
                    set MaybeSpawnedVFX = false

        race:
            ActivateCancelledEvent.Await()

            if (DecayRate = 0.0):
                loop:
                    if (PlayerManager := GetPlayerManager[]):
                        PlayerManager.DownButNotOut.AgentDownedEvent.Await()
                        if (not GetIsActivating()?):
                            break

                    Sleep(0.0)
            else:
                Sleep(Inf)

            if (SpawnedVFX := MaybeSpawnedVFX?, Player := PlayersActivating[0]):
                if (DecayRate = 0.0):
                    SpawnedVFX.MoveTo(transform{Translation := InitialPosition, Rotation := IdentityRotation(), Scale := vector3{X := Size, Y := Size, Z := 1.0}}, ActivationTime)
                    ActivatedEvent.Signal(Player)
                else:
                    loop:
                        IsActivating := GetIsActivating()

                        TargetTransform := transform:
                            Translation := VolumeDevice.GetTransform().Translation
                            Rotation := VolumeDevice.GetTransform().Rotation
                            Scale := vector3{X := Size, Y := Size, Z := IsActivating? and 1.0 or MinScale}
                            
                        OverTime := IsActivating? and GetActivationTime() or GetDecayTime()
                        
                        MoveToResult := race:
                            SpawnedVFX.MoveTo(TargetTransform, Max(MinScale, OverTime))

                            if (AgentDownedEvent := GetGameManager[].DownButNotOut.AgentDownedEvent):
                                AgentDownedEvent.Await()
                                move_to_result.WillNotReachDestination

                            block:
                                if (IsActivating?):
                                    VolumeDevice.AgentExitsEvent.Await()
                                else:
                                    VolumeDevice.AgentEntersEvent.Await()

                                move_to_result.WillNotReachDestination
                            
                        if (IsActivating?, MoveToResult = move_to_result.DestinationReached):
                            ActivatedEvent.Signal(Player)
                            break

                        Sleep(0.0)

    <# ----------------------------------------------------------------------------------------------------
        Event Handlers
    ---------------------------------------------------------------------------------------------------- #>
    OnAgentEnters(Agent:agent):void =
        if (Player := player[Agent], Character := Player.GetFortCharacter[]):
            if (Player.GetIsAFK()?) then return
            if (Character.IsDownButNotOut[]) then return
            if (Player.IsPlayerPaused()?) then return

            if (not GetIsActivating()?):
                set PlayersActivating += array{Player}
                spawn{StartActivate()}
            else:
                set PlayersActivating += array{Player}

            if (IsAudio?):
                Player.PlayAudio(audio_id.Capture)

    OnAgentExits(Agent:agent):void =
        if (Player := player[Agent]):
            set PlayersActivating = PlayersActivating.RemoveAllElements(Player)

            if (DecayRate = 0.0 and PlayersActivating.Length = 0):
                ActivateCancelledEvent.Signal()

            if (IsAudio?):
                Player.StopAudio(audio_id.Capture)