using { /Fortnite.com/UI }
using { /Verse.org/Colors }
using { /Verse.org/Simulation }
using { /UnrealEngine.com/Temporary/SpatialMath }
using { /UnrealEngine.com/Temporary/UI }

<# ----------------------------------------------------------------------------------------------------
    Constants
---------------------------------------------------------------------------------------------------- #>

<# ----------------------------------------------------------------------------------------------------
    Data Definitions
---------------------------------------------------------------------------------------------------- #>


<# ----------------------------------------------------------------------------------------------------
    weapon_evolution_entry_ui
---------------------------------------------------------------------------------------------------- #>
weapon_evolution_entry_ui := class(ui_component):
    SelectEvent:events(tuple(ability_info, int)) = events(tuple(ability_info, int)){}

    SelectButton:button_quiet = button_quiet{}

    Icon:texture_block = texture_block:
        DefaultImage := UserInterface.Icons.WeaponSword_128
        DefaultDesiredSize := vector2{X := 128.0, Y := 128.0}

    NameText:text_block = text_block:
        DefaultTextColor := ThemeColors.PrimaryText
        DefaultShadowOffset := option{vector2{X := 2.0, Y := 2.0}}
        DefaultShadowOpacity := 0.5

    DescriptionText:text_block = text_block:
        DefaultTextColor := ThemeColors.PrimaryText
        DefaultTextOpacity := 0.5

    StatText:text_block = text_block:
        DefaultTextColor := ThemeColors.PrimaryText
        DefaultShadowOffset := option{vector2{X := 2.0, Y := 2.0}}
        DefaultShadowOpacity := 0.5

    StatDeltaText:text_block = text_block:
        DefaultTextColor := ThemeColors.Success
        DefaultShadowColor := ThemeColors.Success * 0.125
        DefaultShadowOffset := option{vector2{X := 2.0, Y := 2.0}}
        DefaultShadowOpacity := 1.0

    var MaybeAbilityInfo:?ability_info = false
    var MaybeWeaponInfo:?weapon_info = false

    <# ----------------------------------------------------------------------------------------------------
        Create UI
    ---------------------------------------------------------------------------------------------------- #>
    CreateUI<override>():widget =
        StackBox := stack_box:
            Orientation := orientation.Vertical
            Slots := array:
                stack_box_slot:
                    HorizontalAlignment := horizontal_alignment.Fill
                    VerticalAlignment := vertical_alignment.Fill
                    Distribution := option{1.0}
                    Padding := margin{Left := 0.0, Top := 0.0, Right := 0.0, Bottom := 0.0}
                    Widget := overlay:
                        Slots := array:
                            overlay_slot:
                                HorizontalAlignment := horizontal_alignment.Fill
                                VerticalAlignment := vertical_alignment.Fill
                                Widget := MakeBox(?Width := 610.0)
                            overlay_slot:
                                HorizontalAlignment := horizontal_alignment.Fill
                                VerticalAlignment := vertical_alignment.Fill
                                Padding := margin{Left := 40.0, Top := 40.0, Right := 40.0, Bottom := 40.0}
                                Widget := CreateInfoUI()

        SelectButton.SetText(MakeMessage("Select"))
        SelectButton.OnClick().Subscribe(OnClickSelect)
        StackBox.AddWidget of stack_box_slot:
            HorizontalAlignment := horizontal_alignment.Fill
            VerticalAlignment := vertical_alignment.Fill
            Padding := margin{Left := 0.0, Top := 10.0, Right := 0.0, Bottom := 0.0}
            Widget := SelectButton

        return StackBox

    CreateInfoUI():widget =
        stack_box:
            Orientation := orientation.Vertical
            Slots := array:
                stack_box_slot:
                    HorizontalAlignment := horizontal_alignment.Fill
                    Widget := color_block:
                        DefaultOpacity := 0.0
                        DefaultDesiredSize := vector2{X := 450.0, Y := 0.0}
                stack_box_slot:
                    HorizontalAlignment := horizontal_alignment.Left
                    VerticalAlignment := vertical_alignment.Top
                    Padding := margin{Left := 0.0, Top := 0.0, Right := 0.0, Bottom := 20.0}
                    Widget := Icon
                stack_box_slot:
                    HorizontalAlignment := horizontal_alignment.Left
                    VerticalAlignment := vertical_alignment.Fill
                    Widget := NameText
                stack_box_slot:
                    HorizontalAlignment := horizontal_alignment.Left
                    VerticalAlignment := vertical_alignment.Fill
                    Distribution := option{1.0}
                    Padding := margin{Left := 0.0, Top := 0.0, Right := 0.0, Bottom := 0.0}
                    Widget := DescriptionText
                stack_box_slot:
                    VerticalAlignment := vertical_alignment.Fill
                    Widget := stack_box:
                        Orientation := orientation.Horizontal
                        Slots := array:
                            stack_box_slot:
                                HorizontalAlignment := horizontal_alignment.Fill
                                Padding := margin{Left := 0.0, Top := 0.0, Right := 0.0, Bottom := 0.0}
                                Widget := StatText
                            stack_box_slot:
                                HorizontalAlignment := horizontal_alignment.Fill
                                Padding := margin{Left := 0.0, Top := 0.0, Right := 0.0, Bottom := 0.0}
                                Widget := StatDeltaText

    <# ----------------------------------------------------------------------------------------------------
        Get Functions
    ---------------------------------------------------------------------------------------------------- #>
    GetActiveAbilityInfo()<transacts>:?ability_info =
        MaybeAbilityInfo

    <# ----------------------------------------------------------------------------------------------------
        Update Functions
    ---------------------------------------------------------------------------------------------------- #>
    Update(AbilityInfo:ability_info, ?IsPreview:logic = false):void =        
        set MaybeAbilityInfo = option{AbilityInfo}

        Icon.SetImage(AbilityInfo.Icon)
        Icon.SetTint(AbilityInfo.IconColor)
        NameText.SetText(MakeMessage("{AbilityInfo.Name}"))
        DescriptionText.SetText(MakeMessage("{AbilityInfo.Description.WrapText(450.0)}"))
        SelectButton.SetVisibility(IsPreview? and widget_visibility.Collapsed or widget_visibility.Visible)

        if (TierValue := AbilityInfo.TierValue[0]?, WeaponId := AbilityInfo.MaybeWeaponId?):
            CurrentValue := Player.GetWeaponEvolutionValue(WeaponId, AbilityInfo.Id)
            UpdatedValue := CurrentValue + TierValue
            FormattedValue := AbilityInfo.FormatValue(CurrentValue)
            if (FormattedValue = "UNIQUE"):
                StatText.SetText(MakeMessage(""))
                StatDeltaText.SetText(MakeMessage(FormattedValue))
            else:
                StatText.SetText(MakeMessage("{FormattedValue} ▶ "))
                StatDeltaText.SetText(MakeMessage("{AbilityInfo.FormatValue(UpdatedValue)} ({TierValue > 0.0 and "+" or ""}{AbilityInfo.FormatValue(TierValue)})"))

    <# ----------------------------------------------------------------------------------------------------
        Event handlers
    ---------------------------------------------------------------------------------------------------- #>
    OnClickSelect(Message:widget_message):void =
        if (AbilityInfo := MaybeAbilityInfo?):
            SelectEvent.Signal(AbilityInfo, 0)