using { /Fortnite.com/UI }
using { /Verse.org/Colors }
using { /Verse.org/Random }
using { /Verse.org/Simulation }
using { /UnrealEngine.com/Temporary/SpatialMath }
using { /UnrealEngine.com/Temporary/UI }

<# ----------------------------------------------------------------------------------------------------
    Constants
---------------------------------------------------------------------------------------------------- #>


<# ----------------------------------------------------------------------------------------------------
    Data Definitions
---------------------------------------------------------------------------------------------------- #>

<# ----------------------------------------------------------------------------------------------------
    Global Functions
---------------------------------------------------------------------------------------------------- #>

<# ----------------------------------------------------------------------------------------------------
    weapon_evolution_ui
---------------------------------------------------------------------------------------------------- #>
weapon_evolution_ui := class(panel_ui):
    Title<override>:string = "Weapon Evolution"
    HasCloseButton<override>:logic = false
    BackgroundOpacity<override>:type {_X:float where 0.000000 <= _X, _X <= 1.000000} = 0.7
    
    CancelAnimationEvent:events(tuple()) = events(tuple()){}

    StatsInfoButton:button_quiet = button_quiet{}

    WeaponIcon:texture_block = texture_block:
        DefaultImage := UserInterface.Icons.WeaponBow_128
        DefaultDesiredSize := vector2{X := 128.0, Y := 128.0}
        
    WeaponNameText:text_block = text_block:
        DefaultTextColor := ThemeColors.PrimaryText

    CloseButton:button_quiet = button_quiet{}

    var MaybeInfoUI:?ui_component = false
    var WeaponEvolutionEntryUIs:[]weapon_evolution_entry_ui = array{}
    var AugmentSlotUIs:[]tuple(texture_block, texture_block, texture_block, text_button_base) = array{}
    
    var Id:weapon_id = weapon_id.Bow
    var ItemShop:logic = false
    var IsPreview:logic = false

    <# ----------------------------------------------------------------------------------------------------
        Initialization
    ---------------------------------------------------------------------------------------------------- #>
    CreateContent<override>():widget =
        ClosedEvent.Subscribe(OnClosed)
        CloseButton.SetText(MakeMessage("Close"))
        CloseButton.OnClick().Subscribe(OnClickClose)

        if (GameManager := GetGameManager[]):
            GameManager.ResumeCountdownStartEvent.Subscribe(OnResumeCountdownStart)
            GameManager.ResumeCountdownCompleteEvent.Subscribe(OnResumeCountdownComplete)
            GameManager.ResumeCountdownEvent.Subscribe(OnResumeCountdown)
            GameManager.GameResumedEvent.Subscribe(OnGameResumed)

        StackBox := stack_box:
            Orientation := orientation.Horizontal
            
        for (Index := 1..3):
            StackBox.AddWidget of stack_box_slot:
                HorizontalAlignment := horizontal_alignment.Fill
                VerticalAlignment := vertical_alignment.Fill
                Padding := margin{Left := (Index = 1) and 0.0 or 20.0, Top := 0.0, Right := 0.0, Bottom := 0.0}
                Distribution := option{1.0}
                Widget := if (Index = 2):
                    CreateInfoUI()
                else:
                    WeaponEvolutionEntryUI := weapon_evolution_entry_ui{Player := Player}
                    WeaponEvolutionEntryUI.Initialize()
                    WeaponEvolutionEntryUI.SelectEvent.Subscribe(OnSelect)

                    set WeaponEvolutionEntryUIs += array{WeaponEvolutionEntryUI}
                    WeaponEvolutionEntryUI.GetUI()

        StatsInfoButton.SetText(MakeMessage("Stats Info"))
        StatsInfoButton.OnClick().Subscribe(OnClickStatsInfo)

        stack_box:
            Orientation := orientation.Vertical
            Slots := array:
                stack_box_slot:
                    HorizontalAlignment := horizontal_alignment.Right
                    VerticalAlignment := vertical_alignment.Fill
                    Padding := margin{Left := 0.0, Top := 0.0, Right := 0.0, Bottom := 20.0}
                    Widget := StatsInfoButton
                stack_box_slot:
                    HorizontalAlignment := horizontal_alignment.Fill
                    VerticalAlignment := vertical_alignment.Fill
                    Padding := margin{Left := 0.0, Top := 0.0, Right := 0.0, Bottom := 20.0}
                    Widget := StackBox
                stack_box_slot:
                    HorizontalAlignment := horizontal_alignment.Fill
                    VerticalAlignment := vertical_alignment.Fill
                    Widget := CloseButton

    CreateInfoUI():widget =
        SelectAugmentText := text_block{DefaultText := MakeMessage("Select Augment")}
        SelectAugmentText.SetColor(ThemeColors.Loud)

        overlay:
            Slots := array:
                overlay_slot:
                    HorizontalAlignment := horizontal_alignment.Fill
                    VerticalAlignment := vertical_alignment.Fill
                    Widget := MakeBox()
                overlay_slot:
                    HorizontalAlignment := horizontal_alignment.Center
                    VerticalAlignment := vertical_alignment.Center
                    Padding := margin{Left := 40.0, Top := 40.0, Right := 40.0, Bottom := 40.0}
                    Widget := stack_box:
                        Orientation := orientation.Vertical
                        Slots := array:
                            stack_box_slot:
                                VerticalAlignment := vertical_alignment.Top
                                Padding := margin{Left := 0.0, Top := 0.0, Right := 0.0, Bottom := 20.0}
                                Widget := WeaponIcon
                            stack_box_slot:
                                VerticalAlignment := vertical_alignment.Fill
                                Padding := margin{Left := 0.0, Top := 0.0, Right := 0.0, Bottom := 40.0}
                                Widget := WeaponNameText
                            stack_box_slot:
                                VerticalAlignment := vertical_alignment.Fill
                                Padding := margin{Left := 0.0, Top := 0.0, Right := 0.0, Bottom := 20.0}
                                Widget := SelectAugmentText
                            stack_box_slot:
                                HorizontalAlignment := horizontal_alignment.Fill
                                VerticalAlignment := vertical_alignment.Fill
                                Distribution := option{1.0}
                                Padding := margin{Left := 0.0, Top := 0.0, Right := 0.0, Bottom := 0.0}
                                Widget := stack_box:
                                    Orientation := orientation.Vertical
                                    Slots := for (Index := 1..3):
                                        stack_box_slot:
                                            HorizontalAlignment := horizontal_alignment.Fill
                                            VerticalAlignment := vertical_alignment.Fill
                                            Padding := margin{Left := 0.0, Top := (Index = 1) and 0.0 or 20.0, Right := 0.0, Bottom := 0.0}
                                            Widget := CreateAugmentSlotUI(Index)

    CreateAugmentSlotUI(Index:int):widget =
        RightArrowIcon := texture_block:
            DefaultImage := UserInterface.Icons.Arrow_Left_128
            DefaultDesiredSize := vector2{X := 64.0, Y := 64.0}
            DefaultTint := ThemeColors.Success

        AugmentIcon := texture_block:
            DefaultImage := UserInterface.Icons.ProjectileCount_64
            DefaultDesiredSize := vector2{X := 64.0, Y := 64.0}

        LeftArrowIcon := texture_block:
            DefaultImage := UserInterface.Icons.Arrow_Right_128
            DefaultDesiredSize := vector2{X := 64.0, Y := 64.0}
            DefaultTint := ThemeColors.Success

        PreviewButton := button_quiet:
            DefaultText := MakeMessage("Evolution {Index.ToRoman()}")

        spawn{AwaitPreviewButtonClicked(PreviewButton.OnClick(), Index)}

        set AugmentSlotUIs += array{(RightArrowIcon, AugmentIcon, LeftArrowIcon, PreviewButton)}

        stack_box:
            Orientation := orientation.Horizontal
            Slots := array:
                stack_box_slot:
                    HorizontalAlignment := horizontal_alignment.Right
                    Widget := RightArrowIcon
                stack_box_slot:
                    HorizontalAlignment := horizontal_alignment.Fill
                    VerticalAlignment := vertical_alignment.Fill
                    Padding := margin{Left := 20.0, Top := 0.0, Right := 20.0, Bottom := 0.0}
                    Distribution := option{1.0}
                    Widget := overlay:
                        Slots := array:
                            overlay_slot:
                                HorizontalAlignment := horizontal_alignment.Fill
                                VerticalAlignment := vertical_alignment.Fill
                                Widget := MakeBox()
                            overlay_slot:
                                Padding := margin{Left := 20.0, Top := 20.0, Right := 20.0, Bottom := 20.0}
                                Widget := AugmentIcon
                            overlay_slot:
                                HorizontalAlignment := horizontal_alignment.Fill
                                VerticalAlignment := vertical_alignment.Fill
                                Widget := PreviewButton
                stack_box_slot:
                    HorizontalAlignment := horizontal_alignment.Left
                    Widget := LeftArrowIcon

    <# ----------------------------------------------------------------------------------------------------
        Get Functions
    ---------------------------------------------------------------------------------------------------- #>


    <# ----------------------------------------------------------------------------------------------------
        Update Functions
    ---------------------------------------------------------------------------------------------------- #>
    Update(Evolution:int):void =
        if (WeaponInfo := GetWeaponInfo[Id]):
            WeaponIcon.SetImage(WeaponInfo.Icon)
            WeaponNameText.SetText(MakeMessage("{WeaponInfo.Name}"))

            for (Index -> WeaponEvolutionEntryUI : WeaponEvolutionEntryUIs, AbilityInfo := WeaponInfo.Evolution[Evolution - 1][Index]):
                WeaponEvolutionEntryUI.Update(AbilityInfo, ?IsPreview := IsPreview)

        if (IsPreview?):
            for (Index -> AugmentSlotUI : AugmentSlotUIs):              
                AugmentSlotUI(0).SetVisibility((Index = Evolution - 1) and widget_visibility.Visible or widget_visibility.Hidden)
                AugmentSlotUI(1).SetVisibility(widget_visibility.Hidden)
                AugmentSlotUI(2).SetVisibility((Index = Evolution - 1) and widget_visibility.Visible or widget_visibility.Hidden)
                AugmentSlotUI(3).SetVisibility(widget_visibility.Visible)
                    
        else:
            WeaponEvolutions := Player.GetWeaponEvolution(Id)
            for (Index -> AugmentSlotUI : AugmentSlotUIs):
                AugmentSlotUI(1).SetVisibility(widget_visibility.Visible)
                AugmentSlotUI(3).SetVisibility(widget_visibility.Collapsed)

                if (Index < Evolution - 1, AbilityInfo := WeaponEvolutions[Index]):
                    AugmentSlotUI(0).SetVisibility(widget_visibility.Hidden)
                    AugmentSlotUI(1).SetImage(AbilityInfo.Icon)
                    AugmentSlotUI(1).SetTint(AbilityInfo.IconColor)
                    AugmentSlotUI(2).SetVisibility(widget_visibility.Hidden)
                else if (Index = Evolution - 1):
                    AugmentSlotUI(0).SetVisibility(widget_visibility.Visible)
                    AugmentSlotUI(1).SetImage(UserInterface.Icons.Unlock_64)
                    AugmentSlotUI(2).SetVisibility(widget_visibility.Visible)
                    spawn{AnimateUI(AugmentSlotUI)}
                else:
                    AugmentSlotUI(0).SetVisibility(widget_visibility.Hidden)
                    AugmentSlotUI(1).SetImage(UserInterface.Icons.Lock_64)
                    AugmentSlotUI(1).SetTint(ThemeColors.Inactive)
                    AugmentSlotUI(2).SetVisibility(widget_visibility.Hidden)

    ToggleInfo():void =
        if (InfoUI := MaybeInfoUI?):
            Visibility := InfoUI.GetUI().GetVisibility()
            if (Visibility = widget_visibility.Visible):
                InfoUI.Hide()
            else:
                InfoUI.Show()
                InfoUI.Update()
        else:
            NewInfoUI := ability_stat_info_ui{Player := Player, BackgroundOpacity := 1.0}
            NewInfoUI.Initialize()
            set MaybeInfoUI = option{NewInfoUI}
            
            if (PanelUI := canvas[MaybeUI?]):
                PanelUI.AddWidget of canvas_slot:
                    Anchors := anchors{Minimum := vector2{X := 0.0, Y := 0.0}, Maximum := vector2{X := 0.0, Y := 0.0}}
                    Alignment := vector2{X := 0.0, Y := 0.0}
                    ZOrder := 1
                    Widget := NewInfoUI.GetUI()
                    
                NewInfoUI.Update()

    SelectRandomAbility()<suspends>:void =
        {}

    AnimateUI(AugmentSlotUI:tuple(texture_block, texture_block, texture_block, text_button_base))<suspends>:void =
        race:
            CancelAnimationEvent.Await()
            loop:
                Sleep(0.0)
                Tint := ThemeColors.Success * (Sin(GetSecondsSinceEpoch() * 2) * 0.7 + 0.8)
                AugmentSlotUI(0).SetTint(Tint)
                AugmentSlotUI(1).SetTint(Tint)
                AugmentSlotUI(2).SetTint(Tint)

    <# ----------------------------------------------------------------------------------------------------
        Event Handlers
    ---------------------------------------------------------------------------------------------------- #>
    OnShow(InId:weapon_id, ?IsItemShop:logic = false, ?InIsPreview:logic = false):void =
        set Id = InId
        set ItemShop = IsItemShop
        set IsPreview = InIsPreview

        TitleText.SetText(MakeMessage("{Title}"))
        TitleText.SetTextColor(ThemeColors.PrimaryText)
        TitleText.SetShadowOpacity(0.0)
        Update(IsPreview? and 1 or Player.GetWeaponEvolutionLevel(Id))

        StatsInfoButton.SetVisibility(IsPreview? and widget_visibility.Collapsed or widget_visibility.Visible)
        CloseButton.SetVisibility(IsPreview? and widget_visibility.Visible or widget_visibility.Collapsed)
        
        if (InfoUI := MaybeInfoUI?):
            InfoUI.Hide()

    OnClosed(InPlayer:player):void =
        CancelAnimationEvent.Signal()
        if (not ItemShop?):
            ResumeGame(Player)

    OnSelect(AbilityInfo:ability_info, Tier:int):void =
        Player.AddWeaponEvolution(AbilityInfo)
        Hide()
        if (ItemShop?):
            Player.ShowItemShopUI(?IsResume := true)

    OnResumeCountdownStart():void =
        if (not IsShowing?) then return
        Player.PlayAudio(audio_id.TimerWarning)
        TitleText.SetText(MakeMessage("{GameResumeCountdownTimeAbilitySelect}s"))
        TitleText.SetTextColor(ThemeColors.Failure)
        TitleText.SetShadowColor(ThemeColors.Failure * 0.125)
        TitleText.SetShadowOffset(option{vector2{X := 2.0, Y := 2.0}})
        TitleText.SetShadowOpacity(1.0)
        
    OnResumeCountdownComplete():void =
        if (not IsShowing?) then return
        TitleText.SetText(MakeMessage("Auto Selected"))
        spawn{SelectRandomAbility()}

    OnResumeCountdown(TimeRemaining:int):void =
        if (not IsShowing?) then return
        TitleText.SetText(MakeMessage("{TimeRemaining}s"))

    OnGameResumed():void =
        if (not IsShowing?) then return
        TitleText.SetText(MakeMessage("{Title}"))
        TitleText.SetTextColor(ThemeColors.PrimaryText)
        TitleText.SetShadowOpacity(0.0)

    OnClickStatsInfo(Message:widget_message):void =
        ToggleInfo()

    OnClickClose<override>(Message:widget_message):void =
        Hide()
        Player.ShowPanelUI(panel_ui_id.WeaponSelect)

    AwaitPreviewButtonClicked(Listenable:listenable(widget_message), Index:int)<suspends>:void =
        loop:
            Listenable.Await()
            CancelAnimationEvent.Signal()
            Update(Index)