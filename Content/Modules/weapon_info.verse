using { /Verse.org/Assets }
using { /Verse.org/Random }
using { /Verse.org/Simulation }
using { WeaponAbilities }

<# ----------------------------------------------------------------------------------------------------
    Constants
---------------------------------------------------------------------------------------------------- #>
DefaultWeapon:weapon_id = weapon_id.Bow
WeaponUnlockLevel1:int = 5
WeaponUnlockLevel2:int = 15
WeaponUnlockLevel3:int = 25
#WeaponUnlockLevel1:int = 500
#WeaponUnlockLevel2:int = 1500
#WeaponUnlockLevel3:int = 2500

<# ----------------------------------------------------------------------------------------------------
    Data Definitions
---------------------------------------------------------------------------------------------------- #>
weapon_id<public> := enum:
    Sword
    Bow
    Shuriken
    Sentry
    Aura
    Torrent
    Lightning
    Trail
    Shield
    ChainLightning
    EnergyBomb
    Boomerang

ToString(Id:weapon_id)<transacts>:string =
    case (Id):
        weapon_id.Sword => "Sword"
        weapon_id.Bow => "Bow"
        weapon_id.Shuriken => "Shuriken"
        weapon_id.Sentry => "Sentry"
        weapon_id.Aura => "Aura"
        weapon_id.Torrent => "Torrent"
        weapon_id.Lightning => "Lightning"
        weapon_id.Trail => "Trail"
        weapon_id.Shield => "Shield"
        weapon_id.ChainLightning => "Chain Lightning"
        weapon_id.EnergyBomb => "Energy Bomb"
        weapon_id.Boomerang => "Boomerang"

(Id:weapon_id).ToInt()<transacts>:int =
    case (Id):
        weapon_id.Sword => 1
        weapon_id.Bow => 2
        weapon_id.Shuriken => 3
        weapon_id.Sentry => 4
        weapon_id.Aura => 5
        weapon_id.Torrent => 6
        weapon_id.Lightning => 7
        weapon_id.Trail => 8
        weapon_id.Shield => 9
        weapon_id.ChainLightning => 10
        weapon_id.EnergyBomb => 11
        weapon_id.Boomerang => 12

(Id:int).ToWeaponId()<transacts>:weapon_id =
    case (Id):
        1 => weapon_id.Sword
        2 => weapon_id.Bow
        3 => weapon_id.Shuriken
        4 => weapon_id.Sentry
        5 => weapon_id.Aura
        6 => weapon_id.Torrent
        7 => weapon_id.Lightning
        8 => weapon_id.Trail
        9 => weapon_id.Shield
        10 => weapon_id.ChainLightning
        11 => weapon_id.EnergyBomb
        12 => weapon_id.Boomerang
         _ => weapon_id.Sword

weapon_info := struct<computes>:
    Id:weapon_id
    Name:string
    Description:string
    Icon:texture
    IconSmall:texture
    Damage:float = 10.0
    Cooldown:float = 3.0
    Size:float = 100.0
    Evolution:[][]ability_info = array{}

StartingWeapons:[]weapon_id = array:
    weapon_id.Sword
    weapon_id.Bow
    weapon_id.Shuriken
    weapon_id.Sentry
    weapon_id.Aura
    weapon_id.Torrent
    weapon_id.Lightning
    weapon_id.Trail
    weapon_id.Shield

<# ----------------------------------------------------------------------------------------------------
    Weapon Info
---------------------------------------------------------------------------------------------------- #>
WeaponInfos:[weapon_id]weapon_info = map:
    weapon_id.Sword => weapon_info:
        Id := weapon_id.Sword
        Name := "Sword"
        Description := "A blade that deals high damage in a small area in the direct the player is facing."
        Icon := UserInterface.Icons.WeaponSword_128
        IconSmall := UserInterface.Icons.WeaponSword_64
        Damage := 30.0
        Cooldown := 2.0
        Size := 200.0
        Evolution := array:
            array:
                ability_info:
                    Id := ability_id.ProjectileCount
                    MaybeWeaponId := option{weapon_id.Sword}
                    Name := "Projectile Count"
                    Description := "Increase the projectile count of the weapon."
                    Icon := UserInterface.Icons.ProjectileCount_128
                    IconSmall := UserInterface.Icons.ProjectileCount_64
                    IconColor := ThemeColors.WeaponEvolutionA
                    TierValue := array{option{1.0}}
                    FormatValue := FormatAsInt
                ability_info:
                    Id := ability_id.SwordAngle
                    MaybeWeaponId := option{weapon_id.Sword}
                    Name := "Slash Angle"
                    Description := "Increases the attack angle of the sword."
                    Icon := UserInterface.Icons.SwordAngle_128
                    IconSmall := UserInterface.Icons.SwordAngle_64
                    IconColor := ThemeColors.WeaponEvolutionB
                    BaseValue := BaseAbilityConeAngle
                    TierValue := array{option{120.0}}
                    FormatValue := FormatAsDegree
            array:
                ability_info:
                    Id := ability_id.ProjectileCount
                    MaybeWeaponId := option{weapon_id.Sword}
                    Name := "Projectile Count"
                    Description := "Increase the projectile count of the weapon."
                    Icon := UserInterface.Icons.ProjectileCount_128
                    IconSmall := UserInterface.Icons.ProjectileCount_64
                    IconColor := ThemeColors.WeaponEvolutionA
                    TierValue := array{option{1.0}}
                    FormatValue := FormatAsInt
                ability_info:
                    Id := ability_id.SwordAngle
                    MaybeWeaponId := option{weapon_id.Sword}
                    Name := "Slash Angle"
                    Description := "Increase the attack angle of the sword."
                    Icon := UserInterface.Icons.SwordAngle_128
                    IconSmall := UserInterface.Icons.SwordAngle_64
                    IconColor := ThemeColors.WeaponEvolutionB
                    BaseValue := BaseAbilityConeAngle
                    TierValue := array{option{120.0}}
                    FormatValue := FormatAsDegree
            array:
                ability_info:
                    Id := ability_id.ProjectileCount
                    MaybeWeaponId := option{weapon_id.Sword}
                    Name := "Projectile Count"
                    Description := "Increase the projectile count of the weapon."
                    Icon := UserInterface.Icons.ProjectileCount_128
                    IconSmall := UserInterface.Icons.ProjectileCount_64
                    IconColor := ThemeColors.WeaponEvolutionA
                    TierValue := array{option{1.0}}
                    FormatValue := FormatAsInt
                ability_info:
                    Id := ability_id.WeaponEvolution
                    MaybeWeaponId := option{weapon_id.Sword}
                    Name := "Precision Strike"
                    Description := "Halves the size of size of the sword but doubles the base damage."
                    Icon := UserInterface.Icons.SwordPrecisionStrike_128
                    IconSmall := UserInterface.Icons.SwordPrecisionStrike_64
                    IconColor := ThemeColors.WeaponEvolutionUnique
                    TierValue := array{option{1.0}}
                    FormatValue := FormatAsUnique
    weapon_id.Bow => weapon_info:
        Id := weapon_id.Bow
        Name := "Bow"
        Description := "Shoots (3) arrows in the direction the player is facing."
        Icon := UserInterface.Icons.WeaponBow_128
        IconSmall := UserInterface.Icons.WeaponBow_64
        Damage := 20.0
        Cooldown := 2.0
        Size := 25.0
        Evolution := array:
            array:
                ability_info:
                    Id := ability_id.ProjectileCount
                    MaybeWeaponId := option{weapon_id.Bow}
                    Name := "Projectile Count"
                    Description := "Increase the projectile count of the weapon."
                    Icon := UserInterface.Icons.ProjectileCount_128
                    IconSmall := UserInterface.Icons.ProjectileCount_64
                    IconColor := ThemeColors.WeaponEvolutionA
                    TierValue := array{option{2.0}}
                    FormatValue := FormatAsInt
                ability_info:
                    Id := ability_id.Multicast
                    MaybeWeaponId := option{weapon_id.Bow}
                    Name := "Multicast"
                    Description := "Increase the amount of times the weapon is casted."
                    Icon := UserInterface.Icons.Multicast_128
                    IconSmall := UserInterface.Icons.Multicast_64
                    IconColor := ThemeColors.WeaponEvolutionB
                    TierValue := array{option{1.0}}
                    FormatValue := FormatAsInt
            array:
                ability_info:
                    Id := ability_id.ProjectileCount
                    MaybeWeaponId := option{weapon_id.Bow}
                    Name := "Projectile Count II"
                    Description := "Increase the projectile count of the weapon."
                    Icon := UserInterface.Icons.ProjectileCount_128
                    IconSmall := UserInterface.Icons.ProjectileCount_64
                    IconColor := ThemeColors.WeaponEvolutionA
                    TierValue := array{option{4.0}}
                    FormatValue := FormatAsInt
                ability_info:
                    Id := ability_id.Multicast
                    MaybeWeaponId := option{weapon_id.Bow}
                    Name := "Multicast II"
                    Description := "Increase the amount of times the weapon is casted."
                    Icon := UserInterface.Icons.Multicast_128
                    IconSmall := UserInterface.Icons.Multicast_64
                    IconColor := ThemeColors.WeaponEvolutionB
                    TierValue := array{option{2.0}}
                    FormatValue := FormatAsInt
            array:
                ability_info:
                    Id := ability_id.ProjectilePiercing
                    MaybeWeaponId := option{weapon_id.Bow}
                    Name := "Projectile Piercing"
                    Description := "Projectile can piece through additional enemies."
                    Icon := UserInterface.Icons.ProjectilePiercing_128
                    IconSmall := UserInterface.Icons.ProjectilePiercing_64
                    IconColor := ThemeColors.WeaponEvolutionA
                    TierValue := array{option{2.0}}
                    FormatValue := FormatAsInt
                ability_info:
                    Id := ability_id.WeaponEvolution
                    MaybeWeaponId := option{weapon_id.Bow}
                    Name := "Supercharged Arrow"
                    Description := "Combines all projectiles into a single powerful arrow. Has 100% piercing and +20% base damage & +50% size per projectile."
                    Icon := UserInterface.Icons.Multicast_128
                    IconSmall := UserInterface.Icons.Multicast_64
                    IconColor := ThemeColors.WeaponEvolutionUnique
                    TierValue := array{option{1.0}}
                    FormatValue := FormatAsUnique
    weapon_id.Shuriken => weapon_info:
        Id := weapon_id.Shuriken
        Name := "Shuriken"
        Description := "Blades that rotate around the player damaging all enemies in its path."
        Icon := UserInterface.Icons.WeaponShuriken_128
        IconSmall := UserInterface.Icons.WeaponShuriken_64
        Damage := 10.0
        Cooldown := 10.0
        Size := 75.0
        Evolution := array:
            array:
                ability_info:
                    Id := ability_id.ProjectileCount
                    MaybeWeaponId := option{weapon_id.Shuriken}
                    Name := "Projectile Count I"
                    Description := "Increase the projectile count of the weapon."
                    Icon := UserInterface.Icons.ProjectileCount_128
                    IconSmall := UserInterface.Icons.ProjectileCount_64
                    IconColor := ThemeColors.WeaponEvolutionA
                    TierValue := array{option{1.0}}
                    FormatValue := FormatAsInt
                ability_info:
                    Id := ability_id.ProjectileSpeed
                    MaybeWeaponId := option{weapon_id.Shuriken}
                    Name := "Weighted Shuriken I"
                    Description := "Reduce projectile speed (33.3%) but increases base damage."
                    Icon := UserInterface.Icons.ShurikenWeight_128
                    IconSmall := UserInterface.Icons.ShurikenWeight_64
                    IconColor := ThemeColors.WeaponEvolutionB
                    TierValue := array{option{0.5}}
                    FormatValue := FormatAsPercent
            array:
                ability_info:
                    Id := ability_id.ProjectileCount
                    MaybeWeaponId := option{weapon_id.Shuriken}
                    Name := "Projectile Count II"
                    Description := "Increase the projectile count of the weapon."
                    Icon := UserInterface.Icons.ProjectileCount_128
                    IconSmall := UserInterface.Icons.ProjectileCount_64
                    IconColor := ThemeColors.WeaponEvolutionA
                    TierValue := array{option{2.0}}
                    FormatValue := FormatAsInt
                ability_info:
                    Id := ability_id.ProjectileSpeed
                    MaybeWeaponId := option{weapon_id.Shuriken}
                    Name := "Weighted Shuriken II"
                    Description := "Reduce projectile speed (33.3%) but increases base damage."
                    Icon := UserInterface.Icons.ShurikenWeight_128
                    IconSmall := UserInterface.Icons.ShurikenWeight_64
                    IconColor := ThemeColors.WeaponEvolutionB
                    TierValue := array{option{1.0}}
                    FormatValue := FormatAsPercent
            array:
                ability_info:
                    Id := ability_id.ProjectileCount
                    MaybeWeaponId := option{weapon_id.Shuriken}
                    Name := "Projectile Count III"
                    Description := "Increase the projectile count of the weapon."
                    Icon := UserInterface.Icons.ProjectileCount_128
                    IconSmall := UserInterface.Icons.ProjectileCount_64
                    IconColor := ThemeColors.WeaponEvolutionA
                    TierValue := array{option{3.0}}
                    FormatValue := FormatAsInt
                ability_info:
                    Id := ability_id.WeaponEvolution
                    MaybeWeaponId := option{weapon_id.Shuriken}
                    Name := "Shuriken Mastery"
                    Description := "Slow and powerful shurikens that deals damage every 0.5s (+100% base damage)."
                    Icon := UserInterface.Icons.ShurikenMastery_128
                    IconSmall := UserInterface.Icons.ShurikenMastery_64
                    IconColor := ThemeColors.WeaponEvolutionUnique
                    TierValue := array{option{1.0}}
                    FormatValue := FormatAsUnique
    weapon_id.Sentry => weapon_info:
        Id := weapon_id.Sentry
        Name := "Sentry"
        Description := "Periodically constructs a sentry turret that automatically shoots enemies in its range."
        Icon := UserInterface.Icons.WeaponSentry_128
        IconSmall := UserInterface.Icons.WeaponSentry_64
        Damage := 10.0
        Cooldown := 5.0
        Size := 80.0
        Evolution := array:
            array:
                ability_info:
                    Id := ability_id.ProjectileCount
                    MaybeWeaponId := option{weapon_id.Sentry}
                    Name := "Projectile Count I"
                    Description := "Increase the projectile count of the weapon."
                    Icon := UserInterface.Icons.ProjectileCount_128
                    IconSmall := UserInterface.Icons.ProjectileCount_64
                    IconColor := ThemeColors.WeaponEvolutionA
                    TierValue := array{option{1.0}}
                    FormatValue := FormatAsInt
                ability_info:
                    Id := ability_id.FireRate
                    MaybeWeaponId := option{weapon_id.Sentry}
                    Name := "Fire Rate II"
                    Description := "Increase the fire rate of the weapon."
                    Icon := UserInterface.Icons.FireRate_128
                    IconSmall := UserInterface.Icons.FireRate_64
                    IconColor := ThemeColors.WeaponEvolutionB
                    TierValue := array{option{0.5}}
                    FormatValue := FormatAsPercent
            array:
                ability_info:
                    Id := ability_id.ProjectileCount
                    MaybeWeaponId := option{weapon_id.Sentry}
                    Name := "Projectile Count II"
                    Description := "Increase the projectile count of the weapon."
                    Icon := UserInterface.Icons.ProjectileCount_128
                    IconSmall := UserInterface.Icons.ProjectileCount_64
                    IconColor := ThemeColors.WeaponEvolutionA
                    TierValue := array{option{2.0}}
                    FormatValue := FormatAsInt
                ability_info:
                    Id := ability_id.FireRate
                    MaybeWeaponId := option{weapon_id.Sentry}
                    Name := "Fire Rate II"
                    Description := "Increase the fire rate of the weapon."
                    Icon := UserInterface.Icons.FireRate_128
                    IconSmall := UserInterface.Icons.FireRate_64
                    IconColor := ThemeColors.WeaponEvolutionB
                    TierValue := array{option{1.0}}
                    FormatValue := FormatAsPercent
            array:
                ability_info:
                    Id := ability_id.ProjectileCount
                    MaybeWeaponId := option{weapon_id.Sentry}
                    Name := "Projectile Count III"
                    Description := "Increase the projectile count of the weapon."
                    Icon := UserInterface.Icons.ProjectileCount_128
                    IconSmall := UserInterface.Icons.ProjectileCount_64
                    IconColor := ThemeColors.WeaponEvolutionA
                    TierValue := array{option{3.0}}
                    FormatValue := FormatAsInt
                ability_info:
                    Id := ability_id.WeaponEvolution
                    MaybeWeaponId := option{weapon_id.Sentry}
                    Name := "Heavy-Duty Sentry"
                    Description := "Powerful sentry with doubled damage, range and duration."
                    Icon := UserInterface.Icons.MovingSentry_128
                    IconSmall := UserInterface.Icons.MovingSentry_64
                    IconColor := ThemeColors.WeaponEvolutionUnique
                    TierValue := array{option{1.0}}
                    FormatValue := FormatAsUnique
    weapon_id.Aura => weapon_info:
        Id := weapon_id.Aura
        Name := "Aura"
        Description := "High frequency damage in small area around the player."
        Icon := UserInterface.Icons.WeaponAura_128
        IconSmall := UserInterface.Icons.WeaponAura_64
        Damage := 8.0
        Cooldown := 0.5
        Size := 175.0 # Requires updating VFX size.
        Evolution := array:
            array:
                ability_info:
                    Id := ability_id.AuraDuration
                    MaybeWeaponId := option{weapon_id.Aura}
                    Name := "Supercharged Duration I"
                    Description := "Increase the duration the aura is in a supercharged state."
                    Icon := UserInterface.Icons.EvolvedAuraDuration_128
                    IconSmall := UserInterface.Icons.EvolvedAuraDuration_64
                    IconColor := ThemeColors.WeaponEvolutionA
                    BaseValue := 1.0
                    TierValue := array{option{1.0}}
                    FormatValue := FormatAsTime
                ability_info:
                    Id := ability_id.AuraPower
                    MaybeWeaponId := option{weapon_id.Aura}
                    Name := "Supercharged Power I"
                    Description := "Increase damage multiplier while in a supercharged state."
                    Icon := UserInterface.Icons.EvolvedAuraPower_128
                    IconSmall := UserInterface.Icons.EvolvedAuraPower_64
                    IconColor := ThemeColors.WeaponEvolutionB
                    TierValue := array{option{1.0}}
                    FormatValue := FormatAsPercent
            array:
                ability_info:
                    Id := ability_id.AuraDuration
                    MaybeWeaponId := option{weapon_id.Aura}
                    Name := "Supercharged Duration I"
                    Description := "Increase the duration the aura is in a supercharged state."
                    Icon := UserInterface.Icons.EvolvedAuraDuration_128
                    IconSmall := UserInterface.Icons.EvolvedAuraDuration_64
                    IconColor := ThemeColors.WeaponEvolutionA
                    BaseValue := 1.0
                    TierValue := array{option{1.0}}
                    FormatValue := FormatAsTime
                ability_info:
                        Id := ability_id.WeaponEvolution
                        MaybeWeaponId := option{weapon_id.Aura}
                        Name := "Knockback"
                        Description := "Aura can apply knockback to enemies while in a supercharged state."
                        Icon := UserInterface.Icons.EvolvedAuraKnockback_128
                        IconSmall := UserInterface.Icons.EvolvedAuraKnockback_64
                        IconColor := ThemeColors.WeaponEvolutionUnique
                        TierValue := array{option{1.0}}
                        FormatValue := FormatAsUnique
            array:
                ability_info:
                    Id := ability_id.AuraDuration
                    MaybeWeaponId := option{weapon_id.Aura}
                    Name := "Supercharged Duration II"
                    Description := "Increase the duration the aura is in a supercharged state."
                    Icon := UserInterface.Icons.EvolvedAuraDuration_128
                    IconSmall := UserInterface.Icons.EvolvedAuraDuration_64
                    IconColor := ThemeColors.WeaponEvolutionA
                    BaseValue := 1.0
                    TierValue := array{option{2.0}}
                    FormatValue := FormatAsTime
                ability_info:
                    Id := ability_id.AuraPower
                    MaybeWeaponId := option{weapon_id.Aura}
                    Name := "Supercharged Power II"
                    Description := "Increase damage multiplier while in a supercharged state."
                    Icon := UserInterface.Icons.EvolvedAuraPower_128
                    IconSmall := UserInterface.Icons.EvolvedAuraPower_64
                    IconColor := ThemeColors.WeaponEvolutionB
                    TierValue := array{option{2.0}}
                    FormatValue := FormatAsPercent
    weapon_id.Torrent => weapon_info:
        Id := weapon_id.Torrent
        Name := "Torrent"
        Description := "Randomly spawns a pool that deals damage over time."
        Icon := UserInterface.Icons.WeaponTorrent_128
        IconSmall := UserInterface.Icons.WeaponTorrent_64
        Damage := 10.0
        Cooldown := 5.0
        Size := 150.0
        Evolution := array:
            array:
                ability_info:
                    Id := ability_id.ProjectileCount
                    MaybeWeaponId := option{weapon_id.Torrent}
                    Name := "Projectile Count"
                    Description := "Increase the projectile count of the weapon."
                    Icon := UserInterface.Icons.ProjectileCount_128
                    IconSmall := UserInterface.Icons.ProjectileCount_64
                    IconColor := ThemeColors.WeaponEvolutionA
                    TierValue := array{option{1.0}}
                    FormatValue := FormatAsInt
                ability_info:
                    Id := ability_id.TorrentSplit
                    MaybeWeaponId := option{weapon_id.Torrent}
                    Name := "Torrent Split"
                    Description := "Torrent splits into 2 smaller torrents when it expires."
                    Icon := UserInterface.Icons.TorrentSplit_128
                    IconSmall := UserInterface.Icons.TorrentSplit_64
                    IconColor := ThemeColors.WeaponEvolutionB
                    TierValue := array{option{2.0}}
                    FormatValue := FormatAsInt
            array:
                ability_info:
                    Id := ability_id.ProjectileCount
                    MaybeWeaponId := option{weapon_id.Torrent}
                    Name := "Projectile Count"
                    Description := "Increase the projectile count of the weapon."
                    Icon := UserInterface.Icons.ProjectileCount_128
                    IconSmall := UserInterface.Icons.ProjectileCount_64
                    IconColor := ThemeColors.WeaponEvolutionA
                    TierValue := array{option{1.0}}
                    FormatValue := FormatAsInt
                ability_info:
                    Id := ability_id.WeaponEvolution
                    MaybeWeaponId := option{weapon_id.Torrent}
                    Name := "Travelling Torrent"
                    Description := "Torrents moves towards the player's position."
                    Icon := UserInterface.Icons.TravellingTorrent_128
                    IconSmall := UserInterface.Icons.TravellingTorrent_64
                    IconColor := ThemeColors.WeaponEvolutionUnique
                    TierValue := array{option{1.0}}
                    FormatValue := FormatAsUnique
            array:
                ability_info:
                    Id := ability_id.ProjectileCount
                    MaybeWeaponId := option{weapon_id.Torrent}
                    Name := "Projectile Count"
                    Description := "Increase the projectile count of the weapon."
                    Icon := UserInterface.Icons.ProjectileCount_128
                    IconSmall := UserInterface.Icons.ProjectileCount_64
                    IconColor := ThemeColors.WeaponEvolutionA
                    TierValue := array{option{1.0}}
                    FormatValue := FormatAsInt
                ability_info:
                    Id := ability_id.TorrentSplit
                    MaybeWeaponId := option{weapon_id.Torrent}
                    Name := "Torrent Split"
                    Description := "Torrent splits into 2 smaller torrents when it expires."
                    Icon := UserInterface.Icons.TorrentSplit_128
                    IconSmall := UserInterface.Icons.TorrentSplit_64
                    IconColor := ThemeColors.WeaponEvolutionB
                    TierValue := array{option{2.0}}
                    FormatValue := FormatAsInt
    weapon_id.Lightning => weapon_info:
        Id := weapon_id.Lightning
        Name := "Lightning"
        Description := "Randomly strikes an enemy dealing high damage in a small area."
        Icon := UserInterface.Icons.WeaponLightning_128
        IconSmall := UserInterface.Icons.WeaponLightning_64
        Damage := 25.0
        Cooldown := 1.0
        Size := 35.0
        Evolution := array:
            array:
                ability_info:
                    Id := ability_id.ProjectileCount
                    MaybeWeaponId := option{weapon_id.Lightning}
                    Name := "Projectile Count I"
                    Description := "Increase the projectile count of the weapon."
                    Icon := UserInterface.Icons.ProjectileCount_128
                    IconSmall := UserInterface.Icons.ProjectileCount_64
                    IconColor := ThemeColors.WeaponEvolutionA
                    TierValue := array{option{1.0}}
                    FormatValue := FormatAsInt
                ability_info:
                    Id := ability_id.WeaponEvolution
                    MaybeWeaponId := option{weapon_id.Lightning}
                    Name := "Focus Fire"
                    Description := "Targets the enemy with the highest health."
                    Icon := UserInterface.Icons.FocusFire_128
                    IconSmall := UserInterface.Icons.FocusFire_64
                    IconColor := ThemeColors.WeaponEvolutionUnique
                    TierValue := array{option{1.0}}
                    FormatValue := FormatAsUnique
            array:
                ability_info:
                    Id := ability_id.ProjectileCount
                    MaybeWeaponId := option{weapon_id.Lightning}
                    Name := "Projectile Count I"
                    Description := "Increase the projectile count of the weapon."
                    Icon := UserInterface.Icons.ProjectileCount_128
                    IconSmall := UserInterface.Icons.ProjectileCount_64
                    IconColor := ThemeColors.WeaponEvolutionA
                    TierValue := array{option{1.0}}
                    FormatValue := FormatAsInt
                ability_info:
                    Id := ability_id.Size
                    MaybeWeaponId := option{weapon_id.Lightning}
                    Name := "Size Multiplier I"
                    Description := "Increase the base size of the weapon."
                    Icon := UserInterface.Icons.AbilitySize_128
                    IconSmall := UserInterface.Icons.AbilitySize_64
                    IconColor := ThemeColors.WeaponEvolutionB
                    TierValue := array{option{1.0}}
                    FormatValue := FormatAsPercent
            array:
                ability_info:
                    Id := ability_id.ProjectileCount
                    MaybeWeaponId := option{weapon_id.Lightning}
                    Name := "Projectile Count II"
                    Description := "Increase the projectile count of the weapon."
                    Icon := UserInterface.Icons.ProjectileCount_128
                    IconSmall := UserInterface.Icons.ProjectileCount_64
                    IconColor := ThemeColors.WeaponEvolutionA
                    TierValue := array{option{2.0}}
                    FormatValue := FormatAsInt
                ability_info:
                    Id := ability_id.Size
                    MaybeWeaponId := option{weapon_id.Lightning}
                    Name := "Size Multiplier II"
                    Description := "Increase the base size of the weapon."
                    Icon := UserInterface.Icons.AbilitySize_128
                    IconSmall := UserInterface.Icons.AbilitySize_64
                    IconColor := ThemeColors.WeaponEvolutionB
                    TierValue := array{option{2.0}}
                    FormatValue := FormatAsPercent
            array:
    weapon_id.Trail => weapon_info:
        Id := weapon_id.Trail
        Name := "Trail"
        Description := "Spawns a trail as the player moves that deals damage over time."
        Icon := UserInterface.Icons.WeaponTrail_128
        IconSmall := UserInterface.Icons.WeaponTrail_64
        Damage := 7.5
        Cooldown := 2.5 # Affects duration.
        Size := 75.0
        Evolution := array:
            array:
                ability_info:
                    Id := ability_id.TrailChargedChance
                    MaybeWeaponId := option{weapon_id.Trail}
                    Name := "Supercharged Trail Chance I"
                    Description := "Increase chance of spawning a supercharged trail (+50% size)."
                    Icon := UserInterface.Icons.ChargedTrailChance_128
                    IconSmall := UserInterface.Icons.ChargedTrailChance_64
                    IconColor := ThemeColors.WeaponEvolutionA
                    BaseValue := 0.1
                    TierValue := array{option{0.025}}
                    FormatValue := FormatAsPercent
                ability_info:
                    Id := ability_id.TrailChargedPower
                    MaybeWeaponId := option{weapon_id.Trail}
                    Name := "Supercharged Trail Power"
                    Description := "Increase damage multiplier of supercharged trail (+50% size)."
                    Icon := UserInterface.Icons.ChargedTrailPower_128
                    IconSmall := UserInterface.Icons.ChargedTrailPower_64
                    IconColor := ThemeColors.WeaponEvolutionB
                    BaseValue := 0.5
                    TierValue := array{option{0.5}}
                    FormatValue := FormatAsPercent
            array:
                ability_info:
                    Id := ability_id.TrailChargedChance
                    MaybeWeaponId := option{weapon_id.Trail}
                    Name := "Supercharged Trail Chance II"
                    Description := "Increase chance of spawning a supercharged trail (+50% size & damage)."
                    Icon := UserInterface.Icons.ChargedTrailChance_128
                    IconSmall := UserInterface.Icons.ChargedTrailChance_64
                    IconColor := ThemeColors.WeaponEvolutionA
                    BaseValue := 0.1
                    TierValue := array{option{0.075}}
                    FormatValue := FormatAsPercent
                ability_info:
                    Id := ability_id.WeaponEvolution
                    MaybeWeaponId := option{weapon_id.Trail}
                    Name := "Supercharged Trail Bomb"
                    Description := "Supercharged trail explodes when it expires applying knockback (+50% size & damage)."
                    Icon := UserInterface.Icons.ChargedTrailPower_128
                    IconSmall := UserInterface.Icons.ChargedTrailPower_64
                    IconColor := ThemeColors.WeaponEvolutionUnique
                    TierValue := array{option{1.0}}
                    FormatValue := FormatAsUnique
            array:
                ability_info:
                    Id := ability_id.TrailChargedChance
                    MaybeWeaponId := option{weapon_id.Trail}
                    Name := "Supercharged Trail Chance III"
                    Description := "Increase chance of spawning a supercharged trail (+50% size & damage)."
                    Icon := UserInterface.Icons.ChargedTrailChance_128
                    IconSmall := UserInterface.Icons.ChargedTrailChance_64
                    IconColor := ThemeColors.WeaponEvolutionA
                    BaseValue := 0.1
                    TierValue := array{option{0.3}}
                    FormatValue := FormatAsPercent
                ability_info:
                    Id := ability_id.TrailChargedPower
                    MaybeWeaponId := option{weapon_id.Trail}
                    Name := "Supercharged Trail Power III"
                    Description := "Increase damage multiplier of supercharged trail."
                    Icon := UserInterface.Icons.ChargedTrailPower_128
                    IconSmall := UserInterface.Icons.ChargedTrailPower_64
                    IconColor := ThemeColors.WeaponEvolutionB
                    BaseValue := 0.5
                    TierValue := array{option{1.0}}
                    FormatValue := FormatAsPercent
    weapon_id.Shield => weapon_info:
        Id := weapon_id.Shield
        Name := "Shield"
        Description := "Periodically activates a shield that generates a blast when hit by an enemy."
        Icon := UserInterface.Icons.WeaponShield_128
        IconSmall := UserInterface.Icons.WeaponShield_64
        Damage := 15.0
        Cooldown := 5.0
        Size := 250.0
        Evolution := array:
            array:
                ability_info:
                    Id := ability_id.ShieldMaxHealth
                    MaybeWeaponId := option{weapon_id.Shield}
                    Name := "Max Shield"
                    Description := "Increase the max shield as a percentage of max health."
                    Icon := UserInterface.Icons.ShieldMaxHealth_128
                    IconSmall := UserInterface.Icons.ShieldMaxHealth_64
                    IconColor := ThemeColors.WeaponEvolutionA
                    BaseValue := 0.25
                    TierValue := array{option{0.25}}
                    FormatValue := FormatAsPercent
                ability_info:
                    Id := ability_id.ShieldRegenerationRate
                    MaybeWeaponId := option{weapon_id.Shield}
                    Name := "Shield Regeneration"
                    Description := "Increase the amount of shield regenerated per cast as a percentage of max health."
                    Icon := UserInterface.Icons.ShieldRegeneration_128
                    IconSmall := UserInterface.Icons.ShieldRegeneration_64
                    IconColor := ThemeColors.WeaponEvolutionB
                    BaseValue := 0.125
                    TierValue := array{option{0.125}}
                    FormatValue := FormatAsPercent
            array:
                ability_info:
                    Id := ability_id.WeaponEvolution
                    MaybeWeaponId := option{weapon_id.Shield}
                    Name := "Cracked Shield Blast"
                    Description := "Big blast when shield is cracked that applies knockback (+50% area & damage)."
                    Icon := UserInterface.Icons.ShieldCracked_128
                    IconSmall := UserInterface.Icons.ShieldCracked_64
                    IconColor := ThemeColors.WeaponEvolutionUnique
                    TierValue := array{option{1.0}}
                    FormatValue := FormatAsUnique
                ability_info:
                    Id := ability_id.ShieldRegenerationRate
                    MaybeWeaponId := option{weapon_id.Shield}
                    Name := "Shield Regeneration"
                    Description := "Increase the amount of shield regenerated per cast as a percentage of max health."
                    Icon := UserInterface.Icons.ShieldRegeneration_128
                    IconSmall := UserInterface.Icons.ShieldRegeneration_64
                    IconColor := ThemeColors.WeaponEvolutionB
                    BaseValue := 0.125
                    TierValue := array{option{0.125}}
                    FormatValue := FormatAsPercent
            array:
                ability_info:
                    Id := ability_id.ShieldMaxHealth
                    MaybeWeaponId := option{weapon_id.Shield}
                    Name := "Max Shield II"
                    Description := "Increase the max shield as a percentage of the player's max health."
                    Icon := UserInterface.Icons.ShieldMaxHealth_128
                    IconSmall := UserInterface.Icons.ShieldMaxHealth_64
                    IconColor := ThemeColors.WeaponEvolutionA
                    BaseValue := 0.25
                    TierValue := array{option{0.5}}
                    FormatValue := FormatAsPercent
                ability_info:
                    Id := ability_id.ShieldRegenerationRate
                    MaybeWeaponId := option{weapon_id.Shield}
                    Name := "Shield Regeneration II"
                    Description := "Increase the amount of shield regenerated per cast as a percentage of max health."
                    Icon := UserInterface.Icons.ShieldRegeneration_128
                    IconSmall := UserInterface.Icons.ShieldRegeneration_64
                    IconColor := ThemeColors.WeaponEvolutionB
                    BaseValue := 0.125
                    TierValue := array{option{0.25}}
                    FormatValue := FormatAsPercent
    weapon_id.ChainLightning => weapon_info:
        Id := weapon_id.ChainLightning
        Name := "Chain Lightning"
        Description := "A projectile that bounces towards the nearest enemies."
        Icon := UserInterface.Icons.WeaponChainLightning_128
        IconSmall := UserInterface.Icons.WeaponChainLightning_64
        Damage := 30.0
        Cooldown := 2.5
        Size := 40.0
        Evolution := array:
            array:
                ability_info:
                    Id := ability_id.ChainLightningBounce
                    MaybeWeaponId := option{weapon_id.ChainLightning}
                    Name := "Chain Bounce I"
                    Description := "Increases the number of times chain lightning bounces."
                    Icon := UserInterface.Icons.ChainLightningBounce_128
                    IconSmall := UserInterface.Icons.ChainLightningBounce_64
                    IconColor := ThemeColors.WeaponEvolutionA
                    BaseValue := 3.0
                    TierValue := array{option{3.0}}
                    FormatValue := FormatAsInt
                ability_info:
                    Id := ability_id.ChainLightningBurst
                    MaybeWeaponId := option{weapon_id.ChainLightning}
                    Name := "Chain Burst"
                    Description := "Damage enemies in an area where the chain lightning bounces (scales with size)."
                    Icon := UserInterface.Icons.ChainLightningBurst_128
                    IconSmall := UserInterface.Icons.ChainLightningBurst_64
                    IconColor := ThemeColors.WeaponEvolutionUnique
                    TierValue := array{option{1.0}}
                    FormatValue := FormatAsUnique
            array:
                ability_info:
                    Id := ability_id.ChainLightningBounce
                    MaybeWeaponId := option{weapon_id.ChainLightning}
                    Name := "Chain Bounce I"
                    Description := "Increases the number of times chain lightning bounces."
                    Icon := UserInterface.Icons.ChainLightningBounce_128
                    IconSmall := UserInterface.Icons.ChainLightningBounce_64
                    IconColor := ThemeColors.WeaponEvolutionA
                    BaseValue := 3.0
                    TierValue := array{option{3.0}}
                    FormatValue := FormatAsInt
                ability_info:
                    Id := ability_id.ChainLightningResilience
                    MaybeWeaponId := option{weapon_id.ChainLightning}
                    Name := "Lightning Resilience I"
                    Description := "Increases the percent damage retained after each bounce."
                    Icon := UserInterface.Icons.ChainLightningResilience_128
                    IconSmall := UserInterface.Icons.ChainLightningResilience_64
                    IconColor := ThemeColors.WeaponEvolutionB
                    BaseValue := 0.8
                    TierValue := array{option{0.1}}
                    FormatValue := FormatAsPercent
            array:
                ability_info:
                    Id := ability_id.ChainLightningBounce
                    MaybeWeaponId := option{weapon_id.ChainLightning}
                    Name := "Chain Bounce II"
                    Description := "Increases the number of times chain lightning bounces."
                    Icon := UserInterface.Icons.ChainLightningBounce_128
                    IconSmall := UserInterface.Icons.ChainLightningBounce_64
                    IconColor := ThemeColors.WeaponEvolutionA
                    BaseValue := 3.0
                    TierValue := array{option{6.0}}
                    FormatValue := FormatAsInt
                ability_info:
                    Id := ability_id.ChainLightningResilience
                    MaybeWeaponId := option{weapon_id.ChainLightning}
                    Name := "Lightning Resilience II"
                    Description := "Increases the percent damage retained after each bounce."
                    Icon := UserInterface.Icons.ChainLightningResilience_128
                    IconSmall := UserInterface.Icons.ChainLightningResilience_64
                    IconColor := ThemeColors.WeaponEvolutionB
                    BaseValue := 0.8
                    TierValue := array{option{0.2}}
                    FormatValue := FormatAsPercent
    weapon_id.EnergyBomb => weapon_info:
        Id := weapon_id.EnergyBomb
        Name := "Energy Bomb"
        Description := "A projectile that bounces towards the nearest enemies."
        Icon := UserInterface.Icons.WeaponEnergyBomb_128
        IconSmall := UserInterface.Icons.WeaponEnergyBomb_64
        Damage := 20.0
        Cooldown := 5.0
        Size := 150.0
        Evolution := array:
            array:
                ability_info:
                    Id := ability_id.ProjectileCount
                    MaybeWeaponId := option{weapon_id.EnergyBomb}
                    Name := "Projectile Count"
                    Description := "Increase the projectile count of the weapon."
                    Icon := UserInterface.Icons.ProjectileCount_128
                    IconSmall := UserInterface.Icons.ProjectileCount_64
                    IconColor := ThemeColors.WeaponEvolutionA
                    TierValue := array{option{1.0}}
                    FormatValue := FormatAsInt
                ability_info:
                    Id := ability_id.Multicast
                    MaybeWeaponId := option{weapon_id.EnergyBomb}
                    Name := "Multicast"
                    Description := "Increase the amount of times the weapon is casted."
                    Icon := UserInterface.Icons.Multicast_128
                    IconSmall := UserInterface.Icons.Multicast_64
                    IconColor := ThemeColors.WeaponEvolutionB
                    TierValue := array{option{1.0}}
                    FormatValue := FormatAsInt
            array:
                ability_info:
                    Id := ability_id.ProjectileCount
                    MaybeWeaponId := option{weapon_id.EnergyBomb}
                    Name := "Projectile Count"
                    Description := "Increase the projectile count of the weapon."
                    Icon := UserInterface.Icons.ProjectileCount_128
                    IconSmall := UserInterface.Icons.ProjectileCount_64
                    IconColor := ThemeColors.WeaponEvolutionA
                    TierValue := array{option{1.0}}
                    FormatValue := FormatAsInt
                ability_info:
                    Id := ability_id.WeaponEvolution
                    MaybeWeaponId := option{weapon_id.EnergyBomb}
                    Name := "Gravity Bomb"
                    Description := "Enemies get pulled towards the bomb before it explodes (only applies to 1st cast)."
                    Icon := UserInterface.Icons.EnergyBombGravity_128
                    IconSmall := UserInterface.Icons.EnergyBombGravity_64
                    IconColor := ThemeColors.WeaponEvolutionUnique
                    TierValue := array{option{1.0}}
                    FormatValue := FormatAsUnique
            array:
                ability_info:
                    Id := ability_id.ProjectileCount
                    MaybeWeaponId := option{weapon_id.EnergyBomb}
                    Name := "Projectile Count"
                    Description := "Increase the projectile count of the weapon."
                    Icon := UserInterface.Icons.ProjectileCount_128
                    IconSmall := UserInterface.Icons.ProjectileCount_64
                    IconColor := ThemeColors.WeaponEvolutionA
                    TierValue := array{option{1.0}}
                    FormatValue := FormatAsInt
                ability_info:
                    Id := ability_id.Multicast
                    MaybeWeaponId := option{weapon_id.EnergyBomb}
                    Name := "Multicast"
                    Description := "Increase the amount of times the weapon is casted."
                    Icon := UserInterface.Icons.Multicast_128
                    IconSmall := UserInterface.Icons.Multicast_64
                    IconColor := ThemeColors.WeaponEvolutionB
                    TierValue := array{option{1.0}}
                    FormatValue := FormatAsInt
    weapon_id.Boomerang => weapon_info:
        Id := weapon_id.Boomerang
        Name := "Boomerang"
        Description := "A projectile that returns to the player."
        Icon := UserInterface.Icons.WeaponBoomerang_128
        IconSmall := UserInterface.Icons.WeaponBoomerang_64
        Damage := 10.0
        Cooldown := 5.0
        Size := 50.0
        Evolution := array:
            array:
                ability_info:
                    Id := ability_id.ProjectileCount
                    MaybeWeaponId := option{weapon_id.Boomerang}
                    Name := "Projectile Count"
                    Description := "Increase the projectile count of the weapon."
                    Icon := UserInterface.Icons.ProjectileCount_128
                    IconSmall := UserInterface.Icons.ProjectileCount_64
                    IconColor := ThemeColors.WeaponEvolutionA
                    TierValue := array{option{1.0}}
                    FormatValue := FormatAsInt
                ability_info:
                    Id := ability_id.ProjectileRange
                    MaybeWeaponId := option{weapon_id.Boomerang}
                    Name := "Projectile Range"
                    Description := "Increase the distance that the projectile travels."
                    Icon := UserInterface.Icons.AbilityRange_128
                    IconSmall := UserInterface.Icons.AbilityRange_64
                    IconColor := ThemeColors.WeaponEvolutionB
                    TierValue := array{option{0.5}}
                    FormatValue := FormatAsPercent
            array:
                ability_info:
                    Id := ability_id.ProjectileCount
                    MaybeWeaponId := option{weapon_id.Boomerang}
                    Name := "Projectile Count"
                    Description := "Increase the projectile count of the weapon."
                    Icon := UserInterface.Icons.ProjectileCount_128
                    IconSmall := UserInterface.Icons.ProjectileCount_64
                    IconColor := ThemeColors.WeaponEvolutionA
                    TierValue := array{option{1.0}}
                    FormatValue := FormatAsInt
                ability_info:
                    Id := ability_id.WeaponEvolution
                    MaybeWeaponId := option{weapon_id.Boomerang}
                    Name := "Boomerang Growth"
                    Description := "Boomerang's size slowly increases as it travels out & deals bonus damage on return (+50% + Projectile Range size & damage)."
                    Icon := UserInterface.Icons.BoomerangBurst_128
                    IconSmall := UserInterface.Icons.BoomerangBurst_64
                    IconColor := ThemeColors.WeaponEvolutionUnique
                    TierValue := array{option{1.0}}
                    FormatValue := FormatAsUnique
            array:
                ability_info:
                    Id := ability_id.ProjectileCount
                    MaybeWeaponId := option{weapon_id.Boomerang}
                    Name := "Projectile Count"
                    Description := "Increase the projectile count of the weapon."
                    Icon := UserInterface.Icons.ProjectileCount_128
                    IconSmall := UserInterface.Icons.ProjectileCount_64
                    IconColor := ThemeColors.WeaponEvolutionA
                    TierValue := array{option{1.0}}
                    FormatValue := FormatAsInt
                ability_info:
                    Id := ability_id.ProjectileRange
                    MaybeWeaponId := option{weapon_id.Boomerang}
                    Name := "Projectile Range"
                    Description := "Increase the distance that the projectile travels."
                    Icon := UserInterface.Icons.AbilityRange_128
                    IconSmall := UserInterface.Icons.AbilityRange_64
                    IconColor := ThemeColors.WeaponEvolutionB
                    TierValue := array{option{0.5}}
                    FormatValue := FormatAsPercent

<# ----------------------------------------------------------------------------------------------------
    Global Functions
---------------------------------------------------------------------------------------------------- #>
GetWeaponInfo()<transacts>:[weapon_id]weapon_info =
    WeaponInfos    

GetWeaponInfo(Id:weapon_id)<decides><transacts>:weapon_info =
    WeaponInfos[Id]

GetRandomWeapons(Player:player, Count:int, ?Ignore:[]weapon_id = array{})<transacts>:[]weapon_info =
    var FilteredWeaponInfos:[]weapon_info = for:
        WeaponInfo : GetWeaponInfo()
        Player.IsWeaponUnlocked(WeaponInfo.Id)?
        not Player.HasWeapon(WeaponInfo.Id)?
        not Ignore.Find[WeaponInfo.Id]
    do:
        WeaponInfo

    ShuffledAbilityInfos := Shuffle(FilteredWeaponInfos)
    ShuffledAbilityInfos.Slice[0, Min(ShuffledAbilityInfos.Length, Count)] or array{}

GetRandomWeaponId()<decides><transacts>:weapon_id =
    Shuffle(GetWeaponIds())[0]

GetTotalWeapons()<transacts>:int =
    WeaponInfos.Length

GetWeaponIds()<transacts>:[]weapon_id =
    for (Id -> WeaponInfo : WeaponInfos) {Id}
