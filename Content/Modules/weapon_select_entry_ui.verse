using { /Fortnite.com/UI }
using { /Verse.org/Colors }
using { /UnrealEngine.com/Temporary/SpatialMath }
using { /UnrealEngine.com/Temporary/UI }

weapon_select_entry_ui := class(ui_component):
    Icon:texture_block = texture_block:
        DefaultImage := UserInterface.Icons.WeaponSword_128
        DefaultDesiredSize := vector2{X := 100.0, Y := 100.0}

    EquipButton:button_quiet = button_quiet{}

    var MaybeWeaponInfo:?weapon_info = false

    <# ----------------------------------------------------------------------------------------------------
        Initialization
    ---------------------------------------------------------------------------------------------------- #>
    CreateUI<override>():widget =
        EquipButton.SetText(MakeMessage("Equip"))
        EquipButton.OnClick().Subscribe(OnClickEquip)

        stack_box:
            Orientation := orientation.Vertical
            Slots := array:
                stack_box_slot:
                    HorizontalAlignment := horizontal_alignment.Center
                    VerticalAlignment := vertical_alignment.Top
                    Widget := Icon
                stack_box_slot:
                    HorizontalAlignment := horizontal_alignment.Fill
                    VerticalAlignment := vertical_alignment.Bottom
                    Padding := margin{Left := 0.0, Top := 20.0, Right := 0.0, Bottom := 0.0}
                    Distribution := option{1.0}
                    Widget := EquipButton

    <# ----------------------------------------------------------------------------------------------------
        Update Functions
    ---------------------------------------------------------------------------------------------------- #>
    Hide<override>():void =
        UI.SetVisibility(widget_visibility.Hidden)

    Set(WeaponInfo:weapon_info):void =
        set MaybeWeaponInfo = option{WeaponInfo}
        Icon.SetImage(WeaponInfo.Icon)
        UI.SetVisibility(widget_visibility.Visible)

    Update<override>():void =
        if (WeaponInfo := MaybeWeaponInfo?):
            if (EquippedWeapon := Player.GetEquippedWeapon()?, WeaponInfo.Id = EquippedWeapon):
                Icon.SetTint(ThemeColors.Highlight * 2.0)
                EquipButton.SetText(MakeMessage("Equipped"))
            else:
                if (Player.IsWeaponUnlocked(WeaponInfo.Id)?):
                    Icon.SetTint(NamedColors.White)
                    EquipButton.SetText(MakeMessage("Equip"))
                else:
                    Icon.SetTint(ThemeColors.Disabled)
                    EquipButton.SetText(MakeMessage("Locked"))

    <# ----------------------------------------------------------------------------------------------------
        Event Handlers
    ---------------------------------------------------------------------------------------------------- #>
    OnClickEquip(Message:widget_message):void =
        if (WeaponInfo := MaybeWeaponInfo?):
            if (not Player.IsWeaponUnlocked(WeaponInfo.Id)?):
                return

            if (EquippedWeapon := Player.GetEquippedWeapon()?):
                if (WeaponInfo.Id <> EquippedWeapon):
                    Player.EquipWeapon(WeaponInfo.Id)
            else:
                Player.EquipWeapon(WeaponInfo.Id)