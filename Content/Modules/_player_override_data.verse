GetPlayerDataOverride()<transacts>:?player_data = false
<#
    option:
        player_data:
            Version := 0
            News := GetNewsInfo().Length
            Settings := player_settings:
                Camera := 0
                CameraDistance := 3
                AutoAim := true
            ArenaProgress := 0
            PowerUps := map:
                power_up_id.Damage.ToInt() => 10
                power_up_id.Cooldown.ToInt() => 10
                power_up_id.Size.ToInt() => 10
                power_up_id.PlayerMaxHealth.ToInt() => 10
                power_up_id.PlayerHealthRegeneration.ToInt() => 10
                power_up_id.PlayerItemPickupDistance.ToInt() => 10
                power_up_id.PlayerCriticalRate.ToInt() => 10
                power_up_id.PlayerCriticalDamage.ToInt() => 10
                power_up_id.PlayerExperienceMultiplier.ToInt() => 10
                power_up_id.AbilityReroll.ToInt() => 100
                power_up_id.AbilityBanish.ToInt() => 10
                power_up_id.AbilityBoost.ToInt() => 5
                power_up_id.WeaponReroll.ToInt() => 2
                power_up_id.DashCount.ToInt() => 4
                power_up_id.Luck.ToInt() => 10
                power_up_id.GoldMultiplier.ToInt() => 3
                power_up_id.TreasureChestLuck.ToInt() => 10
                power_up_id.TreasureChestFocus.ToInt() => 1
                power_up_id.TreasureChestTier.ToInt() => 1
                power_up_id.Revive.ToInt() => 2
                power_up_id.ReviveRegeneration.ToInt() => 10
                power_up_id.ReviveBlast.ToInt() => 1
                power_up_id.AirDash.ToInt() => 1
            Currencies := map:
                currency_id.Gold.ToInt() => 1000000000.0
            WeaponProgress := map:
                weapon_id.Sword.ToInt() => 1
                weapon_id.Bow.ToInt() => 2
                weapon_id.Shuriken.ToInt() => 3
                weapon_id.Sentry.ToInt() => 4
                weapon_id.Aura.ToInt() => 5
                weapon_id.Torrent.ToInt() => 6
                weapon_id.Lightning.ToInt() => 7
                weapon_id.Trail.ToInt() => 7
                weapon_id.Shield.ToInt() => 7
                weapon_id.ChainLightning.ToInt() => 0
                weapon_id.EnergyBomb.ToInt() => 0
                weapon_id.Boomerang.ToInt() => 0
            DailyQuests := player_daily_quests:
                LastQuestAssignedTime := GetSecondsSinceEpoch() + 62135596800.0 - 90000.0
                ActiveQuests := array:
                    daily_quest:
                        Id := daily_quest_id.EliminateEnemies.ToInt()
                        Tier := 0
                        Progress := 4990.0
                    daily_quest:
                        Id := daily_quest_id.GatherTree.ToInt()
                        Tier := 0
                        Progress := 48.0
                    daily_quest:
                        Id := daily_quest_id.WeaponEvolution.ToInt()
                        Tier := 0
                        Progress := 5.0
#>