using { /Fortnite.com/Characters }
using { /Fortnite.com/Devices }
using { /Verse.org/Simulation }
using { /UnrealEngine.com/Temporary/SpatialMath }

<# ----------------------------------------------------------------------------------------------------
    Constants
---------------------------------------------------------------------------------------------------- #>
AbilityAuraBonusRadius:float = 25.0
AbilityAuraUpdateSizeTime:float = 0.5

<# ----------------------------------------------------------------------------------------------------
    ability_aura
---------------------------------------------------------------------------------------------------- #>
ability_aura<public> := class(ability):
    Id<override>:weapon_id = weapon_id.Aura
    DeactivatedEvent:events(tuple()) = events(tuple()){}

    var AttackCooldownTime:float = 1.0
    var AttackDamage:float = 10.0
    var AttackDamageMultiplier:float = 1.0
    var AttackRadius:float = 100.0
    var AttackRadiusMultiplier:float = 1.0
    var EvolutionDuration:float = 0.0
    var EvolutionPower:float = 0.0
    var IsUniqueEvolution:logic = false

    var CinematicSequence:cinematic_sequence_device = cinematic_sequence_device{}
    var IsBoost:logic = false
    var PreviousSize:float = 0.0
    var PauseTime:float = 0.0

    OnInitialize<override>():void =
        VFXPowerSequence := Player.ActivateVFXPowerup(vfx_powerup_sequence_id.AbilityAura)
        set CinematicSequence = VFXPowerSequence.CinematicSequence

    <# ----------------------------------------------------------------------------------------------------
        Get Functions
    ---------------------------------------------------------------------------------------------------- #>
    GetRadiusMultiplier()<transacts>:float =
        AttackRadiusMultiplier * (IsBoost? and 1.5 or 1.0)

    <# ----------------------------------------------------------------------------------------------------
        Update Functions
    ---------------------------------------------------------------------------------------------------- #>
    Activate<override>():void =
        spawn{Start()}
        set IsActive = true

    UpdateSize()<suspends>:void =
        TargetSize := GetRadiusMultiplier()
        DeltaSize := TargetSize - PreviousSize
        
        CinematicSequence.SetPlaybackTime(PreviousSize)
        CinematicSequence.SetPlayRate(Abs(DeltaSize) * (1.0 / AbilityAuraUpdateSizeTime))

        if (Sgn(DeltaSize) = 1.0):
            CinematicSequence.Play()
        else if (Sgn(DeltaSize) = -1.0):
            CinematicSequence.PlayReverse()
            
        Sleep(AbilityAuraUpdateSizeTime)
        CinematicSequence.SetPlaybackTime(TargetSize)
        CinematicSequence.Pause()
        set PreviousSize = TargetSize

    Start()<suspends>:void =
        UpdateStats()

        race:
            DeactivatedEvent.Await()

            if (EvolutionDuration = 10.0):
                set IsBoost = true
                Player.ActivateVFXPowerup(vfx_powerup_sequence_id.AbilityAuraEvolved)
                UpdateSize()
                Sleep(Inf)
            else:              
                block:
                    set IsBoost = false
                    if (PauseTime < 10.0 - EvolutionDuration):
                        Player.ActivateVFXPowerup(vfx_powerup_sequence_id.AbilityAura)
                        UpdateSize()
    
                    Sleep(10.0 - EvolutionDuration - AbilityAuraUpdateSizeTime - PauseTime)
    
                    set IsBoost = logic{Player.GetWeaponEvolutionLevel(Id) > 0}
                    if (IsBoost?):
                        Player.ActivateVFXPowerup(vfx_powerup_sequence_id.AbilityAuraEvolved)
                    else:
                        Player.ActivateVFXPowerup(vfx_powerup_sequence_id.AbilityAura)
    
                    UpdateSize()
                    Sleep(EvolutionDuration - (PauseTime > 10.0 - EvolutionDuration and PauseTime - (10.0 - EvolutionDuration) or 0.0) - AbilityAuraUpdateSizeTime)

                    loop:
                        if (IsBoost?):
                            set IsBoost = false
                            Player.DeactivateVFXPowerup(vfx_powerup_sequence_id.AbilityAuraEvolved)
                            Sleep(0.0)
                            Player.ActivateVFXPowerup(vfx_powerup_sequence_id.AbilityAura)
                            UpdateSize()

                        Sleep(10.0 - EvolutionDuration - AbilityAuraUpdateSizeTime)

                        set IsBoost = logic{Player.GetWeaponEvolutionLevel(Id) > 0}
                        if (IsBoost?):
                            Player.DeactivateVFXPowerup(vfx_powerup_sequence_id.AbilityAura)
                            Sleep(0.0)
                            Player.ActivateVFXPowerup(vfx_powerup_sequence_id.AbilityAuraEvolved)
                            UpdateSize()

                        Sleep(EvolutionDuration - AbilityAuraUpdateSizeTime)

            if (Character := Player.GetFortCharacter[], WeaponInfo := GetWeaponInfo[Id]):
                loop:
                    Sleep(AttackCooldownTime * (IsBoost? and 0.75 or 1.0))
                    AbilityPosition := Character.GetTransform().Translation.Make2D()

                    for:
                        NPCBehaviorBase : GetAllNPCs()
                        NPCCharacter := NPCBehaviorBase.GetAgent[].GetFortCharacter[]
                        NPCPosition := NPCCharacter.GetTransform().Translation.Make2D()
                        DistanceSquared(NPCPosition, AbilityPosition) <= Pow(AttackRadius * GetRadiusMultiplier() + NPCBehaviorBase.GetSize(), 2.0)
                    do:
                        if (IsUniqueEvolution? and IsBoost?, Direction := (NPCPosition - AbilityPosition).MakeUnitVector[]):
                            DirectionRotation := MakeRotation(vector3{X := 0.0, Y := 0.0, Z := 1.0}, ArcTan(Direction.Y, Direction.X))
                            spawn{NPCBehaviorBase.Knockback(1.0, DirectionRotation)}

                        Damage(Player.GetDamageArgs(NPCBehaviorBase, Id, WeaponInfo.Damage * (IsBoost? and 2.0 + EvolutionPower or 1.0), ?Multiplier := AttackDamageMultiplier), ?IsKnockback := false)

    Deactivate<override>(?IsPause:logic = false):void =
        DeactivatedEvent.Signal()
        Player.DeactivateVFXPowerup(vfx_powerup_sequence_id.AbilityAura)
        CinematicSequence.Stop()
        
        set IsActive = false
        if (Remainder := FMod[GetSimulationElapsedTime() - StartTime, 10.0]):
            set PauseTime = Remainder * 1.0
        
    UpdateStats():void =
        if (WeaponInfo := GetWeaponInfo[Id]):
            set AttackCooldownTime = WeaponInfo.Cooldown / (1.0 * (1.0 + Player.GetAbilityValue(Id, ability_id.Cooldown) + Player.GetAbilityValue(ability_id.Cooldown) + Player.GetPowerUpValue(power_up_id.Cooldown)))
            set AttackDamage = WeaponInfo.Damage
            set AttackDamageMultiplier = Player.GetAbilityValue(Id, ability_id.Damage)
            set AttackRadiusMultiplier = 1.0 + Player.GetAbilityValue(Id, ability_id.Size) + Player.GetAbilityValue(ability_id.Size) + Player.GetPowerUpValue(power_up_id.Size)
            set AttackRadius = WeaponInfo.Size

            set EvolutionDuration = Player.GetWeaponEvolutionValue(Id, ability_id.AuraDuration)
            set EvolutionPower = Player.GetWeaponEvolutionValue(Id, ability_id.AuraPower)
            set IsUniqueEvolution = Player.IsWeaponEvolutionAbilityActive(Id, ability_id.WeaponEvolution)