using { /Fortnite.com/Characters }
using { /Fortnite.com/Devices }
using { /Verse.org/Random }
using { /Verse.org/Simulation }
using { /UnrealEngine.com/Temporary/SpatialMath }

<# ----------------------------------------------------------------------------------------------------
    Constants
---------------------------------------------------------------------------------------------------- #>
DefaultAbilityBoomerangRange:float = 600.0
DefaultAbilityBoomerangDuration:float = 1.0
DefaultAbilityBoomerangSpeed:float = 1000.0

<# ----------------------------------------------------------------------------------------------------
    ability_boomerang
---------------------------------------------------------------------------------------------------- #>
ability_boomerang<public> := class(ability):
    Id<override>:weapon_id = weapon_id.Boomerang
    DeactivatedEvent:events(tuple()) = events(tuple()){}

    var AttackDamage:float = 10.0
    var AttackCooldownTime:float = 2.0
    var AttackDamageMultiplier:float = 1.0
    var AttackRadius:float = 25.0
    var ProjectileCount:int = 1
    var ProjectileRange:float = 0.0
    var IsUniqueEvolution:logic = false

    Activate<override>():void =
        spawn{Start()}
        set IsActive = true

    Start()<suspends>:void =
        UpdateStats()
        
        race:
            DeactivatedEvent.Await()
            if (WeaponInfo := GetWeaponInfo[Id]):
                Sleep(CooldownTimeRemaining)
                loop:
                    set CastTime = GetSimulationElapsedTime()
                    spawn{Ability()}
                    Sleep(AttackCooldownTime)

    Deactivate<override>(?IsPause:logic = false):void =
        DeactivatedEvent.Signal()
        set CooldownTimeRemaining = Max(0.0, AttackCooldownTime - (GetSimulationElapsedTime() - CastTime))
        set IsActive = false

    Ability()<suspends>:void =
        for (Index := 1..ProjectileCount):
            spawn{Boomerang()}
            Sleep(0.2)

    Boomerang()<suspends>:void =
        BoomerangStartTime := GetSimulationElapsedTime()
        var MaybeSpawnedProp:?creative_prop = false
        var NPCsHit:[]npc_behavior_base = array{}
        var IsReturn:logic = false

        Player.PlayAudio(audio_id.Projectile)

        if (Character := Player.GetFortCharacter[]):
            CastPosition := Character.GetTransform().Translation.Make2D()
            TargetRotation := Player.GetAutoAimRotation()
            TargetPosition := CastPosition + TargetRotation.GetLocalForward() * DefaultAbilityBoomerangRange * (1.0 + ProjectileRange)
            InitialScale := AttackRadius * 0.1 * UnitVector
            set MaybeSpawnedProp = SpawnVFX(IsUniqueEvolution? and vfx_id.AbilityBoomerangEvolved or vfx_id.AbilityBoomerang, CastPosition, ?Scale := InitialScale, ?Duration := 5.0)
                        
            if (SpawnedProp := MaybeSpawnedProp?):
                race:
                    block:
                        TargetScale := InitialScale * (IsUniqueEvolution? and (1.5 + ProjectileRange) or 1.0)
                        SpawnedProp.MoveTo(transform{Translation := TargetPosition, Rotation := IdentityRotation(), Scale :=  TargetScale}, DefaultAbilityBoomerangDuration)
                        loop:
                            TargetDistance := Distance(SpawnedProp.GetTransform().Translation, Character.GetTransform().Translation.Make2D())
                            MoveDuration := TargetDistance / DefaultAbilityBoomerangSpeed

                            MoveToResult := if (GetSimulationElapsedTime() - BoomerangStartTime >= 4.0):
                                move_to_result.DestinationReached
                            else:
                                race:
                                    SpawnedProp.MoveTo(Character.GetTransform().Translation.Make2D(), IdentityRotation(), MoveDuration)
                                    block:
                                        Sleep(TickTime * 3)
                                        move_to_result.WillNotReachDestination

                            if (MoveToResult = move_to_result.DestinationReached):
                                break
                    
                    block:
                        Sleep(DefaultAbilityBoomerangDuration)
                        set NPCsHit = array{}
                        set IsReturn = true
                        Sleep(Inf)
                    
                    loop:
                        CurrentPosition := SpawnedProp.GetTransform().Translation.Make2D()
                        CurrentScaleMultiplier := SpawnedProp.GetTransform().Scale.Z / InitialScale.Z
                        FinalDamage := AttackDamage * ((IsUniqueEvolution? and IsReturn?) and 1.5 + ProjectileRange or 1.0)

                        for:
                            NPCBehaviorBase : GetAllNPCs()
                            not NPCsHit.Find[NPCBehaviorBase]
                            NPCCharacter := NPCBehaviorBase.GetAgent[].GetFortCharacter[]
                            DistanceSquared(NPCCharacter.GetTransform().Translation.Make2D(), CurrentPosition) <= Pow((AttackRadius * CurrentScaleMultiplier) + NPCBehaviorBase.GetSize(), 2.0)
                        do:
                            Damage(Player.GetDamageArgs(NPCBehaviorBase, Id, FinalDamage, ?Multiplier := AttackDamageMultiplier))
                            set NPCsHit += array{NPCBehaviorBase}

                        Sleep(0.0)

            if (SpawnedProp := MaybeSpawnedProp?):
                SpawnedProp.Dispose()
                set MaybeSpawnedProp = false

    UpdateStats():void =
        if (WeaponInfo := GetWeaponInfo[Id]):
            set IsUniqueEvolution = Player.IsWeaponEvolutionAbilityActive(Id, ability_id.WeaponEvolution)
            set ProjectileCount = 1 + Floor[Player.GetWeaponEvolutionValue(Id, ability_id.ProjectileCount)] or 0
            set ProjectileRange = Player.GetWeaponEvolutionValue(Id, ability_id.ProjectileRange)

            set AttackDamageMultiplier = Player.GetAbilityValue(Id, ability_id.Damage)
            set AttackDamage = WeaponInfo.Damage
            set AttackRadius = WeaponInfo.Size * (1.0 + Player.GetAbilityValue(Id, ability_id.Size) + Player.GetAbilityValue(ability_id.Size) + Player.GetPowerUpValue(power_up_id.Size))
            set AttackCooldownTime = WeaponInfo.Cooldown / (1.0 * (1.0 + Player.GetAbilityValue(Id, ability_id.Cooldown) + Player.GetAbilityValue(ability_id.Cooldown) + Player.GetPowerUpValue(power_up_id.Cooldown)))