using { /Fortnite.com/AI }
using { /Fortnite.com/Characters }
using { /Fortnite.com/Devices }
using { /Verse.org/Simulation }
using { /UnrealEngine.com/Temporary/SpatialMath }

<# ----------------------------------------------------------------------------------------------------
    Constants
---------------------------------------------------------------------------------------------------- #>
DefaultAbilityProjectileDamage:float = 10.0
DefaultAbilityProjectileDistance:float = 1000.0
DefaultAbilityProjectileSpeed:float = 1000.0
DefaultAbilityBowProjectileCount:int = 3

<# ----------------------------------------------------------------------------------------------------
    Data Definitions
---------------------------------------------------------------------------------------------------- #>
projectile_properties<public> := struct:
    Owner<public>:agent
    Rotation<public>:rotation
    Angle<public>:float = 0.0
    Speed<public>:float
    (projectile_properties:)Distance<public>:float
    Damage<public>:float
    DamageMultiplier<public>:float = 1.0
    Radius<public>:float
    VFX<public>:vfx_id
    Piercing<public>:int = 0

poolable := interface:
    GetPooledObject()<transacts>:projectile
    ReturnToObjectPool(Projectile:projectile):void
    
<# ----------------------------------------------------------------------------------------------------
    ability_bow
---------------------------------------------------------------------------------------------------- #>
ability_bow<public> := class(ability, poolable):
    Id<override>:weapon_id = weapon_id.Bow
    DeactivatedEvent:events(tuple()) = events(tuple()){}

    var AttackDamage:float = DefaultAbilityProjectileDamage
    var AttackCooldownTime:float = 2.0
    var AttackDamageMultiplier:float = 1.0
    var AttackRadius:float = 25.0
    var ProjectileCount:int = DefaultAbilityBowProjectileCount
    var ProjectilePiercing:int = 0
    var MulticastCount:int = 1
    var IsChargedArrow:logic = false

    var ProjectileObjectPool:queue(projectile) = queue(projectile){}

    <# ----------------------------------------------------------------------------------------------------
        Get Functions
    ---------------------------------------------------------------------------------------------------- #>
    GetPooledObject<override>()<transacts>:projectile =
        if (DequeuedProjectile := ProjectileObjectPool.Dequeue[]):
            set ProjectileObjectPool = DequeuedProjectile(0)
            DequeuedProjectile(1)
        else:
            NewProjectile := projectile{MaybeManager := option{Self}}       

    ReturnToObjectPool<override>(Projectile:projectile):void =
        UpdatedProjectileObjectPool := ProjectileObjectPool.Enqueue(Projectile)
        set ProjectileObjectPool = UpdatedProjectileObjectPool

    <# ----------------------------------------------------------------------------------------------------
        Update Functions
    ---------------------------------------------------------------------------------------------------- #>
    Activate<override>():void =
        spawn{Start()}
        set IsActive = true

    Start()<suspends>:void =
        UpdateStats()

        race:
            DeactivatedEvent.Await()
            block:
                Sleep(CooldownTimeRemaining)
                loop:
                    set CastTime = GetSimulationElapsedTime()
                    spawn{Ability()}
                    Sleep(AttackCooldownTime)

    Deactivate<override>(?IsPause:logic = false):void =
        DeactivatedEvent.Signal()
        set CooldownTimeRemaining = Max(0.0, AttackCooldownTime - (GetSimulationElapsedTime() - CastTime))
        set IsActive = false

    Ability()<suspends>:void =
        for (CastIndex := 1..MulticastCount):
            Player.PlayAudio(IsChargedArrow? and audio_id.ProjectileEvolved or audio_id.Projectile)

            for (Index := 0..ProjectileCount - 1):
                ProjectileProperties := projectile_properties:
                    Owner := Player
                    Rotation := Player.GetAutoAimRotation()
                    Angle := DegreesToRadians(15.0) * Index - (Floor[ProjectileCount * 1.0 / 2.0] or 0) * 1.0 * DegreesToRadians(15.0)
                    Speed := DefaultAbilityProjectileSpeed
                    Distance := DefaultAbilityProjectileDistance
                    Damage := AttackDamage
                    DamageMultiplier := AttackDamageMultiplier
                    Radius := AttackRadius
                    VFX := IsChargedArrow? and vfx_id.AbilityBowEvolved or vfx_id.AbilityBow
                    Piercing := ProjectilePiercing

                Projectile := GetPooledObject()
                Projectile.Activate(ProjectileProperties)

            Sleep(0.1)

    UpdateStats():void =
        if (WeaponInfo := GetWeaponInfo[Id]):
            set IsChargedArrow = Player.IsWeaponEvolutionAbilityActive(Id, ability_id.WeaponEvolution)
            TargetProjectileCount := DefaultAbilityBowProjectileCount + Floor[Player.GetWeaponEvolutionValue(Id, ability_id.ProjectileCount)] or 0
            set ProjectileCount = IsChargedArrow? and 1 or TargetProjectileCount
            set AttackDamageMultiplier = Player.GetAbilityValue(Id, ability_id.Damage)
            set AttackDamage = WeaponInfo.Damage * (IsChargedArrow? and (1.0 + TargetProjectileCount * 0.2) or 1.0)
            set AttackRadius = WeaponInfo.Size * (IsChargedArrow? and (1.0 + TargetProjectileCount * 0.5) or 1.0) * (1.0 + Player.GetAbilityValue(Id, ability_id.Size) + Player.GetAbilityValue(ability_id.Size) + Player.GetPowerUpValue(power_up_id.Size))
            set AttackCooldownTime = WeaponInfo.Cooldown / (1.0 * (1.0 + Player.GetAbilityValue(Id, ability_id.Cooldown) + Player.GetAbilityValue(ability_id.Cooldown) + Player.GetPowerUpValue(power_up_id.Cooldown)))
            set MulticastCount = 1 + Floor[Player.GetWeaponEvolutionValue(Id, ability_id.Multicast)] or 0
            set ProjectilePiercing = IsChargedArrow? and -1 or (1 + Floor[Player.GetWeaponEvolutionValue(Id, ability_id.ProjectilePiercing)] or 1)
            
    <# ----------------------------------------------------------------------------------------------------
        Event Handlers
    ---------------------------------------------------------------------------------------------------- #>
    OnAbilityChanged(AbilityId:ability_id):void =
        UpdateStats()

<# ----------------------------------------------------------------------------------------------------
    projectile
---------------------------------------------------------------------------------------------------- #>
projectile<public> := class:
    var AgentsHit:[]agent = array{}
    var MaybePropjectileVFX:?creative_prop = false

    MaybeManager:?poolable = false

    <# ----------------------------------------------------------------------------------------------------
        Get Functions
    ---------------------------------------------------------------------------------------------------- #>
    HasHit(InAgentHit:agent)<transacts>:logic =
        for (AgentHit : AgentsHit, AgentHit = InAgentHit) {return true}
        false

    <# ----------------------------------------------------------------------------------------------------
        Update Functions
    ---------------------------------------------------------------------------------------------------- #>
    Activate<public>(Properties:projectile_properties):void =
        spawn{Projectile(Properties)}

    Projectile(Properties:projectile_properties)<suspends>:void =
        if (Character := Properties.Owner.GetFortCharacter[], Character.IsActive[]):
            set AgentsHit = array{}

            Position := Character.GetTransform().Translation
            #TargetRotation := Character.GetTransform().Rotation.ApplyYaw(Properties.Angle)
            TargetRotation := Properties.Rotation.ApplyYaw(Properties.Angle)
            Direction := TargetRotation.GetLocalForward()
            TargetPosition := Position + Direction * Properties.Distance
            Lifetime := Properties.Distance / Properties.Speed

            set MaybePropjectileVFX = SpawnVFX(Properties.VFX, Position, ?Rotation := TargetRotation, ?Scale := Properties.Radius * 0.1 * UnitVector, ?Duration := Lifetime)

            TargetAgents := if (player[Properties.Owner]):
                for (NPCBehaviorBase : GetAllNPCs(), NPCAgent := NPCBehaviorBase.GetAgent[]) {NPCAgent}
            else:
                GetAlivePlayers()

            var StartTime:float = GetSimulationElapsedTime()
            race:
                block:
                    if (PropjectileVFX := MaybePropjectileVFX?):
                        PropjectileVFX.MoveTo(TargetPosition, TargetRotation, Lifetime)
                    else:
                        Sleep(Lifetime)
                    
                loop:
                    ElapsedTime :=  GetSimulationElapsedTime() - StartTime
                    Progress := ElapsedTime / Lifetime
                    CurrentPosition := Lerp(Position, TargetPosition, Progress)

                    for:
                        TargetAgent : TargetAgents
                        not HasHit(TargetAgent)?
                        TargetCharacter := TargetAgent.GetFortCharacter[]
                        TargetSize := player[Properties.Owner] and npc_behavior_base[TargetAgent.GetNPCBehavior[]].GetSize() or PlayerCharacterSize
                        DistanceSquared(vector2{X := TargetCharacter.GetTransform().Translation.X, Y := TargetCharacter.GetTransform().Translation.Y}, vector2{X := CurrentPosition.X, Y := CurrentPosition.Y}) <= Pow(Properties.Radius + TargetSize, 2.0)
                    do:                           
                        if (Player := player[Properties.Owner], NPCBehaviorBase := npc_behavior_base[TargetAgent.GetNPCBehavior[]]):
                            if (Manager := ability_bow[MaybeManager?]):
                                Manager.Damage(Player.GetDamageArgs(NPCBehaviorBase, weapon_id.Bow, Properties.Damage, ?Multiplier := Properties.DamageMultiplier))
                        else if (PlayerTarget := player[TargetAgent], NPCBehaviorBase := npc_behavior_base[Properties.Owner.GetNPCBehavior[]]):
                            PlayerTarget.ApplyDamage(Properties.Damage, ?IsBoss := NPCBehaviorBase.IsBoss)

                        set AgentsHit += array{TargetAgent}

                    if (Progress >= 1.0 or (Properties.Piercing > 0 and AgentsHit.Length >= Properties.Piercing)):
                        break
                    else:
                        Sleep(0.0)

            if (Manager := MaybeManager?):
                if (PropjectileVFX := MaybePropjectileVFX?):
                    PropjectileVFX.Dispose()
                    set MaybePropjectileVFX = false

                Sleep(1.0)
                Manager.ReturnToObjectPool(Self)
