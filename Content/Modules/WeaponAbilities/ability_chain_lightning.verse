using { /Fortnite.com/Characters }
using { /Fortnite.com/Devices }
using { /Verse.org/Random }
using { /Verse.org/Simulation }
using { /UnrealEngine.com/Temporary/SpatialMath }

<# ----------------------------------------------------------------------------------------------------
    Constants
---------------------------------------------------------------------------------------------------- #>
DefaultAbilityChainLightningBounceDuration:float = 0.2
DefaultAbilityChainLightningRange:float = 600.0

<# ----------------------------------------------------------------------------------------------------
    ability_chain_lightning
---------------------------------------------------------------------------------------------------- #>
ability_chain_lightning<public> := class(ability):
    Id<override>:weapon_id = weapon_id.ChainLightning
    DeactivatedEvent:events(tuple()) = events(tuple()){}

    var AttackDamage:float = 10.0
    var AttackCooldownTime:float = 2.0
    var AttackDamageMultiplier:float = 1.0
    var AttackDamageBounceFalloff:float = 0.3
    var AttackRadius:float = 25.0
    var TotalBounceCount:int = 3
    var IsUniqueEvolution:logic = false
    
    Activate<override>():void =
        spawn{Start()}
        set IsActive = true

    Start()<suspends>:void =
        UpdateStats()
        
        race:
            DeactivatedEvent.Await()
            if (WeaponInfo := GetWeaponInfo[Id]):
                Sleep(CooldownTimeRemaining)
                loop:
                    set CastTime = GetSimulationElapsedTime()
                    spawn{Ability()}
                    Sleep(AttackCooldownTime)

    Deactivate<override>(?IsPause:logic = false):void =
        DeactivatedEvent.Signal()
        set CooldownTimeRemaining = Max(0.0, AttackCooldownTime - (GetSimulationElapsedTime() - CastTime))
        set IsActive = false

    Ability()<suspends>:void =
        var MaybeSpawnedProp:?creative_prop = false

        defer:
            if (SpawnedProp := MaybeSpawnedProp?):
                SpawnedProp.Dispose()

        Player.PlayAudio(audio_id.Projectile)

        if (PlayerCharacter := Player.GetFortCharacter[]):
            var BounceCount:int = 0
            var MaybePreviousNPC:?npc_behavior_base = false
            var CurrentPosition:vector3 = PlayerCharacter.GetTransform().Translation.Make2D()

            loop:
                if (BounceCount >= TotalBounceCount):
                    break

                var MaybeNearestNPC:?npc_behavior_base = false
                var NearestDistanceSquared:float = 0.0

                for:
                    NPCBehaviorBase : GetAllNPCs()
                    not MaybePreviousNPC? or NPCBehaviorBase <> MaybePreviousNPC?
                    NPCAgent := NPCBehaviorBase.GetAgent[]
                    NPCCharacter := NPCAgent.GetFortCharacter[]
                    DistanceSquaredFromAbility := DistanceSquared(NPCCharacter.GetTransform().Translation.Make2D(), CurrentPosition)
                    DistanceSquaredFromAbility <= Pow(DefaultAbilityChainLightningRange + NPCBehaviorBase.GetSize(), 2.0)
                    not MaybeNearestNPC? or DistanceSquaredFromAbility < NearestDistanceSquared
                do:
                    set MaybeNearestNPC = option{NPCBehaviorBase}
                    set NearestDistanceSquared = DistanceSquaredFromAbility

                if (NearestNPC := MaybeNearestNPC?, NearestNPCAgent := NearestNPC.GetAgent[], NearestNPCCharacter := NearestNPCAgent.GetFortCharacter[]):
                    if (not MaybeSpawnedProp?):
                        set MaybeSpawnedProp = SpawnVFX(IsUniqueEvolution? and vfx_id.AbilityChainLightningEvolved or vfx_id.AbilityChainLightning, CurrentPosition, ?Rotation := PlayerCharacter.GetTransform().Rotation, ?Scale := vector3{X := 3.0 + (TotalBounceCount * 1.0 / 100.0), Y := IsUniqueEvolution? and AttackRadius * 0.1 or 5.0, Z := DefaultAbilityChainLightningBounceDuration}, ?Duration := TotalBounceCount * DefaultAbilityChainLightningBounceDuration + 1.0)

                    if (SpawnedProp := MaybeSpawnedProp?):
                        SpawnedProp.MoveTo(NearestNPCCharacter.GetTransform().Translation.Make2D(), IdentityRotation(), DefaultAbilityChainLightningBounceDuration)
                        set CurrentPosition = SpawnedProp.GetTransform().Translation

                    if (IsUniqueEvolution?):
                        for:
                            NPCBehaviorBase : GetAllNPCs()
                            NPCCharacter := NPCBehaviorBase.GetAgent[].GetFortCharacter[]
                            DistanceSquared(NPCCharacter.GetTransform().Translation.Make2D(), CurrentPosition) <= Pow(AttackRadius + NPCBehaviorBase.GetSize(), 2.0)
                        do:
                            Damage(Player.GetDamageArgs(NPCBehaviorBase, Id, AttackDamage * Pow(AttackDamageBounceFalloff, BounceCount * 1.0), ?Multiplier := AttackDamageMultiplier))

                    else if (DistanceSquared(NearestNPCCharacter.GetTransform().Translation.Make2D(), CurrentPosition) <= Pow(DefaultAbilityChainLightningRange + NearestNPC.GetSize(), 2.0)):
                        Damage(Player.GetDamageArgs(NearestNPC, Id, AttackDamage * Pow(AttackDamageBounceFalloff, BounceCount * 1.0), ?Multiplier := AttackDamageMultiplier))

                    set BounceCount += 1
                    set MaybePreviousNPC = MaybeNearestNPC
                else:
                    break

    UpdateStats():void =
        if (WeaponInfo := GetWeaponInfo[Id]):
            set IsUniqueEvolution = Player.IsWeaponEvolutionAbilityActive(Id, ability_id.ChainLightningBurst)
            set TotalBounceCount = Floor[Player.GetWeaponEvolutionValue(Id, ability_id.ChainLightningBounce)] or 0
            set AttackDamageBounceFalloff = Player.GetWeaponEvolutionValue(Id, ability_id.ChainLightningResilience)

            set AttackDamageMultiplier = Player.GetAbilityValue(Id, ability_id.Damage)
            set AttackDamage = WeaponInfo.Damage
            set AttackRadius = WeaponInfo.Size * (1.0 + Player.GetAbilityValue(Id, ability_id.Size) + Player.GetAbilityValue(ability_id.Size) + Player.GetPowerUpValue(power_up_id.Size))
            set AttackCooldownTime = WeaponInfo.Cooldown / (1.0 * (1.0 + Player.GetAbilityValue(Id, ability_id.Cooldown) + Player.GetAbilityValue(ability_id.Cooldown) + Player.GetPowerUpValue(power_up_id.Cooldown)))