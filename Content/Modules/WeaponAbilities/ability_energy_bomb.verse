using { /Fortnite.com/Characters }
using { /Fortnite.com/Devices }
using { /Verse.org/Random }
using { /Verse.org/Simulation }
using { /UnrealEngine.com/Temporary/SpatialMath }

<# ----------------------------------------------------------------------------------------------------
    Constants
---------------------------------------------------------------------------------------------------- #>
DefaultAbilityEnergyBombRange:float = 516.0
DefaultAbilityEnergyBombGravityDuration:float = 0.2

<# ----------------------------------------------------------------------------------------------------
    ability_energy_bomb
---------------------------------------------------------------------------------------------------- #>
ability_energy_bomb<public> := class(ability):
    Id<override>:weapon_id = weapon_id.EnergyBomb
    DeactivatedEvent:events(tuple()) = events(tuple()){}

    var AttackDamage:float = 10.0
    var AttackCooldownTime:float = 2.0
    var AttackDamageMultiplier:float = 1.0
    var AttackRadius:float = 25.0
    var ProjectileCount:int = 1
    var Multicast:int = 1
    var IsUniqueEvolution:logic = false
    
    Activate<override>():void =
        spawn{Start()}
        set IsActive = true

    Start()<suspends>:void =
        UpdateStats()
        
        race:
            DeactivatedEvent.Await()
            if (WeaponInfo := GetWeaponInfo[Id]):
                Sleep(CooldownTimeRemaining)
                loop:
                    set CastTime = GetSimulationElapsedTime()
                    spawn{Ability()}
                    Sleep(AttackCooldownTime)

    Deactivate<override>(?IsPause:logic = false):void =
        DeactivatedEvent.Signal()
        set CooldownTimeRemaining = Max(0.0, AttackCooldownTime - (GetSimulationElapsedTime() - CastTime))
        set IsActive = false

    Ability()<suspends>:void =
        for (CastIndex := 1..Multicast):
            for (Index := 1..ProjectileCount):
                spawn{EnergyBomb(CastIndex, Index)}

            Sleep(0.3)

    EnergyBomb(CastIndex:int, Index:int)<suspends>:void =
        if (Character := Player.GetFortCharacter[]):
            CastPosition := Character.GetTransform().Translation.Make2D()
            TargetRotation := if (ProjectileCount > 1):
                Angle := DegreesToRadians(90.0) * (Index - 1) - (Floor[ProjectileCount * 1.0 / 2.0] or 0) * 1.0 * DegreesToRadians(90.0) * (Mod[ProjectileCount, 2] = 0 and 0.5 or 1.0)
                Player.GetAutoAimRotation().ApplyYaw(Angle)
            else:
                Player.GetAutoAimRotation()

            TargetPosition := CastPosition + TargetRotation.GetLocalForward() * DefaultAbilityEnergyBombRange

            SpawnVFX((IsUniqueEvolution? and CastIndex = 1) and vfx_id.AbilityEnergyBombEvolved or vfx_id.AbilityEnergyBomb, CastPosition, ?Rotation := TargetRotation, ?Scale := vector3{X := AttackRadius * 0.1, Y := AttackRadius * 0.1, Z := 2.0}, ?Duration := 3.0)
            Player.PlayAudio(audio_id.Projectile)

            if (IsUniqueEvolution? and CastIndex = 1):
                Sleep(1.5)

                for:
                    NPCBehaviorBase : GetAllNPCs()
                    NPCCharacter := NPCBehaviorBase.GetAgent[].GetFortCharacter[]
                    NPCPosition := NPCCharacter.GetTransform().Translation.Make2D()
                    DistanceSquared(NPCPosition, TargetPosition) <= Pow(AttackRadius * 1.5 + NPCBehaviorBase.GetSize(), 2.0)
                    Direction := (TargetPosition - NPCPosition).MakeUnitVector[]
                do:
                    DirectionRotation := MakeRotation(vector3{X := 0.0, Y := 0.0, Z := 1.0}, ArcTan(Direction.Y, Direction.X))
                    spawn{NPCBehaviorBase.Knockback(1.0, DirectionRotation)}

                Sleep(0.5)
            else:
                Sleep(2.0)
            
            for:
                NPCBehaviorBase : GetAllNPCs()
                NPCCharacter := NPCBehaviorBase.GetAgent[].GetFortCharacter[]
                DistanceSquared(NPCCharacter.GetTransform().Translation.Make2D(), TargetPosition) <= Pow(AttackRadius + NPCBehaviorBase.GetSize(), 2.0)
            do:
                Damage(Player.GetDamageArgs(NPCBehaviorBase, Id, AttackDamage, ?Multiplier := AttackDamageMultiplier))

    UpdateStats():void =
        if (WeaponInfo := GetWeaponInfo[Id]):
            set IsUniqueEvolution = Player.IsWeaponEvolutionAbilityActive(Id, ability_id.WeaponEvolution)
            set ProjectileCount = 1 + Floor[Player.GetWeaponEvolutionValue(Id, ability_id.ProjectileCount)] or 0
            set Multicast = 1 + Floor[Player.GetWeaponEvolutionValue(Id, ability_id.Multicast)] or 0

            set AttackDamageMultiplier = Player.GetAbilityValue(Id, ability_id.Damage)
            set AttackDamage = WeaponInfo.Damage
            set AttackRadius = WeaponInfo.Size * (1.0 + Player.GetAbilityValue(Id, ability_id.Size) + Player.GetAbilityValue(ability_id.Size) + Player.GetPowerUpValue(power_up_id.Size))
            set AttackCooldownTime = WeaponInfo.Cooldown / (1.0 * (1.0 + Player.GetAbilityValue(Id, ability_id.Cooldown) + Player.GetAbilityValue(ability_id.Cooldown) + Player.GetPowerUpValue(power_up_id.Cooldown)))