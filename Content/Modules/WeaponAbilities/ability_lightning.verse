using { /Fortnite.com/Characters }
using { /Verse.org/Random }
using { /Verse.org/Simulation }
using { /UnrealEngine.com/Temporary }
using { /UnrealEngine.com/Temporary/SpatialMath }

<# ----------------------------------------------------------------------------------------------------
    Constants
---------------------------------------------------------------------------------------------------- #>
BaseAbilityLightningRange:float = 900.0

<# ----------------------------------------------------------------------------------------------------
    ability_lightning
---------------------------------------------------------------------------------------------------- #>
ability_lightning<public> := class(ability):
    Id<override>:weapon_id = weapon_id.Lightning
    DeactivatedEvent:events(tuple()) = events(tuple()){}

    var ProjectileCount:int = 1
    var AttackCooldownTime:float = 1.0
    var AttackDamageMultiplier:float = 1.0
    var AttackRadius:float = 100.0
    var EvolutionSize:float = 1.0
    var IsFocusFire:logic = false

    Activate<override>():void =
        spawn{Start()}
        set IsActive = true

    Start()<suspends>:void =
        UpdateStats()

        race:
            DeactivatedEvent.Await()
            if (WeaponInfo := GetWeaponInfo[Id]):
                Sleep(CooldownTimeRemaining)

                loop:
                    set CastTime = GetSimulationElapsedTime()
                    spawn{Ability(WeaponInfo)}
                    Sleep(AttackCooldownTime)

    Ability(WeaponInfo:weapon_info)<suspends>:void =
        if (Character := Player.GetFortCharacter[]):
            var NPCInRangeCount:int = 0
            NPCMaxHealthInRange := for:
                NPCBehaviorBase : GetAllNPCs()
                NPCCharacter := NPCBehaviorBase.GetAgent[].GetFortCharacter[]
                DistanceSquared(NPCCharacter.GetTransform().Translation.Make2D(), Character.GetTransform().Translation.Make2D()) <= Pow(BaseAbilityLightningRange + NPCBehaviorBase.GetSize(), 2.0)
            do:
                set NPCInRangeCount += 1
                (NPCBehaviorBase, NPCBehaviorBase.BaseHealth)
            
            if (NPCInRangeCount = 0) then return

            NPCInRange := if (IsFocusFire?):
                SortBy(NPCMaxHealthInRange, LessNPCMaxHealth)
            else:
                NPCMaxHealthInRange

            var TargetIndex:int = 0
            var MaybeTargetPosition:?vector3 = false

            for (Index := 1..ProjectileCount):
                if (IsFocusFire?):
                    loop:
                        if (TargetIndex > NPCInRangeCount - 1):
                            break

                        if:
                            TargetNPC := NPCInRange[TargetIndex]
                            TargetNPC(0).IsDamagable()?
                            TargetNPCCharacter := TargetNPC(0).GetAgent[].GetFortCharacter[]
                        then:
                            set MaybeTargetPosition = option{TargetNPCCharacter.GetTransform().Translation.Make2D()}
                            break

                        set TargetIndex += 1
                else:
                    if:
                        TargetNPC := NPCInRange[GetRandomInt(0, NPCInRangeCount - 1)]
                        TargetNPCCharacter := TargetNPC(0).GetAgent[].GetFortCharacter[]
                    then:
                        set MaybeTargetPosition = option{TargetNPCCharacter.GetTransform().Translation.Make2D()}

                if (TargetPosition := MaybeTargetPosition?):
                    SpawnVFX(EvolutionSize > 1.0 and vfx_id.AbilityLightningEvolved or vfx_id.AbilityLightning, TargetPosition, ?Scale := vector3{X := AttackRadius * 0.1, Y := AttackRadius * 0.1, Z := 1.0}, ?Duration := 2.0)

                    for:
                        NPCBehaviorBase : GetAllNPCs()
                        NPCCharacter := NPCBehaviorBase.GetAgent[].GetFortCharacter[]
                        DistanceSquared(NPCCharacter.GetTransform().Translation.Make2D(), TargetPosition) <= Pow(AttackRadius + NPCBehaviorBase.GetSize(), 2.0)
                    do:
                        Damage(Player.GetDamageArgs(NPCBehaviorBase, Id, WeaponInfo.Damage, ?Multiplier := AttackDamageMultiplier))

                Sleep(0.1)

    Deactivate<override>(?IsPause:logic = false):void =
        DeactivatedEvent.Signal()
        set CooldownTimeRemaining = Max(0.0, AttackCooldownTime - (GetSimulationElapsedTime() - CastTime))
        set IsActive = false

    LessNPCMaxHealth(NPCMaxHealthA:tuple(npc_behavior_base, float), NPCMaxHealthB:tuple(npc_behavior_base, float))<computes><decides>:void =
        NPCMaxHealthA(1) > NPCMaxHealthB(1)

    UpdateStats():void =
        if (WeaponInfo := GetWeaponInfo[Id]):
            set ProjectileCount = 1 + Floor[Player.GetWeaponEvolutionValue(Id, ability_id.ProjectileCount)] or 0
            set EvolutionSize = 1.0 + Player.GetWeaponEvolutionValue(Id, ability_id.Size)
            set IsFocusFire = Player.IsWeaponEvolutionAbilityActive(Id, ability_id.WeaponEvolution)
            set AttackCooldownTime = WeaponInfo.Cooldown / (1.0 * (1.0 + Player.GetAbilityValue(Id, ability_id.Cooldown) + Player.GetAbilityValue(ability_id.Cooldown) + Player.GetPowerUpValue(power_up_id.Cooldown)))
            set AttackDamageMultiplier = Player.GetAbilityValue(Id, ability_id.Damage)
            set AttackRadius = WeaponInfo.Size * EvolutionSize * (1.0 + Player.GetAbilityValue(Id, ability_id.Size) + Player.GetAbilityValue(ability_id.Size) + Player.GetPowerUpValue(power_up_id.Size))