using { /Fortnite.com/Characters }
using { /Verse.org/Random }
using { /Verse.org/Simulation }
using { /UnrealEngine.com/Temporary/SpatialMath }

<# ----------------------------------------------------------------------------------------------------
    Constants
---------------------------------------------------------------------------------------------------- #>
DefaultAbilitySentryRange:float = 300.0
DefaultAbilitySentryDuration:float = 5.0
DefaultAbilitySentryAttackInterval:float = 1.0
DefaultAbilitySentryRotationDuration:float = 0.15

<# ----------------------------------------------------------------------------------------------------
    ability_sentry
---------------------------------------------------------------------------------------------------- #>
ability_sentry<public> := class(ability):
    Id<override>:weapon_id = weapon_id.Sentry
    DeactivatedEvent:events(tuple()) = events(tuple()){}

    var AttackDamage:float = 10.0
    var AttackCooldownTime:float = 2.0
    var AttackDamageMultiplier:float = 1.0
    var AttackRadius:float = 25.0
    var ProjectileCount:int = DefaultAbilityBowProjectileCount
    var FireRate:float = 1.0
    var IsUniqueEvolution:logic = false

    Activate<override>():void =
        spawn{Start()}
        set IsActive = true

    Start()<suspends>:void =
        UpdateStats()
        
        race:
            DeactivatedEvent.Await()
            if (WeaponInfo := GetWeaponInfo[Id]):
                Sleep(CooldownTimeRemaining)
                loop:
                    set CastTime = GetSimulationElapsedTime()
                    spawn{Ability()}
                    Sleep(AttackCooldownTime)

    Deactivate<override>(?IsPause:logic = false):void =
        DeactivatedEvent.Signal()
        set CooldownTimeRemaining = Max(0.0, AttackCooldownTime - (GetSimulationElapsedTime() - CastTime))
        set IsActive = false

    Ability()<suspends>:void =
        for (CastIndex := 1..ProjectileCount):
            spawn{Sentry()}
            Sleep(0.1)

    Sentry()<suspends>:void =
        if (PlayerCharacter := Player.GetFortCharacter[], WeaponInfo := GetWeaponInfo[Id]):
            SentryPosition := GetRandomPositionOnCircle(GetRandomFloat(150.0, 300.0), ?Origin := PlayerCharacter.GetTransform().Translation.Make2D())
            SentryRange := DefaultAbilitySentryRange * (IsUniqueEvolution? and 2.0 or 1.0)
            MaybeSpawnedProp := SpawnVFX(IsUniqueEvolution? and vfx_id.AbilitySentryEvolved or vfx_id.AbilitySentry, SentryPosition, ?Rotation := PlayerCharacter.GetTransform().Rotation, ?Scale := vector3{X := SentryRange * 0.1, Y := SentryRange * 0.1, Z := 1.0 + (FireRate / 100.0)}, ?Duration := DefaultAbilitySentryDuration)
            Player.PlayAudio(audio_id.Projectile)

            var IsFirst:logic = true
            race:
                Sleep(DefaultAbilitySentryDuration * (IsUniqueEvolution? and 2.0 or 1.0))
                loop:
                    var MaybeNearestNPC:?npc_behavior_base = false
                    var NearestDistanceSquared:float = 0.0

                    for:
                        NPCBehaviorBase : GetAllNPCs()
                        NPCAgent := NPCBehaviorBase.GetAgent[]
                        NPCCharacter := NPCAgent.GetFortCharacter[]
                        DistanceSquaredFromSentry := DistanceSquared(NPCCharacter.GetTransform().Translation, SentryPosition)
                        DistanceSquaredFromSentry <= Pow(SentryRange + NPCBehaviorBase.GetSize(), 2.0)
                        not MaybeNearestNPC? or DistanceSquaredFromSentry < NearestDistanceSquared
                    do:
                        set MaybeNearestNPC = option{NPCBehaviorBase}
                        set NearestDistanceSquared = DistanceSquaredFromSentry

                    if (NearestNPC := MaybeNearestNPC?, NearestNPCAgent := NearestNPC.GetAgent[], NearestNPCCharacter := NearestNPCAgent.GetFortCharacter[]):
                        if (SpawnedProp := MaybeSpawnedProp?, Direction := (NearestNPCCharacter.GetTransform().Translation - SentryPosition).MakeUnitVector[]):
                            TargetRotation := MakeRotation(vector3{X := 0.0, Y := 0.0, Z := 1.0}, ArcTan(Direction.Y, Direction.X))
                            if (IsFirst?):
                                if (SpawnedProp.TeleportTo[SpawnedProp.GetTransform().Translation, TargetRotation]) {}
                            else:                                
                                SpawnedProp.MoveTo(SpawnedProp.GetTransform().Translation, TargetRotation, DefaultAbilitySentryRotationDuration)

                        if:
                            DistanceSquaredFromSentry := DistanceSquared(NearestNPCCharacter.GetTransform().Translation, SentryPosition)
                            DistanceSquaredFromSentry <= Pow(SentryRange + NearestNPC.GetSize(), 2.0)
                        then:
                            SpawnVFX(IsUniqueEvolution? and vfx_id.ImpactAreaEvolved or vfx_id.ImpactArea, NearestNPCCharacter.GetTransform().Translation.Make2D(), ?Scale := AttackRadius * 0.1 * UnitVector, ?Duration := DefaultAbilitySentryDuration)

                            for:
                                NPCBehaviorBase : GetAllNPCs()
                                NPCAgent := NPCBehaviorBase.GetAgent[]
                                NPCCharacter := NPCAgent.GetFortCharacter[]
                                DistanceSquared(NPCCharacter.GetTransform().Translation.Make2D(), NearestNPCCharacter.GetTransform().Translation.Make2D()) <= Pow(AttackRadius + NPCBehaviorBase.GetSize(), 2.0)
                            do:
                                Damage(Player.GetDamageArgs(NPCBehaviorBase, Id, AttackDamage, ?Multiplier := AttackDamageMultiplier))
                    else:
                        if (not IsFirst?):
                            Sleep(DefaultAbilitySentryRotationDuration)

                    Sleep(FireRate - (IsFirst? and DefaultAbilitySentryRotationDuration or 0.0))
                    set IsFirst = false

    UpdateStats():void =
        if (WeaponInfo := GetWeaponInfo[Id]):
            set IsUniqueEvolution = Player.IsWeaponEvolutionAbilityActive(Id, ability_id.WeaponEvolution)
            set ProjectileCount = 1 + Floor[Player.GetWeaponEvolutionValue(Id, ability_id.ProjectileCount)] or 0
            set AttackDamageMultiplier = Player.GetAbilityValue(Id, ability_id.Damage)
            set AttackDamage = WeaponInfo.Damage * (IsUniqueEvolution? and 2.0 or 1.0)
            set AttackRadius = WeaponInfo.Size * (1.0 + Player.GetAbilityValue(Id, ability_id.Size) + Player.GetAbilityValue(ability_id.Size) + Player.GetPowerUpValue(power_up_id.Size))
            set AttackCooldownTime = WeaponInfo.Cooldown / (1.0 * (1.0 + Player.GetAbilityValue(Id, ability_id.Cooldown) + Player.GetAbilityValue(ability_id.Cooldown) + Player.GetPowerUpValue(power_up_id.Cooldown)))
            set FireRate = DefaultAbilitySentryAttackInterval / (1.0 + Player.GetWeaponEvolutionValue(Id, ability_id.FireRate))
