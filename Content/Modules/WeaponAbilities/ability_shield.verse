using { /Fortnite.com/Characters }
using { /Fortnite.com/Game }
using { /Verse.org/Simulation }
using { /UnrealEngine.com/Temporary/SpatialMath }

<# ----------------------------------------------------------------------------------------------------
    Constants
---------------------------------------------------------------------------------------------------- #>
DefaultAbilityShieldAttackTime:float = 0.2

<# ----------------------------------------------------------------------------------------------------
    ability_shield
---------------------------------------------------------------------------------------------------- #>
ability_shield<public> := class(ability):
    Id<override>:weapon_id = weapon_id.Shield
    DeactivatedEvent:events(tuple()) = events(tuple()){}
    
    var AttackCooldownTime:float = 1.0
    var AttackDamage:float = 10.0
    var AttackDamageMultiplier:float = 1.0
    var AttackRadius:float = 100.0

    var ShieldMaxHealth:float = 0.5
    var ShieldRegenerationRate:float = 0.125
    var IsUniqueEvolution:logic = false

    var ShieldIsActive:logic = false
    var MaybeShieldDamagedSubscription:?cancelable = false

    <# ----------------------------------------------------------------------------------------------------
        Update Functions
    ---------------------------------------------------------------------------------------------------- #>
    Activate<override>():void =
        if (Character := Player.GetFortCharacter[]):
            ShieldDamagedSubscription := Character.DamagedShieldEvent().Subscribe(OnDamagedShield)
            set MaybeShieldDamagedSubscription = option{ShieldDamagedSubscription}
            
            if (Character.GetShield() > 0.0):
                Player.ActivateVFXPowerup(IsUniqueEvolution? and vfx_powerup_id.ShieldEvolved or vfx_powerup_id.Shield)

        spawn{Start()}
        set IsActive = true

    Start()<suspends>:void =
        UpdateStats()

        race:
            DeactivatedEvent.Await()
            if (Character := Player.GetFortCharacter[]):
                Sleep(CooldownTimeRemaining)
                loop:
                    set CastTime = GetSimulationElapsedTime()
                    MaxShield := Character.GetMaxHealth() * ShieldMaxHealth
                    ShieldAmount := MaxShield * ShieldRegenerationRate
                    
                    Character.SetMaxShield(MaxShield)
                    if (Character.GetShield() < MaxShield):
                        Character.SetShield(Min(MaxShield, Character.GetShield() + ShieldAmount))
                        Player.PlayAudio(audio_id.Shield)

                    Player.ActivateVFXPowerup(IsUniqueEvolution? and vfx_powerup_id.ShieldEvolved or vfx_powerup_id.Shield)
                    Sleep(AttackCooldownTime)
                    set ShieldIsActive = true

    Deactivate<override>(?IsPause:logic = false):void =
        DeactivatedEvent.Signal()
        Player.DeactivateVFXPowerup(vfx_powerup_id.ShieldEvolved)
        Player.DeactivateVFXPowerup(vfx_powerup_id.Shield)

        if (ShieldDamagedSubscription := MaybeShieldDamagedSubscription?):
            ShieldDamagedSubscription.Cancel()

        set MaybeShieldDamagedSubscription = false
        set CooldownTimeRemaining = Max(0.0, AttackCooldownTime - (GetSimulationElapsedTime() - CastTime))
        set IsActive = false

    UpdateStats():void =
        if (WeaponInfo := GetWeaponInfo[Id]):
            set IsUniqueEvolution = Player.IsWeaponEvolutionAbilityActive(Id, ability_id.WeaponEvolution)
            set ShieldMaxHealth = Player.GetWeaponEvolutionValue(Id, ability_id.ShieldMaxHealth)
            set ShieldRegenerationRate = Player.GetWeaponEvolutionValue(Id, ability_id.ShieldRegenerationRate)

            set AttackCooldownTime = WeaponInfo.Cooldown / (1.0 * (1.0 + Player.GetAbilityValue(Id, ability_id.Cooldown) + Player.GetAbilityValue(ability_id.Cooldown) + Player.GetPowerUpValue(power_up_id.Cooldown)))
            set AttackDamage = WeaponInfo.Damage
            set AttackDamageMultiplier = Player.GetAbilityValue(Id, ability_id.Damage)
            set AttackRadius = WeaponInfo.Size * (1.0 + Player.GetAbilityValue(Id, ability_id.Size) + Player.GetAbilityValue(ability_id.Size) + Player.GetPowerUpValue(power_up_id.Size))

    <# ----------------------------------------------------------------------------------------------------
        Event Handlers
    ---------------------------------------------------------------------------------------------------- #>
    OnDamagedShield(DamageResult:damage_result):void =
        if (ShieldIsActive?):
            spawn{Damage()}

    Damage()<suspends>:void =
        if (Character := Player.GetFortCharacter[]):
            IsCracked := logic{Character.GetShield() = 0.0}
            AbilityPosition := Character.GetTransform().Translation.Make2D()
            FinalAttackRadius := AttackRadius * ((IsUniqueEvolution? and IsCracked?) and 1.5 or 1.0)
            SpawnVFX(IsUniqueEvolution? and vfx_id.AbilityShieldEvolved or vfx_id.AbilityShield, AbilityPosition, ?Scale := vector3{X := FinalAttackRadius * 0.1, Y := FinalAttackRadius * 0.1, Z := DefaultAbilityShieldAttackTime}, ?Duration := DefaultAbilityShieldAttackTime * 2)
            Sleep(DefaultAbilityShieldAttackTime)

            for:
                NPCBehaviorBase : GetAllNPCs()
                NPCCharacter := NPCBehaviorBase.GetAgent[].GetFortCharacter[]
                NPCPosition := NPCCharacter.GetTransform().Translation.Make2D()
                DistanceSquared(NPCPosition, AbilityPosition) <= Pow(FinalAttackRadius + NPCBehaviorBase.GetSize(), 2.0)
            do:
                if (IsUniqueEvolution? and IsCracked?, Direction := (NPCPosition - AbilityPosition).MakeUnitVector[]):
                    DirectionRotation := MakeRotation(vector3{X := 0.0, Y := 0.0, Z := 1.0}, ArcTan(Direction.Y, Direction.X))
                    spawn{NPCBehaviorBase.Knockback(1.0, DirectionRotation)}

                Damage(Player.GetDamageArgs(NPCBehaviorBase, Id, AttackDamage * ((IsUniqueEvolution? and IsCracked?) and 1.5 or 1.0), ?Multiplier := AttackDamageMultiplier))               

            if (IsCracked?):
                Player.DeactivateVFXPowerup(vfx_powerup_id.Shield)
                set ShieldIsActive = false