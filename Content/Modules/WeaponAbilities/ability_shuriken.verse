using { /Fortnite.com/Characters }
using { /Fortnite.com/Devices }
using { /Verse.org/Simulation }
using { /UnrealEngine.com/Temporary/SpatialMath }

<# ----------------------------------------------------------------------------------------------------
    Constants
---------------------------------------------------------------------------------------------------- #>
DefaultAbilityRotatorSpeed:float = PiFloat
DefaultAbilityRotatorSize:float = 400.0
DefaultAbilityRotatorRevolutionTime:float = PiFloat * 2 / DefaultAbilityRotatorSpeed
DefaultAbilityRotatorDuration:float = DefaultAbilityRotatorRevolutionTime * 3.0

<# ----------------------------------------------------------------------------------------------------
    ability_shuriken
---------------------------------------------------------------------------------------------------- #>
ability_shuriken<public> := class(ability):
    Id<override>:weapon_id = weapon_id.Shuriken
    DeactivatedEvent:events(tuple()) = events(tuple()){}

    var AttackDamage:float = 10.0
    var AttackCooldownTime:float = 2.0
    var AttackDamageMultiplier:float = 1.0
    var AttackRadius:float = 25.0
    var ProjectileCount:int = 1
    var ProjectileSpeed:float = 1.0
    var IsUniqueEvolution:logic = false

    var Projectiles:[]ability_shuriken_projectile = array{}
    var ProjectileObjectPool:queue(ability_shuriken_projectile) = queue(ability_shuriken_projectile){}

    <# ----------------------------------------------------------------------------------------------------
        Get Functions
    ---------------------------------------------------------------------------------------------------- #>
    GetPooledObject()<transacts>:ability_shuriken_projectile =
        if (DequeuedProjectile := ProjectileObjectPool.Dequeue[]):
            set ProjectileObjectPool = DequeuedProjectile(0)
            DequeuedProjectile(1)
        else:
            NewProjectile := ability_shuriken_projectile{MaybeManager := option{Self}, Player := Player}

    ReturnToObjectPool(Projectile:ability_shuriken_projectile):void =
        UpdatedProjectileObjectPool := ProjectileObjectPool.Enqueue(Projectile)
        set ProjectileObjectPool = UpdatedProjectileObjectPool

    <# ----------------------------------------------------------------------------------------------------
        Update Functions
    ---------------------------------------------------------------------------------------------------- #>
    Activate<override>():void =
        spawn{Start()}
        Player.ActivateVFXPowerup(vfx_powerup_id.ShurikenBase)
        set IsActive = true

    Start()<suspends>:void =
        UpdateStats()      

        race:
            DeactivatedEvent.Await()
            if (WeaponInfo := GetWeaponInfo[weapon_id.Shuriken]):
                Sleep(CooldownTimeRemaining)
                
                loop:
                    set CastTime = GetSimulationElapsedTime()
                    Player.PlayAudio(audio_id.Projectile)
                    for (Index := 1..ProjectileCount):
                        spawn{GetPooledObject().Activate(Index, AttackDamage, AttackDamageMultiplier, ProjectileCount, ProjectileSpeed, AttackRadius, IsUniqueEvolution)}

                    Sleep(AttackCooldownTime)

    Deactivate<override>(?IsPause:logic = false):void =
        DeactivatedEvent.Signal()
        Player.DeactivateVFXPowerup(vfx_powerup_id.ShurikenBase)
        set CooldownTimeRemaining = Max(0.0, AttackCooldownTime - (GetSimulationElapsedTime() - CastTime))
        set IsActive = false

    UpdateStats():void =
        if (WeaponInfo := GetWeaponInfo[Id]):
            set IsUniqueEvolution = Player.IsWeaponEvolutionAbilityActive(Id, ability_id.WeaponEvolution)
            set ProjectileCount = 1 + Floor[Player.GetWeaponEvolutionValue(Id, ability_id.ProjectileCount)] or 0
            set ProjectileSpeed = DefaultAbilityRotatorSpeed * (1.0 - Player.GetWeaponEvolutionValue(Id, ability_id.ProjectileSpeed) / 3.0) * (IsUniqueEvolution? and 0.5 or 1.0)
            set AttackDamageMultiplier = Player.GetAbilityValue(Id, ability_id.Damage)
            set AttackDamage = WeaponInfo.Damage * (IsUniqueEvolution? and 2.0 or 1.0) * (1.0 + Player.GetWeaponEvolutionValue(Id, ability_id.ProjectileSpeed))
            set AttackRadius = WeaponInfo.Size * (IsUniqueEvolution? and 2.0 or 1.0) * (1.0 + Player.GetAbilityValue(Id, ability_id.Size) + Player.GetAbilityValue(ability_id.Size) + Player.GetPowerUpValue(power_up_id.Size))
            set AttackCooldownTime = WeaponInfo.Cooldown / (1.0 * (1.0 + Player.GetAbilityValue(Id, ability_id.Cooldown) + Player.GetAbilityValue(ability_id.Cooldown) + Player.GetPowerUpValue(power_up_id.Cooldown)))

<# ----------------------------------------------------------------------------------------------------
    ability_shuriken_projectile
---------------------------------------------------------------------------------------------------- #>
ability_shuriken_projectile := class:
    MaybeManager:?ability_shuriken = false
    Player:player
    DeltaTimer:delta_timer = delta_timer{}

    var AgentsHit:[]agent = array{}

    Activate(Index:int, AttackDamage:float, AttackDamageMultiplier:float, ProjectileCount:int, ProjectileSpeed:float, AttackRadius:float, IsUniqueEvolution:logic)<suspends>:void =
        DeltaTimer.Reset()
        InitialRotation := IdentityRotation()

        var MaybeSpawnedProp:?creative_prop = false

        if (Character := Player.GetFortCharacter[], Character.IsActive[]):
            #Angle := (DeltaTimer.GetElapsedTime() * ProjectileSpeed + PiFloat * 2.0 / (ProjectileCount * 1.0) * Index + DegreesToRadians(InitialRotation.GetYawPitchRollDegrees()[0] or 0.0))
            Angle := (DeltaTimer.GetElapsedTime() * ProjectileSpeed + PiFloat * 2.0 / (ProjectileCount * 1.0) * Index)
            
            TargetTransform := transform:
                Translation := Character.GetTransform().Translation - vector3{Z := CharacterHeightOffset}
                Rotation := MakeRotationFromYawPitchRollDegrees(RadiansToDegrees(Angle), 0.0, 0.0)
                Scale := vector3{X := AttackRadius * 0.1, Y := AttackRadius * 0.1, Z := 1.0}

            set MaybeSpawnedProp = SpawnVFX(IsUniqueEvolution? and vfx_id.AbilityShurikenEvolved or vfx_id.AbilityShuriken, TargetTransform, ?Duration := DefaultAbilityRotatorDuration)

        defer:
            if (SpawnedProp := MaybeSpawnedProp?):
                SpawnedProp.Hide()

            if (Manager := MaybeManager?):
                Manager.ReturnToObjectPool(Self)

            set AgentsHit = array{}

        if (WeaponInfo := GetWeaponInfo[weapon_id.Shuriken]):
            race:
                Sleep(DefaultAbilityRotatorDuration)
                loop:
                    Sleep(IsUniqueEvolution? and 0.5 or DefaultAbilityRotatorRevolutionTime)
                    set AgentsHit = array{}
                loop:
                    if (Character := Player.GetFortCharacter[], Character.IsActive[]):
                        #Angle := (DeltaTimer.GetElapsedTime() * ProjectileSpeed + PiFloat * 2.0 / (ProjectileCount * 1.0) * Index + DegreesToRadians(InitialRotation.GetYawPitchRollDegrees()[0] or 0.0))
                        Angle := (DeltaTimer.GetElapsedTime() * ProjectileSpeed + PiFloat * 2.0 / (ProjectileCount * 1.0) * Index)
                        Position := Character.GetTransform().Translation + vector3{X := Cos(Angle) * DefaultAbilityRotatorSize, Y := Sin(Angle) * DefaultAbilityRotatorSize}
                        
                        if (SpawnedProp := MaybeSpawnedProp?):
                            TargetTransform := transform:
                                Translation := Character.GetTransform().Translation - vector3{Z := CharacterHeightOffset}
                                Rotation := MakeRotationFromYawPitchRollDegrees(RadiansToDegrees(Angle), 0.0, 0.0)
                                Scale := vector3{X := AttackRadius * 0.1, Y := AttackRadius * 0.1, Z := 1.0}

                            spawn{SpawnedProp.MoveTo(TargetTransform, TickTime * 3)}

                        for:
                            NPCBehaviorBase : GetAllNPCs()
                            NPCAgent := NPCBehaviorBase.GetAgent[]
                            not HasHit(NPCAgent)?
                            NPCCharacter := NPCAgent.GetFortCharacter[]
                            DistanceSquared(vector2{X := NPCCharacter.GetTransform().Translation.X, Y := NPCCharacter.GetTransform().Translation.Y}, vector2{X := Position.X, Y := Position.Y}) <= Pow(AttackRadius + NPCBehaviorBase.GetSize(), 2.0)
                        do:
                            if (Manager := MaybeManager?):
                                Manager.Damage(Player.GetDamageArgs(NPCBehaviorBase, weapon_id.Shuriken, AttackDamage, ?Multiplier := AttackDamageMultiplier))
                            set AgentsHit += array{NPCAgent}

                    Sleep(0.0)

    HasHit(Agent:agent)<transacts>:logic =
        for (AgentHit : AgentsHit, Agent = AgentHit) {return true}
        return false