using { /Fortnite.com/Characters }
using { /Verse.org/Simulation }
using { /UnrealEngine.com/Temporary/SpatialMath }

<# ----------------------------------------------------------------------------------------------------
    Constants
---------------------------------------------------------------------------------------------------- #>
BaseAbilityConeTime:float = 0.1
BaseAbilityConeAngle<public>:float = 120.0 # Requires updating angle in VFX
BaseAbilityConeAngleBonus:float = 45.0

<# ----------------------------------------------------------------------------------------------------
    ability_sword
---------------------------------------------------------------------------------------------------- #>
ability_sword<public> := class(ability):
    Id<override>:weapon_id = weapon_id.Sword
    DeactivatedEvent:events(tuple()) = events(tuple()){}

    var ProjectileCount:int = 1
    var AttackCooldownTime:float = 1.0
    var AttackDamage:float = 10.0
    var AttackDamageMultiplier:float = 1.0
    var AttackRadius:float = 100.0
    var AttackAngle:float = BaseAbilityConeAngle
    var IsPrecisionStrike:logic = false
    
    <# ----------------------------------------------------------------------------------------------------
        Update Functions
    ---------------------------------------------------------------------------------------------------- #>
    Activate<override>():void =
        spawn{Start()}
        set IsActive = true

    Start()<suspends>:void =
        UpdateStats()

        race:
            DeactivatedEvent.Await()
            if (WeaponInfo := GetWeaponInfo[Id]):
                Sleep(CooldownTimeRemaining)
                loop:
                    set CastTime = GetSimulationElapsedTime()
                    spawn{Ability(WeaponInfo)}
                    Sleep(AttackCooldownTime)
                    

    Deactivate<override>(?IsPause:logic = false):void =
        DeactivatedEvent.Signal()
        set CooldownTimeRemaining = Max(0.0, AttackCooldownTime - (GetSimulationElapsedTime() - CastTime))
        set IsActive = false

    Ability(WeaponInfo:weapon_info)<suspends>:void =
        if (Character := Player.GetFortCharacter[]):
            for (CastIndex := 1..ProjectileCount):
                TargetPosition := Character.GetTransform().Translation - vector3{Z := CharacterHeightOffset}                
                TargetRotation := Player.GetAutoAimRotation()
    
                Player.PlayAudio(audio_id.Projectile)
                if (AttackAngle = BaseAbilityConeAngle):
                    SpawnVFX((IsPrecisionStrike? and vfx_id.AbilitySwordEvolved or vfx_id.AbilitySword), TargetPosition, ?Rotation := TargetRotation, ?Scale := AttackRadius * 0.1 * UnitVector, ?Duration := 0.5)
                else if (AttackAngle = BaseAbilityConeAngle * 2.0):
                    SpawnVFX((IsPrecisionStrike? and vfx_id.AbilitySwordEvolved or vfx_id.AbilitySword), TargetPosition, ?Rotation := TargetRotation.ApplyYaw(DegreesToRadians(60.0)), ?Scale := AttackRadius * 0.1 * UnitVector, ?Duration := 0.5)
                    SpawnVFX((IsPrecisionStrike? and vfx_id.AbilitySwordEvolved or vfx_id.AbilitySword), TargetPosition, ?Rotation := TargetRotation.ApplyYaw(DegreesToRadians(-60.0)), ?Scale := AttackRadius * 0.1 * UnitVector, ?Duration := 0.5)
                else:
                    SpawnVFX((IsPrecisionStrike? and vfx_id.AbilitySwordEvolved or vfx_id.AbilitySword), TargetPosition, ?Rotation := TargetRotation, ?Scale := AttackRadius * 0.1 * UnitVector, ?Duration := 0.5)
                    SpawnVFX((IsPrecisionStrike? and vfx_id.AbilitySwordEvolved or vfx_id.AbilitySword), TargetPosition, ?Rotation := TargetRotation.ApplyYaw(DegreesToRadians(120.0)), ?Scale := AttackRadius * 0.1 * UnitVector, ?Duration := 0.5)
                    SpawnVFX((IsPrecisionStrike? and vfx_id.AbilitySwordEvolved or vfx_id.AbilitySword), TargetPosition, ?Rotation := TargetRotation.ApplyYaw(DegreesToRadians(-120.0)), ?Scale := AttackRadius * 0.1 * UnitVector, ?Duration := 0.5)

                Sleep(BaseAbilityConeTime)
    
                for:
                    NPCBehaviorBase : GetAllNPCs()
                    NPCCharacter := NPCBehaviorBase.GetAgent[].GetFortCharacter[]
                    DistanceSquared(NPCCharacter.GetTransform().Translation.Make2D(), TargetPosition) <= Pow(AttackRadius + NPCBehaviorBase.GetSize(), 2.0)

                    AttackAngle >= 360.0 or (
                        NPCDirection := NPCCharacter.GetTransform().Translation.Make2D() - TargetPosition.Make2D()
                        Angle := RadiansToDegrees(MakeShortestRotationBetween(TargetRotation.GetLocalForward(), NPCDirection).GetAngle())
                        Angle < (AttackAngle + BaseAbilityConeAngleBonus) / 2.0
                    )
                do:
                    Damage(Player.GetDamageArgs(NPCBehaviorBase, Id, AttackDamage, ?Multiplier := AttackDamageMultiplier))

                Sleep(0.2)
            
    UpdateStats():void =
        if (WeaponInfo := GetWeaponInfo[Id]):
            set IsPrecisionStrike = Player.IsWeaponEvolutionAbilityActive(Id, ability_id.WeaponEvolution)
            set AttackDamage = WeaponInfo.Damage * (IsPrecisionStrike? and 2.0 or 1.0)
            set ProjectileCount = 1 + Floor[Player.GetWeaponEvolutionValue(Id, ability_id.ProjectileCount)] or 0
            set AttackCooldownTime = WeaponInfo.Cooldown / (1.0 * (1.0 + Player.GetAbilityValue(Id, ability_id.Cooldown) + Player.GetAbilityValue(ability_id.Cooldown) + Player.GetPowerUpValue(power_up_id.Cooldown)))
            set AttackDamageMultiplier = Player.GetAbilityValue(Id, ability_id.Damage)
            set AttackRadius = WeaponInfo.Size * (IsPrecisionStrike? and 0.5 or 1.0) * (1.0 + Player.GetAbilityValue(Id, ability_id.Size) + Player.GetAbilityValue(ability_id.Size) + Player.GetPowerUpValue(power_up_id.Size))
            set AttackAngle = Player.GetWeaponEvolutionValue(Id, ability_id.SwordAngle)