using { /Fortnite.com/Characters }
using { /Verse.org/Random }
using { /Verse.org/Simulation }
using { /UnrealEngine.com/Temporary/SpatialMath }

<# ----------------------------------------------------------------------------------------------------
    Constants
---------------------------------------------------------------------------------------------------- #>
BaseAbilityTorrentRange:float = 700.0
BaseAbilityTorrentDuration:float = 5.0
BaseAbilityTorrentDamageInterval:float = 1.0

<# ----------------------------------------------------------------------------------------------------
    ability_torrent
---------------------------------------------------------------------------------------------------- #>
ability_torrent<public> := class(ability):
    Id<override>:weapon_id = weapon_id.Torrent
    DeactivatedEvent:events(tuple()) = events(tuple()){}

    var ProjectileCount:int = 1
    var ProjectileSplit:int = 0
    var AttackCooldownTime:float = 1.0
    var AttackDamage:float = 10.0
    var AttackDamageMultiplier:float = 1.0
    var AttackRadius:float = 100.0 
    var IsUniqueEvolution:logic = false

    Activate<override>():void =
        spawn{Start()}
        set IsActive = true

    Start()<suspends>:void =
        UpdateStats()

        race:
            DeactivatedEvent.Await()
            block:
                Sleep(CooldownTimeRemaining)
                loop:
                    set CastTime = GetSimulationElapsedTime()
                    spawn{Ability()}
                    Sleep(AttackCooldownTime)

    Deactivate<override>(?IsPause:logic = false):void =
        DeactivatedEvent.Signal()
        set CooldownTimeRemaining = Max(0.0, AttackCooldownTime - (GetSimulationElapsedTime() - CastTime))
        set IsActive = false

    Ability()<suspends>:void =
        if (Character := Player.GetFortCharacter[]):
            for (Index := 1..ProjectileCount):
                Player.PlayAudio(audio_id.Projectile)
                TargetPosition := GetRandomPositionInCircle(BaseAbilityTorrentRange, ?Origin := Character.GetTransform().Translation.Make2D())
                spawn{Area(true, TargetPosition)}
                Sleep(0.1)

    Area(IsCast:logic, Position:vector3, ?Size:float = 1.0)<suspends>:void =
        if (Character := Player.GetFortCharacter[]):
            FinalAttackRadius := AttackRadius * Size
            MaybeSpawnedVFX := SpawnVFX(IsUniqueEvolution? and vfx_id.AbilityTorrentEvolved or vfx_id.AbilityTorrent, Position, ?Scale := vector3{X := FinalAttackRadius * 0.1, Y := FinalAttackRadius * 0.1, Z := BaseAbilityTorrentDuration}, ?Duration := BaseAbilityTorrentDuration + 0.1)

            if (SpawnedVFX := MaybeSpawnedVFX?):
                race:
                    Sleep(BaseAbilityTorrentDuration)

                    if (IsUniqueEvolution?):
                        loop:
                            Time := Distance(Character.GetTransform().Translation.Make2D(), SpawnedVFX.GetTransform().Translation) / 150.0
                            spawn{SpawnedVFX.MoveTo(Character.GetTransform().Translation.Make2D(), IdentityRotation(), Time)}
                            Sleep(TickTime * 3)
                    else:
                        Sleep(Inf)

                    loop:
                        Sleep(BaseAbilityTorrentDamageInterval)
                        for:
                            NPCBehaviorBase : GetAllNPCs()
                            NPCCharacter := NPCBehaviorBase.GetAgent[].GetFortCharacter[]
                            DistanceSquared(NPCCharacter.GetTransform().Translation.Make2D(), SpawnedVFX.GetTransform().Translation) <= Pow(FinalAttackRadius + NPCBehaviorBase.GetSize(), 2.0)
                        do:
                            Damage(Player.GetDamageArgs(NPCBehaviorBase, Id, AttackDamage, ?Multiplier := AttackDamageMultiplier), ?IsKnockback := false)
                
                if (IsCast?):
                    RandomAngleOffset := GetRandomFloat(0.0, 1.0) * PiFloat * 2.0
                    for (Index := 1..ProjectileSplit):
                        Angle := (PiFloat * 2.0 / (ProjectileSplit * 1.0) * Index) + RandomAngleOffset
                        TargetPosition := SpawnedVFX.GetTransform().Translation + vector3{X := Cos(Angle) * AttackRadius, Y := Sin(Angle) * AttackRadius}
                        spawn{Area(false, TargetPosition, ?Size := 0.5)}

    UpdateStats():void =
        if (WeaponInfo := GetWeaponInfo[Id]):
            set IsUniqueEvolution = Player.IsWeaponEvolutionAbilityActive(Id, ability_id.WeaponEvolution)
            set ProjectileCount = 1 + Floor[Player.GetWeaponEvolutionValue(Id, ability_id.ProjectileCount)] or 0
            set ProjectileSplit = 0 + Floor[Player.GetWeaponEvolutionValue(Id, ability_id.TorrentSplit)] or 0
            set AttackCooldownTime = WeaponInfo.Cooldown / (1.0 * (1.0 + Player.GetAbilityValue(Id, ability_id.Cooldown) + Player.GetAbilityValue(ability_id.Cooldown) + Player.GetPowerUpValue(power_up_id.Cooldown)))
            set AttackDamage = WeaponInfo.Damage
            set AttackDamageMultiplier = Player.GetAbilityValue(Id, ability_id.Damage)
            set AttackRadius = WeaponInfo.Size * (1.0 + Player.GetAbilityValue(Id, ability_id.Size) + Player.GetAbilityValue(ability_id.Size) + Player.GetPowerUpValue(power_up_id.Size))