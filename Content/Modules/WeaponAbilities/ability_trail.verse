using { /Fortnite.com/Characters }
using { /Fortnite.com/Devices }
using { /Verse.org/Random }
using { /Verse.org/Simulation }
using { /UnrealEngine.com/Temporary/SpatialMath }

<# ----------------------------------------------------------------------------------------------------
    Constants
---------------------------------------------------------------------------------------------------- #>
BaseAbilityTrailDamageInterval:float = 1.0

<# ----------------------------------------------------------------------------------------------------
    ability_trail
---------------------------------------------------------------------------------------------------- #>
ability_trail<public> := class(ability):
    Id<override>:weapon_id = weapon_id.Trail
    DeactivatedEvent:events(tuple()) = events(tuple()){}

    var ProjectileCount:int = 1
    var AttackDuration:float = 1.0
    var AttackDamage:float = 10.0
    var AttackDamageMultiplier:float = 1.0
    var AttackRadius:float = 100.0

    var TrailChargedChance:float = 0.0
    var TrailChargedPower:float = 0.0
    var SuperChargeInterval:int = 0
    var IsEvolved:logic = false
    var IsUniqueEvolution:logic = false

    var PreviousPosition:vector3 = vector3{}
    var DistanceTraveledSquared:float = 0.0
    var CastCount:int = 0

    var MaybeVFX:?creative_prop = false

    Activate<override>():void =
        spawn{Start()}
        set IsActive = true

    Start()<suspends>:void =
        UpdateStats()
        set CastCount = 0

        if (not MaybeVFX?, Character := Player.GetFortCharacter[]):
            set MaybeVFX = SpawnVFX(vfx_id.AbilityTrail, Character.GetTransform().Translation.Make2D(), ?Scale := vector3{X := AttackRadius * 0.1, Y := AttackRadius * 0.1, Z := AttackDuration}, ?Duration := 0.0)
        
        race:
            DeactivatedEvent.Await()
            loop:
                Sleep(0.1)
                IsSuperCharged := logic{IsEvolved? and Mod[CastCount, SuperChargeInterval] = 0}
                FinalAttackRadius := AttackRadius * (IsSuperCharged? and 1.5 or 1.0)

                if (Character := Player.GetFortCharacter[]):
                    set DistanceTraveledSquared += DistanceSquared(Character.GetTransform().Translation, PreviousPosition)
                    set PreviousPosition = Character.GetTransform().Translation

                    if (VFX := MaybeVFX?, VFXSize := (Round[FinalAttackRadius * 0.1] * 1.0) + (IsSuperCharged? and 0.001 or 0.0)):
                        spawn{VFX.MoveTo(transform{Translation := Character.GetTransform().Translation.Make2D(), Rotation := IdentityRotation(), Scale := vector3{X := VFXSize, Y := VFXSize, Z := AttackDuration}}, 0.1)}

                if (DistanceTraveledSquared >= Pow(100.0, 2.0)):
                    spawn{Ability(IsSuperCharged, FinalAttackRadius)}
                    set DistanceTraveledSquared = 0.0
                    set CastCount += 1

    Deactivate<override>(?IsPause:logic = false):void =
        DeactivatedEvent.Signal()
        set IsActive = false
        if (not IsPause?, VFX := MaybeVFX?):
            VFX.Dispose()
            set MaybeVFX = false

    Ability(IsSuperCharged:logic, FinalAttackRadius:float)<suspends>:void =
        if (Character := Player.GetFortCharacter[], WeaponInfo := GetWeaponInfo[Id]):
            FinalAttackDamage := AttackDamage * (IsSuperCharged? and (1.0 + TrailChargedPower) or 1.0)
            

            if (Position := Player.GetFortCharacter[].GetTransform().Translation.Make2D()):
                if (IsUniqueEvolution? and IsSuperCharged?):
                    BlastAttackDamage := FinalAttackDamage * 1.5
                    BlastAttackRadius := FinalAttackRadius * 1.5
                    SpawnVFX(vfx_id.AbilityTrailBlast, Position, ?Scale := vector3{X := BlastAttackRadius * 0.1, Y := BlastAttackRadius * 0.1, Z := AttackDuration}, ?Duration := AttackDuration + DefaultAbilityShieldAttackTime)
                    
                    defer:
                        for:
                            NPCBehaviorBase : GetAllNPCs()
                            NPCCharacter := NPCBehaviorBase.GetAgent[].GetFortCharacter[]
                            NPCPosition := NPCCharacter.GetTransform().Translation.Make2D()
                            DistanceSquared(NPCPosition, Position) <= Pow(BlastAttackRadius + NPCBehaviorBase.GetSize(), 2.0)
                            Direction := (NPCPosition - Position).MakeUnitVector[]
                        do:
                            DirectionRotation := MakeRotation(vector3{X := 0.0, Y := 0.0, Z := 1.0}, ArcTan(Direction.Y, Direction.X))
                            spawn{NPCBehaviorBase.Knockback(1.0, DirectionRotation)}

                            Damage(Player.GetDamageArgs(NPCBehaviorBase, Id, BlastAttackDamage, ?Multiplier := AttackDamageMultiplier), ?IsKnockback := false)

                race:
                    Sleep(AttackDuration)
                    loop:
                        Sleep(BaseAbilityTrailDamageInterval)
                        for:
                            NPCBehaviorBase : GetAllNPCs()
                            NPCCharacter := NPCBehaviorBase.GetAgent[].GetFortCharacter[]
                            DistanceSquared(NPCCharacter.GetTransform().Translation.Make2D(), Position) <= Pow(FinalAttackRadius + NPCBehaviorBase.GetSize(), 2.0)
                        do:
                            Damage(Player.GetDamageArgs(NPCBehaviorBase, Id, FinalAttackDamage, ?Multiplier := AttackDamageMultiplier), ?IsKnockback := false)

    UpdateStats():void =
        if (WeaponInfo := GetWeaponInfo[Id]):
            set IsUniqueEvolution = Player.IsWeaponEvolutionAbilityActive(Id, ability_id.WeaponEvolution)
            set IsEvolved = logic{Player.GetWeaponEvolutionLevel(Id) > 0}
            set TrailChargedChance = Player.GetWeaponEvolutionValue(Id, ability_id.TrailChargedChance)
            set TrailChargedPower = Player.GetWeaponEvolutionValue(Id, ability_id.TrailChargedPower)
            set SuperChargeInterval = Ceil[10.0 / (TrailChargedChance * 10.0)] or 0

            set AttackDuration = WeaponInfo.Cooldown * (1.0 * (1.0 + Player.GetAbilityValue(Id, ability_id.Cooldown) + Player.GetAbilityValue(ability_id.Cooldown) + Player.GetPowerUpValue(power_up_id.Cooldown)))
            set AttackDamage = WeaponInfo.Damage
            set AttackDamageMultiplier = Player.GetAbilityValue(Id, ability_id.Damage)
            set AttackRadius = WeaponInfo.Size * (1.0 + Player.GetAbilityValue(Id, ability_id.Size) + Player.GetAbilityValue(ability_id.Size) + Player.GetPowerUpValue(power_up_id.Size))