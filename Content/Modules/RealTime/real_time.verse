# This file defines the real_time device that can be used for obtaining the current
# date and time in Coordinated Universal Time (UTC).
# It supports adding an offset to the current time for debugging and testing purposes.
# It also controls the hands of the clock.

using { /Fortnite.com/Devices }
using { /UnrealEngine.com/Temporary/SpatialMath }
using { /Verse.org/Simulation }

real_time<public> := class<concrete>:
    # The tracker device that keeps track of the current debug time offset.
    # It is persistent to ensure that the offset is kept between games.
    @editable
    DebugTimeOffsetSecondsTracker:tracker_device = tracker_device{}

    # Returns the current date and time.
    GetDateAndTime<public>()<transacts>:date_and_time =
        # Convert from Unix epoch (1/1/1970 00:00:00) to the beginning of the Common Era (1/1/1 00:00:00) since that is the epoch MakeDateAndTime expects.
        Seconds := GetSecondsSinceEpoch() + 62135596800.0 + 1.0 * DebugTimeOffsetSecondsTracker.GetValue()
        return MakeDateAndTime(Seconds)

    # Resets the debug time offset back to 0.
    ResetDebugTimeOffset<public>():void =
        DebugTimeOffsetSecondsTracker.SetValue(0)

    # Adds a number of days to the debug time offset.
    AddDebugTimeOffsetDays<public>(OffsetDays:int):void =
        CurrentDebugTimeOffsetSeconds := DebugTimeOffsetSecondsTracker.GetValue()
        DebugTimeOffsetSecondsTracker.SetValue(CurrentDebugTimeOffsetSeconds + OffsetDays * 86400)
        
    # Adds a number of hours to the debug time offset.
    AddDebugTimeOffsetHours<public>(OffsetHours:int):void =
        CurrentDebugTimeOffsetSeconds := DebugTimeOffsetSecondsTracker.GetValue()
        DebugTimeOffsetSecondsTracker.SetValue(CurrentDebugTimeOffsetSeconds + OffsetHours * 3600)
        
    # Adds a number of minutes to the debug time offset.
    AddDebugTimeOffsetMinutes<public>(OffsetMinutes:int):void =
        CurrentDebugTimeOffsetSeconds := DebugTimeOffsetSecondsTracker.GetValue()
        DebugTimeOffsetSecondsTracker.SetValue(CurrentDebugTimeOffsetSeconds + OffsetMinutes * 60)
        
    # Adds a number of seconds to the debug time offset.
    AddDebugTimeOffsetSeconds<public>(OffsetSeconds:int):void =
        CurrentDebugTimeOffsetSeconds := DebugTimeOffsetSecondsTracker.GetValue()
        DebugTimeOffsetSecondsTracker.SetValue(CurrentDebugTimeOffsetSeconds + OffsetSeconds)

    